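# Niagara trail effect for the TicketPoke player emote, emitting from skeletal mesh bones or sockets read through a user mesh parameter.
# A short-lived ribbon and two GPU sprite emitters, Sparks001 and NE_Snowflake, spawn across two spawn groups for 0.7-second loops.
# Sprites drift with curl noise and gravity, stretch with speed and recolor from bright orange to red.
# Ribbon alpha scales with a time-of-day value; the Mist emitter is disabled.
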NiagaraSystem /BRCosmetics/Animation/Game/MainPlayer/Emotes/TicketPoke/FX/NS_TicketPokeTrail
dependencies {
 usedDataInterfaces [&NiagaraDataInterfaceCurve, &NiagaraDataInterfaceColorCurve]
 usedDynamicInputs [&FNTimeOfDayScalar]
 usedModules [
  &SystemState
  &self:SystemSpawnScript.SystemScriptSource.SystemScriptGraph.NiagaraNodeAssignment_0.SetVariables_#1
 ]
}
emitters {
 Mist {
  bEnabled=false
  emitterUpdateScript [
   &EmitterState(Life Cycle Mode=::NewEnumerator1, Loop Duration=1, UseLoopDelay=false)
   &SpawnRate(SpawnRate=150, Use Spawn Probability=false)
   &SpawnBurst_Instantaneous(Spawn Count=15, Use Spawn Probability=false, UseLoopCountLimit=false) {enabled=false}
  ]
  particleSpawnScript [
   &InitializeParticle(
    Color=<rgba 0.3359 0.6773 1 0.2>
    Lifetime=&RandomRangeFloat(Minimum=0.8, Maximum=1.5)
    Sprite Rotation Angle=&RandomRangeFloat(Minimum=-55, Maximum=98)
    Sprite Rotation Mode=::NewEnumerator0
    Sprite Size Mode=::NewEnumerator3
    Sprite UV Mode=::NewEnumerator2
    Uniform Sprite Size=&Multiply_Float(A=&RandomRangeFloat(Minimum=50, Maximum=70), B=1)
    UsePositionOffset=false
    Write Material Random=true
   )
   &ShapeLocation(
    MaskSpawnGroup=false
    Offset=<0 -1.3139 0>
    Offset Mode=::NewEnumerator0
    Override Seed=false
    Rotation Angle=0.25
    Rotation Axis=<0 1 0>
    Rotation Mode=::NewEnumerator1
    Sphere Radius=10
    Sphere Surface Distribution=1
    UseOwnerScale=true
   ) {enabled=false}
   &InheritVelocity(Inherited Velocity Amount Scale=&VectorFromFloat(Value=0.1), Speed Limit=true, UseSpeedThreshold=true) {enabled=false}
   &CameraOffset(Camera Offset Amount=16.875) {module=CameraOffset001}
   &DynamicMaterialParameters(
    Index 0 Param 4=0.3
    Module.Index0Float0Write=true
    Module.Index0Float1Write=true
    Module.Index0Float2Write=true
    Module.Index0Float3Write=true
   )
   &AddVelocity(Velocity=&MergeHelper_VectorSwitch(
     0=<0 0 0>
     1=<0 0 0>
     2=&Multiply_VectorByFloat(Vector=$System.DirectionVector, Float=&RandomRangeFloat(Minimum=110, Maximum=120))
     3=?
     Index=&ReturnExecIndex()
     Num=::NewEnumerator1
    )) {enabled=false}
   &PointForce(Module.Use Falloff=false, Module.UseLinearForce=false)
   &ApplyInitialForces()
   &SkeletalMeshLocation(
    Bone / Socket Index=&SpawnGroup()
    Bone Sampling Mode=::NewEnumerator4
    MaskSpawnGroup=false
    Skeletal Mesh=$User.Skeletal Mesh Reader
   )
  ]
  particleUpdateScript [
   &ParticleState()
   &GravityForce(Gravity=<0 -150 -52.0864>) {enabled=false}
   &Drag(Drag=2, Use Linear Drag=true, Use Rotational Drag=true)
   &SolveForcesAndVelocity(Clamp Velocity=false, Limit Acceleration=false)
   &ScaleSpriteSize(Scale Factor=&Vector2DFromFloat(Value=&FloatFromCurve(FloatCurve=DataInterface() {propertyValues {Curve {keys [{arriveTangent=1.8373, interpMode=RCIM_Cubic, leaveTangent=1.8373, tangentMode=RCTM_User, time=0, value=0.5}, <at 1 1.3 RCIM_Cubic>]}}})))
   &ScaleColor(
    Scale Alpha=&FloatFromCurve(FloatCurve=DataInterface() {propertyValues {Curve {keys [
         <at 0 0 RCIM_Cubic>
         {arriveTangent=5.6923, interpMode=RCIM_Cubic, leaveTangent=5.6923, time=0.0203, value=0.95}
         {arriveTangent=-0.9697, interpMode=RCIM_Cubic, leaveTangent=-0.9697, time=0.0878, value=0.5}
         <at 1 0 RCIM_Cubic>
        ]}}})
    ScaleA=true
    ScaleRGB=true
   )
   &SubUVAnimation(
    End Frame=26
    SubUV Animation Mode:ENiagara_SubUVLookupMode=::NewEnumerator1
    SubUV Curve=DataInterface() {propertyValues {Curve {keys [
        {arriveTangent=2.3868, interpMode=RCIM_Cubic, leaveTangent=2.3868, tangentMode=RCTM_User, time=0, value=0}
        {interpMode=RCIM_Cubic, tangentMode=RCTM_User, time=1, value=1}
       ]}}}
   ) {enabled=false}
   /SetVariables(Particles.SubImageIndex=&Add_Float(
     A=&RandomRangeFloat(Minimum=0, Maximum=0.3)
     B=&FloatFromCurve(
      FloatCurve=DataInterface() {propertyValues {Curve {keys [{arriveTangent=120.5723, interpMode=RCIM_Cubic, leaveTangent=120.5723, tangentMode=RCTM_User, time=0, value=0}, <at 1 64 RCIM_Cubic>]}}}
      Scale Curve=&RandomRangeFloat(Minimum=0.3, Maximum=0.5)
     )
    )) {isSetParameters=true}
   &CameraOffset(Camera Offset Amount=&FloatFromCurve(FloatCurve=DataInterface() {propertyValues {Curve {keys [<at 0 1>,<at 1 100>]}}}))
   &ScaleColor(
    Color Value To Scale=$Particles.Color
    Scale Mode=::NewEnumerator1
    Scale RGBA=&MakeVector4_FromLinearColor(LinearColor=&FNTimeOfDayColor(
      Night Color=<rgba 0.539 0.6941 1 1>
      Morning Color=<rgba 0.9392 0.8785 1 1>
      Day Color=<rgba 1 1 1 1>
      Evening Color=<rgba 1 0.6882 0.6583 1>
     ))
    ScaleRGBA=true
   ) {module=ScaleColor001}
  ]
  rendererClasses [&NiagaraSpriteRendererProperties]
  renderers [{rendererClass=&NiagaraSpriteRendererProperties}]
 }
 NE_Snowflake {
  simTarget=GPUComputeSim
  emitterUpdateScript [
   &EmitterState(Life Cycle Mode=::NewEnumerator1, Loop Behavior=::NewEnumerator1, Loop Duration=0.7, UseLoopDelay=false)
   &SpawnRate(SpawnRate=60, Use Spawn Probability=false)
  ]
  particleSpawnScript [
   &InitializeParticle(
    Color=&MakeCustomLinearColorFromBool(Bool=&RandomBool(), True Color=<rgba 0.1159 0.4925 1 1>, False Color=<rgba 0.2456 0.5543 1 1>)
    Lifetime Max=0.5
    Lifetime Min=0.2
    Lifetime Mode=::NewEnumerator1
    Sprite Rotation Mode=::NewEnumerator2
    Sprite Size Mode=::NewEnumerator1
    Uniform Sprite Size Max=4
    Uniform Sprite Size Min=2.5
    UsePositionOffset=false
    Write Material Random=true
   )
   &ShapeLocation(Box Size=<10 10 1>, MaskSpawnGroup=false, Override Seed=false, Shape Primitive=::NewEnumerator2, Surface Only=false, UseOwnerScale=true) {module=ShapeLocation001}
   &ApplyInitialForces()
   &FN_SetSpawnGroup(Number of Spawn Groups=2)
   &SkeletalMeshLocation(
    Bone / Socket Index=&SpawnGroup()
    Bone Sampling Mode=::NewEnumerator4
    MaskSpawnGroup=false
    Sampled Position Offset=&RandomRangeVector(Minimum=<-10 -10 -10>, Maximum=<10 10 10>)
    Skeletal Mesh=$User.Skeletal Mesh Reader
   )
  ]
  particleUpdateScript [
   &ParticleState()
   &Color(Color=&ColorFromCurve(ColorCurve=DataInterface() {propertyValues {
       AlphaCurve {keys [<at 0 1 RCIM_Cubic>,<at 1 0 RCIM_Cubic>]}
       BlueCurve {keys [<at 0 0 RCIM_Cubic>,<at 0.1147 0>,<at 0.5143 0 RCIM_Cubic>]}
       GreenCurve {keys [<at 0 9.1707 RCIM_Cubic>,<at 0.1147 3.0139>,<at 0.5143 0.0332 RCIM_Cubic>]}
       RedCurve {keys [<at 0 20 RCIM_Cubic>,<at 0.1147 10>,<at 0.5143 1 RCIM_Cubic>]}
      }}))
   /SetVariables(Particles.Glow03_Color=&MultiplyLinearColor(Linear Color=$Particles.Color, Scale Factor=<1 1 1 0.04>)) {isSetParameters=true, module=SetVariables_#2}
   /SetVariables(Particles.Glow03_Size=&Multiply_Vector2D_ByFloat(
     Float=&FloatFromCurve(
      FloatCurve=DataInterface() {propertyValues {Curve {keys [
          <at 0 0 RCIM_Cubic>
          <at 0.2 1.3 RCIM_Cubic>
          {arriveTangent=-0.1423, interpMode=RCIM_Cubic, leaveTangent=-0.1423, tangentMode=RCTM_User, time=1, value=0}
         ]}}}
      Scale Curve=15
     )
     Vector2D=<2 2>
    )) {isSetParameters=true, module=SetVariables_#3}
   &ScaleColor(
    Scale Alpha=&FloatFromCurve(FloatCurve=DataInterface() {propertyValues {Curve {keys [<at 0.3 1>,<at 0.4 1>,<at 1 1>]}}})
    ScaleA=true
    ScaleRGB=true
   )
   &ScaleSpriteSize(Scale Factor=&Vector2DFromFloat(Value=&FloatFromCurve(FloatCurve=DataInterface() {propertyValues {Curve {keys [
          <at 0 0 RCIM_Cubic>
          {interpMode=RCIM_Cubic, leaveTangent=-3.0296, tangentMode=RCTM_Break, time=0.2, value=1.2}
          {arriveTangent=-1.8107, interpMode=RCIM_Cubic, leaveTangent=-1.8107, tangentMode=RCTM_User, time=1, value=0}
         ]}}}))) {module=ScaleSpriteSize001}
   &GravityForce(Gravity=<0 0 -200>)
   &CurlNoiseForce(
    Noise Frequency=300
    Noise Strength=&Multiply_Float(A=1000, B=1.5)
    Pan Noise=true
    Pan Noise Field=<100 0 1>
    Random Seed=10
   )
   &Drag(Drag=&RandomRangeFloat(Minimum=1, Maximum=5), Use Linear Drag=true, Use Rotational Drag=false)
   &SolveForcesAndVelocity(Clamp Velocity=false, Limit Acceleration=false)
   &ScaleSpriteSizeBySpeed(Max Scale Factor=<1 4>, Velocity Threshold=356)
  ]
  rendererClasses [&NiagaraSpriteRendererProperties]
  renderers [{rendererClass=&NiagaraSpriteRendererProperties}]
 }
 Sparks001 {
  simTarget=GPUComputeSim
  emitterUpdateScript [
   &EmitterState(Life Cycle Mode=::NewEnumerator1, Loop Behavior=::NewEnumerator1, Loop Duration=0.7, UseLoopDelay=false)
   &SpawnRate(SpawnRate=100, Use Spawn Probability=false)
  ]
  particleSpawnScript [
   &InitializeParticle(
    Color=<rgba 0.6253 0.9785 1 0.1304>
    Lifetime Max=0.5
    Lifetime Min=0.2
    Lifetime Mode=::NewEnumerator1
    Sprite Rotation Mode=::NewEnumerator2
    Sprite Size Mode=::NewEnumerator1
    Uniform Sprite Size Max=20
    Uniform Sprite Size Min=15
    UsePositionOffset=false
    Write Material Random=true
   )
   &ShapeLocation(MaskSpawnGroup=false, Override Seed=false, UseOwnerScale=true) {module=ShapeLocation001}
   &ApplyInitialForces()
   &FN_SetSpawnGroup(Number of Spawn Groups=2)
   &SkeletalMeshLocation(
    Bone / Socket Index=&SpawnGroup()
    Bone Sampling Mode=::NewEnumerator4
    MaskSpawnGroup=false
    Sampled Position Offset=&RandomRangeVector(Minimum=<-10 -10 -10>, Maximum=<10 10 10>)
    Skeletal Mesh=$User.Skeletal Mesh Reader
   )
  ]
  particleUpdateScript [
   &ParticleState()
   &Color(Color=&ColorFromCurve(ColorCurve=DataInterface() {propertyValues {
       AlphaCurve {keys [<at 0 1 RCIM_Cubic>,<at 1 0 RCIM_Cubic>]}
       BlueCurve {keys [<at 0 0 RCIM_Cubic>,<at 0.1147 0>,<at 0.5143 0 RCIM_Cubic>]}
       GreenCurve {keys [<at 0 9.1707 RCIM_Cubic>,<at 0.1147 3.0139>,<at 0.5143 0.0332 RCIM_Cubic>]}
       RedCurve {keys [<at 0 20 RCIM_Cubic>,<at 0.1147 10>,<at 0.5143 1 RCIM_Cubic>]}
      }}))
   /SetVariables(Particles.Glow03_Color=&MultiplyLinearColor(Linear Color=$Particles.Color, Scale Factor=<1 1 1 0.04>)) {isSetParameters=true, module=SetVariables_#2}
   /SetVariables(Particles.Glow03_Size=&Multiply_Vector2D_ByFloat(
     Float=&FloatFromCurve(
      FloatCurve=DataInterface() {propertyValues {Curve {keys [
          <at 0 0 RCIM_Cubic>
          <at 0.2 1.3 RCIM_Cubic>
          {arriveTangent=-0.1423, interpMode=RCIM_Cubic, leaveTangent=-0.1423, tangentMode=RCTM_User, time=1, value=0}
         ]}}}
      Scale Curve=15
     )
     Vector2D=<2 2>
    )) {isSetParameters=true, module=SetVariables_#3}
   &ScaleColor(
    Scale Alpha=&FloatFromCurve(FloatCurve=DataInterface() {propertyValues {Curve {keys [<at 0.3 1>,<at 0.4 1>,<at 1 1>]}}})
    ScaleA=true
    ScaleRGB=true
   )
   &ScaleSpriteSize(Scale Factor=&Vector2DFromFloat(Value=&FloatFromCurve(FloatCurve=DataInterface() {propertyValues {Curve {keys [
          <at 0 0 RCIM_Cubic>
          {interpMode=RCIM_Cubic, leaveTangent=-3.0296, tangentMode=RCTM_Break, time=0.2, value=1.2}
          {arriveTangent=-1.8107, interpMode=RCIM_Cubic, leaveTangent=-1.8107, tangentMode=RCTM_User, time=1, value=0}
         ]}}}))) {module=ScaleSpriteSize001}
   &GravityForce(Gravity=<0 0 -200>)
   &CurlNoiseForce(Noise Frequency=300, Noise Strength=200, Pan Noise=true, Pan Noise Field=<100 0 1>, Random Seed=10)
   &Drag(Drag=&RandomRangeFloat(Minimum=1, Maximum=5), Use Linear Drag=true, Use Rotational Drag=false)
   &SolveForcesAndVelocity(Clamp Velocity=false, Limit Acceleration=false)
   &ScaleSpriteSizeBySpeed(Max Scale Factor=<1 4>, Velocity Threshold=356)
  ]
  rendererClasses [&NiagaraSpriteRendererProperties]
  renderers [
   {rendererClass=&NiagaraSpriteRendererProperties}
   {rendererClass=&NiagaraSpriteRendererProperties, rendererIndex=1}
  ]
 }
 trail003 {
  emitterUpdateScript [
   &EmitterState(Inactive Response=::NewEnumerator1, Life Cycle Mode=::NewEnumerator1, Loop Behavior=::NewEnumerator1, Loop Duration=0.7, UseLoopDelay=false)
   &SpawnRate(SpawnRate=60, Use Spawn Probability=false)
  ]
  particleSpawnScript [
   &InitializeParticle(
    Color=<rgba 0.2666 0.3839 0.6562 1>
    Lifetime=0.2
    Ribbon Width=15
    Ribbon Width Mode=::NewEnumerator1
    UsePositionOffset=false
    Write Material Random=true
   )
   /SetVariables(Particles.Random=&RandomRangeFloat(Minimum=0.7, Maximum=1.1)) {isSetParameters=true}
   &FN_SetSpawnGroup(Number of Spawn Groups=2)
   &SkeletalMeshLocation(
    Bone / Socket Index=&SpawnGroup()
    Bone Sampling Mode=::NewEnumerator4
    MaskSpawnGroup=false
    Skeletal Mesh=$User.Skeletal Mesh Reader
   )
  ]
  particleUpdateScript [
   &SolveForcesAndVelocity(Clamp Velocity=false, Limit Acceleration=false)
   &ParticleState()
   &ScaleRibbonWidth(Ribbon Width Scale=&FloatFromCurve(
     FloatCurve=DataInterface() {propertyValues {Curve {keys [{arriveTangent=-1.1377, interpMode=RCIM_Cubic, leaveTangent=-1.1377, tangentMode=RCTM_User, time=0, value=1}, <at 1 1>]}}}
     Scale Curve=2
    ))
   &DynamicMaterialParameters(
    Index 0 Param 1=&FloatFromCurve(FloatCurve=DataInterface() {propertyValues {Curve {keys [<at 0 1>,<at 1 0.3>]}}})
    Index 0 Param 3=-3
    Index 0 Param 4=0.5
    Index 1 Param 1=30
    Module.Index0Float0Write=true
    Module.Index0Float1Write=true
    Module.Index0Float2Write=true
    Module.Index0Float3Write=true
    Module.Index1Float0Write=true
    Module.Index1Float1Write=true
    Module.Index1Float2Write=true
    Module.Index1Float3Write=true
    Write Parameter Index 1=::NewEnumerator1
   )
   &ScaleColor(
    Scale Alpha=&Multiply_Float(
     A=&FloatFromCurve(
      FloatCurve=DataInterface() {propertyValues {Curve {keys [
          {arriveTangent=-2.0492, interpMode=RCIM_Cubic, leaveTangent=-2.0492, tangentMode=RCTM_User, time=0, value=1}
          {arriveTangent=-0.2257, interpMode=RCIM_Cubic, leaveTangent=-0.2257, tangentMode=RCTM_User, time=1, value=0}
         ]}}}
      Scale Curve=0.7
     )
     B=$System.TOD
    )
    ScaleA=true
    ScaleRGB=true
   ) {module=ScaleColor001}
  ]
  rendererClasses [&NiagaraRibbonRendererProperties]
  renderers [{rendererClass=&NiagaraRibbonRendererProperties}]
 }
}
summary {
 systemName=NS_TicketPokeTrail
 emitters [
  {emitterName=trail003, rendererClasses [&NiagaraRibbonRendererProperties]}
  {emitterName=Sparks001, simTarget=GPUComputeSim, rendererClasses [&NiagaraSpriteRendererProperties]}
  {emitterName=NE_Snowflake, simTarget=GPUComputeSim, rendererClasses [&NiagaraSpriteRendererProperties]}
  {bEnabled=false, emitterName=Mist, rendererClasses [&NiagaraSpriteRendererProperties]}
 ]
 userVariables [
  User.bIsFrontEnd {type=NiagaraBool}
  User.bIsFrontEndPreview {type=NiagaraBool}
  User.BPOpacityMult {type=NiagaraFloat, defaultValue=0, description=""}
  User.CurrentExhaust {type=Vector3f, description="", defaultValue {struct=Vector3f, value=<0 0 0>}}
  User.DebugVelocity {type=NiagaraFloat, defaultValue=0, description=""}
  User.IsR2 {type=NiagaraBool, defaultValue=true}
  User.PlayerSpeed {type=NiagaraFloat, defaultValue=0, description=""}
  User.RibbonIDOffset {type=NiagaraInt32}
  User.Skeletal Mesh Reader {
   type=NiagaraDataInterfaceSkeletalMesh
   defaultValue {
    struct=NiagaraExt_VariableValue_DataInterface
    value {dataInterface=&self:NiagaraDataInterfaceSkeletalMesh_1}
   }
  }
  User.SkeletalMesh {
   type=NiagaraDataInterfaceSkeletalMesh
   defaultValue {
    struct=NiagaraExt_VariableValue_DataInterface
    value {dataInterface=&self:NiagaraDataInterfaceSkeletalMesh_0}
   }
  }
  User.TrailLengthScale {type=NiagaraFloat, defaultValue=0, description=""}
  User.TrailWidth {type=NiagaraFloat, defaultValue=0.4, description=""}
  User.Variant_Color {type=LinearColor, description="", defaultValue {struct=LinearColor, value=<rgba 0.1562 0.1698 1 2>}}
 ]
}
