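# Niagara ground-impact effect for the TicketPoke emote, producing a short icy blue puff of snow.
# Three enabled emitters fire bursts: curl-noise Snowflakes flung outward, expanding SubUV Snow_Powder, and upward-facing smoke rings positioned by skeletal mesh sampling.
# All three tint by time of day through color-select dynamic inputs.
# Head glow, Mist, debris, windstreak and ribbon-trail emitters are present but disabled.
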NiagaraSystem /BRCosmetics/Animation/Game/MainPlayer/Emotes/TicketPoke/FX/NS_TicketPoke_GroundImpact
dependencies {
 usedDataInterfaces [
  &NiagaraDataInterfaceCurve
  &NiagaraDataInterfaceParticleRead
  &NiagaraDataInterfaceSpriteRendererInfo
  &NiagaraDataInterfaceVector2DCurve
 ]
 usedDynamicInputs [
  &SimulationPosition
  &MakeFloatFromVector2d
  &Multiply_Vector2
 ]
 usedModules [&SystemState]
}
emitters {
 Head {
  bEnabled=false
  emitterUpdateScript [
   &EmitterState(Inactive Response=::NewEnumerator1, Life Cycle Mode=::NewEnumerator1, Loop Behavior=::NewEnumerator1, Loop Duration=1, UseLoopDelay=false)
   &SpawnBurst_Instantaneous(Use Spawn Probability=false, UseLoopCountLimit=false)
  ]
  particleSpawnScript [
   &InitializeParticle(
    Color=<rgba 0 3.8502 5 1>
    Lifetime=0.2
    Mesh Scale=<0.5 0.5 1.5>
    Mesh Scale Mode=::NewEnumerator4
    Sprite Size Mode=::NewEnumerator3
    Uniform Sprite Size=200
    UsePositionOffset=false
    Write Material Random=true
   )
   &InitialMeshOrientation(Mesh Orientation Mode=::NewEnumerator2, Orientation Vector=<0 0 1>, Use Rotation Vector=false)
  ]
  particleUpdateScript [
   &ParticleState(Kill Particles When Lifetime Has Elapsed=false)
   &ScaleColor(
    Scale Alpha=&FloatFromCurve(FloatCurve=DataInterface() {propertyValues {Curve {keys [
         {arriveTangent=9.0725, interpMode=RCIM_Cubic, leaveTangent=9.0725, tangentMode=RCTM_User, time=0, value=0}
         {arriveTangent=-2.5249, interpMode=RCIM_Cubic, leaveTangent=-2.5249, tangentMode=RCTM_User, time=0.1408, value=0.9847}
         {arriveTangent=-0.0405, interpMode=RCIM_Cubic, leaveTangent=-0.0405, tangentMode=RCTM_User, time=1, value=0}
        ]}}})
    ScaleA=true
    ScaleRGB=true
   )
   &ScaleSpriteSize(Uniform Curve Sprite Scale=DataInterface() {propertyValues {Curve {keys [
        {arriveTangent=9.0725, interpMode=RCIM_Cubic, leaveTangent=9.0725, tangentMode=RCTM_User, time=0, value=0}
        {arriveTangent=-0.5028, interpMode=RCIM_Cubic, leaveTangent=-0.5028, tangentMode=RCTM_User, time=0.1408, value=0.9847}
        {arriveTangent=-0.0405, interpMode=RCIM_Cubic, leaveTangent=-0.0405, tangentMode=RCTM_User, time=1, value=0.4}
       ]}}})
  ]
  rendererClasses [&NiagaraLightRendererProperties, &NiagaraSpriteRendererProperties]
  renderers [
   {rendererClass=&NiagaraLightRendererProperties, rendererIndex=0}
   {rendererClass=&NiagaraSpriteRendererProperties, rendererIndex=1}
  ]
 }
 Mist {
  bEnabled=false
  emitterUpdateScript [
   &EmitterState(Life Cycle Mode=::NewEnumerator1, Loop Duration=1, UseLoopDelay=false)
   &SpawnRate(SpawnRate=150, Use Spawn Probability=false)
   &SpawnBurst_Instantaneous(Spawn Count=15, Use Spawn Probability=false, UseLoopCountLimit=false) {enabled=false}
  ]
  particleSpawnScript [
   &InitializeParticle(
    Color=<rgba 0.3359 0.6773 1 0.2>
    Lifetime=&RandomRangeFloat(Minimum=0.8, Maximum=1.5)
    Sprite Rotation Angle=&RandomRangeFloat(Minimum=-55, Maximum=98)
    Sprite Rotation Mode=::NewEnumerator0
    Sprite Size Mode=::NewEnumerator3
    Sprite UV Mode=::NewEnumerator2
    Uniform Sprite Size=&Multiply_Float(A=&RandomRangeFloat(Minimum=50, Maximum=70), B=1)
    UsePositionOffset=false
    Write Material Random=true
   )
   &ShapeLocation(
    MaskSpawnGroup=false
    Offset=<0 -1.3139 0>
    Offset Mode=::NewEnumerator0
    Override Seed=false
    Rotation Angle=0.25
    Rotation Axis=<0 1 0>
    Rotation Mode=::NewEnumerator1
    Sphere Radius=10
    Sphere Surface Distribution=1
    UseOwnerScale=true
   ) {enabled=false}
   &InheritVelocity(Inherited Velocity Amount Scale=&VectorFromFloat(Value=0.1), Speed Limit=true, UseSpeedThreshold=true) {enabled=false}
   &CameraOffset(Camera Offset Amount=16.875) {module=CameraOffset001}
   &DynamicMaterialParameters(
    Index 0 Param 4=0.3
    Module.Index0Float0Write=true
    Module.Index0Float1Write=true
    Module.Index0Float2Write=true
    Module.Index0Float3Write=true
   )
   &AddVelocity(Velocity=&MergeHelper_VectorSwitch(
     0=<0 0 0>
     1=<0 0 0>
     2=&Multiply_VectorByFloat(Vector=$System.DirectionVector, Float=&RandomRangeFloat(Minimum=110, Maximum=120))
     3=?
     Index=&ReturnExecIndex()
     Num=::NewEnumerator1
    )) {enabled=false}
   &PointForce(Module.Use Falloff=false, Module.UseLinearForce=false)
   &ApplyInitialForces()
   &SkeletalMeshLocation(
    Bone / Socket Index=&SpawnGroup()
    Bone Sampling Mode=::NewEnumerator4
    MaskSpawnGroup=false
    Skeletal Mesh=$User.Skeletal Mesh Reader
   )
   &ShapeLocation(
    Cylinder Height=1
    Cylinder Radius=60
    MaskSpawnGroup=false
    Override Seed=false
    Shape Primitive=::NewEnumerator1
    Surface Only=false
    UseOwnerScale=true
   ) {enabled=false, module=ShapeLocation001}
  ]
  particleUpdateScript [
   &ParticleState()
   &GravityForce(Gravity=<0 -150 -52.0864>) {enabled=false}
   &Drag(Drag=2, Use Linear Drag=true, Use Rotational Drag=true)
   &SolveForcesAndVelocity(Clamp Velocity=false, Limit Acceleration=false)
   &ScaleSpriteSize(Scale Factor=&Vector2DFromFloat(Value=&FloatFromCurve(FloatCurve=DataInterface() {propertyValues {Curve {keys [{arriveTangent=1.8373, interpMode=RCIM_Cubic, leaveTangent=1.8373, tangentMode=RCTM_User, time=0, value=0.5}, <at 1 1.3 RCIM_Cubic>]}}})))
   &ScaleColor(
    Scale Alpha=&FloatFromCurve(FloatCurve=DataInterface() {propertyValues {Curve {keys [
         <at 0 0 RCIM_Cubic>
         {arriveTangent=5.6923, interpMode=RCIM_Cubic, leaveTangent=5.6923, time=0.0203, value=0.95}
         {arriveTangent=-0.9697, interpMode=RCIM_Cubic, leaveTangent=-0.9697, time=0.0878, value=0.5}
         <at 1 0 RCIM_Cubic>
        ]}}})
    ScaleA=true
    ScaleRGB=true
   )
   &SubUVAnimation(
    End Frame=26
    SubUV Animation Mode:ENiagara_SubUVLookupMode=::NewEnumerator1
    SubUV Curve=DataInterface() {propertyValues {Curve {keys [
        {arriveTangent=2.3868, interpMode=RCIM_Cubic, leaveTangent=2.3868, tangentMode=RCTM_User, time=0, value=0}
        {interpMode=RCIM_Cubic, tangentMode=RCTM_User, time=1, value=1}
       ]}}}
   ) {enabled=false}
   /SetVariables(Particles.SubImageIndex=&Add_Float(
     A=&RandomRangeFloat(Minimum=0, Maximum=0.3)
     B=&FloatFromCurve(
      FloatCurve=DataInterface() {propertyValues {Curve {keys [{arriveTangent=120.5723, interpMode=RCIM_Cubic, leaveTangent=120.5723, tangentMode=RCTM_User, time=0, value=0}, <at 1 64 RCIM_Cubic>]}}}
      Scale Curve=&RandomRangeFloat(Minimum=0.3, Maximum=0.5)
     )
    )) {isSetParameters=true}
   &CameraOffset(Camera Offset Amount=&FloatFromCurve(FloatCurve=DataInterface() {propertyValues {Curve {keys [<at 0 1>,<at 1 100>]}}}))
   &ScaleColor(
    Color Value To Scale=$Particles.Color
    Scale Mode=::NewEnumerator1
    Scale RGBA=&MakeVector4_FromLinearColor(LinearColor=&FNTimeOfDayColor(
      Night Color=<rgba 0.539 0.6941 1 1>
      Morning Color=<rgba 0.9392 0.8785 1 1>
      Day Color=<rgba 1 1 1 1>
      Evening Color=<rgba 1 0.6882 0.6583 1>
     ))
    ScaleRGBA=true
   ) {module=ScaleColor001}
  ]
  rendererClasses [&NiagaraSpriteRendererProperties]
  renderers [{rendererClass=&NiagaraSpriteRendererProperties}]
 }
 NE_Debris_02 {
  bEnabled=false
  emitterUpdateScript [
   &EmitterState()
   &SpawnBurst_Instantaneous(Spawn Count=5, Use Spawn Probability=false, UseLoopCountLimit=false)
   &SpawnBurst_Instantaneous(Spawn Count=5, Spawn Group=1, Spawn Time=0.05, Use Spawn Probability=false, UseLoopCountLimit=false) {module=SpawnBurst_Instantaneous001}
   &SpawnBurst_Instantaneous(Spawn Count=5, Spawn Time=0.05, Use Spawn Probability=false, UseLoopCountLimit=false) {module=SpawnBurst_Instantaneous002}
  ]
  particleSpawnScript [
   &InitializeParticle(
    Color Maximum=<rgba 0.4921 0.5944 0.7083 1>
    Color Minimum=<rgba 0.5925 0.8193 0.9844 1>
    Color Mode=::NewEnumerator2
    Lifetime Max=1.5
    Lifetime Min=0.5
    Lifetime Mode=::NewEnumerator1
    Mass Mode=::NewEnumerator1
    Mesh Scale Max=<1 1 1>
    Mesh Scale Min=<0.25 0.25 0.25>
    Mesh Scale Mode=::NewEnumerator2
    Position Offset=<0 0 5>
    Position Offset Coordinate Space=::Local
    UsePositionOffset=true
    Write Material Random=true
   )
   &ShapeLocation(MaskSpawnGroup=false, Override Seed=false, UseOwnerScale=true)
   &AddVelocity(
    Cone Angle=60
    Cone Axis=<0 0 1>
    Module.UseConeFalloff=false
    Module.UseVelDistribution=false
    Override Seed=false
    Velocity Mode=::NewEnumerator2
    Velocity Speed=&RandomRangeFloat(Minimum=300, Maximum=400)
   )
   &AddVelocity(Velocity=&RandomRangeVector(Minimum=<0 0 200>, Maximum=<0 0 -200>)) {module=AddVelocity001}
   /SetVariables(Particles.MeshIndex=&RandomRangeInt(Minimum=0, Maximum=2)) {isSetParameters=true}
   &InitialMeshOrientation(Use Rotation Vector=false)
  ]
  particleUpdateScript [
   &ParticleState()
   &UpdateMeshOrientation(
    Rotation Rate=&RandomRangeFloat(Minimum=1, Maximum=2)
    Rotation Vector=&RandomRangeVector(Minimum=<-1 -1 -1>, Maximum=<1 1 1>)
   )
   &GravityForce(Gravity=&RandomRangeVector(Minimum=<0 0 -360>, Maximum=<1 1 -880>))
   &Drag(Use Linear Drag=true, Use Rotational Drag=true)
   &SolveForcesAndVelocity(Clamp Velocity=false, Limit Acceleration=false)
   &ScaleMeshSize(Scale Factor=&VectorFromFloat(Value=&FloatFromCurve(FloatCurve=DataInterface() {propertyValues {Curve {keys [
          {arriveTangent=-0.008, interpMode=RCIM_Cubic, leaveTangent=-0.008, tangentMode=RCTM_User, time=0, value=1}
          {arriveTangent=-3.1748, interpMode=RCIM_Cubic, leaveTangent=-3.1748, tangentMode=RCTM_User, time=0.9, value=0}
         ]}}})))
  ]
  rendererClasses [&NiagaraMeshRendererProperties]
  renderers [{rendererClass=&NiagaraMeshRendererProperties, rendererIndex=0}]
 }
 NE_SmokeGroundRings001 {
  emitterUpdateScript [
   &EmitterState(Inactive Response=::NewEnumerator1, Life Cycle Mode=::NewEnumerator1, Loop Behavior=::NewEnumerator1, Loop Duration=2, UseLoopDelay=false)
   &SpawnBurst_Instantaneous(Spawn Count=3, Use Spawn Probability=false, UseLoopCountLimit=false)
   &SpawnRate(SpawnRate=5, Use Spawn Probability=false) {enabled=false, module=SpawnRate001}
  ]
  particleSpawnScript [
   &InitializeParticle(
    Color=<rgba 0.5781 0.9866 1 1.5>
    Lifetime=0.5
    Mass Mode=::NewEnumerator1
    Sprite Rotation Angle=&UniformRangedFloat(Minimum=-180, Maximum=180)
    Sprite Rotation Mode=::NewEnumerator0
    Sprite Size=&Vector2DFromFloat(Value=300)
    Sprite Size Mode=::NewEnumerator4
    UsePositionOffset=false
    Write Material Random=true
   ) {module=InitializeParticle001}
   /SetVariables(Particles.SpriteFacing=<0 0 1>) {isSetParameters=true}
  ]
  particleUpdateScript [
   &ParticleState()
   &SkeletalMeshLocation(
    Bone Sampling Mode=::NewEnumerator4
    MaskSpawnGroup=false
    Skeletal Mesh=DataInterface() {propertyValues {
      DefaultMesh=null
      ExcludeBoneName=root
      OptionalComponentClass=null
      PreviewMesh=/FigureCosmetics/Props/Emote/GoudaWheel/SkeletalMesh/SKM_GoudaWheel
      SkinningMode=SkinOnTheFly
      SoftSourceActor=null
      SourceMode=Default
      UvSetIndex=0
      WholeMeshLOD=-1
      bExcludeBone=false
      bReadDeformedGeometry=true
      bRequireCurrentFrameData=true
      ComponentTags []
      FilteredBones [root]
      FilteredSockets []
      MeshUserParameter {parameter {name=None, typeDefHandle {registeredTypeIndex=1}, varData []}}
      SamplingRegions []
     }}
   )
   &DynamicMaterialParameters(
    Index 0 Param 1=&FloatFromCurve(FloatCurve=DataInterface() {propertyValues {Curve {keys [<at 0 0.5>,<at 1 2>]}}})
    Module.Param0WriteEnabled=true
    Module.Param1WriteEnabled=false
    Module.Param2WriteEnabled=false
    Module.Param3WriteEnabled=false
   )
   &ScaleSpriteSize(Scale Factor=&Vector2DFromFloat(Value=&FloatFromCurve(FloatCurve=DataInterface() {propertyValues {Curve {keys [
          {arriveTangent=4.614, interpMode=RCIM_Cubic, leaveTangent=4.614, tangentMode=RCTM_User, time=0, value=0}
          {interpMode=RCIM_Cubic, leaveTangent=0.6903, time=1, value=1.5}
         ]}}})))
   &CameraOffset(Camera Offset Amount=20)
   &ScaleColor(
    Scale Alpha=&FloatFromCurve(FloatCurve=DataInterface() {propertyValues {Curve {keys [
         {arriveTangent=0.0112, interpMode=RCIM_Cubic, leaveTangent=0.0112, tangentMode=RCTM_User, time=0, value=0}
         {arriveTangent=0.002, interpMode=RCIM_Cubic, leaveTangent=0.002, time=0.4, value=0.4}
         {interpMode=RCIM_Cubic, leaveTangent=-4.9981, time=1.0011, value=0.002}
        ]}}})
    Scale RGB=&MakeVectorFromLinearColorRGB(Color=$System.TOD Color)
    ScaleA=true
    ScaleRGB=true
   )
   &InheritVelocity(
    Inherited Velocity Amount Scale=<-1 -1 -1>
    Inherited Velocity Speed Limit=5
    Source Speed Threshold=3
    Speed Limit=true
    UseSpeedThreshold=true
   )
   &SolveForcesAndVelocity(Clamp Velocity=false, Limit Acceleration=false)
   &ScaleColor(
    Scale Alpha=&MakeFloatFromLinearColor(LinearColor=$System.TOD Color, Channel:ENiagaraFloatFromLinearColorOptions=::NewEnumerator3)
    ScaleA=true
    ScaleRGB=false
   ) {module=ScaleColor002}
   &ScaleColor(
    Scale RGB=&MakeVectorFromLinearColorRGB(Color=&FN_Select_Color_from_Time_Of_Day(Morning=<rgba 1 0.8463 0.7189 1>, Day=<rgba 1 1 1 1>, Evening=<rgba 1 0.8371 0.7936 1>, Night=<rgba 0.1429 0.183 0.4497 1>))
    ScaleA=false
    ScaleRGB=true
   ) {module=ScaleColor001}
  ]
  rendererClasses [&NiagaraSpriteRendererProperties]
  renderers [{rendererClass=&NiagaraSpriteRendererProperties}]
 }
 NE_WindstreaksJet001 {
  bEnabled=false
  emitterUpdateScript [
   &EmitterState(Life Cycle Mode=::NewEnumerator1, Loop Behavior=::NewEnumerator1, Loop Delay=0.14, Loop Duration=2, UseLoopDelay=true)
   &SpawnBurst_Instantaneous(Spawn Count=100, Use Spawn Probability=false, UseLoopCountLimit=false)
   &SpawnRate(
    SpawnRate=&Multiply_Float(
     A=&FloatFromCurve(
      FloatCurve=DataInterface() {propertyValues {Curve {keys [
          <at 0 0 RCIM_Cubic>
          {arriveTangent=2.5, interpMode=RCIM_Cubic, leaveTangent=2.5, time=0.2, value=0.15}
          {arriveTangent=1.65, interpMode=RCIM_Cubic, leaveTangent=1.65, time=0.4, value=1}
          {arriveTangent=-1.625, interpMode=RCIM_Cubic, leaveTangent=-1.625, time=0.7, value=0.975}
          <at 1 0.025 RCIM_Cubic>
         ]}}}
      CurveIndex=$Emitter.NormalizedLoopAge
     )
     B=65
    )
    Use Spawn Probability=true
   ) {enabled=false}
  ]
  particleSpawnScript [
   &InitializeParticle(
    Color=<rgba 0.3359 0.6773 1 0.2>
    Lifetime=&UniformRangedFloat(Minimum=0.4, Maximum=1)
    Sprite Size=&UniformRangedVector2D(Minimum=<20 20>, Maximum=<30 30>)
    Sprite Size Mode=::NewEnumerator4
    UsePositionOffset=false
    Write Material Random=true
   )
   &ShapeLocation(
    Cylinder Height=1
    Cylinder Radius=60
    MaskSpawnGroup=false
    Override Seed=false
    Shape Primitive=::NewEnumerator1
    Surface Only=false
    UseOwnerScale=true
   )
   &AddVelocity(Module.Use Falloff Distance=false, Origin Offset=<0 0 -10>, Override Seed=false, Velocity Mode=::NewEnumerator1, Velocity Speed=250)
   &ApplyInitialForces()
   &SubUVAnimation(
    Evaluation Type=::NewEnumerator0
    "Sprite Renderer "=DataInterface() {propertyValues {SpriteRenderer=&self:NE_WindstreaksJet001_0.NiagaraSpriteRendererProperties_0}}
    SubUV Animation Mode:ENiagara_SubUVLookupModeV2=::NewEnumerator2
    UseEndFrame=false
    UseStartFrame=false
   ) {enabled=false}
  ]
  particleUpdateScript [
   &ParticleState()
   &ScaleSpriteSize(Scale Factor=&Vector2DFromCurve(
     Vector2Curve=DataInterface() {propertyValues {
       XCurve {keys [
         {arriveTangent=2.5428, interpMode=RCIM_Cubic, leaveTangent=2.5428, tangentMode=RCTM_User, time=0, value=1}
         {arriveTangent=0.1281, interpMode=RCIM_Cubic, leaveTangent=0.1281, tangentMode=RCTM_User, time=1, value=2}
        ]}
       YCurve {keys [
         {arriveTangent=2.8823, interpMode=RCIM_Cubic, leaveTangent=2.8823, tangentMode=RCTM_User, time=0, value=1}
         {arriveTangent=2.0297, interpMode=RCIM_Cubic, leaveTangent=2.0297, tangentMode=RCTM_User, time=1, value=2}
        ]}
      }}
     ScaleCurve=&Multiply_Vector2D_ByFloat(
      Float=&RandomRangeFloat(Minimum=1, Maximum=2)
      Vector2D=&RandomRangeVector2D(Minimum=<0.6 0.75>, Maximum=<0.8 1>)
     )
    ))
   &DynamicMaterialParameters(
    Index 0 Param 1=20
    Index 0 Param 3=1
    Index 0 Param 4=&FloatFromCurve(FloatCurve=DataInterface() {propertyValues {Curve {keys [{tangentMode=RCTM_User, time=0.1, value=0}, {tangentMode=RCTM_User, time=1, value=0.5}]}}})
    Module.Param0WriteEnabled=true
    Module.Param1WriteEnabled=true
    Module.Param2WriteEnabled=true
    Module.Param3WriteEnabled=true
   ) {module=DynamicMaterialParameters001}
   &Drag(Drag=5, Use Linear Drag=true, Use Rotational Drag=true)
   &InheritVelocity(
    Inherited Velocity Amount Scale=<-1 -1 -1>
    Inherited Velocity Speed Limit=5
    Source Speed Threshold=3
    Speed Limit=true
    UseSpeedThreshold=true
   )
   &SolveForcesAndVelocity(Clamp Velocity=false, Limit Acceleration=false)
   &ScaleColor(
    Scale Alpha=&FloatFromCurve(
     FloatCurve=DataInterface() {propertyValues {Curve {keys [
         {arriveTangent=9.0725, interpMode=RCIM_Cubic, leaveTangent=9.0725, tangentMode=RCTM_User, time=0, value=0}
         {arriveTangent=-2.5249, interpMode=RCIM_Cubic, leaveTangent=-2.5249, tangentMode=RCTM_User, time=0.3, value=0.9847}
         {arriveTangent=-0.0405, interpMode=RCIM_Cubic, leaveTangent=-0.0405, tangentMode=RCTM_User, time=1, value=0}
        ]}}}
     Scale Curve=0.8
    )
    Scale RGB=&MakeVectorFromLinearColorRGB(Color=$System.Tint)
    ScaleA=true
    ScaleRGB=false
   )
   /SetVariables(Particles.SubImageIndex=&Add_Float(
     A=&RandomRangeFloat(Minimum=0, Maximum=0.3)
     B=&FloatFromCurve(
      FloatCurve=DataInterface() {propertyValues {Curve {keys [{arriveTangent=120.5723, interpMode=RCIM_Cubic, leaveTangent=120.5723, tangentMode=RCTM_User, time=0, value=0}, <at 1 64 RCIM_Cubic>]}}}
      Scale Curve=&RandomRangeFloat(Minimum=0.3, Maximum=0.5)
     )
    )) {isSetParameters=true}
   &ScaleColor(
    Scale Alpha=&MakeFloatFromLinearColor(LinearColor=$System.TOD Color, Channel:ENiagaraFloatFromLinearColorOptions=::NewEnumerator3)
    Scale RGB=&MakeVectorFromLinearColorRGB(Color=$System.TOD Color)
    ScaleA=true
    ScaleRGB=true
   ) {module=ScaleColor001}
   &DynamicMaterialParameters(
    Index 0 Param 1=&Add_Float(
     A=&RandomRangeFloat(Minimum=0, Maximum=1)
     B=&FloatFromCurve(
      FloatCurve=DataInterface() {propertyValues {Curve {keys [
          <at 0 0 RCIM_Cubic>
          {arriveTangent=1.5947, interpMode=RCIM_Cubic, leaveTangent=1.5947, tangentMode=RCTM_User, time=0.2557, value=0.5257}
          <at 1 1 RCIM_Cubic>
         ]}}}
      CurveIndex=$Particles.Age
      Scale Curve=0.25
     )
    )
    Module.Index0Float0Write=true
    Module.Index0Float1Write=true
    Module.Index0Float2Write=true
    Module.Index0Float3Write=true
   )
  ]
  rendererClasses [&NiagaraSpriteRendererProperties]
  renderers [
   {rendererClass=&NiagaraSpriteRendererProperties}
   {rendererClass=&NiagaraSpriteRendererProperties, rendererIndex=1}
   {rendererClass=&NiagaraSpriteRendererProperties, rendererIndex=2}
  ]
 }
 NE_flameTrail {
  bEnabled=false
  emitterUpdateScript [
   &EmitterState(Inactive Response=::NewEnumerator1, Life Cycle Mode=::NewEnumerator1, Loop Behavior=::NewEnumerator1, UseLoopDelay=false)
   &SpawnRate(SpawnRate=0, Use Spawn Probability=false) {enabled=false}
   &SpawnParticlesFromOtherEmitter(
    Attribute Reader=DataInterface() {propertyValues {EmitterBinding {bindingMode=Other, emitterName=Snowflakes}}}
    Spawn Rate=100
    Use Maximum Age=false
    Use Minimum Age=false
   )
  ]
  particleSpawnScript [
   &InitializeParticle(
    Color=<rgba 0.2807 0.9052 1 0.1>
    Lifetime=0.5
    Ribbon Width=&RandomRangeFloat(Minimum=10, Maximum=12)
    Ribbon Width Mode=::NewEnumerator1
    UsePositionOffset=false
    Write Material Random=true
   )
   &SampleParticlesFromOtherEmitter(Color Sampling=::NewEnumerator1, Particle ID Sampling=::NewEnumerator1) {module=SampleParticlesFromOtherEmitter001}
  ]
  particleUpdateScript [
   &SolveForcesAndVelocity(Clamp Velocity=false, Limit Acceleration=false)
   &ParticleState()
   &ScaleRibbonWidth(Ribbon Width Scale=&FloatFromCurve(
     FloatCurve=DataInterface() {propertyValues {Curve {keys [{arriveTangent=-1.1377, interpMode=RCIM_Cubic, leaveTangent=-1.1377, tangentMode=RCTM_User, time=0, value=1}, <at 1 1>]}}}
     Scale Curve=0.4
    ))
   &ScaleColor(
    Scale Alpha=&FloatFromCurve(FloatCurve=DataInterface() {propertyValues {Curve {keys [{arriveTangent=0.3606, interpMode=RCIM_Cubic, leaveTangent=0.3606, tangentMode=RCTM_User, time=0, value=1}, <at 1 0>]}}})
    ScaleA=true
    ScaleRGB=true
   )
   &self:ReadRibbonSourceParticleAge_07(PAR=DataInterface() {propertyValues {EmitterBinding {bindingMode=Other, emitterName=Snowflakes}}}) {enabled=false, module=ReadRibbonSourceParticleAge_07}
   &ScaleColor(
    Scale Alpha=&FloatFromCurve(
     FloatCurve=DataInterface() {propertyValues {Curve {keys [
         {arriveTangent=3.0043, interpMode=RCIM_Cubic, leaveTangent=0.8333, tangentMode=RCTM_Break, time=0.1, value=0.5}
         {arriveTangent=-0.6087, interpMode=RCIM_Cubic, leaveTangent=-0.6087, tangentMode=RCTM_User, time=0.8, value=0.5}
         {arriveTangent=-0.0405, interpMode=RCIM_Cubic, leaveTangent=-0.0405, tangentMode=RCTM_User, time=0.895, value=0}
        ]}}}
     CurveIndex=$Emitter.NormalizedLoopAge
     Scale Curve=2
    )
    Scale RGB=&Multiply_VectorByFloat(Vector=<1 1 1>, Float=1)
    ScaleA=true
    ScaleRGB=false
   ) {module=ScaleColor002}
   &DynamicMaterialParameters(
    Index 0 Param 1=&FloatFromCurve(FloatCurve=DataInterface() {propertyValues {Curve {keys [<at 0 1>,<at 1 0.3>]}}})
    Index 0 Param 3=-3
    Index 0 Param 4=1
    Index 1 Param 1=30
    Module.Index0Float0Write=true
    Module.Index0Float1Write=true
    Module.Index0Float2Write=true
    Module.Index0Float3Write=true
    Module.Index1Float0Write=true
    Module.Index1Float1Write=true
    Module.Index1Float2Write=true
    Module.Index1Float3Write=true
    Write Parameter Index 1=::NewEnumerator1
   )
  ]
  rendererClasses [&NiagaraRibbonRendererProperties]
  renderers [{rendererClass=&NiagaraRibbonRendererProperties}]
 }
 Snow_Powder {
  emitterUpdateScript [
   &EmitterState(Life Cycle Mode=::NewEnumerator1, Loop Behavior=::NewEnumerator1, UseLoopDelay=false)
   &SpawnBurst_Instantaneous(Spawn Count=20, Use Spawn Probability=false, UseLoopCountLimit=false)
   &SpawnRate(SpawnRate=50, Use Spawn Probability=false) {enabled=false}
  ]
  particleSpawnScript [
   &InitializeParticle(
    Color=<rgba 0.4492 0.6108 0.8177 0.1>
    Lifetime=&RandomRangeFloat(Minimum=0.8, Maximum=1)
    Sprite Rotation Mode=::NewEnumerator2
    Sprite Size Mode=::NewEnumerator1
    Uniform Sprite Size Max=113
    Uniform Sprite Size Min=69
    UsePositionOffset=false
    Write Material Random=true
   )
   &AddVelocity(
    Cone Angle=360
    Cone Axis=<0 1 1>
    Module.UseConeFalloff=true
    Module.UseVelDistribution=true
    Override Seed=false
    Speed Falloff From Cone Axis=0.9
    Velocity Mode=::NewEnumerator2
    Velocity Speed=150
   ) {enabled=false}
   &ShapeLocation(
    Cylinder Height=1
    Cylinder Radius=30
    MaskSpawnGroup=false
    Override Seed=false
    Shape Primitive=::NewEnumerator1
    Surface Only=false
    UseOwnerScale=true
   )
   &AddVelocity(Module.Use Falloff Distance=false, Override Seed=false, Velocity Mode=::NewEnumerator1, Velocity Speed=250) {enabled=false, module=AddVelocity001}
  ]
  particleUpdateScript [
   &DynamicMaterialParameters(
    Index 0 Param 1=3
    Index 0 Param 2=&FloatFromCurve(FloatCurve=DataInterface() {propertyValues {Curve {keys [
         {arriveTangent=9.0725, interpMode=RCIM_Cubic, leaveTangent=9.0725, tangentMode=RCTM_User, time=0, value=0}
         {arriveTangent=-2.5249, interpMode=RCIM_Cubic, leaveTangent=-2.5249, tangentMode=RCTM_User, time=0.1408, value=0.9847}
         {arriveTangent=-0.0405, interpMode=RCIM_Cubic, leaveTangent=-0.0405, tangentMode=RCTM_User, time=1, value=0}
        ]}}})
    Module.Index0Param0Write=true
    Module.Index0Param1Write=true
    Module.Index0Param2Write=true
    Module.Index0Param3Write=true
   ) {enabled=false}
   &CameraOffset(Camera Offset Amount=&FloatFromCurve(
     FloatCurve=DataInterface() {propertyValues {Curve {keys [
         {arriveTangent=4.1785, interpMode=RCIM_Cubic, leaveTangent=4.1785, tangentMode=RCTM_User, time=0, value=0}
         {arriveTangent=1.1179, interpMode=RCIM_Cubic, leaveTangent=1.1179, tangentMode=RCTM_User, time=0.2042, value=0.6228}
         {arriveTangent=0.0224, interpMode=RCIM_Cubic, leaveTangent=0.0224, tangentMode=RCTM_User, time=1, value=1}
        ]}}}
     Scale Curve=55
    ))
   &Drag(
    Drag=&FloatFromCurve(FloatCurve=DataInterface() {propertyValues {Curve {keys [{arriveTangent=3.3491, interpMode=RCIM_Cubic, leaveTangent=3.3491, tangentMode=RCTM_User, time=0, value=2}, <at 1 3 RCIM_Cubic>]}}})
    Use Linear Drag=true
    Use Rotational Drag=true
   )
   &SolveForcesAndVelocity(Clamp Velocity=false, Limit Acceleration=false)
   &ParticleState()
   &ScaleColor(
    Scale Alpha=&Multiply_Float(
     A=&FloatFromCurve(
      FloatCurve=DataInterface() {propertyValues {Curve {keys [{interpMode=RCIM_Cubic, tangentMode=RCTM_User, time=0, value=1}, {interpMode=RCIM_Cubic, tangentMode=RCTM_User, time=1, value=0}]}}}
      Scale Curve=&RandomRangeFloat(Minimum=0.5, Maximum=1)
     )
     B=&MakeCustomFloatFromBool(Bool=$System.IsSynced, True Float=1, False Float=0.6)
    )
    Scale RGB=&MakeVectorFromLinearColorRGB(Color=$System.TODColorMultiplier)
    ScaleA=true
    ScaleRGB=true
   )
   &SubUVAnimation(
    "Sprite Renderer "=DataInterface() {propertyValues {SpriteRenderer=&self:Snow_Powder_4.NiagaraSpriteRendererProperties_0}}
    UseEndFrame=false
    UseStartFrame=false
   )
   &ScaleSpriteSize(Uniform Curve Sprite Scale=DataInterface() {propertyValues {Curve {keys [
        {arriveTangent=3.1143, interpMode=RCIM_Cubic, leaveTangent=3.1143, tangentMode=RCTM_User, time=0, value=0}
        {arriveTangent=-0.0405, interpMode=RCIM_Cubic, leaveTangent=-0.0405, tangentMode=RCTM_User, time=1, value=1}
       ]}}})
   &ScaleColor(
    Scale RGB=&MakeVectorFromLinearColorRGB(Color=&FN_Select_Color_from_Time_Of_Day(Morning=<rgba 1 0.8463 0.7189 1>, Day=<rgba 1 1 1 1>, Evening=<rgba 1 0.8371 0.7936 1>, Night=<rgba 0.1429 0.183 0.4497 1>))
    ScaleA=false
    ScaleRGB=true
   ) {module=ScaleColor001}
  ]
  rendererClasses [&NiagaraSpriteRendererProperties]
  renderers [{rendererClass=&NiagaraSpriteRendererProperties}]
 }
 Snowflakes {
  emitterUpdateScript [
   &EmitterState(Life Cycle Mode=::NewEnumerator1, Loop Behavior=::NewEnumerator1, Loop Duration=1, UseLoopDelay=false)
   &SpawnBurst_Instantaneous(Spawn Count=45, Use Spawn Probability=false, UseLoopCountLimit=false)
  ]
  particleSpawnScript [
   &InitializeParticle(
    Color=&MakeCustomLinearColorFromBool(Bool=&RandomBool(), True Color=<rgba 0.5794 2.4626 5 1>, False Color=<rgba 0.2456 0.5543 1 1>)
    Lifetime Min=0.5
    Lifetime Mode=::NewEnumerator1
    Sprite Size Mode=::NewEnumerator1
    Uniform Sprite Size Max=2
    Uniform Sprite Size Min=1
    UsePositionOffset=false
    Write Material Random=true
   ) {module=InitializeParticle001}
   &SphereLocation(
    MaskSpawnGroup=false
    Override Seed=false
    Sphere Radius=15
    Surface Only=true
    UseNonUniformScale=false
    UseOffset=true
    UseSphereOrientation=false
   )
   &AddVelocityFromPoint(
    Module.Use Falloff Distance=false
    Origin Offset=<0 0 -10>
    Velocity Strength=&RandomRangeFloat(Minimum=50, Maximum=400)
   )
   &AddVelocity(
    Cone Angle=60
    Cone Axis=<0 0 1>
    Module.UseConeFalloff=false
    Module.UseVelDistribution=false
    Override Seed=false
    Velocity Mode=::NewEnumerator2
    Velocity Speed=&RandomRangeFloat(Minimum=200, Maximum=300)
   ) {enabled=false}
  ]
  particleUpdateScript [
   &ParticleState()
   &CurlNoiseForce(Noise Frequency=10, Noise Strength=150, Pan Noise=false)
   &Drag(
    Drag=&FloatFromCurve(FloatCurve=DataInterface() {propertyValues {Curve {keys [<at 0.5 4>,<at 1 4>]}}})
    Use Linear Drag=true
    Use Rotational Drag=false
   )
   &GravityForce(Gravity=&RandomRangeVector(Minimum=<0 0 10>, Maximum=<0 0 100>))
   &SolveForcesAndVelocity(Clamp Velocity=false, Limit Acceleration=false)
   &SpriteRotationRate(
    Delta Time:NiagaraFloat=$Particles.NormalizedAge
    Rotation Rate=&RandomRangeFloat(Minimum=-5, Maximum=5)
   )
   &ScaleSpriteSize(Scale Factor=&Vector2DFromFloat(Value=&FloatFromCurve(FloatCurve=DataInterface() {propertyValues {Curve {keys [
          {arriveTangent=0.7564, interpMode=RCIM_Cubic, leaveTangent=0.7564, tangentMode=RCTM_User, time=-0.0024, value=0.2}
          {arriveTangent=2.045, interpMode=RCIM_Cubic, leaveTangent=2.045, time=0, value=1.55}
          <at 0.6439 1.5218>
          <at 0.994 0.1778 RCIM_Cubic>
         ]}}})))
   &ScaleColor(
    Scale Alpha=&Multiply_Float(
     A=&FloatFromCurve(FloatCurve=DataInterface() {propertyValues {Curve {keys [{interpMode=RCIM_Cubic, tangentMode=RCTM_User, time=0, value=1}, {interpMode=RCIM_Cubic, tangentMode=RCTM_User, time=1, value=0}]}}})
     B=1
    )
    ScaleA=true
    ScaleRGB=true
   )
   &DynamicMaterialParameters(
    Index 0 Param 2=5
    Module.Param0WriteEnabled=true
    Module.Param1WriteEnabled=true
    Module.Param2WriteEnabled=true
    Module.Param3WriteEnabled=true
   )
   &ScaleColor(
    Scale RGB=&MakeVectorFromLinearColorRGB(Color=&FN_Select_Color_from_Time_Of_Day(Morning=<rgba 1 0.8463 0.7189 1>, Day=<rgba 1 1 1 1>, Evening=<rgba 1 0.8371 0.7936 1>, Night=<rgba 0.1429 0.183 0.4497 1>))
    ScaleA=false
    ScaleRGB=true
   ) {module=ScaleColor001}
  ]
  rendererClasses [&NiagaraSpriteRendererProperties]
  renderers [{rendererClass=&NiagaraSpriteRendererProperties}]
 }
}
summary {
 systemName=NS_TicketPoke_GroundImpact
 emitters [
  {emitterName=Snowflakes, rendererClasses [&NiagaraSpriteRendererProperties]}
  {bEnabled=false, emitterName=NE_Debris_02, rendererClasses [&NiagaraMeshRendererProperties]}
  {
   bEnabled=false
   emitterName=Head
   rendererClasses [&NiagaraLightRendererProperties, &NiagaraSpriteRendererProperties]
  }
  {bEnabled=false, emitterName=NE_flameTrail, rendererClasses [&NiagaraRibbonRendererProperties]}
  {emitterName=NE_SmokeGroundRings001, rendererClasses [&NiagaraSpriteRendererProperties]}
  {bEnabled=false, emitterName=NE_WindstreaksJet001, rendererClasses [&NiagaraSpriteRendererProperties]}
  {bEnabled=false, emitterName=Mist, rendererClasses [&NiagaraSpriteRendererProperties]}
  {emitterName=Snow_Powder, rendererClasses [&NiagaraSpriteRendererProperties]}
 ]
 userVariables [
  User.bIsFrontEndPreview {type=NiagaraBool}
  User.Color {type=LinearColor, description="", defaultValue {struct=LinearColor, value=<rgba 1.3594 2.4773 6 1>}}
  User.Color_Blade_B {type=LinearColor, description="", defaultValue {struct=LinearColor, value=<rgba 0.2266 0.4129 1 1>}}
  User.exp_Delay {type=NiagaraFloat, defaultValue=0.0325, description=""}
  User.Skeletal Mesh Reader {
   type=NiagaraDataInterfaceSkeletalMesh
   defaultValue {
    struct=NiagaraExt_VariableValue_DataInterface
    value {dataInterface=&self:NiagaraDataInterfaceSkeletalMesh_0}
   }
  }
  User.UseCustomColor {type=NiagaraFloat, defaultValue=0, description=""}
  User.UseCustomColor1 {type=NiagaraFloat, defaultValue=0, description=""}
 ]
}
