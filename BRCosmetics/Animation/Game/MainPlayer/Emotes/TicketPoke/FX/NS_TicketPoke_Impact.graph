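# Niagara system for the impact effect of the TicketPoke emote, built from burst-driven emitters.
# Active emitters in the visible portion include Flash_Elements (star, spike and flare sprites plus a light) and HandGlow, a flickering sprite and mesh glow.
# Per-element lifetimes, sizes and colors are stored as custom particle attributes, adjusted for time of day, front end and platform.
# Several other emitters, including BG_Glow, Dust, Flare and Halo, are disabled.
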NiagaraSystem /BRCosmetics/Animation/Game/MainPlayer/Emotes/TicketPoke/FX/NS_TicketPoke_Impact
dependencies {
 usedDataInterfaces [
  &NiagaraDataInterfaceCurve
  &NiagaraDataInterfaceColorCurve
  &NiagaraDataInterfaceParticleRead
  &NiagaraDataInterfaceSpriteRendererInfo
  &NiagaraDataInterfaceVector2DCurve
  &NiagaraDataInterfaceCamera
  &NiagaraDataInterfaceMeshRendererInfo
  &NiagaraDataInterfaceVectorCurve
  &NiagaraDataInterfacePlatformSet
  &NiagaraDataInterfaceArrayFloat
 ]
 usedDynamicInputs [&LerpLinearColors]
 usedModules [
  &SystemState
  &self:SystemSpawnScript.SystemScriptSource.SystemScriptGraph.NiagaraNodeAssignment_0.SetVariables_#1
 ]
}
emitters {
 BG_Glow {
  bEnabled=false
  emitterUpdateScript [
   &EmitterState(Life Cycle Mode=::NewEnumerator1, Loop Behavior=::NewEnumerator1, Loop Delay=0.05, Loop Duration=1, UseLoopDelay=true)
   &SpawnBurst_Instantaneous(Spawn Count=1, Use Spawn Probability=false)
  ]
  particleSpawnScript [
   &InitializeParticle(
    Color=<rgba 0.1074 0.3052 0.3073 1>
    Lifetime=0.5
    Sprite Size Mode=::NewEnumerator3
    Uniform Sprite Size=200
    UsePositionOffset=false
    Write Material Random=true
   )
   &CameraOffset(Camera Offset Amount=-20)
  ]
  particleUpdateScript [
   &ParticleState()
   &ScaleSpriteSize(Scale Factor=&Vector2DFromFloat(Value=&FloatFromCurve(FloatCurve=DataInterface() {propertyValues {Curve {keys [
          {arriveTangent=0.4346, interpMode=RCIM_Cubic, leaveTangent=0.4346, tangentMode=RCTM_User, time=0, value=0}
          {arriveTangent=0.0473, interpMode=RCIM_Cubic, leaveTangent=0.0473, tangentMode=RCTM_User, time=0.2, value=1}
         ]}}})))
   &ScaleColor(
    Scale Alpha=&FloatFromCurve(
     FloatCurve=DataInterface() {propertyValues {Curve {keys [
         {arriveTangent=0.4005, interpMode=RCIM_Cubic, leaveTangent=0.4005, tangentMode=RCTM_User, time=0, value=0}
         {arriveTangent=0.0256, interpMode=RCIM_Cubic, leaveTangent=0.0256, tangentMode=RCTM_User, time=0.2, value=1}
         <at 1 0 RCIM_Cubic>
        ]}}}
     Scale Curve=&FNTimeOfDayScalar(Night Scale=0.8, Morning Scale=0.8, Evening Scale=0.8)
    )
    ScaleA=true
    ScaleRGB=true
   )
   /SetVariables(
    Particles.GlowColor2:LinearColor=&MakeLinearColorFromVectorAndFloat(
     "Vector (RGB)"=&MakeVectorFromLinearColorRGB(Color=<rgba 0.4427 0.6706 5 1>)
     "Float (Alpha)"=&Multiply_Float(
      A=0.2
      B=&FloatFromCurve(FloatCurve=DataInterface() {propertyValues {Curve {keys [
           {arriveTangent=0.4005, interpMode=RCIM_Cubic, leaveTangent=0.4005, tangentMode=RCTM_User, time=0, value=0}
           {arriveTangent=0.0256, interpMode=RCIM_Cubic, leaveTangent=0.0256, tangentMode=RCTM_User, time=0.2, value=1}
           <at 1 0 RCIM_Cubic>
          ]}}})
     )
    )
    Particles.GlowSize2:Vector2f=&Vector2DFromFloat(Value=&Multiply_Float(
      A=120
      B=&FloatFromCurve(
       FloatCurve=DataInterface() {propertyValues {Curve {keys [
           {arriveTangent=0.4346, interpMode=RCIM_Cubic, leaveTangent=0.4346, tangentMode=RCTM_User, time=0, value=0}
           {arriveTangent=0.0473, interpMode=RCIM_Cubic, leaveTangent=0.0473, tangentMode=RCTM_User, time=0.2, value=1}
          ]}}}
       Scale Curve=&FNTimeOfDayScalar(Night Scale=0.6, Morning Scale=0.4, Evening Scale=0.5)
      )
     ))
   ) {isSetParameters=true}
   /SetVariables(
    Particles.GlowColor03:LinearColor=&MultiplyLinearColor(
     Linear Color=&ColorFromCurve(ColorCurve=DataInterface() {propertyValues {
        AlphaCurve {keys [<at 0 1>,<at 0.1 0.7>,<at 1 0>]}
        BlueCurve {keys [<at 0 5>,<at 0.1 1>,<at 1 1>]}
        GreenCurve {keys [<at 0 4.6457>,<at 0.1 0.7037>,<at 1 0.8701>]}
        RedCurve {keys [<at 0 2.0839>,<at 0.1 0.0341>,<at 1 0.3484>]}
       }})
     Scale Factor=&MakeFloat4FromVectorAndFloat(
      "Vector (RGB)"=<1 1 1>
      "Float (Alpha)"=&FNTimeOfDayScalar(Night Scale=0.3, Morning Scale=0.3, Evening Scale=0.5)
     )
    )
    Particles.GlowSize03:Vector2f=&Vector2DFromFloat(Value=&FloatFromCurve(
      FloatCurve=DataInterface() {propertyValues {Curve {keys [{tangentMode=RCTM_User, time=0, value=0}, {tangentMode=RCTM_User, time=0.1, value=1}, <at 1 3>]}}}
      Scale Curve=220
     ))
   ) {isSetParameters=true}
  ]
  rendererClasses [&NiagaraSpriteRendererProperties]
  renderers [
   {rendererClass=&NiagaraSpriteRendererProperties}
   {rendererClass=&NiagaraSpriteRendererProperties, rendererIndex=1}
   {rendererClass=&NiagaraSpriteRendererProperties, rendererIndex=2}
  ]
 }
 Dust {
  bEnabled=false
  emitterUpdateScript [
   &EmitterState(Life Cycle Mode=::NewEnumerator1, Loop Behavior=::NewEnumerator1, Loop Duration=2, UseLoopDelay=false)
   &SpawnBurst_Instantaneous(
    Spawn Count=&RandomRangeInt(Minimum=8, Maximum=12)
    Use Spawn Probability=false
    UseLoopCountLimit=false
   )
   &SpawnBurst_Instantaneous(
    Spawn Count=&RandomRangeInt(Minimum=8, Maximum=12)
    Spawn Group=2
    Use Spawn Probability=false
    UseLoopCountLimit=false
   ) {module=SpawnBurst_Instantaneous002}
   &SpawnBurst_Instantaneous(
    Spawn Count=&RandomRangeInt(Minimum=8, Maximum=12)
    Spawn Group=1
    Spawn Time=0.1
    Use Spawn Probability=false
    UseLoopCountLimit=false
   ) {module=SpawnBurst_Instantaneous001}
   &SpawnBurst_Instantaneous(
    Spawn Count=&RandomRangeInt(Minimum=8, Maximum=12)
    Spawn Group=3
    Spawn Time=0.1
    Use Spawn Probability=false
    UseLoopCountLimit=false
   ) {module=SpawnBurst_Instantaneous003}
  ]
  particleSpawnScript [
   /SetVariables(Particles.VisibilityTag=&MakeCustomIntFromBool(
     Bool=&SetBoolByIntComparison(A=&SpawnGroup(), B=2, Comparison Type=::NewEnumerator2)
     True INT=1
     False INT=0
    )) {isSetParameters=true}
   /SetVariables(
    Particles.SizeMax:NiagaraFloat=&MakeCustomFloatFromBool(
     Bool=&SetBoolByIntComparison(A=&SpawnGroup(), B=2, Comparison Type=::NewEnumerator2)
     True Float=&RandomRangeFloat(Minimum=45, Maximum=65)
     False Float=&RandomRangeFloat(Minimum=155, Maximum=185)
    )
    Particles.SizeMin:NiagaraFloat=&MakeCustomFloatFromBool(
     Bool=&SetBoolByIntComparison(A=&SpawnGroup(), B=2, Comparison Type=::NewEnumerator2)
     True Float=&RandomRangeFloat(Minimum=32.5, Maximum=42.5)
     False Float=&RandomRangeFloat(Minimum=82.5, Maximum=92.5)
    )
   ) {isSetParameters=true}
   /SetVariables(Particles.SizeCurve:NiagaraFloat=&Lerp_Float(
     A=$Particles.SizeMin
     B=$Particles.SizeMax
     Alpha=&FloatFromCurve(
      FloatCurve=DataInterface() {propertyValues {Curve {keys [
          {interpMode=RCIM_Cubic, tangentMode=RCTM_Break, time=0, value=0}
          {arriveTangentWeight=0.9217, interpMode=RCIM_Cubic, tangentMode=RCTM_Break, time=1, value=1}
         ]}}}
      CurveIndex=&RandomRangeFloat(Minimum=0, Maximum=1)
      Scale Curve=0.1
     )
    )) {isSetParameters=true}
   &InitializeParticle(
    Color=&MakeLinearColorFromVectorAndFloat(
     "Vector (RGB)"=&MakeVectorFromLinearColorRGB(Color=&FNTimeOfDayColor(Night Color=<rgba 1 1 1 1>, Morning Color=<rgba 1 1 1 1>, Day Color=<rgba 1 1 1 1>, Evening Color=<rgba 1 0.9 0.9166 1>))
     "Float (Alpha)"=&MaskFloatBySpawnGroup(
      Masked Float=&RandomRangeFloat(Minimum=0.25, Maximum=0.25)
      Passthrough Float=&MaskFloatBySpawnGroup(
       Masked Float=&RandomRangeFloat(Minimum=1, Maximum=1)
       Passthrough Float=&RandomRangeFloat(Minimum=0.3, Maximum=0.3)
       Spawn Group Mask=1
      )
     )
    )
    Lifetime=&MaskFloatBySpawnGroup(
     Masked Float=&RandomRangeFloat(Minimum=1, Maximum=1.5)
     Passthrough Float=&MaskFloatBySpawnGroup(Masked Float=&RandomRangeFloat(Minimum=0.5, Maximum=1), Passthrough Float=1, Spawn Group Mask=1)
    )
    Sprite Rotation Mode=::NewEnumerator2
    Sprite Size Mode=::NewEnumerator3
    Uniform Sprite Size=$Particles.SizeCurve
    UsePositionOffset=false
    Write Material Random=true
   )
   &ShapeLocation(Hemisphere Distribution=<1 0.5>, MaskSpawnGroup=true, Override Seed=false, Spawn Group Mask=1, Sphere Radius=20, UseOwnerScale=true)
   &AddVelocity(Module.Use Falloff Distance=false, Override Seed=false, Velocity Mode=::NewEnumerator1, Velocity Speed=100) {module=AddVelocity001}
   &CameraOffset(Camera Offset Amount=75)
   &SubUVAnimation(
    Evaluation Type=::NewEnumerator0
    "Sprite Renderer "=DataInterface() {propertyValues {SpriteRenderer=&self:Dust_2.NiagaraSpriteRendererProperties_2}}
    SubUV Animation Mode:ENiagara_SubUVLookupModeV2=::NewEnumerator2
    UseEndFrame=false
    UseStartFrame=false
   )
  ]
  particleUpdateScript [
   &ParticleState()
   &ScaleSpriteSize(
    Scale Sprite Size Mode=::NewEnumerator0
    Uniform Scale Factor=&MaskFloatBySpawnGroup(
     Masked Float=&FloatFromCurve(FloatCurve=DataInterface() {propertyValues {Curve {keys [{arriveTangent=6.1888, interpMode=RCIM_Cubic, leaveTangent=6.1888, tangentMode=RCTM_User, time=0, value=0}, <at 0.26 1 RCIM_Cubic>]}}})
     Passthrough Float=1
    )
   )
   &ScaleColor(
    Scale Alpha=&FloatFromCurve(
     FloatCurve=DataInterface() {propertyValues {Curve {keys [<at 0 1>,<at 1 0>]}}}
     Scale Curve=&FNTimeOfDayScalar(Night Scale=0.3, Morning Scale=0.6, Day Scale=0.2, Evening Scale=0.3)
    )
    Scale RGB=&MakeVectorFromLinearColorRGB(Color=&FNTimeOfDayColor(
      Night Color=&ColorFromCurve(ColorCurve=DataInterface() {propertyValues {
         AlphaCurve {keys []}
         BlueCurve {keys [<at 0 0.5 RCIM_Cubic>,<at 1 0.3552 RCIM_Cubic>]}
         GreenCurve {keys [<at 0 0.3598 RCIM_Cubic>,<at 1 0.2628 RCIM_Cubic>]}
         RedCurve {keys [<at 0 0.3076 RCIM_Cubic>,<at 1 0.2315 RCIM_Cubic>]}
        }})
      Morning Color=&ColorFromCurve(ColorCurve=DataInterface() {propertyValues {
         AlphaCurve {keys []}
         BlueCurve {keys [<at 0 0.5104 RCIM_Cubic>,<at 1 0.3552 RCIM_Cubic>]}
         GreenCurve {keys [<at 0 0.5104 RCIM_Cubic>,<at 1 0.3552 RCIM_Cubic>]}
         RedCurve {keys [<at 0 0.5104 RCIM_Cubic>,<at 1 0.3552 RCIM_Cubic>]}
        }})
      Day Color=&ColorFromCurve(ColorCurve=DataInterface() {propertyValues {
         AlphaCurve {keys []}
         BlueCurve {keys [<at 0 0.75 RCIM_Cubic>,<at 1 0.5 RCIM_Cubic>]}
         GreenCurve {keys [<at 0 0.75 RCIM_Cubic>,<at 1 0.5 RCIM_Cubic>]}
         RedCurve {keys [<at 0 0.75 RCIM_Cubic>,<at 1 0.5 RCIM_Cubic>]}
        }})
      Evening Color=&ColorFromCurve(ColorCurve=DataInterface() {propertyValues {
         AlphaCurve {keys []}
         BlueCurve {keys [<at 0 0.1781 RCIM_Cubic>,<at 1 0.2411 RCIM_Cubic>]}
         GreenCurve {keys [<at 0 0.1982 RCIM_Cubic>,<at 1 0.2747 RCIM_Cubic>]}
         RedCurve {keys [<at 0 0.2552 RCIM_Cubic>,<at 1 0.3552 RCIM_Cubic>]}
        }})
     ))
    ScaleA=true
    ScaleRGB=true
   )
   &GravityForce(Gravity=&RandomRangeVector(Minimum=<0 0 -150>, Maximum=<0 0 -200>))
   &Drag(Use Linear Drag=true, Use Rotational Drag=false)
   &PointAttractionForce(
    Attraction Radius=150
    Attraction Strength=&MaskFloatBySpawnGroup(Masked Float=-10, Passthrough Float=0)
    Kill Within Radius=false
    Module.Use Falloff=true
    UseAttractorOffset=false
   )
   &SolveForcesAndVelocity(Clamp Velocity=false, Limit Acceleration=false)
   &DynamicMaterialParameters(
    Index 0 Param 1=&Add_Float(
     A=&RandomRangeFloat(Minimum=0, Maximum=1)
     B=&FloatFromCurve(
      FloatCurve=DataInterface() {propertyValues {Curve {keys [
          {arriveTangent=2.7533, interpMode=RCIM_Cubic, leaveTangent=1.8958, tangentMode=RCTM_User, time=0, value=1}
          {arriveTangent=0.8292, interpMode=RCIM_Cubic, leaveTangent=0.8292, tangentMode=RCTM_User, time=0.321, value=1.3453}
          {arriveTangent=0.0529, interpMode=RCIM_Cubic, leaveTangent=0.2021, tangentMode=RCTM_User, time=1, value=1.5}
         ]}}}
      CurveIndex=&Add_Float(A=&Multiply_Float(A=$Particles.Initial.MaterialRandom, B=0.5), B=$Particles.NormalizedAge)
      Scale Curve=&Lerp_Float(
       A=0.25
       B=0.7
       Alpha=&NormalizeFloat(Value=$Particles.SizeCurve, Min=$Particles.SizeMin, Max=$Particles.SizeMax)
      )
     )
    )
    Index 0 Param 2=&FloatFromCurve(FloatCurve=DataInterface() {propertyValues {Curve {keys [<at 0 1>,<at 1 0>]}}})
    Index 0 Param 3=&FloatFromCurve(
     FloatCurve=DataInterface() {propertyValues {Curve {keys [
         {interpMode=RCIM_Cubic, leaveTangent=16.5652, leaveTangentWeight=0.4385, tangentMode=RCTM_Break, time=0, value=0}
         {arriveTangentWeight=0.0333, interpMode=RCIM_Cubic, leaveTangentWeight=0.9477, tangentMode=RCTM_Break, time=0.1, value=1}
         {arriveTangent=-1.7159, interpMode=RCIM_Cubic, leaveTangent=-1.7159, tangentMode=RCTM_User, time=0.3, value=0.2}
         {arriveTangent=0.2347, interpMode=RCIM_Cubic, leaveTangent=0.2723, tangentMode=RCTM_Break, time=0.6, value=0}
        ]}}}
     Scale Curve=0
    )
    Index 0 Param 4=&FloatFromCurve(FloatCurve=DataInterface() {propertyValues {Curve {keys [{interpMode=RCIM_Cubic, tangentMode=RCTM_Break, time=0.5, value=1}, {interpMode=RCIM_Cubic, tangentMode=RCTM_Break, time=1, value=0}]}}})
    Index 1 Param 1=&FloatFromCurve(
     FloatCurve=DataInterface() {propertyValues {Curve {keys [
         {interpMode=RCIM_Cubic, leaveTangentWeight=0.0833, tangentMode=RCTM_Break, time=0, value=0}
         {arriveTangentWeight=0.0873, interpMode=RCIM_Cubic, tangentMode=RCTM_Break, time=0.25, value=1}
        ]}}}
     Scale Curve=&Lerp_Float(
      A=0.5
      B=0.75
      Alpha=&NormalizeFloat(Value=$Particles.SizeCurve, Min=$Particles.SizeMin, Max=$Particles.SizeMax)
     )
    )
    Index 1 Param 2=&FloatFromCurve(
     FloatCurve=DataInterface() {propertyValues {Curve {keys [
         {interpMode=RCIM_Cubic, tangentMode=RCTM_Break, time=0, value=0.2}
         {interpMode=RCIM_Cubic, leaveTangent=0.9066, tangentMode=RCTM_Break, time=1, value=2}
        ]}}}
     Scale Curve=&RandomRangeFloat(Minimum=0.1, Maximum=0.2)
    )
    Index 1 Param 3=&FloatFromCurve(FloatCurve=DataInterface() {propertyValues {Curve {keys [
         {interpMode=RCIM_Cubic, leaveTangent=6.3917, tangentMode=RCTM_Break, time=0, value=0}
         {arriveTangent=1.7549, interpMode=RCIM_Cubic, leaveTangent=10.6767, tangentMode=RCTM_User, time=0.25, value=1}
        ]}}})
    Index 1 Param 4=&FloatFromCurve(
     FloatCurve=DataInterface() {propertyValues {Curve {keys [{arriveTangent=-0.136, interpMode=RCIM_Cubic, leaveTangent=-0.136, tangentMode=RCTM_User, time=0, value=1}, <at 1 0.95>]}}}
     Scale Curve=&Lerp_Float(
      A=2
      B=5
      Alpha=&NormalizeFloat(Value=$Particles.SizeCurve, Min=$Particles.SizeMin, Max=$Particles.SizeMax)
     )
    )
    Index 2 Param 1=&FloatFromCurve(
     FloatCurve=DataInterface() {propertyValues {Curve {keys [
         {interpMode=RCIM_Cubic, leaveTangent=4.2485, tangentMode=RCTM_Break, time=0, value=0}
         {arriveTangent=1.0548, interpMode=RCIM_Cubic, leaveTangent=1.0548, tangentMode=RCTM_User, time=0.17, value=0.5715}
         <at 1 1>
        ]}}}
     Scale Curve=3
    )
    Index 2 Param 2=&FloatFromCurve(FloatCurve=DataInterface() {propertyValues {Curve {keys [
         {interpMode=RCIM_Cubic, leaveTangentWeight=0.5024, tangentMode=RCTM_User, time=0, value=1}
         {arriveTangentWeight=0.4857, interpMode=RCIM_Cubic, tangentMode=RCTM_User, time=1, value=1}
        ]}}})
    Index 2 Param 3=&FloatFromCurve(
     FloatCurve=DataInterface() {propertyValues {Curve {keys [
         {interpMode=RCIM_Cubic, leaveTangent=-19.1446, leaveTangentWeight=0.1411, tangentMode=RCTM_Break, time=0, value=1}
         {interpMode=RCIM_Cubic, tangentMode=RCTM_Break, time=0.1, value=0}
         {interpMode=RCIM_Cubic, leaveTangent=11.1233, tangentMode=RCTM_Break, time=1, value=1}
        ]}}}
     Scale Curve=5
    )
    Index 2 Param 4=&RandomRangeFloat(Minimum=1, Maximum=1.75)
    Module.Param0WriteEnabled=true
    Module.Param0WriteEnabled1=true
    Module.Param0WriteEnabled2=true
    Module.Param1WriteEnabled=false
    Module.Param1WriteEnabled1=true
    Module.Param1WriteEnabled2=true
    Module.Param2WriteEnabled=true
    Module.Param2WriteEnabled1=true
    Module.Param2WriteEnabled2=true
    Module.Param3WriteEnabled=true
    Module.Param3WriteEnabled1=true
    Module.Param3WriteEnabled2=true
    Write Parameter Index 1=true
    Write Parameter Index 2=true
   )
  ]
  rendererClasses [&NiagaraSpriteRendererProperties]
  renderers [
   {rendererClass=&NiagaraSpriteRendererProperties}
   {rendererClass=&NiagaraSpriteRendererProperties, rendererIndex=1}
  ]
 }
 Flare {
  bEnabled=false
  emitterUpdateScript [
   &EmitterState(
    Inactive Response=::NewEnumerator1
    Life Cycle Mode=::NewEnumerator1
    Loop Behavior=::NewEnumerator1
    Loop Delay=$User.exp_Delay
    Loop Duration=3
    UseLoopDelay=true
   )
   &SpawnBurst_Instantaneous(Use Spawn Probability=false, UseLoopCountLimit=true)
  ]
  particleSpawnScript [
   &InitializeParticle(
    Color=<rgba 0.0344 0.6218 1 1>
    Lifetime Max=0.1667
    Lifetime Min=0.1333
    Lifetime Mode=::NewEnumerator1
    Sprite Size Mode=::NewEnumerator1
    Uniform Sprite Size Max=185
    Uniform Sprite Size Min=170
    UsePositionOffset=true
    Write Material Random=true
   )
   &SpriteFacingAndAlignment(Sprite Alignment=<0 0 1>, Sprite Facing=<0 0 1>, WriteAlignment=true, WriteFacing=false)
   &CameraOffset(Camera Offset Amount=60)
   &DynamicMaterialParameters(Write Parameter Index 0=false)
  ]
  particleUpdateScript [
   &ParticleState()
   &ScaleColor(
    Scale Alpha=&FloatFromCurve(FloatCurve=DataInterface() {propertyValues {Curve {keys [
         {arriveTangent=6.6296, interpMode=RCIM_Cubic, leaveTangent=6.6296, tangentMode=RCTM_User, time=0, value=0}
         {interpMode=RCIM_Cubic, tangentMode=RCTM_User, time=0.25, value=1}
         {tangentMode=RCTM_User, time=1, value=0}
        ]}}})
    ScaleA=true
    ScaleRGB=true
   )
   &ScaleSpriteSize(Uniform Curve Sprite Scale=DataInterface() {propertyValues {Curve {keys [
        {arriveTangent=7.5797, interpMode=RCIM_Cubic, leaveTangent=7.5797, tangentMode=RCTM_User, time=0, value=0.25}
        {interpMode=RCIM_Cubic, tangentMode=RCTM_User, time=0.25, value=1}
        {arriveTangent=-1.3776, interpMode=RCIM_Cubic, leaveTangent=-1.3776, tangentMode=RCTM_User, time=1, value=0.5}
       ]}}})
   /SetVariables(
    Particles.Color Wide:LinearColor=&MultiplyLinearColor(Linear Color=$Particles.Color, Scale Factor=<0.5 0.5 0.5 1>)
    Particles.SpriteSize_Wide:Vector2f=&Multiply_Vector2D(A=$Particles.SpriteSize, B=<2 0.15>)
   ) {isSetParameters=true}
   /SetVariables(Particles.Position_Light:NiagaraPosition=&Add_Vector(A=Linked() {linkedVariable {name=Particles.Initial.Position, type=NiagaraPosition}}, B=<0 0 0>)) {isSetParameters=true}
   &FN_Light_Attributes(
    Color=$User.Color_Blade_B
    Create Color Attribute=true
    Create Exponent Attribute=true
    Create Radius Attribute=true
    Exponent=2.5
    Radius=&Multiply_Float(A=&MakeFloatFromVector2d(Vector 2D=$Particles.SpriteSize), B=1.3)
    Scale Color=&FloatFromCurve(
     FloatCurve=DataInterface() {propertyValues {Curve {keys [{interpMode=RCIM_Cubic, tangentMode=RCTM_User, time=0.25, value=1}, {interpMode=RCIM_Cubic, tangentMode=RCTM_User, time=1, value=0}]}}}
     Scale Curve=3.5
    )
   )
  ]
  rendererClasses [&NiagaraSpriteRendererProperties, &NiagaraLightRendererProperties]
  renderers [
   {rendererClass=&NiagaraSpriteRendererProperties}
   {rendererClass=&NiagaraSpriteRendererProperties, rendererIndex=1}
   {rendererClass=&NiagaraLightRendererProperties, rendererIndex=2}
  ]
 }
 Flash001 {
  bEnabled=false
  emitterUpdateScript [
   &EmitterState(Life Cycle Mode=::NewEnumerator1, Loop Behavior=::NewEnumerator1, Loop Delay=$User.exp_Delay, Loop Duration=3, UseLoopDelay=true)
   &SpawnBurst_Instantaneous(Use Spawn Probability=false, UseLoopCountLimit=false)
  ]
  particleSpawnScript [
   &InitializeParticle(
    Color=<rgba 1.3594 2.4773 6 1>
    Lifetime=0.27
    Mesh Scale Mode=::NewEnumerator3
    Mesh Uniform Scale=0.4
    Sprite Size Mode=::NewEnumerator3
    Uniform Sprite Size=120
    UsePositionOffset=true
    Write Material Random=true
   )
   &CameraOffset(Camera Offset Amount=50)
  ]
  particleUpdateScript [
   &ParticleState()
   &ScaleMeshSize(Scale Factor=&VectorFromFloat(Value=&FloatFromCurve(FloatCurve=DataInterface() {propertyValues {Curve {keys [
          {arriveTangent=1.5171, interpMode=RCIM_Cubic, leaveTangent=1.5171, tangentMode=RCTM_User, time=0, value=0.4}
          {arriveTangent=0.0224, interpMode=RCIM_Cubic, leaveTangent=0.0224, tangentMode=RCTM_User, time=1, value=1}
         ]}}})))
   &DynamicMaterialParameters(
    Index 0 Param 1=1
    Index 0 Param 2=1
    Index 1 Param 1=100
    Index 3 Param 1=&FloatFromCurve(FloatCurve=DataInterface() {propertyValues {Curve {keys [
         {interpMode=RCIM_Cubic, tangentMode=RCTM_User, time=0.5, value=0}
         {arriveTangent=2.1268, interpMode=RCIM_Cubic, leaveTangent=2.1268, tangentMode=RCTM_User, time=1, value=1}
        ]}}})
    Index 3 Param 2=&FloatFromCurve(FloatCurve=DataInterface() {propertyValues {Curve {keys [{arriveTangent=-2.5304, interpMode=RCIM_Cubic, time=0, value=2.5}, {interpMode=RCIM_Cubic, leaveTangent=-3.141, time=1, value=0.4}]}}})
    Index 3 Param 3=&FloatFromCurve(FloatCurve=DataInterface() {propertyValues {Curve {keys [
         {arriveTangent=-1.2425, interpMode=RCIM_Cubic, time=0, value=1}
         {arriveTangent=-4.4048, interpMode=RCIM_Cubic, leaveTangent=-4.4048, tangentMode=RCTM_User, time=1, value=-0.7}
        ]}}})
    Index 3 Param 4=&FloatFromCurve(
     FloatCurve=DataInterface() {propertyValues {Curve {keys [
         {arriveTangent=2.3563, interpMode=RCIM_Cubic, leaveTangent=2.3563, tangentMode=RCTM_User, time=0, value=0}
         {tangentMode=RCTM_User, time=1, value=1}
        ]}}}
     Scale Curve=3
    )
    Module.Index0Param0Write=true
    Module.Index0Param1Write=true
    Module.Index0Param2Write=true
    Module.Index0Param3Write=true
    Module.Index1Param0Write=true
    Module.Index1Param1Write=false
    Module.Index1Param2Write=false
    Module.Index1Param3Write=false
    Module.Index3Param0Write=true
    Module.Index3Param1Write=true
    Module.Index3Param2Write=true
    Module.Index3Param3Write=true
    Write Parameter Index 1=true
    Write Parameter Index 3=true
   )
   &ScaleColor(
    Scale Alpha=&FloatFromCurve(
     FloatCurve=DataInterface() {propertyValues {Curve {keys [
         {arriveTangent=-2.4409, arriveTangentWeight=0.162, interpMode=RCIM_Cubic, leaveTangentWeight=0.2268, time=0.2, value=1}
         {arriveTangent=-3.5548, interpMode=RCIM_Cubic, leaveTangent=-3.5548, tangentMode=RCTM_User, time=0.5, value=0.5}
         <at 1 0>
        ]}}}
     Scale Curve=2.5
    )
    Scale RGB=&VectorFromFloat(Value=0.5)
    ScaleA=true
    ScaleRGB=true
   )
   /SetVariables(Particles.LightColor=&MultiplyLinearColor(Linear Color=$Particles.Color, Scale Factor=&MakeVector4_FromFloat(Float=2))) {isSetParameters=true}
   /SetVariables(Particles.RingColor:LinearColor=&MakeLinearColorFromVectorAndFloat(
     "Vector (RGB)"=&MakeVectorFromLinearColorRGB(Color=$Particles.Initial.Color)
     "Float (Alpha)"=&FloatFromCurve(FloatCurve=DataInterface() {propertyValues {Curve {keys [<at 0 1 RCIM_Cubic>,<at 1 0>]}}}, Scale Curve=0.1)
    )) {isSetParameters=true}
   &ScaleSpriteSize(Uniform Curve Sprite Scale=DataInterface() {propertyValues {Curve {keys [
        {arriveTangent=6.7629, interpMode=RCIM_Cubic, leaveTangent=6.7629, tangentMode=RCTM_User, time=0, value=0.75}
        {arriveTangent=-0.7511, interpMode=RCIM_Cubic, leaveTangent=-0.7511, time=0.11, value=1}
        <at 0.9986 0>
       ]}}})
  ]
  rendererClasses [&NiagaraSpriteRendererProperties, &NiagaraMeshRendererProperties]
  renderers [
   {rendererClass=&NiagaraSpriteRendererProperties}
   {rendererClass=&NiagaraMeshRendererProperties, rendererIndex=1}
  ]
 }
 Flash_Elements {
  emitterUpdateScript [
   &EmitterState(Life Cycle Mode=::NewEnumerator1, Loop Behavior=::NewEnumerator1, Loop Duration=2, UseLoopDelay=true)
   &SpawnBurst_Instantaneous(Spawn Group=1, Use Spawn Probability=false, UseLoopCountLimit=false) {module=SpawnBurst_Instantaneous001}
   &SpawnBurst_Instantaneous(
    Spawn Count=&RandomRangeInt(Minimum=3, Maximum=5)
    Use Spawn Probability=false
    UseLoopCountLimit=false
   )
  ]
  particleSpawnScript [
   &InitializeParticle(Color Mode=::NewEnumerator0, Lifetime=0.4, UsePositionOffset=false, Write Material Random=true)
   &ShapeLocation(MaskSpawnGroup=false, Override Seed=false, Sphere Radius=1, UseOwnerScale=true)
   &SpriteFacingAndAlignment(Sprite Alignment=$Particles.ShapeLocation.ShapeNormal, WriteAlignment=true, WriteFacing=false)
   &FN_PerParticleLOD(Enable SpriteSize Distance Scaling=true, Sprite Size Scale Far=<2.5 2.5>)
   /SetVariables(
    Particles.SpriteSize=&Multiply_Vector2D_ByFloat(Float=&FN_FrontEndFloatModifier(Gameplay=1, Front End=$User.FrontEndScale), Vector2D=<1 1>)
    Particles.VisibilityTag=&SpawnGroup()
   ) {isSetParameters=true}
   /SetVariables(
    Particles.Flash Spike Color:LinearColor=&RandomRangeLinearColor(Minimum=<rgba 0 1 0.5325 1>, Maximum=<rgba 0.1676 0.7775 1 1>)
    Particles.Flash Spike Lifetime:NiagaraFloat=&RandomRangeFloat(Minimum=0.05, Maximum=0.07)
    Particles.Flash Spike Sprite Size:Vector2f=&Multiply_Vector2D_ByFloat(
     Float=&FN_FrontEndFloatModifier(Gameplay=1, Front End=$User.FrontEndScale)
     Vector2D=&RandomRangeVector2D(Minimum=<3 55>, Maximum=<10 70>)
    )
   ) {isSetParameters=true}
   /SetVariables(
    Particles.Flash Star Cam Offset=15
    Particles.Flash Star Color:LinearColor=&MultiplyLinearColor(
     Linear Color=<rgba 0.1214 0.8269 1 5>
     Scale Factor=&MakeVector4_FromLinearColor(LinearColor=$System.TODColorMultiplier)
    )
    Particles.Flash Star Lifetime=0.15
    Particles.Flash Star Rotation:NiagaraFloat=&RandomRangeFloat(Minimum=0, Maximum=360)
    Particles.Flash Star Sprite Size:Vector2f=&Multiply_Vector2D_ByFloat(
     Float=&Multiply_Float(
      A=&MakeCustomFloatFromBool(Bool=$User.bIsFrontEnd, True Float=$User.FrontEndScale, False Float=1)
      B=&RandomRangeFloat(Minimum=120, Maximum=150)
     )
     Vector2D=$Particles.SpriteSize
    )
    Particles.Flash Star Sprite Size Alt:Vector2f=&Multiply_Vector2D_ByFloat(
     Float=&Multiply_Float(
      A=0.85
      B=&Multiply_Float(
       A=&MakeCustomFloatFromBool(Bool=$User.bIsFrontEnd, True Float=0.45, False Float=1)
       B=&RandomRangeFloat(Minimum=120, Maximum=150)
      )
     )
     Vector2D=$Particles.SpriteSize
    )
    Particles.Flash Star SubUV Index:NiagaraFloat=&RandomRangeFloat(Minimum=0, Maximum=3)
    Particles.Flash Star SubUV Index Alt:NiagaraFloat=&RandomRangeFloat(Minimum=0, Maximum=3)
   ) {isSetParameters=true}
   /SetVariables(
    Particles.Flash Flare Color:LinearColor=<rgba 0.0885 0.4672 1 0.0706>
    Particles.Flash Flare Lifetime:NiagaraFloat=&FN_FrontEndFloatModifier(Gameplay=0.1, Front End=0.1)
    Particles.Flash Flare Sprite Size:Vector2f=&Vector2DFromFloat(Value=&Multiply_Float(A=350, B=&FN_FrontEndFloatModifier(Gameplay=1, Front End=$User.FrontEndScale)))
   ) {isSetParameters=true}
   &DynamicMaterialParameters(
    Index 0 Param 1=0.2
    Index 0 Param 2=1
    Module.Index0Param0Write=true
    Module.Index0Param1Write=true
    Module.Index0Param2Write=true
    Module.Index0Param3Write=true
   )
  ]
  particleUpdateScript [
   &ParticleState()
   /SetVariables(
    Particles.Update Flash Flare Age:NiagaraFloat=&Divide_Float(A=$Particles.Age, B=$Particles.Initial.Flash Flare Lifetime)
    Particles.Update Flash Spike Age:NiagaraFloat=&Divide_Float(A=$Particles.Age, B=$Particles.Initial.Flash Spike Lifetime)
    Particles.Update Flash Star Age:NiagaraFloat=&Divide_Float(A=$Particles.Age, B=$Particles.Initial.Flash Star Lifetime)
   ) {isSetParameters=true}
   /SetVariables(
    Particles.Update Flash Spike Color:LinearColor=&MakeLinearColorFromVectorAndFloat(
     "Vector (RGB)"=&MakeVectorFromLinearColorRGB(Color=$Particles.Initial.Flash Spike Color)
     "Float (Alpha)"=&FloatFromCurve(FloatCurve=DataInterface() {propertyValues {Curve {keys [<at 0.7 1>,<at 1 0>]}}}, CurveIndex=$Particles.Update Flash Spike Age)
    )
    Particles.Update Flash Spike Sprite Size:Vector2f=&Multiply_Vector2D_ByFloat(
     Float=&FloatFromCurve(FloatCurve=DataInterface() {propertyValues {Curve {keys [<at 0 0.6>,<at 1 2>]}}}, CurveIndex=$Particles.Update Flash Spike Age)
     Vector2D=$Particles.Initial.Flash Spike Sprite Size
    )
    Particles.Update Flash Star Color:LinearColor=&ColorFromCurve(
     ColorCurve=DataInterface() {propertyValues {
       AlphaCurve {keys [<at 0 1>,<at 0.8992 0>]}
       BlueCurve {keys [<at 0.1978 3>,<at 0.6127 1.5>]}
       GreenCurve {keys [<at 0.1978 2.8279>,<at 0.6127 0.7859>]}
       RedCurve {keys [<at 0.1978 0.8026>,<at 0.6127 0.2353>]}
      }}
     CurveIndex=$Particles.Update Flash Star Age
    )
    Particles.Update Flash Star Sprize Size:Vector2f=&Multiply_Vector2D_ByFloat(
     Float=&FloatFromCurve(
      FloatCurve=DataInterface() {propertyValues {Curve {keys [
          {arriveTangent=2.7013, interpMode=RCIM_Cubic, leaveTangent=2.7013, tangentMode=RCTM_User, time=0, value=0}
          {interpMode=RCIM_Cubic, tangentMode=RCTM_User, time=1, value=1}
         ]}}}
      CurveIndex=$Particles.Update Flash Star Age
     )
     Vector2D=$Particles.Initial.Flash Star Sprite Size
    )
    Particles.Update Flash Star Sprize Size Alt:Vector2f=&Multiply_Vector2D_ByFloat(
     Float=&FloatFromCurve(
      FloatCurve=DataInterface() {propertyValues {Curve {keys [
          {arriveTangent=2.7013, interpMode=RCIM_Cubic, leaveTangent=2.7013, tangentMode=RCTM_User, time=0, value=0}
          {interpMode=RCIM_Cubic, tangentMode=RCTM_User, time=1, value=1}
         ]}}}
      CurveIndex=$Particles.Update Flash Star Age
      Scale Curve=0.8
     )
     Vector2D=$Particles.Initial.Flash Star Sprite Size Alt
    )
   ) {isSetParameters=true}
   /SetVariables(Particles.Update Flash Flare Color:LinearColor=&MakeLinearColorFromVectorAndFloat(
     "Vector (RGB)"=&Multiply_Vector(
      A=&MakeVectorFromLinearColorRGB(Color=$Particles.Initial.Flash Flare Color)
      B=&MakeVectorFromLinearColorRGB(Color=&ColorFromCurve(ColorCurve=DataInterface() {propertyValues {
          AlphaCurve {keys [<at 0 1 RCIM_Cubic>]}
          BlueCurve {keys [<at 0 0.6808 RCIM_Cubic>,<at 0.8966 1 RCIM_Cubic>]}
          GreenCurve {keys [<at 0 1 RCIM_Cubic>,<at 0.8966 0.3662 RCIM_Cubic>]}
          RedCurve {keys [<at 0 0 RCIM_Cubic>,<at 0.8966 0 RCIM_Cubic>]}
         }}))
     )
     "Float (Alpha)"=&Lerp_Float(A=1, B=0, Alpha=$Particles.Update Flash Flare Age)
    )) {isSetParameters=true}
   &FN_Light_Attributes(
    Color=&ColorFromCurve(ColorCurve=DataInterface() {propertyValues {
       AlphaCurve {keys [<at 0 1 RCIM_Cubic>]}
       BlueCurve {keys [<at 0 0.741 RCIM_Cubic>,<at 0.4705 1 RCIM_Cubic>]}
       GreenCurve {keys [<at 0 1 RCIM_Cubic>,<at 0.4705 0.693 RCIM_Cubic>]}
       RedCurve {keys [<at 0 0.3451 RCIM_Cubic>,<at 0.4705 0.5156 RCIM_Cubic>]}
      }})
    Create Color Attribute=true
    Create Exponent Attribute=true
    Create Position Attribute=true
    Create Radius Attribute=true
    Exponent=5
    Position=<15 0 0>
    Radius=120
    Scale Color=&FloatFromCurve(
     FloatCurve=DataInterface() {propertyValues {Curve {keys [
         {interpMode=RCIM_Cubic, tangentMode=RCTM_User, time=0, value=1}
         {arriveTangent=-0.5426, interpMode=RCIM_Cubic, leaveTangent=-0.5426, tangentMode=RCTM_User, time=1, value=0}
        ]}}}
     Scale Curve=10
    )
   )
   &ScaleColor(
    Scale RGB=&MakeVectorFromLinearColorRGB(Color=&FN_Select_Color_from_Time_Of_Day(Morning=<rgba 1 0.8463 0.7189 1>, Day=<rgba 1 1 1 1>, Evening=<rgba 1 0.8371 0.7936 1>, Night=<rgba 0.1429 0.183 0.4497 1>))
    ScaleA=false
    ScaleRGB=true
   ) {module=ScaleColor001}
  ]
  rendererClasses [&NiagaraSpriteRendererProperties, &NiagaraLightRendererProperties]
  renderers [
   {rendererClass=&NiagaraSpriteRendererProperties}
   {rendererClass=&NiagaraSpriteRendererProperties, rendererIndex=1}
   {rendererClass=&NiagaraSpriteRendererProperties, rendererIndex=2}
   {rendererClass=&NiagaraLightRendererProperties, rendererIndex=3}
  ]
 }
 Halo {
  bEnabled=false
  emitterUpdateScript [
   &EmitterState()
   &SpawnBurst_Instantaneous(Spawn Count=1, Use Spawn Probability=false)
  ]
  particleSpawnScript [&InitializeParticle(
    Color=<rgba 1 0.9468 0.999 1>
    Lifetime=0.35
    Sprite Size Mode=::NewEnumerator3
    Uniform Sprite Size=200
    UsePositionOffset=false
    Write Material Random=true
   )]
  particleUpdateScript [
   &ParticleState()
   &ScaleColor(
    Scale Alpha=&Multiply_Float(
     A=&FloatFromCurve(FloatCurve=DataInterface() {propertyValues {Curve {keys [{interpMode=RCIM_Cubic, leaveTangent=-0.6434, tangentMode=RCTM_Break, time=0, value=1}, <at 1 0>]}}})
     B=&FNTimeOfDayScalar(Night Scale=1, Morning Scale=0.3, Evening Scale=0.5)
    )
    ScaleA=true
    ScaleRGB=true
   )
   &CameraOffset(Camera Offset Amount=50)
   &ScaleSpriteSize(
    Uniform Curve Scale=1.5
    Uniform Curve Sprite Scale=DataInterface() {propertyValues {Curve {keys [{tangentMode=RCTM_User, time=0, value=0}, <at 1 1>]}}}
   )
  ]
  rendererClasses [&NiagaraSpriteRendererProperties]
  renderers [{rendererClass=&NiagaraSpriteRendererProperties}]
 }
 HandGlow {
  emitterUpdateScript [
   &EmitterState(Inactive Response=::NewEnumerator1, Life Cycle Mode=::NewEnumerator1, Loop Behavior=::NewEnumerator1, Loop Duration=0.12, UseLoopDelay=false)
   &SpawnBurst_Instantaneous(Use Spawn Probability=false, UseLoopCountLimit=false)
  ]
  particleSpawnScript [
   &InitializeParticle(
    Color=<rgba 0.15 0.15 0.15 1.5>
    Lifetime=&MakeCustomFloatFromBool(Bool=&SetBoolByFloatComparison(A=$User.BoolYay, B=0.5), True Float=1, False Float=0.45)
    Mesh Scale=&Multiply_VectorByFloat(Vector=<1.7 1 1>, Float=0.02)
    Mesh Scale Mode=::NewEnumerator4
    Sprite Rotation Mode=::NewEnumerator1
    Sprite Size Mode=::NewEnumerator3
    Uniform Sprite Size=20
    UsePositionOffset=false
    Write Material Random=true
   )
   &InitialMeshOrientation(Mesh Axis To Orient=<0 1 0>, Mesh Orientation Mode=::NewEnumerator2, Rotation=<0.6323 0.7 0>, Use Rotation Vector=true) {enabled=false}
   &CameraOffset(Camera Offset Amount=5)
   &AddVelocity(Velocity=<0 -0.0001 0>)
   &SpriteFacingAndAlignment(Sprite Facing=<0 1 0>, WriteAlignment=false, WriteFacing=true)
  ]
  particleUpdateScript [
   &ParticleState()
   &ScaleMeshSize(Scale Factor=&Multiply_Vector(
     A=&ScaleFloatByCurve(
      InitialValue:Vector3f=<1 1 1>
      ScaleCurve:NiagaraDataInterfaceVectorCurve=DataInterface() {propertyValues {
        XCurve {keys [<at 0 0.25>,<at 0.1 1>,<at 0.9 1>,<at 1 0>]}
        YCurve {keys [<at 0 1.25>,<at 0.1 2.25>,<at 0.9 2.25>,<at 1 0>]}
        ZCurve {keys [<at 0 1.25>,<at 0.1 2.25>,<at 0.9 2.25>,<at 1 0>]}
       }}
      CurveIndex:NiagaraFloat=$Particles.NormalizedAge
     )
     B=&Divide_Vector(A:Vector3f=<1 1 1>, B:Vector3f=$Engine.Owner.Scale)
    ))
   &ScaleSpriteSize(Uniform Curve Sprite Scale=DataInterface() {propertyValues {Curve {keys [
        <at 0 0>
        <at 0.1 1>
        <at 0.1512 0.7174>
        <at 0.2132 1>
        <at 0.2641 0.7539>
        <at 0.299 1>
        <at 0.3563 0.7037>
        <at 0.4151 1>
        <at 0.4564 0.8017>
        <at 0.5041 1>
        <at 0.5502 0.7561>
        <at 0.5836 1>
        <at 0.6551 0.7653>
        <at 0.6981 1>
        <at 0.7521 0.7721>
        <at 0.7871 1>
        <at 0.8475 0.7106>
        <at 0.9 1>
        {tangentMode=RCTM_User, time=1, value=0}
       ]}}})
   /SetVariables(Particles.Size002:Vector2f=&Multiply_Vector2D(
     A=$Particles.SpriteSize
     B=&Multiply_Vector2D_ByFloat(
      Float=&FloatFromCurve(FloatCurve=DataInterface() {propertyValues {Curve {keys [
           {interpMode=RCIM_Cubic, tangentMode=RCTM_User, time=0, value=1}
           {arriveTangent=-1.234, interpMode=RCIM_Cubic, leaveTangent=-1.234, tangentMode=RCTM_User, time=0.1, value=1.6328}
           {arriveTangentWeight=0.0997, interpMode=RCIM_Cubic, leaveTangentWeight=0.3488, tangentMode=RCTM_User, time=0.1688, value=1}
           {interpMode=RCIM_Cubic, tangentMode=RCTM_User, time=1, value=1}
          ]}}})
      Vector2D=<0.5 1>
     )
    )) {isSetParameters=true}
   /SetVariables(Particles.Size003:Vector2f=&Multiply_Vector2D(A=&Multiply_Vector2D_ByFloat(Float=20, Vector2D=<1 2>), B=<1 1>)) {isSetParameters=true}
   &ScaleColor(
    Scale Alpha=&Multiply_Float(
     A=&FloatFromCurve(
      FloatCurve=DataInterface() {propertyValues {Curve {keys [
          {arriveTangent=4.4475, time=0, value=0}
          {arriveTangent=5.3518, leaveTangent=-3.1774, time=0.1, value=1.2}
          {arriveTangent=-1.4972, leaveTangent=-1.4972, time=0.1523, value=0.7}
          {arriveTangent=-0.1667, leaveTangent=-0.1667, time=0.2052, value=1.1711}
          <at 0.258 0.7713>
          <at 0.2926 1.1775>
          {arriveTangent=-0.25, leaveTangent=-0.25, time=0.3608, value=0.689}
          <at 0.3865 1.193>
          <at 0.4263 0.8642>
          <at 0.4931 1.1191>
          {arriveTangent=-0.25, leaveTangent=-0.25, time=0.5428, value=0.7548}
          <at 0.5846 1.1063>
          <at 0.6571 0.8076>
          <at 0.6976 1.1685>
          {arriveTangent=0.1667, leaveTangent=0.1667, time=0.7542, value=0.7713}
          <at 0.7906 1.1138>
          <at 0.8535 0.8807>
          {arriveTangent=-3.3333, leaveTangent=-3.3333, time=0.9526, value=1.2}
          <at 1 0>
         ]}}}
      Scale Curve=&FNTimeOfDayScalar(Night Scale=0.6, Morning Scale=0.7, Day Scale=0.8, Evening Scale=0.6)
     )
     B=&FN_FrontEndFloatModifier(Gameplay=1, Front End=0.3)
    )
    Scale RGB=&MakeVectorFromLinearColorRGB(Color=&ColorFromCurve(ColorCurve=DataInterface() {propertyValues {
        AlphaCurve {keys []}
        BlueCurve {keys [<at 0.7887 0.437 RCIM_Cubic>,<at 0.9801 0.1562 RCIM_Cubic>]}
        GreenCurve {keys [<at 0.7887 0.4375 RCIM_Cubic>,<at 0.9801 0.1445 RCIM_Cubic>]}
        RedCurve {keys [<at 0.7887 0.3349 RCIM_Cubic>,<at 0.9801 0.1274 RCIM_Cubic>]}
       }}))
    ScaleA=true
    ScaleRGB=true
   )
   /SetVariables(Particles.color002:LinearColor=&MakeLinearColorFromVectorAndFloat(
     "Vector (RGB)"=&MakeVectorFromLinearColorRGB(Color=<rgba 1.7807 2.8114 3 1>)
     "Float (Alpha)"=&FloatFromCurve(
      FloatCurve=DataInterface() {propertyValues {Curve {keys [<at 0 1>,<at 0.1 2>,<at 0.2005 1.7767>,<at 0.306 1.9927>,<at 0.3889 1.7548>,<at 0.5026 2.0076>,<at 0.5773 1.7477>,<at 0.6601 1.9791>,<at 0.7283 1.7134>,<at 0.9495 1.9848>,<at 1 0>]}}}
      Scale Curve=&FN_PickFloatDependOnPlatform(TargetPlatformActive=5, TargetPlatformNoActive=1)
     )
    )) {isSetParameters=true}
   /SetVariables(Particles.Color003:LinearColor=&MakeLinearColorFromVectorAndFloat(
     "Vector (RGB)"=&MakeVectorFromLinearColorRGB(Color=<rgba 0.7068 0.9585 1 1>)
     "Float (Alpha)"=&Multiply_Float(
      A=&FloatFromCurve(
       FloatCurve=DataInterface() {propertyValues {Curve {keys [
           {arriveTangent=0.3376, interpMode=RCIM_Cubic, leaveTangent=0.3376, tangentMode=RCTM_User, time=0.1324, value=0}
           <at 0.3 1>
           <at 0.4 0.7>
           <at 0.5 0.95>
           <at 0.6 0.55>
           <at 0.7 1>
           <at 0.8 0.55>
           <at 0.9555 0.8>
           <at 1 0>
          ]}}}
       Scale Curve=&FNTimeOfDayScalar(Night Scale=0.4, Morning Scale=0.3, Day Scale=0.8, Evening Scale=0.5)
      )
      B=&Multiply_Float(
       A=&FN_FrontEndFloatModifier(Gameplay=0.5, Front End=0.15)
       B=&FN_PickFloatDependOnPlatform(TargetPlatformActive=1, TargetPlatformNoActive=0.4)
      )
     )
    )) {isSetParameters=true}
   /SetVariables(Particles.MeshColor:LinearColor=&MakeLinearColorFromVectorAndFloat(
     "Vector (RGB)"=&MakeVectorFromLinearColorRGB(Color=&ColorFromCurve(ColorCurve=DataInterface() {propertyValues {
         AlphaCurve {keys []}
         BlueCurve {keys [
           {arriveTangent=-1.8762, interpMode=RCIM_Cubic, leaveTangent=-1.8762, tangentMode=RCTM_User, time=0.1298, value=1.9554}
           {interpMode=RCIM_Cubic, leaveTangent=0.2132, time=0.3441, value=1.9833}
          ]}
         GreenCurve {keys [
           {arriveTangent=-0.4331, interpMode=RCIM_Cubic, leaveTangent=-0.4331, tangentMode=RCTM_User, time=0.1298, value=2}
           {interpMode=RCIM_Cubic, leaveTangent=2.0123, time=0.3441, value=2}
          ]}
         RedCurve {keys [
           {arriveTangent=-4.577, interpMode=RCIM_Cubic, leaveTangent=-4.577, tangentMode=RCTM_User, time=0.1298, value=1.8924}
           {interpMode=RCIM_Cubic, leaveTangent=-2.7231, time=0.3441, value=1.6652}
          ]}
        }}))
     "Float (Alpha)"=&Multiply_Float(
      A=&FloatFromCurve(
       FloatCurve=DataInterface() {propertyValues {Curve {keys [
           {arriveTangent=-0.9, time=0, value=0}
           {arriveTangent=-0.0998, leaveTangent=-0.0998, time=0.05, value=1}
           {arriveTangent=-1.0872, leaveTangent=-1.0872, time=0.1804, value=0.7734}
           <at 0.2985 0.937>
           <at 0.4531 0.6834>
           <at 0.5813 0.9005>
           <at 0.7511 0.6392>
           <at 0.8564 0.8>
           {arriveTangent=-1.5912, leaveTangent=-1.5912, time=1, value=0}
          ]}}}
       Scale Curve=&FN_FrontEndFloatModifier(Gameplay=1, Front End=0.5)
      )
      B=&FNTimeOfDayScalar(Night Scale=0.7, Morning Scale=0.5, Evening Scale=0.6)
     )
    )) {isSetParameters=true}
   &SolveForcesAndVelocity(Clamp Velocity=false, Limit Acceleration=false)
   /SetVariables(Particles.Vector 4:Vector4f=<1 1 1 1>) {isSetParameters=true}
   &ScaleColor(
    Scale RGB=&MakeVectorFromLinearColorRGB(Color=&FN_Select_Color_from_Time_Of_Day(Morning=<rgba 1 0.8463 0.7189 1>, Day=<rgba 1 1 1 1>, Evening=<rgba 1 0.8371 0.7936 1>, Night=<rgba 0.1429 0.183 0.4497 1>))
    ScaleA=false
    ScaleRGB=true
   ) {module=ScaleColor001}
  ]
  rendererClasses [&NiagaraSpriteRendererProperties]
  renderers [
   {rendererClass=&NiagaraSpriteRendererProperties}
   {rendererClass=&NiagaraSpriteRendererProperties, rendererIndex=1}
   {rendererClass=&NiagaraSpriteRendererProperties, rendererIndex=2}
  ]
 }
 Head {
  bEnabled=false
  emitterUpdateScript [
   &EmitterState(Inactive Response=::NewEnumerator1, Life Cycle Mode=::NewEnumerator1, Loop Behavior=::NewEnumerator1, Loop Duration=1, UseLoopDelay=false)
   &SpawnBurst_Instantaneous(Use Spawn Probability=false, UseLoopCountLimit=false)
  ]
  particleSpawnScript [
   &InitializeParticle(
    Color=<rgba 0 3.8502 5 1>
    Lifetime=0.2
    Mesh Scale=<0.5 0.5 1.5>
    Mesh Scale Mode=::NewEnumerator4
    Sprite Size Mode=::NewEnumerator3
    Uniform Sprite Size=200
    UsePositionOffset=false
    Write Material Random=true
   )
   &InitialMeshOrientation(Mesh Orientation Mode=::NewEnumerator2, Orientation Vector=<0 0 1>, Use Rotation Vector=false)
  ]
  particleUpdateScript [
   &ParticleState(Kill Particles When Lifetime Has Elapsed=false)
   &ScaleColor(
    Scale Alpha=&FloatFromCurve(FloatCurve=DataInterface() {propertyValues {Curve {keys [
         {arriveTangent=9.0725, interpMode=RCIM_Cubic, leaveTangent=9.0725, tangentMode=RCTM_User, time=0, value=0}
         {arriveTangent=-2.5249, interpMode=RCIM_Cubic, leaveTangent=-2.5249, tangentMode=RCTM_User, time=0.1408, value=0.9847}
         {arriveTangent=-0.0405, interpMode=RCIM_Cubic, leaveTangent=-0.0405, tangentMode=RCTM_User, time=1, value=0}
        ]}}})
    ScaleA=true
    ScaleRGB=true
   )
   &ScaleSpriteSize(Uniform Curve Sprite Scale=DataInterface() {propertyValues {Curve {keys [
        {arriveTangent=9.0725, interpMode=RCIM_Cubic, leaveTangent=9.0725, tangentMode=RCTM_User, time=0, value=0}
        {arriveTangent=-0.5028, interpMode=RCIM_Cubic, leaveTangent=-0.5028, tangentMode=RCTM_User, time=0.1408, value=0.9847}
        {arriveTangent=-0.0405, interpMode=RCIM_Cubic, leaveTangent=-0.0405, tangentMode=RCTM_User, time=1, value=0.4}
       ]}}})
  ]
  rendererClasses [&NiagaraLightRendererProperties, &NiagaraSpriteRendererProperties]
  renderers [
   {rendererClass=&NiagaraLightRendererProperties, rendererIndex=0}
   {rendererClass=&NiagaraSpriteRendererProperties, rendererIndex=1}
  ]
 }
 IceBlast {
  emitterUpdateScript [
   &EmitterState(Life Cycle Mode=::NewEnumerator1, Loop Behavior=::NewEnumerator1, Loop Delay=0.1, Loop Duration=0.1, UseLoopDelay=false)
   &SpawnBurst_Instantaneous(Spawn Count=60, Use Spawn Probability=false, UseLoopCountLimit=false)
   &SpawnRate(SpawnRate=60, Use Spawn Probability=false)
  ]
  particleSpawnScript [
   &InitializeParticle(
    Color=&RandomRangeLinearColor(Minimum=<rgba 0.3851 0.6754 0.9479 0.8>, Maximum=<rgba 0.6933 0.8281 0.8802 0.8>)
    Lifetime Max=0.2
    Lifetime Min=0.1
    Lifetime Mode=::NewEnumerator1
    Position Offset=<0 -1.0927 0>
    Sprite Rotation Angle=&SelectFloatFromArray(
     Direct Array Index=&ReturnExecIndex()
     Direct Array Mode=::NewEnumerator1
     Float Selection Array=DataInterface() {propertyValues {FloatData [0.2,-0.2]}}
    )
    Sprite Rotation Mode=::NewEnumerator1
    Sprite Size Max=<40 80>
    Sprite Size Min=<30 35>
    Sprite Size Mode=::NewEnumerator2
    UsePositionOffset=true
    Write Material Random=true
   )
   &CameraOffset(Camera Offset Amount=5)
   &SpriteFacingAndAlignment(Sprite Alignment=<0 -1 0>, WriteAlignment=true, WriteFacing=false)
  ]
  particleUpdateScript [
   /SetVariables(Particles.SpriteRotation:NiagaraFloat=&SelectFloatFromArray(
     Direct Array Index=&ReturnExecIndex()
     Direct Array Mode=::NewEnumerator1
     Float Selection Array=DataInterface() {propertyValues {FloatData [0,4,-4,0,2,-2,0,1,-1,0,3,-3,0,5,-5,0]}}
    )) {isSetParameters=true}
   &ParticleState()
   &ScaleColor(
    Scale Alpha=&FloatFromCurve(FloatCurve=DataInterface() {propertyValues {Curve {keys [
         {arriveTangent=-0.0888, interpMode=RCIM_Cubic, leaveTangent=-0.0888, tangentMode=RCTM_User, time=0, value=0.5}
         {arriveTangent=0.12, interpMode=RCIM_Cubic, leaveTangent=0.12, tangentMode=RCTM_User, time=0.25, value=0.9847}
         {arriveTangent=-0.0405, interpMode=RCIM_Cubic, leaveTangent=-0.0405, tangentMode=RCTM_User, time=1, value=0}
        ]}}})
    Scale RGB=&VectorFromFloat(Value=1.1)
    ScaleA=true
    ScaleRGB=true
   )
   &ScaleSpriteSize(
    Scale Factor=&Multiply_Vector2D_ByFloat(
     Float=&FloatFromCurve(FloatCurve=DataInterface() {propertyValues {Curve {keys [
          {
           arriveTangent=-0.0024
           arriveTangentWeight=0.5946
           interpMode=RCIM_Cubic
           leaveTangentWeight=0.5946
           tangentMode=RCTM_User
           time=0.0677
           value=0
          }
          {arriveTangent=4.5684, interpMode=RCIM_Cubic, leaveTangent=4.5684, tangentMode=RCTM_User, time=0.7, value=1}
         ]}}})
     Vector2D=<1 2.5>
    )
    Scale Sprite Size Mode=::NewEnumerator2
   ) {module=ScaleSpriteSize003}
   &DynamicMaterialParameters(
    Index 0 Param 1=&Add_Float(
     A=&RandomRangeFloat(Minimum=0, Maximum=1)
     B=&FloatFromCurve(
      FloatCurve=DataInterface() {propertyValues {Curve {keys [
          {arriveTangent=2.7533, interpMode=RCIM_Cubic, leaveTangent=1.8958, tangentMode=RCTM_User, time=0, value=1}
          {arriveTangent=0.8292, interpMode=RCIM_Cubic, leaveTangent=0.8292, tangentMode=RCTM_User, time=0.321, value=1.3453}
          {arriveTangent=0.0529, interpMode=RCIM_Cubic, leaveTangent=0.2021, tangentMode=RCTM_User, time=1, value=1.5}
         ]}}}
      CurveIndex=&Add_Float(A=&Multiply_Float(A=1, B=0.5), B=$Particles.NormalizedAge)
     )
    )
    Index 0 Param 2=&FloatFromCurve(FloatCurve=DataInterface() {propertyValues {Curve {keys [
         {arriveTangent=0.0453, interpMode=RCIM_Cubic, tangentMode=RCTM_Break, time=0, value=-1}
         {arriveTangent=0.0604, interpMode=RCIM_Cubic, leaveTangent=0.0604, tangentMode=RCTM_User, time=1, value=-1}
        ]}}})
    Index 0 Param 3=&FloatFromCurve(
     FloatCurve=DataInterface() {propertyValues {Curve {keys [{tangentMode=RCTM_User, time=0, value=0}, {tangentMode=RCTM_User, time=1, value=0}]}}}
     Scale Curve=0
    )
    Index 0 Param 4=&FloatFromCurve(FloatCurve=DataInterface() {propertyValues {Curve {keys [<at 0.5 1>,<at 1 0>]}}})
    Index 1 Param 1=&FloatFromCurve(
     FloatCurve=DataInterface() {propertyValues {Curve {keys [
         {interpMode=RCIM_Cubic, leaveTangent=4.7088, tangentMode=RCTM_Break, time=0, value=0}
         {arriveTangent=4.2257, interpMode=RCIM_Cubic, leaveTangent=4.2257, tangentMode=RCTM_User, time=0.25, value=1}
        ]}}}
     Scale Curve=&RandomRangeFloat(Minimum=1.5, Maximum=1.5)
    )
    Index 1 Param 2=&FloatFromCurve(
     FloatCurve=DataInterface() {propertyValues {Curve {keys [<at 0 0.2>, {leaveTangent=0.9066, time=1, value=2}]}}}
     Scale Curve=&RandomRangeFloat(Minimum=0.5, Maximum=0.6)
    )
    Index 1 Param 3=&FloatFromCurve(FloatCurve=DataInterface() {propertyValues {Curve {keys [
         {interpMode=RCIM_Cubic, leaveTangent=6.3917, tangentMode=RCTM_Break, time=0, value=0.2}
         {arriveTangent=1.7549, interpMode=RCIM_Cubic, leaveTangent=10.6767, tangentMode=RCTM_User, time=0.25, value=1}
        ]}}})
    Index 1 Param 4=&FloatFromCurve(
     FloatCurve=DataInterface() {propertyValues {Curve {keys [{arriveTangent=-0.136, interpMode=RCIM_Cubic, leaveTangent=-0.136, tangentMode=RCTM_User, time=0, value=1}, <at 1 0.95>]}}}
     Scale Curve=&RandomRangeFloat(Minimum=1.1, Maximum=1.4)
    )
    Index 2 Param 1=&FloatFromCurve(FloatCurve=DataInterface() {propertyValues {Curve {keys [
         {interpMode=RCIM_Cubic, leaveTangent=4.2485, tangentMode=RCTM_Break, time=0, value=0}
         {arriveTangent=1.0548, interpMode=RCIM_Cubic, leaveTangent=1.0548, tangentMode=RCTM_User, time=0.17, value=0.5715}
         <at 1 1>
        ]}}})
    Index 2 Param 2=&FloatFromCurve(
     FloatCurve=DataInterface() {propertyValues {Curve {keys [
         {arriveTangent=1.9056, interpMode=RCIM_Cubic, leaveTangent=1.9056, tangentMode=RCTM_User, time=0, value=0.2}
         {arriveTangent=0.0259, interpMode=RCIM_Cubic, leaveTangent=0.0259, tangentMode=RCTM_User, time=1, value=1.4}
        ]}}}
     Scale Curve=&RandomRangeFloat(Minimum=0.2, Maximum=0.5)
    )
    Index 2 Param 3=&FloatFromCurve(FloatCurve=DataInterface() {propertyValues {Curve {keys [
         {interpMode=RCIM_Cubic, leaveTangent=2.3064, tangentMode=RCTM_Break, time=0, value=1}
         {interpMode=RCIM_Cubic, leaveTangent=11.1233, tangentMode=RCTM_Break, time=1, value=2}
        ]}}})
    Index 2 Param 4=&FloatFromCurve(
     FloatCurve=DataInterface() {propertyValues {Curve {keys [
         {arriveTangent=-0.6867, interpMode=RCIM_Cubic, leaveTangent=-0.6867, tangentMode=RCTM_User, time=0, value=1}
         {interpMode=RCIM_Cubic, leaveTangent=-1.2956, tangentMode=RCTM_Break, time=1, value=0.75}
        ]}}}
     Scale Curve=1.6
    )
    Module.Param0WriteEnabled=true
    Module.Param0WriteEnabled1=false
    Module.Param0WriteEnabled2=true
    Module.Param1WriteEnabled=false
    Module.Param1WriteEnabled1=false
    Module.Param1WriteEnabled2=true
    Module.Param2WriteEnabled=true
    Module.Param2WriteEnabled1=false
    Module.Param2WriteEnabled2=true
    Module.Param3WriteEnabled=true
    Module.Param3WriteEnabled1=false
    Module.Param3WriteEnabled2=true
    Write Parameter Index 1=true
    Write Parameter Index 2=true
   ) {module=DynamicMaterialParameters001}
   &SolveForcesAndVelocity(Clamp Velocity=false, Limit Acceleration=false)
   /SetVariables(Particles.Linear Color=&RandomRangeLinearColor(
     Minimum=&ColorFromCurve(ColorCurve=DataInterface() {propertyValues {
        AlphaCurve {keys [<at 0 0.6665>, {arriveTangent=-0.6665, interpMode=RCIM_Cubic, leaveTangent=-0.6665, time=0.0818, value=2}, <at 1 0 RCIM_Cubic>]}
        BlueCurve {keys [<at 0 0.949 RCIM_Cubic>]}
        GreenCurve {keys [<at 0 0.6745 RCIM_Cubic>]}
        RedCurve {keys [<at 0 0.3843 RCIM_Cubic>]}
       }})
     Maximum=&ColorFromCurve(ColorCurve=DataInterface() {propertyValues {
        AlphaCurve {keys [<at 0 0.6665>, {arriveTangent=-0.6665, interpMode=RCIM_Cubic, leaveTangent=-0.6665, time=0.1479, value=2}, <at 1 0 RCIM_Cubic>]}
        BlueCurve {keys [<at 0 0.8784 RCIM_Cubic>]}
        GreenCurve {keys [<at 0 0.8275 RCIM_Cubic>]}
        RedCurve {keys [<at 0 0.6941 RCIM_Cubic>]}
       }})
    )) {isSetParameters=true}
   /SetVariables(Particles.LowMediumSize:Vector2f=&Multiply_Vector2D_ByFloat(
     Float=&FloatFromCurve(FloatCurve=DataInterface() {propertyValues {Curve {keys [
          {
           arriveTangent=2.6287
           arriveTangentWeight=0.5946
           interpMode=RCIM_Cubic
           leaveTangent=2.6287
           leaveTangentWeight=0.5946
           tangentMode=RCTM_User
           time=0.0677
           value=0
          }
          {arriveTangent=0.3198, interpMode=RCIM_Cubic, leaveTangent=0.3198, tangentMode=RCTM_User, time=0.7, value=1}
         ]}}})
     Vector2D=&Multiply_Vector2D(A=<1 2.5>, B=&RandomRangeVector2D(Minimum=<15 26.0958>, Maximum=<10 28.9994>))
    )) {isSetParameters=true}
   &ScaleColor(
    Scale RGB=&MakeVectorFromLinearColorRGB(Color=&FN_Select_Color_from_Time_Of_Day(Morning=<rgba 1 0.8463 0.7189 1>, Day=<rgba 1 1 1 1>, Evening=<rgba 1 0.8371 0.7936 1>, Night=<rgba 0.1429 0.183 0.4497 1>))
    ScaleA=false
    ScaleRGB=true
   ) {module=ScaleColor001}
  ]
  rendererClasses [&NiagaraSpriteRendererProperties]
  renderers [
   {rendererClass=&NiagaraSpriteRendererProperties}
   {rendererClass=&NiagaraSpriteRendererProperties, rendererIndex=1}
  ]
 }
 IceChunks {
  bEnabled=false
  emitterUpdateScript [
   &EmitterState()
   &SpawnBurst_Instantaneous(Spawn Count=22, Use Spawn Probability=false, UseLoopCountLimit=false)
  ]
  particleSpawnScript [
   &InitializeParticle(
    Color=<rgba 6.96 7.5 7.3919 1>
    Lifetime Max=1
    Lifetime Min=0.25
    Lifetime Mode=::NewEnumerator1
    Mesh Renderer Array Visibility Mode=::NewEnumerator1
    Mesh Renderer Info=DataInterface() {propertyValues {MeshRenderer=&self:IceChunks_0.NiagaraMeshRendererProperties_0}}
    Mesh Renderer Mesh Index:NiagaraInt32=&RandomRangeInt(Minimum=0, Maximum=2)
    Mesh Scale Mode=::NewEnumerator3
    Mesh Uniform Scale=&RandomRangeFloat(Minimum=0.4, Maximum=0.7)
    UsePositionOffset=false
    Write Material Random=true
   )
   &ShapeLocation(MaskSpawnGroup=false, Override Seed=false, Sphere Radius=11, UseOwnerScale=true)
   &AddVelocity(
    Cone Angle=180
    Cone Axis=<0 0 1>
    Distribution Along Cone Axis=0
    Module.UseConeFalloff=true
    Module.UseVelDistribution=true
    Override Seed=false
    Speed Falloff From Cone Axis=0.3
    Velocity Mode=::NewEnumerator2
    Velocity Speed=&RandomRangeFloat(Minimum=222, Maximum=666)
   ) {enabled=false}
   &InitialMeshOrientation(
    Mesh Axis To Orient=<0 0 1>
    Mesh Orientation Mode=::NewEnumerator2
    Orientation Vector=$Particles.ShapeLocation.ShapeNormal
    Use Rotation Vector=false
   )
   &AddVelocity(Velocity=&RandomRangeVector(Minimum=<100 -300 0>, Maximum=<-100 -900 300>)) {module=AddVelocity001}
   &AddVelocity(
    Module.Use Falloff Distance=false
    Override Seed=false
    Velocity Mode=::NewEnumerator1
    Velocity Speed=&RandomRangeFloat(Minimum=100, Maximum=200)
   ) {module=AddVelocity002}
  ]
  particleUpdateScript [
   &ParticleState()
   &UpdateMeshOrientation(
    Rotation Rate=&RandomRangeFloat(Minimum=2, Maximum=4)
    Rotation Vector=&RandomVector(Vector Scale=1)
   )
   &ScaleMeshSize(Scale Factor=&VectorFromFloat(Value=&FloatFromCurve(FloatCurve=DataInterface() {propertyValues {Curve {keys [
          {interpMode=RCIM_Cubic, tangentMode=RCTM_User, time=0, value=1}
          {arriveTangent=-1.6697, interpMode=RCIM_Cubic, leaveTangent=-1.6697, tangentMode=RCTM_User, time=1, value=0}
         ]}}})))
   &Drag(Drag=&RandomRangeFloat(Minimum=2, Maximum=3), Use Linear Drag=true, Use Rotational Drag=true)
   &CurlNoiseForce(Noise Strength=111, Pan Noise=true, Pan Noise Field=<0 0 0.1>)
   &GravityForce()
   &SolveForcesAndVelocity(Clamp Velocity=false, Limit Acceleration=false)
  ]
  rendererClasses [&NiagaraMeshRendererProperties]
  renderers [{rendererClass=&NiagaraMeshRendererProperties, rendererIndex=0}]
 }
 IceSpikes {
  emitterUpdateScript [
   &EmitterState(
    Inactive Response=::NewEnumerator1
    Life Cycle Mode=::NewEnumerator1
    Loop Behavior=::NewEnumerator1
    Loop Delay=&MakeCustomFloatFromBool(Bool=false, True Float=0, False Float=0.05)
    Loop Duration=0.2
    UseLoopDelay=false
   )
   &SpawnBurst_Instantaneous(Spawn Count=30, Use Spawn Probability=false, UseLoopCountLimit=false)
   &SpawnRate(Spawn Probability=0.67, SpawnRate=160, Use Spawn Probability=false) {enabled=false}
  ]
  particleSpawnScript [
   &InitializeParticle(
    Color=&RandomRangeLinearColor(Minimum=<rgba 0.3851 0.6754 0.9479 0.8>, Maximum=<rgba 0.188 0.545 0.8802 0.8>)
    Lifetime=0.15
    Mesh Scale Mode=::NewEnumerator1
    Mesh Uniform Scale Max=1
    Mesh Uniform Scale Min=0.8
    Position Offset=<0 10 0>
    Sprite Rotation Mode=::NewEnumerator2
    Sprite Size=<300 500>
    Sprite Size Mode=::NewEnumerator4
    UsePositionOffset=true
    Write Material Random=true
   ) {module=InitializeParticle001}
   /SetVariables(Particles.Rand:NiagaraFloat=&UniformAOrBFloat(A=0, Evaluation Type=::NewEnumerator0, Recalculate Random Each Loop=true, Override Seed=false)) {isSetParameters=true}
   &InitialMeshOrientation(Mesh Orientation Mode=::NewEnumerator1, Rotation=<0 0 -0.2457>, Rotation Coordinate Space=::NewEnumerator0, Use Rotation Vector=true)
  ]
  particleUpdateScript [
   &ParticleState()
   &DynamicMaterialParameters(
    Index 0 Param 1=$Particles.NormalizedAge
    Index 0 Param 2=&RandomRangeFloat(Minimum=1, Maximum=1.5)
    Index 0 Param 3=&RandomRangeFloat(Minimum=0, Maximum=1)
    Index 0 Param 4=&FloatFromCurve(
     FloatCurve=DataInterface() {propertyValues {Curve {keys [
         {arriveTangent=2.9585, interpMode=RCIM_Cubic, leaveTangent=2.9585, tangentMode=RCTM_User, time=0, value=0}
         {arriveTangent=-0.0405, interpMode=RCIM_Cubic, leaveTangent=-0.0405, tangentMode=RCTM_User, time=1, value=1}
        ]}}}
     Scale Curve=&RandomRangeFloat(Minimum=0.8, Maximum=1)
    )
    Module.Param0WriteEnabled=true
    Module.Param1WriteEnabled=true
    Module.Param2WriteEnabled=true
    Module.Param3WriteEnabled=true
   )
   &ScaleMeshSize(Scale Factor=&VectorFromFloat(Value=&FloatFromCurve(FloatCurve=DataInterface() {propertyValues {Curve {keys [{arriveTangent=0.1649, interpMode=RCIM_Cubic, leaveTangent=0.1649, tangentMode=RCTM_User, time=0.75, value=1}]}}})))
   &ScaleMeshSize(Scale Factor=&VectorFromFloat(Value=&FloatFromCurve(FloatCurve=DataInterface() {propertyValues {Curve {keys [
          {arriveTangent=3.844, interpMode=RCIM_Cubic, leaveTangent=3.844, tangentMode=RCTM_User, time=0, value=0}
          {arriveTangent=0.1649, interpMode=RCIM_Cubic, leaveTangent=0.1649, tangentMode=RCTM_User, time=0.75, value=1}
         ]}}}))) {enabled=false, module=ScaleMeshSize001}
   &ScaleColor(
    Scale Alpha=&FloatFromCurve(
     FloatCurve=DataInterface() {propertyValues {Curve {keys [{interpMode=RCIM_Cubic, tangentMode=RCTM_User, time=0.1, value=1}, {interpMode=RCIM_Cubic, tangentMode=RCTM_User, time=1, value=0}]}}}
     Scale Curve=2
    )
    Scale RGB=&VectorFromFloat(Value=5)
    ScaleA=true
    ScaleRGB=true
   )
   &SolveForcesAndVelocity(Clamp Velocity=false, Limit Acceleration=false)
   &ScaleColor(
    Scale RGB=&MakeVectorFromLinearColorRGB(Color=&FN_Select_Color_from_Time_Of_Day(Morning=<rgba 1 0.8463 0.7189 1>, Day=<rgba 1 1 1 1>, Evening=<rgba 1 0.8371 0.7936 1>, Night=<rgba 0.1429 0.183 0.4497 1>))
    ScaleA=false
    ScaleRGB=true
   ) {module=ScaleColor001}
  ]
  rendererClasses [&NiagaraMeshRendererProperties]
  renderers [{rendererClass=&NiagaraMeshRendererProperties, rendererIndex=0}]
 }
 Mist {
  emitterUpdateScript [
   &EmitterState(Life Cycle Mode=::NewEnumerator1, Loop Behavior=::NewEnumerator1, Loop Duration=1, UseLoopDelay=false)
   &SpawnBurst_Instantaneous(Spawn Count=35, Use Spawn Probability=false, UseLoopCountLimit=false)
  ]
  particleSpawnScript [
   &InitializeParticle(
    Color=<rgba 0.4492 0.6108 0.8177 0.3>
    Lifetime=&RandomRangeFloat(Minimum=0.8, Maximum=1.5)
    Sprite Rotation Angle=&RandomRangeFloat(Minimum=-55, Maximum=98)
    Sprite Rotation Mode=::NewEnumerator0
    Sprite Size Mode=::NewEnumerator3
    Sprite UV Mode=::NewEnumerator2
    Uniform Sprite Size=&Multiply_Float(A=&RandomRangeFloat(Minimum=90, Maximum=120), B=1)
    UsePositionOffset=false
    Write Material Random=true
   )
   &ShapeLocation(
    MaskSpawnGroup=false
    Offset=<0 -1.3139 0>
    Offset Mode=::NewEnumerator0
    Override Seed=false
    Rotation Angle=0.25
    Rotation Axis=<0 1 0>
    Rotation Mode=::NewEnumerator1
    Sphere Radius=10
    Sphere Surface Distribution=1
    UseOwnerScale=true
   )
   &InheritVelocity(Inherited Velocity Amount Scale=&VectorFromFloat(Value=0.1), Speed Limit=true, UseSpeedThreshold=true) {enabled=false}
   &CameraOffset(Camera Offset Amount=16.875) {module=CameraOffset001}
   &DynamicMaterialParameters(
    Index 0 Param 4=0.3
    Module.Index0Float0Write=true
    Module.Index0Float1Write=true
    Module.Index0Float2Write=true
    Module.Index0Float3Write=true
   )
   &AddVelocity(Velocity=&MergeHelper_VectorSwitch(
     0=<0 0 0>
     1=<0 0 0>
     2=&Multiply_VectorByFloat(Vector=$System.DirectionVector, Float=&RandomRangeFloat(Minimum=110, Maximum=120))
     3=?
     Index=&ReturnExecIndex()
     Num=::NewEnumerator1
    )) {enabled=false}
   &PointForce(Module.Use Falloff=false, Module.UseLinearForce=false)
   &ApplyInitialForces()
   &AddVelocity(Velocity=&RandomRangeVector(Minimum=<100 -150 0>, Maximum=<-100 -500 0>)) {module=AddVelocity001}
  ]
  particleUpdateScript [
   &ParticleState()
   &GravityForce(Gravity=<0 -150 -52.0864>)
   &Drag(Drag=2, Use Linear Drag=true, Use Rotational Drag=true)
   &SolveForcesAndVelocity(Clamp Velocity=false, Limit Acceleration=false)
   &ScaleSpriteSize(Scale Factor=&Vector2DFromFloat(Value=&FloatFromCurve(FloatCurve=DataInterface() {propertyValues {Curve {keys [{arriveTangent=1.8373, interpMode=RCIM_Cubic, leaveTangent=1.8373, tangentMode=RCTM_User, time=0, value=0.5}, <at 1 1.3 RCIM_Cubic>]}}})))
   &ScaleColor(
    Scale Alpha=&FloatFromCurve(FloatCurve=DataInterface() {propertyValues {Curve {keys [
         <at 0 0 RCIM_Cubic>
         {arriveTangent=-0.4056, interpMode=RCIM_Cubic, leaveTangent=-0.4056, tangentMode=RCTM_User, time=0.0203, value=0.55}
         {arriveTangent=-0.5614, interpMode=RCIM_Cubic, leaveTangent=-0.5614, time=0.0878, value=0.5}
         <at 1 0 RCIM_Cubic>
        ]}}})
    ScaleA=true
    ScaleRGB=true
   )
   &SubUVAnimation(
    End Frame=26
    SubUV Animation Mode:ENiagara_SubUVLookupMode=::NewEnumerator1
    SubUV Curve=DataInterface() {propertyValues {Curve {keys [
        {arriveTangent=2.3868, interpMode=RCIM_Cubic, leaveTangent=2.3868, tangentMode=RCTM_User, time=0, value=0}
        {interpMode=RCIM_Cubic, tangentMode=RCTM_User, time=1, value=1}
       ]}}}
   ) {enabled=false}
   /SetVariables(Particles.SubImageIndex=&Add_Float(
     A=&RandomRangeFloat(Minimum=0, Maximum=0.3)
     B=&FloatFromCurve(
      FloatCurve=DataInterface() {propertyValues {Curve {keys [{arriveTangent=120.5723, interpMode=RCIM_Cubic, leaveTangent=120.5723, tangentMode=RCTM_User, time=0, value=0}, <at 1 64 RCIM_Cubic>]}}}
      Scale Curve=&RandomRangeFloat(Minimum=0.3, Maximum=0.5)
     )
    )) {enabled=false, isSetParameters=true, module=SetVariables_#2}
   &CameraOffset(Camera Offset Amount=&FloatFromCurve(FloatCurve=DataInterface() {propertyValues {Curve {keys [<at 0 1>,<at 1 100>]}}}))
   &ScaleColor(
    Color Value To Scale=$Particles.Color
    Scale Mode=::NewEnumerator1
    Scale RGBA=&MakeVector4_FromLinearColor(LinearColor=&FNTimeOfDayColor(
      Night Color=<rgba 0.539 0.6941 1 1>
      Morning Color=<rgba 0.9392 0.8785 1 1>
      Day Color=<rgba 1 1 1 1>
      Evening Color=<rgba 1 0.6882 0.6583 1>
     ))
    ScaleRGBA=true
   ) {enabled=false, module=ScaleColor001}
   &SubUVAnimation(
    "Sprite Renderer "=DataInterface() {propertyValues {SpriteRenderer=&self:Mist_6.NiagaraSpriteRendererProperties_1}}
    UseEndFrame=false
    UseStartFrame=false
   ) {module=SubUVAnimation001}
   &ScaleColor(
    Scale Alpha=&MakeFloatFromVector(
     Vector=&MakeVectorFromLinearColorRGB(Color=&FN_Select_Color_from_Time_Of_Day(Morning=<rgba 1 0.8463 0.7189 1>, Day=<rgba 1 1 1 1>, Evening=<rgba 1 0.8371 0.7936 1>, Night=<rgba 0.1429 0.183 0.4497 0.8>))
     Channel=::NewEnumerator0
    )
    Scale RGB=&MakeVectorFromLinearColorRGB(Color=&FN_Select_Color_from_Time_Of_Day(Morning=<rgba 1 0.8463 0.7189 1>, Day=<rgba 1 1 1 1>, Evening=<rgba 1 0.8371 0.7936 1>, Night=<rgba 0.1429 0.183 0.4497 1>))
    ScaleA=true
    ScaleRGB=false
   ) {module=ScaleColor002}
   &ScaleColor(Scale Alpha=&FN_FrontEndFloatModifier(Gameplay=1, Front End=0.1), ScaleA=true, ScaleRGB=false) {module=ScaleColor004}
  ]
  rendererClasses [&NiagaraSpriteRendererProperties]
  renderers [
   {rendererClass=&NiagaraSpriteRendererProperties}
   {rendererClass=&NiagaraSpriteRendererProperties, rendererIndex=1}
  ]
 }
 Mist001 {
  bEnabled=false
  emitterUpdateScript [
   &EmitterState(Life Cycle Mode=::NewEnumerator1, Loop Behavior=::NewEnumerator1, Loop Duration=1, UseLoopDelay=false)
   &SpawnBurst_Instantaneous(Spawn Count=35, Use Spawn Probability=false, UseLoopCountLimit=false)
  ]
  particleSpawnScript [
   &InitializeParticle(
    Color=<rgba 0.4492 0.6108 0.8177 0.3>
    Lifetime=&RandomRangeFloat(Minimum=0.8, Maximum=1.5)
    Sprite Rotation Angle=&RandomRangeFloat(Minimum=-55, Maximum=98)
    Sprite Rotation Mode=::NewEnumerator0
    Sprite Size Mode=::NewEnumerator3
    Sprite UV Mode=::NewEnumerator2
    Uniform Sprite Size=&Multiply_Float(A=&RandomRangeFloat(Minimum=90, Maximum=120), B=1)
    UsePositionOffset=false
    Write Material Random=true
   )
   &ShapeLocation(
    MaskSpawnGroup=false
    Offset=<0 -1.3139 0>
    Offset Mode=::NewEnumerator0
    Override Seed=false
    Rotation Angle=0.25
    Rotation Axis=<0 1 0>
    Rotation Mode=::NewEnumerator1
    Sphere Radius=10
    Sphere Surface Distribution=1
    UseOwnerScale=true
   )
   &InheritVelocity(Inherited Velocity Amount Scale=&VectorFromFloat(Value=0.1), Speed Limit=true, UseSpeedThreshold=true) {enabled=false}
   &CameraOffset(Camera Offset Amount=16.875) {module=CameraOffset001}
   &DynamicMaterialParameters(
    Index 0 Param 4=0.3
    Module.Index0Float0Write=true
    Module.Index0Float1Write=true
    Module.Index0Float2Write=true
    Module.Index0Float3Write=true
   )
   &AddVelocity(Velocity=&MergeHelper_VectorSwitch(
     0=<0 0 0>
     1=<0 0 0>
     2=&Multiply_VectorByFloat(Vector=$System.DirectionVector, Float=&RandomRangeFloat(Minimum=110, Maximum=120))
     3=?
     Index=&ReturnExecIndex()
     Num=::NewEnumerator1
    )) {enabled=false}
   &PointForce(Module.Use Falloff=false, Module.UseLinearForce=false)
   &ApplyInitialForces()
   &AddVelocity(Velocity=&RandomRangeVector(Minimum=<100 0 0>, Maximum=<-100 -300 0>)) {module=AddVelocity001}
  ]
  particleUpdateScript [
   &ParticleState()
   &GravityForce(Gravity=<0 -150 -52.0864>)
   &Drag(Drag=2, Use Linear Drag=true, Use Rotational Drag=true)
   &SolveForcesAndVelocity(Clamp Velocity=false, Limit Acceleration=false)
   &ScaleSpriteSize(Scale Factor=&Vector2DFromFloat(Value=&FloatFromCurve(FloatCurve=DataInterface() {propertyValues {Curve {keys [{arriveTangent=1.8373, interpMode=RCIM_Cubic, leaveTangent=1.8373, tangentMode=RCTM_User, time=0, value=0.5}, <at 1 1.3 RCIM_Cubic>]}}})))
   &ScaleColor(
    Scale Alpha=&FloatFromCurve(FloatCurve=DataInterface() {propertyValues {Curve {keys [
         <at 0 0 RCIM_Cubic>
         {arriveTangent=-0.4056, interpMode=RCIM_Cubic, leaveTangent=-0.4056, tangentMode=RCTM_User, time=0.0203, value=0.8}
         {arriveTangent=-0.8166, interpMode=RCIM_Cubic, leaveTangent=-0.8166, time=0.0878, value=0.5}
         <at 1 0 RCIM_Cubic>
        ]}}})
    ScaleA=true
    ScaleRGB=true
   )
   &SubUVAnimation(
    End Frame=26
    SubUV Animation Mode:ENiagara_SubUVLookupMode=::NewEnumerator1
    SubUV Curve=DataInterface() {propertyValues {Curve {keys [
        {arriveTangent=2.3868, interpMode=RCIM_Cubic, leaveTangent=2.3868, tangentMode=RCTM_User, time=0, value=0}
        {interpMode=RCIM_Cubic, tangentMode=RCTM_User, time=1, value=1}
       ]}}}
   ) {enabled=false}
   /SetVariables(Particles.SubImageIndex=&Add_Float(
     A=&RandomRangeFloat(Minimum=0, Maximum=0.3)
     B=&FloatFromCurve(
      FloatCurve=DataInterface() {propertyValues {Curve {keys [{arriveTangent=120.5723, interpMode=RCIM_Cubic, leaveTangent=120.5723, tangentMode=RCTM_User, time=0, value=0}, <at 1 64 RCIM_Cubic>]}}}
      Scale Curve=&RandomRangeFloat(Minimum=0.3, Maximum=0.5)
     )
    )) {enabled=false, isSetParameters=true, module=SetVariables_#2}
   &CameraOffset(Camera Offset Amount=&FloatFromCurve(FloatCurve=DataInterface() {propertyValues {Curve {keys [<at 0 1>,<at 1 100>]}}}))
   &ScaleColor(
    Color Value To Scale=$Particles.Color
    Scale Mode=::NewEnumerator1
    Scale RGBA=&MakeVector4_FromLinearColor(LinearColor=&FNTimeOfDayColor(
      Night Color=<rgba 0.539 0.6941 1 1>
      Morning Color=<rgba 0.9392 0.8785 1 1>
      Day Color=<rgba 1 1 1 1>
      Evening Color=<rgba 1 0.6882 0.6583 1>
     ))
    ScaleRGBA=true
   ) {module=ScaleColor001}
   &SubUVAnimation(
    "Sprite Renderer "=DataInterface() {propertyValues {SpriteRenderer=&self:Mist001_0.NiagaraSpriteRendererProperties_1}}
    UseEndFrame=false
    UseStartFrame=false
   ) {module=SubUVAnimation001}
  ]
  rendererClasses [&NiagaraSpriteRendererProperties]
  renderers [
   {rendererClass=&NiagaraSpriteRendererProperties}
   {rendererClass=&NiagaraSpriteRendererProperties, rendererIndex=1}
  ]
 }
 NE_Debris_Radial {
  emitterUpdateScript [
   &EmitterState(Life Cycle Mode=::NewEnumerator1, Loop Behavior=::NewEnumerator1, Loop Duration=2, UseLoopDelay=false)
   &SpawnBurst_Instantaneous(
    Spawn Count=20
    Spawn Group:NiagaraInt32=&RandomRangeInt(Minimum=4, Maximum=4)
    Use Spawn Probability=false
    UseLoopCountLimit=false
   ) {module=SpawnBurst_Instantaneous003}
   &SpawnBurst_Instantaneous(Spawn Count=12, Use Spawn Probability=false, UseLoopCountLimit=false)
   &SpawnBurst_Instantaneous(Spawn Count=12, Spawn Group=1, Use Spawn Probability=false, UseLoopCountLimit=false) {module=SpawnBurst_Instantaneous001}
   &SpawnBurst_Instantaneous(Spawn Count=2, Spawn Group=3, Use Spawn Probability=false, UseLoopCountLimit=false) {module=SpawnBurst_Instantaneous002}
  ]
  particleSpawnScript [
   &InitializeParticle(
    Color Maximum=<rgba 0.4921 0.5944 0.7083 1>
    Color Minimum=<rgba 0.5925 0.8193 0.9844 1>
    Color Mode=::NewEnumerator2
    Lifetime Max=0.7
    Lifetime Min=0.3
    Lifetime Mode=::NewEnumerator1
    Mass Mode=::NewEnumerator1
    Mesh Renderer Array Visibility Mode=::NewEnumerator1
    Mesh Renderer Info=DataInterface() {propertyValues {MeshRenderer=&self:NE_Debris_Radial_0.NiagaraMeshRendererProperties_0}}
    Mesh Renderer Mesh Index=4
    Mesh Scale Max=<1.5 1.5 1.5>
    Mesh Scale Min=<0.25 0.25 0.25>
    Mesh Scale Mode=::NewEnumerator2
    Position Offset=<0 0 5>
    Position Offset Coordinate Space=::Local
    UsePositionOffset=true
    Write Material Random=true
   )
   &ShapeLocation(MaskSpawnGroup=false, Override Seed=false, Sphere Radius=1, UseOwnerScale=true)
   &AddVelocity(Velocity=&RandomRangeVector(Minimum=<-100 -300 -100>, Maximum=<100 -65 100>), Velocity Speed Scale=1.5) {module=AddVelocity001}
   /SetVariables(Particles.MeshIndex=&RandomRangeInt(
     Minimum=&Multiply_Int(A=1, B=&SpawnGroup())
     Maximum=&Multiply_Int(A=1, B=&SpawnGroup())
    )) {isSetParameters=true, module=SetVariables_#3}
   &InitialMeshOrientation(Use Rotation Vector=false)
   &FN_SetSpawnGroup(Number of Spawn Groups=7)
  ]
  particleUpdateScript [
   &ParticleState()
   &UpdateMeshOrientation(
    Rotation Rate=&RandomRangeFloat(Minimum=3, Maximum=4)
    Rotation Vector=&RandomRangeVector(Minimum=<-1 -1 -1>, Maximum=<1 1 1>)
   )
   &GravityForce(Gravity=&RandomRangeVector(Minimum=<0 0 -980>, Maximum=<1 0 -980>))
   &Drag(Drag=0.8, Use Linear Drag=true, Use Rotational Drag=true)
   &SolveForcesAndVelocity(Clamp Velocity=false, Limit Acceleration=false)
   &ScaleMeshSize(Scale Factor=&VectorFromFloat(Value=&FloatFromCurve(FloatCurve=DataInterface() {propertyValues {Curve {keys [
          {arriveTangent=-0.008, interpMode=RCIM_Cubic, leaveTangent=-0.008, tangentMode=RCTM_User, time=0.7, value=1}
          {arriveTangent=-3.1748, interpMode=RCIM_Cubic, leaveTangent=-3.1748, tangentMode=RCTM_User, time=0.9, value=0}
         ]}}})))
   &ScaleColor(
    Scale RGB=&MakeVectorFromLinearColorRGB(Color=&FN_Select_Color_from_Time_Of_Day(Morning=<rgba 1 0.8463 0.7189 1>, Day=<rgba 1 1 1 1>, Evening=<rgba 1 0.8371 0.7936 1>, Night=<rgba 0.1429 0.183 0.4497 1>))
    ScaleA=false
    ScaleRGB=true
   ) {module=ScaleColor001}
  ]
  rendererClasses [&NiagaraMeshRendererProperties]
  renderers [{rendererClass=&NiagaraMeshRendererProperties, rendererIndex=0}]
 }
 NE_Debris_dir {
  emitterUpdateScript [
   &EmitterState(Life Cycle Mode=::NewEnumerator1, Loop Behavior=::NewEnumerator1, Loop Delay=0.05, Loop Duration=3, UseLoopDelay=false)
   &SpawnBurst_Instantaneous(Spawn Count=10, Use Spawn Probability=false, UseLoopCountLimit=false)
   &SpawnBurst_Instantaneous(Spawn Count=10, Spawn Group=1, Use Spawn Probability=false, UseLoopCountLimit=false) {module=SpawnBurst_Instantaneous001}
  ]
  particleSpawnScript [
   &InitializeParticle(
    Color Maximum=<rgba 0.4921 0.5944 0.7083 1>
    Color Minimum=<rgba 0.5925 0.8193 0.9844 1>
    Color Mode=::NewEnumerator2
    Lifetime Max=0.6
    Lifetime Min=0.2
    Lifetime Mode=::NewEnumerator1
    Mass Mode=::NewEnumerator1
    Mesh Scale Max=<1.5 1.5 1.5>
    Mesh Scale Min=<0.25 0.25 0.25>
    Mesh Scale Mode=::NewEnumerator2
    Position Offset=<0 0 5>
    Position Offset Coordinate Space=::Local
    UsePositionOffset=true
    Write Material Random=true
   )
   &ShapeLocation(MaskSpawnGroup=false, Override Seed=false, Sphere Radius=5, UseOwnerScale=true)
   &AddVelocity(
    Module.Use Falloff Distance=false
    Override Seed=false
    Velocity Mode=::NewEnumerator1
    Velocity Speed=&RandomRangeFloat(Minimum=100, Maximum=200)
   )
   &AddVelocity(Velocity=&RandomRangeVector(Minimum=<100 -300 0>, Maximum=<-100 -900 100>)) {module=AddVelocity001}
   /SetVariables(Particles.MeshIndex=&RandomRangeInt(Minimum=0, Maximum=3)) {isSetParameters=true, module=SetVariables_#3}
   &InitialMeshOrientation(Use Rotation Vector=false)
  ]
  particleUpdateScript [
   &ParticleState()
   &UpdateMeshOrientation(
    Rotation Rate=&RandomRangeFloat(Minimum=3, Maximum=4)
    Rotation Vector=&RandomRangeVector(Minimum=<-1 -1 -1>, Maximum=<1 1 1>)
   )
   &GravityForce(Gravity=&RandomRangeVector(Minimum=<0 -150 -980>, Maximum=<1 -300 -980>))
   &Drag(Use Linear Drag=true, Use Rotational Drag=true)
   &SolveForcesAndVelocity(Clamp Velocity=false, Limit Acceleration=false)
   &ScaleMeshSize(Scale Factor=&VectorFromFloat(Value=&FloatFromCurve(FloatCurve=DataInterface() {propertyValues {Curve {keys [
          {arriveTangent=-0.008, interpMode=RCIM_Cubic, leaveTangent=-0.008, tangentMode=RCTM_User, time=0, value=1}
          {arriveTangent=-3.1748, interpMode=RCIM_Cubic, leaveTangent=-3.1748, tangentMode=RCTM_User, time=0.9, value=0}
         ]}}})))
   &ScaleColor(
    Scale RGB=&MakeVectorFromLinearColorRGB(Color=&FN_Select_Color_from_Time_Of_Day(Morning=<rgba 1 0.8463 0.7189 1>, Day=<rgba 1 1 1 1>, Evening=<rgba 1 0.8371 0.7936 1>, Night=<rgba 0.1429 0.183 0.4497 1>))
    ScaleA=false
    ScaleRGB=true
   ) {module=ScaleColor001}
  ]
  rendererClasses [&NiagaraMeshRendererProperties]
  renderers [{rendererClass=&NiagaraMeshRendererProperties, rendererIndex=0}]
 }
 NE_flameTrail {
  bEnabled=false
  emitterUpdateScript [
   &EmitterState(Inactive Response=::NewEnumerator1, Life Cycle Mode=::NewEnumerator1, Loop Behavior=::NewEnumerator1, UseLoopDelay=false)
   &SpawnRate(SpawnRate=0, Use Spawn Probability=false) {enabled=false}
   &SpawnParticlesFromOtherEmitter(
    Attribute Reader=DataInterface() {propertyValues {EmitterBinding {bindingMode=Other, emitterName=Snowflakes}}}
    Spawn Rate=100
    Use Maximum Age=false
    Use Minimum Age=false
   )
  ]
  particleSpawnScript [
   &InitializeParticle(
    Color=<rgba 0.2807 0.9052 1 0.1>
    Lifetime=0.4
    Ribbon Width=&RandomRangeFloat(Minimum=12, Maximum=14)
    Ribbon Width Mode=::NewEnumerator1
    UsePositionOffset=false
    Write Material Random=true
   )
   &SampleParticlesFromOtherEmitter(Color Sampling=::NewEnumerator1, Particle ID Sampling=::NewEnumerator1) {module=SampleParticlesFromOtherEmitter001}
  ]
  particleUpdateScript [
   &SolveForcesAndVelocity(Clamp Velocity=false, Limit Acceleration=false)
   &ParticleState()
   &ScaleRibbonWidth(Ribbon Width Scale=&FloatFromCurve(
     FloatCurve=DataInterface() {propertyValues {Curve {keys [{arriveTangent=-1.1377, interpMode=RCIM_Cubic, leaveTangent=-1.1377, tangentMode=RCTM_User, time=0, value=1}, <at 1 1>]}}}
     Scale Curve=0.4
    ))
   &ScaleColor(
    Scale Alpha=&FloatFromCurve(FloatCurve=DataInterface() {propertyValues {Curve {keys [{arriveTangent=0.3606, interpMode=RCIM_Cubic, leaveTangent=0.3606, tangentMode=RCTM_User, time=0, value=1}, <at 1 0>]}}})
    ScaleA=true
    ScaleRGB=true
   )
   &self:ReadRibbonSourceParticleAge_07(PAR=DataInterface() {propertyValues {EmitterBinding {bindingMode=Other, emitterName=Snowflakes}}}) {enabled=false, module=ReadRibbonSourceParticleAge_07}
   &ScaleColor(
    Scale Alpha=&FloatFromCurve(
     FloatCurve=DataInterface() {propertyValues {Curve {keys [
         {arriveTangent=3.0043, interpMode=RCIM_Cubic, leaveTangent=0.8333, tangentMode=RCTM_Break, time=0.1, value=0.5}
         {arriveTangent=-0.6087, interpMode=RCIM_Cubic, leaveTangent=-0.6087, tangentMode=RCTM_User, time=0.8, value=0.5}
         {arriveTangent=-0.0405, interpMode=RCIM_Cubic, leaveTangent=-0.0405, tangentMode=RCTM_User, time=0.895, value=0}
        ]}}}
     CurveIndex=$Emitter.NormalizedLoopAge
     Scale Curve=2
    )
    Scale RGB=&Multiply_VectorByFloat(Vector=<1 1 1>, Float=1)
    ScaleA=true
    ScaleRGB=false
   ) {module=ScaleColor002}
   &DynamicMaterialParameters(
    Index 0 Param 1=&FloatFromCurve(FloatCurve=DataInterface() {propertyValues {Curve {keys [<at 0 1>,<at 1 0.3>]}}})
    Index 0 Param 3=-3
    Index 0 Param 4=0.5
    Index 1 Param 1=30
    Module.Index0Float0Write=true
    Module.Index0Float1Write=true
    Module.Index0Float2Write=true
    Module.Index0Float3Write=true
    Module.Index1Float0Write=true
    Module.Index1Float1Write=true
    Module.Index1Float2Write=true
    Module.Index1Float3Write=true
    Write Parameter Index 1=::NewEnumerator1
   )
  ]
  rendererClasses [&NiagaraRibbonRendererProperties]
  renderers [{rendererClass=&NiagaraRibbonRendererProperties}]
 }
 Snowflakes {
  emitterUpdateScript [
   &EmitterState(Life Cycle Mode=::NewEnumerator1, Loop Behavior=::NewEnumerator1, Loop Duration=1, UseLoopDelay=false)
   &SpawnBurst_Instantaneous(Spawn Count=25, Use Spawn Probability=false, UseLoopCountLimit=false)
  ]
  particleSpawnScript [
   &InitializeParticle(
    Color=&MakeCustomLinearColorFromBool(Bool=&RandomBool(), True Color=<rgba 0.5794 2.4626 5 1>, False Color=<rgba 0.2456 0.5543 1 1>)
    Lifetime Max=1.2
    Lifetime Min=0.5
    Lifetime Mode=::NewEnumerator1
    Sprite Size Mode=::NewEnumerator3
    Uniform Sprite Size=&RandomRangeFloat(Minimum=1.5, Maximum=2)
    UsePositionOffset=false
    Write Material Random=true
   ) {module=InitializeParticle001}
   &SphereLocation(
    MaskSpawnGroup=false
    Override Seed=false
    Sphere Radius=15
    Surface Only=true
    UseNonUniformScale=false
    UseOffset=true
    UseSphereOrientation=false
   )
   &AddVelocityFromPoint(Module.Use Falloff Distance=false, Velocity Strength=&RandomRangeFloat(Minimum=50, Maximum=600))
   &AddVelocity(Velocity=&RandomRangeVector(Minimum=<0 -150 0>, Maximum=<0 -500 0>)) {module=AddVelocity001}
  ]
  particleUpdateScript [
   &ParticleState()
   &CurlNoiseForce(Noise Frequency=10, Noise Strength=200, Pan Noise=false)
   &Drag(
    Drag=&FloatFromCurve(FloatCurve=DataInterface() {propertyValues {Curve {keys [<at 0.5 4>,<at 1 4>]}}})
    Use Linear Drag=true
    Use Rotational Drag=false
   )
   &GravityForce(Gravity=<0 0 -220>)
   &SolveForcesAndVelocity(Clamp Velocity=false, Limit Acceleration=false)
   &SpriteRotationRate(
    Delta Time:NiagaraFloat=$Particles.NormalizedAge
    Rotation Rate=&RandomRangeFloat(Minimum=-5, Maximum=5)
   )
   &ScaleSpriteSize(Scale Factor=&Vector2DFromFloat(Value=&FloatFromCurve(FloatCurve=DataInterface() {propertyValues {Curve {keys [
          {arriveTangent=0.7564, interpMode=RCIM_Cubic, leaveTangent=0.7564, tangentMode=RCTM_User, time=-0.0024, value=0.2}
          {arriveTangent=1.4647, interpMode=RCIM_Cubic, leaveTangent=1.4647, time=0, value=1.55}
          <at 0.9 1.5218>
          <at 0.994 0.1778 RCIM_Cubic>
         ]}}})))
   &ScaleColor(
    Scale Alpha=&Multiply_Float(
     A=&FloatFromCurve(FloatCurve=DataInterface() {propertyValues {Curve {keys [{interpMode=RCIM_Cubic, tangentMode=RCTM_User, time=0.8, value=1}, {interpMode=RCIM_Cubic, tangentMode=RCTM_User, time=1, value=0}]}}})
     B=1
    )
    ScaleA=true
    ScaleRGB=true
   )
   &DynamicMaterialParameters(
    Index 0 Param 2=5
    Module.Param0WriteEnabled=true
    Module.Param1WriteEnabled=true
    Module.Param2WriteEnabled=true
    Module.Param3WriteEnabled=true
   )
   &ScaleColor(
    Scale RGB=&MakeVectorFromLinearColorRGB(Color=&FN_Select_Color_from_Time_Of_Day(Morning=<rgba 1 0.8463 0.7189 1>, Day=<rgba 1 1 1 1>, Evening=<rgba 1 0.8371 0.7936 1>, Night=<rgba 0.1429 0.183 0.4497 1>))
    ScaleA=false
    ScaleRGB=true
   ) {module=ScaleColor001}
  ]
  rendererClasses [&NiagaraSpriteRendererProperties]
  renderers [{rendererClass=&NiagaraSpriteRendererProperties}]
 }
 SnowflakesGPU {
  bEnabled=false
  simTarget=GPUComputeSim
  emitterUpdateScript [
   &EmitterState(Life Cycle Mode=::NewEnumerator1, Loop Behavior=::NewEnumerator1, Loop Duration=3, UseLoopDelay=false)
   &SpawnBurst_Instantaneous(Spawn Count=100, Use Spawn Probability=false, UseLoopCountLimit=false)
   &SpawnRate(SpawnRate=40, Use Spawn Probability=false) {enabled=false}
  ]
  particleSpawnScript [
   &InitializeParticle(
    Color Maximum=<rgba 0.42 0.4712 0.6 0.6>
    Color Minimum=<rgba 0.3882 0.5623 1 0.9>
    Color Mode=::NewEnumerator2
    Lifetime Mode=::NewEnumerator1
    Sprite Rotation Mode=::NewEnumerator2
    Sprite Size Mode=::NewEnumerator1
    Uniform Sprite Size Max=7
    UsePositionOffset=false
    Write Material Random=true
   )
   &ShapeLocation(MaskSpawnGroup=false, Override Seed=false, UseOwnerScale=true)
   &AddVelocity(
    Cone Angle=&MakeCustomFloatFromBool(Bool=$System.IsSynced, True Float=&RandomRangeFloat(Minimum=60, Maximum=180), False Float=180)
    Cone Axis=&MakeCustomVectorFromBool(Bool=$System.IsSynced, True Vector=$System.SnowDirection, False Vector=<0 -0.2 1>)
    Distribution Along Cone Axis=0.75
    Inner Cone Angle=5
    Module.UseConeFalloff=true
    Module.UseVelDistribution=true
    Override Seed=false
    Speed Falloff From Cone Axis=0.3
    Velocity Mode=::NewEnumerator2
    Velocity Speed=&MakeCustomFloatFromBool(Bool=true, True Float=&RandomRangeFloat(Minimum=222, Maximum=333), False Float=555)
   ) {enabled=false}
   &SubUVAnimation(
    Evaluation Type=::NewEnumerator0
    "Sprite Renderer "=DataInterface() {propertyValues {SpriteRenderer=&self:SnowflakesGPU_6.NiagaraSpriteRendererProperties_1}}
    SubUV Animation Mode:ENiagara_SubUVLookupModeV2=::NewEnumerator2
    UseEndFrame=false
    UseStartFrame=false
   )
   &InheritVelocity(Speed Limit=false, UseSpeedThreshold=false)
   &AddVelocity(Module.Use Falloff Distance=false, Origin Offset=<0 0 -20>, Override Seed=false, Velocity Mode=::NewEnumerator1) {module=AddVelocity001}
  ]
  particleUpdateScript [
   &ParticleState()
   &Drag(Drag=&RandomRangeFloat(Minimum=4, Maximum=6), Use Linear Drag=true, Use Rotational Drag=true)
   &GravityForce(Gravity=&RandomRangeVector(Minimum=<0 0 -88>, Maximum=<1 1 -111>))
   &CurlNoiseForce(
    Noise Quality / Cost=::NewEnumerator0
    Noise Strength=&FloatFromCurve(
     FloatCurve=DataInterface() {propertyValues {Curve {keys [
         {arriveTangent=4.1785, interpMode=RCIM_Cubic, leaveTangent=4.1785, tangentMode=RCTM_User, time=0, value=0}
         {arriveTangent=1.1179, interpMode=RCIM_Cubic, leaveTangent=1.1179, tangentMode=RCTM_User, time=0.2042, value=0.6228}
         {arriveTangent=0.0224, interpMode=RCIM_Cubic, leaveTangent=0.0224, tangentMode=RCTM_User, time=1, value=1}
        ]}}}
     Scale Curve=333
    )
    Pan Noise=false
   )
   &AccelerationForce(
    Acceleration=&MakeCustomVectorFromBool(
     Bool=$System.IsSynced
     True Vector=&Multiply_VectorByFloat(
      Vector=&Multiply_Vector(A=$StackContext.ShapeLocation.ShapeNormal, B=<1 1 0>)
      Float=&Lerp_Float(
       A=333
       B=33
       Alpha=&FloatFromCurve(FloatCurve=DataInterface() {propertyValues {Curve {keys [{tangentMode=RCTM_User, time=0, value=0}, {tangentMode=RCTM_User, time=1, value=1}]}}})
      )
     )
     False Vector=<0 0 0>
    )
    Coordinate Space=::Local
   )
   &InheritSourceMovement(ApplyForce=true, ApplyForceFalloff=false, ApplyPosition=false, Module.LimitForceSpeed=false)
   &SolveForcesAndVelocity(Clamp Velocity=false, Limit Acceleration=false)
   &SpriteRotationRate(Rotation Rate=&FloatFromCurve(
     FloatCurve=DataInterface() {propertyValues {Curve {keys [
         {arriveTangent=4.1785, interpMode=RCIM_Cubic, tangentMode=RCTM_Break, time=0, value=0}
         {arriveTangent=0.0224, interpMode=RCIM_Cubic, leaveTangent=0.0224, tangentMode=RCTM_Break, time=0.3, value=1}
        ]}}}
     CurveIndex=$Particles.Age
     Scale Curve=&RandomRangeFloat(Minimum=-360, Maximum=360)
    ))
   &ScaleSpriteSize(
    Scale Factor=&Multiply_Vector2D_ByFloat(
     Float=&FloatFromCurve(
      FloatCurve=DataInterface() {propertyValues {Curve {keys [
          {tangentMode=RCTM_User, time=0, value=1}
          {arriveTangent=-0.0405, interpMode=RCIM_Cubic, leaveTangent=-0.0405, tangentMode=RCTM_User, time=0.7009, value=1}
          {arriveTangent=-8.0089, interpMode=RCIM_Cubic, leaveTangent=-8.0089, tangentMode=RCTM_User, time=1, value=0}
         ]}}}
      Scale Curve=1.75
     )
     Vector2D=&Lerp_Vector2(
      A=<1 2>
      B=&RandomRangeVector2D(Minimum=<0.8 0.8>, Maximum=<1 1>)
      Alpha=&FloatFromCurve(
       FloatCurve=DataInterface() {propertyValues {Curve {keys [
           {arriveTangent=4.1785, interpMode=RCIM_Cubic, tangentMode=RCTM_Break, time=0, value=0}
           {arriveTangent=9.9282, interpMode=RCIM_Cubic, leaveTangent=0.0224, tangentMode=RCTM_Break, time=0.3, value=1}
          ]}}}
       CurveIndex=$Particles.Age
      )
     )
    )
    Scale Sprite Size Mode=::NewEnumerator2
   )
   &SpriteFacingAndAlignment(
    Sprite Alignment=&Lerp_Vector(
     A=&ReturnCameraVector(Camera Vectors=::NewEnumerator2, Coordinate Space=::Simulation)
     B=&NormalizeVector(Vector To Normalize=$Particles.Velocity, Safe Normalize=true, Safe Fallback Threshold=0, Safe Fallback Vector=<0 0 0.001>)
     Alpha=&ClampFloat(Float=&Divide_Float(A=&MakeFloatFromVector(Vector=$Particles.Velocity, Channel=::NewEnumerator2), B=33))
    )
    WriteAlignment=true
    WriteFacing=false
   )
   &ScaleColor(
    Scale Alpha=&Multiply_Float(
     A=&FloatFromCurve(
      FloatCurve=DataInterface() {propertyValues {Curve {keys [
          {interpMode=RCIM_Cubic, leaveTangent=2.9196, tangentMode=RCTM_Break, time=0, value=0}
          {arriveTangent=-0.0405, interpMode=RCIM_Cubic, leaveTangent=-0.0405, tangentMode=RCTM_User, time=0.7009, value=1}
          {arriveTangent=-8.0089, interpMode=RCIM_Cubic, leaveTangent=-8.0089, tangentMode=RCTM_User, time=1, value=0}
         ]}}}
      Scale Curve=&RandomRangeFloat(Minimum=0.6, Maximum=1)
     )
     B=1
    )
    Scale RGB=&MakeVectorFromLinearColorRGB(Color=$System.TODColorMultiplier)
    ScaleA=true
    ScaleRGB=true
   )
   /SetVariables(
    Particles.Sparkles Color:LinearColor=&MakeLinearColorFromVectorAndFloat(
     "Vector (RGB)"=&MakeVectorFromLinearColorRGB(Color=&MultiplyLinearColor(
       Linear Color=&MultiplyLinearColor(
        Linear Color=$Particles.Color
        Scale Factor=&MakeVector4_FromLinearColor(LinearColor=&FN_Select_Color_from_Time_Of_Day(Morning=<rgba 1 0.8463 0.7189 1>, Day=<rgba 1 1 1 1>, Evening=<rgba 1 0.8371 0.7936 1>, Night=<rgba 0.1429 0.183 0.4497 1>))
       )
       Scale Factor=&MakeFloat4FromVectorAndFloat(
        "Vector (RGB)"=&VectorFromFloat(Value=&RandomRangeFloat(Minimum=4, Maximum=11))
        "Float (Alpha)"=1
       )
      ))
     "Float (Alpha)"=&Lerp_Float(
      A=&Multiply_Float(
       A=&FN_Select_Float_from_Time_Of_Day(Morning=2, Day=2, Evening=2, Night=10)
       B=&Waveform(
        Global Phase=$Particles.Age
        "Waveform [1]"=::NewEnumerator0
        "[1] Phase"=$Particles.Age
        "[1] Frequency"=2
        "[1] Amplitude Min/Max"=&MakeVector2D(X=0, Y=&RandomRangeFloat(Minimum=-1, Maximum=1))
        "[1] Scale"=&RandomRangeFloat(Minimum=-1, Maximum=1)
       )
      )
      B=0
      Alpha=&NormalizeDistanceRange(
       Start Position:NiagaraPosition=&TransformVector(
        Vector=&AddVectorToPosition(Position=$Engine.Emitter.SimulationPosition, Vector=<0 100 50>)
        Destination Space=::Local
       )
       End Position:NiagaraPosition=$Particles.Position
       Distance:NiagaraFloat=50
       Invert Normalized Range:NiagaraBool=false
      )
     )
    )
    Particles.SpriteSizeSparkles:Vector2f=&Multiply_Vector2D_ByFloat(Float=&RandomRangeFloat(Minimum=0.5, Maximum=1.5), Vector2D=$Particles.SpriteSize)
   ) {isSetParameters=true}
  ]
  rendererClasses [&NiagaraSpriteRendererProperties]
  renderers [
   {rendererClass=&NiagaraSpriteRendererProperties}
   {rendererClass=&NiagaraSpriteRendererProperties, rendererIndex=1}
  ]
 }
 SoftShockwave {
  bEnabled=false
  emitterUpdateScript [
   &EmitterState()
   &SpawnBurst_Instantaneous(Use Spawn Probability=false, UseLoopCountLimit=false)
  ]
  particleSpawnScript [&InitializeParticle(
    Color=<rgba 0 4.3536 5 0.5>
    Lifetime=0.333
    Mesh Scale Mode=::NewEnumerator3
    Mesh Uniform Scale=1.3434
    UsePositionOffset=false
    Write Material Random=true
   )]
  particleUpdateScript [
   &ParticleState()
   &ScaleColor(
    Scale Alpha=&FloatFromCurve(FloatCurve=DataInterface() {propertyValues {Curve {keys [<at 0 1>, {interpMode=RCIM_Cubic, tangentMode=RCTM_User, time=0.1, value=1}, {interpMode=RCIM_Cubic, tangentMode=RCTM_User, time=1, value=0}]}}})
    ScaleA=true
    ScaleRGB=true
   )
   &DynamicMaterialParameters(
    Index 0 Param 1=1
    Index 0 Param 2=-0.5
    Index 0 Param 3=1
    Index 0 Param 4=1
    Module.Param0WriteEnabled=true
    Module.Param1WriteEnabled=true
    Module.Param2WriteEnabled=true
    Module.Param3WriteEnabled=true
   )
   &ScaleMeshSize(Scale Factor=&VectorFromFloat(Value=&FloatFromCurve(FloatCurve=DataInterface() {propertyValues {Curve {keys [
          {arriveTangent=4.1785, interpMode=RCIM_Cubic, leaveTangent=4.1785, tangentMode=RCTM_User, time=0, value=0.05}
          {arriveTangent=0.9995, interpMode=RCIM_Cubic, leaveTangent=0.9995, tangentMode=RCTM_User, time=0.3, value=0.6}
          {arriveTangent=-0.0405, interpMode=RCIM_Cubic, leaveTangent=-0.0405, tangentMode=RCTM_User, time=1, value=1}
         ]}}})))
  ]
  rendererClasses [&NiagaraMeshRendererProperties]
  renderers [{rendererClass=&NiagaraMeshRendererProperties, rendererIndex=0}]
 }
 SpikeStreaksParticle {
  bEnabled=false
  emitterUpdateScript [
   &EmitterState(
    Enable Distance Culling=true
    Max Distance Response=::KillImmediately
    MaxDistance=3500
    Module.bUseMaxDistance=true
    Module.bUseMinDistance=false
    Scalability Mode=::NewEnumerator1
   )
   &SpawnBurst_Instantaneous(Spawn Count=8, Use Spawn Probability=false)
  ]
  particleSpawnScript [
   /Niagara/Modules/Spawn/Initialization/InitializeParticle(
    Color=<rgba 2.8724 6.7998 10 1>
    Lifetime=&UniformRangedFloat(Minimum=0.15, Maximum=0.2)
    Module.Write Color=true
    Module.Write Lifetime=true
    Module.Write Mass=true
    Module.Write Material Random=true
    Module.Write Position=true
    Module.Write Scale=false
    Module.Write SpriteRotation=false
    Module.Write SpriteSize=true
    Position=&Add_Vector(A=/Niagara/DynamicInputs/Helpers/SimulationPosition(), B=<0 0 0>)
    Sprite Size=&UniformRangedVector2D(Minimum=<10 20>, Maximum=<20 40>)
   )
   /Niagara/Modules/Spawn/Initialization/V2/InitializeParticle(
    Color=<rgba 2.8724 6.7998 10 1>
    Lifetime=&UniformRangedFloat(Minimum=0.15, Maximum=0.2)
    UsePositionOffset=false
    Write Material Random=true
   ) {enabled=false, module=InitializeParticle001}
   &AddVelocityFromPoint(Module.Use Falloff Distance=false, Velocity Strength=50)
  ]
  particleUpdateScript [
   &ParticleState()
   &CameraOffset(Camera Offset Amount=10)
   &SolveForcesAndVelocity(Clamp Velocity=false, Limit Acceleration=false)
   &ScaleSpriteSize(Scale Factor=&Vector2DFromFloat(Value=&FloatFromCurve(FloatCurve=DataInterface() {propertyValues {Curve {keys [<at 0 0.6>,<at 1 2>]}}}, Scale Curve=0.75)))
   &ScaleColor(
    Scale Alpha=&FloatFromCurve(FloatCurve=DataInterface() {propertyValues {Curve {keys [<at 0 1>,<at 1 0>]}}}, Scale Curve=2)
    Scale RGB=&VectorFromFloat(Value=0.1)
    ScaleA=true
    ScaleRGB=true
   )
  ]
  rendererClasses [&NiagaraSpriteRendererProperties]
  renderers [{rendererClass=&NiagaraSpriteRendererProperties}]
 }
 mist_radial {
  emitterUpdateScript [
   &EmitterState(Life Cycle Mode=::NewEnumerator1, Loop Behavior=::NewEnumerator1, Loop Duration=1, UseLoopDelay=false)
   &SpawnBurst_Instantaneous(Spawn Count=15, Use Spawn Probability=false, UseLoopCountLimit=false)
  ]
  particleSpawnScript [
   &InitializeParticle(
    Color=<rgba 0.4492 0.6108 0.8177 0.3>
    Lifetime=&RandomRangeFloat(Minimum=0.8, Maximum=1.5)
    Sprite Rotation Angle=&RandomRangeFloat(Minimum=-55, Maximum=98)
    Sprite Rotation Mode=::NewEnumerator0
    Sprite Size Mode=::NewEnumerator3
    Sprite UV Mode=::NewEnumerator2
    Uniform Sprite Size=&Multiply_Float(A=&RandomRangeFloat(Minimum=90, Maximum=120), B=1)
    UsePositionOffset=false
    Write Material Random=true
   )
   &ShapeLocation(
    MaskSpawnGroup=false
    Offset=<0 -1.3139 0>
    Offset Mode=::NewEnumerator0
    Override Seed=false
    Rotation Angle=0.25
    Rotation Axis=<0 1 0>
    Rotation Mode=::NewEnumerator1
    Sphere Radius=10
    Sphere Surface Distribution=1
    UseOwnerScale=true
   )
   &InheritVelocity(Inherited Velocity Amount Scale=&VectorFromFloat(Value=0.1), Speed Limit=true, UseSpeedThreshold=true) {enabled=false}
   &CameraOffset(Camera Offset Amount=16.875) {module=CameraOffset001}
   &DynamicMaterialParameters(
    Index 0 Param 4=0.3
    Module.Index0Float0Write=true
    Module.Index0Float1Write=true
    Module.Index0Float2Write=true
    Module.Index0Float3Write=true
   )
   &AddVelocity(Velocity=&MergeHelper_VectorSwitch(
     0=<0 0 0>
     1=<0 0 0>
     2=&Multiply_VectorByFloat(Vector=$System.DirectionVector, Float=&RandomRangeFloat(Minimum=110, Maximum=120))
     3=?
     Index=&ReturnExecIndex()
     Num=::NewEnumerator1
    )) {enabled=false}
   &PointForce(Module.Use Falloff=false, Module.UseLinearForce=false)
   &ApplyInitialForces()
   &AddVelocity(Velocity=&RandomRangeVector(Minimum=<50 0 0>, Maximum=<-50 0 0>)) {module=AddVelocity001}
  ]
  particleUpdateScript [
   &ParticleState()
   &GravityForce(Gravity=<0 -150 -52.0864>)
   &Drag(Drag=2, Use Linear Drag=true, Use Rotational Drag=true)
   &SolveForcesAndVelocity(Clamp Velocity=false, Limit Acceleration=false)
   &ScaleSpriteSize(Scale Factor=&Vector2DFromFloat(Value=&FloatFromCurve(FloatCurve=DataInterface() {propertyValues {Curve {keys [{arriveTangent=1.8373, interpMode=RCIM_Cubic, leaveTangent=1.8373, tangentMode=RCTM_User, time=0, value=0.5}, <at 1 1.3 RCIM_Cubic>]}}})))
   &ScaleColor(
    Scale Alpha=&FloatFromCurve(FloatCurve=DataInterface() {propertyValues {Curve {keys [
         <at 0 0 RCIM_Cubic>
         {arriveTangent=-0.4056, interpMode=RCIM_Cubic, leaveTangent=-0.4056, tangentMode=RCTM_User, time=0.0203, value=0.55}
         {arriveTangent=-0.5614, interpMode=RCIM_Cubic, leaveTangent=-0.5614, time=0.0878, value=0.5}
         <at 1 0 RCIM_Cubic>
        ]}}})
    ScaleA=true
    ScaleRGB=true
   )
   &SubUVAnimation(
    End Frame=26
    SubUV Animation Mode:ENiagara_SubUVLookupMode=::NewEnumerator1
    SubUV Curve=DataInterface() {propertyValues {Curve {keys [
        {arriveTangent=2.3868, interpMode=RCIM_Cubic, leaveTangent=2.3868, tangentMode=RCTM_User, time=0, value=0}
        {interpMode=RCIM_Cubic, tangentMode=RCTM_User, time=1, value=1}
       ]}}}
   ) {enabled=false}
   /SetVariables(Particles.SubImageIndex=&Add_Float(
     A=&RandomRangeFloat(Minimum=0, Maximum=0.3)
     B=&FloatFromCurve(
      FloatCurve=DataInterface() {propertyValues {Curve {keys [{arriveTangent=120.5723, interpMode=RCIM_Cubic, leaveTangent=120.5723, tangentMode=RCTM_User, time=0, value=0}, <at 1 64 RCIM_Cubic>]}}}
      Scale Curve=&RandomRangeFloat(Minimum=0.3, Maximum=0.5)
     )
    )) {enabled=false, isSetParameters=true, module=SetVariables_#2}
   &CameraOffset(Camera Offset Amount=&FloatFromCurve(FloatCurve=DataInterface() {propertyValues {Curve {keys [<at 0 1>,<at 1 100>]}}}))
   &ScaleColor(
    Color Value To Scale=$Particles.Color
    Scale Mode=::NewEnumerator1
    Scale RGBA=&MakeVector4_FromLinearColor(LinearColor=&FNTimeOfDayColor(
      Night Color=<rgba 0.539 0.6941 1 1>
      Morning Color=<rgba 0.9392 0.8785 1 1>
      Day Color=<rgba 1 1 1 1>
      Evening Color=<rgba 1 0.6882 0.6583 1>
     ))
    ScaleRGBA=true
   ) {module=ScaleColor001}
   &SubUVAnimation(
    "Sprite Renderer "=DataInterface() {propertyValues {SpriteRenderer=&self:mist_radial_0.NiagaraSpriteRendererProperties_1}}
    UseEndFrame=false
    UseStartFrame=false
   ) {module=SubUVAnimation001}
   &ScaleColor(
    Scale Alpha=&MakeFloatFromVector(
     Vector=&MakeVectorFromLinearColorRGB(Color=&FN_Select_Color_from_Time_Of_Day(Morning=<rgba 1 0.8463 0.7189 1>, Day=<rgba 1 1 1 1>, Evening=<rgba 1 0.8371 0.7936 1>, Night=<rgba 0.1429 0.183 0.4497 0.8>))
     Channel=::NewEnumerator0
    )
    Scale RGB=&MakeVectorFromLinearColorRGB(Color=&FN_Select_Color_from_Time_Of_Day(Morning=<rgba 1 0.8463 0.7189 1>, Day=<rgba 1 1 1 1>, Evening=<rgba 1 0.8371 0.7936 1>, Night=<rgba 0.1429 0.183 0.4497 1>))
    ScaleA=true
    ScaleRGB=false
   ) {module=ScaleColor002}
   &ScaleColor(Scale Alpha=&FN_FrontEndFloatModifier(Gameplay=1, Front End=0.1), ScaleA=true, ScaleRGB=false) {module=ScaleColor004}
  ]
  rendererClasses [&NiagaraSpriteRendererProperties]
  renderers [
   {rendererClass=&NiagaraSpriteRendererProperties}
   {rendererClass=&NiagaraSpriteRendererProperties, rendererIndex=1}
  ]
 }
 mist_radial001 {
  bEnabled=false
  emitterUpdateScript [
   &EmitterState(Life Cycle Mode=::NewEnumerator1, Loop Behavior=::NewEnumerator1, Loop Duration=1, UseLoopDelay=false)
   &SpawnBurst_Instantaneous(Spawn Count=15, Use Spawn Probability=false, UseLoopCountLimit=false)
  ]
  particleSpawnScript [
   &InitializeParticle(
    Color=<rgba 0.4492 0.6108 0.8177 0.3>
    Lifetime=&RandomRangeFloat(Minimum=0.8, Maximum=1.5)
    Sprite Rotation Angle=&RandomRangeFloat(Minimum=-55, Maximum=98)
    Sprite Rotation Mode=::NewEnumerator0
    Sprite Size Mode=::NewEnumerator3
    Sprite UV Mode=::NewEnumerator2
    Uniform Sprite Size=&Multiply_Float(A=&RandomRangeFloat(Minimum=90, Maximum=120), B=1)
    UsePositionOffset=false
    Write Material Random=true
   )
   &ShapeLocation(
    MaskSpawnGroup=false
    Offset=<0 -1.3139 0>
    Offset Mode=::NewEnumerator0
    Override Seed=false
    Rotation Angle=0.25
    Rotation Axis=<0 1 0>
    Rotation Mode=::NewEnumerator1
    Sphere Radius=10
    Sphere Surface Distribution=1
    UseOwnerScale=true
   )
   &InheritVelocity(Inherited Velocity Amount Scale=&VectorFromFloat(Value=0.1), Speed Limit=true, UseSpeedThreshold=true) {enabled=false}
   &CameraOffset(Camera Offset Amount=16.875) {module=CameraOffset001}
   &DynamicMaterialParameters(
    Index 0 Param 4=0.3
    Module.Index0Float0Write=true
    Module.Index0Float1Write=true
    Module.Index0Float2Write=true
    Module.Index0Float3Write=true
   )
   &AddVelocity(Velocity=&MergeHelper_VectorSwitch(
     0=<0 0 0>
     1=<0 0 0>
     2=&Multiply_VectorByFloat(Vector=$System.DirectionVector, Float=&RandomRangeFloat(Minimum=110, Maximum=120))
     3=?
     Index=&ReturnExecIndex()
     Num=::NewEnumerator1
    )) {enabled=false}
   &PointForce(Module.Use Falloff=false, Module.UseLinearForce=false)
   &ApplyInitialForces()
   &AddVelocity(Velocity=&RandomRangeVector(Minimum=<50 0 0>, Maximum=<-50 0 0>)) {module=AddVelocity001}
  ]
  particleUpdateScript [
   &ParticleState()
   &GravityForce(Gravity=<0 -150 -52.0864>)
   &Drag(Drag=2, Use Linear Drag=true, Use Rotational Drag=true)
   &SolveForcesAndVelocity(Clamp Velocity=false, Limit Acceleration=false)
   &ScaleSpriteSize(Scale Factor=&Vector2DFromFloat(Value=&FloatFromCurve(FloatCurve=DataInterface() {propertyValues {Curve {keys [{arriveTangent=1.8373, interpMode=RCIM_Cubic, leaveTangent=1.8373, tangentMode=RCTM_User, time=0, value=0.5}, <at 1 1.3 RCIM_Cubic>]}}})))
   &ScaleColor(
    Scale Alpha=&FloatFromCurve(FloatCurve=DataInterface() {propertyValues {Curve {keys [
         <at 0 0 RCIM_Cubic>
         {arriveTangent=-0.4056, interpMode=RCIM_Cubic, leaveTangent=-0.4056, tangentMode=RCTM_User, time=0.0203, value=0.8}
         {arriveTangent=-0.8166, interpMode=RCIM_Cubic, leaveTangent=-0.8166, time=0.0878, value=0.5}
         <at 1 0 RCIM_Cubic>
        ]}}})
    ScaleA=true
    ScaleRGB=true
   )
   &SubUVAnimation(
    End Frame=26
    SubUV Animation Mode:ENiagara_SubUVLookupMode=::NewEnumerator1
    SubUV Curve=DataInterface() {propertyValues {Curve {keys [
        {arriveTangent=2.3868, interpMode=RCIM_Cubic, leaveTangent=2.3868, tangentMode=RCTM_User, time=0, value=0}
        {interpMode=RCIM_Cubic, tangentMode=RCTM_User, time=1, value=1}
       ]}}}
   ) {enabled=false}
   /SetVariables(Particles.SubImageIndex=&Add_Float(
     A=&RandomRangeFloat(Minimum=0, Maximum=0.3)
     B=&FloatFromCurve(
      FloatCurve=DataInterface() {propertyValues {Curve {keys [{arriveTangent=120.5723, interpMode=RCIM_Cubic, leaveTangent=120.5723, tangentMode=RCTM_User, time=0, value=0}, <at 1 64 RCIM_Cubic>]}}}
      Scale Curve=&RandomRangeFloat(Minimum=0.3, Maximum=0.5)
     )
    )) {enabled=false, isSetParameters=true, module=SetVariables_#2}
   &CameraOffset(Camera Offset Amount=&FloatFromCurve(FloatCurve=DataInterface() {propertyValues {Curve {keys [<at 0 1>,<at 1 100>]}}}))
   &ScaleColor(
    Color Value To Scale=$Particles.Color
    Scale Mode=::NewEnumerator1
    Scale RGBA=&MakeVector4_FromLinearColor(LinearColor=&FNTimeOfDayColor(
      Night Color=<rgba 0.539 0.6941 1 1>
      Morning Color=<rgba 0.9392 0.8785 1 1>
      Day Color=<rgba 1 1 1 1>
      Evening Color=<rgba 1 0.6882 0.6583 1>
     ))
    ScaleRGBA=true
   ) {module=ScaleColor001}
   &SubUVAnimation(
    "Sprite Renderer "=DataInterface() {propertyValues {SpriteRenderer=&self:mist_radial001_0.NiagaraSpriteRendererProperties_1}}
    UseEndFrame=false
    UseStartFrame=false
   ) {module=SubUVAnimation001}
  ]
  rendererClasses [&NiagaraSpriteRendererProperties]
  renderers [
   {rendererClass=&NiagaraSpriteRendererProperties}
   {rendererClass=&NiagaraSpriteRendererProperties, rendererIndex=1}
  ]
 }
 ne_SNOW_CHUNKS {
  emitterUpdateScript [
   &EmitterState(Life Cycle Mode=::NewEnumerator1, Loop Behavior=::NewEnumerator1, Loop Duration=1, UseLoopDelay=true)
   &SpawnBurst_Instantaneous(Spawn Count=20, Use Spawn Probability=false, UseLoopCountLimit=false)
   &SpawnBurst_Instantaneous(Spawn Count=20, Spawn Group=1, Use Spawn Probability=false, UseLoopCountLimit=false) {module=SpawnBurst_Instantaneous001}
  ]
  particleSpawnScript [
   &InitializeParticle(
    Lifetime Max=0.5
    Lifetime Min=0.4
    Lifetime Mode=::NewEnumerator1
    Mass Mode=::NewEnumerator1
    Mesh Scale Max=<0.7 0.7 0.7>
    Mesh Scale Min=<0.25 0.25 0.25>
    Mesh Scale Mode=::NewEnumerator2
    Position Offset=<0 0 5>
    Position Offset Coordinate Space=::Local
    UsePositionOffset=true
    Write Material Random=true
   )
   &ShapeLocation(MaskSpawnGroup=false, Override Seed=false, Sphere Radius=5, UseOwnerScale=true)
   &AddVelocity(
    Module.Use Falloff Distance=false
    Override Seed=false
    Velocity Mode=::NewEnumerator1
    Velocity Speed=&RandomRangeFloat(Minimum=100, Maximum=200)
   )
   &AddVelocity(Velocity=&RandomRangeVector(Minimum=<100 0 0>, Maximum=<-100 -600 100>)) {module=AddVelocity001}
   /SetVariables(Particles.MeshIndex=&RandomRangeInt(Minimum=0, Maximum=3)) {isSetParameters=true, module=SetVariables_#3}
   &InitialMeshOrientation(Use Rotation Vector=false)
  ]
  particleUpdateScript [
   &ParticleState()
   &UpdateMeshOrientation(
    Rotation Rate=&RandomRangeFloat(Minimum=3, Maximum=4)
    Rotation Vector=&RandomRangeVector(Minimum=<-1 -1 -1>, Maximum=<1 1 1>)
   )
   &GravityForce(Gravity=&RandomRangeVector(Minimum=<0 -150 -980>, Maximum=<1 -300 -980>))
   &Drag(Use Linear Drag=true, Use Rotational Drag=true)
   &SolveForcesAndVelocity(Clamp Velocity=false, Limit Acceleration=false)
   &ScaleMeshSize(Scale Factor=&VectorFromFloat(Value=&FloatFromCurve(FloatCurve=DataInterface() {propertyValues {Curve {keys [
          {arriveTangent=-0.008, interpMode=RCIM_Cubic, leaveTangent=-0.008, tangentMode=RCTM_User, time=0, value=1}
          {arriveTangent=-3.1748, interpMode=RCIM_Cubic, leaveTangent=-3.1748, tangentMode=RCTM_User, time=0.9, value=0}
         ]}}})))
   &FN_Mesh_Attributes()
   &ScaleColor(
    Scale RGB=&MakeVectorFromLinearColorRGB(Color=&FN_Select_Color_from_Time_Of_Day(Morning=<rgba 1 0.8463 0.7189 1>, Day=<rgba 1 1 1 1>, Evening=<rgba 1 0.8371 0.7936 1>, Night=<rgba 0.1429 0.183 0.4497 1>))
    ScaleA=false
    ScaleRGB=true
   ) {module=ScaleColor001}
   &DynamicMaterialParameters(
    Index 0 Param 1=&MakeFloatFromLinearColor(
     LinearColor=&FN_Select_Color_from_Time_Of_Day(Morning=<rgba 1 0.8463 0.7189 1>, Day=<rgba 1 1 1 1>, Evening=<rgba 1 0.8371 0.7936 1>, Night=<rgba 0.1429 0.183 0.4497 1>)
     Channel:ENiagaraFloatFromLinearColorOptions=::NewEnumerator0
    )
    Module.Index0Float0Write=true
    Module.Index0Float1Write=true
    Module.Index0Float2Write=true
    Module.Index0Float3Write=true
   )
  ]
  rendererClasses [&NiagaraMeshRendererProperties]
  renderers [{rendererClass=&NiagaraMeshRendererProperties, rendererIndex=0}]
 }
}
summary {
 systemName=NS_TicketPoke_Impact
 emitters [
  {
   bEnabled=false
   emitterName=Flash001
   rendererClasses [&NiagaraSpriteRendererProperties, &NiagaraMeshRendererProperties]
  }
  {
   bEnabled=false
   emitterName=Flare
   rendererClasses [&NiagaraSpriteRendererProperties, &NiagaraLightRendererProperties]
  }
  {emitterName=Snowflakes, rendererClasses [&NiagaraSpriteRendererProperties]}
  {emitterName=NE_Debris_dir, rendererClasses [&NiagaraMeshRendererProperties]}
  {
   bEnabled=false
   emitterName=Head
   rendererClasses [&NiagaraLightRendererProperties, &NiagaraSpriteRendererProperties]
  }
  {bEnabled=false, emitterName=NE_flameTrail, rendererClasses [&NiagaraRibbonRendererProperties]}
  {bEnabled=false, emitterName=SpikeStreaksParticle, rendererClasses [&NiagaraSpriteRendererProperties]}
  {bEnabled=false, emitterName=Dust, rendererClasses [&NiagaraSpriteRendererProperties]}
  {bEnabled=false, emitterName=Halo, rendererClasses [&NiagaraSpriteRendererProperties]}
  {bEnabled=false, emitterName=SoftShockwave, rendererClasses [&NiagaraMeshRendererProperties]}
  {bEnabled=false, emitterName=BG_Glow, rendererClasses [&NiagaraSpriteRendererProperties]}
  {emitterName=Mist, rendererClasses [&NiagaraSpriteRendererProperties]}
  {bEnabled=false, emitterName=SnowflakesGPU, simTarget=GPUComputeSim, rendererClasses [&NiagaraSpriteRendererProperties]}
  {
   emitterName=Flash_Elements
   rendererClasses [&NiagaraSpriteRendererProperties, &NiagaraLightRendererProperties]
  }
  {emitterName=NE_Debris_Radial, rendererClasses [&NiagaraMeshRendererProperties]}
  {emitterName=mist_radial, rendererClasses [&NiagaraSpriteRendererProperties]}
  {bEnabled=false, emitterName=Mist001, rendererClasses [&NiagaraSpriteRendererProperties]}
  {bEnabled=false, emitterName=mist_radial001, rendererClasses [&NiagaraSpriteRendererProperties]}
  {emitterName=HandGlow, rendererClasses [&NiagaraSpriteRendererProperties]}
  {emitterName=IceBlast, rendererClasses [&NiagaraSpriteRendererProperties]}
  {emitterName=IceSpikes, rendererClasses [&NiagaraMeshRendererProperties]}
  {bEnabled=false, emitterName=IceChunks, rendererClasses [&NiagaraMeshRendererProperties]}
  {emitterName=ne_SNOW_CHUNKS, rendererClasses [&NiagaraMeshRendererProperties]}
 ]
 userVariables [
  User.bIsFrontEnd {type=NiagaraBool}
  User.bIsFrontEndPreview {type=NiagaraBool}
  User.BoolYay {type=NiagaraFloat, defaultValue=0, description=""}
  User.Color {type=LinearColor, description="", defaultValue {struct=LinearColor, value=<rgba 1.3594 2.4773 6 1>}}
  User.Color_Blade_B {type=LinearColor, description="", defaultValue {struct=LinearColor, value=<rgba 0.2266 0.4129 1 1>}}
  User.exp_Delay {type=NiagaraFloat, defaultValue=0.0325, description=""}
  User.FrontEndScale {type=NiagaraFloat, defaultValue=0.6, description=""}
  User.UseCustomColor {type=NiagaraFloat, defaultValue=0, description=""}
  User.UseCustomColor1 {type=NiagaraFloat, defaultValue=0, description=""}
 ]
}
