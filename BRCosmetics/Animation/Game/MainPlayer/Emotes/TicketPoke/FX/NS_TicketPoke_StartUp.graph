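# Niagara system for the start-up effect of the TicketPoke player emote in BRCosmetics.
# Script references point to layered emitters including hand glow, flashes, debris, sparkles and ice-named blast, spike and snowflake layers, using sprite, mesh, ribbon and light renderers.
# Particle-read modules spawn from and sample other emitters, and dynamic inputs vary values by time of day and platform.
# Flare, a burst sprite with a light, and Flash001 are disabled.
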NiagaraSystem /BRCosmetics/Animation/Game/MainPlayer/Emotes/TicketPoke/FX/NS_TicketPoke_StartUp
dependencies {
 usedDataInterfaces [
  {refPath=&NiagaraDataInterfaceCurve}
  {refPath=&NiagaraDataInterfaceColorCurve}
  {refPath=&NiagaraDataInterfaceParticleRead}
  {refPath=&NiagaraDataInterfaceSpriteRendererInfo}
  {refPath=&NiagaraDataInterfaceVectorCurve}
  {refPath=&NiagaraDataInterfacePlatformSet}
  {refPath=&NiagaraDataInterfaceArrayFloat}
  {refPath=&NiagaraDataInterfaceMeshRendererInfo}
  {refPath=&NiagaraDataInterfaceVector2DCurve}
  {refPath=&NiagaraDataInterfaceCamera}
 ]
 usedDynamicInputs [
  {refPath=&VectorFromFloat}
  {refPath=&FloatFromCurve}
  {refPath=&MultiplyLinearColor}
  {refPath=&MakeVector4_FromFloat}
  {refPath=&MakeLinearColorFromVectorAndFloat}
  {refPath=&MakeVectorFromLinearColorRGB}
  {refPath=&MakeVector2D}
  {refPath=&Multiply_Vector2D}
  {refPath=&Add_Vector}
  {refPath=&Multiply_Float}
  {refPath=&MakeFloatFromVector2d}
  {refPath=&MakeCustomLinearColorFromBool}
  {refPath=&RandomBool}
  {refPath=&RandomRangeFloat}
  {refPath=&Vector2DFromFloat}
  {refPath=&MergeHelper_FloatSwitch}
  {refPath=&SpawnGroup}
  {refPath=&RandomRangeVector}
  {refPath=&RandomRangeInt}
  {refPath=&Multiply_VectorByFloat}
  {refPath=&UniformRangedFloat}
  {refPath=/Niagara/DynamicInputs/Helpers/SimulationPosition}
  {refPath=&UniformRangedVector2D}
  {refPath=&FN_Select_Color_from_Time_Of_Day}
  {refPath=&MakeVector4_FromLinearColor}
  {refPath=&FNTimeOfDayColor}
  {refPath=&MakeCustomFloatFromBool}
  {refPath=&MakeFloatFromVector}
  {refPath=&FN_FrontEndFloatModifier}
  {refPath=&SetBoolByFloatComparison}
  {refPath=&Multiply_Vector}
  {refPath=&ScaleFloatByCurve}
  {refPath=&Divide_Vector}
  {refPath=&Multiply_Vector2D_ByFloat}
  {refPath=&ColorFromCurve}
  {refPath=&FNTimeOfDayScalar}
  {refPath=&FN_PickFloatDependOnPlatform}
  {refPath=&RandomRangeLinearColor}
  {refPath=&SelectFloatFromArray}
  {refPath=&ReturnExecIndex}
  {refPath=&Add_Float}
  {refPath=&RandomRangeVector2D}
  {refPath=&UniformAOrBFloat}
  {refPath=&MakeCustomVectorFromBool}
  {refPath=&Lerp_Float}
  {refPath=&Lerp_Vector2}
  {refPath=&Lerp_Vector}
  {refPath=&ReturnCameraVector}
  {refPath=&NormalizeVector}
  {refPath=&ClampFloat}
  {refPath=&Divide_Float}
  {refPath=&MakeFloat4FromVectorAndFloat}
  {refPath=&FN_Select_Float_from_Time_Of_Day}
  {refPath=&Waveform}
  {refPath=&NormalizeDistanceRange}
  {refPath=&TransformVector}
  {refPath=&AddVectorToPosition}
  {refPath=&MakeVector}
 ]
 usedModules [
  {refPath=&EmitterState}
  {refPath=&SpawnBurst_Instantaneous}
  {refPath=&InitializeParticle}
  {refPath=&CameraOffset}
  {refPath=&ParticleState}
  {refPath=&ScaleMeshSize}
  {refPath=&DynamicMaterialParameters}
  {refPath=&ScaleColor}
  {refPath=&self:Flash001_1.NiagaraScriptSource_0.NiagaraGraph_0.NiagaraNodeAssignment_0.SetVariables_#1}
  {refPath=&self:Flash001_1.NiagaraScriptSource_0.NiagaraGraph_0.NiagaraNodeAssignment_1.SetVariables_#2}
  {refPath=&ScaleSpriteSize}
  {refPath=&SpriteFacingAndAlignment}
  {refPath=&self:Flare_4.NiagaraScriptSource_0.NiagaraGraph_0.NiagaraNodeAssignment_1.SetVariables_#3}
  {refPath=&self:Flare_4.NiagaraScriptSource_0.NiagaraGraph_0.NiagaraNodeAssignment_0.SetVariables_#4}
  {refPath=&FN_Light_Attributes}
  {refPath=&SphereLocation}
  {refPath=&AddVelocityFromPoint}
  {refPath=&CurlNoiseForce}
  {refPath=&Drag}
  {refPath=&GravityForce}
  {refPath=&SolveForcesAndVelocity}
  {refPath=&SpriteRotationRate}
  {refPath=&ShapeLocation}
  {refPath=&AddVelocity}
  {refPath=&self:NE_Debris_02_0.NiagaraScriptSource_0.NiagaraGraph_0.NiagaraNodeAssignment_0.SetVariables_#5}
  {refPath=&InitialMeshOrientation}
  {refPath=&UpdateMeshOrientation}
  {refPath=&SpawnRate}
  {refPath=&SpawnParticlesFromOtherEmitter}
  {refPath=&SampleParticlesFromOtherEmitter}
  {refPath=&ScaleRibbonWidth}
  {refPath=&self:ReadRibbonSourceParticleAge_07}
  {refPath=/Niagara/Modules/Spawn/Initialization/InitializeParticle}
  {refPath=&SubUVAnimation}
  {refPath=&self:HandGlow_0.NiagaraScriptSource_0.NiagaraGraph_0.NiagaraNodeAssignment_1.SetVariables_#6}
  {refPath=&self:HandGlow_0.NiagaraScriptSource_0.NiagaraGraph_0.NiagaraNodeAssignment_3.SetVariables_#7}
  {refPath=&self:HandGlow_0.NiagaraScriptSource_0.NiagaraGraph_0.NiagaraNodeAssignment_0.SetVariables_#8}
  {refPath=&self:HandGlow_0.NiagaraScriptSource_0.NiagaraGraph_0.NiagaraNodeAssignment_2.SetVariables_#9}
  {refPath=&self:HandGlow_0.NiagaraScriptSource_0.NiagaraGraph_0.NiagaraNodeAssignment_4.SetVariables_#10}
  {refPath=&self:HandGlow_0.NiagaraScriptSource_0.NiagaraGraph_0.NiagaraNodeAssignment_5.SetVariables_#11}
  {refPath=&self:IceBlast_0.NiagaraScriptSource_0.NiagaraGraph_0.NiagaraNodeAssignment_6.SetVariables_#12}
  {refPath=&self:IceBlast_0.NiagaraScriptSource_0.NiagaraGraph_0.NiagaraNodeAssignment_0.SetVariables_#13}
  {refPath=&self:IceBlast_0.NiagaraScriptSource_0.NiagaraGraph_0.NiagaraNodeAssignment_1.SetVariables_#14}
  {refPath=&self:IceSpikes_3.NiagaraScriptSource_0.NiagaraGraph_0.NiagaraNodeAssignment_2.SetVariables_#15}
  {refPath=&InheritVelocity}
  {refPath=&AccelerationForce}
  {refPath=&InheritSourceMovement}
  {refPath=&self:SnowflakesCPU_5.NiagaraScriptSource_0.NiagaraGraph_0.NiagaraNodeAssignment_0.SetVariables_#16}
  {refPath=&self:sparkles_0.NiagaraScriptSource_0.NiagaraGraph_0.NiagaraNodeAssignment_0.SetVariables_#17}
  {refPath=&SystemState}
 ]
 usedRenderers [
  {refPath=&NiagaraSpriteRendererProperties}
  {refPath=&NiagaraMeshRendererProperties}
  {refPath=&NiagaraLightRendererProperties}
  {refPath=&NiagaraRibbonRendererProperties}
 ]
}
emitters {
 Flare {
  bEnabled=false
  emitterSpawnScript {scriptName=EmitterSpawnScript, modules []}
  emitterUpdateScript {
   scriptName=EmitterUpdateScript
   modules [
    {
     bIsSetParametersModule=false
     enabled=true
     moduleName=EmitterState
     inputs [
      Life Cycle Mode {
       bIsDynamic=false
       bIsEditable=true
       bIsStaticSwitch=true
       bIsVisible=true
       type {classStructOrEnum {refPath=/Niagara/Enums/ENiagaraEmitterLifeCycleMode}}
      }
      Inactive Response {
       bIsDynamic=false
       bIsEditable=true
       bIsStaticSwitch=true
       bIsVisible=true
       type {classStructOrEnum {refPath=/Niagara/Enums/ENiagaraInactiveMode}}
      }
      Loop Behavior {
       bIsDynamic=false
       bIsEditable=true
       bIsStaticSwitch=true
       bIsVisible=true
       type {classStructOrEnum {refPath=/Niagara/Enums/ENiagara_EmitterStateOptions}}
      }
      Loop Duration Mode {
       bIsDynamic=false
       bIsEditable=true
       bIsStaticSwitch=true
       bIsVisible=true
       type {classStructOrEnum {refPath=/Niagara/Enums/ENiagara_InfiniteLoopDuration}}
      }
      Loop Count {
       bIsDynamic=false
       bIsEditable=false
       bIsStaticSwitch=false
       bIsVisible=false
       type {classStructOrEnum {refPath=/Script/Niagara.NiagaraInt32}}
      }
      Loop Duration {bIsDynamic=false, bIsEditable=true, bIsStaticSwitch=false, bIsVisible=true, type {classStructOrEnum {refPath=/Script/Niagara.NiagaraFloat}}}
      Recalculate Duration Each Loop {bIsDynamic=false, bIsEditable=false, bIsStaticSwitch=false, bIsVisible=false, type {classStructOrEnum {refPath=/Script/Niagara.NiagaraBool}}}
      UseLoopDelay {bIsDynamic=false, bIsEditable=true, bIsStaticSwitch=false, bIsVisible=true, type {classStructOrEnum {refPath=&NiagaraBool}}}
      Loop Delay {bIsDynamic=false, bIsEditable=true, bIsStaticSwitch=false, bIsVisible=true, type {classStructOrEnum {refPath=&NiagaraFloat}}}
      Delay First Loop Only {bIsDynamic=false, bIsEditable=false, bIsStaticSwitch=false, bIsVisible=false, type {classStructOrEnum {refPath=&NiagaraBool}}}
      Scalability Mode {
       bIsDynamic=false
       bIsEditable=true
       bIsStaticSwitch=true
       bIsVisible=true
       type {classStructOrEnum {refPath=/Niagara/Enums/ENiagaraEmitterScalabilityMode_Limited}}
      }
      Enable Distance Culling {bIsDynamic=false, bIsEditable=false, bIsStaticSwitch=true, bIsVisible=false, type {classStructOrEnum {refPath=&NiagaraBool}}}
      MinDistance {
       bIsDynamic=false
       bIsEditable=false
       bIsStaticSwitch=false
       bIsVisible=false
       type {classStructOrEnum {refPath=&NiagaraFloat}}
      }
      Min Distance Response {
       bIsDynamic=false
       bIsEditable=false
       bIsStaticSwitch=false
       bIsVisible=false
       type {classStructOrEnum {refPath=/Script/Niagara.ENiagaraExecutionStateManagement}}
      }
      MaxDistance {
       bIsDynamic=false
       bIsEditable=false
       bIsStaticSwitch=false
       bIsVisible=false
       type {classStructOrEnum {refPath=&NiagaraFloat}}
      }
      Max Distance Response {
       bIsDynamic=false
       bIsEditable=false
       bIsStaticSwitch=false
       bIsVisible=false
       type {classStructOrEnum {refPath=&ENiagaraExecutionStateManagement}}
      }
      Scale Spawn Count {bIsDynamic=false, bIsEditable=false, bIsStaticSwitch=true, bIsVisible=false, type {classStructOrEnum {refPath=&NiagaraBool}}}
      Spawn Count Scale {
       bIsDynamic=false
       bIsEditable=false
       bIsStaticSwitch=false
       bIsVisible=false
       type {classStructOrEnum {refPath=&NiagaraFloat}}
      }
      Spawn Count Scale By Distance {
       bIsDynamic=false
       bIsEditable=false
       bIsStaticSwitch=false
       bIsVisible=false
       type {classStructOrEnum {refPath=&NiagaraDataInterfaceCurve}}
      }
      Enable Visibility Culling {bIsDynamic=false, bIsEditable=false, bIsStaticSwitch=true, bIsVisible=false, type {classStructOrEnum {refPath=&NiagaraBool}}}
      Visibility Cull Response {
       bIsDynamic=false
       bIsEditable=false
       bIsStaticSwitch=false
       bIsVisible=false
       type {classStructOrEnum {refPath=&ENiagaraExecutionStateManagement}}
      }
      Visibility Cull Delay {
       bIsDynamic=false
       bIsEditable=false
       bIsStaticSwitch=false
       bIsVisible=false
       type {classStructOrEnum {refPath=&NiagaraFloat}}
      }
      Reset Age On Awaken {bIsDynamic=false, bIsEditable=false, bIsStaticSwitch=false, bIsVisible=false, type {classStructOrEnum {refPath=&NiagaraBool}}}
     ]
     moduleScript {refPath=&EmitterState}
    }
    {
     bIsSetParametersModule=false
     enabled=true
     moduleName=SpawnBurst_Instantaneous
     inputs [
      Spawn Count {bIsDynamic=false, bIsEditable=true, bIsStaticSwitch=false, bIsVisible=true, type {classStructOrEnum {refPath=&NiagaraInt32}}}
      Spawn Time {bIsDynamic=false, bIsEditable=true, bIsStaticSwitch=false, bIsVisible=true, type {classStructOrEnum {refPath=&NiagaraFloat}}}
      Use Spawn Probability {bIsDynamic=false, bIsEditable=true, bIsStaticSwitch=false, bIsVisible=true, type {classStructOrEnum {refPath=&NiagaraBool}}}
      Spawn Probability {bIsDynamic=false, bIsEditable=false, bIsStaticSwitch=false, bIsVisible=true, type {classStructOrEnum {refPath=&NiagaraFloat}}}
      UseLoopCountLimit {bIsDynamic=false, bIsEditable=true, bIsStaticSwitch=false, bIsVisible=true, type {classStructOrEnum {refPath=&NiagaraBool}}}
      Loop Count Limit {bIsDynamic=false, bIsEditable=true, bIsStaticSwitch=false, bIsVisible=true, type {classStructOrEnum {refPath=&NiagaraInt32}}}
      Spawn Group {bIsDynamic=false, bIsEditable=true, bIsStaticSwitch=false, bIsVisible=true, type {classStructOrEnum {refPath=&NiagaraInt32}}}
      Age {bIsDynamic=false, bIsEditable=true, bIsStaticSwitch=false, bIsVisible=true, type {classStructOrEnum {refPath=&NiagaraFloat}}}
     ]
     moduleScript {refPath=&SpawnBurst_Instantaneous}
    }
   ]
  }
  particleSpawnScript {
   scriptName=ParticleSpawnScript
   modules [
    {
     bIsSetParametersModule=false
     enabled=true
     moduleName=InitializeParticle
     inputs [
      Lifetime Mode {
       bIsDynamic=false
       bIsEditable=true
       bIsStaticSwitch=true
       bIsVisible=true
       type {classStructOrEnum {refPath=/Niagara/Enums/ENiagara_LifetimeMode}}
      }
      Lifetime Min {bIsDynamic=false, bIsEditable=true, bIsStaticSwitch=false, bIsVisible=true, type {classStructOrEnum {refPath=&NiagaraFloat}}}
      Lifetime {
       bIsDynamic=false
       bIsEditable=false
       bIsStaticSwitch=false
       bIsVisible=false
       type {classStructOrEnum {refPath=&NiagaraFloat}}
      }
      Lifetime Max {bIsDynamic=false, bIsEditable=true, bIsStaticSwitch=false, bIsVisible=true, type {classStructOrEnum {refPath=&NiagaraFloat}}}
      Lifetime Randomness Mode {
       bIsDynamic=false
       bIsEditable=true
       bIsStaticSwitch=true
       bIsVisible=true
       type {classStructOrEnum {refPath=/Niagara/Enums/ENiagaraRandomnessMode}}
      }
      Color Mode {
       bIsDynamic=false
       bIsEditable=true
       bIsStaticSwitch=true
       bIsVisible=true
       type {classStructOrEnum {refPath=/Niagara/Enums/ENiagara_ColorInitializationMode}}
      }
      Color Minimum {
       bIsDynamic=false
       bIsEditable=false
       bIsStaticSwitch=false
       bIsVisible=false
       type {classStructOrEnum {refPath=/Script/CoreUObject.LinearColor}}
      }
      Color {
       bIsDynamic=false
       bIsEditable=true
       bIsStaticSwitch=false
       bIsVisible=true
       type {classStructOrEnum {refPath=&LinearColor}}
      }
      AdjustHue {bIsDynamic=false, bIsEditable=false, bIsStaticSwitch=false, bIsVisible=false, type {classStructOrEnum {refPath=&NiagaraBool}}}
      Hue Shift Range {
       bIsDynamic=false
       bIsEditable=false
       bIsStaticSwitch=false
       bIsVisible=false
       type {classStructOrEnum {refPath=/Script/CoreUObject.Vector2f}}
      }
      Color Maximum {
       bIsDynamic=false
       bIsEditable=false
       bIsStaticSwitch=false
       bIsVisible=false
       type {classStructOrEnum {refPath=&LinearColor}}
      }
      AdjustSaturation {bIsDynamic=false, bIsEditable=false, bIsStaticSwitch=false, bIsVisible=false, type {classStructOrEnum {refPath=&NiagaraBool}}}
      Saturation Range {
       bIsDynamic=false
       bIsEditable=false
       bIsStaticSwitch=false
       bIsVisible=false
       type {classStructOrEnum {refPath=&Vector2f}}
      }
      AdjustValue {bIsDynamic=false, bIsEditable=false, bIsStaticSwitch=false, bIsVisible=false, type {classStructOrEnum {refPath=&NiagaraBool}}}
      Value Range {
       bIsDynamic=false
       bIsEditable=false
       bIsStaticSwitch=false
       bIsVisible=false
       type {classStructOrEnum {refPath=&Vector2f}}
      }
      AdjustAlpha {bIsDynamic=false, bIsEditable=false, bIsStaticSwitch=false, bIsVisible=false, type {classStructOrEnum {refPath=&NiagaraBool}}}
      Alpha Scale Range {
       bIsDynamic=false
       bIsEditable=false
       bIsStaticSwitch=false
       bIsVisible=false
       type {classStructOrEnum {refPath=&Vector2f}}
      }
      Color Channel Mode {
       bIsDynamic=false
       bIsEditable=false
       bIsStaticSwitch=true
       bIsVisible=false
       type {classStructOrEnum {refPath=/Niagara/Enums/ENiagaraChannelCorrelation}}
      }
      Color Randomness Mode {
       bIsDynamic=false
       bIsEditable=false
       bIsStaticSwitch=true
       bIsVisible=false
       type {classStructOrEnum {refPath=&ENiagaraRandomnessMode}}
      }
      Position Mode {
       bIsDynamic=false
       bIsEditable=true
       bIsStaticSwitch=true
       bIsVisible=true
       type {classStructOrEnum {refPath=/Niagara/Enums/ENiagara_PositionInitializationMode}}
      }
      Position {
       bIsDynamic=false
       bIsEditable=false
       bIsStaticSwitch=false
       bIsVisible=false
       type {classStructOrEnum {refPath=/Script/Niagara.NiagaraPosition}}
      }
      UsePositionOffset {bIsDynamic=false, bIsEditable=true, bIsStaticSwitch=false, bIsVisible=true, type {classStructOrEnum {refPath=&NiagaraBool}}}
      Position Offset {bIsDynamic=false, bIsEditable=true, bIsStaticSwitch=false, bIsVisible=true, type {classStructOrEnum {refPath=/Script/CoreUObject.Vector3f}}}
      Position Offset Coordinate Space {
       bIsDynamic=false
       bIsEditable=true
       bIsStaticSwitch=false
       bIsVisible=true
       type {classStructOrEnum {refPath=/Script/Niagara.ENiagaraCoordinateSpace}}
      }
      Mass Mode {
       bIsDynamic=false
       bIsEditable=true
       bIsStaticSwitch=true
       bIsVisible=true
       type {classStructOrEnum {refPath=/Niagara/Enums/ENiagara_MassInitializationMode}}
      }
      Mass Min {
       bIsDynamic=false
       bIsEditable=false
       bIsStaticSwitch=false
       bIsVisible=false
       type {classStructOrEnum {refPath=&NiagaraFloat}}
      }
      Mass {
       bIsDynamic=false
       bIsEditable=false
       bIsStaticSwitch=false
       bIsVisible=false
       type {classStructOrEnum {refPath=&NiagaraFloat}}
      }
      Mass Max {
       bIsDynamic=false
       bIsEditable=false
       bIsStaticSwitch=false
       bIsVisible=false
       type {classStructOrEnum {refPath=&NiagaraFloat}}
      }
      Mass Randomness Mode {
       bIsDynamic=false
       bIsEditable=false
       bIsStaticSwitch=true
       bIsVisible=false
       type {classStructOrEnum {refPath=&ENiagaraRandomnessMode}}
      }
      Write Material Random {bIsDynamic=false, bIsEditable=true, bIsStaticSwitch=false, bIsVisible=true, type {classStructOrEnum {refPath=&NiagaraBool}}}
      Material Random {bIsDynamic=false, bIsEditable=true, bIsStaticSwitch=false, bIsVisible=true, type {classStructOrEnum {refPath=&NiagaraFloat}}}
      Sprite Size Mode {
       bIsDynamic=false
       bIsEditable=true
       bIsStaticSwitch=true
       bIsVisible=true
       type {classStructOrEnum {refPath=/Niagara/Enums/ENiagara_SizeScaleMode}}
      }
      Sprite Size {
       bIsDynamic=false
       bIsEditable=false
       bIsStaticSwitch=false
       bIsVisible=false
       type {classStructOrEnum {refPath=&Vector2f}}
      }
      Uniform Sprite Size {
       bIsDynamic=false
       bIsEditable=false
       bIsStaticSwitch=false
       bIsVisible=false
       type {classStructOrEnum {refPath=&NiagaraFloat}}
      }
      Uniform Sprite Size Min {bIsDynamic=false, bIsEditable=true, bIsStaticSwitch=false, bIsVisible=true, type {classStructOrEnum {refPath=&NiagaraFloat}}}
      Sprite Size Min {
       bIsDynamic=false
       bIsEditable=false
       bIsStaticSwitch=false
       bIsVisible=false
       type {classStructOrEnum {refPath=&Vector2f}}
      }
      Sprite Size Max {
       bIsDynamic=false
       bIsEditable=false
       bIsStaticSwitch=false
       bIsVisible=false
       type {classStructOrEnum {refPath=&Vector2f}}
      }
      Uniform Sprite Size Max {bIsDynamic=false, bIsEditable=true, bIsStaticSwitch=false, bIsVisible=true, type {classStructOrEnum {refPath=&NiagaraFloat}}}
      Sprite Size Randomness Mode {
       bIsDynamic=false
       bIsEditable=true
       bIsStaticSwitch=true
       bIsVisible=true
       type {classStructOrEnum {refPath=&ENiagaraRandomnessMode}}
      }
      Sprite Rotation Mode {
       bIsDynamic=false
       bIsEditable=true
       bIsStaticSwitch=true
       bIsVisible=true
       type {classStructOrEnum {refPath=/Niagara/Enums/ENiagara_SpriteRotationMode}}
      }
      Sprite Rotation Angle Min {
       bIsDynamic=false
       bIsEditable=false
       bIsStaticSwitch=false
       bIsVisible=false
       type {classStructOrEnum {refPath=&NiagaraFloat}}
      }
      Sprite Rotation Angle {
       bIsDynamic=false
       bIsEditable=false
       bIsStaticSwitch=false
       bIsVisible=false
       type {classStructOrEnum {refPath=&NiagaraFloat}}
      }
      Sprite Rotation Angle Max {
       bIsDynamic=false
       bIsEditable=false
       bIsStaticSwitch=false
       bIsVisible=false
       type {classStructOrEnum {refPath=&NiagaraFloat}}
      }
      Sprite Rotation Randomness Mode {
       bIsDynamic=false
       bIsEditable=false
       bIsStaticSwitch=true
       bIsVisible=false
       type {classStructOrEnum {refPath=&ENiagaraRandomnessMode}}
      }
      Sprite UV Mode {
       bIsDynamic=false
       bIsEditable=true
       bIsStaticSwitch=true
       bIsVisible=true
       type {classStructOrEnum {refPath=/Niagara/Enums/ENiagara_UVFlippingMode}}
      }
      Sprite UV Randomness Mode {
       bIsDynamic=false
       bIsEditable=false
       bIsStaticSwitch=true
       bIsVisible=false
       type {classStructOrEnum {refPath=&ENiagaraRandomnessMode}}
      }
      Sprite UV Scale {
       bIsDynamic=false
       bIsEditable=false
       bIsStaticSwitch=false
       bIsVisible=false
       type {classStructOrEnum {refPath=&Vector2f}}
      }
      Mesh Scale Mode {
       bIsDynamic=false
       bIsEditable=true
       bIsStaticSwitch=true
       bIsVisible=true
       type {classStructOrEnum {refPath=&ENiagara_SizeScaleMode}}
      }
      Mesh Uniform Scale Min {
       bIsDynamic=false
       bIsEditable=false
       bIsStaticSwitch=false
       bIsVisible=false
       type {classStructOrEnum {refPath=&NiagaraFloat}}
      }
      Mesh Uniform Scale {
       bIsDynamic=false
       bIsEditable=false
       bIsStaticSwitch=false
       bIsVisible=false
       type {classStructOrEnum {refPath=&NiagaraFloat}}
      }
      Mesh Scale Min {
       bIsDynamic=false
       bIsEditable=false
       bIsStaticSwitch=false
       bIsVisible=false
       type {classStructOrEnum {refPath=&Vector3f}}
      }
      Mesh Scale {
       bIsDynamic=false
       bIsEditable=false
       bIsStaticSwitch=false
       bIsVisible=false
       type {classStructOrEnum {refPath=&Vector3f}}
      }
      Mesh Scale Max {
       bIsDynamic=false
       bIsEditable=false
       bIsStaticSwitch=false
       bIsVisible=false
       type {classStructOrEnum {refPath=&Vector3f}}
      }
      Mesh Uniform Scale Max {
       bIsDynamic=false
       bIsEditable=false
       bIsStaticSwitch=false
       bIsVisible=false
       type {classStructOrEnum {refPath=&NiagaraFloat}}
      }
      Mesh Scale Randomness Mode {
       bIsDynamic=false
       bIsEditable=false
       bIsStaticSwitch=true
       bIsVisible=false
       type {classStructOrEnum {refPath=&ENiagaraRandomnessMode}}
      }
      Mesh Renderer Array Visibility Mode {
       bIsDynamic=false
       bIsEditable=true
       bIsStaticSwitch=true
       bIsVisible=true
       type {classStructOrEnum {refPath=/Niagara/Enums/Ribbons/ENiagara_UnsetDirectSetRandom}}
      }
      Mesh Renderer Mesh Index {
       bIsDynamic=false
       bIsEditable=false
       bIsStaticSwitch=false
       bIsVisible=false
       type {classStructOrEnum {refPath=&NiagaraInt32}}
      }
      Wrap Index By Mesh Count {bIsDynamic=false, bIsEditable=false, bIsStaticSwitch=true, bIsVisible=false, type {classStructOrEnum {refPath=&NiagaraBool}}}
      Mesh Renderer Info {
       bIsDynamic=false
       bIsEditable=false
       bIsStaticSwitch=false
       bIsVisible=false
       type {classStructOrEnum {refPath=&NiagaraDataInterfaceMeshRendererInfo}}
      }
      Mesh Renderer Random Seed {
       bIsDynamic=false
       bIsEditable=false
       bIsStaticSwitch=false
       bIsVisible=false
       type {classStructOrEnum {refPath=&NiagaraInt32}}
      }
      Ribbon Width Mode {
       bIsDynamic=false
       bIsEditable=true
       bIsStaticSwitch=true
       bIsVisible=true
       type {classStructOrEnum {refPath=/Niagara/Enums/Ribbons/ENiagara_UnsetDirectSet}}
      }
      Ribbon Width {
       bIsDynamic=false
       bIsEditable=false
       bIsStaticSwitch=false
       bIsVisible=false
       type {classStructOrEnum {refPath=&NiagaraFloat}}
      }
      Ribbon Facing Mode {
       bIsDynamic=false
       bIsEditable=true
       bIsStaticSwitch=true
       bIsVisible=true
       type {classStructOrEnum {refPath=&ENiagara_UnsetDirectSet}}
      }
      Ribbon Facing Vector {
       bIsDynamic=false
       bIsEditable=false
       bIsStaticSwitch=false
       bIsVisible=false
       type {classStructOrEnum {refPath=&Vector3f}}
      }
      Ribbon Twist Mode {
       bIsDynamic=false
       bIsEditable=true
       bIsStaticSwitch=true
       bIsVisible=true
       type {classStructOrEnum {refPath=&ENiagara_UnsetDirectSet}}
      }
      "Ribbon Twist " {
       bIsDynamic=false
       bIsEditable=false
       bIsStaticSwitch=false
       bIsVisible=false
       type {classStructOrEnum {refPath=&NiagaraFloat}}
      }
     ]
     moduleScript {refPath=/Niagara/Modules/Spawn/Initialization/V2/InitializeParticle}
    }
    {
     bIsSetParametersModule=false
     enabled=true
     moduleName=SpriteFacingAndAlignment
     inputs [
      WriteFacing {bIsDynamic=false, bIsEditable=true, bIsStaticSwitch=false, bIsVisible=true, type {classStructOrEnum {refPath=&NiagaraBool}}}
      Sprite Facing {bIsDynamic=false, bIsEditable=false, bIsStaticSwitch=false, bIsVisible=true, type {classStructOrEnum {refPath=&Vector3f}}}
      Facing Coordinate Space {
       bIsDynamic=false
       bIsEditable=false
       bIsStaticSwitch=false
       bIsVisible=false
       type {classStructOrEnum {refPath=&ENiagaraCoordinateSpace}}
      }
      WriteAlignment {bIsDynamic=false, bIsEditable=true, bIsStaticSwitch=false, bIsVisible=true, type {classStructOrEnum {refPath=&NiagaraBool}}}
      Sprite Alignment {bIsDynamic=false, bIsEditable=true, bIsStaticSwitch=false, bIsVisible=true, type {classStructOrEnum {refPath=&Vector3f}}}
      Alignment Coordinate Space {
       bIsDynamic=false
       bIsEditable=true
       bIsStaticSwitch=false
       bIsVisible=true
       type {classStructOrEnum {refPath=&ENiagaraCoordinateSpace}}
      }
     ]
     moduleScript {refPath=&SpriteFacingAndAlignment}
    }
    {
     bIsSetParametersModule=false
     enabled=true
     moduleName=CameraOffset
     inputs [Camera Offset Amount {bIsDynamic=false, bIsEditable=true, bIsStaticSwitch=false, bIsVisible=true, type {classStructOrEnum {refPath=&NiagaraFloat}}}]
     moduleScript {refPath=&CameraOffset}
    }
    {
     bIsSetParametersModule=false
     enabled=true
     moduleName=DynamicMaterialParameters
     inputs [
      Write Parameter Index 0 {bIsDynamic=false, bIsEditable=true, bIsStaticSwitch=true, bIsVisible=true, type {classStructOrEnum {refPath=&NiagaraBool}}}
      Index 0 Param 1 {
       bIsDynamic=false
       bIsEditable=false
       bIsStaticSwitch=false
       bIsVisible=false
       type {classStructOrEnum {refPath=&NiagaraFloat}}
      }
      Index 0 Param 2 {
       bIsDynamic=false
       bIsEditable=false
       bIsStaticSwitch=false
       bIsVisible=false
       type {classStructOrEnum {refPath=&NiagaraFloat}}
      }
      Index 0 Param 3 {
       bIsDynamic=false
       bIsEditable=false
       bIsStaticSwitch=false
       bIsVisible=false
       type {classStructOrEnum {refPath=&NiagaraFloat}}
      }
      Index 0 Param 4 {
       bIsDynamic=false
       bIsEditable=false
       bIsStaticSwitch=false
       bIsVisible=false
       type {classStructOrEnum {refPath=&NiagaraFloat}}
      }
      Write Parameter Index 1 {bIsDynamic=false, bIsEditable=true, bIsStaticSwitch=true, bIsVisible=true, type {classStructOrEnum {refPath=&NiagaraBool}}}
      Index 1 Param 1 {
       bIsDynamic=false
       bIsEditable=false
       bIsStaticSwitch=false
       bIsVisible=false
       type {classStructOrEnum {refPath=&NiagaraFloat}}
      }
      Index 1 Param 2 {
       bIsDynamic=false
       bIsEditable=false
       bIsStaticSwitch=false
       bIsVisible=false
       type {classStructOrEnum {refPath=&NiagaraFloat}}
      }
      Index 1 Param 3 {
       bIsDynamic=false
       bIsEditable=false
       bIsStaticSwitch=false
       bIsVisible=false
       type {classStructOrEnum {refPath=&NiagaraFloat}}
      }
      Index 1 Param 4 {
       bIsDynamic=false
       bIsEditable=false
       bIsStaticSwitch=false
       bIsVisible=false
       type {classStructOrEnum {refPath=&NiagaraFloat}}
      }
      Write Parameter Index 2 {bIsDynamic=false, bIsEditable=true, bIsStaticSwitch=true, bIsVisible=true, type {classStructOrEnum {refPath=&NiagaraBool}}}
      Index 2 Param 1 {
       bIsDynamic=false
       bIsEditable=false
       bIsStaticSwitch=false
       bIsVisible=false
       type {classStructOrEnum {refPath=&NiagaraFloat}}
      }
      Index 2 Param 2 {
       bIsDynamic=false
       bIsEditable=false
       bIsStaticSwitch=false
       bIsVisible=false
       type {classStructOrEnum {refPath=&NiagaraFloat}}
      }
      Index 2 Param 3 {
       bIsDynamic=false
       bIsEditable=false
       bIsStaticSwitch=false
       bIsVisible=false
       type {classStructOrEnum {refPath=&NiagaraFloat}}
      }
      Index 2 Param 4 {
       bIsDynamic=false
       bIsEditable=false
       bIsStaticSwitch=false
       bIsVisible=false
       type {classStructOrEnum {refPath=&NiagaraFloat}}
      }
      Write Parameter Index 3 {bIsDynamic=false, bIsEditable=true, bIsStaticSwitch=true, bIsVisible=true, type {classStructOrEnum {refPath=&NiagaraBool}}}
      Index 3 Param 1 {
       bIsDynamic=false
       bIsEditable=false
       bIsStaticSwitch=false
       bIsVisible=false
       type {classStructOrEnum {refPath=&NiagaraFloat}}
      }
      Index 3 Param 2 {
       bIsDynamic=false
       bIsEditable=false
       bIsStaticSwitch=false
       bIsVisible=false
       type {classStructOrEnum {refPath=&NiagaraFloat}}
      }
      Index 3 Param 3 {
       bIsDynamic=false
       bIsEditable=false
       bIsStaticSwitch=false
       bIsVisible=false
       type {classStructOrEnum {refPath=&NiagaraFloat}}
      }
      Index 3 Param 4 {
       bIsDynamic=false
       bIsEditable=false
       bIsStaticSwitch=false
       bIsVisible=false
       type {classStructOrEnum {refPath=&NiagaraFloat}}
      }
     ]
     moduleScript {refPath=&DynamicMaterialParameters}
    }
   ]
  }
  particleUpdateScript {
   scriptName=ParticleUpdateScript
   modules [
    {
     bIsSetParametersModule=false
     enabled=true
     moduleName=ParticleState
     inputs [
      Kill Particles When Lifetime Has Elapsed {bIsDynamic=false, bIsEditable=true, bIsStaticSwitch=true, bIsVisible=true, type {classStructOrEnum {refPath=&NiagaraBool}}}
      Loop Particles Lifetime {bIsDynamic=false, bIsEditable=false, bIsStaticSwitch=false, bIsVisible=false, type {classStructOrEnum {refPath=&NiagaraBool}}}
      Let Infinitely Lived Particles Die When Emitter Deactivates {bIsDynamic=false, bIsEditable=false, bIsStaticSwitch=false, bIsVisible=false, type {classStructOrEnum {refPath=&NiagaraBool}}}
      Lifetime {bIsDynamic=false, bIsEditable=true, bIsStaticSwitch=false, bIsVisible=true, type {classStructOrEnum {refPath=&NiagaraFloat}}}
      DeltaTime {bIsDynamic=false, bIsEditable=true, bIsStaticSwitch=false, bIsVisible=true, type {classStructOrEnum {refPath=&NiagaraFloat}}}
     ]
     moduleScript {refPath=&ParticleState}
    }
    {
     bIsSetParametersModule=false
     enabled=true
     moduleName=ScaleColor
     inputs [
      Scale Mode {
       bIsDynamic=false
       bIsEditable=true
       bIsStaticSwitch=true
       bIsVisible=true
       type {classStructOrEnum {refPath=/Niagara/Enums/ENiagaraScaleColorMode}}
      }
      Linear Color Curve {
       bIsDynamic=false
       bIsEditable=false
       bIsStaticSwitch=false
       bIsVisible=false
       type {classStructOrEnum {refPath=&NiagaraDataInterfaceColorCurve}}
      }
      ScaleRGB {bIsDynamic=false, bIsEditable=true, bIsStaticSwitch=false, bIsVisible=true, type {classStructOrEnum {refPath=&NiagaraBool}}}
      Scale RGB {bIsDynamic=false, bIsEditable=true, bIsStaticSwitch=false, bIsVisible=true, type {classStructOrEnum {refPath=&Vector3f}}}
      ScaleRGBA {bIsDynamic=false, bIsEditable=false, bIsStaticSwitch=false, bIsVisible=false, type {classStructOrEnum {refPath=&NiagaraBool}}}
      Scale RGBA {
       bIsDynamic=false
       bIsEditable=false
       bIsStaticSwitch=false
       bIsVisible=false
       type {classStructOrEnum {refPath=/Script/CoreUObject.Vector4f}}
      }
      Curve Index {
       bIsDynamic=false
       bIsEditable=false
       bIsStaticSwitch=false
       bIsVisible=false
       type {classStructOrEnum {refPath=&NiagaraFloat}}
      }
      ScaleA {bIsDynamic=false, bIsEditable=true, bIsStaticSwitch=false, bIsVisible=true, type {classStructOrEnum {refPath=&NiagaraBool}}}
      Scale Alpha {bIsDynamic=true, bIsEditable=true, bIsStaticSwitch=false, bIsVisible=true, type {classStructOrEnum {refPath=&NiagaraFloat}}}
      Color Value To Scale {
       bIsDynamic=false
       bIsEditable=true
       bIsStaticSwitch=false
       bIsVisible=true
       type {classStructOrEnum {refPath=&LinearColor}}
      }
     ]
     moduleScript {refPath=&ScaleColor}
    }
    {
     bIsSetParametersModule=false
     enabled=true
     moduleName=ScaleSpriteSize
     inputs [
      Scale Sprite Size Mode {
       bIsDynamic=false
       bIsEditable=true
       bIsStaticSwitch=true
       bIsVisible=true
       type {classStructOrEnum {refPath=/Niagara/Enums/SpriteRenderer/ENiagara_ScaleSpriteSize}}
      }
      Uniform Curve Sprite Scale {
       bIsDynamic=false
       bIsEditable=true
       bIsStaticSwitch=false
       bIsVisible=true
       type {classStructOrEnum {refPath=&NiagaraDataInterfaceCurve}}
      }
      Non-Uniform Curve Sprite Scale {
       bIsDynamic=false
       bIsEditable=false
       bIsStaticSwitch=false
       bIsVisible=false
       type {classStructOrEnum {refPath=&NiagaraDataInterfaceVector2DCurve}}
      }
      Uniform Curve Index {bIsDynamic=false, bIsEditable=true, bIsStaticSwitch=false, bIsVisible=true, type {classStructOrEnum {refPath=&NiagaraFloat}}}
      Non-Uniform Curve Index {
       bIsDynamic=false
       bIsEditable=false
       bIsStaticSwitch=false
       bIsVisible=false
       type {classStructOrEnum {refPath=&NiagaraFloat}}
      }
      Non-Uniform Curve Scale {
       bIsDynamic=false
       bIsEditable=false
       bIsStaticSwitch=false
       bIsVisible=false
       type {classStructOrEnum {refPath=&Vector2f}}
      }
      Uniform Curve Scale {bIsDynamic=false, bIsEditable=true, bIsStaticSwitch=false, bIsVisible=true, type {classStructOrEnum {refPath=&NiagaraFloat}}}
      Scale Factor {bIsDynamic=true, bIsEditable=false, bIsStaticSwitch=false, bIsVisible=false, type {classStructOrEnum {refPath=&Vector2f}}}
      Uniform Scale Factor {
       bIsDynamic=false
       bIsEditable=false
       bIsStaticSwitch=false
       bIsVisible=false
       type {classStructOrEnum {refPath=&NiagaraFloat}}
      }
      Initial Sprite Size {bIsDynamic=false, bIsEditable=true, bIsStaticSwitch=false, bIsVisible=true, type {classStructOrEnum {refPath=&Vector2f}}}
     ]
     moduleScript {refPath=&ScaleSpriteSize}
    }
    {
     bIsSetParametersModule=true
     enabled=true
     moduleName=SetVariables_#3
     inputs [
      Particles.Color Wide {
       bIsDynamic=true
       bIsEditable=true
       bIsStaticSwitch=false
       bIsVisible=true
       type {classStructOrEnum {refPath=&LinearColor}}
      }
      Particles.SpriteSize_Wide {bIsDynamic=true, bIsEditable=true, bIsStaticSwitch=false, bIsVisible=true, type {classStructOrEnum {refPath=&Vector2f}}}
     ]
     moduleScript {refPath=&self:Flare_4.NiagaraScriptSource_0.NiagaraGraph_0.NiagaraNodeAssignment_1.SetVariables_#3}
    }
    {
     bIsSetParametersModule=true
     enabled=true
     moduleName=SetVariables_#4
     inputs [Particles.Position_Light {
       bIsDynamic=true
       bIsEditable=true
       bIsStaticSwitch=false
       bIsVisible=true
       type {classStructOrEnum {refPath=&NiagaraPosition}}
      }]
     moduleScript {refPath=&self:Flare_4.NiagaraScriptSource_0.NiagaraGraph_0.NiagaraNodeAssignment_0.SetVariables_#4}
    }
    {
     bIsSetParametersModule=false
     enabled=true
     moduleName=FN_Light_Attributes
     inputs [
      Enable Particle Light Enable {bIsDynamic=false, bIsEditable=true, bIsStaticSwitch=true, bIsVisible=true, type {classStructOrEnum {refPath=&NiagaraBool}}}
      Light Enabled {bIsDynamic=false, bIsEditable=false, bIsStaticSwitch=false, bIsVisible=false, type {classStructOrEnum {refPath=&NiagaraBool}}}
      Enable Visibility Tag {bIsDynamic=false, bIsEditable=true, bIsStaticSwitch=true, bIsVisible=true, type {classStructOrEnum {refPath=&NiagaraBool}}}
      VisibilityTag {
       bIsDynamic=false
       bIsEditable=false
       bIsStaticSwitch=false
       bIsVisible=false
       type {classStructOrEnum {refPath=&NiagaraInt32}}
      }
      Create Position Attribute {bIsDynamic=false, bIsEditable=true, bIsStaticSwitch=true, bIsVisible=true, type {classStructOrEnum {refPath=&NiagaraBool}}}
      Position {
       bIsDynamic=false
       bIsEditable=false
       bIsStaticSwitch=false
       bIsVisible=false
       type {classStructOrEnum {refPath=&Vector3f}}
      }
      Create Radius Attribute {bIsDynamic=false, bIsEditable=true, bIsStaticSwitch=true, bIsVisible=true, type {classStructOrEnum {refPath=&NiagaraBool}}}
      Radius {bIsDynamic=true, bIsEditable=true, bIsStaticSwitch=false, bIsVisible=true, type {classStructOrEnum {refPath=&NiagaraFloat}}}
      Create Exponent Attribute {bIsDynamic=false, bIsEditable=true, bIsStaticSwitch=true, bIsVisible=true, type {classStructOrEnum {refPath=&NiagaraBool}}}
      Exponent {bIsDynamic=false, bIsEditable=true, bIsStaticSwitch=false, bIsVisible=true, type {classStructOrEnum {refPath=&NiagaraFloat}}}
      Create Color Attribute {bIsDynamic=false, bIsEditable=true, bIsStaticSwitch=true, bIsVisible=true, type {classStructOrEnum {refPath=&NiagaraBool}}}
      Color {
       bIsDynamic=false
       bIsEditable=true
       bIsStaticSwitch=false
       bIsVisible=true
       type {classStructOrEnum {refPath=&LinearColor}}
      }
      Scale Color {bIsDynamic=true, bIsEditable=true, bIsStaticSwitch=false, bIsVisible=true, type {classStructOrEnum {refPath=&NiagaraFloat}}}
     ]
     moduleScript {refPath=&FN_Light_Attributes}
    }
   ]
  }
  rendererClasses [{refPath=&NiagaraSpriteRendererProperties}, {refPath=&NiagaraLightRendererProperties}]
  renderers [
   {rendererIndex=0, rendererClass {refPath=&NiagaraSpriteRendererProperties}}
   {rendererIndex=1, rendererClass {refPath=&NiagaraSpriteRendererProperties}}
   {rendererIndex=2, rendererClass {refPath=&NiagaraLightRendererProperties}}
  ]
 }
 Flash001 {
  bEnabled=false
  emitterSpawnScript {scriptName=EmitterSpawnScript, modules []}
  emitterUpdateScript {
   scriptName=EmitterUpdateScript
   modules [
    {
     bIsSetParametersModule=false
     enabled=true
     moduleName=EmitterState
     inputs [
      Life Cycle Mode {
       bIsDynamic=false
       bIsEditable=true
       bIsStaticSwitch=true
       bIsVisible=true
       type {classStructOrEnum {refPath=&ENiagaraEmitterLifeCycleMode}}
      }
      Inactive Response {
       bIsDynamic=false
       bIsEditable=true
       bIsStaticSwitch=true
       bIsVisible=true
       type {classStructOrEnum {refPath=&ENiagaraInactiveMode}}
      }
      Loop Behavior {
       bIsDynamic=false
       bIsEditable=true
       bIsStaticSwitch=true
       bIsVisible=true
       type {classStructOrEnum {refPath=&ENiagara_EmitterStateOptions}}
      }
      Loop Duration Mode {
       bIsDynamic=false
       bIsEditable=true
       bIsStaticSwitch=true
       bIsVisible=true
       type {classStructOrEnum {refPath=&ENiagara_InfiniteLoopDuration}}
      }
      Loop Count {
       bIsDynamic=false
       bIsEditable=false
       bIsStaticSwitch=false
       bIsVisible=false
       type {classStructOrEnum {refPath=&NiagaraInt32}}
      }
      Loop Duration {bIsDynamic=false, bIsEditable=true, bIsStaticSwitch=false, bIsVisible=true, type {classStructOrEnum {refPath=&NiagaraFloat}}}
      Recalculate Duration Each Loop {bIsDynamic=false, bIsEditable=false, bIsStaticSwitch=false, bIsVisible=false, type {classStructOrEnum {refPath=&NiagaraBool}}}
      UseLoopDelay {bIsDynamic=false, bIsEditable=true, bIsStaticSwitch=false, bIsVisible=true, type {classStructOrEnum {refPath=&NiagaraBool}}}
      Loop Delay {bIsDynamic=false, bIsEditable=true, bIsStaticSwitch=false, bIsVisible=true, type {classStructOrEnum {refPath=&NiagaraFloat}}}
      Delay First Loop Only {bIsDynamic=false, bIsEditable=false, bIsStaticSwitch=false, bIsVisible=false, type {classStructOrEnum {refPath=&NiagaraBool}}}
      Scalability Mode {
       bIsDynamic=false
       bIsEditable=true
       bIsStaticSwitch=true
       bIsVisible=true
       type {classStructOrEnum {refPath=&ENiagaraEmitterScalabilityMode_Limited}}
      }
      Enable Distance Culling {bIsDynamic=false, bIsEditable=false, bIsStaticSwitch=true, bIsVisible=false, type {classStructOrEnum {refPath=&NiagaraBool}}}
      MinDistance {
       bIsDynamic=false
       bIsEditable=false
       bIsStaticSwitch=false
       bIsVisible=false
       type {classStructOrEnum {refPath=&NiagaraFloat}}
      }
      Min Distance Response {
       bIsDynamic=false
       bIsEditable=false
       bIsStaticSwitch=false
       bIsVisible=false
       type {classStructOrEnum {refPath=&ENiagaraExecutionStateManagement}}
      }
      MaxDistance {
       bIsDynamic=false
       bIsEditable=false
       bIsStaticSwitch=false
       bIsVisible=false
       type {classStructOrEnum {refPath=&NiagaraFloat}}
      }
      Max Distance Response {
       bIsDynamic=false
       bIsEditable=false
       bIsStaticSwitch=false
       bIsVisible=false
       type {classStructOrEnum {refPath=&ENiagaraExecutionStateManagement}}
      }
      Scale Spawn Count {bIsDynamic=false, bIsEditable=false, bIsStaticSwitch=true, bIsVisible=false, type {classStructOrEnum {refPath=&NiagaraBool}}}
      Spawn Count Scale {
       bIsDynamic=false
       bIsEditable=false
       bIsStaticSwitch=false
       bIsVisible=false
       type {classStructOrEnum {refPath=&NiagaraFloat}}
      }
      Spawn Count Scale By Distance {
       bIsDynamic=false
       bIsEditable=false
       bIsStaticSwitch=false
       bIsVisible=false
       type {classStructOrEnum {refPath=&NiagaraDataInterfaceCurve}}
      }
      Enable Visibility Culling {bIsDynamic=false, bIsEditable=false, bIsStaticSwitch=true, bIsVisible=false, type {classStructOrEnum {refPath=&NiagaraBool}}}
      Visibility Cull Response {
       bIsDynamic=false
       bIsEditable=false
       bIsStaticSwitch=false
       bIsVisible=false
       type {classStructOrEnum {refPath=&ENiagaraExecutionStateManagement}}
      }
      Visibility Cull Delay {
       bIsDynamic=false
       bIsEditable=false
       bIsStaticSwitch=false
       bIsVisible=false
       type {classStructOrEnum {refPath=&NiagaraFloat}}
      }
      Reset Age On Awaken {bIsDynamic=false, bIsEditable=false, bIsStaticSwitch=false, bIsVisible=false, type {classStructOrEnum {refPath=&NiagaraBool}}}
     ]
     moduleScript {refPath=&EmitterState}
    }
    {
     bIsSetParametersModule=false
     enabled=true
     moduleName=SpawnBurst_Instantaneous
     inputs [
      Spawn Count {bIsDynamic=false, bIsEditable=true, bIsStaticSwitch=false, bIsVisible=true, type {classStructOrEnum {refPath=&NiagaraInt32}}}
      Spawn Time {bIsDynamic=false, bIsEditable=true, bIsStaticSwitch=false, bIsVisible=true, type {classStructOrEnum {refPath=&NiagaraFloat}}}
      Use Spawn Probability {bIsDynamic=false, bIsEditable=true, bIsStaticSwitch=false, bIsVisible=true, type {classStructOrEnum {refPath=&NiagaraBool}}}
      Spawn Probability {bIsDynamic=false, bIsEditable=false, bIsStaticSwitch=false, bIsVisible=true, type {classStructOrEnum {refPath=&NiagaraFloat}}}
      UseLoopCountLimit {bIsDynamic=false, bIsEditable=true, bIsStaticSwitch=false, bIsVisible=true, type {classStructOrEnum {refPath=&NiagaraBool}}}
      Loop Count Limit {bIsDynamic=false, bIsEditable=false, bIsStaticSwitch=false, bIsVisible=true, type {classStructOrEnum {refPath=&NiagaraInt32}}}
      Spawn Group {bIsDynamic=false, bIsEditable=true, bIsStaticSwitch=false, bIsVisible=true, type {classStructOrEnum {refPath=&NiagaraInt32}}}
      Age {bIsDynamic=false, bIsEditable=true, bIsStaticSwitch=false, bIsVisible=true, type {classStructOrEnum {refPath=&NiagaraFloat}}}
     ]
     moduleScript {refPath=&SpawnBurst_Instantaneous}
    }
   ]
  }
  particleSpawnScript {
   scriptName=ParticleSpawnScript
   modules [
    {
     bIsSetParametersModule=false
     enabled=true
     moduleName=InitializeParticle
     inputs [
      Lifetime Mode {
       bIsDynamic=false
       bIsEditable=true
       bIsStaticSwitch=true
       bIsVisible=true
       type {classStructOrEnum {refPath=&ENiagara_LifetimeMode}}
      }
      Lifetime Min {
       bIsDynamic=false
       bIsEditable=false
       bIsStaticSwitch=false
       bIsVisible=false
       type {classStructOrEnum {refPath=&NiagaraFloat}}
      }
      Lifetime {bIsDynamic=false, bIsEditable=true, bIsStaticSwitch=false, bIsVisible=true, type {classStructOrEnum {refPath=&NiagaraFloat}}}
      Lifetime Max {
       bIsDynamic=false
       bIsEditable=false
       bIsStaticSwitch=false
       bIsVisible=false
       type {classStructOrEnum {refPath=&NiagaraFloat}}
      }
      Lifetime Randomness Mode {
       bIsDynamic=false
       bIsEditable=false
       bIsStaticSwitch=true
       bIsVisible=false
       type {classStructOrEnum {refPath=&ENiagaraRandomnessMode}}
      }
      Color Mode {
       bIsDynamic=false
       bIsEditable=true
       bIsStaticSwitch=true
       bIsVisible=true
       type {classStructOrEnum {refPath=&ENiagara_ColorInitializationMode}}
      }
      Color Minimum {
       bIsDynamic=false
       bIsEditable=false
       bIsStaticSwitch=false
       bIsVisible=false
       type {classStructOrEnum {refPath=&LinearColor}}
      }
      Color {
       bIsDynamic=false
       bIsEditable=true
       bIsStaticSwitch=false
       bIsVisible=true
       type {classStructOrEnum {refPath=&LinearColor}}
      }
      AdjustHue {bIsDynamic=false, bIsEditable=false, bIsStaticSwitch=false, bIsVisible=false, type {classStructOrEnum {refPath=&NiagaraBool}}}
      Hue Shift Range {
       bIsDynamic=false
       bIsEditable=false
       bIsStaticSwitch=false
       bIsVisible=false
       type {classStructOrEnum {refPath=&Vector2f}}
      }
      Color Maximum {
       bIsDynamic=false
       bIsEditable=false
       bIsStaticSwitch=false
       bIsVisible=false
       type {classStructOrEnum {refPath=&LinearColor}}
      }
      AdjustSaturation {bIsDynamic=false, bIsEditable=false, bIsStaticSwitch=false, bIsVisible=false, type {classStructOrEnum {refPath=&NiagaraBool}}}
      Saturation Range {
       bIsDynamic=false
       bIsEditable=false
       bIsStaticSwitch=false
       bIsVisible=false
       type {classStructOrEnum {refPath=&Vector2f}}
      }
      AdjustValue {bIsDynamic=false, bIsEditable=false, bIsStaticSwitch=false, bIsVisible=false, type {classStructOrEnum {refPath=&NiagaraBool}}}
      Value Range {
       bIsDynamic=false
       bIsEditable=false
       bIsStaticSwitch=false
       bIsVisible=false
       type {classStructOrEnum {refPath=&Vector2f}}
      }
      AdjustAlpha {bIsDynamic=false, bIsEditable=false, bIsStaticSwitch=false, bIsVisible=false, type {classStructOrEnum {refPath=&NiagaraBool}}}
      Alpha Scale Range {
       bIsDynamic=false
       bIsEditable=false
       bIsStaticSwitch=false
       bIsVisible=false
       type {classStructOrEnum {refPath=&Vector2f}}
      }
      Color Channel Mode {
       bIsDynamic=false
       bIsEditable=false
       bIsStaticSwitch=true
       bIsVisible=false
       type {classStructOrEnum {refPath=&ENiagaraChannelCorrelation}}
      }
      Color Randomness Mode {
       bIsDynamic=false
       bIsEditable=false
       bIsStaticSwitch=true
       bIsVisible=false
       type {classStructOrEnum {refPath=&ENiagaraRandomnessMode}}
      }
      Position Mode {
       bIsDynamic=false
       bIsEditable=true
       bIsStaticSwitch=true
       bIsVisible=true
       type {classStructOrEnum {refPath=&ENiagara_PositionInitializationMode}}
      }
      Position {
       bIsDynamic=false
       bIsEditable=false
       bIsStaticSwitch=false
       bIsVisible=false
       type {classStructOrEnum {refPath=&NiagaraPosition}}
      }
      UsePositionOffset {bIsDynamic=false, bIsEditable=true, bIsStaticSwitch=false, bIsVisible=true, type {classStructOrEnum {refPath=&NiagaraBool}}}
      Position Offset {bIsDynamic=false, bIsEditable=true, bIsStaticSwitch=false, bIsVisible=true, type {classStructOrEnum {refPath=&Vector3f}}}
      Position Offset Coordinate Space {
       bIsDynamic=false
       bIsEditable=true
       bIsStaticSwitch=false
       bIsVisible=true
       type {classStructOrEnum {refPath=&ENiagaraCoordinateSpace}}
      }
      Mass Mode {
       bIsDynamic=false
       bIsEditable=true
       bIsStaticSwitch=true
       bIsVisible=true
       type {classStructOrEnum {refPath=&ENiagara_MassInitializationMode}}
      }
      Mass Min {
       bIsDynamic=false
       bIsEditable=false
       bIsStaticSwitch=false
       bIsVisible=false
       type {classStructOrEnum {refPath=&NiagaraFloat}}
      }
      Mass {
       bIsDynamic=false
       bIsEditable=false
       bIsStaticSwitch=false
       bIsVisible=false
       type {classStructOrEnum {refPath=&NiagaraFloat}}
      }
      Mass Max {
       bIsDynamic=false
       bIsEditable=false
       bIsStaticSwitch=false
       bIsVisible=false
       type {classStructOrEnum {refPath=&NiagaraFloat}}
      }
      Mass Randomness Mode {
       bIsDynamic=false
       bIsEditable=false
       bIsStaticSwitch=true
       bIsVisible=false
       type {classStructOrEnum {refPath=&ENiagaraRandomnessMode}}
      }
      Write Material Random {bIsDynamic=false, bIsEditable=true, bIsStaticSwitch=false, bIsVisible=true, type {classStructOrEnum {refPath=&NiagaraBool}}}
      Material Random {bIsDynamic=false, bIsEditable=true, bIsStaticSwitch=false, bIsVisible=true, type {classStructOrEnum {refPath=&NiagaraFloat}}}
      Sprite Size Mode {
       bIsDynamic=false
       bIsEditable=true
       bIsStaticSwitch=true
       bIsVisible=true
       type {classStructOrEnum {refPath=&ENiagara_SizeScaleMode}}
      }
      Sprite Size {
       bIsDynamic=false
       bIsEditable=false
       bIsStaticSwitch=false
       bIsVisible=false
       type {classStructOrEnum {refPath=&Vector2f}}
      }
      Uniform Sprite Size {bIsDynamic=false, bIsEditable=true, bIsStaticSwitch=false, bIsVisible=true, type {classStructOrEnum {refPath=&NiagaraFloat}}}
      Uniform Sprite Size Min {
       bIsDynamic=false
       bIsEditable=false
       bIsStaticSwitch=false
       bIsVisible=false
       type {classStructOrEnum {refPath=&NiagaraFloat}}
      }
      Sprite Size Min {
       bIsDynamic=false
       bIsEditable=false
       bIsStaticSwitch=false
       bIsVisible=false
       type {classStructOrEnum {refPath=&Vector2f}}
      }
      Sprite Size Max {
       bIsDynamic=false
       bIsEditable=false
       bIsStaticSwitch=false
       bIsVisible=false
       type {classStructOrEnum {refPath=&Vector2f}}
      }
      Uniform Sprite Size Max {
       bIsDynamic=false
       bIsEditable=false
       bIsStaticSwitch=false
       bIsVisible=false
       type {classStructOrEnum {refPath=&NiagaraFloat}}
      }
      Sprite Size Randomness Mode {
       bIsDynamic=false
       bIsEditable=false
       bIsStaticSwitch=true
       bIsVisible=false
       type {classStructOrEnum {refPath=&ENiagaraRandomnessMode}}
      }
      Sprite Rotation Mode {
       bIsDynamic=false
       bIsEditable=true
       bIsStaticSwitch=true
       bIsVisible=true
       type {classStructOrEnum {refPath=&ENiagara_SpriteRotationMode}}
      }
      Sprite Rotation Angle Min {
       bIsDynamic=false
       bIsEditable=false
       bIsStaticSwitch=false
       bIsVisible=false
       type {classStructOrEnum {refPath=&NiagaraFloat}}
      }
      Sprite Rotation Angle {
       bIsDynamic=false
       bIsEditable=false
       bIsStaticSwitch=false
       bIsVisible=false
       type {classStructOrEnum {refPath=&NiagaraFloat}}
      }
      Sprite Rotation Angle Max {
       bIsDynamic=false
       bIsEditable=false
       bIsStaticSwitch=false
       bIsVisible=false
       type {classStructOrEnum {refPath=&NiagaraFloat}}
      }
      Sprite Rotation Randomness Mode {
       bIsDynamic=false
       bIsEditable=false
       bIsStaticSwitch=true
       bIsVisible=false
       type {classStructOrEnum {refPath=&ENiagaraRandomnessMode}}
      }
      Sprite UV Mode {
       bIsDynamic=false
       bIsEditable=true
       bIsStaticSwitch=true
       bIsVisible=true
       type {classStructOrEnum {refPath=&ENiagara_UVFlippingMode}}
      }
      Sprite UV Randomness Mode {
       bIsDynamic=false
       bIsEditable=false
       bIsStaticSwitch=true
       bIsVisible=false
       type {classStructOrEnum {refPath=&ENiagaraRandomnessMode}}
      }
      Sprite UV Scale {
       bIsDynamic=false
       bIsEditable=false
       bIsStaticSwitch=false
       bIsVisible=false
       type {classStructOrEnum {refPath=&Vector2f}}
      }
      Mesh Scale Mode {
       bIsDynamic=false
       bIsEditable=true
       bIsStaticSwitch=true
       bIsVisible=true
       type {classStructOrEnum {refPath=&ENiagara_SizeScaleMode}}
      }
      Mesh Uniform Scale Min {
       bIsDynamic=false
       bIsEditable=false
       bIsStaticSwitch=false
       bIsVisible=false
       type {classStructOrEnum {refPath=&NiagaraFloat}}
      }
      Mesh Uniform Scale {bIsDynamic=false, bIsEditable=true, bIsStaticSwitch=false, bIsVisible=true, type {classStructOrEnum {refPath=&NiagaraFloat}}}
      Mesh Scale Min {
       bIsDynamic=false
       bIsEditable=false
       bIsStaticSwitch=false
       bIsVisible=false
       type {classStructOrEnum {refPath=&Vector3f}}
      }
      Mesh Scale {
       bIsDynamic=false
       bIsEditable=false
       bIsStaticSwitch=false
       bIsVisible=false
       type {classStructOrEnum {refPath=&Vector3f}}
      }
      Mesh Scale Max {
       bIsDynamic=false
       bIsEditable=false
       bIsStaticSwitch=false
       bIsVisible=false
       type {classStructOrEnum {refPath=&Vector3f}}
      }
      Mesh Uniform Scale Max {
       bIsDynamic=false
       bIsEditable=false
       bIsStaticSwitch=false
       bIsVisible=false
       type {classStructOrEnum {refPath=&NiagaraFloat}}
      }
      Mesh Scale Randomness Mode {
       bIsDynamic=false
       bIsEditable=false
       bIsStaticSwitch=true
       bIsVisible=false
       type {classStructOrEnum {refPath=&ENiagaraRandomnessMode}}
      }
      Mesh Renderer Array Visibility Mode {
       bIsDynamic=false
       bIsEditable=true
       bIsStaticSwitch=true
       bIsVisible=true
       type {classStructOrEnum {refPath=&ENiagara_UnsetDirectSetRandom}}
      }
      Mesh Renderer Mesh Index {
       bIsDynamic=false
       bIsEditable=false
       bIsStaticSwitch=false
       bIsVisible=false
       type {classStructOrEnum {refPath=&NiagaraInt32}}
      }
      Wrap Index By Mesh Count {bIsDynamic=false, bIsEditable=false, bIsStaticSwitch=true, bIsVisible=false, type {classStructOrEnum {refPath=&NiagaraBool}}}
      Mesh Renderer Info {
       bIsDynamic=false
       bIsEditable=false
       bIsStaticSwitch=false
       bIsVisible=false
       type {classStructOrEnum {refPath=&NiagaraDataInterfaceMeshRendererInfo}}
      }
      Mesh Renderer Random Seed {
       bIsDynamic=false
       bIsEditable=false
       bIsStaticSwitch=false
       bIsVisible=false
       type {classStructOrEnum {refPath=&NiagaraInt32}}
      }
      Ribbon Width Mode {
       bIsDynamic=false
       bIsEditable=true
       bIsStaticSwitch=true
       bIsVisible=true
       type {classStructOrEnum {refPath=&ENiagara_UnsetDirectSet}}
      }
      Ribbon Width {
       bIsDynamic=false
       bIsEditable=false
       bIsStaticSwitch=false
       bIsVisible=false
       type {classStructOrEnum {refPath=&NiagaraFloat}}
      }
      Ribbon Facing Mode {
       bIsDynamic=false
       bIsEditable=true
       bIsStaticSwitch=true
       bIsVisible=true
       type {classStructOrEnum {refPath=&ENiagara_UnsetDirectSet}}
      }
      Ribbon Facing Vector {
       bIsDynamic=false
       bIsEditable=false
       bIsStaticSwitch=false
       bIsVisible=false
       type {classStructOrEnum {refPath=&Vector3f}}
      }
      Ribbon Twist Mode {
       bIsDynamic=false
       bIsEditable=true
       bIsStaticSwitch=true
       bIsVisible=true
       type {classStructOrEnum {refPath=&ENiagara_UnsetDirectSet}}
      }
      "Ribbon Twist " {
       bIsDynamic=false
       bIsEditable=false
       bIsStaticSwitch=false
       bIsVisible=false
       type {classStructOrEnum {refPath=&NiagaraFloat}}
      }
     ]
     moduleScript {refPath=&InitializeParticle}
    }
    {
     bIsSetParametersModule=false
     enabled=true
     moduleName=CameraOffset
     inputs [Camera Offset Amount {bIsDynamic=false, bIsEditable=true, bIsStaticSwitch=false, bIsVisible=true, type {classStructOrEnum {refPath=&NiagaraFloat}}}]
     moduleScript {refPath=&CameraOffset}
    }
   ]
  }
  particleUpdateScript {
   scriptName=ParticleUpdateScript
   modules [
    {
     bIsSetParametersModule=false
     enabled=true
     moduleName=ParticleState
     inputs [
      Kill Particles When Lifetime Has Elapsed {bIsDynamic=false, bIsEditable=true, bIsStaticSwitch=true, bIsVisible=true, type {classStructOrEnum {refPath=&NiagaraBool}}}
      Loop Particles Lifetime {bIsDynamic=false, bIsEditable=false, bIsStaticSwitch=false, bIsVisible=false, type {classStructOrEnum {refPath=&NiagaraBool}}}
      Let Infinitely Lived Particles Die When Emitter Deactivates {bIsDynamic=false, bIsEditable=false, bIsStaticSwitch=false, bIsVisible=false, type {classStructOrEnum {refPath=&NiagaraBool}}}
      Lifetime {bIsDynamic=false, bIsEditable=true, bIsStaticSwitch=false, bIsVisible=true, type {classStructOrEnum {refPath=&NiagaraFloat}}}
      DeltaTime {bIsDynamic=false, bIsEditable=true, bIsStaticSwitch=false, bIsVisible=true, type {classStructOrEnum {refPath=&NiagaraFloat}}}
     ]
     moduleScript {refPath=&ParticleState}
    }
    {
     bIsSetParametersModule=false
     enabled=true
     moduleName=ScaleMeshSize
     inputs [
      Scale Factor {bIsDynamic=true, bIsEditable=true, bIsStaticSwitch=false, bIsVisible=true, type {classStructOrEnum {refPath=&Vector3f}}}
      Initial Mesh Scale {bIsDynamic=false, bIsEditable=true, bIsStaticSwitch=false, bIsVisible=true, type {classStructOrEnum {refPath=&Vector3f}}}
     ]
     moduleScript {refPath=&ScaleMeshSize}
    }
    {
     bIsSetParametersModule=false
     enabled=true
     moduleName=DynamicMaterialParameters
     inputs [
      Write Parameter Index 0 {bIsDynamic=false, bIsEditable=true, bIsStaticSwitch=true, bIsVisible=true, type {classStructOrEnum {refPath=&NiagaraBool}}}
      Module.Index0Param0Write {bIsDynamic=false, bIsEditable=true, bIsStaticSwitch=true, bIsVisible=true, type {classStructOrEnum {refPath=&NiagaraBool}}}
      Index 0 Param 1 {bIsDynamic=false, bIsEditable=true, bIsStaticSwitch=false, bIsVisible=true, type {classStructOrEnum {refPath=&NiagaraFloat}}}
      Module.Index0Param1Write {bIsDynamic=false, bIsEditable=true, bIsStaticSwitch=true, bIsVisible=true, type {classStructOrEnum {refPath=&NiagaraBool}}}
      Index 0 Param 2 {bIsDynamic=false, bIsEditable=true, bIsStaticSwitch=false, bIsVisible=true, type {classStructOrEnum {refPath=&NiagaraFloat}}}
      Module.Index0Param2Write {bIsDynamic=false, bIsEditable=true, bIsStaticSwitch=true, bIsVisible=true, type {classStructOrEnum {refPath=&NiagaraBool}}}
      Index 0 Param 3 {bIsDynamic=false, bIsEditable=true, bIsStaticSwitch=false, bIsVisible=true, type {classStructOrEnum {refPath=&NiagaraFloat}}}
      Module.Index0Param3Write {bIsDynamic=false, bIsEditable=true, bIsStaticSwitch=true, bIsVisible=true, type {classStructOrEnum {refPath=&NiagaraBool}}}
      Index 0 Param 4 {bIsDynamic=false, bIsEditable=true, bIsStaticSwitch=false, bIsVisible=true, type {classStructOrEnum {refPath=&NiagaraFloat}}}
      Write Parameter Index 1 {bIsDynamic=false, bIsEditable=true, bIsStaticSwitch=true, bIsVisible=true, type {classStructOrEnum {refPath=&NiagaraBool}}}
      Module.Index1Param0Write {bIsDynamic=false, bIsEditable=true, bIsStaticSwitch=true, bIsVisible=true, type {classStructOrEnum {refPath=&NiagaraBool}}}
      Index 1 Param 1 {bIsDynamic=false, bIsEditable=true, bIsStaticSwitch=false, bIsVisible=true, type {classStructOrEnum {refPath=&NiagaraFloat}}}
      Module.Index1Param1Write {bIsDynamic=false, bIsEditable=true, bIsStaticSwitch=true, bIsVisible=true, type {classStructOrEnum {refPath=&NiagaraBool}}}
      Index 1 Param 2 {bIsDynamic=false, bIsEditable=false, bIsStaticSwitch=false, bIsVisible=true, type {classStructOrEnum {refPath=&NiagaraFloat}}}
      Module.Index1Param2Write {bIsDynamic=false, bIsEditable=true, bIsStaticSwitch=true, bIsVisible=true, type {classStructOrEnum {refPath=&NiagaraBool}}}
      Index 1 Param 3 {bIsDynamic=false, bIsEditable=false, bIsStaticSwitch=false, bIsVisible=true, type {classStructOrEnum {refPath=&NiagaraFloat}}}
      Module.Index1Param3Write {bIsDynamic=false, bIsEditable=true, bIsStaticSwitch=true, bIsVisible=true, type {classStructOrEnum {refPath=&NiagaraBool}}}
      Index 1 Param 4 {bIsDynamic=false, bIsEditable=false, bIsStaticSwitch=false, bIsVisible=true, type {classStructOrEnum {refPath=&NiagaraFloat}}}
      Write Parameter Index 2 {bIsDynamic=false, bIsEditable=true, bIsStaticSwitch=true, bIsVisible=true, type {classStructOrEnum {refPath=&NiagaraBool}}}
      Index 2 Param 1 {
       bIsDynamic=false
       bIsEditable=false
       bIsStaticSwitch=false
       bIsVisible=false
       type {classStructOrEnum {refPath=&NiagaraFloat}}
      }
      Index 2 Param 2 {
       bIsDynamic=false
       bIsEditable=false
       bIsStaticSwitch=false
       bIsVisible=false
       type {classStructOrEnum {refPath=&NiagaraFloat}}
      }
      Index 2 Param 3 {
       bIsDynamic=false
       bIsEditable=false
       bIsStaticSwitch=false
       bIsVisible=false
       type {classStructOrEnum {refPath=&NiagaraFloat}}
      }
      Index 2 Param 4 {
       bIsDynamic=false
       bIsEditable=false
       bIsStaticSwitch=false
       bIsVisible=false
       type {classStructOrEnum {refPath=&NiagaraFloat}}
      }
      Write Parameter Index 3 {bIsDynamic=false, bIsEditable=true, bIsStaticSwitch=true, bIsVisible=true, type {classStructOrEnum {refPath=&NiagaraBool}}}
      Module.Index3Param0Write {bIsDynamic=false, bIsEditable=true, bIsStaticSwitch=true, bIsVisible=true, type {classStructOrEnum {refPath=&NiagaraBool}}}
      Index 3 Param 1 {bIsDynamic=true, bIsEditable=true, bIsStaticSwitch=false, bIsVisible=true, type {classStructOrEnum {refPath=&NiagaraFloat}}}
      Module.Index3Param1Write {bIsDynamic=false, bIsEditable=true, bIsStaticSwitch=true, bIsVisible=true, type {classStructOrEnum {refPath=&NiagaraBool}}}
      Index 3 Param 2 {bIsDynamic=true, bIsEditable=true, bIsStaticSwitch=false, bIsVisible=true, type {classStructOrEnum {refPath=&NiagaraFloat}}}
      Module.Index3Param2Write {bIsDynamic=false, bIsEditable=true, bIsStaticSwitch=true, bIsVisible=true, type {classStructOrEnum {refPath=&NiagaraBool}}}
      Index 3 Param 3 {bIsDynamic=true, bIsEditable=true, bIsStaticSwitch=false, bIsVisible=true, type {classStructOrEnum {refPath=&NiagaraFloat}}}
      Module.Index3Param3Write {bIsDynamic=false, bIsEditable=true, bIsStaticSwitch=true, bIsVisible=true, type {classStructOrEnum {refPath=&NiagaraBool}}}
      Index 3 Param 4 {bIsDynamic=true, bIsEditable=true, bIsStaticSwitch=false, bIsVisible=true, type {classStructOrEnum {refPath=&NiagaraFloat}}}
     ]
     moduleScript {refPath=&DynamicMaterialParameters}
    }
    {
     bIsSetParametersModule=false
     enabled=true
     moduleName=ScaleColor
     inputs [
      Scale Mode {
       bIsDynamic=false
       bIsEditable=true
       bIsStaticSwitch=true
       bIsVisible=true
       type {classStructOrEnum {refPath=&ENiagaraScaleColorMode}}
      }
      Linear Color Curve {
       bIsDynamic=false
       bIsEditable=false
       bIsStaticSwitch=false
       bIsVisible=false
       type {classStructOrEnum {refPath=&NiagaraDataInterfaceColorCurve}}
      }
      ScaleRGB {bIsDynamic=false, bIsEditable=true, bIsStaticSwitch=false, bIsVisible=true, type {classStructOrEnum {refPath=&NiagaraBool}}}
      Scale RGB {bIsDynamic=true, bIsEditable=true, bIsStaticSwitch=false, bIsVisible=true, type {classStructOrEnum {refPath=&Vector3f}}}
      ScaleRGBA {bIsDynamic=false, bIsEditable=false, bIsStaticSwitch=false, bIsVisible=false, type {classStructOrEnum {refPath=&NiagaraBool}}}
      Scale RGBA {
       bIsDynamic=false
       bIsEditable=false
       bIsStaticSwitch=false
       bIsVisible=false
       type {classStructOrEnum {refPath=&Vector4f}}
      }
      Curve Index {
       bIsDynamic=false
       bIsEditable=false
       bIsStaticSwitch=false
       bIsVisible=false
       type {classStructOrEnum {refPath=&NiagaraFloat}}
      }
      ScaleA {bIsDynamic=false, bIsEditable=true, bIsStaticSwitch=false, bIsVisible=true, type {classStructOrEnum {refPath=&NiagaraBool}}}
      Scale Alpha {bIsDynamic=true, bIsEditable=true, bIsStaticSwitch=false, bIsVisible=true, type {classStructOrEnum {refPath=&NiagaraFloat}}}
      Color Value To Scale {
       bIsDynamic=false
       bIsEditable=true
       bIsStaticSwitch=false
       bIsVisible=true
       type {classStructOrEnum {refPath=&LinearColor}}
      }
     ]
     moduleScript {refPath=&ScaleColor}
    }
    {
     bIsSetParametersModule=true
     enabled=true
     moduleName=SetVariables_#1
     inputs [Particles.LightColor {
       bIsDynamic=true
       bIsEditable=true
       bIsStaticSwitch=false
       bIsVisible=true
       type {classStructOrEnum {refPath=&LinearColor}}
      }]
     moduleScript {refPath=&self:Flash001_1.NiagaraScriptSource_0.NiagaraGraph_0.NiagaraNodeAssignment_0.SetVariables_#1}
    }
    {
     bIsSetParametersModule=true
     enabled=true
     moduleName=SetVariables_#2
     inputs [Particles.RingColor {
       bIsDynamic=true
       bIsEditable=true
       bIsStaticSwitch=false
       bIsVisible=true
       type {classStructOrEnum {refPath=&LinearColor}}
      }]
     moduleScript {refPath=&self:Flash001_1.NiagaraScriptSource_0.NiagaraGraph_0.NiagaraNodeAssignment_1.SetVariables_#2}
    }
    {
     bIsSetParametersModule=false
     enabled=true
     moduleName=ScaleSpriteSize
     inputs [
      Scale Sprite Size Mode {
       bIsDynamic=false
       bIsEditable=true
       bIsStaticSwitch=true
       bIsVisible=true
       type {classStructOrEnum {refPath=&ENiagara_ScaleSpriteSize}}
      }
      Uniform Curve Sprite Scale {
       bIsDynamic=false
       bIsEditable=true
       bIsStaticSwitch=false
       bIsVisible=true
       type {classStructOrEnum {refPath=&NiagaraDataInterfaceCurve}}
      }
      Non-Uniform Curve Sprite Scale {
       bIsDynamic=false
       bIsEditable=false
       bIsStaticSwitch=false
       bIsVisible=false
       type {classStructOrEnum {refPath=&NiagaraDataInterfaceVector2DCurve}}
      }
      Uniform Curve Index {bIsDynamic=false, bIsEditable=true, bIsStaticSwitch=false, bIsVisible=true, type {classStructOrEnum {refPath=&NiagaraFloat}}}
      Non-Uniform Curve Index {
       bIsDynamic=false
       bIsEditable=false
       bIsStaticSwitch=false
       bIsVisible=false
       type {classStructOrEnum {refPath=&NiagaraFloat}}
      }
      Non-Uniform Curve Scale {
       bIsDynamic=false
       bIsEditable=false
       bIsStaticSwitch=false
       bIsVisible=false
       type {classStructOrEnum {refPath=&Vector2f}}
      }
      Uniform Curve Scale {bIsDynamic=false, bIsEditable=true, bIsStaticSwitch=false, bIsVisible=true, type {classStructOrEnum {refPath=&NiagaraFloat}}}
      Scale Factor {
       bIsDynamic=false
       bIsEditable=false
       bIsStaticSwitch=false
       bIsVisible=false
       type {classStructOrEnum {refPath=&Vector2f}}
      }
      Uniform Scale Factor {
       bIsDynamic=false
       bIsEditable=false
       bIsStaticSwitch=false
       bIsVisible=false
       type {classStructOrEnum {refPath=&NiagaraFloat}}
      }
      Initial Sprite Size {bIsDynamic=false, bIsEditable=true, bIsStaticSwitch=false, bIsVisible=true, type {classStructOrEnum {refPath=&Vector2f}}}
     ]
     moduleScript {refPath=&ScaleSpriteSize}
    }
   ]
  }
  rendererClasses [{refPath=&NiagaraSpriteRendererProperties}, {refPath=&NiagaraMeshRendererProperties}]
  renderers [
   {rendererIndex=0, rendererClass {refPath=&NiagaraSpriteRendererProperties}}
   {rendererIndex=1, rendererClass {refPath=&NiagaraMeshRendererProperties}}
  ]
 }
 HandGlow {
  emitterSpawnScript {scriptName=EmitterSpawnScript, modules []}
  emitterUpdateScript {
   scriptName=EmitterUpdateScript
   modules [
    {
     bIsSetParametersModule=false
     enabled=true
     moduleName=EmitterState
     inputs [
      Life Cycle Mode {
       bIsDynamic=false
       bIsEditable=true
       bIsStaticSwitch=true
       bIsVisible=true
       type {classStructOrEnum {refPath=&ENiagaraEmitterLifeCycleMode}}
      }
      Inactive Response {
       bIsDynamic=false
       bIsEditable=true
       bIsStaticSwitch=true
       bIsVisible=true
       type {classStructOrEnum {refPath=&ENiagaraInactiveMode}}
      }
      Loop Behavior {
       bIsDynamic=false
       bIsEditable=true
       bIsStaticSwitch=true
       bIsVisible=true
       type {classStructOrEnum {refPath=&ENiagara_EmitterStateOptions}}
      }
      Loop Duration Mode {
       bIsDynamic=false
       bIsEditable=true
       bIsStaticSwitch=true
       bIsVisible=true
       type {classStructOrEnum {refPath=&ENiagara_InfiniteLoopDuration}}
      }
      Loop Count {
       bIsDynamic=false
       bIsEditable=false
       bIsStaticSwitch=false
       bIsVisible=false
       type {classStructOrEnum {refPath=&NiagaraInt32}}
      }
      Loop Duration {bIsDynamic=false, bIsEditable=true, bIsStaticSwitch=false, bIsVisible=true, type {classStructOrEnum {refPath=&NiagaraFloat}}}
      Recalculate Duration Each Loop {bIsDynamic=false, bIsEditable=false, bIsStaticSwitch=false, bIsVisible=false, type {classStructOrEnum {refPath=&NiagaraBool}}}
      UseLoopDelay {bIsDynamic=false, bIsEditable=true, bIsStaticSwitch=false, bIsVisible=true, type {classStructOrEnum {refPath=&NiagaraBool}}}
      Loop Delay {bIsDynamic=false, bIsEditable=false, bIsStaticSwitch=false, bIsVisible=true, type {classStructOrEnum {refPath=&NiagaraFloat}}}
      Delay First Loop Only {bIsDynamic=false, bIsEditable=false, bIsStaticSwitch=false, bIsVisible=false, type {classStructOrEnum {refPath=&NiagaraBool}}}
      Scalability Mode {
       bIsDynamic=false
       bIsEditable=true
       bIsStaticSwitch=true
       bIsVisible=true
       type {classStructOrEnum {refPath=&ENiagaraEmitterScalabilityMode_Limited}}
      }
      Enable Distance Culling {bIsDynamic=false, bIsEditable=false, bIsStaticSwitch=true, bIsVisible=false, type {classStructOrEnum {refPath=&NiagaraBool}}}
      MinDistance {
       bIsDynamic=false
       bIsEditable=false
       bIsStaticSwitch=false
       bIsVisible=false
       type {classStructOrEnum {refPath=&NiagaraFloat}}
      }
      Min Distance Response {
       bIsDynamic=false
       bIsEditable=false
       bIsStaticSwitch=false
       bIsVisible=false
       type {classStructOrEnum {refPath=&ENiagaraExecutionStateManagement}}
      }
      MaxDistance {
       bIsDynamic=false
       bIsEditable=false
       bIsStaticSwitch=false
       bIsVisible=false
       type {classStructOrEnum {refPath=&NiagaraFloat}}
      }
      Max Distance Response {
       bIsDynamic=false
       bIsEditable=false
       bIsStaticSwitch=false
       bIsVisible=false
       type {classStructOrEnum {refPath=&ENiagaraExecutionStateManagement}}
      }
      Scale Spawn Count {bIsDynamic=false, bIsEditable=false, bIsStaticSwitch=true, bIsVisible=false, type {classStructOrEnum {refPath=&NiagaraBool}}}
      Spawn Count Scale {
       bIsDynamic=false
       bIsEditable=false
       bIsStaticSwitch=false
       bIsVisible=false
       type {classStructOrEnum {refPath=&NiagaraFloat}}
      }
      Spawn Count Scale By Distance {
       bIsDynamic=false
       bIsEditable=false
       bIsStaticSwitch=false
       bIsVisible=false
       type {classStructOrEnum {refPath=&NiagaraDataInterfaceCurve}}
      }
      Enable Visibility Culling {bIsDynamic=false, bIsEditable=false, bIsStaticSwitch=true, bIsVisible=false, type {classStructOrEnum {refPath=&NiagaraBool}}}
      Visibility Cull Response {
       bIsDynamic=false
       bIsEditable=false
       bIsStaticSwitch=false
       bIsVisible=false
       type {classStructOrEnum {refPath=&ENiagaraExecutionStateManagement}}
      }
      Visibility Cull Delay {
       bIsDynamic=false
       bIsEditable=false
       bIsStaticSwitch=false
       bIsVisible=false
       type {classStructOrEnum {refPath=&NiagaraFloat}}
      }
      Reset Age On Awaken {bIsDynamic=false, bIsEditable=false, bIsStaticSwitch=false, bIsVisible=false, type {classStructOrEnum {refPath=&NiagaraBool}}}
     ]
     moduleScript {refPath=&EmitterState}
    }
    {
     bIsSetParametersModule=false
     enabled=true
     moduleName=SpawnBurst_Instantaneous
     inputs [
      Spawn Count {bIsDynamic=false, bIsEditable=true, bIsStaticSwitch=false, bIsVisible=true, type {classStructOrEnum {refPath=&NiagaraInt32}}}
      Spawn Time {bIsDynamic=false, bIsEditable=true, bIsStaticSwitch=false, bIsVisible=true, type {classStructOrEnum {refPath=&NiagaraFloat}}}
      Use Spawn Probability {bIsDynamic=false, bIsEditable=true, bIsStaticSwitch=false, bIsVisible=true, type {classStructOrEnum {refPath=&NiagaraBool}}}
      Spawn Probability {bIsDynamic=false, bIsEditable=false, bIsStaticSwitch=false, bIsVisible=true, type {classStructOrEnum {refPath=&NiagaraFloat}}}
      UseLoopCountLimit {bIsDynamic=false, bIsEditable=true, bIsStaticSwitch=false, bIsVisible=true, type {classStructOrEnum {refPath=&NiagaraBool}}}
      Loop Count Limit {bIsDynamic=false, bIsEditable=false, bIsStaticSwitch=false, bIsVisible=true, type {classStructOrEnum {refPath=&NiagaraInt32}}}
      Spawn Group {bIsDynamic=false, bIsEditable=true, bIsStaticSwitch=false, bIsVisible=true, type {classStructOrEnum {refPath=&NiagaraInt32}}}
      Age {bIsDynamic=false, bIsEditable=true, bIsStaticSwitch=false, bIsVisible=true, type {classStructOrEnum {refPath=&NiagaraFloat}}}
     ]
     moduleScript {refPath=&SpawnBurst_Instantaneous}
    }
   ]
  }
  particleSpawnScript {
   scriptName=ParticleSpawnScript
   modules [
    {
     bIsSetParametersModule=false
     enabled=true
     moduleName=InitializeParticle
     inputs [
      Lifetime Mode {
       bIsDynamic=false
       bIsEditable=true
       bIsStaticSwitch=true
       bIsVisible=true
       type {classStructOrEnum {refPath=&ENiagara_LifetimeMode}}
      }
      Lifetime Min {
       bIsDynamic=false
       bIsEditable=false
       bIsStaticSwitch=false
       bIsVisible=false
       type {classStructOrEnum {refPath=&NiagaraFloat}}
      }
      Lifetime {bIsDynamic=true, bIsEditable=true, bIsStaticSwitch=false, bIsVisible=true, type {classStructOrEnum {refPath=&NiagaraFloat}}}
      Lifetime Max {
       bIsDynamic=false
       bIsEditable=false
       bIsStaticSwitch=false
       bIsVisible=false
       type {classStructOrEnum {refPath=&NiagaraFloat}}
      }
      Lifetime Randomness Mode {
       bIsDynamic=false
       bIsEditable=false
       bIsStaticSwitch=true
       bIsVisible=false
       type {classStructOrEnum {refPath=&ENiagaraRandomnessMode}}
      }
      Color Mode {
       bIsDynamic=false
       bIsEditable=true
       bIsStaticSwitch=true
       bIsVisible=true
       type {classStructOrEnum {refPath=&ENiagara_ColorInitializationMode}}
      }
      Color Minimum {
       bIsDynamic=false
       bIsEditable=false
       bIsStaticSwitch=false
       bIsVisible=false
       type {classStructOrEnum {refPath=&LinearColor}}
      }
      Color {
       bIsDynamic=false
       bIsEditable=true
       bIsStaticSwitch=false
       bIsVisible=true
       type {classStructOrEnum {refPath=&LinearColor}}
      }
      AdjustHue {bIsDynamic=false, bIsEditable=false, bIsStaticSwitch=false, bIsVisible=false, type {classStructOrEnum {refPath=&NiagaraBool}}}
      Hue Shift Range {
       bIsDynamic=false
       bIsEditable=false
       bIsStaticSwitch=false
       bIsVisible=false
       type {classStructOrEnum {refPath=&Vector2f}}
      }
      Color Maximum {
       bIsDynamic=false
       bIsEditable=false
       bIsStaticSwitch=false
       bIsVisible=false
       type {classStructOrEnum {refPath=&LinearColor}}
      }
      AdjustSaturation {bIsDynamic=false, bIsEditable=false, bIsStaticSwitch=false, bIsVisible=false, type {classStructOrEnum {refPath=&NiagaraBool}}}
      Saturation Range {
       bIsDynamic=false
       bIsEditable=false
       bIsStaticSwitch=false
       bIsVisible=false
       type {classStructOrEnum {refPath=&Vector2f}}
      }
      AdjustValue {bIsDynamic=false, bIsEditable=false, bIsStaticSwitch=false, bIsVisible=false, type {classStructOrEnum {refPath=&NiagaraBool}}}
      Value Range {
       bIsDynamic=false
       bIsEditable=false
       bIsStaticSwitch=false
       bIsVisible=false
       type {classStructOrEnum {refPath=&Vector2f}}
      }
      AdjustAlpha {bIsDynamic=false, bIsEditable=false, bIsStaticSwitch=false, bIsVisible=false, type {classStructOrEnum {refPath=&NiagaraBool}}}
      Alpha Scale Range {
       bIsDynamic=false
       bIsEditable=false
       bIsStaticSwitch=false
       bIsVisible=false
       type {classStructOrEnum {refPath=&Vector2f}}
      }
      Color Channel Mode {
       bIsDynamic=false
       bIsEditable=false
       bIsStaticSwitch=true
       bIsVisible=false
       type {classStructOrEnum {refPath=&ENiagaraChannelCorrelation}}
      }
      Color Randomness Mode {
       bIsDynamic=false
       bIsEditable=false
       bIsStaticSwitch=true
       bIsVisible=false
       type {classStructOrEnum {refPath=&ENiagaraRandomnessMode}}
      }
      Position Mode {
       bIsDynamic=false
       bIsEditable=true
       bIsStaticSwitch=true
       bIsVisible=true
       type {classStructOrEnum {refPath=&ENiagara_PositionInitializationMode}}
      }
      Position {
       bIsDynamic=false
       bIsEditable=false
       bIsStaticSwitch=false
       bIsVisible=false
       type {classStructOrEnum {refPath=&NiagaraPosition}}
      }
      UsePositionOffset {bIsDynamic=false, bIsEditable=true, bIsStaticSwitch=false, bIsVisible=true, type {classStructOrEnum {refPath=&NiagaraBool}}}
      Position Offset {bIsDynamic=false, bIsEditable=true, bIsStaticSwitch=false, bIsVisible=true, type {classStructOrEnum {refPath=&Vector3f}}}
      Position Offset Coordinate Space {
       bIsDynamic=false
       bIsEditable=true
       bIsStaticSwitch=false
       bIsVisible=true
       type {classStructOrEnum {refPath=&ENiagaraCoordinateSpace}}
      }
      Mass Mode {
       bIsDynamic=false
       bIsEditable=true
       bIsStaticSwitch=true
       bIsVisible=true
       type {classStructOrEnum {refPath=&ENiagara_MassInitializationMode}}
      }
      Mass Min {
       bIsDynamic=false
       bIsEditable=false
       bIsStaticSwitch=false
       bIsVisible=false
       type {classStructOrEnum {refPath=&NiagaraFloat}}
      }
      Mass {
       bIsDynamic=false
       bIsEditable=false
       bIsStaticSwitch=false
       bIsVisible=false
       type {classStructOrEnum {refPath=&NiagaraFloat}}
      }
      Mass Max {
       bIsDynamic=false
       bIsEditable=false
       bIsStaticSwitch=false
       bIsVisible=false
       type {classStructOrEnum {refPath=&NiagaraFloat}}
      }
      Mass Randomness Mode {
       bIsDynamic=false
       bIsEditable=false
       bIsStaticSwitch=true
       bIsVisible=false
       type {classStructOrEnum {refPath=&ENiagaraRandomnessMode}}
      }
      Write Material Random {bIsDynamic=false, bIsEditable=true, bIsStaticSwitch=false, bIsVisible=true, type {classStructOrEnum {refPath=&NiagaraBool}}}
      Material Random {bIsDynamic=false, bIsEditable=true, bIsStaticSwitch=false, bIsVisible=true, type {classStructOrEnum {refPath=&NiagaraFloat}}}
      Sprite Size Mode {
       bIsDynamic=false
       bIsEditable=true
       bIsStaticSwitch=true
       bIsVisible=true
       type {classStructOrEnum {refPath=&ENiagara_SizeScaleMode}}
      }
      Sprite Size {
       bIsDynamic=false
       bIsEditable=false
       bIsStaticSwitch=false
       bIsVisible=false
       type {classStructOrEnum {refPath=&Vector2f}}
      }
      Uniform Sprite Size {bIsDynamic=false, bIsEditable=true, bIsStaticSwitch=false, bIsVisible=true, type {classStructOrEnum {refPath=&NiagaraFloat}}}
      Uniform Sprite Size Min {
       bIsDynamic=false
       bIsEditable=false
       bIsStaticSwitch=false
       bIsVisible=false
       type {classStructOrEnum {refPath=&NiagaraFloat}}
      }
      Sprite Size Min {
       bIsDynamic=false
       bIsEditable=false
       bIsStaticSwitch=false
       bIsVisible=false
       type {classStructOrEnum {refPath=&Vector2f}}
      }
      Sprite Size Max {
       bIsDynamic=false
       bIsEditable=false
       bIsStaticSwitch=false
       bIsVisible=false
       type {classStructOrEnum {refPath=&Vector2f}}
      }
      Uniform Sprite Size Max {
       bIsDynamic=false
       bIsEditable=false
       bIsStaticSwitch=false
       bIsVisible=false
       type {classStructOrEnum {refPath=&NiagaraFloat}}
      }
      Sprite Size Randomness Mode {
       bIsDynamic=false
       bIsEditable=false
       bIsStaticSwitch=true
       bIsVisible=false
       type {classStructOrEnum {refPath=&ENiagaraRandomnessMode}}
      }
      Sprite Rotation Mode {
       bIsDynamic=false
       bIsEditable=true
       bIsStaticSwitch=true
       bIsVisible=true
       type {classStructOrEnum {refPath=&ENiagara_SpriteRotationMode}}
      }
      Sprite Rotation Angle Min {
       bIsDynamic=false
       bIsEditable=false
       bIsStaticSwitch=false
       bIsVisible=false
       type {classStructOrEnum {refPath=&NiagaraFloat}}
      }
      Sprite Rotation Angle {bIsDynamic=false, bIsEditable=true, bIsStaticSwitch=false, bIsVisible=true, type {classStructOrEnum {refPath=&NiagaraFloat}}}
      Sprite Rotation Angle Max {
       bIsDynamic=false
       bIsEditable=false
       bIsStaticSwitch=false
       bIsVisible=false
       type {classStructOrEnum {refPath=&NiagaraFloat}}
      }
      Sprite Rotation Randomness Mode {
       bIsDynamic=false
       bIsEditable=false
       bIsStaticSwitch=true
       bIsVisible=false
       type {classStructOrEnum {refPath=&ENiagaraRandomnessMode}}
      }
      Sprite UV Mode {
       bIsDynamic=false
       bIsEditable=true
       bIsStaticSwitch=true
       bIsVisible=true
       type {classStructOrEnum {refPath=&ENiagara_UVFlippingMode}}
      }
      Sprite UV Randomness Mode {
       bIsDynamic=false
       bIsEditable=false
       bIsStaticSwitch=true
       bIsVisible=false
       type {classStructOrEnum {refPath=&ENiagaraRandomnessMode}}
      }
      Sprite UV Scale {
       bIsDynamic=false
       bIsEditable=false
       bIsStaticSwitch=false
       bIsVisible=false
       type {classStructOrEnum {refPath=&Vector2f}}
      }
      Mesh Scale Mode {
       bIsDynamic=false
       bIsEditable=true
       bIsStaticSwitch=true
       bIsVisible=true
       type {classStructOrEnum {refPath=&ENiagara_SizeScaleMode}}
      }
      Mesh Uniform Scale Min {
       bIsDynamic=false
       bIsEditable=false
       bIsStaticSwitch=false
       bIsVisible=false
       type {classStructOrEnum {refPath=&NiagaraFloat}}
      }
      Mesh Uniform Scale {
       bIsDynamic=false
       bIsEditable=false
       bIsStaticSwitch=false
       bIsVisible=false
       type {classStructOrEnum {refPath=&NiagaraFloat}}
      }
      Mesh Scale Min {
       bIsDynamic=false
       bIsEditable=false
       bIsStaticSwitch=false
       bIsVisible=false
       type {classStructOrEnum {refPath=&Vector3f}}
      }
      Mesh Scale {bIsDynamic=true, bIsEditable=true, bIsStaticSwitch=false, bIsVisible=true, type {classStructOrEnum {refPath=&Vector3f}}}
      Mesh Scale Max {
       bIsDynamic=false
       bIsEditable=false
       bIsStaticSwitch=false
       bIsVisible=false
       type {classStructOrEnum {refPath=&Vector3f}}
      }
      Mesh Uniform Scale Max {
       bIsDynamic=false
       bIsEditable=false
       bIsStaticSwitch=false
       bIsVisible=false
       type {classStructOrEnum {refPath=&NiagaraFloat}}
      }
      Mesh Scale Randomness Mode {
       bIsDynamic=false
       bIsEditable=false
       bIsStaticSwitch=true
       bIsVisible=false
       type {classStructOrEnum {refPath=&ENiagaraRandomnessMode}}
      }
      Mesh Renderer Array Visibility Mode {
       bIsDynamic=false
       bIsEditable=true
       bIsStaticSwitch=true
       bIsVisible=true
       type {classStructOrEnum {refPath=&ENiagara_UnsetDirectSetRandom}}
      }
      Mesh Renderer Mesh Index {
       bIsDynamic=false
       bIsEditable=false
       bIsStaticSwitch=false
       bIsVisible=false
       type {classStructOrEnum {refPath=&NiagaraInt32}}
      }
      Wrap Index By Mesh Count {bIsDynamic=false, bIsEditable=false, bIsStaticSwitch=true, bIsVisible=false, type {classStructOrEnum {refPath=&NiagaraBool}}}
      Mesh Renderer Info {
       bIsDynamic=false
       bIsEditable=false
       bIsStaticSwitch=false
       bIsVisible=false
       type {classStructOrEnum {refPath=&NiagaraDataInterfaceMeshRendererInfo}}
      }
      Mesh Renderer Random Seed {
       bIsDynamic=false
       bIsEditable=false
       bIsStaticSwitch=false
       bIsVisible=false
       type {classStructOrEnum {refPath=&NiagaraInt32}}
      }
      Ribbon Width Mode {
       bIsDynamic=false
       bIsEditable=true
       bIsStaticSwitch=true
       bIsVisible=true
       type {classStructOrEnum {refPath=&ENiagara_UnsetDirectSet}}
      }
      Ribbon Width {
       bIsDynamic=false
       bIsEditable=false
       bIsStaticSwitch=false
       bIsVisible=false
       type {classStructOrEnum {refPath=&NiagaraFloat}}
      }
      Ribbon Facing Mode {
       bIsDynamic=false
       bIsEditable=true
       bIsStaticSwitch=true
       bIsVisible=true
       type {classStructOrEnum {refPath=&ENiagara_UnsetDirectSet}}
      }
      Ribbon Facing Vector {
       bIsDynamic=false
       bIsEditable=false
       bIsStaticSwitch=false
       bIsVisible=false
       type {classStructOrEnum {refPath=&Vector3f}}
      }
      Ribbon Twist Mode {
       bIsDynamic=false
       bIsEditable=true
       bIsStaticSwitch=true
       bIsVisible=true
       type {classStructOrEnum {refPath=&ENiagara_UnsetDirectSet}}
      }
      "Ribbon Twist " {
       bIsDynamic=false
       bIsEditable=false
       bIsStaticSwitch=false
       bIsVisible=false
       type {classStructOrEnum {refPath=&NiagaraFloat}}
      }
     ]
     moduleScript {refPath=&InitializeParticle}
    }
    {
     bIsSetParametersModule=false
     enabled=false
     moduleName=InitialMeshOrientation
     inputs [
      Mesh Orientation Mode {
       bIsDynamic=false
       bIsEditable=true
       bIsStaticSwitch=true
       bIsVisible=true
       type {classStructOrEnum {refPath=/Niagara/Enums/Transforms/ENiagara_MeshOrientationOptions}}
      }
      Orientation Quaternion {bIsDynamic=false, bIsEditable=false, bIsStaticSwitch=false, bIsVisible=false, type {classStructOrEnum {refPath=/Script/CoreUObject.Quat4f}}}
      Orientation Matrix {
       bIsDynamic=false
       bIsEditable=false
       bIsStaticSwitch=false
       bIsVisible=false
       type {classStructOrEnum {refPath=/Script/Niagara.NiagaraMatrix}}
      }
      Orientation Vector {bIsDynamic=false, bIsEditable=true, bIsStaticSwitch=false, bIsVisible=true, type {classStructOrEnum {refPath=&Vector3f}}}
      Orientation Coordinate Space {
       bIsDynamic=false
       bIsEditable=true
       bIsStaticSwitch=true
       bIsVisible=true
       type {classStructOrEnum {refPath=&ENiagaraCoordinateSpace}}
      }
      Mesh Axis To Orient {bIsDynamic=false, bIsEditable=true, bIsStaticSwitch=false, bIsVisible=true, type {classStructOrEnum {refPath=&Vector3f}}}
      Random Seed {
       bIsDynamic=false
       bIsEditable=false
       bIsStaticSwitch=false
       bIsVisible=false
       type {classStructOrEnum {refPath=&NiagaraInt32}}
      }
      Use Rotation Vector {bIsDynamic=false, bIsEditable=true, bIsStaticSwitch=false, bIsVisible=true, type {classStructOrEnum {refPath=&NiagaraBool}}}
      Rotation {bIsDynamic=false, bIsEditable=true, bIsStaticSwitch=false, bIsVisible=true, type {classStructOrEnum {refPath=&Vector3f}}}
      Rotation Coordinate Space {
       bIsDynamic=false
       bIsEditable=true
       bIsStaticSwitch=true
       bIsVisible=true
       type {classStructOrEnum {refPath=/Niagara/Enums/ENiagaraMeshTransforms}}
      }
     ]
     moduleScript {refPath=&InitialMeshOrientation}
    }
    {
     bIsSetParametersModule=false
     enabled=true
     moduleName=CameraOffset
     inputs [Camera Offset Amount {bIsDynamic=false, bIsEditable=true, bIsStaticSwitch=false, bIsVisible=true, type {classStructOrEnum {refPath=&NiagaraFloat}}}]
     moduleScript {refPath=&CameraOffset}
    }
    {
     bIsSetParametersModule=false
     enabled=true
     moduleName=AddVelocity
     inputs [
      Velocity Mode {
       bIsDynamic=false
       bIsEditable=true
       bIsStaticSwitch=true
       bIsVisible=true
       type {classStructOrEnum {refPath=/Niagara/Enums/Utility/ENiagara_VelocityMode}}
      }
      Velocity {bIsDynamic=false, bIsEditable=true, bIsStaticSwitch=false, bIsVisible=true, type {classStructOrEnum {refPath=&Vector3f}}}
      Velocity Speed {
       bIsDynamic=false
       bIsEditable=false
       bIsStaticSwitch=false
       bIsVisible=false
       type {classStructOrEnum {refPath=&NiagaraFloat}}
      }
      Velocity Speed Scale {bIsDynamic=false, bIsEditable=true, bIsStaticSwitch=false, bIsVisible=true, type {classStructOrEnum {refPath=&NiagaraFloat}}}
      Constrain To Radius {
       bIsDynamic=false
       bIsEditable=false
       bIsStaticSwitch=false
       bIsVisible=false
       type {classStructOrEnum {refPath=&NiagaraFloat}}
      }
      Distribution Along Cone Axis {
       bIsDynamic=false
       bIsEditable=false
       bIsStaticSwitch=false
       bIsVisible=false
       type {classStructOrEnum {refPath=&NiagaraFloat}}
      }
      Radius Falloff Near / Far {
       bIsDynamic=false
       bIsEditable=false
       bIsStaticSwitch=false
       bIsVisible=false
       type {classStructOrEnum {refPath=&Vector2f}}
      }
      Radius Falloff Exponent {
       bIsDynamic=false
       bIsEditable=false
       bIsStaticSwitch=false
       bIsVisible=false
       type {classStructOrEnum {refPath=&NiagaraFloat}}
      }
      Speed Falloff From Cone Axis {
       bIsDynamic=false
       bIsEditable=false
       bIsStaticSwitch=false
       bIsVisible=false
       type {classStructOrEnum {refPath=&NiagaraFloat}}
      }
      Invert Velocity Falloff {bIsDynamic=false, bIsEditable=false, bIsStaticSwitch=false, bIsVisible=false, type {classStructOrEnum {refPath=&NiagaraBool}}}
      Override Seed {bIsDynamic=false, bIsEditable=false, bIsStaticSwitch=false, bIsVisible=false, type {classStructOrEnum {refPath=&NiagaraBool}}}
      Random Seed {
       bIsDynamic=false
       bIsEditable=false
       bIsStaticSwitch=false
       bIsVisible=false
       type {classStructOrEnum {refPath=&NiagaraInt32}}
      }
      Fixed Random Seed {bIsDynamic=false, bIsEditable=false, bIsStaticSwitch=true, bIsVisible=false, type {classStructOrEnum {refPath=&NiagaraBool}}}
      Randomness Mode {
       bIsDynamic=false
       bIsEditable=false
       bIsStaticSwitch=true
       bIsVisible=false
       type {classStructOrEnum {refPath=&ENiagaraRandomnessMode}}
      }
      Cone Axis {
       bIsDynamic=false
       bIsEditable=false
       bIsStaticSwitch=false
       bIsVisible=false
       type {classStructOrEnum {refPath=&Vector3f}}
      }
      Cone Angle {
       bIsDynamic=false
       bIsEditable=false
       bIsStaticSwitch=false
       bIsVisible=false
       type {classStructOrEnum {refPath=&NiagaraFloat}}
      }
      Inner Cone Angle {
       bIsDynamic=false
       bIsEditable=false
       bIsStaticSwitch=false
       bIsVisible=false
       type {classStructOrEnum {refPath=&NiagaraFloat}}
      }
      Cone Angle Mode {
       bIsDynamic=false
       bIsEditable=false
       bIsStaticSwitch=true
       bIsVisible=false
       type {classStructOrEnum {refPath=/Niagara/Enums/Angles/ENiagara_AngleInput}}
      }
      Rotation Mode {
       bIsDynamic=false
       bIsEditable=true
       bIsStaticSwitch=true
       bIsVisible=true
       type {classStructOrEnum {refPath=/Niagara/Enums/Transforms/ENiagara_RotationMode}}
      }
      Rotation Axis {
       bIsDynamic=false
       bIsEditable=false
       bIsStaticSwitch=false
       bIsVisible=false
       type {classStructOrEnum {refPath=&Vector3f}}
      }
      Yaw / Pitch / Roll {
       bIsDynamic=false
       bIsEditable=false
       bIsStaticSwitch=false
       bIsVisible=false
       type {classStructOrEnum {refPath=&Vector3f}}
      }
      Rotation Quaternion {bIsDynamic=false, bIsEditable=false, bIsStaticSwitch=false, bIsVisible=false, type {classStructOrEnum {refPath=&Quat4f}}}
      Rotation Matrix {
       bIsDynamic=false
       bIsEditable=false
       bIsStaticSwitch=false
       bIsVisible=false
       type {classStructOrEnum {refPath=&NiagaraMatrix}}
      }
      Rotation Angle {
       bIsDynamic=false
       bIsEditable=false
       bIsStaticSwitch=false
       bIsVisible=false
       type {classStructOrEnum {refPath=&NiagaraFloat}}
      }
      Invert Rotation Quaternion {bIsDynamic=false, bIsEditable=false, bIsStaticSwitch=false, bIsVisible=false, type {classStructOrEnum {refPath=&NiagaraBool}}}
      Additional Quaternion Rotation {bIsDynamic=false, bIsEditable=false, bIsStaticSwitch=false, bIsVisible=false, type {classStructOrEnum {refPath=&NiagaraBool}}}
      Additional Yaw / Pitch / Roll {
       bIsDynamic=false
       bIsEditable=false
       bIsStaticSwitch=false
       bIsVisible=false
       type {classStructOrEnum {refPath=&Vector3f}}
      }
      Rotation Angle Type {
       bIsDynamic=false
       bIsEditable=false
       bIsStaticSwitch=true
       bIsVisible=false
       type {classStructOrEnum {refPath=&ENiagara_AngleInput}}
      }
      Rotation Coordinate Space {
       bIsDynamic=false
       bIsEditable=true
       bIsStaticSwitch=false
       bIsVisible=true
       type {classStructOrEnum {refPath=&ENiagaraCoordinateSpace}}
      }
      Velocity Origin {
       bIsDynamic=false
       bIsEditable=false
       bIsStaticSwitch=false
       bIsVisible=false
       type {classStructOrEnum {refPath=&NiagaraPosition}}
      }
      Origin Offset {
       bIsDynamic=false
       bIsEditable=false
       bIsStaticSwitch=false
       bIsVisible=false
       type {classStructOrEnum {refPath=&Vector3f}}
      }
      Origin Offset Coordinate Space {
       bIsDynamic=false
       bIsEditable=false
       bIsStaticSwitch=false
       bIsVisible=false
       type {classStructOrEnum {refPath=&ENiagaraCoordinateSpace}}
      }
      Default Position {
       bIsDynamic=false
       bIsEditable=false
       bIsStaticSwitch=false
       bIsVisible=false
       type {classStructOrEnum {refPath=&NiagaraPosition}}
      }
     ]
     moduleScript {refPath=&AddVelocity}
    }
    {
     bIsSetParametersModule=false
     enabled=true
     moduleName=SpriteFacingAndAlignment
     inputs [
      WriteFacing {bIsDynamic=false, bIsEditable=true, bIsStaticSwitch=false, bIsVisible=true, type {classStructOrEnum {refPath=&NiagaraBool}}}
      Sprite Facing {bIsDynamic=false, bIsEditable=true, bIsStaticSwitch=false, bIsVisible=true, type {classStructOrEnum {refPath=&Vector3f}}}
      Facing Coordinate Space {
       bIsDynamic=false
       bIsEditable=true
       bIsStaticSwitch=false
       bIsVisible=true
       type {classStructOrEnum {refPath=&ENiagaraCoordinateSpace}}
      }
      WriteAlignment {bIsDynamic=false, bIsEditable=true, bIsStaticSwitch=false, bIsVisible=true, type {classStructOrEnum {refPath=&NiagaraBool}}}
      Sprite Alignment {bIsDynamic=false, bIsEditable=false, bIsStaticSwitch=false, bIsVisible=true, type {classStructOrEnum {refPath=&Vector3f}}}
      Alignment Coordinate Space {
       bIsDynamic=false
       bIsEditable=false
       bIsStaticSwitch=false
       bIsVisible=false
       type {classStructOrEnum {refPath=&ENiagaraCoordinateSpace}}
      }
     ]
     moduleScript {refPath=&SpriteFacingAndAlignment}
    }
   ]
  }
  particleUpdateScript {
   scriptName=ParticleUpdateScript
   modules [
    {
     bIsSetParametersModule=false
     enabled=true
     moduleName=ParticleState
     inputs [
      Kill Particles When Lifetime Has Elapsed {bIsDynamic=false, bIsEditable=true, bIsStaticSwitch=true, bIsVisible=true, type {classStructOrEnum {refPath=&NiagaraBool}}}
      Loop Particles Lifetime {bIsDynamic=false, bIsEditable=false, bIsStaticSwitch=false, bIsVisible=false, type {classStructOrEnum {refPath=&NiagaraBool}}}
      Let Infinitely Lived Particles Die When Emitter Deactivates {bIsDynamic=false, bIsEditable=false, bIsStaticSwitch=false, bIsVisible=false, type {classStructOrEnum {refPath=&NiagaraBool}}}
      Lifetime {bIsDynamic=false, bIsEditable=true, bIsStaticSwitch=false, bIsVisible=true, type {classStructOrEnum {refPath=&NiagaraFloat}}}
      DeltaTime {bIsDynamic=false, bIsEditable=true, bIsStaticSwitch=false, bIsVisible=true, type {classStructOrEnum {refPath=&NiagaraFloat}}}
     ]
     moduleScript {refPath=&ParticleState}
    }
    {
     bIsSetParametersModule=false
     enabled=true
     moduleName=ScaleMeshSize
     inputs [
      Scale Factor {bIsDynamic=true, bIsEditable=true, bIsStaticSwitch=false, bIsVisible=true, type {classStructOrEnum {refPath=&Vector3f}}}
      Initial Mesh Scale {bIsDynamic=false, bIsEditable=true, bIsStaticSwitch=false, bIsVisible=true, type {classStructOrEnum {refPath=&Vector3f}}}
     ]
     moduleScript {refPath=&ScaleMeshSize}
    }
    {
     bIsSetParametersModule=false
     enabled=true
     moduleName=ScaleSpriteSize
     inputs [
      Scale Sprite Size Mode {
       bIsDynamic=false
       bIsEditable=true
       bIsStaticSwitch=true
       bIsVisible=true
       type {classStructOrEnum {refPath=&ENiagara_ScaleSpriteSize}}
      }
      Uniform Curve Sprite Scale {
       bIsDynamic=false
       bIsEditable=true
       bIsStaticSwitch=false
       bIsVisible=true
       type {classStructOrEnum {refPath=&NiagaraDataInterfaceCurve}}
      }
      Non-Uniform Curve Sprite Scale {
       bIsDynamic=false
       bIsEditable=false
       bIsStaticSwitch=false
       bIsVisible=false
       type {classStructOrEnum {refPath=&NiagaraDataInterfaceVector2DCurve}}
      }
      Uniform Curve Index {bIsDynamic=false, bIsEditable=true, bIsStaticSwitch=false, bIsVisible=true, type {classStructOrEnum {refPath=&NiagaraFloat}}}
      Non-Uniform Curve Index {
       bIsDynamic=false
       bIsEditable=false
       bIsStaticSwitch=false
       bIsVisible=false
       type {classStructOrEnum {refPath=&NiagaraFloat}}
      }
      Non-Uniform Curve Scale {
       bIsDynamic=false
       bIsEditable=false
       bIsStaticSwitch=false
       bIsVisible=false
       type {classStructOrEnum {refPath=&Vector2f}}
      }
      Uniform Curve Scale {bIsDynamic=false, bIsEditable=true, bIsStaticSwitch=false, bIsVisible=true, type {classStructOrEnum {refPath=&NiagaraFloat}}}
      Scale Factor {
       bIsDynamic=false
       bIsEditable=false
       bIsStaticSwitch=false
       bIsVisible=false
       type {classStructOrEnum {refPath=&Vector2f}}
      }
      Uniform Scale Factor {
       bIsDynamic=false
       bIsEditable=false
       bIsStaticSwitch=false
       bIsVisible=false
       type {classStructOrEnum {refPath=&NiagaraFloat}}
      }
      Initial Sprite Size {bIsDynamic=false, bIsEditable=true, bIsStaticSwitch=false, bIsVisible=true, type {classStructOrEnum {refPath=&Vector2f}}}
     ]
     moduleScript {refPath=&ScaleSpriteSize}
    }
    {
     bIsSetParametersModule=true
     enabled=true
     moduleName=SetVariables_#6
     inputs [Particles.Size002 {bIsDynamic=true, bIsEditable=true, bIsStaticSwitch=false, bIsVisible=true, type {classStructOrEnum {refPath=&Vector2f}}}]
     moduleScript {refPath=&self:HandGlow_0.NiagaraScriptSource_0.NiagaraGraph_0.NiagaraNodeAssignment_1.SetVariables_#6}
    }
    {
     bIsSetParametersModule=true
     enabled=true
     moduleName=SetVariables_#7
     inputs [Particles.Size003 {bIsDynamic=true, bIsEditable=true, bIsStaticSwitch=false, bIsVisible=true, type {classStructOrEnum {refPath=&Vector2f}}}]
     moduleScript {refPath=&self:HandGlow_0.NiagaraScriptSource_0.NiagaraGraph_0.NiagaraNodeAssignment_3.SetVariables_#7}
    }
    {
     bIsSetParametersModule=false
     enabled=true
     moduleName=ScaleColor
     inputs [
      Scale Mode {
       bIsDynamic=false
       bIsEditable=true
       bIsStaticSwitch=true
       bIsVisible=true
       type {classStructOrEnum {refPath=&ENiagaraScaleColorMode}}
      }
      Linear Color Curve {
       bIsDynamic=false
       bIsEditable=false
       bIsStaticSwitch=false
       bIsVisible=false
       type {classStructOrEnum {refPath=&NiagaraDataInterfaceColorCurve}}
      }
      ScaleRGB {bIsDynamic=false, bIsEditable=true, bIsStaticSwitch=false, bIsVisible=true, type {classStructOrEnum {refPath=&NiagaraBool}}}
      Scale RGB {bIsDynamic=true, bIsEditable=true, bIsStaticSwitch=false, bIsVisible=true, type {classStructOrEnum {refPath=&Vector3f}}}
      ScaleRGBA {bIsDynamic=false, bIsEditable=false, bIsStaticSwitch=false, bIsVisible=false, type {classStructOrEnum {refPath=&NiagaraBool}}}
      Scale RGBA {
       bIsDynamic=false
       bIsEditable=false
       bIsStaticSwitch=false
       bIsVisible=false
       type {classStructOrEnum {refPath=&Vector4f}}
      }
      Curve Index {
       bIsDynamic=false
       bIsEditable=false
       bIsStaticSwitch=false
       bIsVisible=false
       type {classStructOrEnum {refPath=&NiagaraFloat}}
      }
      ScaleA {bIsDynamic=false, bIsEditable=true, bIsStaticSwitch=false, bIsVisible=true, type {classStructOrEnum {refPath=&NiagaraBool}}}
      Scale Alpha {bIsDynamic=true, bIsEditable=true, bIsStaticSwitch=false, bIsVisible=true, type {classStructOrEnum {refPath=&NiagaraFloat}}}
      Color Value To Scale {
       bIsDynamic=false
       bIsEditable=true
       bIsStaticSwitch=false
       bIsVisible=true
       type {classStructOrEnum {refPath=&LinearColor}}
      }
     ]
     moduleScript {refPath=&ScaleColor}
    }
    {
     bIsSetParametersModule=true
     enabled=true
     moduleName=SetVariables_#8
     inputs [Particles.color002 {
       bIsDynamic=true
       bIsEditable=true
       bIsStaticSwitch=false
       bIsVisible=true
       type {classStructOrEnum {refPath=&LinearColor}}
      }]
     moduleScript {refPath=&self:HandGlow_0.NiagaraScriptSource_0.NiagaraGraph_0.NiagaraNodeAssignment_0.SetVariables_#8}
    }
    {
     bIsSetParametersModule=true
     enabled=true
     moduleName=SetVariables_#9
     inputs [Particles.Color003 {
       bIsDynamic=true
       bIsEditable=true
       bIsStaticSwitch=false
       bIsVisible=true
       type {classStructOrEnum {refPath=&LinearColor}}
      }]
     moduleScript {refPath=&self:HandGlow_0.NiagaraScriptSource_0.NiagaraGraph_0.NiagaraNodeAssignment_2.SetVariables_#9}
    }
    {
     bIsSetParametersModule=true
     enabled=true
     moduleName=SetVariables_#10
     inputs [Particles.MeshColor {
       bIsDynamic=true
       bIsEditable=true
       bIsStaticSwitch=false
       bIsVisible=true
       type {classStructOrEnum {refPath=&LinearColor}}
      }]
     moduleScript {refPath=&self:HandGlow_0.NiagaraScriptSource_0.NiagaraGraph_0.NiagaraNodeAssignment_4.SetVariables_#10}
    }
    {
     bIsSetParametersModule=false
     enabled=true
     moduleName=SolveForcesAndVelocity
     inputs [
      Clamp Velocity {bIsDynamic=false, bIsEditable=true, bIsStaticSwitch=false, bIsVisible=true, type {classStructOrEnum {refPath=&NiagaraBool}}}
      Speed Limit {bIsDynamic=false, bIsEditable=false, bIsStaticSwitch=false, bIsVisible=true, type {classStructOrEnum {refPath=&NiagaraFloat}}}
      Limit Acceleration {bIsDynamic=false, bIsEditable=true, bIsStaticSwitch=false, bIsVisible=true, type {classStructOrEnum {refPath=&NiagaraBool}}}
      Acceleration Limit {bIsDynamic=false, bIsEditable=false, bIsStaticSwitch=false, bIsVisible=true, type {classStructOrEnum {refPath=&NiagaraFloat}}}
      Manually Enable Rotational Solver {bIsDynamic=false, bIsEditable=true, bIsStaticSwitch=false, bIsVisible=true, type {classStructOrEnum {refPath=&NiagaraBool}}}
      Write to Intrinsic Properties {bIsDynamic=false, bIsEditable=true, bIsStaticSwitch=false, bIsVisible=true, type {classStructOrEnum {refPath=&NiagaraBool}}}
      Write to Presolve Properties {bIsDynamic=false, bIsEditable=true, bIsStaticSwitch=true, bIsVisible=true, type {classStructOrEnum {refPath=&NiagaraBool}}}
      DeltaTime {bIsDynamic=false, bIsEditable=true, bIsStaticSwitch=false, bIsVisible=true, type {classStructOrEnum {refPath=&NiagaraFloat}}}
      Mass {bIsDynamic=false, bIsEditable=true, bIsStaticSwitch=false, bIsVisible=true, type {classStructOrEnum {refPath=&NiagaraFloat}}}
      Position {
       bIsDynamic=false
       bIsEditable=true
       bIsStaticSwitch=false
       bIsVisible=true
       type {classStructOrEnum {refPath=&NiagaraPosition}}
      }
      Velocity {bIsDynamic=false, bIsEditable=true, bIsStaticSwitch=false, bIsVisible=true, type {classStructOrEnum {refPath=&Vector3f}}}
      PreviousVelocity {bIsDynamic=false, bIsEditable=true, bIsStaticSwitch=false, bIsVisible=true, type {classStructOrEnum {refPath=&Vector3f}}}
      Force {bIsDynamic=false, bIsEditable=true, bIsStaticSwitch=false, bIsVisible=true, type {classStructOrEnum {refPath=&Vector3f}}}
      Mesh Orientation {bIsDynamic=false, bIsEditable=false, bIsStaticSwitch=false, bIsVisible=false, type {classStructOrEnum {refPath=&Quat4f}}}
      Rotational Velocity {
       bIsDynamic=false
       bIsEditable=false
       bIsStaticSwitch=false
       bIsVisible=false
       type {classStructOrEnum {refPath=&Vector3f}}
      }
      Rotational Force {
       bIsDynamic=false
       bIsEditable=false
       bIsStaticSwitch=false
       bIsVisible=false
       type {classStructOrEnum {refPath=&Vector3f}}
      }
      Rotational Inertia {
       bIsDynamic=false
       bIsEditable=false
       bIsStaticSwitch=false
       bIsVisible=false
       type {classStructOrEnum {refPath=&NiagaraFloat}}
      }
     ]
     moduleScript {refPath=&SolveForcesAndVelocity}
    }
    {
     bIsSetParametersModule=true
     enabled=true
     moduleName=SetVariables_#11
     inputs [Particles.Vector 4 {bIsDynamic=false, bIsEditable=true, bIsStaticSwitch=false, bIsVisible=true, type {classStructOrEnum {refPath=&Vector4f}}}]
     moduleScript {refPath=&self:HandGlow_0.NiagaraScriptSource_0.NiagaraGraph_0.NiagaraNodeAssignment_5.SetVariables_#11}
    }
    {
     bIsSetParametersModule=false
     enabled=true
     moduleName=ScaleColor001
     inputs [
      Scale Mode {bIsDynamic=false, bIsEditable=true, bIsStaticSwitch=true, bIsVisible=true, type {classStructOrEnum {refPath=&NiagaraInt32}}}
      Linear Color Curve {
       bIsDynamic=false
       bIsEditable=false
       bIsStaticSwitch=false
       bIsVisible=false
       type {classStructOrEnum {refPath=&NiagaraDataInterfaceColorCurve}}
      }
      ScaleRGB {bIsDynamic=false, bIsEditable=true, bIsStaticSwitch=false, bIsVisible=true, type {classStructOrEnum {refPath=&NiagaraBool}}}
      Scale RGB {bIsDynamic=true, bIsEditable=true, bIsStaticSwitch=false, bIsVisible=true, type {classStructOrEnum {refPath=&Vector3f}}}
      ScaleRGBA {bIsDynamic=false, bIsEditable=false, bIsStaticSwitch=false, bIsVisible=false, type {classStructOrEnum {refPath=&NiagaraBool}}}
      Scale RGBA {
       bIsDynamic=false
       bIsEditable=false
       bIsStaticSwitch=false
       bIsVisible=false
       type {classStructOrEnum {refPath=&Vector4f}}
      }
      Curve Index {
       bIsDynamic=false
       bIsEditable=false
       bIsStaticSwitch=false
       bIsVisible=false
       type {classStructOrEnum {refPath=&NiagaraFloat}}
      }
      ScaleA {bIsDynamic=false, bIsEditable=true, bIsStaticSwitch=false, bIsVisible=true, type {classStructOrEnum {refPath=&NiagaraBool}}}
      Scale Alpha {bIsDynamic=false, bIsEditable=false, bIsStaticSwitch=false, bIsVisible=true, type {classStructOrEnum {refPath=&NiagaraFloat}}}
      Color Value To Scale {
       bIsDynamic=false
       bIsEditable=true
       bIsStaticSwitch=false
       bIsVisible=true
       type {classStructOrEnum {refPath=&LinearColor}}
      }
      Color Scale {
       bIsDynamic=false
       bIsEditable=false
       bIsStaticSwitch=false
       bIsVisible=false
       type {classStructOrEnum {refPath=&LinearColor}}
      }
     ]
     moduleScript {refPath=&ScaleColor}
    }
   ]
  }
  rendererClasses [{refPath=&NiagaraSpriteRendererProperties}]
  renderers [
   {rendererIndex=0, rendererClass {refPath=&NiagaraSpriteRendererProperties}}
   {rendererIndex=1, rendererClass {refPath=&NiagaraSpriteRendererProperties}}
   {rendererIndex=2, rendererClass {refPath=&NiagaraSpriteRendererProperties}}
  ]
 }
 Head {
  bEnabled=false
  emitterSpawnScript {scriptName=EmitterSpawnScript, modules []}
  emitterUpdateScript {
   scriptName=EmitterUpdateScript
   modules [
    {
     bIsSetParametersModule=false
     enabled=true
     moduleName=EmitterState
     inputs [
      Life Cycle Mode {
       bIsDynamic=false
       bIsEditable=true
       bIsStaticSwitch=true
       bIsVisible=true
       type {classStructOrEnum {refPath=&ENiagaraEmitterLifeCycleMode}}
      }
      Inactive Response {
       bIsDynamic=false
       bIsEditable=true
       bIsStaticSwitch=true
       bIsVisible=true
       type {classStructOrEnum {refPath=&ENiagaraInactiveMode}}
      }
      Loop Behavior {
       bIsDynamic=false
       bIsEditable=true
       bIsStaticSwitch=true
       bIsVisible=true
       type {classStructOrEnum {refPath=&ENiagara_EmitterStateOptions}}
      }
      Loop Duration Mode {
       bIsDynamic=false
       bIsEditable=true
       bIsStaticSwitch=true
       bIsVisible=true
       type {classStructOrEnum {refPath=&ENiagara_InfiniteLoopDuration}}
      }
      Loop Count {
       bIsDynamic=false
       bIsEditable=false
       bIsStaticSwitch=false
       bIsVisible=false
       type {classStructOrEnum {refPath=&NiagaraInt32}}
      }
      Loop Duration {bIsDynamic=false, bIsEditable=true, bIsStaticSwitch=false, bIsVisible=true, type {classStructOrEnum {refPath=&NiagaraFloat}}}
      Recalculate Duration Each Loop {bIsDynamic=false, bIsEditable=false, bIsStaticSwitch=false, bIsVisible=false, type {classStructOrEnum {refPath=&NiagaraBool}}}
      UseLoopDelay {bIsDynamic=false, bIsEditable=true, bIsStaticSwitch=false, bIsVisible=true, type {classStructOrEnum {refPath=&NiagaraBool}}}
      Loop Delay {bIsDynamic=false, bIsEditable=false, bIsStaticSwitch=false, bIsVisible=true, type {classStructOrEnum {refPath=&NiagaraFloat}}}
      Delay First Loop Only {bIsDynamic=false, bIsEditable=false, bIsStaticSwitch=false, bIsVisible=false, type {classStructOrEnum {refPath=&NiagaraBool}}}
      Scalability Mode {
       bIsDynamic=false
       bIsEditable=true
       bIsStaticSwitch=true
       bIsVisible=true
       type {classStructOrEnum {refPath=&ENiagaraEmitterScalabilityMode_Limited}}
      }
      Enable Distance Culling {bIsDynamic=false, bIsEditable=false, bIsStaticSwitch=true, bIsVisible=false, type {classStructOrEnum {refPath=&NiagaraBool}}}
      MinDistance {
       bIsDynamic=false
       bIsEditable=false
       bIsStaticSwitch=false
       bIsVisible=false
       type {classStructOrEnum {refPath=&NiagaraFloat}}
      }
      Min Distance Response {
       bIsDynamic=false
       bIsEditable=false
       bIsStaticSwitch=false
       bIsVisible=false
       type {classStructOrEnum {refPath=&ENiagaraExecutionStateManagement}}
      }
      MaxDistance {
       bIsDynamic=false
       bIsEditable=false
       bIsStaticSwitch=false
       bIsVisible=false
       type {classStructOrEnum {refPath=&NiagaraFloat}}
      }
      Max Distance Response {
       bIsDynamic=false
       bIsEditable=false
       bIsStaticSwitch=false
       bIsVisible=false
       type {classStructOrEnum {refPath=&ENiagaraExecutionStateManagement}}
      }
      Scale Spawn Count {bIsDynamic=false, bIsEditable=false, bIsStaticSwitch=true, bIsVisible=false, type {classStructOrEnum {refPath=&NiagaraBool}}}
      Spawn Count Scale {
       bIsDynamic=false
       bIsEditable=false
       bIsStaticSwitch=false
       bIsVisible=false
       type {classStructOrEnum {refPath=&NiagaraFloat}}
      }
      Spawn Count Scale By Distance {
       bIsDynamic=false
       bIsEditable=false
       bIsStaticSwitch=false
       bIsVisible=false
       type {classStructOrEnum {refPath=&NiagaraDataInterfaceCurve}}
      }
      Enable Visibility Culling {bIsDynamic=false, bIsEditable=false, bIsStaticSwitch=true, bIsVisible=false, type {classStructOrEnum {refPath=&NiagaraBool}}}
      Visibility Cull Response {
       bIsDynamic=false
       bIsEditable=false
       bIsStaticSwitch=false
       bIsVisible=false
       type {classStructOrEnum {refPath=&ENiagaraExecutionStateManagement}}
      }
      Visibility Cull Delay {
       bIsDynamic=false
       bIsEditable=false
       bIsStaticSwitch=false
       bIsVisible=false
       type {classStructOrEnum {refPath=&NiagaraFloat}}
      }
      Reset Age On Awaken {bIsDynamic=false, bIsEditable=false, bIsStaticSwitch=false, bIsVisible=false, type {classStructOrEnum {refPath=&NiagaraBool}}}
     ]
     moduleScript {refPath=&EmitterState}
    }
    {
     bIsSetParametersModule=false
     enabled=true
     moduleName=SpawnBurst_Instantaneous
     inputs [
      Spawn Count {bIsDynamic=false, bIsEditable=true, bIsStaticSwitch=false, bIsVisible=true, type {classStructOrEnum {refPath=&NiagaraInt32}}}
      Spawn Time {bIsDynamic=false, bIsEditable=true, bIsStaticSwitch=false, bIsVisible=true, type {classStructOrEnum {refPath=&NiagaraFloat}}}
      Use Spawn Probability {bIsDynamic=false, bIsEditable=true, bIsStaticSwitch=false, bIsVisible=true, type {classStructOrEnum {refPath=&NiagaraBool}}}
      Spawn Probability {bIsDynamic=false, bIsEditable=false, bIsStaticSwitch=false, bIsVisible=true, type {classStructOrEnum {refPath=&NiagaraFloat}}}
      UseLoopCountLimit {bIsDynamic=false, bIsEditable=true, bIsStaticSwitch=false, bIsVisible=true, type {classStructOrEnum {refPath=&NiagaraBool}}}
      Loop Count Limit {bIsDynamic=false, bIsEditable=false, bIsStaticSwitch=false, bIsVisible=true, type {classStructOrEnum {refPath=&NiagaraInt32}}}
      Spawn Group {bIsDynamic=false, bIsEditable=true, bIsStaticSwitch=false, bIsVisible=true, type {classStructOrEnum {refPath=&NiagaraInt32}}}
      Age {bIsDynamic=false, bIsEditable=true, bIsStaticSwitch=false, bIsVisible=true, type {classStructOrEnum {refPath=&NiagaraFloat}}}
     ]
     moduleScript {refPath=&SpawnBurst_Instantaneous}
    }
   ]
  }
  particleSpawnScript {
   scriptName=ParticleSpawnScript
   modules [
    {
     bIsSetParametersModule=false
     enabled=true
     moduleName=InitializeParticle
     inputs [
      Lifetime Mode {
       bIsDynamic=false
       bIsEditable=true
       bIsStaticSwitch=true
       bIsVisible=true
       type {classStructOrEnum {refPath=&ENiagara_LifetimeMode}}
      }
      Lifetime Min {
       bIsDynamic=false
       bIsEditable=false
       bIsStaticSwitch=false
       bIsVisible=false
       type {classStructOrEnum {refPath=&NiagaraFloat}}
      }
      Lifetime {bIsDynamic=false, bIsEditable=true, bIsStaticSwitch=false, bIsVisible=true, type {classStructOrEnum {refPath=&NiagaraFloat}}}
      Lifetime Max {
       bIsDynamic=false
       bIsEditable=false
       bIsStaticSwitch=false
       bIsVisible=false
       type {classStructOrEnum {refPath=&NiagaraFloat}}
      }
      Lifetime Randomness Mode {
       bIsDynamic=false
       bIsEditable=false
       bIsStaticSwitch=true
       bIsVisible=false
       type {classStructOrEnum {refPath=&ENiagaraRandomnessMode}}
      }
      Color Mode {
       bIsDynamic=false
       bIsEditable=true
       bIsStaticSwitch=true
       bIsVisible=true
       type {classStructOrEnum {refPath=&ENiagara_ColorInitializationMode}}
      }
      Color Minimum {
       bIsDynamic=false
       bIsEditable=false
       bIsStaticSwitch=false
       bIsVisible=false
       type {classStructOrEnum {refPath=&LinearColor}}
      }
      Color {
       bIsDynamic=false
       bIsEditable=true
       bIsStaticSwitch=false
       bIsVisible=true
       type {classStructOrEnum {refPath=&LinearColor}}
      }
      AdjustHue {bIsDynamic=false, bIsEditable=false, bIsStaticSwitch=false, bIsVisible=false, type {classStructOrEnum {refPath=&NiagaraBool}}}
      Hue Shift Range {
       bIsDynamic=false
       bIsEditable=false
       bIsStaticSwitch=false
       bIsVisible=false
       type {classStructOrEnum {refPath=&Vector2f}}
      }
      Color Maximum {
       bIsDynamic=false
       bIsEditable=false
       bIsStaticSwitch=false
       bIsVisible=false
       type {classStructOrEnum {refPath=&LinearColor}}
      }
      AdjustSaturation {bIsDynamic=false, bIsEditable=false, bIsStaticSwitch=false, bIsVisible=false, type {classStructOrEnum {refPath=&NiagaraBool}}}
      Saturation Range {
       bIsDynamic=false
       bIsEditable=false
       bIsStaticSwitch=false
       bIsVisible=false
       type {classStructOrEnum {refPath=&Vector2f}}
      }
      AdjustValue {bIsDynamic=false, bIsEditable=false, bIsStaticSwitch=false, bIsVisible=false, type {classStructOrEnum {refPath=&NiagaraBool}}}
      Value Range {
       bIsDynamic=false
       bIsEditable=false
       bIsStaticSwitch=false
       bIsVisible=false
       type {classStructOrEnum {refPath=&Vector2f}}
      }
      AdjustAlpha {bIsDynamic=false, bIsEditable=false, bIsStaticSwitch=false, bIsVisible=false, type {classStructOrEnum {refPath=&NiagaraBool}}}
      Alpha Scale Range {
       bIsDynamic=false
       bIsEditable=false
       bIsStaticSwitch=false
       bIsVisible=false
       type {classStructOrEnum {refPath=&Vector2f}}
      }
      Color Channel Mode {
       bIsDynamic=false
       bIsEditable=false
       bIsStaticSwitch=true
       bIsVisible=false
       type {classStructOrEnum {refPath=&ENiagaraChannelCorrelation}}
      }
      Color Randomness Mode {
       bIsDynamic=false
       bIsEditable=false
       bIsStaticSwitch=true
       bIsVisible=false
       type {classStructOrEnum {refPath=&ENiagaraRandomnessMode}}
      }
      Position Mode {
       bIsDynamic=false
       bIsEditable=true
       bIsStaticSwitch=true
       bIsVisible=true
       type {classStructOrEnum {refPath=&ENiagara_PositionInitializationMode}}
      }
      Position {
       bIsDynamic=false
       bIsEditable=false
       bIsStaticSwitch=false
       bIsVisible=false
       type {classStructOrEnum {refPath=&NiagaraPosition}}
      }
      UsePositionOffset {bIsDynamic=false, bIsEditable=true, bIsStaticSwitch=false, bIsVisible=true, type {classStructOrEnum {refPath=&NiagaraBool}}}
      Position Offset {bIsDynamic=false, bIsEditable=false, bIsStaticSwitch=false, bIsVisible=true, type {classStructOrEnum {refPath=&Vector3f}}}
      Position Offset Coordinate Space {
       bIsDynamic=false
       bIsEditable=false
       bIsStaticSwitch=false
       bIsVisible=false
       type {classStructOrEnum {refPath=&ENiagaraCoordinateSpace}}
      }
      Mass Mode {
       bIsDynamic=false
       bIsEditable=true
       bIsStaticSwitch=true
       bIsVisible=true
       type {classStructOrEnum {refPath=&ENiagara_MassInitializationMode}}
      }
      Mass Min {
       bIsDynamic=false
       bIsEditable=false
       bIsStaticSwitch=false
       bIsVisible=false
       type {classStructOrEnum {refPath=&NiagaraFloat}}
      }
      Mass {
       bIsDynamic=false
       bIsEditable=false
       bIsStaticSwitch=false
       bIsVisible=false
       type {classStructOrEnum {refPath=&NiagaraFloat}}
      }
      Mass Max {
       bIsDynamic=false
       bIsEditable=false
       bIsStaticSwitch=false
       bIsVisible=false
       type {classStructOrEnum {refPath=&NiagaraFloat}}
      }
      Mass Randomness Mode {
       bIsDynamic=false
       bIsEditable=false
       bIsStaticSwitch=true
       bIsVisible=false
       type {classStructOrEnum {refPath=&ENiagaraRandomnessMode}}
      }
      Write Material Random {bIsDynamic=false, bIsEditable=true, bIsStaticSwitch=false, bIsVisible=true, type {classStructOrEnum {refPath=&NiagaraBool}}}
      Material Random {bIsDynamic=false, bIsEditable=true, bIsStaticSwitch=false, bIsVisible=true, type {classStructOrEnum {refPath=&NiagaraFloat}}}
      Sprite Size Mode {
       bIsDynamic=false
       bIsEditable=true
       bIsStaticSwitch=true
       bIsVisible=true
       type {classStructOrEnum {refPath=&ENiagara_SizeScaleMode}}
      }
      Sprite Size {
       bIsDynamic=false
       bIsEditable=false
       bIsStaticSwitch=false
       bIsVisible=false
       type {classStructOrEnum {refPath=&Vector2f}}
      }
      Uniform Sprite Size {bIsDynamic=false, bIsEditable=true, bIsStaticSwitch=false, bIsVisible=true, type {classStructOrEnum {refPath=&NiagaraFloat}}}
      Uniform Sprite Size Min {
       bIsDynamic=false
       bIsEditable=false
       bIsStaticSwitch=false
       bIsVisible=false
       type {classStructOrEnum {refPath=&NiagaraFloat}}
      }
      Sprite Size Min {
       bIsDynamic=false
       bIsEditable=false
       bIsStaticSwitch=false
       bIsVisible=false
       type {classStructOrEnum {refPath=&Vector2f}}
      }
      Sprite Size Max {
       bIsDynamic=false
       bIsEditable=false
       bIsStaticSwitch=false
       bIsVisible=false
       type {classStructOrEnum {refPath=&Vector2f}}
      }
      Uniform Sprite Size Max {
       bIsDynamic=false
       bIsEditable=false
       bIsStaticSwitch=false
       bIsVisible=false
       type {classStructOrEnum {refPath=&NiagaraFloat}}
      }
      Sprite Size Randomness Mode {
       bIsDynamic=false
       bIsEditable=false
       bIsStaticSwitch=true
       bIsVisible=false
       type {classStructOrEnum {refPath=&ENiagaraRandomnessMode}}
      }
      Sprite Rotation Mode {
       bIsDynamic=false
       bIsEditable=true
       bIsStaticSwitch=true
       bIsVisible=true
       type {classStructOrEnum {refPath=&ENiagara_SpriteRotationMode}}
      }
      Sprite Rotation Angle Min {
       bIsDynamic=false
       bIsEditable=false
       bIsStaticSwitch=false
       bIsVisible=false
       type {classStructOrEnum {refPath=&NiagaraFloat}}
      }
      Sprite Rotation Angle {
       bIsDynamic=false
       bIsEditable=false
       bIsStaticSwitch=false
       bIsVisible=false
       type {classStructOrEnum {refPath=&NiagaraFloat}}
      }
      Sprite Rotation Angle Max {
       bIsDynamic=false
       bIsEditable=false
       bIsStaticSwitch=false
       bIsVisible=false
       type {classStructOrEnum {refPath=&NiagaraFloat}}
      }
      Sprite Rotation Randomness Mode {
       bIsDynamic=false
       bIsEditable=false
       bIsStaticSwitch=true
       bIsVisible=false
       type {classStructOrEnum {refPath=&ENiagaraRandomnessMode}}
      }
      Sprite UV Mode {
       bIsDynamic=false
       bIsEditable=true
       bIsStaticSwitch=true
       bIsVisible=true
       type {classStructOrEnum {refPath=&ENiagara_UVFlippingMode}}
      }
      Sprite UV Randomness Mode {
       bIsDynamic=false
       bIsEditable=false
       bIsStaticSwitch=true
       bIsVisible=false
       type {classStructOrEnum {refPath=&ENiagaraRandomnessMode}}
      }
      Sprite UV Scale {
       bIsDynamic=false
       bIsEditable=false
       bIsStaticSwitch=false
       bIsVisible=false
       type {classStructOrEnum {refPath=&Vector2f}}
      }
      Mesh Scale Mode {
       bIsDynamic=false
       bIsEditable=true
       bIsStaticSwitch=true
       bIsVisible=true
       type {classStructOrEnum {refPath=&ENiagara_SizeScaleMode}}
      }
      Mesh Uniform Scale Min {
       bIsDynamic=false
       bIsEditable=false
       bIsStaticSwitch=false
       bIsVisible=false
       type {classStructOrEnum {refPath=&NiagaraFloat}}
      }
      Mesh Uniform Scale {
       bIsDynamic=false
       bIsEditable=false
       bIsStaticSwitch=false
       bIsVisible=false
       type {classStructOrEnum {refPath=&NiagaraFloat}}
      }
      Mesh Scale Min {
       bIsDynamic=false
       bIsEditable=false
       bIsStaticSwitch=false
       bIsVisible=false
       type {classStructOrEnum {refPath=&Vector3f}}
      }
      Mesh Scale {bIsDynamic=false, bIsEditable=true, bIsStaticSwitch=false, bIsVisible=true, type {classStructOrEnum {refPath=&Vector3f}}}
      Mesh Scale Max {
       bIsDynamic=false
       bIsEditable=false
       bIsStaticSwitch=false
       bIsVisible=false
       type {classStructOrEnum {refPath=&Vector3f}}
      }
      Mesh Uniform Scale Max {
       bIsDynamic=false
       bIsEditable=false
       bIsStaticSwitch=false
       bIsVisible=false
       type {classStructOrEnum {refPath=&NiagaraFloat}}
      }
      Mesh Scale Randomness Mode {
       bIsDynamic=false
       bIsEditable=false
       bIsStaticSwitch=true
       bIsVisible=false
       type {classStructOrEnum {refPath=&ENiagaraRandomnessMode}}
      }
      Mesh Renderer Array Visibility Mode {
       bIsDynamic=false
       bIsEditable=true
       bIsStaticSwitch=true
       bIsVisible=true
       type {classStructOrEnum {refPath=&ENiagara_UnsetDirectSetRandom}}
      }
      Mesh Renderer Mesh Index {
       bIsDynamic=false
       bIsEditable=false
       bIsStaticSwitch=false
       bIsVisible=false
       type {classStructOrEnum {refPath=&NiagaraInt32}}
      }
      Wrap Index By Mesh Count {bIsDynamic=false, bIsEditable=false, bIsStaticSwitch=true, bIsVisible=false, type {classStructOrEnum {refPath=&NiagaraBool}}}
      Mesh Renderer Info {
       bIsDynamic=false
       bIsEditable=false
       bIsStaticSwitch=false
       bIsVisible=false
       type {classStructOrEnum {refPath=&NiagaraDataInterfaceMeshRendererInfo}}
      }
      Mesh Renderer Random Seed {
       bIsDynamic=false
       bIsEditable=false
       bIsStaticSwitch=false
       bIsVisible=false
       type {classStructOrEnum {refPath=&NiagaraInt32}}
      }
      Ribbon Width Mode {
       bIsDynamic=false
       bIsEditable=true
       bIsStaticSwitch=true
       bIsVisible=true
       type {classStructOrEnum {refPath=&ENiagara_UnsetDirectSet}}
      }
      Ribbon Width {
       bIsDynamic=false
       bIsEditable=false
       bIsStaticSwitch=false
       bIsVisible=false
       type {classStructOrEnum {refPath=&NiagaraFloat}}
      }
      Ribbon Facing Mode {
       bIsDynamic=false
       bIsEditable=true
       bIsStaticSwitch=true
       bIsVisible=true
       type {classStructOrEnum {refPath=&ENiagara_UnsetDirectSet}}
      }
      Ribbon Facing Vector {
       bIsDynamic=false
       bIsEditable=false
       bIsStaticSwitch=false
       bIsVisible=false
       type {classStructOrEnum {refPath=&Vector3f}}
      }
      Ribbon Twist Mode {
       bIsDynamic=false
       bIsEditable=true
       bIsStaticSwitch=true
       bIsVisible=true
       type {classStructOrEnum {refPath=&ENiagara_UnsetDirectSet}}
      }
      "Ribbon Twist " {
       bIsDynamic=false
       bIsEditable=false
       bIsStaticSwitch=false
       bIsVisible=false
       type {classStructOrEnum {refPath=&NiagaraFloat}}
      }
     ]
     moduleScript {refPath=&InitializeParticle}
    }
    {
     bIsSetParametersModule=false
     enabled=true
     moduleName=InitialMeshOrientation
     inputs [
      Mesh Orientation Mode {
       bIsDynamic=false
       bIsEditable=true
       bIsStaticSwitch=true
       bIsVisible=true
       type {classStructOrEnum {refPath=&ENiagara_MeshOrientationOptions}}
      }
      Orientation Quaternion {bIsDynamic=false, bIsEditable=false, bIsStaticSwitch=false, bIsVisible=false, type {classStructOrEnum {refPath=&Quat4f}}}
      Orientation Matrix {
       bIsDynamic=false
       bIsEditable=false
       bIsStaticSwitch=false
       bIsVisible=false
       type {classStructOrEnum {refPath=&NiagaraMatrix}}
      }
      Orientation Vector {bIsDynamic=false, bIsEditable=true, bIsStaticSwitch=false, bIsVisible=true, type {classStructOrEnum {refPath=&Vector3f}}}
      Orientation Coordinate Space {
       bIsDynamic=false
       bIsEditable=true
       bIsStaticSwitch=true
       bIsVisible=true
       type {classStructOrEnum {refPath=&ENiagaraCoordinateSpace}}
      }
      Mesh Axis To Orient {bIsDynamic=false, bIsEditable=true, bIsStaticSwitch=false, bIsVisible=true, type {classStructOrEnum {refPath=&Vector3f}}}
      Random Seed {
       bIsDynamic=false
       bIsEditable=false
       bIsStaticSwitch=false
       bIsVisible=false
       type {classStructOrEnum {refPath=&NiagaraInt32}}
      }
      Use Rotation Vector {bIsDynamic=false, bIsEditable=true, bIsStaticSwitch=false, bIsVisible=true, type {classStructOrEnum {refPath=&NiagaraBool}}}
      Rotation {bIsDynamic=false, bIsEditable=false, bIsStaticSwitch=false, bIsVisible=true, type {classStructOrEnum {refPath=&Vector3f}}}
      Rotation Coordinate Space {
       bIsDynamic=false
       bIsEditable=false
       bIsStaticSwitch=true
       bIsVisible=true
       type {classStructOrEnum {refPath=&ENiagaraMeshTransforms}}
      }
     ]
     moduleScript {refPath=&InitialMeshOrientation}
    }
   ]
  }
  particleUpdateScript {
   scriptName=ParticleUpdateScript
   modules [
    {
     bIsSetParametersModule=false
     enabled=true
     moduleName=ParticleState
     inputs [
      Kill Particles When Lifetime Has Elapsed {bIsDynamic=false, bIsEditable=true, bIsStaticSwitch=true, bIsVisible=true, type {classStructOrEnum {refPath=&NiagaraBool}}}
      Loop Particles Lifetime {bIsDynamic=false, bIsEditable=true, bIsStaticSwitch=false, bIsVisible=true, type {classStructOrEnum {refPath=&NiagaraBool}}}
      Let Infinitely Lived Particles Die When Emitter Deactivates {bIsDynamic=false, bIsEditable=true, bIsStaticSwitch=false, bIsVisible=true, type {classStructOrEnum {refPath=&NiagaraBool}}}
      Lifetime {bIsDynamic=false, bIsEditable=true, bIsStaticSwitch=false, bIsVisible=true, type {classStructOrEnum {refPath=&NiagaraFloat}}}
      DeltaTime {bIsDynamic=false, bIsEditable=true, bIsStaticSwitch=false, bIsVisible=true, type {classStructOrEnum {refPath=&NiagaraFloat}}}
     ]
     moduleScript {refPath=&ParticleState}
    }
    {
     bIsSetParametersModule=false
     enabled=true
     moduleName=ScaleColor
     inputs [
      Scale Mode {
       bIsDynamic=false
       bIsEditable=true
       bIsStaticSwitch=true
       bIsVisible=true
       type {classStructOrEnum {refPath=&ENiagaraScaleColorMode}}
      }
      Linear Color Curve {
       bIsDynamic=false
       bIsEditable=false
       bIsStaticSwitch=false
       bIsVisible=false
       type {classStructOrEnum {refPath=&NiagaraDataInterfaceColorCurve}}
      }
      ScaleRGB {bIsDynamic=false, bIsEditable=true, bIsStaticSwitch=false, bIsVisible=true, type {classStructOrEnum {refPath=&NiagaraBool}}}
      Scale RGB {bIsDynamic=false, bIsEditable=true, bIsStaticSwitch=false, bIsVisible=true, type {classStructOrEnum {refPath=&Vector3f}}}
      ScaleRGBA {bIsDynamic=false, bIsEditable=false, bIsStaticSwitch=false, bIsVisible=false, type {classStructOrEnum {refPath=&NiagaraBool}}}
      Scale RGBA {
       bIsDynamic=false
       bIsEditable=false
       bIsStaticSwitch=false
       bIsVisible=false
       type {classStructOrEnum {refPath=&Vector4f}}
      }
      Curve Index {
       bIsDynamic=false
       bIsEditable=false
       bIsStaticSwitch=false
       bIsVisible=false
       type {classStructOrEnum {refPath=&NiagaraFloat}}
      }
      ScaleA {bIsDynamic=false, bIsEditable=true, bIsStaticSwitch=false, bIsVisible=true, type {classStructOrEnum {refPath=&NiagaraBool}}}
      Scale Alpha {bIsDynamic=true, bIsEditable=true, bIsStaticSwitch=false, bIsVisible=true, type {classStructOrEnum {refPath=&NiagaraFloat}}}
      Color Value To Scale {
       bIsDynamic=false
       bIsEditable=true
       bIsStaticSwitch=false
       bIsVisible=true
       type {classStructOrEnum {refPath=&LinearColor}}
      }
     ]
     moduleScript {refPath=&ScaleColor}
    }
    {
     bIsSetParametersModule=false
     enabled=true
     moduleName=ScaleSpriteSize
     inputs [
      Scale Sprite Size Mode {
       bIsDynamic=false
       bIsEditable=true
       bIsStaticSwitch=true
       bIsVisible=true
       type {classStructOrEnum {refPath=&ENiagara_ScaleSpriteSize}}
      }
      Uniform Curve Sprite Scale {
       bIsDynamic=false
       bIsEditable=true
       bIsStaticSwitch=false
       bIsVisible=true
       type {classStructOrEnum {refPath=&NiagaraDataInterfaceCurve}}
      }
      Non-Uniform Curve Sprite Scale {
       bIsDynamic=false
       bIsEditable=false
       bIsStaticSwitch=false
       bIsVisible=false
       type {classStructOrEnum {refPath=&NiagaraDataInterfaceVector2DCurve}}
      }
      Uniform Curve Index {bIsDynamic=false, bIsEditable=true, bIsStaticSwitch=false, bIsVisible=true, type {classStructOrEnum {refPath=&NiagaraFloat}}}
      Non-Uniform Curve Index {
       bIsDynamic=false
       bIsEditable=false
       bIsStaticSwitch=false
       bIsVisible=false
       type {classStructOrEnum {refPath=&NiagaraFloat}}
      }
      Non-Uniform Curve Scale {
       bIsDynamic=false
       bIsEditable=false
       bIsStaticSwitch=false
       bIsVisible=false
       type {classStructOrEnum {refPath=&Vector2f}}
      }
      Uniform Curve Scale {bIsDynamic=false, bIsEditable=true, bIsStaticSwitch=false, bIsVisible=true, type {classStructOrEnum {refPath=&NiagaraFloat}}}
      Scale Factor {
       bIsDynamic=false
       bIsEditable=false
       bIsStaticSwitch=false
       bIsVisible=false
       type {classStructOrEnum {refPath=&Vector2f}}
      }
      Uniform Scale Factor {
       bIsDynamic=false
       bIsEditable=false
       bIsStaticSwitch=false
       bIsVisible=false
       type {classStructOrEnum {refPath=&NiagaraFloat}}
      }
      Initial Sprite Size {bIsDynamic=false, bIsEditable=true, bIsStaticSwitch=false, bIsVisible=true, type {classStructOrEnum {refPath=&Vector2f}}}
     ]
     moduleScript {refPath=&ScaleSpriteSize}
    }
   ]
  }
  rendererClasses [{refPath=&NiagaraLightRendererProperties}, {refPath=&NiagaraSpriteRendererProperties}]
  renderers [
   {rendererIndex=0, rendererClass {refPath=&NiagaraLightRendererProperties}}
   {rendererIndex=1, rendererClass {refPath=&NiagaraSpriteRendererProperties}}
  ]
 }
 IceBlast {
  emitterSpawnScript {scriptName=EmitterSpawnScript, modules []}
  emitterUpdateScript {
   scriptName=EmitterUpdateScript
   modules [
    {
     bIsSetParametersModule=false
     enabled=true
     moduleName=EmitterState
     inputs [
      Life Cycle Mode {
       bIsDynamic=false
       bIsEditable=true
       bIsStaticSwitch=true
       bIsVisible=true
       type {classStructOrEnum {refPath=&ENiagaraEmitterLifeCycleMode}}
      }
      Inactive Response {
       bIsDynamic=false
       bIsEditable=true
       bIsStaticSwitch=true
       bIsVisible=true
       type {classStructOrEnum {refPath=&ENiagaraInactiveMode}}
      }
      Loop Behavior {
       bIsDynamic=false
       bIsEditable=true
       bIsStaticSwitch=true
       bIsVisible=true
       type {classStructOrEnum {refPath=&ENiagara_EmitterStateOptions}}
      }
      Loop Duration Mode {
       bIsDynamic=false
       bIsEditable=true
       bIsStaticSwitch=true
       bIsVisible=true
       type {classStructOrEnum {refPath=&ENiagara_InfiniteLoopDuration}}
      }
      Loop Count {
       bIsDynamic=false
       bIsEditable=false
       bIsStaticSwitch=false
       bIsVisible=false
       type {classStructOrEnum {refPath=&NiagaraInt32}}
      }
      Loop Duration {bIsDynamic=false, bIsEditable=true, bIsStaticSwitch=false, bIsVisible=true, type {classStructOrEnum {refPath=&NiagaraFloat}}}
      Recalculate Duration Each Loop {bIsDynamic=false, bIsEditable=false, bIsStaticSwitch=false, bIsVisible=false, type {classStructOrEnum {refPath=&NiagaraBool}}}
      UseLoopDelay {bIsDynamic=false, bIsEditable=true, bIsStaticSwitch=false, bIsVisible=true, type {classStructOrEnum {refPath=&NiagaraBool}}}
      Loop Delay {bIsDynamic=false, bIsEditable=false, bIsStaticSwitch=false, bIsVisible=true, type {classStructOrEnum {refPath=&NiagaraFloat}}}
      Delay First Loop Only {bIsDynamic=false, bIsEditable=false, bIsStaticSwitch=false, bIsVisible=false, type {classStructOrEnum {refPath=&NiagaraBool}}}
      Scalability Mode {
       bIsDynamic=false
       bIsEditable=true
       bIsStaticSwitch=true
       bIsVisible=true
       type {classStructOrEnum {refPath=&ENiagaraEmitterScalabilityMode_Limited}}
      }
      Enable Distance Culling {bIsDynamic=false, bIsEditable=false, bIsStaticSwitch=true, bIsVisible=false, type {classStructOrEnum {refPath=&NiagaraBool}}}
      MinDistance {
       bIsDynamic=false
       bIsEditable=false
       bIsStaticSwitch=false
       bIsVisible=false
       type {classStructOrEnum {refPath=&NiagaraFloat}}
      }
      Min Distance Response {
       bIsDynamic=false
       bIsEditable=false
       bIsStaticSwitch=false
       bIsVisible=false
       type {classStructOrEnum {refPath=&ENiagaraExecutionStateManagement}}
      }
      MaxDistance {
       bIsDynamic=false
       bIsEditable=false
       bIsStaticSwitch=false
       bIsVisible=false
       type {classStructOrEnum {refPath=&NiagaraFloat}}
      }
      Max Distance Response {
       bIsDynamic=false
       bIsEditable=false
       bIsStaticSwitch=false
       bIsVisible=false
       type {classStructOrEnum {refPath=&ENiagaraExecutionStateManagement}}
      }
      Scale Spawn Count {bIsDynamic=false, bIsEditable=false, bIsStaticSwitch=true, bIsVisible=false, type {classStructOrEnum {refPath=&NiagaraBool}}}
      Spawn Count Scale {
       bIsDynamic=false
       bIsEditable=false
       bIsStaticSwitch=false
       bIsVisible=false
       type {classStructOrEnum {refPath=&NiagaraFloat}}
      }
      Spawn Count Scale By Distance {
       bIsDynamic=false
       bIsEditable=false
       bIsStaticSwitch=false
       bIsVisible=false
       type {classStructOrEnum {refPath=&NiagaraDataInterfaceCurve}}
      }
      Enable Visibility Culling {bIsDynamic=false, bIsEditable=false, bIsStaticSwitch=true, bIsVisible=false, type {classStructOrEnum {refPath=&NiagaraBool}}}
      Visibility Cull Response {
       bIsDynamic=false
       bIsEditable=false
       bIsStaticSwitch=false
       bIsVisible=false
       type {classStructOrEnum {refPath=&ENiagaraExecutionStateManagement}}
      }
      Visibility Cull Delay {
       bIsDynamic=false
       bIsEditable=false
       bIsStaticSwitch=false
       bIsVisible=false
       type {classStructOrEnum {refPath=&NiagaraFloat}}
      }
      Reset Age On Awaken {bIsDynamic=false, bIsEditable=false, bIsStaticSwitch=false, bIsVisible=false, type {classStructOrEnum {refPath=&NiagaraBool}}}
     ]
     moduleScript {refPath=&EmitterState}
    }
    {
     bIsSetParametersModule=false
     enabled=true
     moduleName=SpawnBurst_Instantaneous
     inputs [
      Spawn Count {bIsDynamic=false, bIsEditable=true, bIsStaticSwitch=false, bIsVisible=true, type {classStructOrEnum {refPath=&NiagaraInt32}}}
      Spawn Time {bIsDynamic=false, bIsEditable=true, bIsStaticSwitch=false, bIsVisible=true, type {classStructOrEnum {refPath=&NiagaraFloat}}}
      Use Spawn Probability {bIsDynamic=false, bIsEditable=true, bIsStaticSwitch=false, bIsVisible=true, type {classStructOrEnum {refPath=&NiagaraBool}}}
      Spawn Probability {bIsDynamic=false, bIsEditable=false, bIsStaticSwitch=false, bIsVisible=true, type {classStructOrEnum {refPath=&NiagaraFloat}}}
      UseLoopCountLimit {bIsDynamic=false, bIsEditable=true, bIsStaticSwitch=false, bIsVisible=true, type {classStructOrEnum {refPath=&NiagaraBool}}}
      Loop Count Limit {bIsDynamic=false, bIsEditable=false, bIsStaticSwitch=false, bIsVisible=true, type {classStructOrEnum {refPath=&NiagaraInt32}}}
      Spawn Group {bIsDynamic=false, bIsEditable=true, bIsStaticSwitch=false, bIsVisible=true, type {classStructOrEnum {refPath=&NiagaraInt32}}}
      Age {bIsDynamic=false, bIsEditable=true, bIsStaticSwitch=false, bIsVisible=true, type {classStructOrEnum {refPath=&NiagaraFloat}}}
     ]
     moduleScript {refPath=&SpawnBurst_Instantaneous}
    }
    {
     bIsSetParametersModule=false
     enabled=true
     moduleName=SpawnRate
     inputs [
      SpawnRate {bIsDynamic=false, bIsEditable=true, bIsStaticSwitch=false, bIsVisible=true, type {classStructOrEnum {refPath=&NiagaraFloat}}}
      Use Spawn Probability {bIsDynamic=false, bIsEditable=true, bIsStaticSwitch=false, bIsVisible=true, type {classStructOrEnum {refPath=&NiagaraBool}}}
      Spawn Probability {bIsDynamic=false, bIsEditable=false, bIsStaticSwitch=false, bIsVisible=true, type {classStructOrEnum {refPath=&NiagaraFloat}}}
      SpawnGroup {bIsDynamic=false, bIsEditable=true, bIsStaticSwitch=false, bIsVisible=true, type {classStructOrEnum {refPath=&NiagaraInt32}}}
     ]
     moduleScript {refPath=&SpawnRate}
    }
   ]
  }
  particleSpawnScript {
   scriptName=ParticleSpawnScript
   modules [
    {
     bIsSetParametersModule=false
     enabled=true
     moduleName=InitializeParticle
     inputs [
      Lifetime Mode {
       bIsDynamic=false
       bIsEditable=true
       bIsStaticSwitch=true
       bIsVisible=true
       type {classStructOrEnum {refPath=&ENiagara_LifetimeMode}}
      }
      Lifetime Min {bIsDynamic=false, bIsEditable=true, bIsStaticSwitch=false, bIsVisible=true, type {classStructOrEnum {refPath=&NiagaraFloat}}}
      Lifetime {
       bIsDynamic=false
       bIsEditable=false
       bIsStaticSwitch=false
       bIsVisible=false
       type {classStructOrEnum {refPath=&NiagaraFloat}}
      }
      Lifetime Max {bIsDynamic=false, bIsEditable=true, bIsStaticSwitch=false, bIsVisible=true, type {classStructOrEnum {refPath=&NiagaraFloat}}}
      Lifetime Randomness Mode {
       bIsDynamic=false
       bIsEditable=true
       bIsStaticSwitch=true
       bIsVisible=true
       type {classStructOrEnum {refPath=&ENiagaraRandomnessMode}}
      }
      Color Mode {
       bIsDynamic=false
       bIsEditable=true
       bIsStaticSwitch=true
       bIsVisible=true
       type {classStructOrEnum {refPath=&ENiagara_ColorInitializationMode}}
      }
      Color Minimum {
       bIsDynamic=false
       bIsEditable=false
       bIsStaticSwitch=false
       bIsVisible=false
       type {classStructOrEnum {refPath=&LinearColor}}
      }
      Color {
       bIsDynamic=true
       bIsEditable=true
       bIsStaticSwitch=false
       bIsVisible=true
       type {classStructOrEnum {refPath=&LinearColor}}
      }
      AdjustHue {bIsDynamic=false, bIsEditable=false, bIsStaticSwitch=false, bIsVisible=false, type {classStructOrEnum {refPath=&NiagaraBool}}}
      Hue Shift Range {
       bIsDynamic=false
       bIsEditable=false
       bIsStaticSwitch=false
       bIsVisible=false
       type {classStructOrEnum {refPath=&Vector2f}}
      }
      Color Maximum {
       bIsDynamic=false
       bIsEditable=false
       bIsStaticSwitch=false
       bIsVisible=false
       type {classStructOrEnum {refPath=&LinearColor}}
      }
      AdjustSaturation {bIsDynamic=false, bIsEditable=false, bIsStaticSwitch=false, bIsVisible=false, type {classStructOrEnum {refPath=&NiagaraBool}}}
      Saturation Range {
       bIsDynamic=false
       bIsEditable=false
       bIsStaticSwitch=false
       bIsVisible=false
       type {classStructOrEnum {refPath=&Vector2f}}
      }
      AdjustValue {bIsDynamic=false, bIsEditable=false, bIsStaticSwitch=false, bIsVisible=false, type {classStructOrEnum {refPath=&NiagaraBool}}}
      Value Range {
       bIsDynamic=false
       bIsEditable=false
       bIsStaticSwitch=false
       bIsVisible=false
       type {classStructOrEnum {refPath=&Vector2f}}
      }
      AdjustAlpha {bIsDynamic=false, bIsEditable=false, bIsStaticSwitch=false, bIsVisible=false, type {classStructOrEnum {refPath=&NiagaraBool}}}
      Alpha Scale Range {
       bIsDynamic=false
       bIsEditable=false
       bIsStaticSwitch=false
       bIsVisible=false
       type {classStructOrEnum {refPath=&Vector2f}}
      }
      Color Channel Mode {
       bIsDynamic=false
       bIsEditable=false
       bIsStaticSwitch=true
       bIsVisible=false
       type {classStructOrEnum {refPath=&ENiagaraChannelCorrelation}}
      }
      Color Randomness Mode {
       bIsDynamic=false
       bIsEditable=false
       bIsStaticSwitch=true
       bIsVisible=false
       type {classStructOrEnum {refPath=&ENiagaraRandomnessMode}}
      }
      Position Mode {
       bIsDynamic=false
       bIsEditable=true
       bIsStaticSwitch=true
       bIsVisible=true
       type {classStructOrEnum {refPath=&ENiagara_PositionInitializationMode}}
      }
      Position {
       bIsDynamic=false
       bIsEditable=false
       bIsStaticSwitch=false
       bIsVisible=false
       type {classStructOrEnum {refPath=&NiagaraPosition}}
      }
      UsePositionOffset {bIsDynamic=false, bIsEditable=true, bIsStaticSwitch=false, bIsVisible=true, type {classStructOrEnum {refPath=&NiagaraBool}}}
      Position Offset {bIsDynamic=false, bIsEditable=true, bIsStaticSwitch=false, bIsVisible=true, type {classStructOrEnum {refPath=&Vector3f}}}
      Position Offset Coordinate Space {
       bIsDynamic=false
       bIsEditable=true
       bIsStaticSwitch=false
       bIsVisible=true
       type {classStructOrEnum {refPath=&ENiagaraCoordinateSpace}}
      }
      Mass Mode {
       bIsDynamic=false
       bIsEditable=true
       bIsStaticSwitch=true
       bIsVisible=true
       type {classStructOrEnum {refPath=&ENiagara_MassInitializationMode}}
      }
      Mass Min {
       bIsDynamic=false
       bIsEditable=false
       bIsStaticSwitch=false
       bIsVisible=false
       type {classStructOrEnum {refPath=&NiagaraFloat}}
      }
      Mass {
       bIsDynamic=false
       bIsEditable=false
       bIsStaticSwitch=false
       bIsVisible=false
       type {classStructOrEnum {refPath=&NiagaraFloat}}
      }
      Mass Max {
       bIsDynamic=false
       bIsEditable=false
       bIsStaticSwitch=false
       bIsVisible=false
       type {classStructOrEnum {refPath=&NiagaraFloat}}
      }
      Mass Randomness Mode {
       bIsDynamic=false
       bIsEditable=false
       bIsStaticSwitch=true
       bIsVisible=false
       type {classStructOrEnum {refPath=&ENiagaraRandomnessMode}}
      }
      Write Material Random {bIsDynamic=false, bIsEditable=true, bIsStaticSwitch=false, bIsVisible=true, type {classStructOrEnum {refPath=&NiagaraBool}}}
      Material Random {bIsDynamic=false, bIsEditable=true, bIsStaticSwitch=false, bIsVisible=true, type {classStructOrEnum {refPath=&NiagaraFloat}}}
      Sprite Size Mode {
       bIsDynamic=false
       bIsEditable=true
       bIsStaticSwitch=true
       bIsVisible=true
       type {classStructOrEnum {refPath=&ENiagara_SizeScaleMode}}
      }
      Sprite Size {
       bIsDynamic=false
       bIsEditable=false
       bIsStaticSwitch=false
       bIsVisible=false
       type {classStructOrEnum {refPath=&Vector2f}}
      }
      Uniform Sprite Size {
       bIsDynamic=false
       bIsEditable=false
       bIsStaticSwitch=false
       bIsVisible=false
       type {classStructOrEnum {refPath=&NiagaraFloat}}
      }
      Uniform Sprite Size Min {
       bIsDynamic=false
       bIsEditable=false
       bIsStaticSwitch=false
       bIsVisible=false
       type {classStructOrEnum {refPath=&NiagaraFloat}}
      }
      Sprite Size Min {bIsDynamic=false, bIsEditable=true, bIsStaticSwitch=false, bIsVisible=true, type {classStructOrEnum {refPath=&Vector2f}}}
      Sprite Size Max {bIsDynamic=false, bIsEditable=true, bIsStaticSwitch=false, bIsVisible=true, type {classStructOrEnum {refPath=&Vector2f}}}
      Uniform Sprite Size Max {
       bIsDynamic=false
       bIsEditable=false
       bIsStaticSwitch=false
       bIsVisible=false
       type {classStructOrEnum {refPath=&NiagaraFloat}}
      }
      Sprite Size Randomness Mode {
       bIsDynamic=false
       bIsEditable=true
       bIsStaticSwitch=true
       bIsVisible=true
       type {classStructOrEnum {refPath=&ENiagaraRandomnessMode}}
      }
      Sprite Rotation Mode {
       bIsDynamic=false
       bIsEditable=true
       bIsStaticSwitch=true
       bIsVisible=true
       type {classStructOrEnum {refPath=&ENiagara_SpriteRotationMode}}
      }
      Sprite Rotation Angle Min {
       bIsDynamic=false
       bIsEditable=false
       bIsStaticSwitch=false
       bIsVisible=false
       type {classStructOrEnum {refPath=&NiagaraFloat}}
      }
      Sprite Rotation Angle {bIsDynamic=true, bIsEditable=true, bIsStaticSwitch=false, bIsVisible=true, type {classStructOrEnum {refPath=&NiagaraFloat}}}
      Sprite Rotation Angle Max {
       bIsDynamic=false
       bIsEditable=false
       bIsStaticSwitch=false
       bIsVisible=false
       type {classStructOrEnum {refPath=&NiagaraFloat}}
      }
      Sprite Rotation Randomness Mode {
       bIsDynamic=false
       bIsEditable=false
       bIsStaticSwitch=true
       bIsVisible=false
       type {classStructOrEnum {refPath=&ENiagaraRandomnessMode}}
      }
      Sprite UV Mode {
       bIsDynamic=false
       bIsEditable=true
       bIsStaticSwitch=true
       bIsVisible=true
       type {classStructOrEnum {refPath=&ENiagara_UVFlippingMode}}
      }
      Sprite UV Randomness Mode {
       bIsDynamic=false
       bIsEditable=false
       bIsStaticSwitch=true
       bIsVisible=false
       type {classStructOrEnum {refPath=&ENiagaraRandomnessMode}}
      }
      Sprite UV Scale {
       bIsDynamic=false
       bIsEditable=false
       bIsStaticSwitch=false
       bIsVisible=false
       type {classStructOrEnum {refPath=&Vector2f}}
      }
      Mesh Scale Mode {
       bIsDynamic=false
       bIsEditable=true
       bIsStaticSwitch=true
       bIsVisible=true
       type {classStructOrEnum {refPath=&ENiagara_SizeScaleMode}}
      }
      Mesh Uniform Scale Min {
       bIsDynamic=false
       bIsEditable=false
       bIsStaticSwitch=false
       bIsVisible=false
       type {classStructOrEnum {refPath=&NiagaraFloat}}
      }
      Mesh Uniform Scale {
       bIsDynamic=false
       bIsEditable=false
       bIsStaticSwitch=false
       bIsVisible=false
       type {classStructOrEnum {refPath=&NiagaraFloat}}
      }
      Mesh Scale Min {
       bIsDynamic=false
       bIsEditable=false
       bIsStaticSwitch=false
       bIsVisible=false
       type {classStructOrEnum {refPath=&Vector3f}}
      }
      Mesh Scale {
       bIsDynamic=false
       bIsEditable=false
       bIsStaticSwitch=false
       bIsVisible=false
       type {classStructOrEnum {refPath=&Vector3f}}
      }
      Mesh Scale Max {
       bIsDynamic=false
       bIsEditable=false
       bIsStaticSwitch=false
       bIsVisible=false
       type {classStructOrEnum {refPath=&Vector3f}}
      }
      Mesh Uniform Scale Max {
       bIsDynamic=false
       bIsEditable=false
       bIsStaticSwitch=false
       bIsVisible=false
       type {classStructOrEnum {refPath=&NiagaraFloat}}
      }
      Mesh Scale Randomness Mode {
       bIsDynamic=false
       bIsEditable=false
       bIsStaticSwitch=true
       bIsVisible=false
       type {classStructOrEnum {refPath=&ENiagaraRandomnessMode}}
      }
      Mesh Renderer Array Visibility Mode {
       bIsDynamic=false
       bIsEditable=true
       bIsStaticSwitch=true
       bIsVisible=true
       type {classStructOrEnum {refPath=&ENiagara_UnsetDirectSetRandom}}
      }
      Mesh Renderer Mesh Index {
       bIsDynamic=false
       bIsEditable=false
       bIsStaticSwitch=false
       bIsVisible=false
       type {classStructOrEnum {refPath=&NiagaraInt32}}
      }
      Wrap Index By Mesh Count {bIsDynamic=false, bIsEditable=false, bIsStaticSwitch=true, bIsVisible=false, type {classStructOrEnum {refPath=&NiagaraBool}}}
      Mesh Renderer Info {
       bIsDynamic=false
       bIsEditable=false
       bIsStaticSwitch=false
       bIsVisible=false
       type {classStructOrEnum {refPath=&NiagaraDataInterfaceMeshRendererInfo}}
      }
      Mesh Renderer Random Seed {
       bIsDynamic=false
       bIsEditable=false
       bIsStaticSwitch=false
       bIsVisible=false
       type {classStructOrEnum {refPath=&NiagaraInt32}}
      }
      Ribbon Width Mode {
       bIsDynamic=false
       bIsEditable=true
       bIsStaticSwitch=true
       bIsVisible=true
       type {classStructOrEnum {refPath=&ENiagara_UnsetDirectSet}}
      }
      Ribbon Width {
       bIsDynamic=false
       bIsEditable=false
       bIsStaticSwitch=false
       bIsVisible=false
       type {classStructOrEnum {refPath=&NiagaraFloat}}
      }
      Ribbon Facing Mode {
       bIsDynamic=false
       bIsEditable=true
       bIsStaticSwitch=true
       bIsVisible=true
       type {classStructOrEnum {refPath=&ENiagara_UnsetDirectSet}}
      }
      Ribbon Facing Vector {
       bIsDynamic=false
       bIsEditable=false
       bIsStaticSwitch=false
       bIsVisible=false
       type {classStructOrEnum {refPath=&Vector3f}}
      }
      Ribbon Twist Mode {
       bIsDynamic=false
       bIsEditable=true
       bIsStaticSwitch=true
       bIsVisible=true
       type {classStructOrEnum {refPath=&ENiagara_UnsetDirectSet}}
      }
      "Ribbon Twist " {
       bIsDynamic=false
       bIsEditable=false
       bIsStaticSwitch=false
       bIsVisible=false
       type {classStructOrEnum {refPath=&NiagaraFloat}}
      }
     ]
     moduleScript {refPath=&InitializeParticle}
    }
    {
     bIsSetParametersModule=false
     enabled=true
     moduleName=CameraOffset
     inputs [Camera Offset Amount {bIsDynamic=false, bIsEditable=true, bIsStaticSwitch=false, bIsVisible=true, type {classStructOrEnum {refPath=&NiagaraFloat}}}]
     moduleScript {refPath=&CameraOffset}
    }
    {
     bIsSetParametersModule=false
     enabled=true
     moduleName=SpriteFacingAndAlignment
     inputs [
      WriteFacing {bIsDynamic=false, bIsEditable=true, bIsStaticSwitch=false, bIsVisible=true, type {classStructOrEnum {refPath=&NiagaraBool}}}
      Sprite Facing {bIsDynamic=false, bIsEditable=false, bIsStaticSwitch=false, bIsVisible=true, type {classStructOrEnum {refPath=&Vector3f}}}
      Facing Coordinate Space {
       bIsDynamic=false
       bIsEditable=false
       bIsStaticSwitch=false
       bIsVisible=false
       type {classStructOrEnum {refPath=&ENiagaraCoordinateSpace}}
      }
      WriteAlignment {bIsDynamic=false, bIsEditable=true, bIsStaticSwitch=false, bIsVisible=true, type {classStructOrEnum {refPath=&NiagaraBool}}}
      Sprite Alignment {bIsDynamic=false, bIsEditable=true, bIsStaticSwitch=false, bIsVisible=true, type {classStructOrEnum {refPath=&Vector3f}}}
      Alignment Coordinate Space {
       bIsDynamic=false
       bIsEditable=true
       bIsStaticSwitch=false
       bIsVisible=true
       type {classStructOrEnum {refPath=&ENiagaraCoordinateSpace}}
      }
     ]
     moduleScript {refPath=&SpriteFacingAndAlignment}
    }
   ]
  }
  particleUpdateScript {
   scriptName=ParticleUpdateScript
   modules [
    {
     bIsSetParametersModule=true
     enabled=true
     moduleName=SetVariables_#12
     inputs [Particles.SpriteRotation {bIsDynamic=true, bIsEditable=true, bIsStaticSwitch=false, bIsVisible=true, type {classStructOrEnum {refPath=&NiagaraFloat}}}]
     moduleScript {refPath=&self:IceBlast_0.NiagaraScriptSource_0.NiagaraGraph_0.NiagaraNodeAssignment_6.SetVariables_#12}
    }
    {
     bIsSetParametersModule=false
     enabled=true
     moduleName=ParticleState
     inputs [
      Kill Particles When Lifetime Has Elapsed {bIsDynamic=false, bIsEditable=true, bIsStaticSwitch=true, bIsVisible=true, type {classStructOrEnum {refPath=&NiagaraBool}}}
      Loop Particles Lifetime {bIsDynamic=false, bIsEditable=false, bIsStaticSwitch=false, bIsVisible=false, type {classStructOrEnum {refPath=&NiagaraBool}}}
      Let Infinitely Lived Particles Die When Emitter Deactivates {bIsDynamic=false, bIsEditable=false, bIsStaticSwitch=false, bIsVisible=false, type {classStructOrEnum {refPath=&NiagaraBool}}}
      Lifetime {bIsDynamic=false, bIsEditable=true, bIsStaticSwitch=false, bIsVisible=true, type {classStructOrEnum {refPath=&NiagaraFloat}}}
      DeltaTime {bIsDynamic=false, bIsEditable=true, bIsStaticSwitch=false, bIsVisible=true, type {classStructOrEnum {refPath=&NiagaraFloat}}}
     ]
     moduleScript {refPath=&ParticleState}
    }
    {
     bIsSetParametersModule=false
     enabled=true
     moduleName=ScaleColor
     inputs [
      Scale Mode {
       bIsDynamic=false
       bIsEditable=true
       bIsStaticSwitch=true
       bIsVisible=true
       type {classStructOrEnum {refPath=&ENiagaraScaleColorMode}}
      }
      Linear Color Curve {
       bIsDynamic=false
       bIsEditable=false
       bIsStaticSwitch=false
       bIsVisible=false
       type {classStructOrEnum {refPath=&NiagaraDataInterfaceColorCurve}}
      }
      ScaleRGB {bIsDynamic=false, bIsEditable=true, bIsStaticSwitch=false, bIsVisible=true, type {classStructOrEnum {refPath=&NiagaraBool}}}
      Scale RGB {bIsDynamic=true, bIsEditable=true, bIsStaticSwitch=false, bIsVisible=true, type {classStructOrEnum {refPath=&Vector3f}}}
      ScaleRGBA {bIsDynamic=false, bIsEditable=false, bIsStaticSwitch=false, bIsVisible=false, type {classStructOrEnum {refPath=&NiagaraBool}}}
      Scale RGBA {
       bIsDynamic=false
       bIsEditable=false
       bIsStaticSwitch=false
       bIsVisible=false
       type {classStructOrEnum {refPath=&Vector4f}}
      }
      Curve Index {
       bIsDynamic=false
       bIsEditable=false
       bIsStaticSwitch=false
       bIsVisible=false
       type {classStructOrEnum {refPath=&NiagaraFloat}}
      }
      ScaleA {bIsDynamic=false, bIsEditable=true, bIsStaticSwitch=false, bIsVisible=true, type {classStructOrEnum {refPath=&NiagaraBool}}}
      Scale Alpha {bIsDynamic=true, bIsEditable=true, bIsStaticSwitch=false, bIsVisible=true, type {classStructOrEnum {refPath=&NiagaraFloat}}}
      Color Value To Scale {
       bIsDynamic=false
       bIsEditable=true
       bIsStaticSwitch=false
       bIsVisible=true
       type {classStructOrEnum {refPath=&LinearColor}}
      }
     ]
     moduleScript {refPath=&ScaleColor}
    }
    {
     bIsSetParametersModule=false
     enabled=true
     moduleName=ScaleSpriteSize003
     inputs [
      Scale Sprite Size Mode {
       bIsDynamic=false
       bIsEditable=true
       bIsStaticSwitch=true
       bIsVisible=true
       type {classStructOrEnum {refPath=&ENiagara_ScaleSpriteSize}}
      }
      Uniform Curve Sprite Scale {
       bIsDynamic=false
       bIsEditable=false
       bIsStaticSwitch=false
       bIsVisible=false
       type {classStructOrEnum {refPath=&NiagaraDataInterfaceCurve}}
      }
      Non-Uniform Curve Sprite Scale {
       bIsDynamic=false
       bIsEditable=false
       bIsStaticSwitch=false
       bIsVisible=false
       type {classStructOrEnum {refPath=&NiagaraDataInterfaceVector2DCurve}}
      }
      Uniform Curve Index {
       bIsDynamic=false
       bIsEditable=false
       bIsStaticSwitch=false
       bIsVisible=false
       type {classStructOrEnum {refPath=&NiagaraFloat}}
      }
      Non-Uniform Curve Index {
       bIsDynamic=false
       bIsEditable=false
       bIsStaticSwitch=false
       bIsVisible=false
       type {classStructOrEnum {refPath=&NiagaraFloat}}
      }
      Non-Uniform Curve Scale {
       bIsDynamic=false
       bIsEditable=false
       bIsStaticSwitch=false
       bIsVisible=false
       type {classStructOrEnum {refPath=&Vector2f}}
      }
      Uniform Curve Scale {
       bIsDynamic=false
       bIsEditable=false
       bIsStaticSwitch=false
       bIsVisible=false
       type {classStructOrEnum {refPath=&NiagaraFloat}}
      }
      Scale Factor {bIsDynamic=true, bIsEditable=true, bIsStaticSwitch=false, bIsVisible=true, type {classStructOrEnum {refPath=&Vector2f}}}
      Uniform Scale Factor {
       bIsDynamic=false
       bIsEditable=false
       bIsStaticSwitch=false
       bIsVisible=false
       type {classStructOrEnum {refPath=&NiagaraFloat}}
      }
      Initial Sprite Size {bIsDynamic=false, bIsEditable=true, bIsStaticSwitch=false, bIsVisible=true, type {classStructOrEnum {refPath=&Vector2f}}}
     ]
     moduleScript {refPath=&ScaleSpriteSize}
    }
    {
     bIsSetParametersModule=false
     enabled=true
     moduleName=DynamicMaterialParameters001
     inputs [
      Write Parameter Index 0 {bIsDynamic=false, bIsEditable=true, bIsStaticSwitch=true, bIsVisible=true, type {classStructOrEnum {refPath=&NiagaraBool}}}
      Module.Param0WriteEnabled {bIsDynamic=false, bIsEditable=true, bIsStaticSwitch=false, bIsVisible=true, type {classStructOrEnum {refPath=&NiagaraBool}}}
      Index 0 Param 1 {bIsDynamic=true, bIsEditable=true, bIsStaticSwitch=false, bIsVisible=true, type {classStructOrEnum {refPath=&NiagaraFloat}}}
      Module.Param1WriteEnabled {bIsDynamic=false, bIsEditable=true, bIsStaticSwitch=false, bIsVisible=true, type {classStructOrEnum {refPath=&NiagaraBool}}}
      Index 0 Param 2 {bIsDynamic=true, bIsEditable=false, bIsStaticSwitch=false, bIsVisible=true, type {classStructOrEnum {refPath=&NiagaraFloat}}}
      Module.Param2WriteEnabled {bIsDynamic=false, bIsEditable=true, bIsStaticSwitch=false, bIsVisible=true, type {classStructOrEnum {refPath=&NiagaraBool}}}
      Index 0 Param 3 {bIsDynamic=true, bIsEditable=true, bIsStaticSwitch=false, bIsVisible=true, type {classStructOrEnum {refPath=&NiagaraFloat}}}
      Module.Param3WriteEnabled {bIsDynamic=false, bIsEditable=true, bIsStaticSwitch=false, bIsVisible=true, type {classStructOrEnum {refPath=&NiagaraBool}}}
      Index 0 Param 4 {bIsDynamic=true, bIsEditable=true, bIsStaticSwitch=false, bIsVisible=true, type {classStructOrEnum {refPath=&NiagaraFloat}}}
      Write Parameter Index 1 {bIsDynamic=false, bIsEditable=true, bIsStaticSwitch=true, bIsVisible=true, type {classStructOrEnum {refPath=&NiagaraBool}}}
      Module.Param0WriteEnabled1 {bIsDynamic=false, bIsEditable=true, bIsStaticSwitch=false, bIsVisible=true, type {classStructOrEnum {refPath=&NiagaraBool}}}
      Index 1 Param 1 {bIsDynamic=true, bIsEditable=false, bIsStaticSwitch=false, bIsVisible=true, type {classStructOrEnum {refPath=&NiagaraFloat}}}
      Module.Param1WriteEnabled1 {bIsDynamic=false, bIsEditable=true, bIsStaticSwitch=false, bIsVisible=true, type {classStructOrEnum {refPath=&NiagaraBool}}}
      Index 1 Param 2 {bIsDynamic=true, bIsEditable=false, bIsStaticSwitch=false, bIsVisible=true, type {classStructOrEnum {refPath=&NiagaraFloat}}}
      Module.Param2WriteEnabled1 {bIsDynamic=false, bIsEditable=true, bIsStaticSwitch=false, bIsVisible=true, type {classStructOrEnum {refPath=&NiagaraBool}}}
      Index 1 Param 3 {bIsDynamic=true, bIsEditable=false, bIsStaticSwitch=false, bIsVisible=true, type {classStructOrEnum {refPath=&NiagaraFloat}}}
      Module.Param3WriteEnabled1 {bIsDynamic=false, bIsEditable=true, bIsStaticSwitch=false, bIsVisible=true, type {classStructOrEnum {refPath=&NiagaraBool}}}
      Index 1 Param 4 {bIsDynamic=true, bIsEditable=false, bIsStaticSwitch=false, bIsVisible=true, type {classStructOrEnum {refPath=&NiagaraFloat}}}
      Write Parameter Index 2 {bIsDynamic=false, bIsEditable=true, bIsStaticSwitch=true, bIsVisible=true, type {classStructOrEnum {refPath=&NiagaraBool}}}
      Module.Param0WriteEnabled2 {bIsDynamic=false, bIsEditable=true, bIsStaticSwitch=false, bIsVisible=true, type {classStructOrEnum {refPath=&NiagaraBool}}}
      Index 2 Param 1 {bIsDynamic=true, bIsEditable=true, bIsStaticSwitch=false, bIsVisible=true, type {classStructOrEnum {refPath=&NiagaraFloat}}}
      Module.Param1WriteEnabled2 {bIsDynamic=false, bIsEditable=true, bIsStaticSwitch=false, bIsVisible=true, type {classStructOrEnum {refPath=&NiagaraBool}}}
      Index 2 Param 2 {bIsDynamic=true, bIsEditable=true, bIsStaticSwitch=false, bIsVisible=true, type {classStructOrEnum {refPath=&NiagaraFloat}}}
      Module.Param2WriteEnabled2 {bIsDynamic=false, bIsEditable=true, bIsStaticSwitch=false, bIsVisible=true, type {classStructOrEnum {refPath=&NiagaraBool}}}
      Index 2 Param 3 {bIsDynamic=true, bIsEditable=true, bIsStaticSwitch=false, bIsVisible=true, type {classStructOrEnum {refPath=&NiagaraFloat}}}
      Module.Param3WriteEnabled2 {bIsDynamic=false, bIsEditable=true, bIsStaticSwitch=false, bIsVisible=true, type {classStructOrEnum {refPath=&NiagaraBool}}}
      Index 2 Param 4 {bIsDynamic=true, bIsEditable=true, bIsStaticSwitch=false, bIsVisible=true, type {classStructOrEnum {refPath=&NiagaraFloat}}}
      Write Parameter Index 3 {bIsDynamic=false, bIsEditable=true, bIsStaticSwitch=true, bIsVisible=true, type {classStructOrEnum {refPath=&NiagaraBool}}}
      Index 3 Param 1 {
       bIsDynamic=false
       bIsEditable=false
       bIsStaticSwitch=false
       bIsVisible=false
       type {classStructOrEnum {refPath=&NiagaraFloat}}
      }
      Index 3 Param 2 {
       bIsDynamic=false
       bIsEditable=false
       bIsStaticSwitch=false
       bIsVisible=false
       type {classStructOrEnum {refPath=&NiagaraFloat}}
      }
      Index 3 Param 3 {
       bIsDynamic=false
       bIsEditable=false
       bIsStaticSwitch=false
       bIsVisible=false
       type {classStructOrEnum {refPath=&NiagaraFloat}}
      }
      Index 3 Param 4 {
       bIsDynamic=false
       bIsEditable=false
       bIsStaticSwitch=false
       bIsVisible=false
       type {classStructOrEnum {refPath=&NiagaraFloat}}
      }
     ]
     moduleScript {refPath=&DynamicMaterialParameters}
    }
    {
     bIsSetParametersModule=false
     enabled=true
     moduleName=SolveForcesAndVelocity
     inputs [
      Clamp Velocity {bIsDynamic=false, bIsEditable=true, bIsStaticSwitch=false, bIsVisible=true, type {classStructOrEnum {refPath=&NiagaraBool}}}
      Speed Limit {bIsDynamic=false, bIsEditable=false, bIsStaticSwitch=false, bIsVisible=true, type {classStructOrEnum {refPath=&NiagaraFloat}}}
      Limit Acceleration {bIsDynamic=false, bIsEditable=true, bIsStaticSwitch=false, bIsVisible=true, type {classStructOrEnum {refPath=&NiagaraBool}}}
      Acceleration Limit {bIsDynamic=false, bIsEditable=false, bIsStaticSwitch=false, bIsVisible=true, type {classStructOrEnum {refPath=&NiagaraFloat}}}
      Manually Enable Rotational Solver {bIsDynamic=false, bIsEditable=true, bIsStaticSwitch=false, bIsVisible=true, type {classStructOrEnum {refPath=&NiagaraBool}}}
      Write to Intrinsic Properties {bIsDynamic=false, bIsEditable=true, bIsStaticSwitch=false, bIsVisible=true, type {classStructOrEnum {refPath=&NiagaraBool}}}
      Write to Presolve Properties {bIsDynamic=false, bIsEditable=true, bIsStaticSwitch=true, bIsVisible=true, type {classStructOrEnum {refPath=&NiagaraBool}}}
      DeltaTime {bIsDynamic=false, bIsEditable=true, bIsStaticSwitch=false, bIsVisible=true, type {classStructOrEnum {refPath=&NiagaraFloat}}}
      Mass {bIsDynamic=false, bIsEditable=true, bIsStaticSwitch=false, bIsVisible=true, type {classStructOrEnum {refPath=&NiagaraFloat}}}
      Position {
       bIsDynamic=false
       bIsEditable=true
       bIsStaticSwitch=false
       bIsVisible=true
       type {classStructOrEnum {refPath=&NiagaraPosition}}
      }
      Velocity {bIsDynamic=false, bIsEditable=true, bIsStaticSwitch=false, bIsVisible=true, type {classStructOrEnum {refPath=&Vector3f}}}
      PreviousVelocity {bIsDynamic=false, bIsEditable=true, bIsStaticSwitch=false, bIsVisible=true, type {classStructOrEnum {refPath=&Vector3f}}}
      Force {bIsDynamic=false, bIsEditable=true, bIsStaticSwitch=false, bIsVisible=true, type {classStructOrEnum {refPath=&Vector3f}}}
      Mesh Orientation {bIsDynamic=false, bIsEditable=false, bIsStaticSwitch=false, bIsVisible=false, type {classStructOrEnum {refPath=&Quat4f}}}
      Rotational Velocity {
       bIsDynamic=false
       bIsEditable=false
       bIsStaticSwitch=false
       bIsVisible=false
       type {classStructOrEnum {refPath=&Vector3f}}
      }
      Rotational Force {
       bIsDynamic=false
       bIsEditable=false
       bIsStaticSwitch=false
       bIsVisible=false
       type {classStructOrEnum {refPath=&Vector3f}}
      }
      Rotational Inertia {
       bIsDynamic=false
       bIsEditable=false
       bIsStaticSwitch=false
       bIsVisible=false
       type {classStructOrEnum {refPath=&NiagaraFloat}}
      }
     ]
     moduleScript {refPath=&SolveForcesAndVelocity}
    }
    {
     bIsSetParametersModule=true
     enabled=true
     moduleName=SetVariables_#13
     inputs [Particles.Linear Color {
       bIsDynamic=true
       bIsEditable=true
       bIsStaticSwitch=false
       bIsVisible=true
       type {classStructOrEnum {refPath=&LinearColor}}
      }]
     moduleScript {refPath=&self:IceBlast_0.NiagaraScriptSource_0.NiagaraGraph_0.NiagaraNodeAssignment_0.SetVariables_#13}
    }
    {
     bIsSetParametersModule=true
     enabled=true
     moduleName=SetVariables_#14
     inputs [Particles.LowMediumSize {bIsDynamic=true, bIsEditable=true, bIsStaticSwitch=false, bIsVisible=true, type {classStructOrEnum {refPath=&Vector2f}}}]
     moduleScript {refPath=&self:IceBlast_0.NiagaraScriptSource_0.NiagaraGraph_0.NiagaraNodeAssignment_1.SetVariables_#14}
    }
    {
     bIsSetParametersModule=false
     enabled=true
     moduleName=ScaleColor001
     inputs [
      Scale Mode {bIsDynamic=false, bIsEditable=true, bIsStaticSwitch=true, bIsVisible=true, type {classStructOrEnum {refPath=&NiagaraInt32}}}
      Linear Color Curve {
       bIsDynamic=false
       bIsEditable=false
       bIsStaticSwitch=false
       bIsVisible=false
       type {classStructOrEnum {refPath=&NiagaraDataInterfaceColorCurve}}
      }
      ScaleRGB {bIsDynamic=false, bIsEditable=true, bIsStaticSwitch=false, bIsVisible=true, type {classStructOrEnum {refPath=&NiagaraBool}}}
      Scale RGB {bIsDynamic=true, bIsEditable=true, bIsStaticSwitch=false, bIsVisible=true, type {classStructOrEnum {refPath=&Vector3f}}}
      ScaleRGBA {bIsDynamic=false, bIsEditable=false, bIsStaticSwitch=false, bIsVisible=false, type {classStructOrEnum {refPath=&NiagaraBool}}}
      Scale RGBA {
       bIsDynamic=false
       bIsEditable=false
       bIsStaticSwitch=false
       bIsVisible=false
       type {classStructOrEnum {refPath=&Vector4f}}
      }
      Curve Index {
       bIsDynamic=false
       bIsEditable=false
       bIsStaticSwitch=false
       bIsVisible=false
       type {classStructOrEnum {refPath=&NiagaraFloat}}
      }
      ScaleA {bIsDynamic=false, bIsEditable=true, bIsStaticSwitch=false, bIsVisible=true, type {classStructOrEnum {refPath=&NiagaraBool}}}
      Scale Alpha {bIsDynamic=false, bIsEditable=false, bIsStaticSwitch=false, bIsVisible=true, type {classStructOrEnum {refPath=&NiagaraFloat}}}
      Color Value To Scale {
       bIsDynamic=false
       bIsEditable=true
       bIsStaticSwitch=false
       bIsVisible=true
       type {classStructOrEnum {refPath=&LinearColor}}
      }
      Color Scale {
       bIsDynamic=false
       bIsEditable=false
       bIsStaticSwitch=false
       bIsVisible=false
       type {classStructOrEnum {refPath=&LinearColor}}
      }
     ]
     moduleScript {refPath=&ScaleColor}
    }
   ]
  }
  rendererClasses [{refPath=&NiagaraSpriteRendererProperties}]
  renderers [
   {rendererIndex=0, rendererClass {refPath=&NiagaraSpriteRendererProperties}}
   {rendererIndex=1, rendererClass {refPath=&NiagaraSpriteRendererProperties}}
  ]
 }
 IceSpikes {
  emitterSpawnScript {scriptName=EmitterSpawnScript, modules []}
  emitterUpdateScript {
   scriptName=EmitterUpdateScript
   modules [
    {
     bIsSetParametersModule=false
     enabled=true
     moduleName=EmitterState
     inputs [
      Life Cycle Mode {
       bIsDynamic=false
       bIsEditable=true
       bIsStaticSwitch=true
       bIsVisible=true
       type {classStructOrEnum {refPath=&ENiagaraEmitterLifeCycleMode}}
      }
      Inactive Response {
       bIsDynamic=false
       bIsEditable=true
       bIsStaticSwitch=true
       bIsVisible=true
       type {classStructOrEnum {refPath=&ENiagaraInactiveMode}}
      }
      Loop Behavior {
       bIsDynamic=false
       bIsEditable=true
       bIsStaticSwitch=true
       bIsVisible=true
       type {classStructOrEnum {refPath=&ENiagara_EmitterStateOptions}}
      }
      Loop Duration Mode {
       bIsDynamic=false
       bIsEditable=true
       bIsStaticSwitch=true
       bIsVisible=true
       type {classStructOrEnum {refPath=&ENiagara_InfiniteLoopDuration}}
      }
      Loop Count {
       bIsDynamic=false
       bIsEditable=false
       bIsStaticSwitch=false
       bIsVisible=false
       type {classStructOrEnum {refPath=&NiagaraInt32}}
      }
      Loop Duration {bIsDynamic=false, bIsEditable=true, bIsStaticSwitch=false, bIsVisible=true, type {classStructOrEnum {refPath=&NiagaraFloat}}}
      Recalculate Duration Each Loop {bIsDynamic=false, bIsEditable=false, bIsStaticSwitch=false, bIsVisible=false, type {classStructOrEnum {refPath=&NiagaraBool}}}
      UseLoopDelay {bIsDynamic=false, bIsEditable=true, bIsStaticSwitch=false, bIsVisible=true, type {classStructOrEnum {refPath=&NiagaraBool}}}
      Loop Delay {bIsDynamic=true, bIsEditable=false, bIsStaticSwitch=false, bIsVisible=true, type {classStructOrEnum {refPath=&NiagaraFloat}}}
      Delay First Loop Only {bIsDynamic=false, bIsEditable=false, bIsStaticSwitch=false, bIsVisible=false, type {classStructOrEnum {refPath=&NiagaraBool}}}
      Scalability Mode {
       bIsDynamic=false
       bIsEditable=true
       bIsStaticSwitch=true
       bIsVisible=true
       type {classStructOrEnum {refPath=&ENiagaraEmitterScalabilityMode_Limited}}
      }
      Enable Distance Culling {bIsDynamic=false, bIsEditable=false, bIsStaticSwitch=true, bIsVisible=false, type {classStructOrEnum {refPath=&NiagaraBool}}}
      MinDistance {
       bIsDynamic=false
       bIsEditable=false
       bIsStaticSwitch=false
       bIsVisible=false
       type {classStructOrEnum {refPath=&NiagaraFloat}}
      }
      Min Distance Response {
       bIsDynamic=false
       bIsEditable=false
       bIsStaticSwitch=false
       bIsVisible=false
       type {classStructOrEnum {refPath=&ENiagaraExecutionStateManagement}}
      }
      MaxDistance {
       bIsDynamic=false
       bIsEditable=false
       bIsStaticSwitch=false
       bIsVisible=false
       type {classStructOrEnum {refPath=&NiagaraFloat}}
      }
      Max Distance Response {
       bIsDynamic=false
       bIsEditable=false
       bIsStaticSwitch=false
       bIsVisible=false
       type {classStructOrEnum {refPath=&ENiagaraExecutionStateManagement}}
      }
      Scale Spawn Count {bIsDynamic=false, bIsEditable=false, bIsStaticSwitch=true, bIsVisible=false, type {classStructOrEnum {refPath=&NiagaraBool}}}
      Spawn Count Scale {
       bIsDynamic=false
       bIsEditable=false
       bIsStaticSwitch=false
       bIsVisible=false
       type {classStructOrEnum {refPath=&NiagaraFloat}}
      }
      Spawn Count Scale By Distance {
       bIsDynamic=false
       bIsEditable=false
       bIsStaticSwitch=false
       bIsVisible=false
       type {classStructOrEnum {refPath=&NiagaraDataInterfaceCurve}}
      }
      Enable Visibility Culling {bIsDynamic=false, bIsEditable=false, bIsStaticSwitch=true, bIsVisible=false, type {classStructOrEnum {refPath=&NiagaraBool}}}
      Visibility Cull Response {
       bIsDynamic=false
       bIsEditable=false
       bIsStaticSwitch=false
       bIsVisible=false
       type {classStructOrEnum {refPath=&ENiagaraExecutionStateManagement}}
      }
      Visibility Cull Delay {
       bIsDynamic=false
       bIsEditable=false
       bIsStaticSwitch=false
       bIsVisible=false
       type {classStructOrEnum {refPath=&NiagaraFloat}}
      }
      Reset Age On Awaken {bIsDynamic=false, bIsEditable=false, bIsStaticSwitch=false, bIsVisible=false, type {classStructOrEnum {refPath=&NiagaraBool}}}
     ]
     moduleScript {refPath=&EmitterState}
    }
    {
     bIsSetParametersModule=false
     enabled=true
     moduleName=SpawnBurst_Instantaneous
     inputs [
      Spawn Count {bIsDynamic=false, bIsEditable=true, bIsStaticSwitch=false, bIsVisible=true, type {classStructOrEnum {refPath=&NiagaraInt32}}}
      Spawn Time {bIsDynamic=false, bIsEditable=true, bIsStaticSwitch=false, bIsVisible=true, type {classStructOrEnum {refPath=&NiagaraFloat}}}
      Use Spawn Probability {bIsDynamic=false, bIsEditable=true, bIsStaticSwitch=false, bIsVisible=true, type {classStructOrEnum {refPath=&NiagaraBool}}}
      Spawn Probability {bIsDynamic=false, bIsEditable=false, bIsStaticSwitch=false, bIsVisible=true, type {classStructOrEnum {refPath=&NiagaraFloat}}}
      UseLoopCountLimit {bIsDynamic=false, bIsEditable=true, bIsStaticSwitch=false, bIsVisible=true, type {classStructOrEnum {refPath=&NiagaraBool}}}
      Loop Count Limit {bIsDynamic=false, bIsEditable=false, bIsStaticSwitch=false, bIsVisible=true, type {classStructOrEnum {refPath=&NiagaraInt32}}}
      Spawn Group {bIsDynamic=false, bIsEditable=true, bIsStaticSwitch=false, bIsVisible=true, type {classStructOrEnum {refPath=&NiagaraInt32}}}
      Age {bIsDynamic=false, bIsEditable=true, bIsStaticSwitch=false, bIsVisible=true, type {classStructOrEnum {refPath=&NiagaraFloat}}}
     ]
     moduleScript {refPath=&SpawnBurst_Instantaneous}
    }
    {
     bIsSetParametersModule=false
     enabled=false
     moduleName=SpawnRate
     inputs [
      SpawnRate {bIsDynamic=false, bIsEditable=true, bIsStaticSwitch=false, bIsVisible=true, type {classStructOrEnum {refPath=&NiagaraFloat}}}
      Use Spawn Probability {bIsDynamic=false, bIsEditable=true, bIsStaticSwitch=false, bIsVisible=true, type {classStructOrEnum {refPath=&NiagaraBool}}}
      Spawn Probability {bIsDynamic=false, bIsEditable=false, bIsStaticSwitch=false, bIsVisible=true, type {classStructOrEnum {refPath=&NiagaraFloat}}}
      SpawnGroup {bIsDynamic=false, bIsEditable=true, bIsStaticSwitch=false, bIsVisible=true, type {classStructOrEnum {refPath=&NiagaraInt32}}}
     ]
     moduleScript {refPath=&SpawnRate}
    }
   ]
  }
  particleSpawnScript {
   scriptName=ParticleSpawnScript
   modules [
    {
     bIsSetParametersModule=false
     enabled=true
     moduleName=InitializeParticle001
     inputs [
      Lifetime Mode {
       bIsDynamic=false
       bIsEditable=true
       bIsStaticSwitch=true
       bIsVisible=true
       type {classStructOrEnum {refPath=&ENiagara_LifetimeMode}}
      }
      Lifetime Min {
       bIsDynamic=false
       bIsEditable=false
       bIsStaticSwitch=false
       bIsVisible=false
       type {classStructOrEnum {refPath=&NiagaraFloat}}
      }
      Lifetime {bIsDynamic=false, bIsEditable=true, bIsStaticSwitch=false, bIsVisible=true, type {classStructOrEnum {refPath=&NiagaraFloat}}}
      Lifetime Max {
       bIsDynamic=false
       bIsEditable=false
       bIsStaticSwitch=false
       bIsVisible=false
       type {classStructOrEnum {refPath=&NiagaraFloat}}
      }
      Lifetime Randomness Mode {
       bIsDynamic=false
       bIsEditable=false
       bIsStaticSwitch=true
       bIsVisible=false
       type {classStructOrEnum {refPath=&ENiagaraRandomnessMode}}
      }
      Color Mode {
       bIsDynamic=false
       bIsEditable=true
       bIsStaticSwitch=true
       bIsVisible=true
       type {classStructOrEnum {refPath=&ENiagara_ColorInitializationMode}}
      }
      Color Minimum {
       bIsDynamic=false
       bIsEditable=false
       bIsStaticSwitch=false
       bIsVisible=false
       type {classStructOrEnum {refPath=&LinearColor}}
      }
      Color {
       bIsDynamic=true
       bIsEditable=true
       bIsStaticSwitch=false
       bIsVisible=true
       type {classStructOrEnum {refPath=&LinearColor}}
      }
      AdjustHue {bIsDynamic=false, bIsEditable=false, bIsStaticSwitch=false, bIsVisible=false, type {classStructOrEnum {refPath=&NiagaraBool}}}
      Hue Shift Range {
       bIsDynamic=false
       bIsEditable=false
       bIsStaticSwitch=false
       bIsVisible=false
       type {classStructOrEnum {refPath=&Vector2f}}
      }
      Color Maximum {
       bIsDynamic=false
       bIsEditable=false
       bIsStaticSwitch=false
       bIsVisible=false
       type {classStructOrEnum {refPath=&LinearColor}}
      }
      AdjustSaturation {bIsDynamic=false, bIsEditable=false, bIsStaticSwitch=false, bIsVisible=false, type {classStructOrEnum {refPath=&NiagaraBool}}}
      Saturation Range {
       bIsDynamic=false
       bIsEditable=false
       bIsStaticSwitch=false
       bIsVisible=false
       type {classStructOrEnum {refPath=&Vector2f}}
      }
      AdjustValue {bIsDynamic=false, bIsEditable=false, bIsStaticSwitch=false, bIsVisible=false, type {classStructOrEnum {refPath=&NiagaraBool}}}
      Value Range {
       bIsDynamic=false
       bIsEditable=false
       bIsStaticSwitch=false
       bIsVisible=false
       type {classStructOrEnum {refPath=&Vector2f}}
      }
      AdjustAlpha {bIsDynamic=false, bIsEditable=false, bIsStaticSwitch=false, bIsVisible=false, type {classStructOrEnum {refPath=&NiagaraBool}}}
      Alpha Scale Range {
       bIsDynamic=false
       bIsEditable=false
       bIsStaticSwitch=false
       bIsVisible=false
       type {classStructOrEnum {refPath=&Vector2f}}
      }
      Color Channel Mode {
       bIsDynamic=false
       bIsEditable=false
       bIsStaticSwitch=true
       bIsVisible=false
       type {classStructOrEnum {refPath=&ENiagaraChannelCorrelation}}
      }
      Color Randomness Mode {
       bIsDynamic=false
       bIsEditable=false
       bIsStaticSwitch=true
       bIsVisible=false
       type {classStructOrEnum {refPath=&ENiagaraRandomnessMode}}
      }
      Position Mode {
       bIsDynamic=false
       bIsEditable=true
       bIsStaticSwitch=true
       bIsVisible=true
       type {classStructOrEnum {refPath=&ENiagara_PositionInitializationMode}}
      }
      Position {
       bIsDynamic=false
       bIsEditable=false
       bIsStaticSwitch=false
       bIsVisible=false
       type {classStructOrEnum {refPath=&NiagaraPosition}}
      }
      UsePositionOffset {bIsDynamic=false, bIsEditable=true, bIsStaticSwitch=false, bIsVisible=true, type {classStructOrEnum {refPath=&NiagaraBool}}}
      Position Offset {bIsDynamic=false, bIsEditable=true, bIsStaticSwitch=false, bIsVisible=true, type {classStructOrEnum {refPath=&Vector3f}}}
      Position Offset Coordinate Space {
       bIsDynamic=false
       bIsEditable=true
       bIsStaticSwitch=false
       bIsVisible=true
       type {classStructOrEnum {refPath=&ENiagaraCoordinateSpace}}
      }
      Mass Mode {
       bIsDynamic=false
       bIsEditable=true
       bIsStaticSwitch=true
       bIsVisible=true
       type {classStructOrEnum {refPath=&ENiagara_MassInitializationMode}}
      }
      Mass Min {
       bIsDynamic=false
       bIsEditable=false
       bIsStaticSwitch=false
       bIsVisible=false
       type {classStructOrEnum {refPath=&NiagaraFloat}}
      }
      Mass {
       bIsDynamic=false
       bIsEditable=false
       bIsStaticSwitch=false
       bIsVisible=false
       type {classStructOrEnum {refPath=&NiagaraFloat}}
      }
      Mass Max {
       bIsDynamic=false
       bIsEditable=false
       bIsStaticSwitch=false
       bIsVisible=false
       type {classStructOrEnum {refPath=&NiagaraFloat}}
      }
      Mass Randomness Mode {
       bIsDynamic=false
       bIsEditable=false
       bIsStaticSwitch=true
       bIsVisible=false
       type {classStructOrEnum {refPath=&ENiagaraRandomnessMode}}
      }
      Write Material Random {bIsDynamic=false, bIsEditable=true, bIsStaticSwitch=false, bIsVisible=true, type {classStructOrEnum {refPath=&NiagaraBool}}}
      Material Random {bIsDynamic=false, bIsEditable=true, bIsStaticSwitch=false, bIsVisible=true, type {classStructOrEnum {refPath=&NiagaraFloat}}}
      Sprite Size Mode {
       bIsDynamic=false
       bIsEditable=true
       bIsStaticSwitch=true
       bIsVisible=true
       type {classStructOrEnum {refPath=&ENiagara_SizeScaleMode}}
      }
      Sprite Size {bIsDynamic=false, bIsEditable=true, bIsStaticSwitch=false, bIsVisible=true, type {classStructOrEnum {refPath=&Vector2f}}}
      Uniform Sprite Size {
       bIsDynamic=false
       bIsEditable=false
       bIsStaticSwitch=false
       bIsVisible=false
       type {classStructOrEnum {refPath=&NiagaraFloat}}
      }
      Uniform Sprite Size Min {
       bIsDynamic=false
       bIsEditable=false
       bIsStaticSwitch=false
       bIsVisible=false
       type {classStructOrEnum {refPath=&NiagaraFloat}}
      }
      Sprite Size Min {
       bIsDynamic=false
       bIsEditable=false
       bIsStaticSwitch=false
       bIsVisible=false
       type {classStructOrEnum {refPath=&Vector2f}}
      }
      Sprite Size Max {
       bIsDynamic=false
       bIsEditable=false
       bIsStaticSwitch=false
       bIsVisible=false
       type {classStructOrEnum {refPath=&Vector2f}}
      }
      Uniform Sprite Size Max {
       bIsDynamic=false
       bIsEditable=false
       bIsStaticSwitch=false
       bIsVisible=false
       type {classStructOrEnum {refPath=&NiagaraFloat}}
      }
      Sprite Size Randomness Mode {
       bIsDynamic=false
       bIsEditable=false
       bIsStaticSwitch=true
       bIsVisible=false
       type {classStructOrEnum {refPath=&ENiagaraRandomnessMode}}
      }
      Sprite Rotation Mode {
       bIsDynamic=false
       bIsEditable=true
       bIsStaticSwitch=true
       bIsVisible=true
       type {classStructOrEnum {refPath=&ENiagara_SpriteRotationMode}}
      }
      Sprite Rotation Angle Min {bIsDynamic=false, bIsEditable=true, bIsStaticSwitch=false, bIsVisible=true, type {classStructOrEnum {refPath=&NiagaraFloat}}}
      Sprite Rotation Angle {
       bIsDynamic=false
       bIsEditable=false
       bIsStaticSwitch=false
       bIsVisible=false
       type {classStructOrEnum {refPath=&NiagaraFloat}}
      }
      Sprite Rotation Angle Max {bIsDynamic=false, bIsEditable=true, bIsStaticSwitch=false, bIsVisible=true, type {classStructOrEnum {refPath=&NiagaraFloat}}}
      Sprite Rotation Randomness Mode {
       bIsDynamic=false
       bIsEditable=true
       bIsStaticSwitch=true
       bIsVisible=true
       type {classStructOrEnum {refPath=&ENiagaraRandomnessMode}}
      }
      Sprite UV Mode {
       bIsDynamic=false
       bIsEditable=true
       bIsStaticSwitch=true
       bIsVisible=true
       type {classStructOrEnum {refPath=&ENiagara_UVFlippingMode}}
      }
      Sprite UV Randomness Mode {
       bIsDynamic=false
       bIsEditable=false
       bIsStaticSwitch=true
       bIsVisible=false
       type {classStructOrEnum {refPath=&ENiagaraRandomnessMode}}
      }
      Sprite UV Scale {
       bIsDynamic=false
       bIsEditable=false
       bIsStaticSwitch=false
       bIsVisible=false
       type {classStructOrEnum {refPath=&Vector2f}}
      }
      Mesh Scale Mode {
       bIsDynamic=false
       bIsEditable=true
       bIsStaticSwitch=true
       bIsVisible=true
       type {classStructOrEnum {refPath=&ENiagara_SizeScaleMode}}
      }
      Mesh Uniform Scale Min {bIsDynamic=false, bIsEditable=true, bIsStaticSwitch=false, bIsVisible=true, type {classStructOrEnum {refPath=&NiagaraFloat}}}
      Mesh Uniform Scale {
       bIsDynamic=false
       bIsEditable=false
       bIsStaticSwitch=false
       bIsVisible=false
       type {classStructOrEnum {refPath=&NiagaraFloat}}
      }
      Mesh Scale Min {
       bIsDynamic=false
       bIsEditable=false
       bIsStaticSwitch=false
       bIsVisible=false
       type {classStructOrEnum {refPath=&Vector3f}}
      }
      Mesh Scale {
       bIsDynamic=false
       bIsEditable=false
       bIsStaticSwitch=false
       bIsVisible=false
       type {classStructOrEnum {refPath=&Vector3f}}
      }
      Mesh Scale Max {
       bIsDynamic=false
       bIsEditable=false
       bIsStaticSwitch=false
       bIsVisible=false
       type {classStructOrEnum {refPath=&Vector3f}}
      }
      Mesh Uniform Scale Max {bIsDynamic=false, bIsEditable=true, bIsStaticSwitch=false, bIsVisible=true, type {classStructOrEnum {refPath=&NiagaraFloat}}}
      Mesh Scale Randomness Mode {
       bIsDynamic=false
       bIsEditable=true
       bIsStaticSwitch=true
       bIsVisible=true
       type {classStructOrEnum {refPath=&ENiagaraRandomnessMode}}
      }
      Mesh Renderer Array Visibility Mode {
       bIsDynamic=false
       bIsEditable=true
       bIsStaticSwitch=true
       bIsVisible=true
       type {classStructOrEnum {refPath=&ENiagara_UnsetDirectSetRandom}}
      }
      Mesh Renderer Mesh Index {
       bIsDynamic=false
       bIsEditable=false
       bIsStaticSwitch=false
       bIsVisible=false
       type {classStructOrEnum {refPath=&NiagaraInt32}}
      }
      Wrap Index By Mesh Count {bIsDynamic=false, bIsEditable=false, bIsStaticSwitch=true, bIsVisible=false, type {classStructOrEnum {refPath=&NiagaraBool}}}
      Mesh Renderer Info {
       bIsDynamic=false
       bIsEditable=false
       bIsStaticSwitch=false
       bIsVisible=false
       type {classStructOrEnum {refPath=&NiagaraDataInterfaceMeshRendererInfo}}
      }
      Mesh Renderer Random Seed {
       bIsDynamic=false
       bIsEditable=false
       bIsStaticSwitch=false
       bIsVisible=false
       type {classStructOrEnum {refPath=&NiagaraInt32}}
      }
      Ribbon Width Mode {
       bIsDynamic=false
       bIsEditable=true
       bIsStaticSwitch=true
       bIsVisible=true
       type {classStructOrEnum {refPath=&ENiagara_UnsetDirectSet}}
      }
      Ribbon Width {
       bIsDynamic=false
       bIsEditable=false
       bIsStaticSwitch=false
       bIsVisible=false
       type {classStructOrEnum {refPath=&NiagaraFloat}}
      }
      Ribbon Facing Mode {
       bIsDynamic=false
       bIsEditable=true
       bIsStaticSwitch=true
       bIsVisible=true
       type {classStructOrEnum {refPath=&ENiagara_UnsetDirectSet}}
      }
      Ribbon Facing Vector {
       bIsDynamic=false
       bIsEditable=false
       bIsStaticSwitch=false
       bIsVisible=false
       type {classStructOrEnum {refPath=&Vector3f}}
      }
      Ribbon Twist Mode {
       bIsDynamic=false
       bIsEditable=true
       bIsStaticSwitch=true
       bIsVisible=true
       type {classStructOrEnum {refPath=&ENiagara_UnsetDirectSet}}
      }
      "Ribbon Twist " {
       bIsDynamic=false
       bIsEditable=false
       bIsStaticSwitch=false
       bIsVisible=false
       type {classStructOrEnum {refPath=&NiagaraFloat}}
      }
     ]
     moduleScript {refPath=&InitializeParticle}
    }
    {
     bIsSetParametersModule=true
     enabled=true
     moduleName=SetVariables_#15
     inputs [Particles.Rand {bIsDynamic=true, bIsEditable=true, bIsStaticSwitch=false, bIsVisible=true, type {classStructOrEnum {refPath=&NiagaraFloat}}}]
     moduleScript {refPath=&self:IceSpikes_3.NiagaraScriptSource_0.NiagaraGraph_0.NiagaraNodeAssignment_2.SetVariables_#15}
    }
    {
     bIsSetParametersModule=false
     enabled=true
     moduleName=InitialMeshOrientation
     inputs [
      Mesh Orientation Mode {
       bIsDynamic=false
       bIsEditable=true
       bIsStaticSwitch=true
       bIsVisible=true
       type {classStructOrEnum {refPath=&ENiagara_MeshOrientationOptions}}
      }
      Orientation Quaternion {bIsDynamic=false, bIsEditable=false, bIsStaticSwitch=false, bIsVisible=false, type {classStructOrEnum {refPath=&Quat4f}}}
      Orientation Matrix {
       bIsDynamic=false
       bIsEditable=false
       bIsStaticSwitch=false
       bIsVisible=false
       type {classStructOrEnum {refPath=&NiagaraMatrix}}
      }
      Orientation Vector {
       bIsDynamic=false
       bIsEditable=false
       bIsStaticSwitch=false
       bIsVisible=false
       type {classStructOrEnum {refPath=&Vector3f}}
      }
      Orientation Coordinate Space {
       bIsDynamic=false
       bIsEditable=false
       bIsStaticSwitch=true
       bIsVisible=false
       type {classStructOrEnum {refPath=&ENiagaraCoordinateSpace}}
      }
      Mesh Axis To Orient {
       bIsDynamic=false
       bIsEditable=false
       bIsStaticSwitch=false
       bIsVisible=false
       type {classStructOrEnum {refPath=&Vector3f}}
      }
      Random Seed {
       bIsDynamic=false
       bIsEditable=false
       bIsStaticSwitch=false
       bIsVisible=false
       type {classStructOrEnum {refPath=&NiagaraInt32}}
      }
      Use Rotation Vector {bIsDynamic=false, bIsEditable=true, bIsStaticSwitch=false, bIsVisible=true, type {classStructOrEnum {refPath=&NiagaraBool}}}
      Rotation {bIsDynamic=false, bIsEditable=true, bIsStaticSwitch=false, bIsVisible=true, type {classStructOrEnum {refPath=&Vector3f}}}
      Rotation Coordinate Space {
       bIsDynamic=false
       bIsEditable=true
       bIsStaticSwitch=true
       bIsVisible=true
       type {classStructOrEnum {refPath=&ENiagaraMeshTransforms}}
      }
     ]
     moduleScript {refPath=&InitialMeshOrientation}
    }
   ]
  }
  particleUpdateScript {
   scriptName=ParticleUpdateScript
   modules [
    {
     bIsSetParametersModule=false
     enabled=true
     moduleName=ParticleState
     inputs [
      Kill Particles When Lifetime Has Elapsed {bIsDynamic=false, bIsEditable=true, bIsStaticSwitch=true, bIsVisible=true, type {classStructOrEnum {refPath=&NiagaraBool}}}
      Loop Particles Lifetime {bIsDynamic=false, bIsEditable=false, bIsStaticSwitch=false, bIsVisible=false, type {classStructOrEnum {refPath=&NiagaraBool}}}
      Let Infinitely Lived Particles Die When Emitter Deactivates {bIsDynamic=false, bIsEditable=false, bIsStaticSwitch=false, bIsVisible=false, type {classStructOrEnum {refPath=&NiagaraBool}}}
      Lifetime {bIsDynamic=false, bIsEditable=true, bIsStaticSwitch=false, bIsVisible=true, type {classStructOrEnum {refPath=&NiagaraFloat}}}
      DeltaTime {bIsDynamic=false, bIsEditable=true, bIsStaticSwitch=false, bIsVisible=true, type {classStructOrEnum {refPath=&NiagaraFloat}}}
     ]
     moduleScript {refPath=&ParticleState}
    }
    {
     bIsSetParametersModule=false
     enabled=true
     moduleName=DynamicMaterialParameters
     inputs [
      Write Parameter Index 0 {bIsDynamic=false, bIsEditable=true, bIsStaticSwitch=true, bIsVisible=true, type {classStructOrEnum {refPath=&NiagaraBool}}}
      Module.Param0WriteEnabled {bIsDynamic=false, bIsEditable=true, bIsStaticSwitch=false, bIsVisible=true, type {classStructOrEnum {refPath=&NiagaraBool}}}
      Index 0 Param 1 {bIsDynamic=false, bIsEditable=true, bIsStaticSwitch=false, bIsVisible=true, type {classStructOrEnum {refPath=&NiagaraFloat}}}
      Module.Param1WriteEnabled {bIsDynamic=false, bIsEditable=true, bIsStaticSwitch=false, bIsVisible=true, type {classStructOrEnum {refPath=&NiagaraBool}}}
      Index 0 Param 2 {bIsDynamic=true, bIsEditable=true, bIsStaticSwitch=false, bIsVisible=true, type {classStructOrEnum {refPath=&NiagaraFloat}}}
      Module.Param2WriteEnabled {bIsDynamic=false, bIsEditable=true, bIsStaticSwitch=false, bIsVisible=true, type {classStructOrEnum {refPath=&NiagaraBool}}}
      Index 0 Param 3 {bIsDynamic=true, bIsEditable=true, bIsStaticSwitch=false, bIsVisible=true, type {classStructOrEnum {refPath=&NiagaraFloat}}}
      Module.Param3WriteEnabled {bIsDynamic=false, bIsEditable=true, bIsStaticSwitch=false, bIsVisible=true, type {classStructOrEnum {refPath=&NiagaraBool}}}
      Index 0 Param 4 {bIsDynamic=true, bIsEditable=true, bIsStaticSwitch=false, bIsVisible=true, type {classStructOrEnum {refPath=&NiagaraFloat}}}
      Write Parameter Index 1 {bIsDynamic=false, bIsEditable=true, bIsStaticSwitch=true, bIsVisible=true, type {classStructOrEnum {refPath=&NiagaraBool}}}
      Index 1 Param 1 {
       bIsDynamic=false
       bIsEditable=false
       bIsStaticSwitch=false
       bIsVisible=false
       type {classStructOrEnum {refPath=&NiagaraFloat}}
      }
      Index 1 Param 2 {
       bIsDynamic=false
       bIsEditable=false
       bIsStaticSwitch=false
       bIsVisible=false
       type {classStructOrEnum {refPath=&NiagaraFloat}}
      }
      Index 1 Param 3 {
       bIsDynamic=false
       bIsEditable=false
       bIsStaticSwitch=false
       bIsVisible=false
       type {classStructOrEnum {refPath=&NiagaraFloat}}
      }
      Index 1 Param 4 {
       bIsDynamic=false
       bIsEditable=false
       bIsStaticSwitch=false
       bIsVisible=false
       type {classStructOrEnum {refPath=&NiagaraFloat}}
      }
      Write Parameter Index 2 {bIsDynamic=false, bIsEditable=true, bIsStaticSwitch=true, bIsVisible=true, type {classStructOrEnum {refPath=&NiagaraBool}}}
      Index 2 Param 1 {
       bIsDynamic=false
       bIsEditable=false
       bIsStaticSwitch=false
       bIsVisible=false
       type {classStructOrEnum {refPath=&NiagaraFloat}}
      }
      Index 2 Param 2 {
       bIsDynamic=false
       bIsEditable=false
       bIsStaticSwitch=false
       bIsVisible=false
       type {classStructOrEnum {refPath=&NiagaraFloat}}
      }
      Index 2 Param 3 {
       bIsDynamic=false
       bIsEditable=false
       bIsStaticSwitch=false
       bIsVisible=false
       type {classStructOrEnum {refPath=&NiagaraFloat}}
      }
      Index 2 Param 4 {
       bIsDynamic=false
       bIsEditable=false
       bIsStaticSwitch=false
       bIsVisible=false
       type {classStructOrEnum {refPath=&NiagaraFloat}}
      }
      Write Parameter Index 3 {bIsDynamic=false, bIsEditable=true, bIsStaticSwitch=true, bIsVisible=true, type {classStructOrEnum {refPath=&NiagaraBool}}}
      Index 3 Param 1 {
       bIsDynamic=false
       bIsEditable=false
       bIsStaticSwitch=false
       bIsVisible=false
       type {classStructOrEnum {refPath=&NiagaraFloat}}
      }
      Index 3 Param 2 {
       bIsDynamic=false
       bIsEditable=false
       bIsStaticSwitch=false
       bIsVisible=false
       type {classStructOrEnum {refPath=&NiagaraFloat}}
      }
      Index 3 Param 3 {
       bIsDynamic=false
       bIsEditable=false
       bIsStaticSwitch=false
       bIsVisible=false
       type {classStructOrEnum {refPath=&NiagaraFloat}}
      }
      Index 3 Param 4 {
       bIsDynamic=false
       bIsEditable=false
       bIsStaticSwitch=false
       bIsVisible=false
       type {classStructOrEnum {refPath=&NiagaraFloat}}
      }
     ]
     moduleScript {refPath=&DynamicMaterialParameters}
    }
    {
     bIsSetParametersModule=false
     enabled=true
     moduleName=ScaleMeshSize
     inputs [
      Scale Factor {bIsDynamic=true, bIsEditable=true, bIsStaticSwitch=false, bIsVisible=true, type {classStructOrEnum {refPath=&Vector3f}}}
      Initial Mesh Scale {bIsDynamic=false, bIsEditable=true, bIsStaticSwitch=false, bIsVisible=true, type {classStructOrEnum {refPath=&Vector3f}}}
     ]
     moduleScript {refPath=&ScaleMeshSize}
    }
    {
     bIsSetParametersModule=false
     enabled=false
     moduleName=ScaleMeshSize001
     inputs [
      Scale Factor {bIsDynamic=true, bIsEditable=true, bIsStaticSwitch=false, bIsVisible=true, type {classStructOrEnum {refPath=&Vector3f}}}
      Initial Mesh Scale {bIsDynamic=false, bIsEditable=true, bIsStaticSwitch=false, bIsVisible=true, type {classStructOrEnum {refPath=&Vector3f}}}
     ]
     moduleScript {refPath=&ScaleMeshSize}
    }
    {
     bIsSetParametersModule=false
     enabled=true
     moduleName=ScaleColor
     inputs [
      Scale Mode {
       bIsDynamic=false
       bIsEditable=true
       bIsStaticSwitch=true
       bIsVisible=true
       type {classStructOrEnum {refPath=&ENiagaraScaleColorMode}}
      }
      Linear Color Curve {
       bIsDynamic=false
       bIsEditable=false
       bIsStaticSwitch=false
       bIsVisible=false
       type {classStructOrEnum {refPath=&NiagaraDataInterfaceColorCurve}}
      }
      ScaleRGB {bIsDynamic=false, bIsEditable=true, bIsStaticSwitch=false, bIsVisible=true, type {classStructOrEnum {refPath=&NiagaraBool}}}
      Scale RGB {bIsDynamic=true, bIsEditable=true, bIsStaticSwitch=false, bIsVisible=true, type {classStructOrEnum {refPath=&Vector3f}}}
      ScaleRGBA {bIsDynamic=false, bIsEditable=false, bIsStaticSwitch=false, bIsVisible=false, type {classStructOrEnum {refPath=&NiagaraBool}}}
      Scale RGBA {
       bIsDynamic=false
       bIsEditable=false
       bIsStaticSwitch=false
       bIsVisible=false
       type {classStructOrEnum {refPath=&Vector4f}}
      }
      Curve Index {
       bIsDynamic=false
       bIsEditable=false
       bIsStaticSwitch=false
       bIsVisible=false
       type {classStructOrEnum {refPath=&NiagaraFloat}}
      }
      ScaleA {bIsDynamic=false, bIsEditable=true, bIsStaticSwitch=false, bIsVisible=true, type {classStructOrEnum {refPath=&NiagaraBool}}}
      Scale Alpha {bIsDynamic=true, bIsEditable=true, bIsStaticSwitch=false, bIsVisible=true, type {classStructOrEnum {refPath=&NiagaraFloat}}}
      Color Value To Scale {
       bIsDynamic=false
       bIsEditable=true
       bIsStaticSwitch=false
       bIsVisible=true
       type {classStructOrEnum {refPath=&LinearColor}}
      }
     ]
     moduleScript {refPath=&ScaleColor}
    }
    {
     bIsSetParametersModule=false
     enabled=true
     moduleName=SolveForcesAndVelocity
     inputs [
      Clamp Velocity {bIsDynamic=false, bIsEditable=true, bIsStaticSwitch=false, bIsVisible=true, type {classStructOrEnum {refPath=&NiagaraBool}}}
      Speed Limit {bIsDynamic=false, bIsEditable=false, bIsStaticSwitch=false, bIsVisible=true, type {classStructOrEnum {refPath=&NiagaraFloat}}}
      Limit Acceleration {bIsDynamic=false, bIsEditable=true, bIsStaticSwitch=false, bIsVisible=true, type {classStructOrEnum {refPath=&NiagaraBool}}}
      Acceleration Limit {bIsDynamic=false, bIsEditable=false, bIsStaticSwitch=false, bIsVisible=true, type {classStructOrEnum {refPath=&NiagaraFloat}}}
      Manually Enable Rotational Solver {bIsDynamic=false, bIsEditable=true, bIsStaticSwitch=false, bIsVisible=true, type {classStructOrEnum {refPath=&NiagaraBool}}}
      Write to Intrinsic Properties {bIsDynamic=false, bIsEditable=true, bIsStaticSwitch=false, bIsVisible=true, type {classStructOrEnum {refPath=&NiagaraBool}}}
      Write to Presolve Properties {bIsDynamic=false, bIsEditable=true, bIsStaticSwitch=true, bIsVisible=true, type {classStructOrEnum {refPath=&NiagaraBool}}}
      DeltaTime {bIsDynamic=false, bIsEditable=true, bIsStaticSwitch=false, bIsVisible=true, type {classStructOrEnum {refPath=&NiagaraFloat}}}
      Mass {bIsDynamic=false, bIsEditable=true, bIsStaticSwitch=false, bIsVisible=true, type {classStructOrEnum {refPath=&NiagaraFloat}}}
      Position {
       bIsDynamic=false
       bIsEditable=true
       bIsStaticSwitch=false
       bIsVisible=true
       type {classStructOrEnum {refPath=&NiagaraPosition}}
      }
      Velocity {bIsDynamic=false, bIsEditable=true, bIsStaticSwitch=false, bIsVisible=true, type {classStructOrEnum {refPath=&Vector3f}}}
      PreviousVelocity {bIsDynamic=false, bIsEditable=true, bIsStaticSwitch=false, bIsVisible=true, type {classStructOrEnum {refPath=&Vector3f}}}
      Force {bIsDynamic=false, bIsEditable=true, bIsStaticSwitch=false, bIsVisible=true, type {classStructOrEnum {refPath=&Vector3f}}}
      Mesh Orientation {bIsDynamic=false, bIsEditable=false, bIsStaticSwitch=false, bIsVisible=false, type {classStructOrEnum {refPath=&Quat4f}}}
      Rotational Velocity {
       bIsDynamic=false
       bIsEditable=false
       bIsStaticSwitch=false
       bIsVisible=false
       type {classStructOrEnum {refPath=&Vector3f}}
      }
      Rotational Force {
       bIsDynamic=false
       bIsEditable=false
       bIsStaticSwitch=false
       bIsVisible=false
       type {classStructOrEnum {refPath=&Vector3f}}
      }
      Rotational Inertia {
       bIsDynamic=false
       bIsEditable=false
       bIsStaticSwitch=false
       bIsVisible=false
       type {classStructOrEnum {refPath=&NiagaraFloat}}
      }
     ]
     moduleScript {refPath=&SolveForcesAndVelocity}
    }
    {
     bIsSetParametersModule=false
     enabled=true
     moduleName=ScaleColor001
     inputs [
      Scale Mode {bIsDynamic=false, bIsEditable=true, bIsStaticSwitch=true, bIsVisible=true, type {classStructOrEnum {refPath=&NiagaraInt32}}}
      Linear Color Curve {
       bIsDynamic=false
       bIsEditable=false
       bIsStaticSwitch=false
       bIsVisible=false
       type {classStructOrEnum {refPath=&NiagaraDataInterfaceColorCurve}}
      }
      ScaleRGB {bIsDynamic=false, bIsEditable=true, bIsStaticSwitch=false, bIsVisible=true, type {classStructOrEnum {refPath=&NiagaraBool}}}
      Scale RGB {bIsDynamic=true, bIsEditable=true, bIsStaticSwitch=false, bIsVisible=true, type {classStructOrEnum {refPath=&Vector3f}}}
      ScaleRGBA {bIsDynamic=false, bIsEditable=false, bIsStaticSwitch=false, bIsVisible=false, type {classStructOrEnum {refPath=&NiagaraBool}}}
      Scale RGBA {
       bIsDynamic=false
       bIsEditable=false
       bIsStaticSwitch=false
       bIsVisible=false
       type {classStructOrEnum {refPath=&Vector4f}}
      }
      Curve Index {
       bIsDynamic=false
       bIsEditable=false
       bIsStaticSwitch=false
       bIsVisible=false
       type {classStructOrEnum {refPath=&NiagaraFloat}}
      }
      ScaleA {bIsDynamic=false, bIsEditable=true, bIsStaticSwitch=false, bIsVisible=true, type {classStructOrEnum {refPath=&NiagaraBool}}}
      Scale Alpha {bIsDynamic=false, bIsEditable=false, bIsStaticSwitch=false, bIsVisible=true, type {classStructOrEnum {refPath=&NiagaraFloat}}}
      Color Value To Scale {
       bIsDynamic=false
       bIsEditable=true
       bIsStaticSwitch=false
       bIsVisible=true
       type {classStructOrEnum {refPath=&LinearColor}}
      }
      Color Scale {
       bIsDynamic=false
       bIsEditable=false
       bIsStaticSwitch=false
       bIsVisible=false
       type {classStructOrEnum {refPath=&LinearColor}}
      }
     ]
     moduleScript {refPath=&ScaleColor}
    }
   ]
  }
  rendererClasses [{refPath=&NiagaraMeshRendererProperties}]
  renderers [{rendererIndex=0, rendererClass {refPath=&NiagaraMeshRendererProperties}}]
 }
 NE_Debris_02 {
  bEnabled=false
  emitterSpawnScript {scriptName=EmitterSpawnScript, modules []}
  emitterUpdateScript {
   scriptName=EmitterUpdateScript
   modules [
    {
     bIsSetParametersModule=false
     enabled=true
     moduleName=EmitterState
     inputs [
      Life Cycle Mode {
       bIsDynamic=false
       bIsEditable=true
       bIsStaticSwitch=true
       bIsVisible=true
       type {classStructOrEnum {refPath=&ENiagaraEmitterLifeCycleMode}}
      }
      Inactive Response {
       bIsDynamic=false
       bIsEditable=false
       bIsStaticSwitch=true
       bIsVisible=false
       type {classStructOrEnum {refPath=&ENiagaraInactiveMode}}
      }
      Loop Behavior {
       bIsDynamic=false
       bIsEditable=false
       bIsStaticSwitch=true
       bIsVisible=false
       type {classStructOrEnum {refPath=&ENiagara_EmitterStateOptions}}
      }
      Loop Duration Mode {
       bIsDynamic=false
       bIsEditable=false
       bIsStaticSwitch=true
       bIsVisible=false
       type {classStructOrEnum {refPath=&ENiagara_InfiniteLoopDuration}}
      }
      Loop Count {
       bIsDynamic=false
       bIsEditable=false
       bIsStaticSwitch=false
       bIsVisible=false
       type {classStructOrEnum {refPath=&NiagaraInt32}}
      }
      Loop Duration {
       bIsDynamic=false
       bIsEditable=false
       bIsStaticSwitch=false
       bIsVisible=false
       type {classStructOrEnum {refPath=&NiagaraFloat}}
      }
      Recalculate Duration Each Loop {bIsDynamic=false, bIsEditable=false, bIsStaticSwitch=false, bIsVisible=false, type {classStructOrEnum {refPath=&NiagaraBool}}}
      UseLoopDelay {bIsDynamic=false, bIsEditable=false, bIsStaticSwitch=false, bIsVisible=false, type {classStructOrEnum {refPath=&NiagaraBool}}}
      Loop Delay {
       bIsDynamic=false
       bIsEditable=false
       bIsStaticSwitch=false
       bIsVisible=false
       type {classStructOrEnum {refPath=&NiagaraFloat}}
      }
      Delay First Loop Only {bIsDynamic=false, bIsEditable=false, bIsStaticSwitch=false, bIsVisible=false, type {classStructOrEnum {refPath=&NiagaraBool}}}
      Scalability Mode {
       bIsDynamic=false
       bIsEditable=true
       bIsStaticSwitch=true
       bIsVisible=true
       type {classStructOrEnum {refPath=&ENiagaraEmitterScalabilityMode_Limited}}
      }
      Enable Distance Culling {bIsDynamic=false, bIsEditable=false, bIsStaticSwitch=true, bIsVisible=false, type {classStructOrEnum {refPath=&NiagaraBool}}}
      MinDistance {
       bIsDynamic=false
       bIsEditable=false
       bIsStaticSwitch=false
       bIsVisible=false
       type {classStructOrEnum {refPath=&NiagaraFloat}}
      }
      Min Distance Response {
       bIsDynamic=false
       bIsEditable=false
       bIsStaticSwitch=false
       bIsVisible=false
       type {classStructOrEnum {refPath=&ENiagaraExecutionStateManagement}}
      }
      MaxDistance {
       bIsDynamic=false
       bIsEditable=false
       bIsStaticSwitch=false
       bIsVisible=false
       type {classStructOrEnum {refPath=&NiagaraFloat}}
      }
      Max Distance Response {
       bIsDynamic=false
       bIsEditable=false
       bIsStaticSwitch=false
       bIsVisible=false
       type {classStructOrEnum {refPath=&ENiagaraExecutionStateManagement}}
      }
      Scale Spawn Count {bIsDynamic=false, bIsEditable=false, bIsStaticSwitch=true, bIsVisible=false, type {classStructOrEnum {refPath=&NiagaraBool}}}
      Spawn Count Scale {
       bIsDynamic=false
       bIsEditable=false
       bIsStaticSwitch=false
       bIsVisible=false
       type {classStructOrEnum {refPath=&NiagaraFloat}}
      }
      Spawn Count Scale By Distance {
       bIsDynamic=false
       bIsEditable=false
       bIsStaticSwitch=false
       bIsVisible=false
       type {classStructOrEnum {refPath=&NiagaraDataInterfaceCurve}}
      }
      Enable Visibility Culling {bIsDynamic=false, bIsEditable=false, bIsStaticSwitch=true, bIsVisible=false, type {classStructOrEnum {refPath=&NiagaraBool}}}
      Visibility Cull Response {
       bIsDynamic=false
       bIsEditable=false
       bIsStaticSwitch=false
       bIsVisible=false
       type {classStructOrEnum {refPath=&ENiagaraExecutionStateManagement}}
      }
      Visibility Cull Delay {
       bIsDynamic=false
       bIsEditable=false
       bIsStaticSwitch=false
       bIsVisible=false
       type {classStructOrEnum {refPath=&NiagaraFloat}}
      }
      Reset Age On Awaken {bIsDynamic=false, bIsEditable=false, bIsStaticSwitch=false, bIsVisible=false, type {classStructOrEnum {refPath=&NiagaraBool}}}
     ]
     moduleScript {refPath=&EmitterState}
    }
    {
     bIsSetParametersModule=false
     enabled=true
     moduleName=SpawnBurst_Instantaneous
     inputs [
      Spawn Count {bIsDynamic=false, bIsEditable=true, bIsStaticSwitch=false, bIsVisible=true, type {classStructOrEnum {refPath=&NiagaraInt32}}}
      Spawn Time {bIsDynamic=false, bIsEditable=true, bIsStaticSwitch=false, bIsVisible=true, type {classStructOrEnum {refPath=&NiagaraFloat}}}
      Use Spawn Probability {bIsDynamic=false, bIsEditable=true, bIsStaticSwitch=false, bIsVisible=true, type {classStructOrEnum {refPath=&NiagaraBool}}}
      Spawn Probability {bIsDynamic=false, bIsEditable=false, bIsStaticSwitch=false, bIsVisible=true, type {classStructOrEnum {refPath=&NiagaraFloat}}}
      UseLoopCountLimit {bIsDynamic=false, bIsEditable=true, bIsStaticSwitch=false, bIsVisible=true, type {classStructOrEnum {refPath=&NiagaraBool}}}
      Loop Count Limit {bIsDynamic=false, bIsEditable=false, bIsStaticSwitch=false, bIsVisible=true, type {classStructOrEnum {refPath=&NiagaraInt32}}}
      Spawn Group {bIsDynamic=false, bIsEditable=true, bIsStaticSwitch=false, bIsVisible=true, type {classStructOrEnum {refPath=&NiagaraInt32}}}
      Age {bIsDynamic=false, bIsEditable=true, bIsStaticSwitch=false, bIsVisible=true, type {classStructOrEnum {refPath=&NiagaraFloat}}}
     ]
     moduleScript {refPath=&SpawnBurst_Instantaneous}
    }
    {
     bIsSetParametersModule=false
     enabled=true
     moduleName=SpawnBurst_Instantaneous001
     inputs [
      Spawn Count {bIsDynamic=false, bIsEditable=true, bIsStaticSwitch=false, bIsVisible=true, type {classStructOrEnum {refPath=&NiagaraInt32}}}
      Spawn Time {bIsDynamic=false, bIsEditable=true, bIsStaticSwitch=false, bIsVisible=true, type {classStructOrEnum {refPath=&NiagaraFloat}}}
      Use Spawn Probability {bIsDynamic=false, bIsEditable=true, bIsStaticSwitch=false, bIsVisible=true, type {classStructOrEnum {refPath=&NiagaraBool}}}
      Spawn Probability {bIsDynamic=false, bIsEditable=false, bIsStaticSwitch=false, bIsVisible=true, type {classStructOrEnum {refPath=&NiagaraFloat}}}
      UseLoopCountLimit {bIsDynamic=false, bIsEditable=true, bIsStaticSwitch=false, bIsVisible=true, type {classStructOrEnum {refPath=&NiagaraBool}}}
      Loop Count Limit {bIsDynamic=false, bIsEditable=false, bIsStaticSwitch=false, bIsVisible=true, type {classStructOrEnum {refPath=&NiagaraInt32}}}
      Spawn Group {bIsDynamic=false, bIsEditable=true, bIsStaticSwitch=false, bIsVisible=true, type {classStructOrEnum {refPath=&NiagaraInt32}}}
      Age {bIsDynamic=false, bIsEditable=true, bIsStaticSwitch=false, bIsVisible=true, type {classStructOrEnum {refPath=&NiagaraFloat}}}
     ]
     moduleScript {refPath=&SpawnBurst_Instantaneous}
    }
   ]
  }
  particleSpawnScript {
   scriptName=ParticleSpawnScript
   modules [
    {
     bIsSetParametersModule=false
     enabled=true
     moduleName=InitializeParticle
     inputs [
      Lifetime Mode {
       bIsDynamic=false
       bIsEditable=true
       bIsStaticSwitch=true
       bIsVisible=true
       type {classStructOrEnum {refPath=&ENiagara_LifetimeMode}}
      }
      Lifetime Min {bIsDynamic=false, bIsEditable=true, bIsStaticSwitch=false, bIsVisible=true, type {classStructOrEnum {refPath=&NiagaraFloat}}}
      Lifetime {
       bIsDynamic=false
       bIsEditable=false
       bIsStaticSwitch=false
       bIsVisible=false
       type {classStructOrEnum {refPath=&NiagaraFloat}}
      }
      Lifetime Max {bIsDynamic=false, bIsEditable=true, bIsStaticSwitch=false, bIsVisible=true, type {classStructOrEnum {refPath=&NiagaraFloat}}}
      Lifetime Randomness Mode {
       bIsDynamic=false
       bIsEditable=true
       bIsStaticSwitch=true
       bIsVisible=true
       type {classStructOrEnum {refPath=&ENiagaraRandomnessMode}}
      }
      Color Mode {
       bIsDynamic=false
       bIsEditable=true
       bIsStaticSwitch=true
       bIsVisible=true
       type {classStructOrEnum {refPath=&ENiagara_ColorInitializationMode}}
      }
      Color Minimum {
       bIsDynamic=false
       bIsEditable=true
       bIsStaticSwitch=false
       bIsVisible=true
       type {classStructOrEnum {refPath=&LinearColor}}
      }
      Color {
       bIsDynamic=false
       bIsEditable=false
       bIsStaticSwitch=false
       bIsVisible=false
       type {classStructOrEnum {refPath=&LinearColor}}
      }
      AdjustHue {bIsDynamic=false, bIsEditable=false, bIsStaticSwitch=false, bIsVisible=false, type {classStructOrEnum {refPath=&NiagaraBool}}}
      Hue Shift Range {
       bIsDynamic=false
       bIsEditable=false
       bIsStaticSwitch=false
       bIsVisible=false
       type {classStructOrEnum {refPath=&Vector2f}}
      }
      Color Maximum {
       bIsDynamic=false
       bIsEditable=true
       bIsStaticSwitch=false
       bIsVisible=true
       type {classStructOrEnum {refPath=&LinearColor}}
      }
      AdjustSaturation {bIsDynamic=false, bIsEditable=false, bIsStaticSwitch=false, bIsVisible=false, type {classStructOrEnum {refPath=&NiagaraBool}}}
      Saturation Range {
       bIsDynamic=false
       bIsEditable=false
       bIsStaticSwitch=false
       bIsVisible=false
       type {classStructOrEnum {refPath=&Vector2f}}
      }
      AdjustValue {bIsDynamic=false, bIsEditable=false, bIsStaticSwitch=false, bIsVisible=false, type {classStructOrEnum {refPath=&NiagaraBool}}}
      Value Range {
       bIsDynamic=false
       bIsEditable=false
       bIsStaticSwitch=false
       bIsVisible=false
       type {classStructOrEnum {refPath=&Vector2f}}
      }
      AdjustAlpha {bIsDynamic=false, bIsEditable=false, bIsStaticSwitch=false, bIsVisible=false, type {classStructOrEnum {refPath=&NiagaraBool}}}
      Alpha Scale Range {
       bIsDynamic=false
       bIsEditable=false
       bIsStaticSwitch=false
       bIsVisible=false
       type {classStructOrEnum {refPath=&Vector2f}}
      }
      Color Channel Mode {
       bIsDynamic=false
       bIsEditable=true
       bIsStaticSwitch=true
       bIsVisible=true
       type {classStructOrEnum {refPath=&ENiagaraChannelCorrelation}}
      }
      Color Randomness Mode {
       bIsDynamic=false
       bIsEditable=true
       bIsStaticSwitch=true
       bIsVisible=true
       type {classStructOrEnum {refPath=&ENiagaraRandomnessMode}}
      }
      Position Mode {
       bIsDynamic=false
       bIsEditable=true
       bIsStaticSwitch=true
       bIsVisible=true
       type {classStructOrEnum {refPath=&ENiagara_PositionInitializationMode}}
      }
      Position {
       bIsDynamic=false
       bIsEditable=false
       bIsStaticSwitch=false
       bIsVisible=false
       type {classStructOrEnum {refPath=&NiagaraPosition}}
      }
      UsePositionOffset {bIsDynamic=false, bIsEditable=true, bIsStaticSwitch=false, bIsVisible=true, type {classStructOrEnum {refPath=&NiagaraBool}}}
      Position Offset {bIsDynamic=false, bIsEditable=true, bIsStaticSwitch=false, bIsVisible=true, type {classStructOrEnum {refPath=&Vector3f}}}
      Position Offset Coordinate Space {
       bIsDynamic=false
       bIsEditable=true
       bIsStaticSwitch=false
       bIsVisible=true
       type {classStructOrEnum {refPath=&ENiagaraCoordinateSpace}}
      }
      Mass Mode {
       bIsDynamic=false
       bIsEditable=true
       bIsStaticSwitch=true
       bIsVisible=true
       type {classStructOrEnum {refPath=&ENiagara_MassInitializationMode}}
      }
      Mass Min {
       bIsDynamic=false
       bIsEditable=false
       bIsStaticSwitch=false
       bIsVisible=false
       type {classStructOrEnum {refPath=&NiagaraFloat}}
      }
      Mass {bIsDynamic=false, bIsEditable=true, bIsStaticSwitch=false, bIsVisible=true, type {classStructOrEnum {refPath=&NiagaraFloat}}}
      Mass Max {
       bIsDynamic=false
       bIsEditable=false
       bIsStaticSwitch=false
       bIsVisible=false
       type {classStructOrEnum {refPath=&NiagaraFloat}}
      }
      Mass Randomness Mode {
       bIsDynamic=false
       bIsEditable=false
       bIsStaticSwitch=true
       bIsVisible=false
       type {classStructOrEnum {refPath=&ENiagaraRandomnessMode}}
      }
      Write Material Random {bIsDynamic=false, bIsEditable=true, bIsStaticSwitch=false, bIsVisible=true, type {classStructOrEnum {refPath=&NiagaraBool}}}
      Material Random {bIsDynamic=false, bIsEditable=true, bIsStaticSwitch=false, bIsVisible=true, type {classStructOrEnum {refPath=&NiagaraFloat}}}
      Sprite Size Mode {
       bIsDynamic=false
       bIsEditable=true
       bIsStaticSwitch=true
       bIsVisible=true
       type {classStructOrEnum {refPath=&ENiagara_SizeScaleMode}}
      }
      Sprite Size {
       bIsDynamic=false
       bIsEditable=false
       bIsStaticSwitch=false
       bIsVisible=false
       type {classStructOrEnum {refPath=&Vector2f}}
      }
      Uniform Sprite Size {
       bIsDynamic=false
       bIsEditable=false
       bIsStaticSwitch=false
       bIsVisible=false
       type {classStructOrEnum {refPath=&NiagaraFloat}}
      }
      Uniform Sprite Size Min {
       bIsDynamic=false
       bIsEditable=false
       bIsStaticSwitch=false
       bIsVisible=false
       type {classStructOrEnum {refPath=&NiagaraFloat}}
      }
      Sprite Size Min {
       bIsDynamic=false
       bIsEditable=false
       bIsStaticSwitch=false
       bIsVisible=false
       type {classStructOrEnum {refPath=&Vector2f}}
      }
      Sprite Size Max {
       bIsDynamic=false
       bIsEditable=false
       bIsStaticSwitch=false
       bIsVisible=false
       type {classStructOrEnum {refPath=&Vector2f}}
      }
      Uniform Sprite Size Max {
       bIsDynamic=false
       bIsEditable=false
       bIsStaticSwitch=false
       bIsVisible=false
       type {classStructOrEnum {refPath=&NiagaraFloat}}
      }
      Sprite Size Randomness Mode {
       bIsDynamic=false
       bIsEditable=false
       bIsStaticSwitch=true
       bIsVisible=false
       type {classStructOrEnum {refPath=&ENiagaraRandomnessMode}}
      }
      Sprite Rotation Mode {
       bIsDynamic=false
       bIsEditable=true
       bIsStaticSwitch=true
       bIsVisible=true
       type {classStructOrEnum {refPath=&ENiagara_SpriteRotationMode}}
      }
      Sprite Rotation Angle Min {
       bIsDynamic=false
       bIsEditable=false
       bIsStaticSwitch=false
       bIsVisible=false
       type {classStructOrEnum {refPath=&NiagaraFloat}}
      }
      Sprite Rotation Angle {
       bIsDynamic=false
       bIsEditable=false
       bIsStaticSwitch=false
       bIsVisible=false
       type {classStructOrEnum {refPath=&NiagaraFloat}}
      }
      Sprite Rotation Angle Max {
       bIsDynamic=false
       bIsEditable=false
       bIsStaticSwitch=false
       bIsVisible=false
       type {classStructOrEnum {refPath=&NiagaraFloat}}
      }
      Sprite Rotation Randomness Mode {
       bIsDynamic=false
       bIsEditable=false
       bIsStaticSwitch=true
       bIsVisible=false
       type {classStructOrEnum {refPath=&ENiagaraRandomnessMode}}
      }
      Sprite UV Mode {
       bIsDynamic=false
       bIsEditable=true
       bIsStaticSwitch=true
       bIsVisible=true
       type {classStructOrEnum {refPath=&ENiagara_UVFlippingMode}}
      }
      Sprite UV Randomness Mode {
       bIsDynamic=false
       bIsEditable=false
       bIsStaticSwitch=true
       bIsVisible=false
       type {classStructOrEnum {refPath=&ENiagaraRandomnessMode}}
      }
      Sprite UV Scale {
       bIsDynamic=false
       bIsEditable=false
       bIsStaticSwitch=false
       bIsVisible=false
       type {classStructOrEnum {refPath=&Vector2f}}
      }
      Mesh Scale Mode {
       bIsDynamic=false
       bIsEditable=true
       bIsStaticSwitch=true
       bIsVisible=true
       type {classStructOrEnum {refPath=&ENiagara_SizeScaleMode}}
      }
      Mesh Uniform Scale Min {
       bIsDynamic=false
       bIsEditable=false
       bIsStaticSwitch=false
       bIsVisible=false
       type {classStructOrEnum {refPath=&NiagaraFloat}}
      }
      Mesh Uniform Scale {
       bIsDynamic=false
       bIsEditable=false
       bIsStaticSwitch=false
       bIsVisible=false
       type {classStructOrEnum {refPath=&NiagaraFloat}}
      }
      Mesh Scale Min {bIsDynamic=false, bIsEditable=true, bIsStaticSwitch=false, bIsVisible=true, type {classStructOrEnum {refPath=&Vector3f}}}
      Mesh Scale {
       bIsDynamic=false
       bIsEditable=false
       bIsStaticSwitch=false
       bIsVisible=false
       type {classStructOrEnum {refPath=&Vector3f}}
      }
      Mesh Scale Max {bIsDynamic=false, bIsEditable=true, bIsStaticSwitch=false, bIsVisible=true, type {classStructOrEnum {refPath=&Vector3f}}}
      Mesh Uniform Scale Max {
       bIsDynamic=false
       bIsEditable=false
       bIsStaticSwitch=false
       bIsVisible=false
       type {classStructOrEnum {refPath=&NiagaraFloat}}
      }
      Mesh Scale Randomness Mode {
       bIsDynamic=false
       bIsEditable=true
       bIsStaticSwitch=true
       bIsVisible=true
       type {classStructOrEnum {refPath=&ENiagaraRandomnessMode}}
      }
      Mesh Renderer Array Visibility Mode {
       bIsDynamic=false
       bIsEditable=true
       bIsStaticSwitch=true
       bIsVisible=true
       type {classStructOrEnum {refPath=&ENiagara_UnsetDirectSetRandom}}
      }
      Mesh Renderer Mesh Index {
       bIsDynamic=false
       bIsEditable=false
       bIsStaticSwitch=false
       bIsVisible=false
       type {classStructOrEnum {refPath=&NiagaraInt32}}
      }
      Wrap Index By Mesh Count {bIsDynamic=false, bIsEditable=false, bIsStaticSwitch=true, bIsVisible=false, type {classStructOrEnum {refPath=&NiagaraBool}}}
      Mesh Renderer Info {
       bIsDynamic=false
       bIsEditable=false
       bIsStaticSwitch=false
       bIsVisible=false
       type {classStructOrEnum {refPath=&NiagaraDataInterfaceMeshRendererInfo}}
      }
      Mesh Renderer Random Seed {
       bIsDynamic=false
       bIsEditable=false
       bIsStaticSwitch=false
       bIsVisible=false
       type {classStructOrEnum {refPath=&NiagaraInt32}}
      }
      Ribbon Width Mode {
       bIsDynamic=false
       bIsEditable=true
       bIsStaticSwitch=true
       bIsVisible=true
       type {classStructOrEnum {refPath=&ENiagara_UnsetDirectSet}}
      }
      Ribbon Width {
       bIsDynamic=false
       bIsEditable=false
       bIsStaticSwitch=false
       bIsVisible=false
       type {classStructOrEnum {refPath=&NiagaraFloat}}
      }
      Ribbon Facing Mode {
       bIsDynamic=false
       bIsEditable=true
       bIsStaticSwitch=true
       bIsVisible=true
       type {classStructOrEnum {refPath=&ENiagara_UnsetDirectSet}}
      }
      Ribbon Facing Vector {
       bIsDynamic=false
       bIsEditable=false
       bIsStaticSwitch=false
       bIsVisible=false
       type {classStructOrEnum {refPath=&Vector3f}}
      }
      Ribbon Twist Mode {
       bIsDynamic=false
       bIsEditable=true
       bIsStaticSwitch=true
       bIsVisible=true
       type {classStructOrEnum {refPath=&ENiagara_UnsetDirectSet}}
      }
      "Ribbon Twist " {
       bIsDynamic=false
       bIsEditable=false
       bIsStaticSwitch=false
       bIsVisible=false
       type {classStructOrEnum {refPath=&NiagaraFloat}}
      }
     ]
     moduleScript {refPath=&InitializeParticle}
    }
    {
     bIsSetParametersModule=false
     enabled=true
     moduleName=ShapeLocation
     inputs [
      Shape Primitive {
       bIsDynamic=false
       bIsEditable=true
       bIsStaticSwitch=true
       bIsVisible=true
       type {classStructOrEnum {refPath=/Niagara/Enums/Location/ENiagara_LocationShapes}}
      }
      Box / Plane Mode {
       bIsDynamic=false
       bIsEditable=false
       bIsStaticSwitch=true
       bIsVisible=false
       type {classStructOrEnum {refPath=/Niagara/Enums/Location/ENiagara_BoxPlaneMode}}
      }
      Cone Mode {
       bIsDynamic=false
       bIsEditable=false
       bIsStaticSwitch=true
       bIsVisible=false
       type {classStructOrEnum {refPath=/Niagara/Enums/Location/ENiagara_ConeMode}}
      }
      Cylinder Height {
       bIsDynamic=false
       bIsEditable=false
       bIsStaticSwitch=false
       bIsVisible=false
       type {classStructOrEnum {refPath=&NiagaraFloat}}
      }
      Sphere Radius {bIsDynamic=true, bIsEditable=true, bIsStaticSwitch=false, bIsVisible=true, type {classStructOrEnum {refPath=&NiagaraFloat}}}
      Ring / Disc Mode {
       bIsDynamic=false
       bIsEditable=false
       bIsStaticSwitch=true
       bIsVisible=false
       type {classStructOrEnum {refPath=/Niagara/Enums/Location/ENiagaraRingDiscMode}}
      }
      Torus Mode {
       bIsDynamic=false
       bIsEditable=false
       bIsStaticSwitch=true
       bIsVisible=false
       type {classStructOrEnum {refPath=/Niagara/Enums/ENiagaraShapeTorusMode}}
      }
      Knot Ring Loops {
       bIsDynamic=false
       bIsEditable=false
       bIsStaticSwitch=false
       bIsVisible=false
       type {classStructOrEnum {refPath=&NiagaraInt32}}
      }
      Knot Helix Loops {
       bIsDynamic=false
       bIsEditable=false
       bIsStaticSwitch=false
       bIsVisible=false
       type {classStructOrEnum {refPath=&NiagaraInt32}}
      }
      Enable Torus Knot Radius {bIsDynamic=false, bIsEditable=false, bIsStaticSwitch=false, bIsVisible=false, type {classStructOrEnum {refPath=&NiagaraBool}}}
      Torus Knot Tube Radius {
       bIsDynamic=false
       bIsEditable=false
       bIsStaticSwitch=false
       bIsVisible=false
       type {classStructOrEnum {refPath=&NiagaraFloat}}
      }
      Cylinder Radius {
       bIsDynamic=false
       bIsEditable=false
       bIsStaticSwitch=false
       bIsVisible=false
       type {classStructOrEnum {refPath=&NiagaraFloat}}
      }
      Cylinder Height Midpoint {
       bIsDynamic=false
       bIsEditable=false
       bIsStaticSwitch=false
       bIsVisible=false
       type {classStructOrEnum {refPath=&NiagaraFloat}}
      }
      Ring Radius {
       bIsDynamic=false
       bIsEditable=false
       bIsStaticSwitch=false
       bIsVisible=false
       type {classStructOrEnum {refPath=&NiagaraFloat}}
      }
      Box Size {
       bIsDynamic=false
       bIsEditable=false
       bIsStaticSwitch=false
       bIsVisible=false
       type {classStructOrEnum {refPath=&Vector3f}}
      }
      Box Midpoint {
       bIsDynamic=false
       bIsEditable=false
       bIsStaticSwitch=false
       bIsVisible=false
       type {classStructOrEnum {refPath=&Vector3f}}
      }
      Cone Length {
       bIsDynamic=false
       bIsEditable=false
       bIsStaticSwitch=false
       bIsVisible=false
       type {classStructOrEnum {refPath=&NiagaraFloat}}
      }
      Plane Size {
       bIsDynamic=false
       bIsEditable=false
       bIsStaticSwitch=false
       bIsVisible=false
       type {classStructOrEnum {refPath=&Vector2f}}
      }
      Wedge Horizontal Angle {
       bIsDynamic=false
       bIsEditable=false
       bIsStaticSwitch=false
       bIsVisible=false
       type {classStructOrEnum {refPath=&NiagaraFloat}}
      }
      Large Radius {
       bIsDynamic=false
       bIsEditable=false
       bIsStaticSwitch=false
       bIsVisible=false
       type {classStructOrEnum {refPath=&NiagaraFloat}}
      }
      Plane Midpoint {
       bIsDynamic=false
       bIsEditable=false
       bIsStaticSwitch=false
       bIsVisible=false
       type {classStructOrEnum {refPath=&Vector2f}}
      }
      Wedge Vertical Angle {
       bIsDynamic=false
       bIsEditable=false
       bIsStaticSwitch=false
       bIsVisible=false
       type {classStructOrEnum {refPath=&NiagaraFloat}}
      }
      Cone Angle {
       bIsDynamic=false
       bIsEditable=false
       bIsStaticSwitch=false
       bIsVisible=false
       type {classStructOrEnum {refPath=&NiagaraFloat}}
      }
      Handle Radius {
       bIsDynamic=false
       bIsEditable=false
       bIsStaticSwitch=false
       bIsVisible=false
       type {classStructOrEnum {refPath=&NiagaraFloat}}
      }
      Cone Inner Angle {
       bIsDynamic=false
       bIsEditable=false
       bIsStaticSwitch=false
       bIsVisible=false
       type {classStructOrEnum {refPath=&NiagaraFloat}}
      }
      Cone Radial Angle {
       bIsDynamic=false
       bIsEditable=false
       bIsStaticSwitch=false
       bIsVisible=false
       type {classStructOrEnum {refPath=&NiagaraFloat}}
      }
      Cone Angle Type {
       bIsDynamic=false
       bIsEditable=false
       bIsStaticSwitch=true
       bIsVisible=false
       type {classStructOrEnum {refPath=&ENiagara_AngleInput}}
      }
      Flatten Endcaps {bIsDynamic=false, bIsEditable=false, bIsStaticSwitch=true, bIsVisible=false, type {classStructOrEnum {refPath=&NiagaraBool}}}
      Apply To Particle Position {
       bIsDynamic=false
       bIsEditable=true
       bIsStaticSwitch=true
       bIsVisible=true
       type {classStructOrEnum {refPath=/Niagara/Enums/ENiagara_AttributeSamplingApplyOutput}}
      }
      Cone Surface Distribution {
       bIsDynamic=false
       bIsEditable=false
       bIsStaticSwitch=false
       bIsVisible=false
       type {classStructOrEnum {refPath=&NiagaraFloat}}
      }
      Ring / Disc Distribution Mode {
       bIsDynamic=false
       bIsEditable=false
       bIsStaticSwitch=true
       bIsVisible=false
       type {classStructOrEnum {refPath=/Niagara/Enums/ENiagaraSphereDistributionMode}}
      }
      Sphere Distribution {
       bIsDynamic=false
       bIsEditable=true
       bIsStaticSwitch=true
       bIsVisible=true
       type {classStructOrEnum {refPath=&ENiagaraSphereDistributionMode}}
      }
      Torus Distribution Mode {
       bIsDynamic=false
       bIsEditable=false
       bIsStaticSwitch=true
       bIsVisible=false
       type {classStructOrEnum {refPath=/Niagara/Enums/ENiagaraTorusDistributionMode}}
      }
      Sphere Surface Distribution {bIsDynamic=false, bIsEditable=true, bIsStaticSwitch=false, bIsVisible=true, type {classStructOrEnum {refPath=&NiagaraFloat}}}
      Hemisphere Distribution {bIsDynamic=false, bIsEditable=true, bIsStaticSwitch=false, bIsVisible=true, type {classStructOrEnum {refPath=&Vector2f}}}
      Cylinder Distribution {
       bIsDynamic=false
       bIsEditable=false
       bIsStaticSwitch=true
       bIsVisible=false
       type {classStructOrEnum {refPath=/Niagara/Enums/Location/ENiagara_CylinderMode}}
      }
      Surface Only {bIsDynamic=false, bIsEditable=false, bIsStaticSwitch=false, bIsVisible=false, type {classStructOrEnum {refPath=&NiagaraBool}}}
      Surface Only Band Thickness {
       bIsDynamic=false
       bIsEditable=false
       bIsStaticSwitch=false
       bIsVisible=false
       type {classStructOrEnum {refPath=&NiagaraFloat}}
      }
      Use Endcaps In Surface Only Mode {bIsDynamic=false, bIsEditable=false, bIsStaticSwitch=true, bIsVisible=false, type {classStructOrEnum {refPath=&NiagaraBool}}}
      Hemicircle X {bIsDynamic=false, bIsEditable=false, bIsStaticSwitch=true, bIsVisible=false, type {classStructOrEnum {refPath=&NiagaraBool}}}
      Hemicircle Y {bIsDynamic=false, bIsEditable=false, bIsStaticSwitch=true, bIsVisible=false, type {classStructOrEnum {refPath=&NiagaraBool}}}
      Disc Coverage {
       bIsDynamic=false
       bIsEditable=false
       bIsStaticSwitch=false
       bIsVisible=false
       type {classStructOrEnum {refPath=&NiagaraFloat}}
      }
      Hemicircle Internal Cap {bIsDynamic=false, bIsEditable=false, bIsStaticSwitch=true, bIsVisible=false, type {classStructOrEnum {refPath=&NiagaraBool}}}
      U Position {
       bIsDynamic=false
       bIsEditable=false
       bIsStaticSwitch=false
       bIsVisible=false
       type {classStructOrEnum {refPath=&NiagaraFloat}}
      }
      Enable Lathe Profile {bIsDynamic=false, bIsEditable=false, bIsStaticSwitch=true, bIsVisible=false, type {classStructOrEnum {refPath=&NiagaraBool}}}
      Lathe Profile {
       bIsDynamic=false
       bIsEditable=false
       bIsStaticSwitch=false
       bIsVisible=false
       type {classStructOrEnum {refPath=&NiagaraDataInterfaceCurve}}
      }
      Surface Distribution {
       bIsDynamic=false
       bIsEditable=false
       bIsStaticSwitch=false
       bIsVisible=false
       type {classStructOrEnum {refPath=&NiagaraFloat}}
      }
      Surface Expansion Mode {
       bIsDynamic=false
       bIsEditable=false
       bIsStaticSwitch=false
       bIsVisible=false
       type {classStructOrEnum {refPath=/Niagara/Enums/ENiagaraExpansionMode}}
      }
      U Distribution {
       bIsDynamic=false
       bIsEditable=false
       bIsStaticSwitch=false
       bIsVisible=false
       type {classStructOrEnum {refPath=&NiagaraFloat}}
      }
      V Position {
       bIsDynamic=false
       bIsEditable=false
       bIsStaticSwitch=false
       bIsVisible=false
       type {classStructOrEnum {refPath=&NiagaraFloat}}
      }
      Radius Position {
       bIsDynamic=false
       bIsEditable=false
       bIsStaticSwitch=false
       bIsVisible=false
       type {classStructOrEnum {refPath=&NiagaraFloat}}
      }
      Sphere Radius Position {
       bIsDynamic=false
       bIsEditable=false
       bIsStaticSwitch=false
       bIsVisible=false
       type {classStructOrEnum {refPath=&NiagaraFloat}}
      }
      V Distribution {
       bIsDynamic=false
       bIsEditable=false
       bIsStaticSwitch=false
       bIsVisible=false
       type {classStructOrEnum {refPath=&NiagaraFloat}}
      }
      Evaluation Type {
       bIsDynamic=false
       bIsEditable=false
       bIsStaticSwitch=true
       bIsVisible=false
       type {classStructOrEnum {refPath=/Niagara/Enums/ENiagaraRandomnessEvaluation}}
      }
      Uniform Distribution {
       bIsDynamic=false
       bIsEditable=false
       bIsStaticSwitch=false
       bIsVisible=false
       type {classStructOrEnum {refPath=&NiagaraFloat}}
      }
      Uniform Spiral Amount {
       bIsDynamic=false
       bIsEditable=false
       bIsStaticSwitch=false
       bIsVisible=false
       type {classStructOrEnum {refPath=&NiagaraFloat}}
      }
      Uniform Spiral Falloff {
       bIsDynamic=false
       bIsEditable=false
       bIsStaticSwitch=false
       bIsVisible=false
       type {classStructOrEnum {refPath=&NiagaraFloat}}
      }
      Uniform Count {
       bIsDynamic=false
       bIsEditable=false
       bIsStaticSwitch=false
       bIsVisible=false
       type {classStructOrEnum {refPath=&NiagaraInt32}}
      }
      Uniform Particle Index {
       bIsDynamic=false
       bIsEditable=false
       bIsStaticSwitch=false
       bIsVisible=false
       type {classStructOrEnum {refPath=&NiagaraInt32}}
      }
      Hemisphere Angle Type {
       bIsDynamic=false
       bIsEditable=true
       bIsStaticSwitch=true
       bIsVisible=true
       type {classStructOrEnum {refPath=&ENiagara_AngleInput}}
      }
      Uniform Endpoint Offset {
       bIsDynamic=false
       bIsEditable=false
       bIsStaticSwitch=false
       bIsVisible=false
       type {classStructOrEnum {refPath=&NiagaraFloat}}
      }
      MaskSpawnGroup {bIsDynamic=false, bIsEditable=true, bIsStaticSwitch=false, bIsVisible=true, type {classStructOrEnum {refPath=&NiagaraBool}}}
      Spawn Group Mask {bIsDynamic=false, bIsEditable=false, bIsStaticSwitch=false, bIsVisible=true, type {classStructOrEnum {refPath=&NiagaraInt32}}}
      Spawn Group {
       bIsDynamic=false
       bIsEditable=false
       bIsStaticSwitch=false
       bIsVisible=false
       type {classStructOrEnum {refPath=&NiagaraInt32}}
      }
      Randomness Mode {
       bIsDynamic=false
       bIsEditable=true
       bIsStaticSwitch=true
       bIsVisible=true
       type {classStructOrEnum {refPath=&ENiagaraRandomnessMode}}
      }
      Override Seed {bIsDynamic=false, bIsEditable=true, bIsStaticSwitch=false, bIsVisible=true, type {classStructOrEnum {refPath=&NiagaraBool}}}
      Random Seed {bIsDynamic=false, bIsEditable=false, bIsStaticSwitch=false, bIsVisible=true, type {classStructOrEnum {refPath=&NiagaraInt32}}}
      Fixed Random Seed {bIsDynamic=false, bIsEditable=false, bIsStaticSwitch=true, bIsVisible=false, type {classStructOrEnum {refPath=&NiagaraBool}}}
      Transform Order {
       bIsDynamic=false
       bIsEditable=true
       bIsStaticSwitch=true
       bIsVisible=true
       type {classStructOrEnum {refPath=/Niagara/Enums/Transforms/ENiagara_TransformOrder}}
      }
      Scale Mode {
       bIsDynamic=false
       bIsEditable=true
       bIsStaticSwitch=true
       bIsVisible=true
       type {classStructOrEnum {refPath=/Niagara/Enums/Transforms/ENiagara_ScaleMode}}
      }
      Non Uniform Scale {bIsDynamic=false, bIsEditable=true, bIsStaticSwitch=false, bIsVisible=true, type {classStructOrEnum {refPath=&Vector3f}}}
      UseOwnerScale {bIsDynamic=false, bIsEditable=true, bIsStaticSwitch=false, bIsVisible=true, type {classStructOrEnum {refPath=&NiagaraBool}}}
      Apply Owner Scale {bIsDynamic=false, bIsEditable=true, bIsStaticSwitch=false, bIsVisible=true, type {classStructOrEnum {refPath=&Vector3f}}}
      Rotation Mode {
       bIsDynamic=false
       bIsEditable=true
       bIsStaticSwitch=true
       bIsVisible=true
       type {classStructOrEnum {refPath=&ENiagara_RotationMode}}
      }
      Rotation Axis {
       bIsDynamic=false
       bIsEditable=false
       bIsStaticSwitch=false
       bIsVisible=false
       type {classStructOrEnum {refPath=&Vector3f}}
      }
      Rotation Matrix {
       bIsDynamic=false
       bIsEditable=false
       bIsStaticSwitch=false
       bIsVisible=false
       type {classStructOrEnum {refPath=&NiagaraMatrix}}
      }
      Rotation Quaternion {bIsDynamic=false, bIsEditable=false, bIsStaticSwitch=false, bIsVisible=false, type {classStructOrEnum {refPath=&Quat4f}}}
      Yaw / Pitch / Roll {
       bIsDynamic=false
       bIsEditable=false
       bIsStaticSwitch=false
       bIsVisible=false
       type {classStructOrEnum {refPath=&Vector3f}}
      }
      Rotation Angle {
       bIsDynamic=false
       bIsEditable=false
       bIsStaticSwitch=false
       bIsVisible=false
       type {classStructOrEnum {refPath=&NiagaraFloat}}
      }
      Invert Rotation Quaternion {bIsDynamic=false, bIsEditable=false, bIsStaticSwitch=false, bIsVisible=false, type {classStructOrEnum {refPath=&NiagaraBool}}}
      Additional Quaternion Rotation {bIsDynamic=false, bIsEditable=false, bIsStaticSwitch=false, bIsVisible=false, type {classStructOrEnum {refPath=&NiagaraBool}}}
      Additional Yaw / Pitch / Roll {
       bIsDynamic=false
       bIsEditable=false
       bIsStaticSwitch=false
       bIsVisible=false
       type {classStructOrEnum {refPath=&Vector3f}}
      }
      Rotation Angle Type {
       bIsDynamic=false
       bIsEditable=false
       bIsStaticSwitch=true
       bIsVisible=false
       type {classStructOrEnum {refPath=&ENiagara_AngleInput}}
      }
      Rotation Coordinate Space {
       bIsDynamic=false
       bIsEditable=true
       bIsStaticSwitch=false
       bIsVisible=true
       type {classStructOrEnum {refPath=&ENiagaraCoordinateSpace}}
      }
      Offset Mode {
       bIsDynamic=false
       bIsEditable=true
       bIsStaticSwitch=true
       bIsVisible=true
       type {classStructOrEnum {refPath=/Niagara/Enums/Transforms/ENiagara_OffsetMode}}
      }
      Shape Origin {
       bIsDynamic=false
       bIsEditable=true
       bIsStaticSwitch=false
       bIsVisible=true
       type {classStructOrEnum {refPath=&NiagaraPosition}}
      }
      Offset {
       bIsDynamic=false
       bIsEditable=false
       bIsStaticSwitch=false
       bIsVisible=false
       type {classStructOrEnum {refPath=&Vector3f}}
      }
      Offset Coordinate Space {
       bIsDynamic=false
       bIsEditable=false
       bIsStaticSwitch=false
       bIsVisible=false
       type {classStructOrEnum {refPath=&ENiagaraCoordinateSpace}}
      }
      Custom Transform Matrix {
       bIsDynamic=false
       bIsEditable=false
       bIsStaticSwitch=false
       bIsVisible=false
       type {classStructOrEnum {refPath=&NiagaraMatrix}}
      }
      Transform Method {
       bIsDynamic=false
       bIsEditable=true
       bIsStaticSwitch=true
       bIsVisible=true
       type {classStructOrEnum {refPath=/Niagara/Enums/Transforms/ENiagara_TransformType}}
      }
     ]
     moduleScript {refPath=&ShapeLocation}
    }
    {
     bIsSetParametersModule=false
     enabled=true
     moduleName=AddVelocity
     inputs [
      Velocity Mode {
       bIsDynamic=false
       bIsEditable=true
       bIsStaticSwitch=true
       bIsVisible=true
       type {classStructOrEnum {refPath=&ENiagara_VelocityMode}}
      }
      Velocity {
       bIsDynamic=false
       bIsEditable=false
       bIsStaticSwitch=false
       bIsVisible=false
       type {classStructOrEnum {refPath=&Vector3f}}
      }
      Velocity Speed {bIsDynamic=true, bIsEditable=true, bIsStaticSwitch=false, bIsVisible=true, type {classStructOrEnum {refPath=&NiagaraFloat}}}
      Velocity Speed Scale {
       bIsDynamic=false
       bIsEditable=false
       bIsStaticSwitch=false
       bIsVisible=false
       type {classStructOrEnum {refPath=&NiagaraFloat}}
      }
      Module.Use Falloff Distance {bIsDynamic=false, bIsEditable=true, bIsStaticSwitch=false, bIsVisible=true, type {classStructOrEnum {refPath=&NiagaraBool}}}
      Constrain To Radius {bIsDynamic=false, bIsEditable=false, bIsStaticSwitch=false, bIsVisible=true, type {classStructOrEnum {refPath=&NiagaraFloat}}}
      Distribution Along Cone Axis {
       bIsDynamic=false
       bIsEditable=false
       bIsStaticSwitch=false
       bIsVisible=false
       type {classStructOrEnum {refPath=&NiagaraFloat}}
      }
      Radius Falloff Near / Far {
       bIsDynamic=false
       bIsEditable=false
       bIsStaticSwitch=false
       bIsVisible=false
       type {classStructOrEnum {refPath=&Vector2f}}
      }
      Radius Falloff Exponent {
       bIsDynamic=false
       bIsEditable=false
       bIsStaticSwitch=false
       bIsVisible=false
       type {classStructOrEnum {refPath=&NiagaraFloat}}
      }
      Speed Falloff From Cone Axis {
       bIsDynamic=false
       bIsEditable=false
       bIsStaticSwitch=false
       bIsVisible=false
       type {classStructOrEnum {refPath=&NiagaraFloat}}
      }
      Invert Velocity Falloff {bIsDynamic=false, bIsEditable=false, bIsStaticSwitch=false, bIsVisible=false, type {classStructOrEnum {refPath=&NiagaraBool}}}
      Override Seed {bIsDynamic=false, bIsEditable=true, bIsStaticSwitch=false, bIsVisible=true, type {classStructOrEnum {refPath=&NiagaraBool}}}
      Random Seed {bIsDynamic=false, bIsEditable=false, bIsStaticSwitch=false, bIsVisible=true, type {classStructOrEnum {refPath=&NiagaraInt32}}}
      Fixed Random Seed {bIsDynamic=false, bIsEditable=false, bIsStaticSwitch=true, bIsVisible=false, type {classStructOrEnum {refPath=&NiagaraBool}}}
      Randomness Mode {
       bIsDynamic=false
       bIsEditable=true
       bIsStaticSwitch=true
       bIsVisible=true
       type {classStructOrEnum {refPath=&ENiagaraRandomnessMode}}
      }
      Cone Axis {
       bIsDynamic=false
       bIsEditable=false
       bIsStaticSwitch=false
       bIsVisible=false
       type {classStructOrEnum {refPath=&Vector3f}}
      }
      Cone Angle {
       bIsDynamic=false
       bIsEditable=false
       bIsStaticSwitch=false
       bIsVisible=false
       type {classStructOrEnum {refPath=&NiagaraFloat}}
      }
      Inner Cone Angle {
       bIsDynamic=false
       bIsEditable=false
       bIsStaticSwitch=false
       bIsVisible=false
       type {classStructOrEnum {refPath=&NiagaraFloat}}
      }
      Cone Angle Mode {
       bIsDynamic=false
       bIsEditable=false
       bIsStaticSwitch=true
       bIsVisible=false
       type {classStructOrEnum {refPath=&ENiagara_AngleInput}}
      }
      Rotation Mode {
       bIsDynamic=false
       bIsEditable=false
       bIsStaticSwitch=true
       bIsVisible=false
       type {classStructOrEnum {refPath=&ENiagara_RotationMode}}
      }
      Rotation Axis {
       bIsDynamic=false
       bIsEditable=false
       bIsStaticSwitch=false
       bIsVisible=false
       type {classStructOrEnum {refPath=&Vector3f}}
      }
      Yaw / Pitch / Roll {
       bIsDynamic=false
       bIsEditable=false
       bIsStaticSwitch=false
       bIsVisible=false
       type {classStructOrEnum {refPath=&Vector3f}}
      }
      Rotation Quaternion {bIsDynamic=false, bIsEditable=false, bIsStaticSwitch=false, bIsVisible=false, type {classStructOrEnum {refPath=&Quat4f}}}
      Rotation Matrix {
       bIsDynamic=false
       bIsEditable=false
       bIsStaticSwitch=false
       bIsVisible=false
       type {classStructOrEnum {refPath=&NiagaraMatrix}}
      }
      Rotation Angle {
       bIsDynamic=false
       bIsEditable=false
       bIsStaticSwitch=false
       bIsVisible=false
       type {classStructOrEnum {refPath=&NiagaraFloat}}
      }
      Invert Rotation Quaternion {bIsDynamic=false, bIsEditable=false, bIsStaticSwitch=false, bIsVisible=false, type {classStructOrEnum {refPath=&NiagaraBool}}}
      Additional Quaternion Rotation {bIsDynamic=false, bIsEditable=false, bIsStaticSwitch=false, bIsVisible=false, type {classStructOrEnum {refPath=&NiagaraBool}}}
      Additional Yaw / Pitch / Roll {
       bIsDynamic=false
       bIsEditable=false
       bIsStaticSwitch=false
       bIsVisible=false
       type {classStructOrEnum {refPath=&Vector3f}}
      }
      Rotation Angle Type {
       bIsDynamic=false
       bIsEditable=false
       bIsStaticSwitch=true
       bIsVisible=false
       type {classStructOrEnum {refPath=&ENiagara_AngleInput}}
      }
      Rotation Coordinate Space {
       bIsDynamic=false
       bIsEditable=false
       bIsStaticSwitch=false
       bIsVisible=false
       type {classStructOrEnum {refPath=&ENiagaraCoordinateSpace}}
      }
      Velocity Origin {
       bIsDynamic=false
       bIsEditable=true
       bIsStaticSwitch=false
       bIsVisible=true
       type {classStructOrEnum {refPath=&NiagaraPosition}}
      }
      Origin Offset {bIsDynamic=false, bIsEditable=true, bIsStaticSwitch=false, bIsVisible=true, type {classStructOrEnum {refPath=&Vector3f}}}
      Origin Offset Coordinate Space {
       bIsDynamic=false
       bIsEditable=true
       bIsStaticSwitch=false
       bIsVisible=true
       type {classStructOrEnum {refPath=&ENiagaraCoordinateSpace}}
      }
      Default Position {
       bIsDynamic=false
       bIsEditable=true
       bIsStaticSwitch=false
       bIsVisible=true
       type {classStructOrEnum {refPath=&NiagaraPosition}}
      }
     ]
     moduleScript {refPath=&AddVelocity}
    }
    {
     bIsSetParametersModule=false
     enabled=true
     moduleName=AddVelocity001
     inputs [
      Velocity Mode {
       bIsDynamic=false
       bIsEditable=true
       bIsStaticSwitch=true
       bIsVisible=true
       type {classStructOrEnum {refPath=&ENiagara_VelocityMode}}
      }
      Velocity {bIsDynamic=true, bIsEditable=true, bIsStaticSwitch=false, bIsVisible=true, type {classStructOrEnum {refPath=&Vector3f}}}
      Velocity Speed {
       bIsDynamic=false
       bIsEditable=false
       bIsStaticSwitch=false
       bIsVisible=false
       type {classStructOrEnum {refPath=&NiagaraFloat}}
      }
      Velocity Speed Scale {bIsDynamic=false, bIsEditable=true, bIsStaticSwitch=false, bIsVisible=true, type {classStructOrEnum {refPath=&NiagaraFloat}}}
      Constrain To Radius {
       bIsDynamic=false
       bIsEditable=false
       bIsStaticSwitch=false
       bIsVisible=false
       type {classStructOrEnum {refPath=&NiagaraFloat}}
      }
      Distribution Along Cone Axis {
       bIsDynamic=false
       bIsEditable=false
       bIsStaticSwitch=false
       bIsVisible=false
       type {classStructOrEnum {refPath=&NiagaraFloat}}
      }
      Radius Falloff Near / Far {
       bIsDynamic=false
       bIsEditable=false
       bIsStaticSwitch=false
       bIsVisible=false
       type {classStructOrEnum {refPath=&Vector2f}}
      }
      Radius Falloff Exponent {
       bIsDynamic=false
       bIsEditable=false
       bIsStaticSwitch=false
       bIsVisible=false
       type {classStructOrEnum {refPath=&NiagaraFloat}}
      }
      Speed Falloff From Cone Axis {
       bIsDynamic=false
       bIsEditable=false
       bIsStaticSwitch=false
       bIsVisible=false
       type {classStructOrEnum {refPath=&NiagaraFloat}}
      }
      Invert Velocity Falloff {bIsDynamic=false, bIsEditable=false, bIsStaticSwitch=false, bIsVisible=false, type {classStructOrEnum {refPath=&NiagaraBool}}}
      Override Seed {bIsDynamic=false, bIsEditable=false, bIsStaticSwitch=false, bIsVisible=false, type {classStructOrEnum {refPath=&NiagaraBool}}}
      Random Seed {
       bIsDynamic=false
       bIsEditable=false
       bIsStaticSwitch=false
       bIsVisible=false
       type {classStructOrEnum {refPath=&NiagaraInt32}}
      }
      Fixed Random Seed {bIsDynamic=false, bIsEditable=false, bIsStaticSwitch=true, bIsVisible=false, type {classStructOrEnum {refPath=&NiagaraBool}}}
      Randomness Mode {
       bIsDynamic=false
       bIsEditable=false
       bIsStaticSwitch=true
       bIsVisible=false
       type {classStructOrEnum {refPath=&ENiagaraRandomnessMode}}
      }
      Cone Axis {
       bIsDynamic=false
       bIsEditable=false
       bIsStaticSwitch=false
       bIsVisible=false
       type {classStructOrEnum {refPath=&Vector3f}}
      }
      Cone Angle {
       bIsDynamic=false
       bIsEditable=false
       bIsStaticSwitch=false
       bIsVisible=false
       type {classStructOrEnum {refPath=&NiagaraFloat}}
      }
      Inner Cone Angle {
       bIsDynamic=false
       bIsEditable=false
       bIsStaticSwitch=false
       bIsVisible=false
       type {classStructOrEnum {refPath=&NiagaraFloat}}
      }
      Cone Angle Mode {
       bIsDynamic=false
       bIsEditable=false
       bIsStaticSwitch=true
       bIsVisible=false
       type {classStructOrEnum {refPath=&ENiagara_AngleInput}}
      }
      Rotation Mode {
       bIsDynamic=false
       bIsEditable=true
       bIsStaticSwitch=true
       bIsVisible=true
       type {classStructOrEnum {refPath=&ENiagara_RotationMode}}
      }
      Rotation Axis {
       bIsDynamic=false
       bIsEditable=false
       bIsStaticSwitch=false
       bIsVisible=false
       type {classStructOrEnum {refPath=&Vector3f}}
      }
      Yaw / Pitch / Roll {
       bIsDynamic=false
       bIsEditable=false
       bIsStaticSwitch=false
       bIsVisible=false
       type {classStructOrEnum {refPath=&Vector3f}}
      }
      Rotation Quaternion {bIsDynamic=false, bIsEditable=false, bIsStaticSwitch=false, bIsVisible=false, type {classStructOrEnum {refPath=&Quat4f}}}
      Rotation Matrix {
       bIsDynamic=false
       bIsEditable=false
       bIsStaticSwitch=false
       bIsVisible=false
       type {classStructOrEnum {refPath=&NiagaraMatrix}}
      }
      Rotation Angle {
       bIsDynamic=false
       bIsEditable=false
       bIsStaticSwitch=false
       bIsVisible=false
       type {classStructOrEnum {refPath=&NiagaraFloat}}
      }
      Invert Rotation Quaternion {bIsDynamic=false, bIsEditable=false, bIsStaticSwitch=false, bIsVisible=false, type {classStructOrEnum {refPath=&NiagaraBool}}}
      Additional Quaternion Rotation {bIsDynamic=false, bIsEditable=false, bIsStaticSwitch=false, bIsVisible=false, type {classStructOrEnum {refPath=&NiagaraBool}}}
      Additional Yaw / Pitch / Roll {
       bIsDynamic=false
       bIsEditable=false
       bIsStaticSwitch=false
       bIsVisible=false
       type {classStructOrEnum {refPath=&Vector3f}}
      }
      Rotation Angle Type {
       bIsDynamic=false
       bIsEditable=false
       bIsStaticSwitch=true
       bIsVisible=false
       type {classStructOrEnum {refPath=&ENiagara_AngleInput}}
      }
      Rotation Coordinate Space {
       bIsDynamic=false
       bIsEditable=true
       bIsStaticSwitch=false
       bIsVisible=true
       type {classStructOrEnum {refPath=&ENiagaraCoordinateSpace}}
      }
      Velocity Origin {
       bIsDynamic=false
       bIsEditable=false
       bIsStaticSwitch=false
       bIsVisible=false
       type {classStructOrEnum {refPath=&NiagaraPosition}}
      }
      Origin Offset {
       bIsDynamic=false
       bIsEditable=false
       bIsStaticSwitch=false
       bIsVisible=false
       type {classStructOrEnum {refPath=&Vector3f}}
      }
      Origin Offset Coordinate Space {
       bIsDynamic=false
       bIsEditable=false
       bIsStaticSwitch=false
       bIsVisible=false
       type {classStructOrEnum {refPath=&ENiagaraCoordinateSpace}}
      }
      Default Position {
       bIsDynamic=false
       bIsEditable=false
       bIsStaticSwitch=false
       bIsVisible=false
       type {classStructOrEnum {refPath=&NiagaraPosition}}
      }
     ]
     moduleScript {refPath=&AddVelocity}
    }
    {
     bIsSetParametersModule=true
     enabled=true
     moduleName=SetVariables_#5
     inputs [Particles.MeshIndex {bIsDynamic=true, bIsEditable=true, bIsStaticSwitch=false, bIsVisible=true, type {classStructOrEnum {refPath=&NiagaraInt32}}}]
     moduleScript {refPath=&self:NE_Debris_02_0.NiagaraScriptSource_0.NiagaraGraph_0.NiagaraNodeAssignment_0.SetVariables_#5}
    }
    {
     bIsSetParametersModule=false
     enabled=true
     moduleName=InitialMeshOrientation
     inputs [
      Mesh Orientation Mode {
       bIsDynamic=false
       bIsEditable=true
       bIsStaticSwitch=true
       bIsVisible=true
       type {classStructOrEnum {refPath=&ENiagara_MeshOrientationOptions}}
      }
      Orientation Quaternion {bIsDynamic=false, bIsEditable=false, bIsStaticSwitch=false, bIsVisible=false, type {classStructOrEnum {refPath=&Quat4f}}}
      Orientation Matrix {
       bIsDynamic=false
       bIsEditable=false
       bIsStaticSwitch=false
       bIsVisible=false
       type {classStructOrEnum {refPath=&NiagaraMatrix}}
      }
      Orientation Vector {
       bIsDynamic=false
       bIsEditable=false
       bIsStaticSwitch=false
       bIsVisible=false
       type {classStructOrEnum {refPath=&Vector3f}}
      }
      Orientation Coordinate Space {
       bIsDynamic=false
       bIsEditable=false
       bIsStaticSwitch=true
       bIsVisible=false
       type {classStructOrEnum {refPath=&ENiagaraCoordinateSpace}}
      }
      Mesh Axis To Orient {
       bIsDynamic=false
       bIsEditable=false
       bIsStaticSwitch=false
       bIsVisible=false
       type {classStructOrEnum {refPath=&Vector3f}}
      }
      Random Seed {bIsDynamic=false, bIsEditable=true, bIsStaticSwitch=false, bIsVisible=true, type {classStructOrEnum {refPath=&NiagaraInt32}}}
      Use Rotation Vector {bIsDynamic=false, bIsEditable=true, bIsStaticSwitch=false, bIsVisible=true, type {classStructOrEnum {refPath=&NiagaraBool}}}
      Rotation {bIsDynamic=false, bIsEditable=false, bIsStaticSwitch=false, bIsVisible=true, type {classStructOrEnum {refPath=&Vector3f}}}
      Rotation Coordinate Space {
       bIsDynamic=false
       bIsEditable=false
       bIsStaticSwitch=true
       bIsVisible=true
       type {classStructOrEnum {refPath=&ENiagaraMeshTransforms}}
      }
     ]
     moduleScript {refPath=&InitialMeshOrientation}
    }
   ]
  }
  particleUpdateScript {
   scriptName=ParticleUpdateScript
   modules [
    {
     bIsSetParametersModule=false
     enabled=true
     moduleName=ParticleState
     inputs [
      Kill Particles When Lifetime Has Elapsed {bIsDynamic=false, bIsEditable=true, bIsStaticSwitch=true, bIsVisible=true, type {classStructOrEnum {refPath=&NiagaraBool}}}
      Loop Particles Lifetime {bIsDynamic=false, bIsEditable=false, bIsStaticSwitch=false, bIsVisible=false, type {classStructOrEnum {refPath=&NiagaraBool}}}
      Let Infinitely Lived Particles Die When Emitter Deactivates {bIsDynamic=false, bIsEditable=false, bIsStaticSwitch=false, bIsVisible=false, type {classStructOrEnum {refPath=&NiagaraBool}}}
      Lifetime {bIsDynamic=false, bIsEditable=true, bIsStaticSwitch=false, bIsVisible=true, type {classStructOrEnum {refPath=&NiagaraFloat}}}
      DeltaTime {bIsDynamic=false, bIsEditable=true, bIsStaticSwitch=false, bIsVisible=true, type {classStructOrEnum {refPath=&NiagaraFloat}}}
     ]
     moduleScript {refPath=&ParticleState}
    }
    {
     bIsSetParametersModule=false
     enabled=true
     moduleName=UpdateMeshOrientation
     inputs [
      Orientation Method {
       bIsDynamic=false
       bIsEditable=true
       bIsStaticSwitch=true
       bIsVisible=true
       type {classStructOrEnum {refPath=/Niagara/Enums/Utility/ENiagara_UpdateMeshOrientationMode}}
      }
      Facing Position {
       bIsDynamic=false
       bIsEditable=false
       bIsStaticSwitch=false
       bIsVisible=false
       type {classStructOrEnum {refPath=&NiagaraPosition}}
      }
      Facing Direction {
       bIsDynamic=false
       bIsEditable=false
       bIsStaticSwitch=false
       bIsVisible=false
       type {classStructOrEnum {refPath=&Vector3f}}
      }
      TransformFacingPosition {bIsDynamic=false, bIsEditable=false, bIsStaticSwitch=false, bIsVisible=false, type {classStructOrEnum {refPath=&NiagaraBool}}}
      Facing Position Coordinate Space {
       bIsDynamic=false
       bIsEditable=false
       bIsStaticSwitch=false
       bIsVisible=false
       type {classStructOrEnum {refPath=&ENiagaraCoordinateSpace}}
      }
      "Facing Origin " {
       bIsDynamic=false
       bIsEditable=false
       bIsStaticSwitch=false
       bIsVisible=false
       type {classStructOrEnum {refPath=&NiagaraPosition}}
      }
      TransformFacingOrigin {bIsDynamic=false, bIsEditable=false, bIsStaticSwitch=false, bIsVisible=false, type {classStructOrEnum {refPath=&NiagaraBool}}}
      Facing Origin Coordinate Space {
       bIsDynamic=false
       bIsEditable=false
       bIsStaticSwitch=false
       bIsVisible=false
       type {classStructOrEnum {refPath=&ENiagaraCoordinateSpace}}
      }
      TransformFacing {bIsDynamic=false, bIsEditable=false, bIsStaticSwitch=false, bIsVisible=false, type {classStructOrEnum {refPath=&NiagaraBool}}}
      Facing Coordinate Space {
       bIsDynamic=false
       bIsEditable=false
       bIsStaticSwitch=false
       bIsVisible=false
       type {classStructOrEnum {refPath=&ENiagaraCoordinateSpace}}
      }
      Facing Interpolation {bIsDynamic=false, bIsEditable=false, bIsStaticSwitch=false, bIsVisible=false, type {classStructOrEnum {refPath=&NiagaraBool}}}
      Facing Interpolation Rate {
       bIsDynamic=false
       bIsEditable=false
       bIsStaticSwitch=false
       bIsVisible=false
       type {classStructOrEnum {refPath=&NiagaraFloat}}
      }
      Stabilization Mode {
       bIsDynamic=false
       bIsEditable=false
       bIsStaticSwitch=true
       bIsVisible=false
       type {classStructOrEnum {refPath=/Niagara/Enums/Utility/ENiagara_OrientVectorCount}}
      }
      Side Direction {
       bIsDynamic=false
       bIsEditable=false
       bIsStaticSwitch=false
       bIsVisible=false
       type {classStructOrEnum {refPath=&Vector3f}}
      }
      TransformSide {bIsDynamic=false, bIsEditable=false, bIsStaticSwitch=false, bIsVisible=false, type {classStructOrEnum {refPath=&NiagaraBool}}}
      Side Coordinate Space {
       bIsDynamic=false
       bIsEditable=false
       bIsStaticSwitch=false
       bIsVisible=false
       type {classStructOrEnum {refPath=&ENiagaraCoordinateSpace}}
      }
      Up Direction {
       bIsDynamic=false
       bIsEditable=false
       bIsStaticSwitch=false
       bIsVisible=false
       type {classStructOrEnum {refPath=&Vector3f}}
      }
      TransformUp {bIsDynamic=false, bIsEditable=false, bIsStaticSwitch=false, bIsVisible=false, type {classStructOrEnum {refPath=&NiagaraBool}}}
      Up Coordinate Space {
       bIsDynamic=false
       bIsEditable=false
       bIsStaticSwitch=false
       bIsVisible=false
       type {classStructOrEnum {refPath=&ENiagaraCoordinateSpace}}
      }
      Rotation Vector {bIsDynamic=true, bIsEditable=true, bIsStaticSwitch=false, bIsVisible=true, type {classStructOrEnum {refPath=&Vector3f}}}
      Rotation Rate {bIsDynamic=true, bIsEditable=true, bIsStaticSwitch=false, bIsVisible=true, type {classStructOrEnum {refPath=&NiagaraFloat}}}
      Surface Normal {
       bIsDynamic=false
       bIsEditable=false
       bIsStaticSwitch=false
       bIsVisible=false
       type {classStructOrEnum {refPath=&Vector3f}}
      }
      Enabled {bIsDynamic=false, bIsEditable=false, bIsStaticSwitch=false, bIsVisible=false, type {classStructOrEnum {refPath=&NiagaraBool}}}
      Particle Velocity {
       bIsDynamic=false
       bIsEditable=false
       bIsStaticSwitch=false
       bIsVisible=false
       type {classStructOrEnum {refPath=&Vector3f}}
      }
      Rotation Coordinate Space {
       bIsDynamic=false
       bIsEditable=true
       bIsStaticSwitch=true
       bIsVisible=true
       type {classStructOrEnum {refPath=&ENiagaraMeshTransforms}}
      }
      Initial Forward Vector {
       bIsDynamic=false
       bIsEditable=false
       bIsStaticSwitch=false
       bIsVisible=false
       type {classStructOrEnum {refPath=&Vector3f}}
      }
      Reference Axis Vector {
       bIsDynamic=false
       bIsEditable=false
       bIsStaticSwitch=false
       bIsVisible=false
       type {classStructOrEnum {refPath=&Vector3f}}
      }
      Update Rotation Rate {bIsDynamic=false, bIsEditable=false, bIsStaticSwitch=false, bIsVisible=false, type {classStructOrEnum {refPath=&NiagaraBool}}}
      Velocity {
       bIsDynamic=false
       bIsEditable=false
       bIsStaticSwitch=false
       bIsVisible=false
       type {classStructOrEnum {refPath=&Vector3f}}
      }
      Override Quat {bIsDynamic=false, bIsEditable=false, bIsStaticSwitch=false, bIsVisible=false, type {classStructOrEnum {refPath=&NiagaraBool}}}
      Override Incoming Orientation {
       bIsDynamic=false
       bIsEditable=false
       bIsStaticSwitch=false
       bIsVisible=false
       type {classStructOrEnum {refPath=&NiagaraFloat}}
      }
      Override Quaternion {bIsDynamic=false, bIsEditable=false, bIsStaticSwitch=false, bIsVisible=false, type {classStructOrEnum {refPath=&Quat4f}}}
      First Frame {bIsDynamic=false, bIsEditable=true, bIsStaticSwitch=false, bIsVisible=true, type {classStructOrEnum {refPath=&NiagaraBool}}}
      Write to Orientation {
       bIsDynamic=false
       bIsEditable=true
       bIsStaticSwitch=true
       bIsVisible=true
       type {classStructOrEnum {refPath=&ENiagara_AttributeSamplingApplyOutput}}
      }
      Automatically Calculate Particle Radius {bIsDynamic=false, bIsEditable=false, bIsStaticSwitch=true, bIsVisible=false, type {classStructOrEnum {refPath=&NiagaraBool}}}
      Particle Radius {
       bIsDynamic=false
       bIsEditable=false
       bIsStaticSwitch=false
       bIsVisible=false
       type {classStructOrEnum {refPath=&NiagaraFloat}}
      }
      Particle Radius Scale {
       bIsDynamic=false
       bIsEditable=false
       bIsStaticSwitch=false
       bIsVisible=false
       type {classStructOrEnum {refPath=&NiagaraFloat}}
      }
      Mesh Radius Calculation Method {
       bIsDynamic=false
       bIsEditable=false
       bIsStaticSwitch=true
       bIsVisible=false
       type {classStructOrEnum {refPath=/Niagara/Enums/ENiagaraCalculateRadiusOptions}}
      }
      Mesh Renderer {
       bIsDynamic=false
       bIsEditable=false
       bIsStaticSwitch=false
       bIsVisible=false
       type {classStructOrEnum {refPath=&NiagaraDataInterfaceMeshRendererInfo}}
      }
      Mesh Renderer Index {
       bIsDynamic=false
       bIsEditable=false
       bIsStaticSwitch=false
       bIsVisible=false
       type {classStructOrEnum {refPath=&NiagaraInt32}}
      }
      Scale {
       bIsDynamic=false
       bIsEditable=false
       bIsStaticSwitch=false
       bIsVisible=false
       type {classStructOrEnum {refPath=&Vector3f}}
      }
      "Foward Vector (Mesh Space)" {
       bIsDynamic=false
       bIsEditable=false
       bIsStaticSwitch=false
       bIsVisible=false
       type {classStructOrEnum {refPath=&Vector3f}}
      }
      "Right Vector (Mesh Space)" {
       bIsDynamic=false
       bIsEditable=false
       bIsStaticSwitch=false
       bIsVisible=false
       type {classStructOrEnum {refPath=&Vector3f}}
      }
      Up Vector {
       bIsDynamic=false
       bIsEditable=false
       bIsStaticSwitch=false
       bIsVisible=false
       type {classStructOrEnum {refPath=&Vector3f}}
      }
      Up Vector Coordinate Space {
       bIsDynamic=false
       bIsEditable=false
       bIsStaticSwitch=false
       bIsVisible=false
       type {classStructOrEnum {refPath=&ENiagaraCoordinateSpace}}
      }
      Bank Rate {
       bIsDynamic=false
       bIsEditable=false
       bIsStaticSwitch=false
       bIsVisible=false
       type {classStructOrEnum {refPath=&NiagaraFloat}}
      }
      Banking Angle Min / Max {
       bIsDynamic=false
       bIsEditable=false
       bIsStaticSwitch=false
       bIsVisible=false
       type {classStructOrEnum {refPath=&Vector2f}}
      }
      Rotation Decay Rate In Degrees {
       bIsDynamic=false
       bIsEditable=false
       bIsStaticSwitch=false
       bIsVisible=false
       type {classStructOrEnum {refPath=&NiagaraFloat}}
      }
      Upright Roll Correction Speed {
       bIsDynamic=false
       bIsEditable=false
       bIsStaticSwitch=false
       bIsVisible=false
       type {classStructOrEnum {refPath=&NiagaraFloat}}
      }
      Velocity Orientation Convergence Rate {
       bIsDynamic=false
       bIsEditable=false
       bIsStaticSwitch=false
       bIsVisible=false
       type {classStructOrEnum {refPath=&NiagaraFloat}}
      }
      Delta Time {bIsDynamic=false, bIsEditable=true, bIsStaticSwitch=false, bIsVisible=true, type {classStructOrEnum {refPath=&NiagaraFloat}}}
      Debug Coordinate System Scale {
       bIsDynamic=false
       bIsEditable=false
       bIsStaticSwitch=false
       bIsVisible=false
       type {classStructOrEnum {refPath=&NiagaraFloat}}
      }
     ]
     moduleScript {refPath=&UpdateMeshOrientation}
    }
    {
     bIsSetParametersModule=false
     enabled=true
     moduleName=GravityForce
     inputs [
      Gravity {bIsDynamic=true, bIsEditable=true, bIsStaticSwitch=false, bIsVisible=true, type {classStructOrEnum {refPath=&Vector3f}}}
      Coordinate Space {
       bIsDynamic=false
       bIsEditable=true
       bIsStaticSwitch=true
       bIsVisible=true
       type {classStructOrEnum {refPath=&ENiagaraCoordinateSpace}}
      }
     ]
     moduleScript {refPath=&GravityForce}
    }
    {
     bIsSetParametersModule=false
     enabled=true
     moduleName=Drag
     inputs [
      Use Linear Drag {bIsDynamic=false, bIsEditable=true, bIsStaticSwitch=false, bIsVisible=true, type {classStructOrEnum {refPath=&NiagaraBool}}}
      Drag {bIsDynamic=false, bIsEditable=true, bIsStaticSwitch=false, bIsVisible=true, type {classStructOrEnum {refPath=&NiagaraFloat}}}
      Use Rotational Drag {bIsDynamic=false, bIsEditable=true, bIsStaticSwitch=false, bIsVisible=true, type {classStructOrEnum {refPath=&NiagaraBool}}}
      Rotational Drag {bIsDynamic=false, bIsEditable=true, bIsStaticSwitch=false, bIsVisible=true, type {classStructOrEnum {refPath=&NiagaraFloat}}}
      Ignore Mass {bIsDynamic=false, bIsEditable=true, bIsStaticSwitch=false, bIsVisible=true, type {classStructOrEnum {refPath=&NiagaraBool}}}
     ]
     moduleScript {refPath=&Drag}
    }
    {
     bIsSetParametersModule=false
     enabled=true
     moduleName=SolveForcesAndVelocity
     inputs [
      Clamp Velocity {bIsDynamic=false, bIsEditable=true, bIsStaticSwitch=false, bIsVisible=true, type {classStructOrEnum {refPath=&NiagaraBool}}}
      Speed Limit {bIsDynamic=false, bIsEditable=false, bIsStaticSwitch=false, bIsVisible=true, type {classStructOrEnum {refPath=&NiagaraFloat}}}
      Limit Acceleration {bIsDynamic=false, bIsEditable=true, bIsStaticSwitch=false, bIsVisible=true, type {classStructOrEnum {refPath=&NiagaraBool}}}
      Acceleration Limit {bIsDynamic=false, bIsEditable=false, bIsStaticSwitch=false, bIsVisible=true, type {classStructOrEnum {refPath=&NiagaraFloat}}}
      Manually Enable Rotational Solver {bIsDynamic=false, bIsEditable=true, bIsStaticSwitch=false, bIsVisible=true, type {classStructOrEnum {refPath=&NiagaraBool}}}
      Write to Intrinsic Properties {bIsDynamic=false, bIsEditable=true, bIsStaticSwitch=false, bIsVisible=true, type {classStructOrEnum {refPath=&NiagaraBool}}}
      Write to Presolve Properties {bIsDynamic=false, bIsEditable=true, bIsStaticSwitch=true, bIsVisible=true, type {classStructOrEnum {refPath=&NiagaraBool}}}
      DeltaTime {bIsDynamic=false, bIsEditable=true, bIsStaticSwitch=false, bIsVisible=true, type {classStructOrEnum {refPath=&NiagaraFloat}}}
      Mass {bIsDynamic=false, bIsEditable=true, bIsStaticSwitch=false, bIsVisible=true, type {classStructOrEnum {refPath=&NiagaraFloat}}}
      Position {
       bIsDynamic=false
       bIsEditable=true
       bIsStaticSwitch=false
       bIsVisible=true
       type {classStructOrEnum {refPath=&NiagaraPosition}}
      }
      Velocity {bIsDynamic=false, bIsEditable=true, bIsStaticSwitch=false, bIsVisible=true, type {classStructOrEnum {refPath=&Vector3f}}}
      PreviousVelocity {bIsDynamic=false, bIsEditable=true, bIsStaticSwitch=false, bIsVisible=true, type {classStructOrEnum {refPath=&Vector3f}}}
      Force {bIsDynamic=false, bIsEditable=true, bIsStaticSwitch=false, bIsVisible=true, type {classStructOrEnum {refPath=&Vector3f}}}
      Mesh Orientation {bIsDynamic=false, bIsEditable=false, bIsStaticSwitch=false, bIsVisible=false, type {classStructOrEnum {refPath=&Quat4f}}}
      Rotational Velocity {
       bIsDynamic=false
       bIsEditable=false
       bIsStaticSwitch=false
       bIsVisible=false
       type {classStructOrEnum {refPath=&Vector3f}}
      }
      Rotational Force {
       bIsDynamic=false
       bIsEditable=false
       bIsStaticSwitch=false
       bIsVisible=false
       type {classStructOrEnum {refPath=&Vector3f}}
      }
      Rotational Inertia {
       bIsDynamic=false
       bIsEditable=false
       bIsStaticSwitch=false
       bIsVisible=false
       type {classStructOrEnum {refPath=&NiagaraFloat}}
      }
     ]
     moduleScript {refPath=&SolveForcesAndVelocity}
    }
    {
     bIsSetParametersModule=false
     enabled=true
     moduleName=ScaleMeshSize
     inputs [
      Scale Factor {bIsDynamic=true, bIsEditable=true, bIsStaticSwitch=false, bIsVisible=true, type {classStructOrEnum {refPath=&Vector3f}}}
      Initial Mesh Scale {bIsDynamic=false, bIsEditable=true, bIsStaticSwitch=false, bIsVisible=true, type {classStructOrEnum {refPath=&Vector3f}}}
     ]
     moduleScript {refPath=&ScaleMeshSize}
    }
   ]
  }
  rendererClasses [{refPath=&NiagaraMeshRendererProperties}]
  renderers [{rendererIndex=0, rendererClass {refPath=&NiagaraMeshRendererProperties}}]
 }
 NE_flameTrail {
  bEnabled=false
  emitterSpawnScript {scriptName=EmitterSpawnScript, modules []}
  emitterUpdateScript {
   scriptName=EmitterUpdateScript
   modules [
    {
     bIsSetParametersModule=false
     enabled=true
     moduleName=EmitterState
     inputs [
      Life Cycle Mode {
       bIsDynamic=false
       bIsEditable=true
       bIsStaticSwitch=true
       bIsVisible=true
       type {classStructOrEnum {refPath=&ENiagaraEmitterLifeCycleMode}}
      }
      Inactive Response {
       bIsDynamic=false
       bIsEditable=true
       bIsStaticSwitch=true
       bIsVisible=true
       type {classStructOrEnum {refPath=&ENiagaraInactiveMode}}
      }
      Loop Behavior {
       bIsDynamic=false
       bIsEditable=true
       bIsStaticSwitch=true
       bIsVisible=true
       type {classStructOrEnum {refPath=&ENiagara_EmitterStateOptions}}
      }
      Loop Duration Mode {
       bIsDynamic=false
       bIsEditable=true
       bIsStaticSwitch=true
       bIsVisible=true
       type {classStructOrEnum {refPath=&ENiagara_InfiniteLoopDuration}}
      }
      Loop Count {
       bIsDynamic=false
       bIsEditable=false
       bIsStaticSwitch=false
       bIsVisible=false
       type {classStructOrEnum {refPath=&NiagaraInt32}}
      }
      Loop Duration {bIsDynamic=false, bIsEditable=true, bIsStaticSwitch=false, bIsVisible=true, type {classStructOrEnum {refPath=&NiagaraFloat}}}
      Recalculate Duration Each Loop {bIsDynamic=false, bIsEditable=false, bIsStaticSwitch=false, bIsVisible=false, type {classStructOrEnum {refPath=&NiagaraBool}}}
      UseLoopDelay {bIsDynamic=false, bIsEditable=true, bIsStaticSwitch=false, bIsVisible=true, type {classStructOrEnum {refPath=&NiagaraBool}}}
      Loop Delay {bIsDynamic=false, bIsEditable=false, bIsStaticSwitch=false, bIsVisible=true, type {classStructOrEnum {refPath=&NiagaraFloat}}}
      Delay First Loop Only {bIsDynamic=false, bIsEditable=false, bIsStaticSwitch=false, bIsVisible=false, type {classStructOrEnum {refPath=&NiagaraBool}}}
      Scalability Mode {
       bIsDynamic=false
       bIsEditable=true
       bIsStaticSwitch=true
       bIsVisible=true
       type {classStructOrEnum {refPath=&ENiagaraEmitterScalabilityMode_Limited}}
      }
      Enable Distance Culling {bIsDynamic=false, bIsEditable=false, bIsStaticSwitch=true, bIsVisible=false, type {classStructOrEnum {refPath=&NiagaraBool}}}
      MinDistance {
       bIsDynamic=false
       bIsEditable=false
       bIsStaticSwitch=false
       bIsVisible=false
       type {classStructOrEnum {refPath=&NiagaraFloat}}
      }
      Min Distance Response {
       bIsDynamic=false
       bIsEditable=false
       bIsStaticSwitch=false
       bIsVisible=false
       type {classStructOrEnum {refPath=&ENiagaraExecutionStateManagement}}
      }
      MaxDistance {
       bIsDynamic=false
       bIsEditable=false
       bIsStaticSwitch=false
       bIsVisible=false
       type {classStructOrEnum {refPath=&NiagaraFloat}}
      }
      Max Distance Response {
       bIsDynamic=false
       bIsEditable=false
       bIsStaticSwitch=false
       bIsVisible=false
       type {classStructOrEnum {refPath=&ENiagaraExecutionStateManagement}}
      }
      Scale Spawn Count {bIsDynamic=false, bIsEditable=false, bIsStaticSwitch=true, bIsVisible=false, type {classStructOrEnum {refPath=&NiagaraBool}}}
      Spawn Count Scale {
       bIsDynamic=false
       bIsEditable=false
       bIsStaticSwitch=false
       bIsVisible=false
       type {classStructOrEnum {refPath=&NiagaraFloat}}
      }
      Spawn Count Scale By Distance {
       bIsDynamic=false
       bIsEditable=false
       bIsStaticSwitch=false
       bIsVisible=false
       type {classStructOrEnum {refPath=&NiagaraDataInterfaceCurve}}
      }
      Enable Visibility Culling {bIsDynamic=false, bIsEditable=false, bIsStaticSwitch=true, bIsVisible=false, type {classStructOrEnum {refPath=&NiagaraBool}}}
      Visibility Cull Response {
       bIsDynamic=false
       bIsEditable=false
       bIsStaticSwitch=false
       bIsVisible=false
       type {classStructOrEnum {refPath=&ENiagaraExecutionStateManagement}}
      }
      Visibility Cull Delay {
       bIsDynamic=false
       bIsEditable=false
       bIsStaticSwitch=false
       bIsVisible=false
       type {classStructOrEnum {refPath=&NiagaraFloat}}
      }
      Reset Age On Awaken {bIsDynamic=false, bIsEditable=false, bIsStaticSwitch=false, bIsVisible=false, type {classStructOrEnum {refPath=&NiagaraBool}}}
     ]
     moduleScript {refPath=&EmitterState}
    }
    {
     bIsSetParametersModule=false
     enabled=false
     moduleName=SpawnRate
     inputs [
      SpawnRate {bIsDynamic=false, bIsEditable=true, bIsStaticSwitch=false, bIsVisible=true, type {classStructOrEnum {refPath=&NiagaraFloat}}}
      Use Spawn Probability {bIsDynamic=false, bIsEditable=true, bIsStaticSwitch=false, bIsVisible=true, type {classStructOrEnum {refPath=&NiagaraBool}}}
      Spawn Probability {bIsDynamic=false, bIsEditable=false, bIsStaticSwitch=false, bIsVisible=true, type {classStructOrEnum {refPath=&NiagaraFloat}}}
      SpawnGroup {bIsDynamic=false, bIsEditable=true, bIsStaticSwitch=false, bIsVisible=true, type {classStructOrEnum {refPath=&NiagaraInt32}}}
     ]
     moduleScript {refPath=&SpawnRate}
    }
    {
     bIsSetParametersModule=false
     enabled=true
     moduleName=SpawnParticlesFromOtherEmitter
     inputs [
      Attribute Reader {
       bIsDynamic=false
       bIsEditable=true
       bIsStaticSwitch=false
       bIsVisible=true
       type {classStructOrEnum {refPath=&NiagaraDataInterfaceParticleRead}}
      }
      Spawn Rate {bIsDynamic=false, bIsEditable=true, bIsStaticSwitch=false, bIsVisible=true, type {classStructOrEnum {refPath=&NiagaraFloat}}}
      Calculate Spawn Rate Per Particle {bIsDynamic=false, bIsEditable=true, bIsStaticSwitch=true, bIsVisible=true, type {classStructOrEnum {refPath=&NiagaraBool}}}
      Spawn Rate Per Particle Cap {bIsDynamic=false, bIsEditable=true, bIsStaticSwitch=false, bIsVisible=true, type {classStructOrEnum {refPath=&NiagaraFloat}}}
      Spawning Enabled {bIsDynamic=false, bIsEditable=true, bIsStaticSwitch=false, bIsVisible=true, type {classStructOrEnum {refPath=&NiagaraBool}}}
      Use Minimum Age {bIsDynamic=false, bIsEditable=true, bIsStaticSwitch=false, bIsVisible=true, type {classStructOrEnum {refPath=&NiagaraBool}}}
      Minimum Age Spawn Threshold {bIsDynamic=false, bIsEditable=false, bIsStaticSwitch=false, bIsVisible=true, type {classStructOrEnum {refPath=&NiagaraFloat}}}
      Use Maximum Age {bIsDynamic=false, bIsEditable=true, bIsStaticSwitch=false, bIsVisible=true, type {classStructOrEnum {refPath=&NiagaraBool}}}
      Maximum Age Spawn Threshold {bIsDynamic=false, bIsEditable=false, bIsStaticSwitch=false, bIsVisible=true, type {classStructOrEnum {refPath=&NiagaraFloat}}}
      Spawn Group {bIsDynamic=false, bIsEditable=true, bIsStaticSwitch=false, bIsVisible=true, type {classStructOrEnum {refPath=&NiagaraInt32}}}
     ]
     moduleScript {refPath=&SpawnParticlesFromOtherEmitter}
    }
   ]
  }
  particleSpawnScript {
   scriptName=ParticleSpawnScript
   modules [
    {
     bIsSetParametersModule=false
     enabled=true
     moduleName=InitializeParticle
     inputs [
      Lifetime Mode {
       bIsDynamic=false
       bIsEditable=true
       bIsStaticSwitch=true
       bIsVisible=true
       type {classStructOrEnum {refPath=&ENiagara_LifetimeMode}}
      }
      Lifetime Min {
       bIsDynamic=false
       bIsEditable=false
       bIsStaticSwitch=false
       bIsVisible=false
       type {classStructOrEnum {refPath=&NiagaraFloat}}
      }
      Lifetime {bIsDynamic=false, bIsEditable=true, bIsStaticSwitch=false, bIsVisible=true, type {classStructOrEnum {refPath=&NiagaraFloat}}}
      Lifetime Max {
       bIsDynamic=false
       bIsEditable=false
       bIsStaticSwitch=false
       bIsVisible=false
       type {classStructOrEnum {refPath=&NiagaraFloat}}
      }
      Lifetime Randomness Mode {
       bIsDynamic=false
       bIsEditable=false
       bIsStaticSwitch=true
       bIsVisible=false
       type {classStructOrEnum {refPath=&ENiagaraRandomnessMode}}
      }
      Color Mode {
       bIsDynamic=false
       bIsEditable=true
       bIsStaticSwitch=true
       bIsVisible=true
       type {classStructOrEnum {refPath=&ENiagara_ColorInitializationMode}}
      }
      Color Minimum {
       bIsDynamic=false
       bIsEditable=false
       bIsStaticSwitch=false
       bIsVisible=false
       type {classStructOrEnum {refPath=&LinearColor}}
      }
      Color {
       bIsDynamic=false
       bIsEditable=true
       bIsStaticSwitch=false
       bIsVisible=true
       type {classStructOrEnum {refPath=&LinearColor}}
      }
      AdjustHue {bIsDynamic=false, bIsEditable=false, bIsStaticSwitch=false, bIsVisible=false, type {classStructOrEnum {refPath=&NiagaraBool}}}
      Hue Shift Range {
       bIsDynamic=false
       bIsEditable=false
       bIsStaticSwitch=false
       bIsVisible=false
       type {classStructOrEnum {refPath=&Vector2f}}
      }
      Color Maximum {
       bIsDynamic=false
       bIsEditable=false
       bIsStaticSwitch=false
       bIsVisible=false
       type {classStructOrEnum {refPath=&LinearColor}}
      }
      AdjustSaturation {bIsDynamic=false, bIsEditable=false, bIsStaticSwitch=false, bIsVisible=false, type {classStructOrEnum {refPath=&NiagaraBool}}}
      Saturation Range {
       bIsDynamic=false
       bIsEditable=false
       bIsStaticSwitch=false
       bIsVisible=false
       type {classStructOrEnum {refPath=&Vector2f}}
      }
      AdjustValue {bIsDynamic=false, bIsEditable=false, bIsStaticSwitch=false, bIsVisible=false, type {classStructOrEnum {refPath=&NiagaraBool}}}
      Value Range {
       bIsDynamic=false
       bIsEditable=false
       bIsStaticSwitch=false
       bIsVisible=false
       type {classStructOrEnum {refPath=&Vector2f}}
      }
      AdjustAlpha {bIsDynamic=false, bIsEditable=false, bIsStaticSwitch=false, bIsVisible=false, type {classStructOrEnum {refPath=&NiagaraBool}}}
      Alpha Scale Range {
       bIsDynamic=false
       bIsEditable=false
       bIsStaticSwitch=false
       bIsVisible=false
       type {classStructOrEnum {refPath=&Vector2f}}
      }
      Color Channel Mode {
       bIsDynamic=false
       bIsEditable=false
       bIsStaticSwitch=true
       bIsVisible=false
       type {classStructOrEnum {refPath=&ENiagaraChannelCorrelation}}
      }
      Color Randomness Mode {
       bIsDynamic=false
       bIsEditable=false
       bIsStaticSwitch=true
       bIsVisible=false
       type {classStructOrEnum {refPath=&ENiagaraRandomnessMode}}
      }
      Position Mode {
       bIsDynamic=false
       bIsEditable=true
       bIsStaticSwitch=true
       bIsVisible=true
       type {classStructOrEnum {refPath=&ENiagara_PositionInitializationMode}}
      }
      Position {
       bIsDynamic=false
       bIsEditable=false
       bIsStaticSwitch=false
       bIsVisible=false
       type {classStructOrEnum {refPath=&NiagaraPosition}}
      }
      UsePositionOffset {bIsDynamic=false, bIsEditable=true, bIsStaticSwitch=false, bIsVisible=true, type {classStructOrEnum {refPath=&NiagaraBool}}}
      Position Offset {bIsDynamic=false, bIsEditable=false, bIsStaticSwitch=false, bIsVisible=true, type {classStructOrEnum {refPath=&Vector3f}}}
      Position Offset Coordinate Space {
       bIsDynamic=false
       bIsEditable=false
       bIsStaticSwitch=false
       bIsVisible=false
       type {classStructOrEnum {refPath=&ENiagaraCoordinateSpace}}
      }
      Mass Mode {
       bIsDynamic=false
       bIsEditable=true
       bIsStaticSwitch=true
       bIsVisible=true
       type {classStructOrEnum {refPath=&ENiagara_MassInitializationMode}}
      }
      Mass Min {
       bIsDynamic=false
       bIsEditable=false
       bIsStaticSwitch=false
       bIsVisible=false
       type {classStructOrEnum {refPath=&NiagaraFloat}}
      }
      Mass {
       bIsDynamic=false
       bIsEditable=false
       bIsStaticSwitch=false
       bIsVisible=false
       type {classStructOrEnum {refPath=&NiagaraFloat}}
      }
      Mass Max {
       bIsDynamic=false
       bIsEditable=false
       bIsStaticSwitch=false
       bIsVisible=false
       type {classStructOrEnum {refPath=&NiagaraFloat}}
      }
      Mass Randomness Mode {
       bIsDynamic=false
       bIsEditable=false
       bIsStaticSwitch=true
       bIsVisible=false
       type {classStructOrEnum {refPath=&ENiagaraRandomnessMode}}
      }
      Write Material Random {bIsDynamic=false, bIsEditable=true, bIsStaticSwitch=false, bIsVisible=true, type {classStructOrEnum {refPath=&NiagaraBool}}}
      Material Random {bIsDynamic=false, bIsEditable=true, bIsStaticSwitch=false, bIsVisible=true, type {classStructOrEnum {refPath=&NiagaraFloat}}}
      Sprite Size Mode {
       bIsDynamic=false
       bIsEditable=true
       bIsStaticSwitch=true
       bIsVisible=true
       type {classStructOrEnum {refPath=&ENiagara_SizeScaleMode}}
      }
      Sprite Size {
       bIsDynamic=false
       bIsEditable=false
       bIsStaticSwitch=false
       bIsVisible=false
       type {classStructOrEnum {refPath=&Vector2f}}
      }
      Uniform Sprite Size {
       bIsDynamic=false
       bIsEditable=false
       bIsStaticSwitch=false
       bIsVisible=false
       type {classStructOrEnum {refPath=&NiagaraFloat}}
      }
      Uniform Sprite Size Min {
       bIsDynamic=false
       bIsEditable=false
       bIsStaticSwitch=false
       bIsVisible=false
       type {classStructOrEnum {refPath=&NiagaraFloat}}
      }
      Sprite Size Min {
       bIsDynamic=false
       bIsEditable=false
       bIsStaticSwitch=false
       bIsVisible=false
       type {classStructOrEnum {refPath=&Vector2f}}
      }
      Sprite Size Max {
       bIsDynamic=false
       bIsEditable=false
       bIsStaticSwitch=false
       bIsVisible=false
       type {classStructOrEnum {refPath=&Vector2f}}
      }
      Uniform Sprite Size Max {
       bIsDynamic=false
       bIsEditable=false
       bIsStaticSwitch=false
       bIsVisible=false
       type {classStructOrEnum {refPath=&NiagaraFloat}}
      }
      Sprite Size Randomness Mode {
       bIsDynamic=false
       bIsEditable=false
       bIsStaticSwitch=true
       bIsVisible=false
       type {classStructOrEnum {refPath=&ENiagaraRandomnessMode}}
      }
      Sprite Rotation Mode {
       bIsDynamic=false
       bIsEditable=true
       bIsStaticSwitch=true
       bIsVisible=true
       type {classStructOrEnum {refPath=&ENiagara_SpriteRotationMode}}
      }
      Sprite Rotation Angle Min {
       bIsDynamic=false
       bIsEditable=false
       bIsStaticSwitch=false
       bIsVisible=false
       type {classStructOrEnum {refPath=&NiagaraFloat}}
      }
      Sprite Rotation Angle {
       bIsDynamic=false
       bIsEditable=false
       bIsStaticSwitch=false
       bIsVisible=false
       type {classStructOrEnum {refPath=&NiagaraFloat}}
      }
      Sprite Rotation Angle Max {
       bIsDynamic=false
       bIsEditable=false
       bIsStaticSwitch=false
       bIsVisible=false
       type {classStructOrEnum {refPath=&NiagaraFloat}}
      }
      Sprite Rotation Randomness Mode {
       bIsDynamic=false
       bIsEditable=false
       bIsStaticSwitch=true
       bIsVisible=false
       type {classStructOrEnum {refPath=&ENiagaraRandomnessMode}}
      }
      Sprite UV Mode {
       bIsDynamic=false
       bIsEditable=true
       bIsStaticSwitch=true
       bIsVisible=true
       type {classStructOrEnum {refPath=&ENiagara_UVFlippingMode}}
      }
      Sprite UV Randomness Mode {
       bIsDynamic=false
       bIsEditable=false
       bIsStaticSwitch=true
       bIsVisible=false
       type {classStructOrEnum {refPath=&ENiagaraRandomnessMode}}
      }
      Sprite UV Scale {
       bIsDynamic=false
       bIsEditable=false
       bIsStaticSwitch=false
       bIsVisible=false
       type {classStructOrEnum {refPath=&Vector2f}}
      }
      Mesh Scale Mode {
       bIsDynamic=false
       bIsEditable=true
       bIsStaticSwitch=true
       bIsVisible=true
       type {classStructOrEnum {refPath=&ENiagara_SizeScaleMode}}
      }
      Mesh Uniform Scale Min {
       bIsDynamic=false
       bIsEditable=false
       bIsStaticSwitch=false
       bIsVisible=false
       type {classStructOrEnum {refPath=&NiagaraFloat}}
      }
      Mesh Uniform Scale {
       bIsDynamic=false
       bIsEditable=false
       bIsStaticSwitch=false
       bIsVisible=false
       type {classStructOrEnum {refPath=&NiagaraFloat}}
      }
      Mesh Scale Min {
       bIsDynamic=false
       bIsEditable=false
       bIsStaticSwitch=false
       bIsVisible=false
       type {classStructOrEnum {refPath=&Vector3f}}
      }
      Mesh Scale {
       bIsDynamic=false
       bIsEditable=false
       bIsStaticSwitch=false
       bIsVisible=false
       type {classStructOrEnum {refPath=&Vector3f}}
      }
      Mesh Scale Max {
       bIsDynamic=false
       bIsEditable=false
       bIsStaticSwitch=false
       bIsVisible=false
       type {classStructOrEnum {refPath=&Vector3f}}
      }
      Mesh Uniform Scale Max {
       bIsDynamic=false
       bIsEditable=false
       bIsStaticSwitch=false
       bIsVisible=false
       type {classStructOrEnum {refPath=&NiagaraFloat}}
      }
      Mesh Scale Randomness Mode {
       bIsDynamic=false
       bIsEditable=false
       bIsStaticSwitch=true
       bIsVisible=false
       type {classStructOrEnum {refPath=&ENiagaraRandomnessMode}}
      }
      Mesh Renderer Array Visibility Mode {
       bIsDynamic=false
       bIsEditable=true
       bIsStaticSwitch=true
       bIsVisible=true
       type {classStructOrEnum {refPath=&ENiagara_UnsetDirectSetRandom}}
      }
      Mesh Renderer Mesh Index {
       bIsDynamic=false
       bIsEditable=false
       bIsStaticSwitch=false
       bIsVisible=false
       type {classStructOrEnum {refPath=&NiagaraInt32}}
      }
      Wrap Index By Mesh Count {bIsDynamic=false, bIsEditable=false, bIsStaticSwitch=true, bIsVisible=false, type {classStructOrEnum {refPath=&NiagaraBool}}}
      Mesh Renderer Info {
       bIsDynamic=false
       bIsEditable=false
       bIsStaticSwitch=false
       bIsVisible=false
       type {classStructOrEnum {refPath=&NiagaraDataInterfaceMeshRendererInfo}}
      }
      Mesh Renderer Random Seed {
       bIsDynamic=false
       bIsEditable=false
       bIsStaticSwitch=false
       bIsVisible=false
       type {classStructOrEnum {refPath=&NiagaraInt32}}
      }
      Ribbon Width Mode {
       bIsDynamic=false
       bIsEditable=true
       bIsStaticSwitch=true
       bIsVisible=true
       type {classStructOrEnum {refPath=&ENiagara_UnsetDirectSet}}
      }
      Ribbon Width {bIsDynamic=true, bIsEditable=true, bIsStaticSwitch=false, bIsVisible=true, type {classStructOrEnum {refPath=&NiagaraFloat}}}
      Ribbon Facing Mode {
       bIsDynamic=false
       bIsEditable=true
       bIsStaticSwitch=true
       bIsVisible=true
       type {classStructOrEnum {refPath=&ENiagara_UnsetDirectSet}}
      }
      Ribbon Facing Vector {
       bIsDynamic=false
       bIsEditable=false
       bIsStaticSwitch=false
       bIsVisible=false
       type {classStructOrEnum {refPath=&Vector3f}}
      }
      Ribbon Twist Mode {
       bIsDynamic=false
       bIsEditable=true
       bIsStaticSwitch=true
       bIsVisible=true
       type {classStructOrEnum {refPath=&ENiagara_UnsetDirectSet}}
      }
      "Ribbon Twist " {
       bIsDynamic=false
       bIsEditable=false
       bIsStaticSwitch=false
       bIsVisible=false
       type {classStructOrEnum {refPath=&NiagaraFloat}}
      }
     ]
     moduleScript {refPath=&InitializeParticle}
    }
    {
     bIsSetParametersModule=false
     enabled=true
     moduleName=SampleParticlesFromOtherEmitter001
     inputs [
      Emitter Sampling Mode {
       bIsDynamic=false
       bIsEditable=true
       bIsStaticSwitch=true
       bIsVisible=true
       type {classStructOrEnum {refPath=/Niagara/Enums/ENiagara_EmitterLocSamplingMode}}
      }
      Transform Data Option {bIsDynamic=false, bIsEditable=true, bIsStaticSwitch=true, bIsVisible=true, type {classStructOrEnum {refPath=&NiagaraInt32}}}
      Reject Particle By Sampled Age {bIsDynamic=false, bIsEditable=true, bIsStaticSwitch=true, bIsVisible=true, type {classStructOrEnum {refPath=&NiagaraBool}}}
      Minimum Sampled Particle Age {
       bIsDynamic=false
       bIsEditable=false
       bIsStaticSwitch=false
       bIsVisible=false
       type {classStructOrEnum {refPath=&NiagaraFloat}}
      }
      Maximum Sampled Particle Age {
       bIsDynamic=false
       bIsEditable=false
       bIsStaticSwitch=false
       bIsVisible=false
       type {classStructOrEnum {refPath=&NiagaraFloat}}
      }
      Reject Particle If Sampled Index Is Invalid {bIsDynamic=false, bIsEditable=true, bIsStaticSwitch=true, bIsVisible=true, type {classStructOrEnum {refPath=&NiagaraBool}}}
      Reject Particle on Sampled Particle's First Frame {bIsDynamic=false, bIsEditable=true, bIsStaticSwitch=true, bIsVisible=true, type {classStructOrEnum {refPath=&NiagaraBool}}}
      Position Sampling {
       bIsDynamic=false
       bIsEditable=true
       bIsStaticSwitch=true
       bIsVisible=true
       type {classStructOrEnum {refPath=/Niagara/Enums/ENiagara_DirectReadSamplingMode}}
      }
      Applied Position Offset {bIsDynamic=false, bIsEditable=true, bIsStaticSwitch=false, bIsVisible=true, type {classStructOrEnum {refPath=&Vector3f}}}
      Velocity Sampling {
       bIsDynamic=false
       bIsEditable=true
       bIsStaticSwitch=true
       bIsVisible=true
       type {classStructOrEnum {refPath=&ENiagara_DirectReadSamplingMode}}
      }
      Velocity Application {
       bIsDynamic=false
       bIsEditable=false
       bIsStaticSwitch=true
       bIsVisible=false
       type {classStructOrEnum {refPath=/Niagara/Enums/ENiagara_DirectReadApplicationMode}}
      }
      Applied Velocity Scale {
       bIsDynamic=false
       bIsEditable=false
       bIsStaticSwitch=false
       bIsVisible=false
       type {classStructOrEnum {refPath=&NiagaraFloat}}
      }
      Color Sampling {
       bIsDynamic=false
       bIsEditable=true
       bIsStaticSwitch=true
       bIsVisible=true
       type {classStructOrEnum {refPath=&ENiagara_DirectReadSamplingMode}}
      }
      Apply Sampled RGB {bIsDynamic=false, bIsEditable=false, bIsStaticSwitch=true, bIsVisible=false, type {classStructOrEnum {refPath=&NiagaraBool}}}
      Applied RGB Scale {
       bIsDynamic=false
       bIsEditable=false
       bIsStaticSwitch=false
       bIsVisible=false
       type {classStructOrEnum {refPath=&Vector3f}}
      }
      Apply Sampled Alpha {bIsDynamic=false, bIsEditable=false, bIsStaticSwitch=true, bIsVisible=false, type {classStructOrEnum {refPath=&NiagaraBool}}}
      Applied Alpha Scale {
       bIsDynamic=false
       bIsEditable=false
       bIsStaticSwitch=false
       bIsVisible=false
       type {classStructOrEnum {refPath=&NiagaraFloat}}
      }
      Age Sampling {
       bIsDynamic=false
       bIsEditable=true
       bIsStaticSwitch=true
       bIsVisible=true
       type {classStructOrEnum {refPath=&ENiagara_DirectReadSamplingMode}}
      }
      Lifetime Sampling {
       bIsDynamic=false
       bIsEditable=true
       bIsStaticSwitch=true
       bIsVisible=true
       type {classStructOrEnum {refPath=&ENiagara_DirectReadSamplingMode}}
      }
      Lifetime Application {
       bIsDynamic=false
       bIsEditable=true
       bIsStaticSwitch=true
       bIsVisible=true
       type {classStructOrEnum {refPath=/Niagara/Enums/ENiagara_LifetimeInheritanceOptions}}
      }
      Mass Sampling {
       bIsDynamic=false
       bIsEditable=true
       bIsStaticSwitch=true
       bIsVisible=true
       type {classStructOrEnum {refPath=&ENiagara_DirectReadSamplingMode}}
      }
      Mass Application {
       bIsDynamic=false
       bIsEditable=false
       bIsStaticSwitch=true
       bIsVisible=false
       type {classStructOrEnum {refPath=&ENiagara_LifetimeInheritanceOptions}}
      }
      Material Random Sampling {
       bIsDynamic=false
       bIsEditable=true
       bIsStaticSwitch=true
       bIsVisible=true
       type {classStructOrEnum {refPath=&ENiagara_DirectReadSamplingMode}}
      }
      Sprite Size Sampling {
       bIsDynamic=false
       bIsEditable=true
       bIsStaticSwitch=true
       bIsVisible=true
       type {classStructOrEnum {refPath=&ENiagara_DirectReadSamplingMode}}
      }
      Sprite Rotation Sampling {
       bIsDynamic=false
       bIsEditable=true
       bIsStaticSwitch=true
       bIsVisible=true
       type {classStructOrEnum {refPath=&ENiagara_DirectReadSamplingMode}}
      }
      Sprite Facing Sampling {
       bIsDynamic=false
       bIsEditable=true
       bIsStaticSwitch=true
       bIsVisible=true
       type {classStructOrEnum {refPath=&ENiagara_DirectReadSamplingMode}}
      }
      Sprite Alignment Sampling {
       bIsDynamic=false
       bIsEditable=true
       bIsStaticSwitch=true
       bIsVisible=true
       type {classStructOrEnum {refPath=&ENiagara_DirectReadSamplingMode}}
      }
      Mesh Scale Sampling {
       bIsDynamic=false
       bIsEditable=true
       bIsStaticSwitch=true
       bIsVisible=true
       type {classStructOrEnum {refPath=&ENiagara_DirectReadSamplingMode}}
      }
      Mesh Orientation Sampling {
       bIsDynamic=false
       bIsEditable=true
       bIsStaticSwitch=true
       bIsVisible=true
       type {classStructOrEnum {refPath=&ENiagara_DirectReadSamplingMode}}
      }
      Dynamic Material Parameter Sampling {
       bIsDynamic=false
       bIsEditable=true
       bIsStaticSwitch=true
       bIsVisible=true
       type {classStructOrEnum {refPath=&ENiagara_DirectReadSamplingMode}}
      }
      Dynamic Material Parameter 1 Sampling {
       bIsDynamic=false
       bIsEditable=true
       bIsStaticSwitch=true
       bIsVisible=true
       type {classStructOrEnum {refPath=&ENiagara_DirectReadSamplingMode}}
      }
      Dynamic Material Parameter 2 Sampling {
       bIsDynamic=false
       bIsEditable=true
       bIsStaticSwitch=true
       bIsVisible=true
       type {classStructOrEnum {refPath=&ENiagara_DirectReadSamplingMode}}
      }
      Dynamic Material Parameter 3 Sampling {
       bIsDynamic=false
       bIsEditable=true
       bIsStaticSwitch=true
       bIsVisible=true
       type {classStructOrEnum {refPath=&ENiagara_DirectReadSamplingMode}}
      }
      Particle ID Sampling {
       bIsDynamic=false
       bIsEditable=true
       bIsStaticSwitch=true
       bIsVisible=true
       type {classStructOrEnum {refPath=/Niagara/Enums/ENiagara_DirectReadParticleIDSampling}}
      }
      Ribbon UV Distance {
       bIsDynamic=false
       bIsEditable=true
       bIsStaticSwitch=true
       bIsVisible=true
       type {classStructOrEnum {refPath=&ENiagara_DirectReadSamplingMode}}
      }
     ]
     moduleScript {refPath=&SampleParticlesFromOtherEmitter}
    }
   ]
  }
  particleUpdateScript {
   scriptName=ParticleUpdateScript
   modules [
    {
     bIsSetParametersModule=false
     enabled=true
     moduleName=SolveForcesAndVelocity
     inputs [
      Clamp Velocity {bIsDynamic=false, bIsEditable=true, bIsStaticSwitch=false, bIsVisible=true, type {classStructOrEnum {refPath=&NiagaraBool}}}
      Speed Limit {bIsDynamic=false, bIsEditable=false, bIsStaticSwitch=false, bIsVisible=true, type {classStructOrEnum {refPath=&NiagaraFloat}}}
      Limit Acceleration {bIsDynamic=false, bIsEditable=true, bIsStaticSwitch=false, bIsVisible=true, type {classStructOrEnum {refPath=&NiagaraBool}}}
      Acceleration Limit {bIsDynamic=false, bIsEditable=false, bIsStaticSwitch=false, bIsVisible=true, type {classStructOrEnum {refPath=&NiagaraFloat}}}
      Manually Enable Rotational Solver {bIsDynamic=false, bIsEditable=true, bIsStaticSwitch=false, bIsVisible=true, type {classStructOrEnum {refPath=&NiagaraBool}}}
      Write to Intrinsic Properties {bIsDynamic=false, bIsEditable=true, bIsStaticSwitch=false, bIsVisible=true, type {classStructOrEnum {refPath=&NiagaraBool}}}
      Write to Presolve Properties {bIsDynamic=false, bIsEditable=true, bIsStaticSwitch=true, bIsVisible=true, type {classStructOrEnum {refPath=&NiagaraBool}}}
      DeltaTime {bIsDynamic=false, bIsEditable=true, bIsStaticSwitch=false, bIsVisible=true, type {classStructOrEnum {refPath=&NiagaraFloat}}}
      Mass {bIsDynamic=false, bIsEditable=true, bIsStaticSwitch=false, bIsVisible=true, type {classStructOrEnum {refPath=&NiagaraFloat}}}
      Position {
       bIsDynamic=false
       bIsEditable=true
       bIsStaticSwitch=false
       bIsVisible=true
       type {classStructOrEnum {refPath=&NiagaraPosition}}
      }
      Velocity {bIsDynamic=false, bIsEditable=true, bIsStaticSwitch=false, bIsVisible=true, type {classStructOrEnum {refPath=&Vector3f}}}
      PreviousVelocity {bIsDynamic=false, bIsEditable=true, bIsStaticSwitch=false, bIsVisible=true, type {classStructOrEnum {refPath=&Vector3f}}}
      Force {bIsDynamic=false, bIsEditable=true, bIsStaticSwitch=false, bIsVisible=true, type {classStructOrEnum {refPath=&Vector3f}}}
      Mesh Orientation {bIsDynamic=false, bIsEditable=false, bIsStaticSwitch=false, bIsVisible=false, type {classStructOrEnum {refPath=&Quat4f}}}
      Rotational Velocity {
       bIsDynamic=false
       bIsEditable=false
       bIsStaticSwitch=false
       bIsVisible=false
       type {classStructOrEnum {refPath=&Vector3f}}
      }
      Rotational Force {
       bIsDynamic=false
       bIsEditable=false
       bIsStaticSwitch=false
       bIsVisible=false
       type {classStructOrEnum {refPath=&Vector3f}}
      }
      Rotational Inertia {
       bIsDynamic=false
       bIsEditable=false
       bIsStaticSwitch=false
       bIsVisible=false
       type {classStructOrEnum {refPath=&NiagaraFloat}}
      }
     ]
     moduleScript {refPath=&SolveForcesAndVelocity}
    }
    {
     bIsSetParametersModule=false
     enabled=true
     moduleName=ParticleState
     inputs [
      Kill Particles When Lifetime Has Elapsed {bIsDynamic=false, bIsEditable=true, bIsStaticSwitch=true, bIsVisible=true, type {classStructOrEnum {refPath=&NiagaraBool}}}
      Loop Particles Lifetime {bIsDynamic=false, bIsEditable=false, bIsStaticSwitch=false, bIsVisible=false, type {classStructOrEnum {refPath=&NiagaraBool}}}
      Let Infinitely Lived Particles Die When Emitter Deactivates {bIsDynamic=false, bIsEditable=false, bIsStaticSwitch=false, bIsVisible=false, type {classStructOrEnum {refPath=&NiagaraBool}}}
      Lifetime {bIsDynamic=false, bIsEditable=true, bIsStaticSwitch=false, bIsVisible=true, type {classStructOrEnum {refPath=&NiagaraFloat}}}
      DeltaTime {bIsDynamic=false, bIsEditable=true, bIsStaticSwitch=false, bIsVisible=true, type {classStructOrEnum {refPath=&NiagaraFloat}}}
     ]
     moduleScript {refPath=&ParticleState}
    }
    {
     bIsSetParametersModule=false
     enabled=true
     moduleName=ScaleRibbonWidth
     inputs [Ribbon Width Scale {bIsDynamic=true, bIsEditable=true, bIsStaticSwitch=false, bIsVisible=true, type {classStructOrEnum {refPath=&NiagaraFloat}}}]
     moduleScript {refPath=&ScaleRibbonWidth}
    }
    {
     bIsSetParametersModule=false
     enabled=true
     moduleName=ScaleColor
     inputs [
      Scale Mode {
       bIsDynamic=false
       bIsEditable=true
       bIsStaticSwitch=true
       bIsVisible=true
       type {classStructOrEnum {refPath=&ENiagaraScaleColorMode}}
      }
      Linear Color Curve {
       bIsDynamic=false
       bIsEditable=false
       bIsStaticSwitch=false
       bIsVisible=false
       type {classStructOrEnum {refPath=&NiagaraDataInterfaceColorCurve}}
      }
      ScaleRGB {bIsDynamic=false, bIsEditable=true, bIsStaticSwitch=false, bIsVisible=true, type {classStructOrEnum {refPath=&NiagaraBool}}}
      Scale RGB {bIsDynamic=false, bIsEditable=true, bIsStaticSwitch=false, bIsVisible=true, type {classStructOrEnum {refPath=&Vector3f}}}
      ScaleRGBA {bIsDynamic=false, bIsEditable=false, bIsStaticSwitch=false, bIsVisible=false, type {classStructOrEnum {refPath=&NiagaraBool}}}
      Scale RGBA {
       bIsDynamic=false
       bIsEditable=false
       bIsStaticSwitch=false
       bIsVisible=false
       type {classStructOrEnum {refPath=&Vector4f}}
      }
      Curve Index {
       bIsDynamic=false
       bIsEditable=false
       bIsStaticSwitch=false
       bIsVisible=false
       type {classStructOrEnum {refPath=&NiagaraFloat}}
      }
      ScaleA {bIsDynamic=false, bIsEditable=true, bIsStaticSwitch=false, bIsVisible=true, type {classStructOrEnum {refPath=&NiagaraBool}}}
      Scale Alpha {bIsDynamic=true, bIsEditable=true, bIsStaticSwitch=false, bIsVisible=true, type {classStructOrEnum {refPath=&NiagaraFloat}}}
      Color Value To Scale {
       bIsDynamic=false
       bIsEditable=true
       bIsStaticSwitch=false
       bIsVisible=true
       type {classStructOrEnum {refPath=&LinearColor}}
      }
     ]
     moduleScript {refPath=&ScaleColor}
    }
    {
     bIsSetParametersModule=false
     enabled=false
     moduleName=ReadRibbonSourceParticleAge_07
     inputs [PAR {
       bIsDynamic=false
       bIsEditable=true
       bIsStaticSwitch=false
       bIsVisible=true
       type {classStructOrEnum {refPath=&NiagaraDataInterfaceParticleRead}}
      }]
     moduleScript {refPath=&self:ReadRibbonSourceParticleAge_07}
    }
    {
     bIsSetParametersModule=false
     enabled=true
     moduleName=ScaleColor002
     inputs [
      Scale Mode {
       bIsDynamic=false
       bIsEditable=true
       bIsStaticSwitch=true
       bIsVisible=true
       type {classStructOrEnum {refPath=&ENiagaraScaleColorMode}}
      }
      Linear Color Curve {
       bIsDynamic=false
       bIsEditable=false
       bIsStaticSwitch=false
       bIsVisible=false
       type {classStructOrEnum {refPath=&NiagaraDataInterfaceColorCurve}}
      }
      ScaleRGB {bIsDynamic=false, bIsEditable=true, bIsStaticSwitch=false, bIsVisible=true, type {classStructOrEnum {refPath=&NiagaraBool}}}
      Scale RGB {bIsDynamic=true, bIsEditable=false, bIsStaticSwitch=false, bIsVisible=true, type {classStructOrEnum {refPath=&Vector3f}}}
      ScaleRGBA {bIsDynamic=false, bIsEditable=false, bIsStaticSwitch=false, bIsVisible=false, type {classStructOrEnum {refPath=&NiagaraBool}}}
      Scale RGBA {
       bIsDynamic=false
       bIsEditable=false
       bIsStaticSwitch=false
       bIsVisible=false
       type {classStructOrEnum {refPath=&Vector4f}}
      }
      Curve Index {
       bIsDynamic=false
       bIsEditable=false
       bIsStaticSwitch=false
       bIsVisible=false
       type {classStructOrEnum {refPath=&NiagaraFloat}}
      }
      ScaleA {bIsDynamic=false, bIsEditable=true, bIsStaticSwitch=false, bIsVisible=true, type {classStructOrEnum {refPath=&NiagaraBool}}}
      Scale Alpha {bIsDynamic=true, bIsEditable=true, bIsStaticSwitch=false, bIsVisible=true, type {classStructOrEnum {refPath=&NiagaraFloat}}}
      Color Value To Scale {
       bIsDynamic=false
       bIsEditable=true
       bIsStaticSwitch=false
       bIsVisible=true
       type {classStructOrEnum {refPath=&LinearColor}}
      }
     ]
     moduleScript {refPath=&ScaleColor}
    }
    {
     bIsSetParametersModule=false
     enabled=true
     moduleName=DynamicMaterialParameters
     inputs [
      Write Parameter Index 0 {
       bIsDynamic=false
       bIsEditable=true
       bIsStaticSwitch=true
       bIsVisible=true
       type {classStructOrEnum {refPath=/Niagara/Enums/Utility/ENiagara_DynamicMaterialParametersInputMode}}
      }
      Index 0 Vector 1 {
       bIsDynamic=false
       bIsEditable=false
       bIsStaticSwitch=false
       bIsVisible=false
       type {classStructOrEnum {refPath=&Vector3f}}
      }
      Module.Index0Float0Write {bIsDynamic=false, bIsEditable=true, bIsStaticSwitch=true, bIsVisible=true, type {classStructOrEnum {refPath=&NiagaraBool}}}
      Index 0 Param 1 {bIsDynamic=true, bIsEditable=true, bIsStaticSwitch=false, bIsVisible=true, type {classStructOrEnum {refPath=&NiagaraFloat}}}
      Module.Index0Float1Write {bIsDynamic=false, bIsEditable=true, bIsStaticSwitch=true, bIsVisible=true, type {classStructOrEnum {refPath=&NiagaraBool}}}
      Index 0 Param 2 {bIsDynamic=false, bIsEditable=true, bIsStaticSwitch=false, bIsVisible=true, type {classStructOrEnum {refPath=&NiagaraFloat}}}
      Module.Index0Float2Write {bIsDynamic=false, bIsEditable=true, bIsStaticSwitch=true, bIsVisible=true, type {classStructOrEnum {refPath=&NiagaraBool}}}
      Index 0 Param 3 {bIsDynamic=false, bIsEditable=true, bIsStaticSwitch=false, bIsVisible=true, type {classStructOrEnum {refPath=&NiagaraFloat}}}
      Module.Index0Float3Write {bIsDynamic=false, bIsEditable=true, bIsStaticSwitch=true, bIsVisible=true, type {classStructOrEnum {refPath=&NiagaraBool}}}
      Index 0 Param 4 {bIsDynamic=false, bIsEditable=true, bIsStaticSwitch=false, bIsVisible=true, type {classStructOrEnum {refPath=&NiagaraFloat}}}
      Index 0 Vector2D 1 {
       bIsDynamic=false
       bIsEditable=false
       bIsStaticSwitch=false
       bIsVisible=false
       type {classStructOrEnum {refPath=&Vector2f}}
      }
      Index 0 Vector2D 2 {
       bIsDynamic=false
       bIsEditable=false
       bIsStaticSwitch=false
       bIsVisible=false
       type {classStructOrEnum {refPath=&Vector2f}}
      }
      Index 0 Vector 4 {
       bIsDynamic=false
       bIsEditable=false
       bIsStaticSwitch=false
       bIsVisible=false
       type {classStructOrEnum {refPath=&Vector4f}}
      }
      Index 0 Linear Color {
       bIsDynamic=false
       bIsEditable=false
       bIsStaticSwitch=false
       bIsVisible=false
       type {classStructOrEnum {refPath=&LinearColor}}
      }
      Write Parameter Index 1 {
       bIsDynamic=false
       bIsEditable=true
       bIsStaticSwitch=true
       bIsVisible=true
       type {classStructOrEnum {refPath=&ENiagara_DynamicMaterialParametersInputMode}}
      }
      Index 1 Vector 1 {
       bIsDynamic=false
       bIsEditable=false
       bIsStaticSwitch=false
       bIsVisible=false
       type {classStructOrEnum {refPath=&Vector3f}}
      }
      Module.Index1Float0Write {bIsDynamic=false, bIsEditable=true, bIsStaticSwitch=true, bIsVisible=true, type {classStructOrEnum {refPath=&NiagaraBool}}}
      Index 1 Param 1 {bIsDynamic=false, bIsEditable=true, bIsStaticSwitch=false, bIsVisible=true, type {classStructOrEnum {refPath=&NiagaraFloat}}}
      Module.Index1Float1Write {bIsDynamic=false, bIsEditable=true, bIsStaticSwitch=true, bIsVisible=true, type {classStructOrEnum {refPath=&NiagaraBool}}}
      Index 1 Param 2 {bIsDynamic=false, bIsEditable=true, bIsStaticSwitch=false, bIsVisible=true, type {classStructOrEnum {refPath=&NiagaraFloat}}}
      Module.Index1Float2Write {bIsDynamic=false, bIsEditable=true, bIsStaticSwitch=true, bIsVisible=true, type {classStructOrEnum {refPath=&NiagaraBool}}}
      Index 1 Param 3 {bIsDynamic=false, bIsEditable=true, bIsStaticSwitch=false, bIsVisible=true, type {classStructOrEnum {refPath=&NiagaraFloat}}}
      Module.Index1Float3Write {bIsDynamic=false, bIsEditable=true, bIsStaticSwitch=true, bIsVisible=true, type {classStructOrEnum {refPath=&NiagaraBool}}}
      Index 1 Param 4 {bIsDynamic=false, bIsEditable=true, bIsStaticSwitch=false, bIsVisible=true, type {classStructOrEnum {refPath=&NiagaraFloat}}}
      Index 1 Vector2D 1 {
       bIsDynamic=false
       bIsEditable=false
       bIsStaticSwitch=false
       bIsVisible=false
       type {classStructOrEnum {refPath=&Vector2f}}
      }
      Index 1 Vector2D 2 {
       bIsDynamic=false
       bIsEditable=false
       bIsStaticSwitch=false
       bIsVisible=false
       type {classStructOrEnum {refPath=&Vector2f}}
      }
      Index 1 Vector 4 {
       bIsDynamic=false
       bIsEditable=false
       bIsStaticSwitch=false
       bIsVisible=false
       type {classStructOrEnum {refPath=&Vector4f}}
      }
      Index 1 Linear Color {
       bIsDynamic=false
       bIsEditable=false
       bIsStaticSwitch=false
       bIsVisible=false
       type {classStructOrEnum {refPath=&LinearColor}}
      }
      Write Parameter Index 2 {
       bIsDynamic=false
       bIsEditable=true
       bIsStaticSwitch=true
       bIsVisible=true
       type {classStructOrEnum {refPath=&ENiagara_DynamicMaterialParametersInputMode}}
      }
      Index 2 Vector 1 {
       bIsDynamic=false
       bIsEditable=false
       bIsStaticSwitch=false
       bIsVisible=false
       type {classStructOrEnum {refPath=&Vector3f}}
      }
      Index 2 Param 1 {
       bIsDynamic=false
       bIsEditable=false
       bIsStaticSwitch=false
       bIsVisible=false
       type {classStructOrEnum {refPath=&NiagaraFloat}}
      }
      Index 2 Param 2 {
       bIsDynamic=false
       bIsEditable=false
       bIsStaticSwitch=false
       bIsVisible=false
       type {classStructOrEnum {refPath=&NiagaraFloat}}
      }
      Index 2 Param 3 {
       bIsDynamic=false
       bIsEditable=false
       bIsStaticSwitch=false
       bIsVisible=false
       type {classStructOrEnum {refPath=&NiagaraFloat}}
      }
      Index 2 Param 4 {
       bIsDynamic=false
       bIsEditable=false
       bIsStaticSwitch=false
       bIsVisible=false
       type {classStructOrEnum {refPath=&NiagaraFloat}}
      }
      Index 2 Vector2D 1 {
       bIsDynamic=false
       bIsEditable=false
       bIsStaticSwitch=false
       bIsVisible=false
       type {classStructOrEnum {refPath=&Vector2f}}
      }
      Index 2 Vector2D 2 {
       bIsDynamic=false
       bIsEditable=false
       bIsStaticSwitch=false
       bIsVisible=false
       type {classStructOrEnum {refPath=&Vector2f}}
      }
      Index 2 Vector 4 {
       bIsDynamic=false
       bIsEditable=false
       bIsStaticSwitch=false
       bIsVisible=false
       type {classStructOrEnum {refPath=&Vector4f}}
      }
      Index 2 Linear Color {
       bIsDynamic=false
       bIsEditable=false
       bIsStaticSwitch=false
       bIsVisible=false
       type {classStructOrEnum {refPath=&LinearColor}}
      }
      Write Parameter Index 3 {
       bIsDynamic=false
       bIsEditable=true
       bIsStaticSwitch=true
       bIsVisible=true
       type {classStructOrEnum {refPath=&ENiagara_DynamicMaterialParametersInputMode}}
      }
      Index 3 Vector 1 {
       bIsDynamic=false
       bIsEditable=false
       bIsStaticSwitch=false
       bIsVisible=false
       type {classStructOrEnum {refPath=&Vector3f}}
      }
      Index 3 Param 1 {
       bIsDynamic=false
       bIsEditable=false
       bIsStaticSwitch=false
       bIsVisible=false
       type {classStructOrEnum {refPath=&NiagaraFloat}}
      }
      Index 3 Param 2 {
       bIsDynamic=false
       bIsEditable=false
       bIsStaticSwitch=false
       bIsVisible=false
       type {classStructOrEnum {refPath=&NiagaraFloat}}
      }
      Index 3 Param 3 {
       bIsDynamic=false
       bIsEditable=false
       bIsStaticSwitch=false
       bIsVisible=false
       type {classStructOrEnum {refPath=&NiagaraFloat}}
      }
      Index 3 Param 4 {
       bIsDynamic=false
       bIsEditable=false
       bIsStaticSwitch=false
       bIsVisible=false
       type {classStructOrEnum {refPath=&NiagaraFloat}}
      }
      Index 3 Vector2D 1 {
       bIsDynamic=false
       bIsEditable=false
       bIsStaticSwitch=false
       bIsVisible=false
       type {classStructOrEnum {refPath=&Vector2f}}
      }
      Index 3 Vector2D 2 {
       bIsDynamic=false
       bIsEditable=false
       bIsStaticSwitch=false
       bIsVisible=false
       type {classStructOrEnum {refPath=&Vector2f}}
      }
      Index 3 Vector 4 {
       bIsDynamic=false
       bIsEditable=false
       bIsStaticSwitch=false
       bIsVisible=false
       type {classStructOrEnum {refPath=&Vector4f}}
      }
      Index 3 Linear Color {
       bIsDynamic=false
       bIsEditable=false
       bIsStaticSwitch=false
       bIsVisible=false
       type {classStructOrEnum {refPath=&LinearColor}}
      }
     ]
     moduleScript {refPath=&DynamicMaterialParameters}
    }
   ]
  }
  rendererClasses [{refPath=&NiagaraRibbonRendererProperties}]
  renderers [{rendererIndex=0, rendererClass {refPath=&NiagaraRibbonRendererProperties}}]
 }
 Snow_Powder {
  emitterSpawnScript {scriptName=EmitterSpawnScript, modules []}
  emitterUpdateScript {
   scriptName=EmitterUpdateScript
   modules [
    {
     bIsSetParametersModule=false
     enabled=true
     moduleName=EmitterState
     inputs [
      Life Cycle Mode {
       bIsDynamic=false
       bIsEditable=true
       bIsStaticSwitch=true
       bIsVisible=true
       type {classStructOrEnum {refPath=&ENiagaraEmitterLifeCycleMode}}
      }
      Inactive Response {
       bIsDynamic=false
       bIsEditable=true
       bIsStaticSwitch=true
       bIsVisible=true
       type {classStructOrEnum {refPath=&ENiagaraInactiveMode}}
      }
      Loop Behavior {
       bIsDynamic=false
       bIsEditable=true
       bIsStaticSwitch=true
       bIsVisible=true
       type {classStructOrEnum {refPath=&ENiagara_EmitterStateOptions}}
      }
      Loop Duration Mode {
       bIsDynamic=false
       bIsEditable=true
       bIsStaticSwitch=true
       bIsVisible=true
       type {classStructOrEnum {refPath=&ENiagara_InfiniteLoopDuration}}
      }
      Loop Count {
       bIsDynamic=false
       bIsEditable=false
       bIsStaticSwitch=false
       bIsVisible=false
       type {classStructOrEnum {refPath=&NiagaraInt32}}
      }
      Loop Duration {bIsDynamic=false, bIsEditable=true, bIsStaticSwitch=false, bIsVisible=true, type {classStructOrEnum {refPath=&NiagaraFloat}}}
      Recalculate Duration Each Loop {bIsDynamic=false, bIsEditable=false, bIsStaticSwitch=false, bIsVisible=false, type {classStructOrEnum {refPath=&NiagaraBool}}}
      UseLoopDelay {bIsDynamic=false, bIsEditable=true, bIsStaticSwitch=false, bIsVisible=true, type {classStructOrEnum {refPath=&NiagaraBool}}}
      Loop Delay {bIsDynamic=false, bIsEditable=false, bIsStaticSwitch=false, bIsVisible=true, type {classStructOrEnum {refPath=&NiagaraFloat}}}
      Delay First Loop Only {bIsDynamic=false, bIsEditable=false, bIsStaticSwitch=false, bIsVisible=false, type {classStructOrEnum {refPath=&NiagaraBool}}}
      Scalability Mode {
       bIsDynamic=false
       bIsEditable=true
       bIsStaticSwitch=true
       bIsVisible=true
       type {classStructOrEnum {refPath=&ENiagaraEmitterScalabilityMode_Limited}}
      }
      Enable Distance Culling {bIsDynamic=false, bIsEditable=false, bIsStaticSwitch=true, bIsVisible=false, type {classStructOrEnum {refPath=&NiagaraBool}}}
      MinDistance {
       bIsDynamic=false
       bIsEditable=false
       bIsStaticSwitch=false
       bIsVisible=false
       type {classStructOrEnum {refPath=&NiagaraFloat}}
      }
      Min Distance Response {
       bIsDynamic=false
       bIsEditable=false
       bIsStaticSwitch=false
       bIsVisible=false
       type {classStructOrEnum {refPath=&ENiagaraExecutionStateManagement}}
      }
      MaxDistance {
       bIsDynamic=false
       bIsEditable=false
       bIsStaticSwitch=false
       bIsVisible=false
       type {classStructOrEnum {refPath=&NiagaraFloat}}
      }
      Max Distance Response {
       bIsDynamic=false
       bIsEditable=false
       bIsStaticSwitch=false
       bIsVisible=false
       type {classStructOrEnum {refPath=&ENiagaraExecutionStateManagement}}
      }
      Scale Spawn Count {bIsDynamic=false, bIsEditable=false, bIsStaticSwitch=true, bIsVisible=false, type {classStructOrEnum {refPath=&NiagaraBool}}}
      Spawn Count Scale {
       bIsDynamic=false
       bIsEditable=false
       bIsStaticSwitch=false
       bIsVisible=false
       type {classStructOrEnum {refPath=&NiagaraFloat}}
      }
      Spawn Count Scale By Distance {
       bIsDynamic=false
       bIsEditable=false
       bIsStaticSwitch=false
       bIsVisible=false
       type {classStructOrEnum {refPath=&NiagaraDataInterfaceCurve}}
      }
      Enable Visibility Culling {bIsDynamic=false, bIsEditable=false, bIsStaticSwitch=true, bIsVisible=false, type {classStructOrEnum {refPath=&NiagaraBool}}}
      Visibility Cull Response {
       bIsDynamic=false
       bIsEditable=false
       bIsStaticSwitch=false
       bIsVisible=false
       type {classStructOrEnum {refPath=&ENiagaraExecutionStateManagement}}
      }
      Visibility Cull Delay {
       bIsDynamic=false
       bIsEditable=false
       bIsStaticSwitch=false
       bIsVisible=false
       type {classStructOrEnum {refPath=&NiagaraFloat}}
      }
      Reset Age On Awaken {bIsDynamic=false, bIsEditable=false, bIsStaticSwitch=false, bIsVisible=false, type {classStructOrEnum {refPath=&NiagaraBool}}}
     ]
     moduleScript {refPath=&EmitterState}
    }
    {
     bIsSetParametersModule=false
     enabled=true
     moduleName=SpawnBurst_Instantaneous
     inputs [
      Spawn Count {bIsDynamic=false, bIsEditable=true, bIsStaticSwitch=false, bIsVisible=true, type {classStructOrEnum {refPath=&NiagaraInt32}}}
      Spawn Time {bIsDynamic=false, bIsEditable=true, bIsStaticSwitch=false, bIsVisible=true, type {classStructOrEnum {refPath=&NiagaraFloat}}}
      Use Spawn Probability {bIsDynamic=false, bIsEditable=true, bIsStaticSwitch=false, bIsVisible=true, type {classStructOrEnum {refPath=&NiagaraBool}}}
      Spawn Probability {bIsDynamic=false, bIsEditable=false, bIsStaticSwitch=false, bIsVisible=true, type {classStructOrEnum {refPath=&NiagaraFloat}}}
      UseLoopCountLimit {bIsDynamic=false, bIsEditable=true, bIsStaticSwitch=false, bIsVisible=true, type {classStructOrEnum {refPath=&NiagaraBool}}}
      Loop Count Limit {bIsDynamic=false, bIsEditable=false, bIsStaticSwitch=false, bIsVisible=true, type {classStructOrEnum {refPath=&NiagaraInt32}}}
      Spawn Group {bIsDynamic=false, bIsEditable=true, bIsStaticSwitch=false, bIsVisible=true, type {classStructOrEnum {refPath=&NiagaraInt32}}}
      Age {bIsDynamic=false, bIsEditable=true, bIsStaticSwitch=false, bIsVisible=true, type {classStructOrEnum {refPath=&NiagaraFloat}}}
     ]
     moduleScript {refPath=&SpawnBurst_Instantaneous}
    }
    {
     bIsSetParametersModule=false
     enabled=true
     moduleName=SpawnRate
     inputs [
      SpawnRate {bIsDynamic=false, bIsEditable=true, bIsStaticSwitch=false, bIsVisible=true, type {classStructOrEnum {refPath=&NiagaraFloat}}}
      Use Spawn Probability {bIsDynamic=false, bIsEditable=true, bIsStaticSwitch=false, bIsVisible=true, type {classStructOrEnum {refPath=&NiagaraBool}}}
      Spawn Probability {bIsDynamic=false, bIsEditable=false, bIsStaticSwitch=false, bIsVisible=true, type {classStructOrEnum {refPath=&NiagaraFloat}}}
      SpawnGroup {bIsDynamic=false, bIsEditable=true, bIsStaticSwitch=false, bIsVisible=true, type {classStructOrEnum {refPath=&NiagaraInt32}}}
     ]
     moduleScript {refPath=&SpawnRate}
    }
   ]
  }
  particleSpawnScript {
   scriptName=ParticleSpawnScript
   modules [
    {
     bIsSetParametersModule=false
     enabled=true
     moduleName=InitializeParticle
     inputs [
      Lifetime Mode {
       bIsDynamic=false
       bIsEditable=true
       bIsStaticSwitch=true
       bIsVisible=true
       type {classStructOrEnum {refPath=&ENiagara_LifetimeMode}}
      }
      Lifetime Min {
       bIsDynamic=false
       bIsEditable=false
       bIsStaticSwitch=false
       bIsVisible=false
       type {classStructOrEnum {refPath=&NiagaraFloat}}
      }
      Lifetime {bIsDynamic=true, bIsEditable=true, bIsStaticSwitch=false, bIsVisible=true, type {classStructOrEnum {refPath=&NiagaraFloat}}}
      Lifetime Max {
       bIsDynamic=false
       bIsEditable=false
       bIsStaticSwitch=false
       bIsVisible=false
       type {classStructOrEnum {refPath=&NiagaraFloat}}
      }
      Lifetime Randomness Mode {
       bIsDynamic=false
       bIsEditable=false
       bIsStaticSwitch=true
       bIsVisible=false
       type {classStructOrEnum {refPath=&ENiagaraRandomnessMode}}
      }
      Color Mode {
       bIsDynamic=false
       bIsEditable=true
       bIsStaticSwitch=true
       bIsVisible=true
       type {classStructOrEnum {refPath=&ENiagara_ColorInitializationMode}}
      }
      Color Minimum {
       bIsDynamic=false
       bIsEditable=false
       bIsStaticSwitch=false
       bIsVisible=false
       type {classStructOrEnum {refPath=&LinearColor}}
      }
      Color {
       bIsDynamic=false
       bIsEditable=true
       bIsStaticSwitch=false
       bIsVisible=true
       type {classStructOrEnum {refPath=&LinearColor}}
      }
      AdjustHue {bIsDynamic=false, bIsEditable=false, bIsStaticSwitch=false, bIsVisible=false, type {classStructOrEnum {refPath=&NiagaraBool}}}
      Hue Shift Range {
       bIsDynamic=false
       bIsEditable=false
       bIsStaticSwitch=false
       bIsVisible=false
       type {classStructOrEnum {refPath=&Vector2f}}
      }
      Color Maximum {
       bIsDynamic=false
       bIsEditable=false
       bIsStaticSwitch=false
       bIsVisible=false
       type {classStructOrEnum {refPath=&LinearColor}}
      }
      AdjustSaturation {bIsDynamic=false, bIsEditable=false, bIsStaticSwitch=false, bIsVisible=false, type {classStructOrEnum {refPath=&NiagaraBool}}}
      Saturation Range {
       bIsDynamic=false
       bIsEditable=false
       bIsStaticSwitch=false
       bIsVisible=false
       type {classStructOrEnum {refPath=&Vector2f}}
      }
      AdjustValue {bIsDynamic=false, bIsEditable=false, bIsStaticSwitch=false, bIsVisible=false, type {classStructOrEnum {refPath=&NiagaraBool}}}
      Value Range {
       bIsDynamic=false
       bIsEditable=false
       bIsStaticSwitch=false
       bIsVisible=false
       type {classStructOrEnum {refPath=&Vector2f}}
      }
      AdjustAlpha {bIsDynamic=false, bIsEditable=false, bIsStaticSwitch=false, bIsVisible=false, type {classStructOrEnum {refPath=&NiagaraBool}}}
      Alpha Scale Range {
       bIsDynamic=false
       bIsEditable=false
       bIsStaticSwitch=false
       bIsVisible=false
       type {classStructOrEnum {refPath=&Vector2f}}
      }
      Color Channel Mode {
       bIsDynamic=false
       bIsEditable=false
       bIsStaticSwitch=true
       bIsVisible=false
       type {classStructOrEnum {refPath=&ENiagaraChannelCorrelation}}
      }
      Color Randomness Mode {
       bIsDynamic=false
       bIsEditable=false
       bIsStaticSwitch=true
       bIsVisible=false
       type {classStructOrEnum {refPath=&ENiagaraRandomnessMode}}
      }
      Position Mode {
       bIsDynamic=false
       bIsEditable=true
       bIsStaticSwitch=true
       bIsVisible=true
       type {classStructOrEnum {refPath=&ENiagara_PositionInitializationMode}}
      }
      Position {
       bIsDynamic=false
       bIsEditable=false
       bIsStaticSwitch=false
       bIsVisible=false
       type {classStructOrEnum {refPath=&NiagaraPosition}}
      }
      UsePositionOffset {bIsDynamic=false, bIsEditable=true, bIsStaticSwitch=false, bIsVisible=true, type {classStructOrEnum {refPath=&NiagaraBool}}}
      Position Offset {bIsDynamic=false, bIsEditable=false, bIsStaticSwitch=false, bIsVisible=true, type {classStructOrEnum {refPath=&Vector3f}}}
      Position Offset Coordinate Space {
       bIsDynamic=false
       bIsEditable=false
       bIsStaticSwitch=false
       bIsVisible=false
       type {classStructOrEnum {refPath=&ENiagaraCoordinateSpace}}
      }
      Mass Mode {
       bIsDynamic=false
       bIsEditable=true
       bIsStaticSwitch=true
       bIsVisible=true
       type {classStructOrEnum {refPath=&ENiagara_MassInitializationMode}}
      }
      Mass Min {
       bIsDynamic=false
       bIsEditable=false
       bIsStaticSwitch=false
       bIsVisible=false
       type {classStructOrEnum {refPath=&NiagaraFloat}}
      }
      Mass {
       bIsDynamic=false
       bIsEditable=false
       bIsStaticSwitch=false
       bIsVisible=false
       type {classStructOrEnum {refPath=&NiagaraFloat}}
      }
      Mass Max {
       bIsDynamic=false
       bIsEditable=false
       bIsStaticSwitch=false
       bIsVisible=false
       type {classStructOrEnum {refPath=&NiagaraFloat}}
      }
      Mass Randomness Mode {
       bIsDynamic=false
       bIsEditable=false
       bIsStaticSwitch=true
       bIsVisible=false
       type {classStructOrEnum {refPath=&ENiagaraRandomnessMode}}
      }
      Write Material Random {bIsDynamic=false, bIsEditable=true, bIsStaticSwitch=false, bIsVisible=true, type {classStructOrEnum {refPath=&NiagaraBool}}}
      Material Random {bIsDynamic=false, bIsEditable=true, bIsStaticSwitch=false, bIsVisible=true, type {classStructOrEnum {refPath=&NiagaraFloat}}}
      Sprite Size Mode {
       bIsDynamic=false
       bIsEditable=true
       bIsStaticSwitch=true
       bIsVisible=true
       type {classStructOrEnum {refPath=&ENiagara_SizeScaleMode}}
      }
      Sprite Size {
       bIsDynamic=false
       bIsEditable=false
       bIsStaticSwitch=false
       bIsVisible=false
       type {classStructOrEnum {refPath=&Vector2f}}
      }
      Uniform Sprite Size {
       bIsDynamic=false
       bIsEditable=false
       bIsStaticSwitch=false
       bIsVisible=false
       type {classStructOrEnum {refPath=&NiagaraFloat}}
      }
      Uniform Sprite Size Min {bIsDynamic=false, bIsEditable=true, bIsStaticSwitch=false, bIsVisible=true, type {classStructOrEnum {refPath=&NiagaraFloat}}}
      Sprite Size Min {
       bIsDynamic=false
       bIsEditable=false
       bIsStaticSwitch=false
       bIsVisible=false
       type {classStructOrEnum {refPath=&Vector2f}}
      }
      Sprite Size Max {
       bIsDynamic=false
       bIsEditable=false
       bIsStaticSwitch=false
       bIsVisible=false
       type {classStructOrEnum {refPath=&Vector2f}}
      }
      Uniform Sprite Size Max {bIsDynamic=false, bIsEditable=true, bIsStaticSwitch=false, bIsVisible=true, type {classStructOrEnum {refPath=&NiagaraFloat}}}
      Sprite Size Randomness Mode {
       bIsDynamic=false
       bIsEditable=true
       bIsStaticSwitch=true
       bIsVisible=true
       type {classStructOrEnum {refPath=&ENiagaraRandomnessMode}}
      }
      Sprite Rotation Mode {
       bIsDynamic=false
       bIsEditable=true
       bIsStaticSwitch=true
       bIsVisible=true
       type {classStructOrEnum {refPath=&ENiagara_SpriteRotationMode}}
      }
      Sprite Rotation Angle Min {bIsDynamic=false, bIsEditable=true, bIsStaticSwitch=false, bIsVisible=true, type {classStructOrEnum {refPath=&NiagaraFloat}}}
      Sprite Rotation Angle {
       bIsDynamic=false
       bIsEditable=false
       bIsStaticSwitch=false
       bIsVisible=false
       type {classStructOrEnum {refPath=&NiagaraFloat}}
      }
      Sprite Rotation Angle Max {bIsDynamic=false, bIsEditable=true, bIsStaticSwitch=false, bIsVisible=true, type {classStructOrEnum {refPath=&NiagaraFloat}}}
      Sprite Rotation Randomness Mode {
       bIsDynamic=false
       bIsEditable=true
       bIsStaticSwitch=true
       bIsVisible=true
       type {classStructOrEnum {refPath=&ENiagaraRandomnessMode}}
      }
      Sprite UV Mode {
       bIsDynamic=false
       bIsEditable=true
       bIsStaticSwitch=true
       bIsVisible=true
       type {classStructOrEnum {refPath=&ENiagara_UVFlippingMode}}
      }
      Sprite UV Randomness Mode {
       bIsDynamic=false
       bIsEditable=false
       bIsStaticSwitch=true
       bIsVisible=false
       type {classStructOrEnum {refPath=&ENiagaraRandomnessMode}}
      }
      Sprite UV Scale {
       bIsDynamic=false
       bIsEditable=false
       bIsStaticSwitch=false
       bIsVisible=false
       type {classStructOrEnum {refPath=&Vector2f}}
      }
      Mesh Scale Mode {
       bIsDynamic=false
       bIsEditable=true
       bIsStaticSwitch=true
       bIsVisible=true
       type {classStructOrEnum {refPath=&ENiagara_SizeScaleMode}}
      }
      Mesh Uniform Scale Min {
       bIsDynamic=false
       bIsEditable=false
       bIsStaticSwitch=false
       bIsVisible=false
       type {classStructOrEnum {refPath=&NiagaraFloat}}
      }
      Mesh Uniform Scale {
       bIsDynamic=false
       bIsEditable=false
       bIsStaticSwitch=false
       bIsVisible=false
       type {classStructOrEnum {refPath=&NiagaraFloat}}
      }
      Mesh Scale Min {
       bIsDynamic=false
       bIsEditable=false
       bIsStaticSwitch=false
       bIsVisible=false
       type {classStructOrEnum {refPath=&Vector3f}}
      }
      Mesh Scale {bIsDynamic=true, bIsEditable=false, bIsStaticSwitch=false, bIsVisible=false, type {classStructOrEnum {refPath=&Vector3f}}}
      Mesh Scale Max {
       bIsDynamic=false
       bIsEditable=false
       bIsStaticSwitch=false
       bIsVisible=false
       type {classStructOrEnum {refPath=&Vector3f}}
      }
      Mesh Uniform Scale Max {
       bIsDynamic=false
       bIsEditable=false
       bIsStaticSwitch=false
       bIsVisible=false
       type {classStructOrEnum {refPath=&NiagaraFloat}}
      }
      Mesh Scale Randomness Mode {
       bIsDynamic=false
       bIsEditable=false
       bIsStaticSwitch=true
       bIsVisible=false
       type {classStructOrEnum {refPath=&ENiagaraRandomnessMode}}
      }
      Mesh Renderer Array Visibility Mode {
       bIsDynamic=false
       bIsEditable=true
       bIsStaticSwitch=true
       bIsVisible=true
       type {classStructOrEnum {refPath=&ENiagara_UnsetDirectSetRandom}}
      }
      Mesh Renderer Mesh Index {
       bIsDynamic=false
       bIsEditable=false
       bIsStaticSwitch=false
       bIsVisible=false
       type {classStructOrEnum {refPath=&NiagaraInt32}}
      }
      Wrap Index By Mesh Count {bIsDynamic=false, bIsEditable=false, bIsStaticSwitch=true, bIsVisible=false, type {classStructOrEnum {refPath=&NiagaraBool}}}
      Mesh Renderer Info {
       bIsDynamic=false
       bIsEditable=false
       bIsStaticSwitch=false
       bIsVisible=false
       type {classStructOrEnum {refPath=&NiagaraDataInterfaceMeshRendererInfo}}
      }
      Mesh Renderer Random Seed {
       bIsDynamic=false
       bIsEditable=false
       bIsStaticSwitch=false
       bIsVisible=false
       type {classStructOrEnum {refPath=&NiagaraInt32}}
      }
      Ribbon Width Mode {
       bIsDynamic=false
       bIsEditable=true
       bIsStaticSwitch=true
       bIsVisible=true
       type {classStructOrEnum {refPath=&ENiagara_UnsetDirectSet}}
      }
      Ribbon Width {
       bIsDynamic=false
       bIsEditable=false
       bIsStaticSwitch=false
       bIsVisible=false
       type {classStructOrEnum {refPath=&NiagaraFloat}}
      }
      Ribbon Facing Mode {
       bIsDynamic=false
       bIsEditable=true
       bIsStaticSwitch=true
       bIsVisible=true
       type {classStructOrEnum {refPath=&ENiagara_UnsetDirectSet}}
      }
      Ribbon Facing Vector {
       bIsDynamic=false
       bIsEditable=false
       bIsStaticSwitch=false
       bIsVisible=false
       type {classStructOrEnum {refPath=&Vector3f}}
      }
      Ribbon Twist Mode {
       bIsDynamic=false
       bIsEditable=true
       bIsStaticSwitch=true
       bIsVisible=true
       type {classStructOrEnum {refPath=&ENiagara_UnsetDirectSet}}
      }
      "Ribbon Twist " {
       bIsDynamic=false
       bIsEditable=false
       bIsStaticSwitch=false
       bIsVisible=false
       type {classStructOrEnum {refPath=&NiagaraFloat}}
      }
     ]
     moduleScript {refPath=&InitializeParticle}
    }
    {
     bIsSetParametersModule=false
     enabled=false
     moduleName=AddVelocity
     inputs [
      Velocity Mode {
       bIsDynamic=false
       bIsEditable=true
       bIsStaticSwitch=true
       bIsVisible=true
       type {classStructOrEnum {refPath=&ENiagara_VelocityMode}}
      }
      Velocity {
       bIsDynamic=false
       bIsEditable=false
       bIsStaticSwitch=false
       bIsVisible=false
       type {classStructOrEnum {refPath=&Vector3f}}
      }
      Velocity Speed {bIsDynamic=false, bIsEditable=true, bIsStaticSwitch=false, bIsVisible=true, type {classStructOrEnum {refPath=&NiagaraFloat}}}
      Velocity Speed Scale {
       bIsDynamic=false
       bIsEditable=false
       bIsStaticSwitch=false
       bIsVisible=false
       type {classStructOrEnum {refPath=&NiagaraFloat}}
      }
      Constrain To Radius {
       bIsDynamic=false
       bIsEditable=false
       bIsStaticSwitch=false
       bIsVisible=false
       type {classStructOrEnum {refPath=&NiagaraFloat}}
      }
      Module.UseVelDistribution {bIsDynamic=false, bIsEditable=true, bIsStaticSwitch=false, bIsVisible=true, type {classStructOrEnum {refPath=&NiagaraBool}}}
      Distribution Along Cone Axis {bIsDynamic=false, bIsEditable=true, bIsStaticSwitch=false, bIsVisible=true, type {classStructOrEnum {refPath=&NiagaraFloat}}}
      Radius Falloff Near / Far {
       bIsDynamic=false
       bIsEditable=false
       bIsStaticSwitch=false
       bIsVisible=false
       type {classStructOrEnum {refPath=&Vector2f}}
      }
      Radius Falloff Exponent {
       bIsDynamic=false
       bIsEditable=false
       bIsStaticSwitch=false
       bIsVisible=false
       type {classStructOrEnum {refPath=&NiagaraFloat}}
      }
      Module.UseConeFalloff {bIsDynamic=false, bIsEditable=true, bIsStaticSwitch=false, bIsVisible=true, type {classStructOrEnum {refPath=&NiagaraBool}}}
      Speed Falloff From Cone Axis {bIsDynamic=false, bIsEditable=true, bIsStaticSwitch=false, bIsVisible=true, type {classStructOrEnum {refPath=&NiagaraFloat}}}
      Invert Velocity Falloff {bIsDynamic=false, bIsEditable=false, bIsStaticSwitch=false, bIsVisible=false, type {classStructOrEnum {refPath=&NiagaraBool}}}
      Override Seed {bIsDynamic=false, bIsEditable=true, bIsStaticSwitch=false, bIsVisible=true, type {classStructOrEnum {refPath=&NiagaraBool}}}
      Random Seed {bIsDynamic=false, bIsEditable=false, bIsStaticSwitch=false, bIsVisible=true, type {classStructOrEnum {refPath=&NiagaraInt32}}}
      Fixed Random Seed {bIsDynamic=false, bIsEditable=false, bIsStaticSwitch=true, bIsVisible=false, type {classStructOrEnum {refPath=&NiagaraBool}}}
      Randomness Mode {
       bIsDynamic=false
       bIsEditable=true
       bIsStaticSwitch=true
       bIsVisible=true
       type {classStructOrEnum {refPath=&ENiagaraRandomnessMode}}
      }
      Cone Axis {bIsDynamic=false, bIsEditable=true, bIsStaticSwitch=false, bIsVisible=true, type {classStructOrEnum {refPath=&Vector3f}}}
      Cone Angle {bIsDynamic=false, bIsEditable=true, bIsStaticSwitch=false, bIsVisible=true, type {classStructOrEnum {refPath=&NiagaraFloat}}}
      Inner Cone Angle {bIsDynamic=false, bIsEditable=true, bIsStaticSwitch=false, bIsVisible=true, type {classStructOrEnum {refPath=&NiagaraFloat}}}
      Cone Angle Mode {
       bIsDynamic=false
       bIsEditable=true
       bIsStaticSwitch=true
       bIsVisible=true
       type {classStructOrEnum {refPath=&ENiagara_AngleInput}}
      }
      Rotation Mode {
       bIsDynamic=false
       bIsEditable=true
       bIsStaticSwitch=true
       bIsVisible=true
       type {classStructOrEnum {refPath=&ENiagara_RotationMode}}
      }
      Rotation Axis {
       bIsDynamic=false
       bIsEditable=false
       bIsStaticSwitch=false
       bIsVisible=false
       type {classStructOrEnum {refPath=&Vector3f}}
      }
      Yaw / Pitch / Roll {
       bIsDynamic=false
       bIsEditable=false
       bIsStaticSwitch=false
       bIsVisible=false
       type {classStructOrEnum {refPath=&Vector3f}}
      }
      Rotation Quaternion {bIsDynamic=false, bIsEditable=false, bIsStaticSwitch=false, bIsVisible=false, type {classStructOrEnum {refPath=&Quat4f}}}
      Rotation Matrix {
       bIsDynamic=false
       bIsEditable=false
       bIsStaticSwitch=false
       bIsVisible=false
       type {classStructOrEnum {refPath=&NiagaraMatrix}}
      }
      Rotation Angle {
       bIsDynamic=false
       bIsEditable=false
       bIsStaticSwitch=false
       bIsVisible=false
       type {classStructOrEnum {refPath=&NiagaraFloat}}
      }
      Invert Rotation Quaternion {bIsDynamic=false, bIsEditable=false, bIsStaticSwitch=false, bIsVisible=false, type {classStructOrEnum {refPath=&NiagaraBool}}}
      Additional Quaternion Rotation {bIsDynamic=false, bIsEditable=false, bIsStaticSwitch=false, bIsVisible=false, type {classStructOrEnum {refPath=&NiagaraBool}}}
      Additional Yaw / Pitch / Roll {
       bIsDynamic=false
       bIsEditable=false
       bIsStaticSwitch=false
       bIsVisible=false
       type {classStructOrEnum {refPath=&Vector3f}}
      }
      Rotation Angle Type {
       bIsDynamic=false
       bIsEditable=false
       bIsStaticSwitch=true
       bIsVisible=false
       type {classStructOrEnum {refPath=&ENiagara_AngleInput}}
      }
      Rotation Coordinate Space {
       bIsDynamic=false
       bIsEditable=true
       bIsStaticSwitch=false
       bIsVisible=true
       type {classStructOrEnum {refPath=&ENiagaraCoordinateSpace}}
      }
      Velocity Origin {
       bIsDynamic=false
       bIsEditable=false
       bIsStaticSwitch=false
       bIsVisible=false
       type {classStructOrEnum {refPath=&NiagaraPosition}}
      }
      Origin Offset {
       bIsDynamic=false
       bIsEditable=false
       bIsStaticSwitch=false
       bIsVisible=false
       type {classStructOrEnum {refPath=&Vector3f}}
      }
      Origin Offset Coordinate Space {
       bIsDynamic=false
       bIsEditable=false
       bIsStaticSwitch=false
       bIsVisible=false
       type {classStructOrEnum {refPath=&ENiagaraCoordinateSpace}}
      }
      Default Position {
       bIsDynamic=false
       bIsEditable=false
       bIsStaticSwitch=false
       bIsVisible=false
       type {classStructOrEnum {refPath=&NiagaraPosition}}
      }
     ]
     moduleScript {refPath=&AddVelocity}
    }
    {
     bIsSetParametersModule=false
     enabled=true
     moduleName=ShapeLocation
     inputs [
      Shape Primitive {
       bIsDynamic=false
       bIsEditable=true
       bIsStaticSwitch=true
       bIsVisible=true
       type {classStructOrEnum {refPath=&ENiagara_LocationShapes}}
      }
      Box / Plane Mode {
       bIsDynamic=false
       bIsEditable=false
       bIsStaticSwitch=true
       bIsVisible=false
       type {classStructOrEnum {refPath=&ENiagara_BoxPlaneMode}}
      }
      Cone Mode {
       bIsDynamic=false
       bIsEditable=false
       bIsStaticSwitch=true
       bIsVisible=false
       type {classStructOrEnum {refPath=&ENiagara_ConeMode}}
      }
      Cylinder Height {
       bIsDynamic=false
       bIsEditable=false
       bIsStaticSwitch=false
       bIsVisible=false
       type {classStructOrEnum {refPath=&NiagaraFloat}}
      }
      Sphere Radius {bIsDynamic=false, bIsEditable=true, bIsStaticSwitch=false, bIsVisible=true, type {classStructOrEnum {refPath=&NiagaraFloat}}}
      Ring / Disc Mode {
       bIsDynamic=false
       bIsEditable=false
       bIsStaticSwitch=true
       bIsVisible=false
       type {classStructOrEnum {refPath=&ENiagaraRingDiscMode}}
      }
      Torus Mode {
       bIsDynamic=false
       bIsEditable=false
       bIsStaticSwitch=true
       bIsVisible=false
       type {classStructOrEnum {refPath=&ENiagaraShapeTorusMode}}
      }
      Knot Ring Loops {
       bIsDynamic=false
       bIsEditable=false
       bIsStaticSwitch=false
       bIsVisible=false
       type {classStructOrEnum {refPath=&NiagaraInt32}}
      }
      Knot Helix Loops {
       bIsDynamic=false
       bIsEditable=false
       bIsStaticSwitch=false
       bIsVisible=false
       type {classStructOrEnum {refPath=&NiagaraInt32}}
      }
      Enable Torus Knot Radius {bIsDynamic=false, bIsEditable=false, bIsStaticSwitch=false, bIsVisible=false, type {classStructOrEnum {refPath=&NiagaraBool}}}
      Torus Knot Tube Radius {
       bIsDynamic=false
       bIsEditable=false
       bIsStaticSwitch=false
       bIsVisible=false
       type {classStructOrEnum {refPath=&NiagaraFloat}}
      }
      Cylinder Radius {
       bIsDynamic=false
       bIsEditable=false
       bIsStaticSwitch=false
       bIsVisible=false
       type {classStructOrEnum {refPath=&NiagaraFloat}}
      }
      Cylinder Height Midpoint {
       bIsDynamic=false
       bIsEditable=false
       bIsStaticSwitch=false
       bIsVisible=false
       type {classStructOrEnum {refPath=&NiagaraFloat}}
      }
      Ring Radius {
       bIsDynamic=false
       bIsEditable=false
       bIsStaticSwitch=false
       bIsVisible=false
       type {classStructOrEnum {refPath=&NiagaraFloat}}
      }
      Box Size {
       bIsDynamic=false
       bIsEditable=false
       bIsStaticSwitch=false
       bIsVisible=false
       type {classStructOrEnum {refPath=&Vector3f}}
      }
      Box Midpoint {
       bIsDynamic=false
       bIsEditable=false
       bIsStaticSwitch=false
       bIsVisible=false
       type {classStructOrEnum {refPath=&Vector3f}}
      }
      Cone Length {
       bIsDynamic=false
       bIsEditable=false
       bIsStaticSwitch=false
       bIsVisible=false
       type {classStructOrEnum {refPath=&NiagaraFloat}}
      }
      Plane Size {
       bIsDynamic=false
       bIsEditable=false
       bIsStaticSwitch=false
       bIsVisible=false
       type {classStructOrEnum {refPath=&Vector2f}}
      }
      Wedge Horizontal Angle {
       bIsDynamic=false
       bIsEditable=false
       bIsStaticSwitch=false
       bIsVisible=false
       type {classStructOrEnum {refPath=&NiagaraFloat}}
      }
      Large Radius {
       bIsDynamic=false
       bIsEditable=false
       bIsStaticSwitch=false
       bIsVisible=false
       type {classStructOrEnum {refPath=&NiagaraFloat}}
      }
      Plane Midpoint {
       bIsDynamic=false
       bIsEditable=false
       bIsStaticSwitch=false
       bIsVisible=false
       type {classStructOrEnum {refPath=&Vector2f}}
      }
      Wedge Vertical Angle {
       bIsDynamic=false
       bIsEditable=false
       bIsStaticSwitch=false
       bIsVisible=false
       type {classStructOrEnum {refPath=&NiagaraFloat}}
      }
      Cone Angle {
       bIsDynamic=false
       bIsEditable=false
       bIsStaticSwitch=false
       bIsVisible=false
       type {classStructOrEnum {refPath=&NiagaraFloat}}
      }
      Handle Radius {
       bIsDynamic=false
       bIsEditable=false
       bIsStaticSwitch=false
       bIsVisible=false
       type {classStructOrEnum {refPath=&NiagaraFloat}}
      }
      Cone Inner Angle {
       bIsDynamic=false
       bIsEditable=false
       bIsStaticSwitch=false
       bIsVisible=false
       type {classStructOrEnum {refPath=&NiagaraFloat}}
      }
      Cone Radial Angle {
       bIsDynamic=false
       bIsEditable=false
       bIsStaticSwitch=false
       bIsVisible=false
       type {classStructOrEnum {refPath=&NiagaraFloat}}
      }
      Cone Angle Type {
       bIsDynamic=false
       bIsEditable=false
       bIsStaticSwitch=true
       bIsVisible=false
       type {classStructOrEnum {refPath=&ENiagara_AngleInput}}
      }
      Flatten Endcaps {bIsDynamic=false, bIsEditable=false, bIsStaticSwitch=true, bIsVisible=false, type {classStructOrEnum {refPath=&NiagaraBool}}}
      Apply To Particle Position {
       bIsDynamic=false
       bIsEditable=true
       bIsStaticSwitch=true
       bIsVisible=true
       type {classStructOrEnum {refPath=&ENiagara_AttributeSamplingApplyOutput}}
      }
      Cone Surface Distribution {
       bIsDynamic=false
       bIsEditable=false
       bIsStaticSwitch=false
       bIsVisible=false
       type {classStructOrEnum {refPath=&NiagaraFloat}}
      }
      Ring / Disc Distribution Mode {
       bIsDynamic=false
       bIsEditable=false
       bIsStaticSwitch=true
       bIsVisible=false
       type {classStructOrEnum {refPath=&ENiagaraSphereDistributionMode}}
      }
      Sphere Distribution {
       bIsDynamic=false
       bIsEditable=true
       bIsStaticSwitch=true
       bIsVisible=true
       type {classStructOrEnum {refPath=&ENiagaraSphereDistributionMode}}
      }
      Torus Distribution Mode {
       bIsDynamic=false
       bIsEditable=false
       bIsStaticSwitch=true
       bIsVisible=false
       type {classStructOrEnum {refPath=&ENiagaraTorusDistributionMode}}
      }
      Sphere Surface Distribution {bIsDynamic=false, bIsEditable=true, bIsStaticSwitch=false, bIsVisible=true, type {classStructOrEnum {refPath=&NiagaraFloat}}}
      Hemisphere Distribution {bIsDynamic=false, bIsEditable=true, bIsStaticSwitch=false, bIsVisible=true, type {classStructOrEnum {refPath=&Vector2f}}}
      Cylinder Distribution {
       bIsDynamic=false
       bIsEditable=false
       bIsStaticSwitch=true
       bIsVisible=false
       type {classStructOrEnum {refPath=&ENiagara_CylinderMode}}
      }
      Surface Only {bIsDynamic=false, bIsEditable=false, bIsStaticSwitch=false, bIsVisible=false, type {classStructOrEnum {refPath=&NiagaraBool}}}
      Surface Only Band Thickness {
       bIsDynamic=false
       bIsEditable=false
       bIsStaticSwitch=false
       bIsVisible=false
       type {classStructOrEnum {refPath=&NiagaraFloat}}
      }
      Use Endcaps In Surface Only Mode {bIsDynamic=false, bIsEditable=false, bIsStaticSwitch=true, bIsVisible=false, type {classStructOrEnum {refPath=&NiagaraBool}}}
      Hemicircle X {bIsDynamic=false, bIsEditable=false, bIsStaticSwitch=true, bIsVisible=false, type {classStructOrEnum {refPath=&NiagaraBool}}}
      Hemicircle Y {bIsDynamic=false, bIsEditable=false, bIsStaticSwitch=true, bIsVisible=false, type {classStructOrEnum {refPath=&NiagaraBool}}}
      Disc Coverage {
       bIsDynamic=false
       bIsEditable=false
       bIsStaticSwitch=false
       bIsVisible=false
       type {classStructOrEnum {refPath=&NiagaraFloat}}
      }
      Hemicircle Internal Cap {bIsDynamic=false, bIsEditable=false, bIsStaticSwitch=true, bIsVisible=false, type {classStructOrEnum {refPath=&NiagaraBool}}}
      U Position {
       bIsDynamic=false
       bIsEditable=false
       bIsStaticSwitch=false
       bIsVisible=false
       type {classStructOrEnum {refPath=&NiagaraFloat}}
      }
      Enable Lathe Profile {bIsDynamic=false, bIsEditable=false, bIsStaticSwitch=true, bIsVisible=false, type {classStructOrEnum {refPath=&NiagaraBool}}}
      Lathe Profile {
       bIsDynamic=false
       bIsEditable=false
       bIsStaticSwitch=false
       bIsVisible=false
       type {classStructOrEnum {refPath=&NiagaraDataInterfaceCurve}}
      }
      Surface Distribution {
       bIsDynamic=false
       bIsEditable=false
       bIsStaticSwitch=false
       bIsVisible=false
       type {classStructOrEnum {refPath=&NiagaraFloat}}
      }
      Surface Expansion Mode {
       bIsDynamic=false
       bIsEditable=false
       bIsStaticSwitch=false
       bIsVisible=false
       type {classStructOrEnum {refPath=&ENiagaraExpansionMode}}
      }
      U Distribution {
       bIsDynamic=false
       bIsEditable=false
       bIsStaticSwitch=false
       bIsVisible=false
       type {classStructOrEnum {refPath=&NiagaraFloat}}
      }
      V Position {
       bIsDynamic=false
       bIsEditable=false
       bIsStaticSwitch=false
       bIsVisible=false
       type {classStructOrEnum {refPath=&NiagaraFloat}}
      }
      Radius Position {
       bIsDynamic=false
       bIsEditable=false
       bIsStaticSwitch=false
       bIsVisible=false
       type {classStructOrEnum {refPath=&NiagaraFloat}}
      }
      Sphere Radius Position {
       bIsDynamic=false
       bIsEditable=false
       bIsStaticSwitch=false
       bIsVisible=false
       type {classStructOrEnum {refPath=&NiagaraFloat}}
      }
      V Distribution {
       bIsDynamic=false
       bIsEditable=false
       bIsStaticSwitch=false
       bIsVisible=false
       type {classStructOrEnum {refPath=&NiagaraFloat}}
      }
      Evaluation Type {
       bIsDynamic=false
       bIsEditable=false
       bIsStaticSwitch=true
       bIsVisible=false
       type {classStructOrEnum {refPath=&ENiagaraRandomnessEvaluation}}
      }
      Uniform Distribution {
       bIsDynamic=false
       bIsEditable=false
       bIsStaticSwitch=false
       bIsVisible=false
       type {classStructOrEnum {refPath=&NiagaraFloat}}
      }
      Uniform Spiral Amount {
       bIsDynamic=false
       bIsEditable=false
       bIsStaticSwitch=false
       bIsVisible=false
       type {classStructOrEnum {refPath=&NiagaraFloat}}
      }
      Uniform Spiral Falloff {
       bIsDynamic=false
       bIsEditable=false
       bIsStaticSwitch=false
       bIsVisible=false
       type {classStructOrEnum {refPath=&NiagaraFloat}}
      }
      Uniform Count {
       bIsDynamic=false
       bIsEditable=false
       bIsStaticSwitch=false
       bIsVisible=false
       type {classStructOrEnum {refPath=&NiagaraInt32}}
      }
      Uniform Particle Index {
       bIsDynamic=false
       bIsEditable=false
       bIsStaticSwitch=false
       bIsVisible=false
       type {classStructOrEnum {refPath=&NiagaraInt32}}
      }
      Hemisphere Angle Type {
       bIsDynamic=false
       bIsEditable=true
       bIsStaticSwitch=true
       bIsVisible=true
       type {classStructOrEnum {refPath=&ENiagara_AngleInput}}
      }
      Uniform Endpoint Offset {
       bIsDynamic=false
       bIsEditable=false
       bIsStaticSwitch=false
       bIsVisible=false
       type {classStructOrEnum {refPath=&NiagaraFloat}}
      }
      MaskSpawnGroup {bIsDynamic=false, bIsEditable=true, bIsStaticSwitch=false, bIsVisible=true, type {classStructOrEnum {refPath=&NiagaraBool}}}
      Spawn Group Mask {bIsDynamic=false, bIsEditable=false, bIsStaticSwitch=false, bIsVisible=true, type {classStructOrEnum {refPath=&NiagaraInt32}}}
      Spawn Group {
       bIsDynamic=false
       bIsEditable=false
       bIsStaticSwitch=false
       bIsVisible=false
       type {classStructOrEnum {refPath=&NiagaraInt32}}
      }
      Randomness Mode {
       bIsDynamic=false
       bIsEditable=true
       bIsStaticSwitch=true
       bIsVisible=true
       type {classStructOrEnum {refPath=&ENiagaraRandomnessMode}}
      }
      Override Seed {bIsDynamic=false, bIsEditable=true, bIsStaticSwitch=false, bIsVisible=true, type {classStructOrEnum {refPath=&NiagaraBool}}}
      Random Seed {bIsDynamic=false, bIsEditable=false, bIsStaticSwitch=false, bIsVisible=true, type {classStructOrEnum {refPath=&NiagaraInt32}}}
      Fixed Random Seed {bIsDynamic=false, bIsEditable=false, bIsStaticSwitch=true, bIsVisible=false, type {classStructOrEnum {refPath=&NiagaraBool}}}
      Transform Order {
       bIsDynamic=false
       bIsEditable=true
       bIsStaticSwitch=true
       bIsVisible=true
       type {classStructOrEnum {refPath=&ENiagara_TransformOrder}}
      }
      Scale Mode {
       bIsDynamic=false
       bIsEditable=true
       bIsStaticSwitch=true
       bIsVisible=true
       type {classStructOrEnum {refPath=&ENiagara_ScaleMode}}
      }
      Non Uniform Scale {bIsDynamic=false, bIsEditable=true, bIsStaticSwitch=false, bIsVisible=true, type {classStructOrEnum {refPath=&Vector3f}}}
      UseOwnerScale {bIsDynamic=false, bIsEditable=true, bIsStaticSwitch=false, bIsVisible=true, type {classStructOrEnum {refPath=&NiagaraBool}}}
      Apply Owner Scale {bIsDynamic=false, bIsEditable=true, bIsStaticSwitch=false, bIsVisible=true, type {classStructOrEnum {refPath=&Vector3f}}}
      Rotation Mode {
       bIsDynamic=false
       bIsEditable=true
       bIsStaticSwitch=true
       bIsVisible=true
       type {classStructOrEnum {refPath=&ENiagara_RotationMode}}
      }
      Rotation Axis {
       bIsDynamic=false
       bIsEditable=false
       bIsStaticSwitch=false
       bIsVisible=false
       type {classStructOrEnum {refPath=&Vector3f}}
      }
      Rotation Matrix {
       bIsDynamic=false
       bIsEditable=false
       bIsStaticSwitch=false
       bIsVisible=false
       type {classStructOrEnum {refPath=&NiagaraMatrix}}
      }
      Rotation Quaternion {bIsDynamic=false, bIsEditable=false, bIsStaticSwitch=false, bIsVisible=false, type {classStructOrEnum {refPath=&Quat4f}}}
      Yaw / Pitch / Roll {
       bIsDynamic=false
       bIsEditable=false
       bIsStaticSwitch=false
       bIsVisible=false
       type {classStructOrEnum {refPath=&Vector3f}}
      }
      Rotation Angle {
       bIsDynamic=false
       bIsEditable=false
       bIsStaticSwitch=false
       bIsVisible=false
       type {classStructOrEnum {refPath=&NiagaraFloat}}
      }
      Invert Rotation Quaternion {bIsDynamic=false, bIsEditable=false, bIsStaticSwitch=false, bIsVisible=false, type {classStructOrEnum {refPath=&NiagaraBool}}}
      Additional Quaternion Rotation {bIsDynamic=false, bIsEditable=false, bIsStaticSwitch=false, bIsVisible=false, type {classStructOrEnum {refPath=&NiagaraBool}}}
      Additional Yaw / Pitch / Roll {
       bIsDynamic=false
       bIsEditable=false
       bIsStaticSwitch=false
       bIsVisible=false
       type {classStructOrEnum {refPath=&Vector3f}}
      }
      Rotation Angle Type {
       bIsDynamic=false
       bIsEditable=false
       bIsStaticSwitch=true
       bIsVisible=false
       type {classStructOrEnum {refPath=&ENiagara_AngleInput}}
      }
      Rotation Coordinate Space {
       bIsDynamic=false
       bIsEditable=true
       bIsStaticSwitch=false
       bIsVisible=true
       type {classStructOrEnum {refPath=&ENiagaraCoordinateSpace}}
      }
      Offset Mode {
       bIsDynamic=false
       bIsEditable=true
       bIsStaticSwitch=true
       bIsVisible=true
       type {classStructOrEnum {refPath=&ENiagara_OffsetMode}}
      }
      Shape Origin {
       bIsDynamic=false
       bIsEditable=true
       bIsStaticSwitch=false
       bIsVisible=true
       type {classStructOrEnum {refPath=&NiagaraPosition}}
      }
      Offset {
       bIsDynamic=false
       bIsEditable=false
       bIsStaticSwitch=false
       bIsVisible=false
       type {classStructOrEnum {refPath=&Vector3f}}
      }
      Offset Coordinate Space {
       bIsDynamic=false
       bIsEditable=false
       bIsStaticSwitch=false
       bIsVisible=false
       type {classStructOrEnum {refPath=&ENiagaraCoordinateSpace}}
      }
      Custom Transform Matrix {
       bIsDynamic=false
       bIsEditable=false
       bIsStaticSwitch=false
       bIsVisible=false
       type {classStructOrEnum {refPath=&NiagaraMatrix}}
      }
      Transform Method {
       bIsDynamic=false
       bIsEditable=true
       bIsStaticSwitch=true
       bIsVisible=true
       type {classStructOrEnum {refPath=&ENiagara_TransformType}}
      }
     ]
     moduleScript {refPath=&ShapeLocation}
    }
    {
     bIsSetParametersModule=false
     enabled=true
     moduleName=AddVelocity001
     inputs [
      Velocity Mode {
       bIsDynamic=false
       bIsEditable=true
       bIsStaticSwitch=true
       bIsVisible=true
       type {classStructOrEnum {refPath=&ENiagara_VelocityMode}}
      }
      Velocity {bIsDynamic=true, bIsEditable=false, bIsStaticSwitch=false, bIsVisible=false, type {classStructOrEnum {refPath=&Vector3f}}}
      Velocity Speed {bIsDynamic=true, bIsEditable=true, bIsStaticSwitch=false, bIsVisible=true, type {classStructOrEnum {refPath=&NiagaraFloat}}}
      Velocity Speed Scale {
       bIsDynamic=false
       bIsEditable=false
       bIsStaticSwitch=false
       bIsVisible=false
       type {classStructOrEnum {refPath=&NiagaraFloat}}
      }
      Module.Use Falloff Distance {bIsDynamic=false, bIsEditable=true, bIsStaticSwitch=false, bIsVisible=true, type {classStructOrEnum {refPath=&NiagaraBool}}}
      Constrain To Radius {bIsDynamic=false, bIsEditable=false, bIsStaticSwitch=false, bIsVisible=true, type {classStructOrEnum {refPath=&NiagaraFloat}}}
      Distribution Along Cone Axis {
       bIsDynamic=false
       bIsEditable=false
       bIsStaticSwitch=false
       bIsVisible=false
       type {classStructOrEnum {refPath=&NiagaraFloat}}
      }
      Radius Falloff Near / Far {
       bIsDynamic=false
       bIsEditable=false
       bIsStaticSwitch=false
       bIsVisible=false
       type {classStructOrEnum {refPath=&Vector2f}}
      }
      Radius Falloff Exponent {
       bIsDynamic=false
       bIsEditable=false
       bIsStaticSwitch=false
       bIsVisible=false
       type {classStructOrEnum {refPath=&NiagaraFloat}}
      }
      Speed Falloff From Cone Axis {
       bIsDynamic=false
       bIsEditable=false
       bIsStaticSwitch=false
       bIsVisible=false
       type {classStructOrEnum {refPath=&NiagaraFloat}}
      }
      Invert Velocity Falloff {bIsDynamic=false, bIsEditable=false, bIsStaticSwitch=false, bIsVisible=false, type {classStructOrEnum {refPath=&NiagaraBool}}}
      Override Seed {bIsDynamic=false, bIsEditable=true, bIsStaticSwitch=false, bIsVisible=true, type {classStructOrEnum {refPath=&NiagaraBool}}}
      Random Seed {bIsDynamic=false, bIsEditable=false, bIsStaticSwitch=false, bIsVisible=true, type {classStructOrEnum {refPath=&NiagaraInt32}}}
      Fixed Random Seed {bIsDynamic=false, bIsEditable=false, bIsStaticSwitch=true, bIsVisible=false, type {classStructOrEnum {refPath=&NiagaraBool}}}
      Randomness Mode {
       bIsDynamic=false
       bIsEditable=true
       bIsStaticSwitch=true
       bIsVisible=true
       type {classStructOrEnum {refPath=&ENiagaraRandomnessMode}}
      }
      Cone Axis {
       bIsDynamic=false
       bIsEditable=false
       bIsStaticSwitch=false
       bIsVisible=false
       type {classStructOrEnum {refPath=&Vector3f}}
      }
      Cone Angle {
       bIsDynamic=false
       bIsEditable=false
       bIsStaticSwitch=false
       bIsVisible=false
       type {classStructOrEnum {refPath=&NiagaraFloat}}
      }
      Inner Cone Angle {
       bIsDynamic=false
       bIsEditable=false
       bIsStaticSwitch=false
       bIsVisible=false
       type {classStructOrEnum {refPath=&NiagaraFloat}}
      }
      Cone Angle Mode {
       bIsDynamic=false
       bIsEditable=false
       bIsStaticSwitch=true
       bIsVisible=false
       type {classStructOrEnum {refPath=&ENiagara_AngleInput}}
      }
      Rotation Mode {
       bIsDynamic=false
       bIsEditable=false
       bIsStaticSwitch=true
       bIsVisible=false
       type {classStructOrEnum {refPath=&ENiagara_RotationMode}}
      }
      Rotation Axis {
       bIsDynamic=false
       bIsEditable=false
       bIsStaticSwitch=false
       bIsVisible=false
       type {classStructOrEnum {refPath=&Vector3f}}
      }
      Yaw / Pitch / Roll {
       bIsDynamic=false
       bIsEditable=false
       bIsStaticSwitch=false
       bIsVisible=false
       type {classStructOrEnum {refPath=&Vector3f}}
      }
      Rotation Quaternion {bIsDynamic=false, bIsEditable=false, bIsStaticSwitch=false, bIsVisible=false, type {classStructOrEnum {refPath=&Quat4f}}}
      Rotation Matrix {
       bIsDynamic=false
       bIsEditable=false
       bIsStaticSwitch=false
       bIsVisible=false
       type {classStructOrEnum {refPath=&NiagaraMatrix}}
      }
      Rotation Angle {
       bIsDynamic=false
       bIsEditable=false
       bIsStaticSwitch=false
       bIsVisible=false
       type {classStructOrEnum {refPath=&NiagaraFloat}}
      }
      Invert Rotation Quaternion {bIsDynamic=false, bIsEditable=false, bIsStaticSwitch=false, bIsVisible=false, type {classStructOrEnum {refPath=&NiagaraBool}}}
      Additional Quaternion Rotation {bIsDynamic=false, bIsEditable=false, bIsStaticSwitch=false, bIsVisible=false, type {classStructOrEnum {refPath=&NiagaraBool}}}
      Additional Yaw / Pitch / Roll {
       bIsDynamic=false
       bIsEditable=false
       bIsStaticSwitch=false
       bIsVisible=false
       type {classStructOrEnum {refPath=&Vector3f}}
      }
      Rotation Angle Type {
       bIsDynamic=false
       bIsEditable=false
       bIsStaticSwitch=true
       bIsVisible=false
       type {classStructOrEnum {refPath=&ENiagara_AngleInput}}
      }
      Rotation Coordinate Space {
       bIsDynamic=false
       bIsEditable=false
       bIsStaticSwitch=false
       bIsVisible=false
       type {classStructOrEnum {refPath=&ENiagaraCoordinateSpace}}
      }
      Velocity Origin {
       bIsDynamic=false
       bIsEditable=true
       bIsStaticSwitch=false
       bIsVisible=true
       type {classStructOrEnum {refPath=&NiagaraPosition}}
      }
      Origin Offset {bIsDynamic=false, bIsEditable=true, bIsStaticSwitch=false, bIsVisible=true, type {classStructOrEnum {refPath=&Vector3f}}}
      Origin Offset Coordinate Space {
       bIsDynamic=false
       bIsEditable=true
       bIsStaticSwitch=false
       bIsVisible=true
       type {classStructOrEnum {refPath=&ENiagaraCoordinateSpace}}
      }
      Default Position {
       bIsDynamic=false
       bIsEditable=true
       bIsStaticSwitch=false
       bIsVisible=true
       type {classStructOrEnum {refPath=&NiagaraPosition}}
      }
     ]
     moduleScript {refPath=&AddVelocity}
    }
   ]
  }
  particleUpdateScript {
   scriptName=ParticleUpdateScript
   modules [
    {
     bIsSetParametersModule=false
     enabled=true
     moduleName=DynamicMaterialParameters
     inputs [
      Write Parameter Index 0 {bIsDynamic=false, bIsEditable=true, bIsStaticSwitch=true, bIsVisible=true, type {classStructOrEnum {refPath=&NiagaraBool}}}
      Module.Index0Param0Write {bIsDynamic=false, bIsEditable=true, bIsStaticSwitch=true, bIsVisible=true, type {classStructOrEnum {refPath=&NiagaraBool}}}
      Index 0 Param 1 {bIsDynamic=false, bIsEditable=true, bIsStaticSwitch=false, bIsVisible=true, type {classStructOrEnum {refPath=&NiagaraFloat}}}
      Module.Index0Param1Write {bIsDynamic=false, bIsEditable=true, bIsStaticSwitch=true, bIsVisible=true, type {classStructOrEnum {refPath=&NiagaraBool}}}
      Index 0 Param 2 {bIsDynamic=true, bIsEditable=true, bIsStaticSwitch=false, bIsVisible=true, type {classStructOrEnum {refPath=&NiagaraFloat}}}
      Module.Index0Param2Write {bIsDynamic=false, bIsEditable=true, bIsStaticSwitch=true, bIsVisible=true, type {classStructOrEnum {refPath=&NiagaraBool}}}
      Index 0 Param 3 {bIsDynamic=false, bIsEditable=true, bIsStaticSwitch=false, bIsVisible=true, type {classStructOrEnum {refPath=&NiagaraFloat}}}
      Module.Index0Param3Write {bIsDynamic=false, bIsEditable=true, bIsStaticSwitch=true, bIsVisible=true, type {classStructOrEnum {refPath=&NiagaraBool}}}
      Index 0 Param 4 {bIsDynamic=false, bIsEditable=true, bIsStaticSwitch=false, bIsVisible=true, type {classStructOrEnum {refPath=&NiagaraFloat}}}
      Write Parameter Index 1 {bIsDynamic=false, bIsEditable=true, bIsStaticSwitch=true, bIsVisible=true, type {classStructOrEnum {refPath=&NiagaraBool}}}
      Index 1 Param 1 {
       bIsDynamic=false
       bIsEditable=false
       bIsStaticSwitch=false
       bIsVisible=false
       type {classStructOrEnum {refPath=&NiagaraFloat}}
      }
      Index 1 Param 2 {
       bIsDynamic=false
       bIsEditable=false
       bIsStaticSwitch=false
       bIsVisible=false
       type {classStructOrEnum {refPath=&NiagaraFloat}}
      }
      Index 1 Param 3 {
       bIsDynamic=false
       bIsEditable=false
       bIsStaticSwitch=false
       bIsVisible=false
       type {classStructOrEnum {refPath=&NiagaraFloat}}
      }
      Index 1 Param 4 {
       bIsDynamic=false
       bIsEditable=false
       bIsStaticSwitch=false
       bIsVisible=false
       type {classStructOrEnum {refPath=&NiagaraFloat}}
      }
      Write Parameter Index 2 {bIsDynamic=false, bIsEditable=true, bIsStaticSwitch=true, bIsVisible=true, type {classStructOrEnum {refPath=&NiagaraBool}}}
      Index 2 Param 1 {
       bIsDynamic=false
       bIsEditable=false
       bIsStaticSwitch=false
       bIsVisible=false
       type {classStructOrEnum {refPath=&NiagaraFloat}}
      }
      Index 2 Param 2 {
       bIsDynamic=false
       bIsEditable=false
       bIsStaticSwitch=false
       bIsVisible=false
       type {classStructOrEnum {refPath=&NiagaraFloat}}
      }
      Index 2 Param 3 {
       bIsDynamic=false
       bIsEditable=false
       bIsStaticSwitch=false
       bIsVisible=false
       type {classStructOrEnum {refPath=&NiagaraFloat}}
      }
      Index 2 Param 4 {
       bIsDynamic=false
       bIsEditable=false
       bIsStaticSwitch=false
       bIsVisible=false
       type {classStructOrEnum {refPath=&NiagaraFloat}}
      }
      Write Parameter Index 3 {bIsDynamic=false, bIsEditable=true, bIsStaticSwitch=true, bIsVisible=true, type {classStructOrEnum {refPath=&NiagaraBool}}}
      Index 3 Param 1 {
       bIsDynamic=false
       bIsEditable=false
       bIsStaticSwitch=false
       bIsVisible=false
       type {classStructOrEnum {refPath=&NiagaraFloat}}
      }
      Index 3 Param 2 {
       bIsDynamic=false
       bIsEditable=false
       bIsStaticSwitch=false
       bIsVisible=false
       type {classStructOrEnum {refPath=&NiagaraFloat}}
      }
      Index 3 Param 3 {
       bIsDynamic=false
       bIsEditable=false
       bIsStaticSwitch=false
       bIsVisible=false
       type {classStructOrEnum {refPath=&NiagaraFloat}}
      }
      Index 3 Param 4 {
       bIsDynamic=false
       bIsEditable=false
       bIsStaticSwitch=false
       bIsVisible=false
       type {classStructOrEnum {refPath=&NiagaraFloat}}
      }
     ]
     moduleScript {refPath=&DynamicMaterialParameters}
    }
    {
     bIsSetParametersModule=false
     enabled=true
     moduleName=CameraOffset
     inputs [Camera Offset Amount {bIsDynamic=true, bIsEditable=true, bIsStaticSwitch=false, bIsVisible=true, type {classStructOrEnum {refPath=&NiagaraFloat}}}]
     moduleScript {refPath=&CameraOffset}
    }
    {
     bIsSetParametersModule=false
     enabled=true
     moduleName=Drag
     inputs [
      Use Linear Drag {bIsDynamic=false, bIsEditable=true, bIsStaticSwitch=false, bIsVisible=true, type {classStructOrEnum {refPath=&NiagaraBool}}}
      Drag {bIsDynamic=false, bIsEditable=true, bIsStaticSwitch=false, bIsVisible=true, type {classStructOrEnum {refPath=&NiagaraFloat}}}
      Use Rotational Drag {bIsDynamic=false, bIsEditable=true, bIsStaticSwitch=false, bIsVisible=true, type {classStructOrEnum {refPath=&NiagaraBool}}}
      Rotational Drag {bIsDynamic=false, bIsEditable=true, bIsStaticSwitch=false, bIsVisible=true, type {classStructOrEnum {refPath=&NiagaraFloat}}}
      Ignore Mass {bIsDynamic=false, bIsEditable=true, bIsStaticSwitch=false, bIsVisible=true, type {classStructOrEnum {refPath=&NiagaraBool}}}
     ]
     moduleScript {refPath=&Drag}
    }
    {
     bIsSetParametersModule=false
     enabled=true
     moduleName=SolveForcesAndVelocity
     inputs [
      Clamp Velocity {bIsDynamic=false, bIsEditable=true, bIsStaticSwitch=false, bIsVisible=true, type {classStructOrEnum {refPath=&NiagaraBool}}}
      Speed Limit {bIsDynamic=false, bIsEditable=false, bIsStaticSwitch=false, bIsVisible=true, type {classStructOrEnum {refPath=&NiagaraFloat}}}
      Limit Acceleration {bIsDynamic=false, bIsEditable=true, bIsStaticSwitch=false, bIsVisible=true, type {classStructOrEnum {refPath=&NiagaraBool}}}
      Acceleration Limit {bIsDynamic=false, bIsEditable=false, bIsStaticSwitch=false, bIsVisible=true, type {classStructOrEnum {refPath=&NiagaraFloat}}}
      Manually Enable Rotational Solver {bIsDynamic=false, bIsEditable=true, bIsStaticSwitch=false, bIsVisible=true, type {classStructOrEnum {refPath=&NiagaraBool}}}
      Write to Intrinsic Properties {bIsDynamic=false, bIsEditable=true, bIsStaticSwitch=false, bIsVisible=true, type {classStructOrEnum {refPath=&NiagaraBool}}}
      Write to Presolve Properties {bIsDynamic=false, bIsEditable=true, bIsStaticSwitch=true, bIsVisible=true, type {classStructOrEnum {refPath=&NiagaraBool}}}
      DeltaTime {bIsDynamic=false, bIsEditable=true, bIsStaticSwitch=false, bIsVisible=true, type {classStructOrEnum {refPath=&NiagaraFloat}}}
      Mass {bIsDynamic=false, bIsEditable=true, bIsStaticSwitch=false, bIsVisible=true, type {classStructOrEnum {refPath=&NiagaraFloat}}}
      Position {
       bIsDynamic=false
       bIsEditable=true
       bIsStaticSwitch=false
       bIsVisible=true
       type {classStructOrEnum {refPath=&NiagaraPosition}}
      }
      Velocity {bIsDynamic=false, bIsEditable=true, bIsStaticSwitch=false, bIsVisible=true, type {classStructOrEnum {refPath=&Vector3f}}}
      PreviousVelocity {bIsDynamic=false, bIsEditable=true, bIsStaticSwitch=false, bIsVisible=true, type {classStructOrEnum {refPath=&Vector3f}}}
      Force {bIsDynamic=false, bIsEditable=true, bIsStaticSwitch=false, bIsVisible=true, type {classStructOrEnum {refPath=&Vector3f}}}
      Mesh Orientation {bIsDynamic=false, bIsEditable=false, bIsStaticSwitch=false, bIsVisible=false, type {classStructOrEnum {refPath=&Quat4f}}}
      Rotational Velocity {
       bIsDynamic=false
       bIsEditable=false
       bIsStaticSwitch=false
       bIsVisible=false
       type {classStructOrEnum {refPath=&Vector3f}}
      }
      Rotational Force {
       bIsDynamic=false
       bIsEditable=false
       bIsStaticSwitch=false
       bIsVisible=false
       type {classStructOrEnum {refPath=&Vector3f}}
      }
      Rotational Inertia {
       bIsDynamic=false
       bIsEditable=false
       bIsStaticSwitch=false
       bIsVisible=false
       type {classStructOrEnum {refPath=&NiagaraFloat}}
      }
     ]
     moduleScript {refPath=&SolveForcesAndVelocity}
    }
    {
     bIsSetParametersModule=false
     enabled=true
     moduleName=ParticleState
     inputs [
      Kill Particles When Lifetime Has Elapsed {bIsDynamic=false, bIsEditable=true, bIsStaticSwitch=true, bIsVisible=true, type {classStructOrEnum {refPath=&NiagaraBool}}}
      Loop Particles Lifetime {bIsDynamic=false, bIsEditable=false, bIsStaticSwitch=false, bIsVisible=false, type {classStructOrEnum {refPath=&NiagaraBool}}}
      Let Infinitely Lived Particles Die When Emitter Deactivates {bIsDynamic=false, bIsEditable=false, bIsStaticSwitch=false, bIsVisible=false, type {classStructOrEnum {refPath=&NiagaraBool}}}
      Lifetime {bIsDynamic=false, bIsEditable=true, bIsStaticSwitch=false, bIsVisible=true, type {classStructOrEnum {refPath=&NiagaraFloat}}}
      DeltaTime {bIsDynamic=false, bIsEditable=true, bIsStaticSwitch=false, bIsVisible=true, type {classStructOrEnum {refPath=&NiagaraFloat}}}
     ]
     moduleScript {refPath=&ParticleState}
    }
    {
     bIsSetParametersModule=false
     enabled=false
     moduleName=ScaleColor001
     inputs [
      Scale Mode {
       bIsDynamic=false
       bIsEditable=true
       bIsStaticSwitch=true
       bIsVisible=true
       type {classStructOrEnum {refPath=&ENiagaraScaleColorMode}}
      }
      Linear Color Curve {
       bIsDynamic=false
       bIsEditable=false
       bIsStaticSwitch=false
       bIsVisible=false
       type {classStructOrEnum {refPath=&NiagaraDataInterfaceColorCurve}}
      }
      ScaleRGB {bIsDynamic=false, bIsEditable=false, bIsStaticSwitch=false, bIsVisible=false, type {classStructOrEnum {refPath=&NiagaraBool}}}
      Scale RGB {
       bIsDynamic=false
       bIsEditable=false
       bIsStaticSwitch=false
       bIsVisible=false
       type {classStructOrEnum {refPath=&Vector3f}}
      }
      ScaleRGBA {bIsDynamic=false, bIsEditable=true, bIsStaticSwitch=false, bIsVisible=true, type {classStructOrEnum {refPath=&NiagaraBool}}}
      Scale RGBA {bIsDynamic=true, bIsEditable=true, bIsStaticSwitch=false, bIsVisible=true, type {classStructOrEnum {refPath=&Vector4f}}}
      Curve Index {
       bIsDynamic=false
       bIsEditable=false
       bIsStaticSwitch=false
       bIsVisible=false
       type {classStructOrEnum {refPath=&NiagaraFloat}}
      }
      ScaleA {bIsDynamic=false, bIsEditable=false, bIsStaticSwitch=false, bIsVisible=false, type {classStructOrEnum {refPath=&NiagaraBool}}}
      Scale Alpha {
       bIsDynamic=false
       bIsEditable=false
       bIsStaticSwitch=false
       bIsVisible=false
       type {classStructOrEnum {refPath=&NiagaraFloat}}
      }
      Color Value To Scale {
       bIsDynamic=false
       bIsEditable=true
       bIsStaticSwitch=false
       bIsVisible=true
       type {classStructOrEnum {refPath=&LinearColor}}
      }
     ]
     moduleScript {refPath=&ScaleColor}
    }
    {
     bIsSetParametersModule=false
     enabled=true
     moduleName=ScaleColor
     inputs [
      Scale Mode {
       bIsDynamic=false
       bIsEditable=true
       bIsStaticSwitch=true
       bIsVisible=true
       type {classStructOrEnum {refPath=&ENiagaraScaleColorMode}}
      }
      Linear Color Curve {
       bIsDynamic=false
       bIsEditable=false
       bIsStaticSwitch=false
       bIsVisible=false
       type {classStructOrEnum {refPath=&NiagaraDataInterfaceColorCurve}}
      }
      ScaleRGB {bIsDynamic=false, bIsEditable=true, bIsStaticSwitch=false, bIsVisible=true, type {classStructOrEnum {refPath=&NiagaraBool}}}
      Scale RGB {bIsDynamic=true, bIsEditable=false, bIsStaticSwitch=false, bIsVisible=true, type {classStructOrEnum {refPath=&Vector3f}}}
      ScaleRGBA {bIsDynamic=false, bIsEditable=false, bIsStaticSwitch=false, bIsVisible=false, type {classStructOrEnum {refPath=&NiagaraBool}}}
      Scale RGBA {
       bIsDynamic=false
       bIsEditable=false
       bIsStaticSwitch=false
       bIsVisible=false
       type {classStructOrEnum {refPath=&Vector4f}}
      }
      Curve Index {
       bIsDynamic=false
       bIsEditable=false
       bIsStaticSwitch=false
       bIsVisible=false
       type {classStructOrEnum {refPath=&NiagaraFloat}}
      }
      ScaleA {bIsDynamic=false, bIsEditable=true, bIsStaticSwitch=false, bIsVisible=true, type {classStructOrEnum {refPath=&NiagaraBool}}}
      Scale Alpha {bIsDynamic=true, bIsEditable=true, bIsStaticSwitch=false, bIsVisible=true, type {classStructOrEnum {refPath=&NiagaraFloat}}}
      Color Value To Scale {
       bIsDynamic=false
       bIsEditable=true
       bIsStaticSwitch=false
       bIsVisible=true
       type {classStructOrEnum {refPath=&LinearColor}}
      }
     ]
     moduleScript {refPath=&ScaleColor}
    }
    {
     bIsSetParametersModule=false
     enabled=true
     moduleName=SubUVAnimation
     inputs [
      Manual SubImage Size {
       bIsDynamic=false
       bIsEditable=false
       bIsStaticSwitch=false
       bIsVisible=false
       type {classStructOrEnum {refPath=&Vector2f}}
      }
      Is SubUV Blending Enabled On Renderer {bIsDynamic=false, bIsEditable=false, bIsStaticSwitch=false, bIsVisible=false, type {classStructOrEnum {refPath=&NiagaraBool}}}
      SubUV Animation Mode {
       bIsDynamic=false
       bIsEditable=true
       bIsStaticSwitch=true
       bIsVisible=true
       type {classStructOrEnum {refPath=/Niagara/Enums/ENiagara_SubUVLookupModeV2}}
      }
      Renderer Type {
       bIsDynamic=false
       bIsEditable=true
       bIsStaticSwitch=true
       bIsVisible=true
       type {classStructOrEnum {refPath=/Niagara/Enums/ENiagaraMeshOrSprite}}
      }
      Mesh Renderer {
       bIsDynamic=false
       bIsEditable=false
       bIsStaticSwitch=false
       bIsVisible=false
       type {classStructOrEnum {refPath=&NiagaraDataInterfaceMeshRendererInfo}}
      }
      "Sprite Renderer " {
       bIsDynamic=false
       bIsEditable=true
       bIsStaticSwitch=false
       bIsVisible=true
       type {classStructOrEnum {refPath=&NiagaraDataInterfaceSpriteRendererInfo}}
      }
      Define SubUV Setup Manually {
       bIsDynamic=false
       bIsEditable=true
       bIsStaticSwitch=true
       bIsVisible=true
       type {classStructOrEnum {refPath=/Niagara/Enums/SpriteRenderer/ENiagara_AutomaticManual}}
      }
      Direct Frame Index {
       bIsDynamic=false
       bIsEditable=false
       bIsStaticSwitch=false
       bIsVisible=false
       type {classStructOrEnum {refPath=&NiagaraFloat}}
      }
      Random Start Frame {bIsDynamic=false, bIsEditable=true, bIsStaticSwitch=true, bIsVisible=true, type {classStructOrEnum {refPath=&NiagaraBool}}}
      Start Frame Offset {bIsDynamic=false, bIsEditable=true, bIsStaticSwitch=false, bIsVisible=true, type {classStructOrEnum {refPath=&NiagaraInt32}}}
      Start Time Offset {
       bIsDynamic=false
       bIsEditable=false
       bIsStaticSwitch=false
       bIsVisible=false
       type {classStructOrEnum {refPath=&NiagaraFloat}}
      }
      SubUV Lookup Index {bIsDynamic=false, bIsEditable=true, bIsStaticSwitch=false, bIsVisible=true, type {classStructOrEnum {refPath=&NiagaraFloat}}}
      X Coordinate {
       bIsDynamic=false
       bIsEditable=false
       bIsStaticSwitch=false
       bIsVisible=false
       type {classStructOrEnum {refPath=&NiagaraFloat}}
      }
      Playback Mode {
       bIsDynamic=false
       bIsEditable=false
       bIsStaticSwitch=true
       bIsVisible=false
       type {classStructOrEnum {refPath=/Niagara/Enums/SpriteRenderer/ENiagara_FPSPlayrate}}
      }
      Frames Per Second {
       bIsDynamic=false
       bIsEditable=false
       bIsStaticSwitch=false
       bIsVisible=false
       type {classStructOrEnum {refPath=&NiagaraInt32}}
      }
      Play Rate {
       bIsDynamic=false
       bIsEditable=false
       bIsStaticSwitch=false
       bIsVisible=false
       type {classStructOrEnum {refPath=&NiagaraFloat}}
      }
      Playback Time {
       bIsDynamic=false
       bIsEditable=false
       bIsStaticSwitch=false
       bIsVisible=false
       type {classStructOrEnum {refPath=&NiagaraFloat}}
      }
      UsePlayRateScale {bIsDynamic=false, bIsEditable=false, bIsStaticSwitch=false, bIsVisible=false, type {classStructOrEnum {refPath=&NiagaraBool}}}
      Playback Scale Curve {
       bIsDynamic=false
       bIsEditable=false
       bIsStaticSwitch=false
       bIsVisible=false
       type {classStructOrEnum {refPath=&NiagaraDataInterfaceCurve}}
      }
      Playback Scale Curve Index {
       bIsDynamic=false
       bIsEditable=false
       bIsStaticSwitch=false
       bIsVisible=false
       type {classStructOrEnum {refPath=&NiagaraFloat}}
      }
      SubUV Curve {
       bIsDynamic=false
       bIsEditable=false
       bIsStaticSwitch=false
       bIsVisible=false
       type {classStructOrEnum {refPath=&NiagaraDataInterfaceCurve}}
      }
      SubUV Lookup Index Scale {bIsDynamic=false, bIsEditable=true, bIsStaticSwitch=false, bIsVisible=true, type {classStructOrEnum {refPath=&NiagaraFloat}}}
      Y Coordinate {
       bIsDynamic=false
       bIsEditable=false
       bIsStaticSwitch=false
       bIsVisible=false
       type {classStructOrEnum {refPath=&NiagaraFloat}}
      }
      SubUV Curve Index {
       bIsDynamic=false
       bIsEditable=false
       bIsStaticSwitch=false
       bIsVisible=false
       type {classStructOrEnum {refPath=&NiagaraFloat}}
      }
      Index By X/Y Coordinate {bIsDynamic=false, bIsEditable=false, bIsStaticSwitch=true, bIsVisible=false, type {classStructOrEnum {refPath=&NiagaraBool}}}
      Evaluation Type {
       bIsDynamic=false
       bIsEditable=false
       bIsStaticSwitch=true
       bIsVisible=false
       type {classStructOrEnum {refPath=&ENiagaraRandomnessEvaluation}}
      }
      Random Change Interval {
       bIsDynamic=false
       bIsEditable=false
       bIsStaticSwitch=false
       bIsVisible=false
       type {classStructOrEnum {refPath=&NiagaraFloat}}
      }
      Loop Range Boundary {
       bIsDynamic=false
       bIsEditable=false
       bIsStaticSwitch=true
       bIsVisible=false
       type {classStructOrEnum {refPath=/Niagara/Enums/ENiagara_WrapClamp}}
      }
      Loop Boundary Min/Max {
       bIsDynamic=false
       bIsEditable=false
       bIsStaticSwitch=false
       bIsVisible=false
       type {classStructOrEnum {refPath=&Vector2f}}
      }
      Truncate Y Coordinate {bIsDynamic=false, bIsEditable=false, bIsStaticSwitch=true, bIsVisible=false, type {classStructOrEnum {refPath=&NiagaraBool}}}
      Random Change Interval Time {
       bIsDynamic=false
       bIsEditable=false
       bIsStaticSwitch=false
       bIsVisible=false
       type {classStructOrEnum {refPath=&NiagaraFloat}}
      }
      Randomness Mode {
       bIsDynamic=false
       bIsEditable=false
       bIsStaticSwitch=true
       bIsVisible=false
       type {classStructOrEnum {refPath=&ENiagaraRandomnessMode}}
      }
      Override Seed {bIsDynamic=false, bIsEditable=false, bIsStaticSwitch=true, bIsVisible=false, type {classStructOrEnum {refPath=&NiagaraBool}}}
      Random Seed {
       bIsDynamic=false
       bIsEditable=false
       bIsStaticSwitch=false
       bIsVisible=false
       type {classStructOrEnum {refPath=&NiagaraInt32}}
      }
      Fixed Random Seed {bIsDynamic=false, bIsEditable=false, bIsStaticSwitch=true, bIsVisible=false, type {classStructOrEnum {refPath=&NiagaraBool}}}
      Output Alternate Renderer Binding {bIsDynamic=false, bIsEditable=true, bIsStaticSwitch=true, bIsVisible=true, type {classStructOrEnum {refPath=&NiagaraBool}}}
      UseStartFrame {bIsDynamic=false, bIsEditable=true, bIsStaticSwitch=false, bIsVisible=true, type {classStructOrEnum {refPath=&NiagaraBool}}}
      Start Frame Range Override {bIsDynamic=false, bIsEditable=false, bIsStaticSwitch=false, bIsVisible=true, type {classStructOrEnum {refPath=&NiagaraInt32}}}
      UseEndFrame {bIsDynamic=false, bIsEditable=true, bIsStaticSwitch=false, bIsVisible=true, type {classStructOrEnum {refPath=&NiagaraBool}}}
      End Frame Range Override {bIsDynamic=false, bIsEditable=false, bIsStaticSwitch=false, bIsVisible=true, type {classStructOrEnum {refPath=&NiagaraInt32}}}
     ]
     moduleScript {refPath=&SubUVAnimation}
    }
    {
     bIsSetParametersModule=false
     enabled=false
     moduleName=ScaleSpriteSize
     inputs [
      Scale Sprite Size Mode {
       bIsDynamic=false
       bIsEditable=true
       bIsStaticSwitch=true
       bIsVisible=true
       type {classStructOrEnum {refPath=&ENiagara_ScaleSpriteSize}}
      }
      Uniform Curve Sprite Scale {
       bIsDynamic=false
       bIsEditable=true
       bIsStaticSwitch=false
       bIsVisible=true
       type {classStructOrEnum {refPath=&NiagaraDataInterfaceCurve}}
      }
      Non-Uniform Curve Sprite Scale {
       bIsDynamic=false
       bIsEditable=false
       bIsStaticSwitch=false
       bIsVisible=false
       type {classStructOrEnum {refPath=&NiagaraDataInterfaceVector2DCurve}}
      }
      Uniform Curve Index {bIsDynamic=false, bIsEditable=true, bIsStaticSwitch=false, bIsVisible=true, type {classStructOrEnum {refPath=&NiagaraFloat}}}
      Non-Uniform Curve Index {
       bIsDynamic=false
       bIsEditable=false
       bIsStaticSwitch=false
       bIsVisible=false
       type {classStructOrEnum {refPath=&NiagaraFloat}}
      }
      Non-Uniform Curve Scale {
       bIsDynamic=false
       bIsEditable=false
       bIsStaticSwitch=false
       bIsVisible=false
       type {classStructOrEnum {refPath=&Vector2f}}
      }
      Uniform Curve Scale {bIsDynamic=false, bIsEditable=true, bIsStaticSwitch=false, bIsVisible=true, type {classStructOrEnum {refPath=&NiagaraFloat}}}
      Scale Factor {
       bIsDynamic=false
       bIsEditable=false
       bIsStaticSwitch=false
       bIsVisible=false
       type {classStructOrEnum {refPath=&Vector2f}}
      }
      Uniform Scale Factor {
       bIsDynamic=false
       bIsEditable=false
       bIsStaticSwitch=false
       bIsVisible=false
       type {classStructOrEnum {refPath=&NiagaraFloat}}
      }
      Initial Sprite Size {bIsDynamic=false, bIsEditable=true, bIsStaticSwitch=false, bIsVisible=true, type {classStructOrEnum {refPath=&Vector2f}}}
     ]
     moduleScript {refPath=&ScaleSpriteSize}
    }
    {
     bIsSetParametersModule=false
     enabled=true
     moduleName=ScaleSpriteSize001
     inputs [
      Scale Sprite Size Mode {
       bIsDynamic=false
       bIsEditable=true
       bIsStaticSwitch=true
       bIsVisible=true
       type {classStructOrEnum {refPath=&ENiagara_ScaleSpriteSize}}
      }
      Uniform Curve Sprite Scale {
       bIsDynamic=false
       bIsEditable=true
       bIsStaticSwitch=false
       bIsVisible=true
       type {classStructOrEnum {refPath=&NiagaraDataInterfaceCurve}}
      }
      Non-Uniform Curve Sprite Scale {
       bIsDynamic=false
       bIsEditable=false
       bIsStaticSwitch=false
       bIsVisible=false
       type {classStructOrEnum {refPath=&NiagaraDataInterfaceVector2DCurve}}
      }
      Uniform Curve Index {bIsDynamic=false, bIsEditable=true, bIsStaticSwitch=false, bIsVisible=true, type {classStructOrEnum {refPath=&NiagaraFloat}}}
      Non-Uniform Curve Index {
       bIsDynamic=false
       bIsEditable=false
       bIsStaticSwitch=false
       bIsVisible=false
       type {classStructOrEnum {refPath=&NiagaraFloat}}
      }
      Non-Uniform Curve Scale {
       bIsDynamic=false
       bIsEditable=false
       bIsStaticSwitch=false
       bIsVisible=false
       type {classStructOrEnum {refPath=&Vector2f}}
      }
      Uniform Curve Scale {bIsDynamic=false, bIsEditable=true, bIsStaticSwitch=false, bIsVisible=true, type {classStructOrEnum {refPath=&NiagaraFloat}}}
      Scale Factor {
       bIsDynamic=false
       bIsEditable=false
       bIsStaticSwitch=false
       bIsVisible=false
       type {classStructOrEnum {refPath=&Vector2f}}
      }
      Uniform Scale Factor {
       bIsDynamic=false
       bIsEditable=false
       bIsStaticSwitch=false
       bIsVisible=false
       type {classStructOrEnum {refPath=&NiagaraFloat}}
      }
      Initial Sprite Size {bIsDynamic=false, bIsEditable=true, bIsStaticSwitch=false, bIsVisible=true, type {classStructOrEnum {refPath=&Vector2f}}}
     ]
     moduleScript {refPath=&ScaleSpriteSize}
    }
    {
     bIsSetParametersModule=false
     enabled=true
     moduleName=ScaleColor002
     inputs [
      Scale Mode {bIsDynamic=false, bIsEditable=true, bIsStaticSwitch=true, bIsVisible=true, type {classStructOrEnum {refPath=&NiagaraInt32}}}
      Linear Color Curve {
       bIsDynamic=false
       bIsEditable=false
       bIsStaticSwitch=false
       bIsVisible=false
       type {classStructOrEnum {refPath=&NiagaraDataInterfaceColorCurve}}
      }
      ScaleRGB {bIsDynamic=false, bIsEditable=true, bIsStaticSwitch=false, bIsVisible=true, type {classStructOrEnum {refPath=&NiagaraBool}}}
      Scale RGB {bIsDynamic=true, bIsEditable=false, bIsStaticSwitch=false, bIsVisible=true, type {classStructOrEnum {refPath=&Vector3f}}}
      ScaleRGBA {bIsDynamic=false, bIsEditable=false, bIsStaticSwitch=false, bIsVisible=false, type {classStructOrEnum {refPath=&NiagaraBool}}}
      Scale RGBA {
       bIsDynamic=false
       bIsEditable=false
       bIsStaticSwitch=false
       bIsVisible=false
       type {classStructOrEnum {refPath=&Vector4f}}
      }
      Curve Index {
       bIsDynamic=false
       bIsEditable=false
       bIsStaticSwitch=false
       bIsVisible=false
       type {classStructOrEnum {refPath=&NiagaraFloat}}
      }
      ScaleA {bIsDynamic=false, bIsEditable=true, bIsStaticSwitch=false, bIsVisible=true, type {classStructOrEnum {refPath=&NiagaraBool}}}
      Scale Alpha {bIsDynamic=true, bIsEditable=true, bIsStaticSwitch=false, bIsVisible=true, type {classStructOrEnum {refPath=&NiagaraFloat}}}
      Color Value To Scale {
       bIsDynamic=false
       bIsEditable=true
       bIsStaticSwitch=false
       bIsVisible=true
       type {classStructOrEnum {refPath=&LinearColor}}
      }
      Color Scale {
       bIsDynamic=false
       bIsEditable=false
       bIsStaticSwitch=false
       bIsVisible=false
       type {classStructOrEnum {refPath=&LinearColor}}
      }
     ]
     moduleScript {refPath=&ScaleColor}
    }
    {
     bIsSetParametersModule=false
     enabled=false
     moduleName=ScaleColor003
     inputs [
      Scale Mode {bIsDynamic=false, bIsEditable=true, bIsStaticSwitch=true, bIsVisible=true, type {classStructOrEnum {refPath=&NiagaraInt32}}}
      Linear Color Curve {
       bIsDynamic=false
       bIsEditable=false
       bIsStaticSwitch=false
       bIsVisible=false
       type {classStructOrEnum {refPath=&NiagaraDataInterfaceColorCurve}}
      }
      ScaleRGB {bIsDynamic=false, bIsEditable=true, bIsStaticSwitch=false, bIsVisible=true, type {classStructOrEnum {refPath=&NiagaraBool}}}
      Scale RGB {bIsDynamic=true, bIsEditable=true, bIsStaticSwitch=false, bIsVisible=true, type {classStructOrEnum {refPath=&Vector3f}}}
      ScaleRGBA {bIsDynamic=false, bIsEditable=false, bIsStaticSwitch=false, bIsVisible=false, type {classStructOrEnum {refPath=&NiagaraBool}}}
      Scale RGBA {
       bIsDynamic=false
       bIsEditable=false
       bIsStaticSwitch=false
       bIsVisible=false
       type {classStructOrEnum {refPath=&Vector4f}}
      }
      Curve Index {
       bIsDynamic=false
       bIsEditable=false
       bIsStaticSwitch=false
       bIsVisible=false
       type {classStructOrEnum {refPath=&NiagaraFloat}}
      }
      ScaleA {bIsDynamic=false, bIsEditable=true, bIsStaticSwitch=false, bIsVisible=true, type {classStructOrEnum {refPath=&NiagaraBool}}}
      Scale Alpha {bIsDynamic=false, bIsEditable=false, bIsStaticSwitch=false, bIsVisible=true, type {classStructOrEnum {refPath=&NiagaraFloat}}}
      Color Value To Scale {
       bIsDynamic=false
       bIsEditable=true
       bIsStaticSwitch=false
       bIsVisible=true
       type {classStructOrEnum {refPath=&LinearColor}}
      }
      Color Scale {
       bIsDynamic=false
       bIsEditable=false
       bIsStaticSwitch=false
       bIsVisible=false
       type {classStructOrEnum {refPath=&LinearColor}}
      }
     ]
     moduleScript {refPath=&ScaleColor}
    }
    {
     bIsSetParametersModule=false
     enabled=true
     moduleName=ScaleColor004
     inputs [
      Scale Mode {bIsDynamic=false, bIsEditable=true, bIsStaticSwitch=true, bIsVisible=true, type {classStructOrEnum {refPath=&NiagaraInt32}}}
      Linear Color Curve {
       bIsDynamic=false
       bIsEditable=false
       bIsStaticSwitch=false
       bIsVisible=false
       type {classStructOrEnum {refPath=&NiagaraDataInterfaceColorCurve}}
      }
      ScaleRGB {bIsDynamic=false, bIsEditable=true, bIsStaticSwitch=false, bIsVisible=true, type {classStructOrEnum {refPath=&NiagaraBool}}}
      Scale RGB {bIsDynamic=false, bIsEditable=false, bIsStaticSwitch=false, bIsVisible=true, type {classStructOrEnum {refPath=&Vector3f}}}
      ScaleRGBA {bIsDynamic=false, bIsEditable=false, bIsStaticSwitch=false, bIsVisible=false, type {classStructOrEnum {refPath=&NiagaraBool}}}
      Scale RGBA {
       bIsDynamic=false
       bIsEditable=false
       bIsStaticSwitch=false
       bIsVisible=false
       type {classStructOrEnum {refPath=&Vector4f}}
      }
      Curve Index {
       bIsDynamic=false
       bIsEditable=false
       bIsStaticSwitch=false
       bIsVisible=false
       type {classStructOrEnum {refPath=&NiagaraFloat}}
      }
      ScaleA {bIsDynamic=false, bIsEditable=true, bIsStaticSwitch=false, bIsVisible=true, type {classStructOrEnum {refPath=&NiagaraBool}}}
      Scale Alpha {bIsDynamic=true, bIsEditable=true, bIsStaticSwitch=false, bIsVisible=true, type {classStructOrEnum {refPath=&NiagaraFloat}}}
      Color Value To Scale {
       bIsDynamic=false
       bIsEditable=true
       bIsStaticSwitch=false
       bIsVisible=true
       type {classStructOrEnum {refPath=&LinearColor}}
      }
      Color Scale {
       bIsDynamic=false
       bIsEditable=false
       bIsStaticSwitch=false
       bIsVisible=false
       type {classStructOrEnum {refPath=&LinearColor}}
      }
     ]
     moduleScript {refPath=&ScaleColor}
    }
   ]
  }
  rendererClasses [{refPath=&NiagaraSpriteRendererProperties}]
  renderers [{rendererIndex=0, rendererClass {refPath=&NiagaraSpriteRendererProperties}}]
 }
 Snowflakedsff {
  emitterSpawnScript {scriptName=EmitterSpawnScript, modules []}
  emitterUpdateScript {
   scriptName=EmitterUpdateScript
   modules [
    {
     bIsSetParametersModule=false
     enabled=true
     moduleName=EmitterState
     inputs [
      Life Cycle Mode {
       bIsDynamic=false
       bIsEditable=true
       bIsStaticSwitch=true
       bIsVisible=true
       type {classStructOrEnum {refPath=&ENiagaraEmitterLifeCycleMode}}
      }
      Inactive Response {
       bIsDynamic=false
       bIsEditable=true
       bIsStaticSwitch=true
       bIsVisible=true
       type {classStructOrEnum {refPath=&ENiagaraInactiveMode}}
      }
      Loop Behavior {
       bIsDynamic=false
       bIsEditable=true
       bIsStaticSwitch=true
       bIsVisible=true
       type {classStructOrEnum {refPath=&ENiagara_EmitterStateOptions}}
      }
      Loop Duration Mode {
       bIsDynamic=false
       bIsEditable=true
       bIsStaticSwitch=true
       bIsVisible=true
       type {classStructOrEnum {refPath=&ENiagara_InfiniteLoopDuration}}
      }
      Loop Count {
       bIsDynamic=false
       bIsEditable=false
       bIsStaticSwitch=false
       bIsVisible=false
       type {classStructOrEnum {refPath=&NiagaraInt32}}
      }
      Loop Duration {bIsDynamic=false, bIsEditable=true, bIsStaticSwitch=false, bIsVisible=true, type {classStructOrEnum {refPath=&NiagaraFloat}}}
      Recalculate Duration Each Loop {bIsDynamic=false, bIsEditable=false, bIsStaticSwitch=false, bIsVisible=false, type {classStructOrEnum {refPath=&NiagaraBool}}}
      UseLoopDelay {bIsDynamic=false, bIsEditable=true, bIsStaticSwitch=false, bIsVisible=true, type {classStructOrEnum {refPath=&NiagaraBool}}}
      Loop Delay {bIsDynamic=false, bIsEditable=true, bIsStaticSwitch=false, bIsVisible=true, type {classStructOrEnum {refPath=&NiagaraFloat}}}
      Delay First Loop Only {bIsDynamic=false, bIsEditable=false, bIsStaticSwitch=false, bIsVisible=false, type {classStructOrEnum {refPath=&NiagaraBool}}}
      Scalability Mode {
       bIsDynamic=false
       bIsEditable=true
       bIsStaticSwitch=true
       bIsVisible=true
       type {classStructOrEnum {refPath=&ENiagaraEmitterScalabilityMode_Limited}}
      }
      Enable Distance Culling {bIsDynamic=false, bIsEditable=false, bIsStaticSwitch=true, bIsVisible=false, type {classStructOrEnum {refPath=&NiagaraBool}}}
      MinDistance {
       bIsDynamic=false
       bIsEditable=false
       bIsStaticSwitch=false
       bIsVisible=false
       type {classStructOrEnum {refPath=&NiagaraFloat}}
      }
      Min Distance Response {
       bIsDynamic=false
       bIsEditable=false
       bIsStaticSwitch=false
       bIsVisible=false
       type {classStructOrEnum {refPath=&ENiagaraExecutionStateManagement}}
      }
      MaxDistance {
       bIsDynamic=false
       bIsEditable=false
       bIsStaticSwitch=false
       bIsVisible=false
       type {classStructOrEnum {refPath=&NiagaraFloat}}
      }
      Max Distance Response {
       bIsDynamic=false
       bIsEditable=false
       bIsStaticSwitch=false
       bIsVisible=false
       type {classStructOrEnum {refPath=&ENiagaraExecutionStateManagement}}
      }
      Scale Spawn Count {bIsDynamic=false, bIsEditable=false, bIsStaticSwitch=true, bIsVisible=false, type {classStructOrEnum {refPath=&NiagaraBool}}}
      Spawn Count Scale {
       bIsDynamic=false
       bIsEditable=false
       bIsStaticSwitch=false
       bIsVisible=false
       type {classStructOrEnum {refPath=&NiagaraFloat}}
      }
      Spawn Count Scale By Distance {
       bIsDynamic=false
       bIsEditable=false
       bIsStaticSwitch=false
       bIsVisible=false
       type {classStructOrEnum {refPath=&NiagaraDataInterfaceCurve}}
      }
      Enable Visibility Culling {bIsDynamic=false, bIsEditable=false, bIsStaticSwitch=true, bIsVisible=false, type {classStructOrEnum {refPath=&NiagaraBool}}}
      Visibility Cull Response {
       bIsDynamic=false
       bIsEditable=false
       bIsStaticSwitch=false
       bIsVisible=false
       type {classStructOrEnum {refPath=&ENiagaraExecutionStateManagement}}
      }
      Visibility Cull Delay {
       bIsDynamic=false
       bIsEditable=false
       bIsStaticSwitch=false
       bIsVisible=false
       type {classStructOrEnum {refPath=&NiagaraFloat}}
      }
      Reset Age On Awaken {bIsDynamic=false, bIsEditable=false, bIsStaticSwitch=false, bIsVisible=false, type {classStructOrEnum {refPath=&NiagaraBool}}}
     ]
     moduleScript {refPath=&EmitterState}
    }
    {
     bIsSetParametersModule=false
     enabled=true
     moduleName=SpawnRate
     inputs [
      SpawnRate {bIsDynamic=false, bIsEditable=true, bIsStaticSwitch=false, bIsVisible=true, type {classStructOrEnum {refPath=&NiagaraFloat}}}
      Use Spawn Probability {bIsDynamic=false, bIsEditable=true, bIsStaticSwitch=false, bIsVisible=true, type {classStructOrEnum {refPath=&NiagaraBool}}}
      Spawn Probability {bIsDynamic=false, bIsEditable=false, bIsStaticSwitch=false, bIsVisible=true, type {classStructOrEnum {refPath=&NiagaraFloat}}}
      SpawnGroup {bIsDynamic=false, bIsEditable=true, bIsStaticSwitch=false, bIsVisible=true, type {classStructOrEnum {refPath=&NiagaraInt32}}}
     ]
     moduleScript {refPath=&SpawnRate}
    }
    {
     bIsSetParametersModule=false
     enabled=true
     moduleName=SpawnBurst_Instantaneous
     inputs [
      Spawn Count {bIsDynamic=false, bIsEditable=true, bIsStaticSwitch=false, bIsVisible=true, type {classStructOrEnum {refPath=&NiagaraInt32}}}
      Spawn Time {bIsDynamic=false, bIsEditable=true, bIsStaticSwitch=false, bIsVisible=true, type {classStructOrEnum {refPath=&NiagaraFloat}}}
      Use Spawn Probability {bIsDynamic=false, bIsEditable=true, bIsStaticSwitch=false, bIsVisible=true, type {classStructOrEnum {refPath=&NiagaraBool}}}
      Spawn Probability {bIsDynamic=false, bIsEditable=false, bIsStaticSwitch=false, bIsVisible=true, type {classStructOrEnum {refPath=&NiagaraFloat}}}
      UseLoopCountLimit {bIsDynamic=false, bIsEditable=true, bIsStaticSwitch=false, bIsVisible=true, type {classStructOrEnum {refPath=&NiagaraBool}}}
      Loop Count Limit {bIsDynamic=false, bIsEditable=false, bIsStaticSwitch=false, bIsVisible=true, type {classStructOrEnum {refPath=&NiagaraInt32}}}
      Spawn Group {bIsDynamic=false, bIsEditable=true, bIsStaticSwitch=false, bIsVisible=true, type {classStructOrEnum {refPath=&NiagaraInt32}}}
      Age {bIsDynamic=false, bIsEditable=true, bIsStaticSwitch=false, bIsVisible=true, type {classStructOrEnum {refPath=&NiagaraFloat}}}
     ]
     moduleScript {refPath=&SpawnBurst_Instantaneous}
    }
   ]
  }
  particleSpawnScript {
   scriptName=ParticleSpawnScript
   modules [
    {
     bIsSetParametersModule=false
     enabled=true
     moduleName=InitializeParticle001
     inputs [
      Lifetime Mode {
       bIsDynamic=false
       bIsEditable=true
       bIsStaticSwitch=true
       bIsVisible=true
       type {classStructOrEnum {refPath=&ENiagara_LifetimeMode}}
      }
      Lifetime Min {bIsDynamic=false, bIsEditable=true, bIsStaticSwitch=false, bIsVisible=true, type {classStructOrEnum {refPath=&NiagaraFloat}}}
      Lifetime {
       bIsDynamic=false
       bIsEditable=false
       bIsStaticSwitch=false
       bIsVisible=false
       type {classStructOrEnum {refPath=&NiagaraFloat}}
      }
      Lifetime Max {bIsDynamic=false, bIsEditable=true, bIsStaticSwitch=false, bIsVisible=true, type {classStructOrEnum {refPath=&NiagaraFloat}}}
      Lifetime Randomness Mode {
       bIsDynamic=false
       bIsEditable=true
       bIsStaticSwitch=true
       bIsVisible=true
       type {classStructOrEnum {refPath=&ENiagaraRandomnessMode}}
      }
      Color Mode {
       bIsDynamic=false
       bIsEditable=true
       bIsStaticSwitch=true
       bIsVisible=true
       type {classStructOrEnum {refPath=&ENiagara_ColorInitializationMode}}
      }
      Color Minimum {
       bIsDynamic=false
       bIsEditable=false
       bIsStaticSwitch=false
       bIsVisible=false
       type {classStructOrEnum {refPath=&LinearColor}}
      }
      Color {
       bIsDynamic=true
       bIsEditable=true
       bIsStaticSwitch=false
       bIsVisible=true
       type {classStructOrEnum {refPath=&LinearColor}}
      }
      AdjustHue {bIsDynamic=false, bIsEditable=false, bIsStaticSwitch=false, bIsVisible=false, type {classStructOrEnum {refPath=&NiagaraBool}}}
      Hue Shift Range {
       bIsDynamic=false
       bIsEditable=false
       bIsStaticSwitch=false
       bIsVisible=false
       type {classStructOrEnum {refPath=&Vector2f}}
      }
      Color Maximum {
       bIsDynamic=false
       bIsEditable=false
       bIsStaticSwitch=false
       bIsVisible=false
       type {classStructOrEnum {refPath=&LinearColor}}
      }
      AdjustSaturation {bIsDynamic=false, bIsEditable=false, bIsStaticSwitch=false, bIsVisible=false, type {classStructOrEnum {refPath=&NiagaraBool}}}
      Saturation Range {
       bIsDynamic=false
       bIsEditable=false
       bIsStaticSwitch=false
       bIsVisible=false
       type {classStructOrEnum {refPath=&Vector2f}}
      }
      AdjustValue {bIsDynamic=false, bIsEditable=false, bIsStaticSwitch=false, bIsVisible=false, type {classStructOrEnum {refPath=&NiagaraBool}}}
      Value Range {
       bIsDynamic=false
       bIsEditable=false
       bIsStaticSwitch=false
       bIsVisible=false
       type {classStructOrEnum {refPath=&Vector2f}}
      }
      AdjustAlpha {bIsDynamic=false, bIsEditable=false, bIsStaticSwitch=false, bIsVisible=false, type {classStructOrEnum {refPath=&NiagaraBool}}}
      Alpha Scale Range {
       bIsDynamic=false
       bIsEditable=false
       bIsStaticSwitch=false
       bIsVisible=false
       type {classStructOrEnum {refPath=&Vector2f}}
      }
      Color Channel Mode {
       bIsDynamic=false
       bIsEditable=false
       bIsStaticSwitch=true
       bIsVisible=false
       type {classStructOrEnum {refPath=&ENiagaraChannelCorrelation}}
      }
      Color Randomness Mode {
       bIsDynamic=false
       bIsEditable=false
       bIsStaticSwitch=true
       bIsVisible=false
       type {classStructOrEnum {refPath=&ENiagaraRandomnessMode}}
      }
      Position Mode {
       bIsDynamic=false
       bIsEditable=true
       bIsStaticSwitch=true
       bIsVisible=true
       type {classStructOrEnum {refPath=&ENiagara_PositionInitializationMode}}
      }
      Position {
       bIsDynamic=false
       bIsEditable=false
       bIsStaticSwitch=false
       bIsVisible=false
       type {classStructOrEnum {refPath=&NiagaraPosition}}
      }
      UsePositionOffset {bIsDynamic=false, bIsEditable=true, bIsStaticSwitch=false, bIsVisible=true, type {classStructOrEnum {refPath=&NiagaraBool}}}
      Position Offset {bIsDynamic=false, bIsEditable=true, bIsStaticSwitch=false, bIsVisible=true, type {classStructOrEnum {refPath=&Vector3f}}}
      Position Offset Coordinate Space {
       bIsDynamic=false
       bIsEditable=true
       bIsStaticSwitch=false
       bIsVisible=true
       type {classStructOrEnum {refPath=&ENiagaraCoordinateSpace}}
      }
      Mass Mode {
       bIsDynamic=false
       bIsEditable=true
       bIsStaticSwitch=true
       bIsVisible=true
       type {classStructOrEnum {refPath=&ENiagara_MassInitializationMode}}
      }
      Mass Min {
       bIsDynamic=false
       bIsEditable=false
       bIsStaticSwitch=false
       bIsVisible=false
       type {classStructOrEnum {refPath=&NiagaraFloat}}
      }
      Mass {
       bIsDynamic=false
       bIsEditable=false
       bIsStaticSwitch=false
       bIsVisible=false
       type {classStructOrEnum {refPath=&NiagaraFloat}}
      }
      Mass Max {
       bIsDynamic=false
       bIsEditable=false
       bIsStaticSwitch=false
       bIsVisible=false
       type {classStructOrEnum {refPath=&NiagaraFloat}}
      }
      Mass Randomness Mode {
       bIsDynamic=false
       bIsEditable=false
       bIsStaticSwitch=true
       bIsVisible=false
       type {classStructOrEnum {refPath=&ENiagaraRandomnessMode}}
      }
      Write Material Random {bIsDynamic=false, bIsEditable=true, bIsStaticSwitch=false, bIsVisible=true, type {classStructOrEnum {refPath=&NiagaraBool}}}
      Material Random {bIsDynamic=false, bIsEditable=true, bIsStaticSwitch=false, bIsVisible=true, type {classStructOrEnum {refPath=&NiagaraFloat}}}
      Sprite Size Mode {
       bIsDynamic=false
       bIsEditable=true
       bIsStaticSwitch=true
       bIsVisible=true
       type {classStructOrEnum {refPath=&ENiagara_SizeScaleMode}}
      }
      Sprite Size {
       bIsDynamic=false
       bIsEditable=false
       bIsStaticSwitch=false
       bIsVisible=false
       type {classStructOrEnum {refPath=&Vector2f}}
      }
      Uniform Sprite Size {bIsDynamic=true, bIsEditable=true, bIsStaticSwitch=false, bIsVisible=true, type {classStructOrEnum {refPath=&NiagaraFloat}}}
      Uniform Sprite Size Min {
       bIsDynamic=false
       bIsEditable=false
       bIsStaticSwitch=false
       bIsVisible=false
       type {classStructOrEnum {refPath=&NiagaraFloat}}
      }
      Sprite Size Min {
       bIsDynamic=false
       bIsEditable=false
       bIsStaticSwitch=false
       bIsVisible=false
       type {classStructOrEnum {refPath=&Vector2f}}
      }
      Sprite Size Max {
       bIsDynamic=false
       bIsEditable=false
       bIsStaticSwitch=false
       bIsVisible=false
       type {classStructOrEnum {refPath=&Vector2f}}
      }
      Uniform Sprite Size Max {
       bIsDynamic=false
       bIsEditable=false
       bIsStaticSwitch=false
       bIsVisible=false
       type {classStructOrEnum {refPath=&NiagaraFloat}}
      }
      Sprite Size Randomness Mode {
       bIsDynamic=false
       bIsEditable=false
       bIsStaticSwitch=true
       bIsVisible=false
       type {classStructOrEnum {refPath=&ENiagaraRandomnessMode}}
      }
      Sprite Rotation Mode {
       bIsDynamic=false
       bIsEditable=true
       bIsStaticSwitch=true
       bIsVisible=true
       type {classStructOrEnum {refPath=&ENiagara_SpriteRotationMode}}
      }
      Sprite Rotation Angle Min {
       bIsDynamic=false
       bIsEditable=false
       bIsStaticSwitch=false
       bIsVisible=false
       type {classStructOrEnum {refPath=&NiagaraFloat}}
      }
      Sprite Rotation Angle {
       bIsDynamic=false
       bIsEditable=false
       bIsStaticSwitch=false
       bIsVisible=false
       type {classStructOrEnum {refPath=&NiagaraFloat}}
      }
      Sprite Rotation Angle Max {
       bIsDynamic=false
       bIsEditable=false
       bIsStaticSwitch=false
       bIsVisible=false
       type {classStructOrEnum {refPath=&NiagaraFloat}}
      }
      Sprite Rotation Randomness Mode {
       bIsDynamic=false
       bIsEditable=false
       bIsStaticSwitch=true
       bIsVisible=false
       type {classStructOrEnum {refPath=&ENiagaraRandomnessMode}}
      }
      Sprite UV Mode {
       bIsDynamic=false
       bIsEditable=true
       bIsStaticSwitch=true
       bIsVisible=true
       type {classStructOrEnum {refPath=&ENiagara_UVFlippingMode}}
      }
      Sprite UV Randomness Mode {
       bIsDynamic=false
       bIsEditable=false
       bIsStaticSwitch=true
       bIsVisible=false
       type {classStructOrEnum {refPath=&ENiagaraRandomnessMode}}
      }
      Sprite UV Scale {
       bIsDynamic=false
       bIsEditable=false
       bIsStaticSwitch=false
       bIsVisible=false
       type {classStructOrEnum {refPath=&Vector2f}}
      }
      Mesh Scale Mode {
       bIsDynamic=false
       bIsEditable=true
       bIsStaticSwitch=true
       bIsVisible=true
       type {classStructOrEnum {refPath=&ENiagara_SizeScaleMode}}
      }
      Mesh Uniform Scale Min {
       bIsDynamic=false
       bIsEditable=false
       bIsStaticSwitch=false
       bIsVisible=false
       type {classStructOrEnum {refPath=&NiagaraFloat}}
      }
      Mesh Uniform Scale {
       bIsDynamic=false
       bIsEditable=false
       bIsStaticSwitch=false
       bIsVisible=false
       type {classStructOrEnum {refPath=&NiagaraFloat}}
      }
      Mesh Scale Min {
       bIsDynamic=false
       bIsEditable=false
       bIsStaticSwitch=false
       bIsVisible=false
       type {classStructOrEnum {refPath=&Vector3f}}
      }
      Mesh Scale {
       bIsDynamic=false
       bIsEditable=false
       bIsStaticSwitch=false
       bIsVisible=false
       type {classStructOrEnum {refPath=&Vector3f}}
      }
      Mesh Scale Max {
       bIsDynamic=false
       bIsEditable=false
       bIsStaticSwitch=false
       bIsVisible=false
       type {classStructOrEnum {refPath=&Vector3f}}
      }
      Mesh Uniform Scale Max {
       bIsDynamic=false
       bIsEditable=false
       bIsStaticSwitch=false
       bIsVisible=false
       type {classStructOrEnum {refPath=&NiagaraFloat}}
      }
      Mesh Scale Randomness Mode {
       bIsDynamic=false
       bIsEditable=false
       bIsStaticSwitch=true
       bIsVisible=false
       type {classStructOrEnum {refPath=&ENiagaraRandomnessMode}}
      }
      Mesh Renderer Array Visibility Mode {
       bIsDynamic=false
       bIsEditable=true
       bIsStaticSwitch=true
       bIsVisible=true
       type {classStructOrEnum {refPath=&ENiagara_UnsetDirectSetRandom}}
      }
      Mesh Renderer Mesh Index {
       bIsDynamic=false
       bIsEditable=false
       bIsStaticSwitch=false
       bIsVisible=false
       type {classStructOrEnum {refPath=&NiagaraInt32}}
      }
      Wrap Index By Mesh Count {bIsDynamic=false, bIsEditable=false, bIsStaticSwitch=true, bIsVisible=false, type {classStructOrEnum {refPath=&NiagaraBool}}}
      Mesh Renderer Info {
       bIsDynamic=false
       bIsEditable=false
       bIsStaticSwitch=false
       bIsVisible=false
       type {classStructOrEnum {refPath=&NiagaraDataInterfaceMeshRendererInfo}}
      }
      Mesh Renderer Random Seed {
       bIsDynamic=false
       bIsEditable=false
       bIsStaticSwitch=false
       bIsVisible=false
       type {classStructOrEnum {refPath=&NiagaraInt32}}
      }
      Ribbon Width Mode {
       bIsDynamic=false
       bIsEditable=true
       bIsStaticSwitch=true
       bIsVisible=true
       type {classStructOrEnum {refPath=&ENiagara_UnsetDirectSet}}
      }
      Ribbon Width {
       bIsDynamic=false
       bIsEditable=false
       bIsStaticSwitch=false
       bIsVisible=false
       type {classStructOrEnum {refPath=&NiagaraFloat}}
      }
      Ribbon Facing Mode {
       bIsDynamic=false
       bIsEditable=true
       bIsStaticSwitch=true
       bIsVisible=true
       type {classStructOrEnum {refPath=&ENiagara_UnsetDirectSet}}
      }
      Ribbon Facing Vector {
       bIsDynamic=false
       bIsEditable=false
       bIsStaticSwitch=false
       bIsVisible=false
       type {classStructOrEnum {refPath=&Vector3f}}
      }
      Ribbon Twist Mode {
       bIsDynamic=false
       bIsEditable=true
       bIsStaticSwitch=true
       bIsVisible=true
       type {classStructOrEnum {refPath=&ENiagara_UnsetDirectSet}}
      }
      "Ribbon Twist " {
       bIsDynamic=false
       bIsEditable=false
       bIsStaticSwitch=false
       bIsVisible=false
       type {classStructOrEnum {refPath=&NiagaraFloat}}
      }
     ]
     moduleScript {refPath=&InitializeParticle}
    }
    {
     bIsSetParametersModule=false
     enabled=true
     moduleName=SphereLocation
     inputs [
      Sphere Radius {bIsDynamic=false, bIsEditable=true, bIsStaticSwitch=false, bIsVisible=true, type {classStructOrEnum {refPath=&NiagaraFloat}}}
      UseOffset {bIsDynamic=false, bIsEditable=true, bIsStaticSwitch=false, bIsVisible=true, type {classStructOrEnum {refPath=&NiagaraBool}}}
      Offset {bIsDynamic=false, bIsEditable=true, bIsStaticSwitch=false, bIsVisible=true, type {classStructOrEnum {refPath=&Vector3f}}}
      Offset Coordinate Space {
       bIsDynamic=false
       bIsEditable=true
       bIsStaticSwitch=false
       bIsVisible=true
       type {classStructOrEnum {refPath=&ENiagaraCoordinateSpace}}
      }
      UseNonUniformScale {bIsDynamic=false, bIsEditable=true, bIsStaticSwitch=false, bIsVisible=true, type {classStructOrEnum {refPath=&NiagaraBool}}}
      Non Uniform Scale {bIsDynamic=false, bIsEditable=false, bIsStaticSwitch=false, bIsVisible=true, type {classStructOrEnum {refPath=&Vector3f}}}
      UseSphereOrientation {bIsDynamic=false, bIsEditable=true, bIsStaticSwitch=false, bIsVisible=true, type {classStructOrEnum {refPath=&NiagaraBool}}}
      Sphere Orientation Axis {bIsDynamic=false, bIsEditable=false, bIsStaticSwitch=false, bIsVisible=true, type {classStructOrEnum {refPath=&Vector3f}}}
      Sphere Origin {
       bIsDynamic=false
       bIsEditable=true
       bIsStaticSwitch=false
       bIsVisible=true
       type {classStructOrEnum {refPath=&NiagaraPosition}}
      }
      Sphere Coordinate Space {
       bIsDynamic=false
       bIsEditable=true
       bIsStaticSwitch=false
       bIsVisible=true
       type {classStructOrEnum {refPath=&ENiagaraCoordinateSpace}}
      }
      Write to Intrinsic Variables {bIsDynamic=false, bIsEditable=true, bIsStaticSwitch=true, bIsVisible=true, type {classStructOrEnum {refPath=&NiagaraBool}}}
      Sphere Distribution {
       bIsDynamic=false
       bIsEditable=true
       bIsStaticSwitch=true
       bIsVisible=true
       type {classStructOrEnum {refPath=&ENiagaraSphereDistributionMode}}
      }
      Evaluation Type {
       bIsDynamic=false
       bIsEditable=false
       bIsStaticSwitch=true
       bIsVisible=false
       type {classStructOrEnum {refPath=&ENiagaraRandomnessEvaluation}}
      }
      Hemisphere X {bIsDynamic=false, bIsEditable=true, bIsStaticSwitch=false, bIsVisible=true, type {classStructOrEnum {refPath=&NiagaraBool}}}
      Uniform Distribution {
       bIsDynamic=false
       bIsEditable=false
       bIsStaticSwitch=false
       bIsVisible=false
       type {classStructOrEnum {refPath=&NiagaraFloat}}
      }
      Hemisphere Y {bIsDynamic=false, bIsEditable=true, bIsStaticSwitch=false, bIsVisible=true, type {classStructOrEnum {refPath=&NiagaraBool}}}
      U Position {
       bIsDynamic=false
       bIsEditable=false
       bIsStaticSwitch=false
       bIsVisible=false
       type {classStructOrEnum {refPath=&NiagaraFloat}}
      }
      Uniform Spiral Amount {
       bIsDynamic=false
       bIsEditable=false
       bIsStaticSwitch=false
       bIsVisible=false
       type {classStructOrEnum {refPath=&NiagaraFloat}}
      }
      Hemisphere Z {bIsDynamic=false, bIsEditable=true, bIsStaticSwitch=false, bIsVisible=true, type {classStructOrEnum {refPath=&NiagaraBool}}}
      Uniform Endpoint Offset {
       bIsDynamic=false
       bIsEditable=false
       bIsStaticSwitch=false
       bIsVisible=false
       type {classStructOrEnum {refPath=&NiagaraFloat}}
      }
      V Position {
       bIsDynamic=false
       bIsEditable=false
       bIsStaticSwitch=false
       bIsVisible=false
       type {classStructOrEnum {refPath=&NiagaraFloat}}
      }
      Radius Position {
       bIsDynamic=false
       bIsEditable=false
       bIsStaticSwitch=false
       bIsVisible=false
       type {classStructOrEnum {refPath=&NiagaraFloat}}
      }
      Surface Only {bIsDynamic=false, bIsEditable=true, bIsStaticSwitch=false, bIsVisible=true, type {classStructOrEnum {refPath=&NiagaraBool}}}
      Surface Only Band Thickness {bIsDynamic=false, bIsEditable=true, bIsStaticSwitch=false, bIsVisible=true, type {classStructOrEnum {refPath=&NiagaraFloat}}}
      Surface Expansion Mode {
       bIsDynamic=false
       bIsEditable=true
       bIsStaticSwitch=false
       bIsVisible=true
       type {classStructOrEnum {refPath=&ENiagaraExpansionMode}}
      }
      Normalize Thickness {bIsDynamic=false, bIsEditable=true, bIsStaticSwitch=false, bIsVisible=true, type {classStructOrEnum {refPath=&NiagaraBool}}}
      Randomness Mode {
       bIsDynamic=false
       bIsEditable=true
       bIsStaticSwitch=true
       bIsVisible=true
       type {classStructOrEnum {refPath=&ENiagaraRandomnessMode}}
      }
      Override Seed {bIsDynamic=false, bIsEditable=true, bIsStaticSwitch=false, bIsVisible=true, type {classStructOrEnum {refPath=&NiagaraBool}}}
      Random Seed {bIsDynamic=false, bIsEditable=false, bIsStaticSwitch=false, bIsVisible=true, type {classStructOrEnum {refPath=&NiagaraInt32}}}
      Fixed Random Seed {bIsDynamic=false, bIsEditable=false, bIsStaticSwitch=true, bIsVisible=false, type {classStructOrEnum {refPath=&NiagaraBool}}}
      MaskSpawnGroup {bIsDynamic=false, bIsEditable=true, bIsStaticSwitch=false, bIsVisible=true, type {classStructOrEnum {refPath=&NiagaraBool}}}
      Spawn Group Mask {bIsDynamic=false, bIsEditable=false, bIsStaticSwitch=false, bIsVisible=true, type {classStructOrEnum {refPath=&NiagaraInt32}}}
     ]
     moduleScript {refPath=&SphereLocation}
    }
    {
     bIsSetParametersModule=false
     enabled=false
     moduleName=AddVelocityFromPoint
     inputs [
      Velocity Strength {bIsDynamic=true, bIsEditable=true, bIsStaticSwitch=false, bIsVisible=true, type {classStructOrEnum {refPath=&NiagaraFloat}}}
      Origin Offset {bIsDynamic=false, bIsEditable=true, bIsStaticSwitch=false, bIsVisible=true, type {classStructOrEnum {refPath=&Vector3f}}}
      Origin Offset Coordinate Space {
       bIsDynamic=false
       bIsEditable=true
       bIsStaticSwitch=false
       bIsVisible=true
       type {classStructOrEnum {refPath=&ENiagaraCoordinateSpace}}
      }
      Default Position {bIsDynamic=false, bIsEditable=true, bIsStaticSwitch=false, bIsVisible=true, type {classStructOrEnum {refPath=&Vector3f}}}
      Module.Use Falloff Distance {bIsDynamic=false, bIsEditable=true, bIsStaticSwitch=false, bIsVisible=true, type {classStructOrEnum {refPath=&NiagaraBool}}}
      Velocity Falloff Distance {bIsDynamic=false, bIsEditable=false, bIsStaticSwitch=false, bIsVisible=true, type {classStructOrEnum {refPath=&NiagaraFloat}}}
      Invert Velocity Falloff {bIsDynamic=false, bIsEditable=false, bIsStaticSwitch=false, bIsVisible=false, type {classStructOrEnum {refPath=&NiagaraBool}}}
      Velocity Origin {bIsDynamic=false, bIsEditable=true, bIsStaticSwitch=false, bIsVisible=true, type {classStructOrEnum {refPath=&Vector3f}}}
     ]
     moduleScript {refPath=&AddVelocityFromPoint}
    }
    {
     bIsSetParametersModule=false
     enabled=true
     moduleName=AddVelocity
     inputs [
      Velocity Mode {
       bIsDynamic=false
       bIsEditable=true
       bIsStaticSwitch=true
       bIsVisible=true
       type {classStructOrEnum {refPath=&ENiagara_VelocityMode}}
      }
      Velocity {
       bIsDynamic=false
       bIsEditable=false
       bIsStaticSwitch=false
       bIsVisible=false
       type {classStructOrEnum {refPath=&Vector3f}}
      }
      Velocity Speed {bIsDynamic=true, bIsEditable=true, bIsStaticSwitch=false, bIsVisible=true, type {classStructOrEnum {refPath=&NiagaraFloat}}}
      Velocity Speed Scale {
       bIsDynamic=false
       bIsEditable=false
       bIsStaticSwitch=false
       bIsVisible=false
       type {classStructOrEnum {refPath=&NiagaraFloat}}
      }
      Constrain To Radius {
       bIsDynamic=false
       bIsEditable=false
       bIsStaticSwitch=false
       bIsVisible=false
       type {classStructOrEnum {refPath=&NiagaraFloat}}
      }
      Module.UseVelDistribution {bIsDynamic=false, bIsEditable=true, bIsStaticSwitch=false, bIsVisible=true, type {classStructOrEnum {refPath=&NiagaraBool}}}
      Distribution Along Cone Axis {bIsDynamic=false, bIsEditable=false, bIsStaticSwitch=false, bIsVisible=true, type {classStructOrEnum {refPath=&NiagaraFloat}}}
      Radius Falloff Near / Far {
       bIsDynamic=false
       bIsEditable=false
       bIsStaticSwitch=false
       bIsVisible=false
       type {classStructOrEnum {refPath=&Vector2f}}
      }
      Radius Falloff Exponent {
       bIsDynamic=false
       bIsEditable=false
       bIsStaticSwitch=false
       bIsVisible=false
       type {classStructOrEnum {refPath=&NiagaraFloat}}
      }
      Module.UseConeFalloff {bIsDynamic=false, bIsEditable=true, bIsStaticSwitch=false, bIsVisible=true, type {classStructOrEnum {refPath=&NiagaraBool}}}
      Speed Falloff From Cone Axis {bIsDynamic=false, bIsEditable=false, bIsStaticSwitch=false, bIsVisible=true, type {classStructOrEnum {refPath=&NiagaraFloat}}}
      Invert Velocity Falloff {bIsDynamic=false, bIsEditable=false, bIsStaticSwitch=false, bIsVisible=false, type {classStructOrEnum {refPath=&NiagaraBool}}}
      Override Seed {bIsDynamic=false, bIsEditable=true, bIsStaticSwitch=false, bIsVisible=true, type {classStructOrEnum {refPath=&NiagaraBool}}}
      Random Seed {bIsDynamic=false, bIsEditable=false, bIsStaticSwitch=false, bIsVisible=true, type {classStructOrEnum {refPath=&NiagaraInt32}}}
      Fixed Random Seed {bIsDynamic=false, bIsEditable=false, bIsStaticSwitch=true, bIsVisible=false, type {classStructOrEnum {refPath=&NiagaraBool}}}
      Randomness Mode {
       bIsDynamic=false
       bIsEditable=true
       bIsStaticSwitch=true
       bIsVisible=true
       type {classStructOrEnum {refPath=&ENiagaraRandomnessMode}}
      }
      Cone Axis {bIsDynamic=false, bIsEditable=true, bIsStaticSwitch=false, bIsVisible=true, type {classStructOrEnum {refPath=&Vector3f}}}
      Cone Angle {bIsDynamic=false, bIsEditable=true, bIsStaticSwitch=false, bIsVisible=true, type {classStructOrEnum {refPath=&NiagaraFloat}}}
      Inner Cone Angle {bIsDynamic=false, bIsEditable=true, bIsStaticSwitch=false, bIsVisible=true, type {classStructOrEnum {refPath=&NiagaraFloat}}}
      Cone Angle Mode {
       bIsDynamic=false
       bIsEditable=true
       bIsStaticSwitch=true
       bIsVisible=true
       type {classStructOrEnum {refPath=&ENiagara_AngleInput}}
      }
      Rotation Mode {
       bIsDynamic=false
       bIsEditable=true
       bIsStaticSwitch=true
       bIsVisible=true
       type {classStructOrEnum {refPath=&ENiagara_RotationMode}}
      }
      Rotation Axis {
       bIsDynamic=false
       bIsEditable=false
       bIsStaticSwitch=false
       bIsVisible=false
       type {classStructOrEnum {refPath=&Vector3f}}
      }
      Yaw / Pitch / Roll {
       bIsDynamic=false
       bIsEditable=false
       bIsStaticSwitch=false
       bIsVisible=false
       type {classStructOrEnum {refPath=&Vector3f}}
      }
      Rotation Quaternion {bIsDynamic=false, bIsEditable=false, bIsStaticSwitch=false, bIsVisible=false, type {classStructOrEnum {refPath=&Quat4f}}}
      Rotation Matrix {
       bIsDynamic=false
       bIsEditable=false
       bIsStaticSwitch=false
       bIsVisible=false
       type {classStructOrEnum {refPath=&NiagaraMatrix}}
      }
      Rotation Angle {
       bIsDynamic=false
       bIsEditable=false
       bIsStaticSwitch=false
       bIsVisible=false
       type {classStructOrEnum {refPath=&NiagaraFloat}}
      }
      Invert Rotation Quaternion {bIsDynamic=false, bIsEditable=false, bIsStaticSwitch=false, bIsVisible=false, type {classStructOrEnum {refPath=&NiagaraBool}}}
      Additional Quaternion Rotation {bIsDynamic=false, bIsEditable=false, bIsStaticSwitch=false, bIsVisible=false, type {classStructOrEnum {refPath=&NiagaraBool}}}
      Additional Yaw / Pitch / Roll {
       bIsDynamic=false
       bIsEditable=false
       bIsStaticSwitch=false
       bIsVisible=false
       type {classStructOrEnum {refPath=&Vector3f}}
      }
      Rotation Angle Type {
       bIsDynamic=false
       bIsEditable=false
       bIsStaticSwitch=true
       bIsVisible=false
       type {classStructOrEnum {refPath=&ENiagara_AngleInput}}
      }
      Rotation Coordinate Space {
       bIsDynamic=false
       bIsEditable=true
       bIsStaticSwitch=false
       bIsVisible=true
       type {classStructOrEnum {refPath=&ENiagaraCoordinateSpace}}
      }
      Velocity Origin {
       bIsDynamic=false
       bIsEditable=false
       bIsStaticSwitch=false
       bIsVisible=false
       type {classStructOrEnum {refPath=&NiagaraPosition}}
      }
      Origin Offset {
       bIsDynamic=false
       bIsEditable=false
       bIsStaticSwitch=false
       bIsVisible=false
       type {classStructOrEnum {refPath=&Vector3f}}
      }
      Origin Offset Coordinate Space {
       bIsDynamic=false
       bIsEditable=false
       bIsStaticSwitch=false
       bIsVisible=false
       type {classStructOrEnum {refPath=&ENiagaraCoordinateSpace}}
      }
      Default Position {
       bIsDynamic=false
       bIsEditable=false
       bIsStaticSwitch=false
       bIsVisible=false
       type {classStructOrEnum {refPath=&NiagaraPosition}}
      }
     ]
     moduleScript {refPath=&AddVelocity}
    }
   ]
  }
  particleUpdateScript {
   scriptName=ParticleUpdateScript
   modules [
    {
     bIsSetParametersModule=false
     enabled=true
     moduleName=ParticleState
     inputs [
      Kill Particles When Lifetime Has Elapsed {bIsDynamic=false, bIsEditable=true, bIsStaticSwitch=true, bIsVisible=true, type {classStructOrEnum {refPath=&NiagaraBool}}}
      Loop Particles Lifetime {bIsDynamic=false, bIsEditable=false, bIsStaticSwitch=false, bIsVisible=false, type {classStructOrEnum {refPath=&NiagaraBool}}}
      Let Infinitely Lived Particles Die When Emitter Deactivates {bIsDynamic=false, bIsEditable=false, bIsStaticSwitch=false, bIsVisible=false, type {classStructOrEnum {refPath=&NiagaraBool}}}
      Lifetime {bIsDynamic=false, bIsEditable=true, bIsStaticSwitch=false, bIsVisible=true, type {classStructOrEnum {refPath=&NiagaraFloat}}}
      DeltaTime {bIsDynamic=false, bIsEditable=true, bIsStaticSwitch=false, bIsVisible=true, type {classStructOrEnum {refPath=&NiagaraFloat}}}
     ]
     moduleScript {refPath=&ParticleState}
    }
    {
     bIsSetParametersModule=false
     enabled=true
     moduleName=CurlNoiseForce
     inputs [
      Noise Strength {bIsDynamic=false, bIsEditable=true, bIsStaticSwitch=false, bIsVisible=true, type {classStructOrEnum {refPath=&NiagaraFloat}}}
      Noise Frequency {bIsDynamic=false, bIsEditable=true, bIsStaticSwitch=false, bIsVisible=true, type {classStructOrEnum {refPath=&NiagaraFloat}}}
      Noise Quality / Cost {
       bIsDynamic=false
       bIsEditable=true
       bIsStaticSwitch=true
       bIsVisible=true
       type {classStructOrEnum {refPath=/Niagara/Enums/ENiagaraCurlNoiseQuality}}
      }
      Pan Noise {bIsDynamic=false, bIsEditable=true, bIsStaticSwitch=false, bIsVisible=true, type {classStructOrEnum {refPath=&NiagaraBool}}}
      Pan Noise Field {bIsDynamic=false, bIsEditable=false, bIsStaticSwitch=false, bIsVisible=true, type {classStructOrEnum {refPath=&Vector3f}}}
      Write to Intrinsic Parameters {bIsDynamic=false, bIsEditable=true, bIsStaticSwitch=true, bIsVisible=true, type {classStructOrEnum {refPath=&NiagaraBool}}}
      Sample Position {bIsDynamic=false, bIsEditable=true, bIsStaticSwitch=false, bIsVisible=true, type {classStructOrEnum {refPath=&Vector3f}}}
      Randomize Noise Sample {bIsDynamic=false, bIsEditable=true, bIsStaticSwitch=true, bIsVisible=true, type {classStructOrEnum {refPath=&NiagaraBool}}}
      Random Seed {bIsDynamic=false, bIsEditable=true, bIsStaticSwitch=false, bIsVisible=true, type {classStructOrEnum {refPath=&NiagaraInt32}}}
      Randomization Vector {bIsDynamic=false, bIsEditable=true, bIsStaticSwitch=false, bIsVisible=true, type {classStructOrEnum {refPath=&Vector3f}}}
      Scale Curl Lookup By Engine Owner Scale {bIsDynamic=false, bIsEditable=true, bIsStaticSwitch=true, bIsVisible=true, type {classStructOrEnum {refPath=&NiagaraBool}}}
      CurlNoise {
       bIsDynamic=false
       bIsEditable=false
       bIsStaticSwitch=false
       bIsVisible=false
       type {classStructOrEnum {refPath=&NiagaraDataInterfaceCurlNoise}}
      }
      Mask Curl Noise {bIsDynamic=false, bIsEditable=true, bIsStaticSwitch=true, bIsVisible=true, type {classStructOrEnum {refPath=&NiagaraBool}}}
      Curl Noise Cone Mask Angle {
       bIsDynamic=false
       bIsEditable=false
       bIsStaticSwitch=false
       bIsVisible=false
       type {classStructOrEnum {refPath=&NiagaraFloat}}
      }
      Curl Noise Cone Mask Falloff Angle {
       bIsDynamic=false
       bIsEditable=false
       bIsStaticSwitch=false
       bIsVisible=false
       type {classStructOrEnum {refPath=&NiagaraFloat}}
      }
      Curl Noise Cone Mask Axis {
       bIsDynamic=false
       bIsEditable=false
       bIsStaticSwitch=false
       bIsVisible=false
       type {classStructOrEnum {refPath=&Vector3f}}
      }
     ]
     moduleScript {refPath=&CurlNoiseForce}
    }
    {
     bIsSetParametersModule=false
     enabled=true
     moduleName=Drag
     inputs [
      Use Linear Drag {bIsDynamic=false, bIsEditable=true, bIsStaticSwitch=false, bIsVisible=true, type {classStructOrEnum {refPath=&NiagaraBool}}}
      Drag {bIsDynamic=true, bIsEditable=true, bIsStaticSwitch=false, bIsVisible=true, type {classStructOrEnum {refPath=&NiagaraFloat}}}
      Use Rotational Drag {bIsDynamic=false, bIsEditable=true, bIsStaticSwitch=false, bIsVisible=true, type {classStructOrEnum {refPath=&NiagaraBool}}}
      Rotational Drag {bIsDynamic=false, bIsEditable=false, bIsStaticSwitch=false, bIsVisible=true, type {classStructOrEnum {refPath=&NiagaraFloat}}}
     ]
     moduleScript {refPath=&Drag}
    }
    {
     bIsSetParametersModule=false
     enabled=true
     moduleName=GravityForce
     inputs [
      Gravity {bIsDynamic=false, bIsEditable=true, bIsStaticSwitch=false, bIsVisible=true, type {classStructOrEnum {refPath=&Vector3f}}}
      Coordinate Space {
       bIsDynamic=false
       bIsEditable=true
       bIsStaticSwitch=true
       bIsVisible=true
       type {classStructOrEnum {refPath=&ENiagaraCoordinateSpace}}
      }
     ]
     moduleScript {refPath=&GravityForce}
    }
    {
     bIsSetParametersModule=false
     enabled=true
     moduleName=SolveForcesAndVelocity
     inputs [
      Clamp Velocity {bIsDynamic=false, bIsEditable=true, bIsStaticSwitch=false, bIsVisible=true, type {classStructOrEnum {refPath=&NiagaraBool}}}
      Speed Limit {bIsDynamic=false, bIsEditable=false, bIsStaticSwitch=false, bIsVisible=true, type {classStructOrEnum {refPath=&NiagaraFloat}}}
      Limit Acceleration {bIsDynamic=false, bIsEditable=true, bIsStaticSwitch=false, bIsVisible=true, type {classStructOrEnum {refPath=&NiagaraBool}}}
      Acceleration Limit {bIsDynamic=false, bIsEditable=false, bIsStaticSwitch=false, bIsVisible=true, type {classStructOrEnum {refPath=&NiagaraFloat}}}
      Manually Enable Rotational Solver {bIsDynamic=false, bIsEditable=true, bIsStaticSwitch=false, bIsVisible=true, type {classStructOrEnum {refPath=&NiagaraBool}}}
      Write to Intrinsic Properties {bIsDynamic=false, bIsEditable=true, bIsStaticSwitch=false, bIsVisible=true, type {classStructOrEnum {refPath=&NiagaraBool}}}
      Write to Presolve Properties {bIsDynamic=false, bIsEditable=true, bIsStaticSwitch=true, bIsVisible=true, type {classStructOrEnum {refPath=&NiagaraBool}}}
      DeltaTime {bIsDynamic=false, bIsEditable=true, bIsStaticSwitch=false, bIsVisible=true, type {classStructOrEnum {refPath=&NiagaraFloat}}}
      Mass {bIsDynamic=false, bIsEditable=true, bIsStaticSwitch=false, bIsVisible=true, type {classStructOrEnum {refPath=&NiagaraFloat}}}
      Position {
       bIsDynamic=false
       bIsEditable=true
       bIsStaticSwitch=false
       bIsVisible=true
       type {classStructOrEnum {refPath=&NiagaraPosition}}
      }
      Velocity {bIsDynamic=false, bIsEditable=true, bIsStaticSwitch=false, bIsVisible=true, type {classStructOrEnum {refPath=&Vector3f}}}
      PreviousVelocity {bIsDynamic=false, bIsEditable=true, bIsStaticSwitch=false, bIsVisible=true, type {classStructOrEnum {refPath=&Vector3f}}}
      Force {bIsDynamic=false, bIsEditable=true, bIsStaticSwitch=false, bIsVisible=true, type {classStructOrEnum {refPath=&Vector3f}}}
      Mesh Orientation {bIsDynamic=false, bIsEditable=false, bIsStaticSwitch=false, bIsVisible=false, type {classStructOrEnum {refPath=&Quat4f}}}
      Rotational Velocity {
       bIsDynamic=false
       bIsEditable=false
       bIsStaticSwitch=false
       bIsVisible=false
       type {classStructOrEnum {refPath=&Vector3f}}
      }
      Rotational Force {
       bIsDynamic=false
       bIsEditable=false
       bIsStaticSwitch=false
       bIsVisible=false
       type {classStructOrEnum {refPath=&Vector3f}}
      }
      Rotational Inertia {
       bIsDynamic=false
       bIsEditable=false
       bIsStaticSwitch=false
       bIsVisible=false
       type {classStructOrEnum {refPath=&NiagaraFloat}}
      }
     ]
     moduleScript {refPath=&SolveForcesAndVelocity}
    }
    {
     bIsSetParametersModule=false
     enabled=true
     moduleName=SpriteRotationRate
     inputs [
      Rotation Rate {bIsDynamic=true, bIsEditable=true, bIsStaticSwitch=false, bIsVisible=true, type {classStructOrEnum {refPath=&NiagaraFloat}}}
      Delta Time {bIsDynamic=false, bIsEditable=true, bIsStaticSwitch=false, bIsVisible=true, type {classStructOrEnum {refPath=&NiagaraFloat}}}
     ]
     moduleScript {refPath=&SpriteRotationRate}
    }
    {
     bIsSetParametersModule=false
     enabled=true
     moduleName=ScaleSpriteSize
     inputs [
      Scale Factor {bIsDynamic=true, bIsEditable=true, bIsStaticSwitch=false, bIsVisible=true, type {classStructOrEnum {refPath=&Vector2f}}}
      Initial Sprite Size {bIsDynamic=false, bIsEditable=true, bIsStaticSwitch=false, bIsVisible=true, type {classStructOrEnum {refPath=&Vector2f}}}
     ]
     moduleScript {refPath=&ScaleSpriteSize}
    }
    {
     bIsSetParametersModule=false
     enabled=true
     moduleName=ScaleColor
     inputs [
      Scale Mode {
       bIsDynamic=false
       bIsEditable=true
       bIsStaticSwitch=true
       bIsVisible=true
       type {classStructOrEnum {refPath=&ENiagaraScaleColorMode}}
      }
      Linear Color Curve {
       bIsDynamic=false
       bIsEditable=false
       bIsStaticSwitch=false
       bIsVisible=false
       type {classStructOrEnum {refPath=&NiagaraDataInterfaceColorCurve}}
      }
      ScaleRGB {bIsDynamic=false, bIsEditable=true, bIsStaticSwitch=false, bIsVisible=true, type {classStructOrEnum {refPath=&NiagaraBool}}}
      Scale RGB {bIsDynamic=false, bIsEditable=true, bIsStaticSwitch=false, bIsVisible=true, type {classStructOrEnum {refPath=&Vector3f}}}
      ScaleRGBA {bIsDynamic=false, bIsEditable=false, bIsStaticSwitch=false, bIsVisible=false, type {classStructOrEnum {refPath=&NiagaraBool}}}
      Scale RGBA {
       bIsDynamic=false
       bIsEditable=false
       bIsStaticSwitch=false
       bIsVisible=false
       type {classStructOrEnum {refPath=&Vector4f}}
      }
      Curve Index {
       bIsDynamic=false
       bIsEditable=false
       bIsStaticSwitch=false
       bIsVisible=false
       type {classStructOrEnum {refPath=&NiagaraFloat}}
      }
      ScaleA {bIsDynamic=false, bIsEditable=true, bIsStaticSwitch=false, bIsVisible=true, type {classStructOrEnum {refPath=&NiagaraBool}}}
      Scale Alpha {bIsDynamic=true, bIsEditable=true, bIsStaticSwitch=false, bIsVisible=true, type {classStructOrEnum {refPath=&NiagaraFloat}}}
      Color Value To Scale {
       bIsDynamic=false
       bIsEditable=true
       bIsStaticSwitch=false
       bIsVisible=true
       type {classStructOrEnum {refPath=&LinearColor}}
      }
     ]
     moduleScript {refPath=&ScaleColor}
    }
    {
     bIsSetParametersModule=false
     enabled=true
     moduleName=DynamicMaterialParameters
     inputs [
      Write Parameter Index 0 {bIsDynamic=false, bIsEditable=true, bIsStaticSwitch=true, bIsVisible=true, type {classStructOrEnum {refPath=&NiagaraBool}}}
      Module.Param0WriteEnabled {bIsDynamic=false, bIsEditable=true, bIsStaticSwitch=false, bIsVisible=true, type {classStructOrEnum {refPath=&NiagaraBool}}}
      Index 0 Param 1 {bIsDynamic=false, bIsEditable=true, bIsStaticSwitch=false, bIsVisible=true, type {classStructOrEnum {refPath=&NiagaraFloat}}}
      Module.Param1WriteEnabled {bIsDynamic=false, bIsEditable=true, bIsStaticSwitch=false, bIsVisible=true, type {classStructOrEnum {refPath=&NiagaraBool}}}
      Index 0 Param 2 {bIsDynamic=false, bIsEditable=true, bIsStaticSwitch=false, bIsVisible=true, type {classStructOrEnum {refPath=&NiagaraFloat}}}
      Module.Param2WriteEnabled {bIsDynamic=false, bIsEditable=true, bIsStaticSwitch=false, bIsVisible=true, type {classStructOrEnum {refPath=&NiagaraBool}}}
      Index 0 Param 3 {bIsDynamic=false, bIsEditable=true, bIsStaticSwitch=false, bIsVisible=true, type {classStructOrEnum {refPath=&NiagaraFloat}}}
      Module.Param3WriteEnabled {bIsDynamic=false, bIsEditable=true, bIsStaticSwitch=false, bIsVisible=true, type {classStructOrEnum {refPath=&NiagaraBool}}}
      Index 0 Param 4 {bIsDynamic=false, bIsEditable=true, bIsStaticSwitch=false, bIsVisible=true, type {classStructOrEnum {refPath=&NiagaraFloat}}}
      Write Parameter Index 1 {bIsDynamic=false, bIsEditable=true, bIsStaticSwitch=true, bIsVisible=true, type {classStructOrEnum {refPath=&NiagaraBool}}}
      Index 1 Param 1 {
       bIsDynamic=false
       bIsEditable=false
       bIsStaticSwitch=false
       bIsVisible=false
       type {classStructOrEnum {refPath=&NiagaraFloat}}
      }
      Index 1 Param 2 {
       bIsDynamic=false
       bIsEditable=false
       bIsStaticSwitch=false
       bIsVisible=false
       type {classStructOrEnum {refPath=&NiagaraFloat}}
      }
      Index 1 Param 3 {
       bIsDynamic=false
       bIsEditable=false
       bIsStaticSwitch=false
       bIsVisible=false
       type {classStructOrEnum {refPath=&NiagaraFloat}}
      }
      Index 1 Param 4 {
       bIsDynamic=false
       bIsEditable=false
       bIsStaticSwitch=false
       bIsVisible=false
       type {classStructOrEnum {refPath=&NiagaraFloat}}
      }
      Write Parameter Index 2 {bIsDynamic=false, bIsEditable=true, bIsStaticSwitch=true, bIsVisible=true, type {classStructOrEnum {refPath=&NiagaraBool}}}
      Index 2 Param 1 {
       bIsDynamic=false
       bIsEditable=false
       bIsStaticSwitch=false
       bIsVisible=false
       type {classStructOrEnum {refPath=&NiagaraFloat}}
      }
      Index 2 Param 2 {
       bIsDynamic=false
       bIsEditable=false
       bIsStaticSwitch=false
       bIsVisible=false
       type {classStructOrEnum {refPath=&NiagaraFloat}}
      }
      Index 2 Param 3 {
       bIsDynamic=false
       bIsEditable=false
       bIsStaticSwitch=false
       bIsVisible=false
       type {classStructOrEnum {refPath=&NiagaraFloat}}
      }
      Index 2 Param 4 {
       bIsDynamic=false
       bIsEditable=false
       bIsStaticSwitch=false
       bIsVisible=false
       type {classStructOrEnum {refPath=&NiagaraFloat}}
      }
      Write Parameter Index 3 {bIsDynamic=false, bIsEditable=true, bIsStaticSwitch=true, bIsVisible=true, type {classStructOrEnum {refPath=&NiagaraBool}}}
      Index 3 Param 1 {
       bIsDynamic=false
       bIsEditable=false
       bIsStaticSwitch=false
       bIsVisible=false
       type {classStructOrEnum {refPath=&NiagaraFloat}}
      }
      Index 3 Param 2 {
       bIsDynamic=false
       bIsEditable=false
       bIsStaticSwitch=false
       bIsVisible=false
       type {classStructOrEnum {refPath=&NiagaraFloat}}
      }
      Index 3 Param 3 {
       bIsDynamic=false
       bIsEditable=false
       bIsStaticSwitch=false
       bIsVisible=false
       type {classStructOrEnum {refPath=&NiagaraFloat}}
      }
      Index 3 Param 4 {
       bIsDynamic=false
       bIsEditable=false
       bIsStaticSwitch=false
       bIsVisible=false
       type {classStructOrEnum {refPath=&NiagaraFloat}}
      }
     ]
     moduleScript {refPath=&DynamicMaterialParameters}
    }
    {
     bIsSetParametersModule=false
     enabled=true
     moduleName=ScaleColor001
     inputs [
      Scale Mode {bIsDynamic=false, bIsEditable=true, bIsStaticSwitch=true, bIsVisible=true, type {classStructOrEnum {refPath=&NiagaraInt32}}}
      Linear Color Curve {
       bIsDynamic=false
       bIsEditable=false
       bIsStaticSwitch=false
       bIsVisible=false
       type {classStructOrEnum {refPath=&NiagaraDataInterfaceColorCurve}}
      }
      ScaleRGB {bIsDynamic=false, bIsEditable=true, bIsStaticSwitch=false, bIsVisible=true, type {classStructOrEnum {refPath=&NiagaraBool}}}
      Scale RGB {bIsDynamic=true, bIsEditable=true, bIsStaticSwitch=false, bIsVisible=true, type {classStructOrEnum {refPath=&Vector3f}}}
      ScaleRGBA {bIsDynamic=false, bIsEditable=false, bIsStaticSwitch=false, bIsVisible=false, type {classStructOrEnum {refPath=&NiagaraBool}}}
      Scale RGBA {
       bIsDynamic=false
       bIsEditable=false
       bIsStaticSwitch=false
       bIsVisible=false
       type {classStructOrEnum {refPath=&Vector4f}}
      }
      Curve Index {
       bIsDynamic=false
       bIsEditable=false
       bIsStaticSwitch=false
       bIsVisible=false
       type {classStructOrEnum {refPath=&NiagaraFloat}}
      }
      ScaleA {bIsDynamic=false, bIsEditable=true, bIsStaticSwitch=false, bIsVisible=true, type {classStructOrEnum {refPath=&NiagaraBool}}}
      Scale Alpha {bIsDynamic=false, bIsEditable=false, bIsStaticSwitch=false, bIsVisible=true, type {classStructOrEnum {refPath=&NiagaraFloat}}}
      Color Value To Scale {
       bIsDynamic=false
       bIsEditable=true
       bIsStaticSwitch=false
       bIsVisible=true
       type {classStructOrEnum {refPath=&LinearColor}}
      }
      Color Scale {
       bIsDynamic=false
       bIsEditable=false
       bIsStaticSwitch=false
       bIsVisible=false
       type {classStructOrEnum {refPath=&LinearColor}}
      }
     ]
     moduleScript {refPath=&ScaleColor}
    }
   ]
  }
  rendererClasses [{refPath=&NiagaraSpriteRendererProperties}]
  renderers [{rendererIndex=0, rendererClass {refPath=&NiagaraSpriteRendererProperties}}]
 }
 Snowflakedsff001 {
  bEnabled=false
  emitterSpawnScript {scriptName=EmitterSpawnScript, modules []}
  emitterUpdateScript {
   scriptName=EmitterUpdateScript
   modules [
    {
     bIsSetParametersModule=false
     enabled=true
     moduleName=EmitterState
     inputs [
      Life Cycle Mode {
       bIsDynamic=false
       bIsEditable=true
       bIsStaticSwitch=true
       bIsVisible=true
       type {classStructOrEnum {refPath=&ENiagaraEmitterLifeCycleMode}}
      }
      Inactive Response {
       bIsDynamic=false
       bIsEditable=true
       bIsStaticSwitch=true
       bIsVisible=true
       type {classStructOrEnum {refPath=&ENiagaraInactiveMode}}
      }
      Loop Behavior {
       bIsDynamic=false
       bIsEditable=true
       bIsStaticSwitch=true
       bIsVisible=true
       type {classStructOrEnum {refPath=&ENiagara_EmitterStateOptions}}
      }
      Loop Duration Mode {
       bIsDynamic=false
       bIsEditable=false
       bIsStaticSwitch=true
       bIsVisible=false
       type {classStructOrEnum {refPath=&ENiagara_InfiniteLoopDuration}}
      }
      Loop Count {
       bIsDynamic=false
       bIsEditable=false
       bIsStaticSwitch=false
       bIsVisible=false
       type {classStructOrEnum {refPath=&NiagaraInt32}}
      }
      Loop Duration {bIsDynamic=false, bIsEditable=true, bIsStaticSwitch=false, bIsVisible=true, type {classStructOrEnum {refPath=&NiagaraFloat}}}
      Recalculate Duration Each Loop {bIsDynamic=false, bIsEditable=true, bIsStaticSwitch=false, bIsVisible=true, type {classStructOrEnum {refPath=&NiagaraBool}}}
      UseLoopDelay {bIsDynamic=false, bIsEditable=true, bIsStaticSwitch=false, bIsVisible=true, type {classStructOrEnum {refPath=&NiagaraBool}}}
      Loop Delay {bIsDynamic=false, bIsEditable=true, bIsStaticSwitch=false, bIsVisible=true, type {classStructOrEnum {refPath=&NiagaraFloat}}}
      Delay First Loop Only {bIsDynamic=false, bIsEditable=true, bIsStaticSwitch=false, bIsVisible=true, type {classStructOrEnum {refPath=&NiagaraBool}}}
      Scalability Mode {
       bIsDynamic=false
       bIsEditable=true
       bIsStaticSwitch=true
       bIsVisible=true
       type {classStructOrEnum {refPath=&ENiagaraEmitterScalabilityMode_Limited}}
      }
      Enable Distance Culling {bIsDynamic=false, bIsEditable=false, bIsStaticSwitch=true, bIsVisible=false, type {classStructOrEnum {refPath=&NiagaraBool}}}
      MinDistance {
       bIsDynamic=false
       bIsEditable=false
       bIsStaticSwitch=false
       bIsVisible=false
       type {classStructOrEnum {refPath=&NiagaraFloat}}
      }
      Min Distance Response {
       bIsDynamic=false
       bIsEditable=false
       bIsStaticSwitch=false
       bIsVisible=false
       type {classStructOrEnum {refPath=&ENiagaraExecutionStateManagement}}
      }
      MaxDistance {
       bIsDynamic=false
       bIsEditable=false
       bIsStaticSwitch=false
       bIsVisible=false
       type {classStructOrEnum {refPath=&NiagaraFloat}}
      }
      Max Distance Response {
       bIsDynamic=false
       bIsEditable=false
       bIsStaticSwitch=false
       bIsVisible=false
       type {classStructOrEnum {refPath=&ENiagaraExecutionStateManagement}}
      }
      Scale Spawn Count {bIsDynamic=false, bIsEditable=false, bIsStaticSwitch=true, bIsVisible=false, type {classStructOrEnum {refPath=&NiagaraBool}}}
      Spawn Count Scale {
       bIsDynamic=false
       bIsEditable=false
       bIsStaticSwitch=false
       bIsVisible=false
       type {classStructOrEnum {refPath=&NiagaraFloat}}
      }
      Spawn Count Scale By Distance {
       bIsDynamic=false
       bIsEditable=false
       bIsStaticSwitch=false
       bIsVisible=false
       type {classStructOrEnum {refPath=&NiagaraDataInterfaceCurve}}
      }
      Enable Visibility Culling {bIsDynamic=false, bIsEditable=false, bIsStaticSwitch=true, bIsVisible=false, type {classStructOrEnum {refPath=&NiagaraBool}}}
      Visibility Cull Response {
       bIsDynamic=false
       bIsEditable=false
       bIsStaticSwitch=false
       bIsVisible=false
       type {classStructOrEnum {refPath=&ENiagaraExecutionStateManagement}}
      }
      Visibility Cull Delay {
       bIsDynamic=false
       bIsEditable=false
       bIsStaticSwitch=false
       bIsVisible=false
       type {classStructOrEnum {refPath=&NiagaraFloat}}
      }
      Reset Age On Awaken {bIsDynamic=false, bIsEditable=false, bIsStaticSwitch=false, bIsVisible=false, type {classStructOrEnum {refPath=&NiagaraBool}}}
     ]
     moduleScript {refPath=&EmitterState}
    }
    {
     bIsSetParametersModule=false
     enabled=true
     moduleName=SpawnRate
     inputs [
      SpawnRate {bIsDynamic=false, bIsEditable=true, bIsStaticSwitch=false, bIsVisible=true, type {classStructOrEnum {refPath=&NiagaraFloat}}}
      Use Spawn Probability {bIsDynamic=false, bIsEditable=true, bIsStaticSwitch=false, bIsVisible=true, type {classStructOrEnum {refPath=&NiagaraBool}}}
      Spawn Probability {bIsDynamic=false, bIsEditable=false, bIsStaticSwitch=false, bIsVisible=true, type {classStructOrEnum {refPath=&NiagaraFloat}}}
      SpawnGroup {bIsDynamic=false, bIsEditable=true, bIsStaticSwitch=false, bIsVisible=true, type {classStructOrEnum {refPath=&NiagaraInt32}}}
     ]
     moduleScript {refPath=&SpawnRate}
    }
    {
     bIsSetParametersModule=false
     enabled=true
     moduleName=SpawnBurst_Instantaneous
     inputs [
      Spawn Count {bIsDynamic=false, bIsEditable=true, bIsStaticSwitch=false, bIsVisible=true, type {classStructOrEnum {refPath=&NiagaraInt32}}}
      Spawn Time {bIsDynamic=false, bIsEditable=true, bIsStaticSwitch=false, bIsVisible=true, type {classStructOrEnum {refPath=&NiagaraFloat}}}
      Use Spawn Probability {bIsDynamic=false, bIsEditable=true, bIsStaticSwitch=false, bIsVisible=true, type {classStructOrEnum {refPath=&NiagaraBool}}}
      Spawn Probability {bIsDynamic=false, bIsEditable=false, bIsStaticSwitch=false, bIsVisible=true, type {classStructOrEnum {refPath=&NiagaraFloat}}}
      UseLoopCountLimit {bIsDynamic=false, bIsEditable=true, bIsStaticSwitch=false, bIsVisible=true, type {classStructOrEnum {refPath=&NiagaraBool}}}
      Loop Count Limit {bIsDynamic=false, bIsEditable=false, bIsStaticSwitch=false, bIsVisible=true, type {classStructOrEnum {refPath=&NiagaraInt32}}}
      Spawn Group {bIsDynamic=false, bIsEditable=true, bIsStaticSwitch=false, bIsVisible=true, type {classStructOrEnum {refPath=&NiagaraInt32}}}
      Age {bIsDynamic=false, bIsEditable=true, bIsStaticSwitch=false, bIsVisible=true, type {classStructOrEnum {refPath=&NiagaraFloat}}}
     ]
     moduleScript {refPath=&SpawnBurst_Instantaneous}
    }
   ]
  }
  particleSpawnScript {
   scriptName=ParticleSpawnScript
   modules [
    {
     bIsSetParametersModule=false
     enabled=true
     moduleName=InitializeParticle001
     inputs [
      Lifetime Mode {
       bIsDynamic=false
       bIsEditable=true
       bIsStaticSwitch=true
       bIsVisible=true
       type {classStructOrEnum {refPath=&ENiagara_LifetimeMode}}
      }
      Lifetime Min {bIsDynamic=false, bIsEditable=true, bIsStaticSwitch=false, bIsVisible=true, type {classStructOrEnum {refPath=&NiagaraFloat}}}
      Lifetime {
       bIsDynamic=false
       bIsEditable=false
       bIsStaticSwitch=false
       bIsVisible=false
       type {classStructOrEnum {refPath=&NiagaraFloat}}
      }
      Lifetime Max {bIsDynamic=false, bIsEditable=true, bIsStaticSwitch=false, bIsVisible=true, type {classStructOrEnum {refPath=&NiagaraFloat}}}
      Lifetime Randomness Mode {
       bIsDynamic=false
       bIsEditable=true
       bIsStaticSwitch=true
       bIsVisible=true
       type {classStructOrEnum {refPath=&ENiagaraRandomnessMode}}
      }
      Color Mode {
       bIsDynamic=false
       bIsEditable=true
       bIsStaticSwitch=true
       bIsVisible=true
       type {classStructOrEnum {refPath=&ENiagara_ColorInitializationMode}}
      }
      Color Minimum {
       bIsDynamic=false
       bIsEditable=false
       bIsStaticSwitch=false
       bIsVisible=false
       type {classStructOrEnum {refPath=&LinearColor}}
      }
      Color {
       bIsDynamic=true
       bIsEditable=true
       bIsStaticSwitch=false
       bIsVisible=true
       type {classStructOrEnum {refPath=&LinearColor}}
      }
      AdjustHue {bIsDynamic=false, bIsEditable=false, bIsStaticSwitch=false, bIsVisible=false, type {classStructOrEnum {refPath=&NiagaraBool}}}
      Hue Shift Range {
       bIsDynamic=false
       bIsEditable=false
       bIsStaticSwitch=false
       bIsVisible=false
       type {classStructOrEnum {refPath=&Vector2f}}
      }
      Color Maximum {
       bIsDynamic=false
       bIsEditable=false
       bIsStaticSwitch=false
       bIsVisible=false
       type {classStructOrEnum {refPath=&LinearColor}}
      }
      AdjustSaturation {bIsDynamic=false, bIsEditable=false, bIsStaticSwitch=false, bIsVisible=false, type {classStructOrEnum {refPath=&NiagaraBool}}}
      Saturation Range {
       bIsDynamic=false
       bIsEditable=false
       bIsStaticSwitch=false
       bIsVisible=false
       type {classStructOrEnum {refPath=&Vector2f}}
      }
      AdjustValue {bIsDynamic=false, bIsEditable=false, bIsStaticSwitch=false, bIsVisible=false, type {classStructOrEnum {refPath=&NiagaraBool}}}
      Value Range {
       bIsDynamic=false
       bIsEditable=false
       bIsStaticSwitch=false
       bIsVisible=false
       type {classStructOrEnum {refPath=&Vector2f}}
      }
      AdjustAlpha {bIsDynamic=false, bIsEditable=false, bIsStaticSwitch=false, bIsVisible=false, type {classStructOrEnum {refPath=&NiagaraBool}}}
      Alpha Scale Range {
       bIsDynamic=false
       bIsEditable=false
       bIsStaticSwitch=false
       bIsVisible=false
       type {classStructOrEnum {refPath=&Vector2f}}
      }
      Color Channel Mode {
       bIsDynamic=false
       bIsEditable=false
       bIsStaticSwitch=true
       bIsVisible=false
       type {classStructOrEnum {refPath=&ENiagaraChannelCorrelation}}
      }
      Color Randomness Mode {
       bIsDynamic=false
       bIsEditable=false
       bIsStaticSwitch=true
       bIsVisible=false
       type {classStructOrEnum {refPath=&ENiagaraRandomnessMode}}
      }
      Position Mode {
       bIsDynamic=false
       bIsEditable=true
       bIsStaticSwitch=true
       bIsVisible=true
       type {classStructOrEnum {refPath=&ENiagara_PositionInitializationMode}}
      }
      Position {
       bIsDynamic=false
       bIsEditable=false
       bIsStaticSwitch=false
       bIsVisible=false
       type {classStructOrEnum {refPath=&NiagaraPosition}}
      }
      UsePositionOffset {bIsDynamic=false, bIsEditable=true, bIsStaticSwitch=false, bIsVisible=true, type {classStructOrEnum {refPath=&NiagaraBool}}}
      Position Offset {bIsDynamic=false, bIsEditable=true, bIsStaticSwitch=false, bIsVisible=true, type {classStructOrEnum {refPath=&Vector3f}}}
      Position Offset Coordinate Space {
       bIsDynamic=false
       bIsEditable=true
       bIsStaticSwitch=false
       bIsVisible=true
       type {classStructOrEnum {refPath=&ENiagaraCoordinateSpace}}
      }
      Mass Mode {
       bIsDynamic=false
       bIsEditable=true
       bIsStaticSwitch=true
       bIsVisible=true
       type {classStructOrEnum {refPath=&ENiagara_MassInitializationMode}}
      }
      Mass Min {
       bIsDynamic=false
       bIsEditable=false
       bIsStaticSwitch=false
       bIsVisible=false
       type {classStructOrEnum {refPath=&NiagaraFloat}}
      }
      Mass {
       bIsDynamic=false
       bIsEditable=false
       bIsStaticSwitch=false
       bIsVisible=false
       type {classStructOrEnum {refPath=&NiagaraFloat}}
      }
      Mass Max {
       bIsDynamic=false
       bIsEditable=false
       bIsStaticSwitch=false
       bIsVisible=false
       type {classStructOrEnum {refPath=&NiagaraFloat}}
      }
      Mass Randomness Mode {
       bIsDynamic=false
       bIsEditable=false
       bIsStaticSwitch=true
       bIsVisible=false
       type {classStructOrEnum {refPath=&ENiagaraRandomnessMode}}
      }
      Write Material Random {bIsDynamic=false, bIsEditable=true, bIsStaticSwitch=false, bIsVisible=true, type {classStructOrEnum {refPath=&NiagaraBool}}}
      Material Random {bIsDynamic=false, bIsEditable=true, bIsStaticSwitch=false, bIsVisible=true, type {classStructOrEnum {refPath=&NiagaraFloat}}}
      Sprite Size Mode {
       bIsDynamic=false
       bIsEditable=true
       bIsStaticSwitch=true
       bIsVisible=true
       type {classStructOrEnum {refPath=&ENiagara_SizeScaleMode}}
      }
      Sprite Size {
       bIsDynamic=false
       bIsEditable=false
       bIsStaticSwitch=false
       bIsVisible=false
       type {classStructOrEnum {refPath=&Vector2f}}
      }
      Uniform Sprite Size {bIsDynamic=true, bIsEditable=true, bIsStaticSwitch=false, bIsVisible=true, type {classStructOrEnum {refPath=&NiagaraFloat}}}
      Uniform Sprite Size Min {
       bIsDynamic=false
       bIsEditable=false
       bIsStaticSwitch=false
       bIsVisible=false
       type {classStructOrEnum {refPath=&NiagaraFloat}}
      }
      Sprite Size Min {
       bIsDynamic=false
       bIsEditable=false
       bIsStaticSwitch=false
       bIsVisible=false
       type {classStructOrEnum {refPath=&Vector2f}}
      }
      Sprite Size Max {
       bIsDynamic=false
       bIsEditable=false
       bIsStaticSwitch=false
       bIsVisible=false
       type {classStructOrEnum {refPath=&Vector2f}}
      }
      Uniform Sprite Size Max {
       bIsDynamic=false
       bIsEditable=false
       bIsStaticSwitch=false
       bIsVisible=false
       type {classStructOrEnum {refPath=&NiagaraFloat}}
      }
      Sprite Size Randomness Mode {
       bIsDynamic=false
       bIsEditable=false
       bIsStaticSwitch=true
       bIsVisible=false
       type {classStructOrEnum {refPath=&ENiagaraRandomnessMode}}
      }
      Sprite Rotation Mode {
       bIsDynamic=false
       bIsEditable=true
       bIsStaticSwitch=true
       bIsVisible=true
       type {classStructOrEnum {refPath=&ENiagara_SpriteRotationMode}}
      }
      Sprite Rotation Angle Min {
       bIsDynamic=false
       bIsEditable=false
       bIsStaticSwitch=false
       bIsVisible=false
       type {classStructOrEnum {refPath=&NiagaraFloat}}
      }
      Sprite Rotation Angle {
       bIsDynamic=false
       bIsEditable=false
       bIsStaticSwitch=false
       bIsVisible=false
       type {classStructOrEnum {refPath=&NiagaraFloat}}
      }
      Sprite Rotation Angle Max {
       bIsDynamic=false
       bIsEditable=false
       bIsStaticSwitch=false
       bIsVisible=false
       type {classStructOrEnum {refPath=&NiagaraFloat}}
      }
      Sprite Rotation Randomness Mode {
       bIsDynamic=false
       bIsEditable=false
       bIsStaticSwitch=true
       bIsVisible=false
       type {classStructOrEnum {refPath=&ENiagaraRandomnessMode}}
      }
      Sprite UV Mode {
       bIsDynamic=false
       bIsEditable=true
       bIsStaticSwitch=true
       bIsVisible=true
       type {classStructOrEnum {refPath=&ENiagara_UVFlippingMode}}
      }
      Sprite UV Randomness Mode {
       bIsDynamic=false
       bIsEditable=false
       bIsStaticSwitch=true
       bIsVisible=false
       type {classStructOrEnum {refPath=&ENiagaraRandomnessMode}}
      }
      Sprite UV Scale {
       bIsDynamic=false
       bIsEditable=false
       bIsStaticSwitch=false
       bIsVisible=false
       type {classStructOrEnum {refPath=&Vector2f}}
      }
      Mesh Scale Mode {
       bIsDynamic=false
       bIsEditable=true
       bIsStaticSwitch=true
       bIsVisible=true
       type {classStructOrEnum {refPath=&ENiagara_SizeScaleMode}}
      }
      Mesh Uniform Scale Min {
       bIsDynamic=false
       bIsEditable=false
       bIsStaticSwitch=false
       bIsVisible=false
       type {classStructOrEnum {refPath=&NiagaraFloat}}
      }
      Mesh Uniform Scale {
       bIsDynamic=false
       bIsEditable=false
       bIsStaticSwitch=false
       bIsVisible=false
       type {classStructOrEnum {refPath=&NiagaraFloat}}
      }
      Mesh Scale Min {bIsDynamic=false, bIsEditable=true, bIsStaticSwitch=false, bIsVisible=true, type {classStructOrEnum {refPath=&Vector3f}}}
      Mesh Scale {
       bIsDynamic=false
       bIsEditable=false
       bIsStaticSwitch=false
       bIsVisible=false
       type {classStructOrEnum {refPath=&Vector3f}}
      }
      Mesh Scale Max {bIsDynamic=false, bIsEditable=true, bIsStaticSwitch=false, bIsVisible=true, type {classStructOrEnum {refPath=&Vector3f}}}
      Mesh Uniform Scale Max {
       bIsDynamic=false
       bIsEditable=false
       bIsStaticSwitch=false
       bIsVisible=false
       type {classStructOrEnum {refPath=&NiagaraFloat}}
      }
      Mesh Scale Randomness Mode {
       bIsDynamic=false
       bIsEditable=true
       bIsStaticSwitch=true
       bIsVisible=true
       type {classStructOrEnum {refPath=&ENiagaraRandomnessMode}}
      }
      Mesh Renderer Array Visibility Mode {
       bIsDynamic=false
       bIsEditable=true
       bIsStaticSwitch=true
       bIsVisible=true
       type {classStructOrEnum {refPath=&ENiagara_UnsetDirectSetRandom}}
      }
      Mesh Renderer Mesh Index {
       bIsDynamic=false
       bIsEditable=false
       bIsStaticSwitch=false
       bIsVisible=false
       type {classStructOrEnum {refPath=&NiagaraInt32}}
      }
      Wrap Index By Mesh Count {bIsDynamic=false, bIsEditable=false, bIsStaticSwitch=true, bIsVisible=false, type {classStructOrEnum {refPath=&NiagaraBool}}}
      Mesh Renderer Info {
       bIsDynamic=false
       bIsEditable=false
       bIsStaticSwitch=false
       bIsVisible=false
       type {classStructOrEnum {refPath=&NiagaraDataInterfaceMeshRendererInfo}}
      }
      Mesh Renderer Random Seed {
       bIsDynamic=false
       bIsEditable=false
       bIsStaticSwitch=false
       bIsVisible=false
       type {classStructOrEnum {refPath=&NiagaraInt32}}
      }
      Ribbon Width Mode {
       bIsDynamic=false
       bIsEditable=true
       bIsStaticSwitch=true
       bIsVisible=true
       type {classStructOrEnum {refPath=&ENiagara_UnsetDirectSet}}
      }
      Ribbon Width {
       bIsDynamic=false
       bIsEditable=false
       bIsStaticSwitch=false
       bIsVisible=false
       type {classStructOrEnum {refPath=&NiagaraFloat}}
      }
      Ribbon Facing Mode {
       bIsDynamic=false
       bIsEditable=true
       bIsStaticSwitch=true
       bIsVisible=true
       type {classStructOrEnum {refPath=&ENiagara_UnsetDirectSet}}
      }
      Ribbon Facing Vector {
       bIsDynamic=false
       bIsEditable=false
       bIsStaticSwitch=false
       bIsVisible=false
       type {classStructOrEnum {refPath=&Vector3f}}
      }
      Ribbon Twist Mode {
       bIsDynamic=false
       bIsEditable=true
       bIsStaticSwitch=true
       bIsVisible=true
       type {classStructOrEnum {refPath=&ENiagara_UnsetDirectSet}}
      }
      "Ribbon Twist " {
       bIsDynamic=false
       bIsEditable=false
       bIsStaticSwitch=false
       bIsVisible=false
       type {classStructOrEnum {refPath=&NiagaraFloat}}
      }
     ]
     moduleScript {refPath=&InitializeParticle}
    }
    {
     bIsSetParametersModule=false
     enabled=true
     moduleName=SphereLocation
     inputs [
      Sphere Radius {bIsDynamic=false, bIsEditable=true, bIsStaticSwitch=false, bIsVisible=true, type {classStructOrEnum {refPath=&NiagaraFloat}}}
      UseOffset {bIsDynamic=false, bIsEditable=true, bIsStaticSwitch=false, bIsVisible=true, type {classStructOrEnum {refPath=&NiagaraBool}}}
      Offset {bIsDynamic=false, bIsEditable=true, bIsStaticSwitch=false, bIsVisible=true, type {classStructOrEnum {refPath=&Vector3f}}}
      Offset Coordinate Space {
       bIsDynamic=false
       bIsEditable=true
       bIsStaticSwitch=false
       bIsVisible=true
       type {classStructOrEnum {refPath=&ENiagaraCoordinateSpace}}
      }
      UseNonUniformScale {bIsDynamic=false, bIsEditable=true, bIsStaticSwitch=false, bIsVisible=true, type {classStructOrEnum {refPath=&NiagaraBool}}}
      Non Uniform Scale {bIsDynamic=false, bIsEditable=false, bIsStaticSwitch=false, bIsVisible=true, type {classStructOrEnum {refPath=&Vector3f}}}
      UseSphereOrientation {bIsDynamic=false, bIsEditable=true, bIsStaticSwitch=false, bIsVisible=true, type {classStructOrEnum {refPath=&NiagaraBool}}}
      Sphere Orientation Axis {bIsDynamic=false, bIsEditable=false, bIsStaticSwitch=false, bIsVisible=true, type {classStructOrEnum {refPath=&Vector3f}}}
      Sphere Origin {
       bIsDynamic=false
       bIsEditable=true
       bIsStaticSwitch=false
       bIsVisible=true
       type {classStructOrEnum {refPath=&NiagaraPosition}}
      }
      Sphere Coordinate Space {
       bIsDynamic=false
       bIsEditable=true
       bIsStaticSwitch=false
       bIsVisible=true
       type {classStructOrEnum {refPath=&ENiagaraCoordinateSpace}}
      }
      Write to Intrinsic Variables {bIsDynamic=false, bIsEditable=true, bIsStaticSwitch=true, bIsVisible=true, type {classStructOrEnum {refPath=&NiagaraBool}}}
      Sphere Distribution {
       bIsDynamic=false
       bIsEditable=true
       bIsStaticSwitch=true
       bIsVisible=true
       type {classStructOrEnum {refPath=&ENiagaraSphereDistributionMode}}
      }
      Evaluation Type {
       bIsDynamic=false
       bIsEditable=false
       bIsStaticSwitch=true
       bIsVisible=false
       type {classStructOrEnum {refPath=&ENiagaraRandomnessEvaluation}}
      }
      Hemisphere X {bIsDynamic=false, bIsEditable=true, bIsStaticSwitch=false, bIsVisible=true, type {classStructOrEnum {refPath=&NiagaraBool}}}
      Uniform Distribution {
       bIsDynamic=false
       bIsEditable=false
       bIsStaticSwitch=false
       bIsVisible=false
       type {classStructOrEnum {refPath=&NiagaraFloat}}
      }
      Hemisphere Y {bIsDynamic=false, bIsEditable=true, bIsStaticSwitch=false, bIsVisible=true, type {classStructOrEnum {refPath=&NiagaraBool}}}
      U Position {
       bIsDynamic=false
       bIsEditable=false
       bIsStaticSwitch=false
       bIsVisible=false
       type {classStructOrEnum {refPath=&NiagaraFloat}}
      }
      Uniform Spiral Amount {
       bIsDynamic=false
       bIsEditable=false
       bIsStaticSwitch=false
       bIsVisible=false
       type {classStructOrEnum {refPath=&NiagaraFloat}}
      }
      Hemisphere Z {bIsDynamic=false, bIsEditable=true, bIsStaticSwitch=false, bIsVisible=true, type {classStructOrEnum {refPath=&NiagaraBool}}}
      Uniform Endpoint Offset {
       bIsDynamic=false
       bIsEditable=false
       bIsStaticSwitch=false
       bIsVisible=false
       type {classStructOrEnum {refPath=&NiagaraFloat}}
      }
      V Position {
       bIsDynamic=false
       bIsEditable=false
       bIsStaticSwitch=false
       bIsVisible=false
       type {classStructOrEnum {refPath=&NiagaraFloat}}
      }
      Radius Position {
       bIsDynamic=false
       bIsEditable=false
       bIsStaticSwitch=false
       bIsVisible=false
       type {classStructOrEnum {refPath=&NiagaraFloat}}
      }
      Surface Only {bIsDynamic=false, bIsEditable=true, bIsStaticSwitch=false, bIsVisible=true, type {classStructOrEnum {refPath=&NiagaraBool}}}
      Surface Only Band Thickness {bIsDynamic=false, bIsEditable=true, bIsStaticSwitch=false, bIsVisible=true, type {classStructOrEnum {refPath=&NiagaraFloat}}}
      Surface Expansion Mode {
       bIsDynamic=false
       bIsEditable=true
       bIsStaticSwitch=false
       bIsVisible=true
       type {classStructOrEnum {refPath=&ENiagaraExpansionMode}}
      }
      Normalize Thickness {bIsDynamic=false, bIsEditable=true, bIsStaticSwitch=false, bIsVisible=true, type {classStructOrEnum {refPath=&NiagaraBool}}}
      Randomness Mode {
       bIsDynamic=false
       bIsEditable=true
       bIsStaticSwitch=true
       bIsVisible=true
       type {classStructOrEnum {refPath=&ENiagaraRandomnessMode}}
      }
      Override Seed {bIsDynamic=false, bIsEditable=true, bIsStaticSwitch=false, bIsVisible=true, type {classStructOrEnum {refPath=&NiagaraBool}}}
      Random Seed {bIsDynamic=false, bIsEditable=false, bIsStaticSwitch=false, bIsVisible=true, type {classStructOrEnum {refPath=&NiagaraInt32}}}
      Fixed Random Seed {bIsDynamic=false, bIsEditable=false, bIsStaticSwitch=true, bIsVisible=false, type {classStructOrEnum {refPath=&NiagaraBool}}}
      MaskSpawnGroup {bIsDynamic=false, bIsEditable=true, bIsStaticSwitch=false, bIsVisible=true, type {classStructOrEnum {refPath=&NiagaraBool}}}
      Spawn Group Mask {bIsDynamic=false, bIsEditable=false, bIsStaticSwitch=false, bIsVisible=true, type {classStructOrEnum {refPath=&NiagaraInt32}}}
     ]
     moduleScript {refPath=&SphereLocation}
    }
    {
     bIsSetParametersModule=false
     enabled=false
     moduleName=AddVelocityFromPoint
     inputs [
      Velocity Strength {bIsDynamic=true, bIsEditable=true, bIsStaticSwitch=false, bIsVisible=true, type {classStructOrEnum {refPath=&NiagaraFloat}}}
      Origin Offset {bIsDynamic=false, bIsEditable=true, bIsStaticSwitch=false, bIsVisible=true, type {classStructOrEnum {refPath=&Vector3f}}}
      Origin Offset Coordinate Space {
       bIsDynamic=false
       bIsEditable=true
       bIsStaticSwitch=false
       bIsVisible=true
       type {classStructOrEnum {refPath=&ENiagaraCoordinateSpace}}
      }
      Default Position {bIsDynamic=false, bIsEditable=true, bIsStaticSwitch=false, bIsVisible=true, type {classStructOrEnum {refPath=&Vector3f}}}
      Module.Use Falloff Distance {bIsDynamic=false, bIsEditable=true, bIsStaticSwitch=false, bIsVisible=true, type {classStructOrEnum {refPath=&NiagaraBool}}}
      Velocity Falloff Distance {bIsDynamic=false, bIsEditable=false, bIsStaticSwitch=false, bIsVisible=true, type {classStructOrEnum {refPath=&NiagaraFloat}}}
      Invert Velocity Falloff {bIsDynamic=false, bIsEditable=false, bIsStaticSwitch=false, bIsVisible=false, type {classStructOrEnum {refPath=&NiagaraBool}}}
      Velocity Origin {bIsDynamic=false, bIsEditable=true, bIsStaticSwitch=false, bIsVisible=true, type {classStructOrEnum {refPath=&Vector3f}}}
     ]
     moduleScript {refPath=&AddVelocityFromPoint}
    }
    {
     bIsSetParametersModule=false
     enabled=true
     moduleName=AddVelocity
     inputs [
      Velocity Mode {
       bIsDynamic=false
       bIsEditable=true
       bIsStaticSwitch=true
       bIsVisible=true
       type {classStructOrEnum {refPath=&ENiagara_VelocityMode}}
      }
      Velocity {
       bIsDynamic=false
       bIsEditable=false
       bIsStaticSwitch=false
       bIsVisible=false
       type {classStructOrEnum {refPath=&Vector3f}}
      }
      Velocity Speed {bIsDynamic=true, bIsEditable=true, bIsStaticSwitch=false, bIsVisible=true, type {classStructOrEnum {refPath=&NiagaraFloat}}}
      Velocity Speed Scale {
       bIsDynamic=false
       bIsEditable=false
       bIsStaticSwitch=false
       bIsVisible=false
       type {classStructOrEnum {refPath=&NiagaraFloat}}
      }
      Constrain To Radius {
       bIsDynamic=false
       bIsEditable=false
       bIsStaticSwitch=false
       bIsVisible=false
       type {classStructOrEnum {refPath=&NiagaraFloat}}
      }
      Module.UseVelDistribution {bIsDynamic=false, bIsEditable=true, bIsStaticSwitch=false, bIsVisible=true, type {classStructOrEnum {refPath=&NiagaraBool}}}
      Distribution Along Cone Axis {bIsDynamic=false, bIsEditable=false, bIsStaticSwitch=false, bIsVisible=true, type {classStructOrEnum {refPath=&NiagaraFloat}}}
      Radius Falloff Near / Far {
       bIsDynamic=false
       bIsEditable=false
       bIsStaticSwitch=false
       bIsVisible=false
       type {classStructOrEnum {refPath=&Vector2f}}
      }
      Radius Falloff Exponent {
       bIsDynamic=false
       bIsEditable=false
       bIsStaticSwitch=false
       bIsVisible=false
       type {classStructOrEnum {refPath=&NiagaraFloat}}
      }
      Module.UseConeFalloff {bIsDynamic=false, bIsEditable=true, bIsStaticSwitch=false, bIsVisible=true, type {classStructOrEnum {refPath=&NiagaraBool}}}
      Speed Falloff From Cone Axis {bIsDynamic=false, bIsEditable=false, bIsStaticSwitch=false, bIsVisible=true, type {classStructOrEnum {refPath=&NiagaraFloat}}}
      Invert Velocity Falloff {bIsDynamic=false, bIsEditable=false, bIsStaticSwitch=false, bIsVisible=false, type {classStructOrEnum {refPath=&NiagaraBool}}}
      Override Seed {bIsDynamic=false, bIsEditable=true, bIsStaticSwitch=false, bIsVisible=true, type {classStructOrEnum {refPath=&NiagaraBool}}}
      Random Seed {bIsDynamic=false, bIsEditable=false, bIsStaticSwitch=false, bIsVisible=true, type {classStructOrEnum {refPath=&NiagaraInt32}}}
      Fixed Random Seed {bIsDynamic=false, bIsEditable=false, bIsStaticSwitch=true, bIsVisible=false, type {classStructOrEnum {refPath=&NiagaraBool}}}
      Randomness Mode {
       bIsDynamic=false
       bIsEditable=true
       bIsStaticSwitch=true
       bIsVisible=true
       type {classStructOrEnum {refPath=&ENiagaraRandomnessMode}}
      }
      Cone Axis {bIsDynamic=false, bIsEditable=true, bIsStaticSwitch=false, bIsVisible=true, type {classStructOrEnum {refPath=&Vector3f}}}
      Cone Angle {bIsDynamic=false, bIsEditable=true, bIsStaticSwitch=false, bIsVisible=true, type {classStructOrEnum {refPath=&NiagaraFloat}}}
      Inner Cone Angle {bIsDynamic=false, bIsEditable=true, bIsStaticSwitch=false, bIsVisible=true, type {classStructOrEnum {refPath=&NiagaraFloat}}}
      Cone Angle Mode {
       bIsDynamic=false
       bIsEditable=true
       bIsStaticSwitch=true
       bIsVisible=true
       type {classStructOrEnum {refPath=&ENiagara_AngleInput}}
      }
      Rotation Mode {
       bIsDynamic=false
       bIsEditable=true
       bIsStaticSwitch=true
       bIsVisible=true
       type {classStructOrEnum {refPath=&ENiagara_RotationMode}}
      }
      Rotation Axis {
       bIsDynamic=false
       bIsEditable=false
       bIsStaticSwitch=false
       bIsVisible=false
       type {classStructOrEnum {refPath=&Vector3f}}
      }
      Yaw / Pitch / Roll {
       bIsDynamic=false
       bIsEditable=false
       bIsStaticSwitch=false
       bIsVisible=false
       type {classStructOrEnum {refPath=&Vector3f}}
      }
      Rotation Quaternion {bIsDynamic=false, bIsEditable=false, bIsStaticSwitch=false, bIsVisible=false, type {classStructOrEnum {refPath=&Quat4f}}}
      Rotation Matrix {
       bIsDynamic=false
       bIsEditable=false
       bIsStaticSwitch=false
       bIsVisible=false
       type {classStructOrEnum {refPath=&NiagaraMatrix}}
      }
      Rotation Angle {
       bIsDynamic=false
       bIsEditable=false
       bIsStaticSwitch=false
       bIsVisible=false
       type {classStructOrEnum {refPath=&NiagaraFloat}}
      }
      Invert Rotation Quaternion {bIsDynamic=false, bIsEditable=false, bIsStaticSwitch=false, bIsVisible=false, type {classStructOrEnum {refPath=&NiagaraBool}}}
      Additional Quaternion Rotation {bIsDynamic=false, bIsEditable=false, bIsStaticSwitch=false, bIsVisible=false, type {classStructOrEnum {refPath=&NiagaraBool}}}
      Additional Yaw / Pitch / Roll {
       bIsDynamic=false
       bIsEditable=false
       bIsStaticSwitch=false
       bIsVisible=false
       type {classStructOrEnum {refPath=&Vector3f}}
      }
      Rotation Angle Type {
       bIsDynamic=false
       bIsEditable=false
       bIsStaticSwitch=true
       bIsVisible=false
       type {classStructOrEnum {refPath=&ENiagara_AngleInput}}
      }
      Rotation Coordinate Space {
       bIsDynamic=false
       bIsEditable=true
       bIsStaticSwitch=false
       bIsVisible=true
       type {classStructOrEnum {refPath=&ENiagaraCoordinateSpace}}
      }
      Velocity Origin {
       bIsDynamic=false
       bIsEditable=false
       bIsStaticSwitch=false
       bIsVisible=false
       type {classStructOrEnum {refPath=&NiagaraPosition}}
      }
      Origin Offset {
       bIsDynamic=false
       bIsEditable=false
       bIsStaticSwitch=false
       bIsVisible=false
       type {classStructOrEnum {refPath=&Vector3f}}
      }
      Origin Offset Coordinate Space {
       bIsDynamic=false
       bIsEditable=false
       bIsStaticSwitch=false
       bIsVisible=false
       type {classStructOrEnum {refPath=&ENiagaraCoordinateSpace}}
      }
      Default Position {
       bIsDynamic=false
       bIsEditable=false
       bIsStaticSwitch=false
       bIsVisible=false
       type {classStructOrEnum {refPath=&NiagaraPosition}}
      }
     ]
     moduleScript {refPath=&AddVelocity}
    }
   ]
  }
  particleUpdateScript {
   scriptName=ParticleUpdateScript
   modules [
    {
     bIsSetParametersModule=false
     enabled=true
     moduleName=ParticleState
     inputs [
      Kill Particles When Lifetime Has Elapsed {bIsDynamic=false, bIsEditable=true, bIsStaticSwitch=true, bIsVisible=true, type {classStructOrEnum {refPath=&NiagaraBool}}}
      Loop Particles Lifetime {bIsDynamic=false, bIsEditable=false, bIsStaticSwitch=false, bIsVisible=false, type {classStructOrEnum {refPath=&NiagaraBool}}}
      Let Infinitely Lived Particles Die When Emitter Deactivates {bIsDynamic=false, bIsEditable=false, bIsStaticSwitch=false, bIsVisible=false, type {classStructOrEnum {refPath=&NiagaraBool}}}
      Lifetime {bIsDynamic=false, bIsEditable=true, bIsStaticSwitch=false, bIsVisible=true, type {classStructOrEnum {refPath=&NiagaraFloat}}}
      DeltaTime {bIsDynamic=false, bIsEditable=true, bIsStaticSwitch=false, bIsVisible=true, type {classStructOrEnum {refPath=&NiagaraFloat}}}
     ]
     moduleScript {refPath=&ParticleState}
    }
    {
     bIsSetParametersModule=false
     enabled=true
     moduleName=ScaleMeshSize
     inputs [
      Scale Factor {bIsDynamic=true, bIsEditable=true, bIsStaticSwitch=false, bIsVisible=true, type {classStructOrEnum {refPath=&Vector3f}}}
      Initial Mesh Scale {bIsDynamic=false, bIsEditable=true, bIsStaticSwitch=false, bIsVisible=true, type {classStructOrEnum {refPath=&Vector3f}}}
     ]
     moduleScript {refPath=&ScaleMeshSize}
    }
    {
     bIsSetParametersModule=false
     enabled=true
     moduleName=CurlNoiseForce
     inputs [
      Noise Strength {bIsDynamic=false, bIsEditable=true, bIsStaticSwitch=false, bIsVisible=true, type {classStructOrEnum {refPath=&NiagaraFloat}}}
      Noise Frequency {bIsDynamic=false, bIsEditable=true, bIsStaticSwitch=false, bIsVisible=true, type {classStructOrEnum {refPath=&NiagaraFloat}}}
      Noise Quality / Cost {
       bIsDynamic=false
       bIsEditable=true
       bIsStaticSwitch=true
       bIsVisible=true
       type {classStructOrEnum {refPath=&ENiagaraCurlNoiseQuality}}
      }
      Pan Noise {bIsDynamic=false, bIsEditable=true, bIsStaticSwitch=false, bIsVisible=true, type {classStructOrEnum {refPath=&NiagaraBool}}}
      Pan Noise Field {bIsDynamic=false, bIsEditable=false, bIsStaticSwitch=false, bIsVisible=true, type {classStructOrEnum {refPath=&Vector3f}}}
      Write to Intrinsic Parameters {bIsDynamic=false, bIsEditable=true, bIsStaticSwitch=true, bIsVisible=true, type {classStructOrEnum {refPath=&NiagaraBool}}}
      Sample Position {bIsDynamic=false, bIsEditable=true, bIsStaticSwitch=false, bIsVisible=true, type {classStructOrEnum {refPath=&Vector3f}}}
      Randomize Noise Sample {bIsDynamic=false, bIsEditable=true, bIsStaticSwitch=true, bIsVisible=true, type {classStructOrEnum {refPath=&NiagaraBool}}}
      Random Seed {bIsDynamic=false, bIsEditable=true, bIsStaticSwitch=false, bIsVisible=true, type {classStructOrEnum {refPath=&NiagaraInt32}}}
      Randomization Vector {bIsDynamic=false, bIsEditable=true, bIsStaticSwitch=false, bIsVisible=true, type {classStructOrEnum {refPath=&Vector3f}}}
      Scale Curl Lookup By Engine Owner Scale {bIsDynamic=false, bIsEditable=true, bIsStaticSwitch=true, bIsVisible=true, type {classStructOrEnum {refPath=&NiagaraBool}}}
      CurlNoise {
       bIsDynamic=false
       bIsEditable=false
       bIsStaticSwitch=false
       bIsVisible=false
       type {classStructOrEnum {refPath=&NiagaraDataInterfaceCurlNoise}}
      }
      Mask Curl Noise {bIsDynamic=false, bIsEditable=true, bIsStaticSwitch=true, bIsVisible=true, type {classStructOrEnum {refPath=&NiagaraBool}}}
      Curl Noise Cone Mask Angle {
       bIsDynamic=false
       bIsEditable=false
       bIsStaticSwitch=false
       bIsVisible=false
       type {classStructOrEnum {refPath=&NiagaraFloat}}
      }
      Curl Noise Cone Mask Falloff Angle {
       bIsDynamic=false
       bIsEditable=false
       bIsStaticSwitch=false
       bIsVisible=false
       type {classStructOrEnum {refPath=&NiagaraFloat}}
      }
      Curl Noise Cone Mask Axis {
       bIsDynamic=false
       bIsEditable=false
       bIsStaticSwitch=false
       bIsVisible=false
       type {classStructOrEnum {refPath=&Vector3f}}
      }
     ]
     moduleScript {refPath=&CurlNoiseForce}
    }
    {
     bIsSetParametersModule=false
     enabled=true
     moduleName=Drag
     inputs [
      Use Linear Drag {bIsDynamic=false, bIsEditable=true, bIsStaticSwitch=false, bIsVisible=true, type {classStructOrEnum {refPath=&NiagaraBool}}}
      Drag {bIsDynamic=true, bIsEditable=true, bIsStaticSwitch=false, bIsVisible=true, type {classStructOrEnum {refPath=&NiagaraFloat}}}
      Use Rotational Drag {bIsDynamic=false, bIsEditable=true, bIsStaticSwitch=false, bIsVisible=true, type {classStructOrEnum {refPath=&NiagaraBool}}}
      Rotational Drag {bIsDynamic=false, bIsEditable=false, bIsStaticSwitch=false, bIsVisible=true, type {classStructOrEnum {refPath=&NiagaraFloat}}}
     ]
     moduleScript {refPath=&Drag}
    }
    {
     bIsSetParametersModule=false
     enabled=true
     moduleName=GravityForce
     inputs [
      Gravity {bIsDynamic=false, bIsEditable=true, bIsStaticSwitch=false, bIsVisible=true, type {classStructOrEnum {refPath=&Vector3f}}}
      Coordinate Space {
       bIsDynamic=false
       bIsEditable=true
       bIsStaticSwitch=true
       bIsVisible=true
       type {classStructOrEnum {refPath=&ENiagaraCoordinateSpace}}
      }
     ]
     moduleScript {refPath=&GravityForce}
    }
    {
     bIsSetParametersModule=false
     enabled=true
     moduleName=SolveForcesAndVelocity
     inputs [
      Clamp Velocity {bIsDynamic=false, bIsEditable=true, bIsStaticSwitch=false, bIsVisible=true, type {classStructOrEnum {refPath=&NiagaraBool}}}
      Speed Limit {bIsDynamic=false, bIsEditable=false, bIsStaticSwitch=false, bIsVisible=true, type {classStructOrEnum {refPath=&NiagaraFloat}}}
      Limit Acceleration {bIsDynamic=false, bIsEditable=true, bIsStaticSwitch=false, bIsVisible=true, type {classStructOrEnum {refPath=&NiagaraBool}}}
      Acceleration Limit {bIsDynamic=false, bIsEditable=false, bIsStaticSwitch=false, bIsVisible=true, type {classStructOrEnum {refPath=&NiagaraFloat}}}
      Manually Enable Rotational Solver {bIsDynamic=false, bIsEditable=true, bIsStaticSwitch=false, bIsVisible=true, type {classStructOrEnum {refPath=&NiagaraBool}}}
      Write to Intrinsic Properties {bIsDynamic=false, bIsEditable=true, bIsStaticSwitch=false, bIsVisible=true, type {classStructOrEnum {refPath=&NiagaraBool}}}
      Write to Presolve Properties {bIsDynamic=false, bIsEditable=true, bIsStaticSwitch=true, bIsVisible=true, type {classStructOrEnum {refPath=&NiagaraBool}}}
      DeltaTime {bIsDynamic=false, bIsEditable=true, bIsStaticSwitch=false, bIsVisible=true, type {classStructOrEnum {refPath=&NiagaraFloat}}}
      Mass {bIsDynamic=false, bIsEditable=true, bIsStaticSwitch=false, bIsVisible=true, type {classStructOrEnum {refPath=&NiagaraFloat}}}
      Position {
       bIsDynamic=false
       bIsEditable=true
       bIsStaticSwitch=false
       bIsVisible=true
       type {classStructOrEnum {refPath=&NiagaraPosition}}
      }
      Velocity {bIsDynamic=false, bIsEditable=true, bIsStaticSwitch=false, bIsVisible=true, type {classStructOrEnum {refPath=&Vector3f}}}
      PreviousVelocity {bIsDynamic=false, bIsEditable=true, bIsStaticSwitch=false, bIsVisible=true, type {classStructOrEnum {refPath=&Vector3f}}}
      Force {bIsDynamic=false, bIsEditable=true, bIsStaticSwitch=false, bIsVisible=true, type {classStructOrEnum {refPath=&Vector3f}}}
      Mesh Orientation {bIsDynamic=false, bIsEditable=false, bIsStaticSwitch=false, bIsVisible=false, type {classStructOrEnum {refPath=&Quat4f}}}
      Rotational Velocity {
       bIsDynamic=false
       bIsEditable=false
       bIsStaticSwitch=false
       bIsVisible=false
       type {classStructOrEnum {refPath=&Vector3f}}
      }
      Rotational Force {
       bIsDynamic=false
       bIsEditable=false
       bIsStaticSwitch=false
       bIsVisible=false
       type {classStructOrEnum {refPath=&Vector3f}}
      }
      Rotational Inertia {
       bIsDynamic=false
       bIsEditable=false
       bIsStaticSwitch=false
       bIsVisible=false
       type {classStructOrEnum {refPath=&NiagaraFloat}}
      }
     ]
     moduleScript {refPath=&SolveForcesAndVelocity}
    }
    {
     bIsSetParametersModule=false
     enabled=true
     moduleName=SpriteRotationRate
     inputs [
      Rotation Rate {bIsDynamic=true, bIsEditable=true, bIsStaticSwitch=false, bIsVisible=true, type {classStructOrEnum {refPath=&NiagaraFloat}}}
      Delta Time {bIsDynamic=false, bIsEditable=true, bIsStaticSwitch=false, bIsVisible=true, type {classStructOrEnum {refPath=&NiagaraFloat}}}
     ]
     moduleScript {refPath=&SpriteRotationRate}
    }
    {
     bIsSetParametersModule=false
     enabled=true
     moduleName=ScaleSpriteSize
     inputs [
      Scale Factor {bIsDynamic=true, bIsEditable=true, bIsStaticSwitch=false, bIsVisible=true, type {classStructOrEnum {refPath=&Vector2f}}}
      Initial Sprite Size {bIsDynamic=false, bIsEditable=true, bIsStaticSwitch=false, bIsVisible=true, type {classStructOrEnum {refPath=&Vector2f}}}
     ]
     moduleScript {refPath=&ScaleSpriteSize}
    }
    {
     bIsSetParametersModule=false
     enabled=true
     moduleName=ScaleColor
     inputs [
      Scale Mode {
       bIsDynamic=false
       bIsEditable=true
       bIsStaticSwitch=true
       bIsVisible=true
       type {classStructOrEnum {refPath=&ENiagaraScaleColorMode}}
      }
      Linear Color Curve {
       bIsDynamic=false
       bIsEditable=false
       bIsStaticSwitch=false
       bIsVisible=false
       type {classStructOrEnum {refPath=&NiagaraDataInterfaceColorCurve}}
      }
      ScaleRGB {bIsDynamic=false, bIsEditable=true, bIsStaticSwitch=false, bIsVisible=true, type {classStructOrEnum {refPath=&NiagaraBool}}}
      Scale RGB {bIsDynamic=false, bIsEditable=true, bIsStaticSwitch=false, bIsVisible=true, type {classStructOrEnum {refPath=&Vector3f}}}
      ScaleRGBA {bIsDynamic=false, bIsEditable=false, bIsStaticSwitch=false, bIsVisible=false, type {classStructOrEnum {refPath=&NiagaraBool}}}
      Scale RGBA {
       bIsDynamic=false
       bIsEditable=false
       bIsStaticSwitch=false
       bIsVisible=false
       type {classStructOrEnum {refPath=&Vector4f}}
      }
      Curve Index {
       bIsDynamic=false
       bIsEditable=false
       bIsStaticSwitch=false
       bIsVisible=false
       type {classStructOrEnum {refPath=&NiagaraFloat}}
      }
      ScaleA {bIsDynamic=false, bIsEditable=true, bIsStaticSwitch=false, bIsVisible=true, type {classStructOrEnum {refPath=&NiagaraBool}}}
      Scale Alpha {bIsDynamic=true, bIsEditable=true, bIsStaticSwitch=false, bIsVisible=true, type {classStructOrEnum {refPath=&NiagaraFloat}}}
      Color Value To Scale {
       bIsDynamic=false
       bIsEditable=true
       bIsStaticSwitch=false
       bIsVisible=true
       type {classStructOrEnum {refPath=&LinearColor}}
      }
     ]
     moduleScript {refPath=&ScaleColor}
    }
    {
     bIsSetParametersModule=false
     enabled=true
     moduleName=DynamicMaterialParameters
     inputs [
      Write Parameter Index 0 {bIsDynamic=false, bIsEditable=true, bIsStaticSwitch=true, bIsVisible=true, type {classStructOrEnum {refPath=&NiagaraBool}}}
      Module.Param0WriteEnabled {bIsDynamic=false, bIsEditable=true, bIsStaticSwitch=false, bIsVisible=true, type {classStructOrEnum {refPath=&NiagaraBool}}}
      Index 0 Param 1 {bIsDynamic=false, bIsEditable=true, bIsStaticSwitch=false, bIsVisible=true, type {classStructOrEnum {refPath=&NiagaraFloat}}}
      Module.Param1WriteEnabled {bIsDynamic=false, bIsEditable=true, bIsStaticSwitch=false, bIsVisible=true, type {classStructOrEnum {refPath=&NiagaraBool}}}
      Index 0 Param 2 {bIsDynamic=false, bIsEditable=true, bIsStaticSwitch=false, bIsVisible=true, type {classStructOrEnum {refPath=&NiagaraFloat}}}
      Module.Param2WriteEnabled {bIsDynamic=false, bIsEditable=true, bIsStaticSwitch=false, bIsVisible=true, type {classStructOrEnum {refPath=&NiagaraBool}}}
      Index 0 Param 3 {bIsDynamic=false, bIsEditable=true, bIsStaticSwitch=false, bIsVisible=true, type {classStructOrEnum {refPath=&NiagaraFloat}}}
      Module.Param3WriteEnabled {bIsDynamic=false, bIsEditable=true, bIsStaticSwitch=false, bIsVisible=true, type {classStructOrEnum {refPath=&NiagaraBool}}}
      Index 0 Param 4 {bIsDynamic=false, bIsEditable=true, bIsStaticSwitch=false, bIsVisible=true, type {classStructOrEnum {refPath=&NiagaraFloat}}}
      Write Parameter Index 1 {bIsDynamic=false, bIsEditable=true, bIsStaticSwitch=true, bIsVisible=true, type {classStructOrEnum {refPath=&NiagaraBool}}}
      Index 1 Param 1 {
       bIsDynamic=false
       bIsEditable=false
       bIsStaticSwitch=false
       bIsVisible=false
       type {classStructOrEnum {refPath=&NiagaraFloat}}
      }
      Index 1 Param 2 {
       bIsDynamic=false
       bIsEditable=false
       bIsStaticSwitch=false
       bIsVisible=false
       type {classStructOrEnum {refPath=&NiagaraFloat}}
      }
      Index 1 Param 3 {
       bIsDynamic=false
       bIsEditable=false
       bIsStaticSwitch=false
       bIsVisible=false
       type {classStructOrEnum {refPath=&NiagaraFloat}}
      }
      Index 1 Param 4 {
       bIsDynamic=false
       bIsEditable=false
       bIsStaticSwitch=false
       bIsVisible=false
       type {classStructOrEnum {refPath=&NiagaraFloat}}
      }
      Write Parameter Index 2 {bIsDynamic=false, bIsEditable=true, bIsStaticSwitch=true, bIsVisible=true, type {classStructOrEnum {refPath=&NiagaraBool}}}
      Index 2 Param 1 {
       bIsDynamic=false
       bIsEditable=false
       bIsStaticSwitch=false
       bIsVisible=false
       type {classStructOrEnum {refPath=&NiagaraFloat}}
      }
      Index 2 Param 2 {
       bIsDynamic=false
       bIsEditable=false
       bIsStaticSwitch=false
       bIsVisible=false
       type {classStructOrEnum {refPath=&NiagaraFloat}}
      }
      Index 2 Param 3 {
       bIsDynamic=false
       bIsEditable=false
       bIsStaticSwitch=false
       bIsVisible=false
       type {classStructOrEnum {refPath=&NiagaraFloat}}
      }
      Index 2 Param 4 {
       bIsDynamic=false
       bIsEditable=false
       bIsStaticSwitch=false
       bIsVisible=false
       type {classStructOrEnum {refPath=&NiagaraFloat}}
      }
      Write Parameter Index 3 {bIsDynamic=false, bIsEditable=true, bIsStaticSwitch=true, bIsVisible=true, type {classStructOrEnum {refPath=&NiagaraBool}}}
      Index 3 Param 1 {
       bIsDynamic=false
       bIsEditable=false
       bIsStaticSwitch=false
       bIsVisible=false
       type {classStructOrEnum {refPath=&NiagaraFloat}}
      }
      Index 3 Param 2 {
       bIsDynamic=false
       bIsEditable=false
       bIsStaticSwitch=false
       bIsVisible=false
       type {classStructOrEnum {refPath=&NiagaraFloat}}
      }
      Index 3 Param 3 {
       bIsDynamic=false
       bIsEditable=false
       bIsStaticSwitch=false
       bIsVisible=false
       type {classStructOrEnum {refPath=&NiagaraFloat}}
      }
      Index 3 Param 4 {
       bIsDynamic=false
       bIsEditable=false
       bIsStaticSwitch=false
       bIsVisible=false
       type {classStructOrEnum {refPath=&NiagaraFloat}}
      }
     ]
     moduleScript {refPath=&DynamicMaterialParameters}
    }
    {
     bIsSetParametersModule=false
     enabled=false
     moduleName=ScaleColor001
     inputs [
      Scale Mode {bIsDynamic=false, bIsEditable=true, bIsStaticSwitch=true, bIsVisible=true, type {classStructOrEnum {refPath=&NiagaraInt32}}}
      Linear Color Curve {
       bIsDynamic=false
       bIsEditable=false
       bIsStaticSwitch=false
       bIsVisible=false
       type {classStructOrEnum {refPath=&NiagaraDataInterfaceColorCurve}}
      }
      ScaleRGB {bIsDynamic=false, bIsEditable=true, bIsStaticSwitch=false, bIsVisible=true, type {classStructOrEnum {refPath=&NiagaraBool}}}
      Scale RGB {bIsDynamic=true, bIsEditable=true, bIsStaticSwitch=false, bIsVisible=true, type {classStructOrEnum {refPath=&Vector3f}}}
      ScaleRGBA {bIsDynamic=false, bIsEditable=false, bIsStaticSwitch=false, bIsVisible=false, type {classStructOrEnum {refPath=&NiagaraBool}}}
      Scale RGBA {
       bIsDynamic=false
       bIsEditable=false
       bIsStaticSwitch=false
       bIsVisible=false
       type {classStructOrEnum {refPath=&Vector4f}}
      }
      Curve Index {
       bIsDynamic=false
       bIsEditable=false
       bIsStaticSwitch=false
       bIsVisible=false
       type {classStructOrEnum {refPath=&NiagaraFloat}}
      }
      ScaleA {bIsDynamic=false, bIsEditable=true, bIsStaticSwitch=false, bIsVisible=true, type {classStructOrEnum {refPath=&NiagaraBool}}}
      Scale Alpha {bIsDynamic=false, bIsEditable=false, bIsStaticSwitch=false, bIsVisible=true, type {classStructOrEnum {refPath=&NiagaraFloat}}}
      Color Value To Scale {
       bIsDynamic=false
       bIsEditable=true
       bIsStaticSwitch=false
       bIsVisible=true
       type {classStructOrEnum {refPath=&LinearColor}}
      }
      Color Scale {
       bIsDynamic=false
       bIsEditable=false
       bIsStaticSwitch=false
       bIsVisible=false
       type {classStructOrEnum {refPath=&LinearColor}}
      }
     ]
     moduleScript {refPath=&ScaleColor}
    }
    {
     bIsSetParametersModule=false
     enabled=true
     moduleName=ScaleColor002
     inputs [
      Scale Mode {bIsDynamic=false, bIsEditable=true, bIsStaticSwitch=true, bIsVisible=true, type {classStructOrEnum {refPath=&NiagaraInt32}}}
      Linear Color Curve {
       bIsDynamic=false
       bIsEditable=false
       bIsStaticSwitch=false
       bIsVisible=false
       type {classStructOrEnum {refPath=&NiagaraDataInterfaceColorCurve}}
      }
      ScaleRGB {bIsDynamic=false, bIsEditable=true, bIsStaticSwitch=false, bIsVisible=true, type {classStructOrEnum {refPath=&NiagaraBool}}}
      Scale RGB {bIsDynamic=true, bIsEditable=true, bIsStaticSwitch=false, bIsVisible=true, type {classStructOrEnum {refPath=&Vector3f}}}
      ScaleRGBA {bIsDynamic=false, bIsEditable=false, bIsStaticSwitch=false, bIsVisible=false, type {classStructOrEnum {refPath=&NiagaraBool}}}
      Scale RGBA {
       bIsDynamic=false
       bIsEditable=false
       bIsStaticSwitch=false
       bIsVisible=false
       type {classStructOrEnum {refPath=&Vector4f}}
      }
      Curve Index {
       bIsDynamic=false
       bIsEditable=false
       bIsStaticSwitch=false
       bIsVisible=false
       type {classStructOrEnum {refPath=&NiagaraFloat}}
      }
      ScaleA {bIsDynamic=false, bIsEditable=true, bIsStaticSwitch=false, bIsVisible=true, type {classStructOrEnum {refPath=&NiagaraBool}}}
      Scale Alpha {bIsDynamic=false, bIsEditable=false, bIsStaticSwitch=false, bIsVisible=true, type {classStructOrEnum {refPath=&NiagaraFloat}}}
      Color Value To Scale {
       bIsDynamic=false
       bIsEditable=true
       bIsStaticSwitch=false
       bIsVisible=true
       type {classStructOrEnum {refPath=&LinearColor}}
      }
      Color Scale {
       bIsDynamic=false
       bIsEditable=false
       bIsStaticSwitch=false
       bIsVisible=false
       type {classStructOrEnum {refPath=&LinearColor}}
      }
     ]
     moduleScript {refPath=&ScaleColor}
    }
   ]
  }
  rendererClasses [{refPath=&NiagaraSpriteRendererProperties}, {refPath=&NiagaraMeshRendererProperties}]
  renderers [
   {rendererIndex=0, rendererClass {refPath=&NiagaraSpriteRendererProperties}}
   {rendererIndex=1, rendererClass {refPath=&NiagaraMeshRendererProperties}}
  ]
 }
 Snowflakedsff002 {
  emitterSpawnScript {scriptName=EmitterSpawnScript, modules []}
  emitterUpdateScript {
   scriptName=EmitterUpdateScript
   modules [
    {
     bIsSetParametersModule=false
     enabled=true
     moduleName=EmitterState
     inputs [
      Life Cycle Mode {
       bIsDynamic=false
       bIsEditable=true
       bIsStaticSwitch=true
       bIsVisible=true
       type {classStructOrEnum {refPath=&ENiagaraEmitterLifeCycleMode}}
      }
      Inactive Response {
       bIsDynamic=false
       bIsEditable=true
       bIsStaticSwitch=true
       bIsVisible=true
       type {classStructOrEnum {refPath=&ENiagaraInactiveMode}}
      }
      Loop Behavior {
       bIsDynamic=false
       bIsEditable=true
       bIsStaticSwitch=true
       bIsVisible=true
       type {classStructOrEnum {refPath=&ENiagara_EmitterStateOptions}}
      }
      Loop Duration Mode {
       bIsDynamic=false
       bIsEditable=true
       bIsStaticSwitch=true
       bIsVisible=true
       type {classStructOrEnum {refPath=&ENiagara_InfiniteLoopDuration}}
      }
      Loop Count {
       bIsDynamic=false
       bIsEditable=false
       bIsStaticSwitch=false
       bIsVisible=false
       type {classStructOrEnum {refPath=&NiagaraInt32}}
      }
      Loop Duration {bIsDynamic=false, bIsEditable=true, bIsStaticSwitch=false, bIsVisible=true, type {classStructOrEnum {refPath=&NiagaraFloat}}}
      Recalculate Duration Each Loop {bIsDynamic=false, bIsEditable=false, bIsStaticSwitch=false, bIsVisible=false, type {classStructOrEnum {refPath=&NiagaraBool}}}
      UseLoopDelay {bIsDynamic=false, bIsEditable=true, bIsStaticSwitch=false, bIsVisible=true, type {classStructOrEnum {refPath=&NiagaraBool}}}
      Loop Delay {bIsDynamic=false, bIsEditable=false, bIsStaticSwitch=false, bIsVisible=true, type {classStructOrEnum {refPath=&NiagaraFloat}}}
      Delay First Loop Only {bIsDynamic=false, bIsEditable=false, bIsStaticSwitch=false, bIsVisible=false, type {classStructOrEnum {refPath=&NiagaraBool}}}
      Scalability Mode {
       bIsDynamic=false
       bIsEditable=true
       bIsStaticSwitch=true
       bIsVisible=true
       type {classStructOrEnum {refPath=&ENiagaraEmitterScalabilityMode_Limited}}
      }
      Enable Distance Culling {bIsDynamic=false, bIsEditable=false, bIsStaticSwitch=true, bIsVisible=false, type {classStructOrEnum {refPath=&NiagaraBool}}}
      MinDistance {
       bIsDynamic=false
       bIsEditable=false
       bIsStaticSwitch=false
       bIsVisible=false
       type {classStructOrEnum {refPath=&NiagaraFloat}}
      }
      Min Distance Response {
       bIsDynamic=false
       bIsEditable=false
       bIsStaticSwitch=false
       bIsVisible=false
       type {classStructOrEnum {refPath=&ENiagaraExecutionStateManagement}}
      }
      MaxDistance {
       bIsDynamic=false
       bIsEditable=false
       bIsStaticSwitch=false
       bIsVisible=false
       type {classStructOrEnum {refPath=&NiagaraFloat}}
      }
      Max Distance Response {
       bIsDynamic=false
       bIsEditable=false
       bIsStaticSwitch=false
       bIsVisible=false
       type {classStructOrEnum {refPath=&ENiagaraExecutionStateManagement}}
      }
      Scale Spawn Count {bIsDynamic=false, bIsEditable=false, bIsStaticSwitch=true, bIsVisible=false, type {classStructOrEnum {refPath=&NiagaraBool}}}
      Spawn Count Scale {
       bIsDynamic=false
       bIsEditable=false
       bIsStaticSwitch=false
       bIsVisible=false
       type {classStructOrEnum {refPath=&NiagaraFloat}}
      }
      Spawn Count Scale By Distance {
       bIsDynamic=false
       bIsEditable=false
       bIsStaticSwitch=false
       bIsVisible=false
       type {classStructOrEnum {refPath=&NiagaraDataInterfaceCurve}}
      }
      Enable Visibility Culling {bIsDynamic=false, bIsEditable=false, bIsStaticSwitch=true, bIsVisible=false, type {classStructOrEnum {refPath=&NiagaraBool}}}
      Visibility Cull Response {
       bIsDynamic=false
       bIsEditable=false
       bIsStaticSwitch=false
       bIsVisible=false
       type {classStructOrEnum {refPath=&ENiagaraExecutionStateManagement}}
      }
      Visibility Cull Delay {
       bIsDynamic=false
       bIsEditable=false
       bIsStaticSwitch=false
       bIsVisible=false
       type {classStructOrEnum {refPath=&NiagaraFloat}}
      }
      Reset Age On Awaken {bIsDynamic=false, bIsEditable=false, bIsStaticSwitch=false, bIsVisible=false, type {classStructOrEnum {refPath=&NiagaraBool}}}
     ]
     moduleScript {refPath=&EmitterState}
    }
    {
     bIsSetParametersModule=false
     enabled=true
     moduleName=SpawnRate
     inputs [
      SpawnRate {bIsDynamic=false, bIsEditable=true, bIsStaticSwitch=false, bIsVisible=true, type {classStructOrEnum {refPath=&NiagaraFloat}}}
      Use Spawn Probability {bIsDynamic=false, bIsEditable=true, bIsStaticSwitch=false, bIsVisible=true, type {classStructOrEnum {refPath=&NiagaraBool}}}
      Spawn Probability {bIsDynamic=false, bIsEditable=false, bIsStaticSwitch=false, bIsVisible=true, type {classStructOrEnum {refPath=&NiagaraFloat}}}
      SpawnGroup {bIsDynamic=false, bIsEditable=true, bIsStaticSwitch=false, bIsVisible=true, type {classStructOrEnum {refPath=&NiagaraInt32}}}
     ]
     moduleScript {refPath=&SpawnRate}
    }
    {
     bIsSetParametersModule=false
     enabled=true
     moduleName=SpawnBurst_Instantaneous
     inputs [
      Spawn Count {bIsDynamic=false, bIsEditable=true, bIsStaticSwitch=false, bIsVisible=true, type {classStructOrEnum {refPath=&NiagaraInt32}}}
      Spawn Time {bIsDynamic=false, bIsEditable=true, bIsStaticSwitch=false, bIsVisible=true, type {classStructOrEnum {refPath=&NiagaraFloat}}}
      Use Spawn Probability {bIsDynamic=false, bIsEditable=true, bIsStaticSwitch=false, bIsVisible=true, type {classStructOrEnum {refPath=&NiagaraBool}}}
      Spawn Probability {bIsDynamic=false, bIsEditable=false, bIsStaticSwitch=false, bIsVisible=true, type {classStructOrEnum {refPath=&NiagaraFloat}}}
      UseLoopCountLimit {bIsDynamic=false, bIsEditable=true, bIsStaticSwitch=false, bIsVisible=true, type {classStructOrEnum {refPath=&NiagaraBool}}}
      Loop Count Limit {bIsDynamic=false, bIsEditable=false, bIsStaticSwitch=false, bIsVisible=true, type {classStructOrEnum {refPath=&NiagaraInt32}}}
      Spawn Group {bIsDynamic=false, bIsEditable=true, bIsStaticSwitch=false, bIsVisible=true, type {classStructOrEnum {refPath=&NiagaraInt32}}}
      Age {bIsDynamic=false, bIsEditable=true, bIsStaticSwitch=false, bIsVisible=true, type {classStructOrEnum {refPath=&NiagaraFloat}}}
     ]
     moduleScript {refPath=&SpawnBurst_Instantaneous}
    }
   ]
  }
  particleSpawnScript {
   scriptName=ParticleSpawnScript
   modules [
    {
     bIsSetParametersModule=false
     enabled=true
     moduleName=InitializeParticle001
     inputs [
      Lifetime Mode {
       bIsDynamic=false
       bIsEditable=true
       bIsStaticSwitch=true
       bIsVisible=true
       type {classStructOrEnum {refPath=&ENiagara_LifetimeMode}}
      }
      Lifetime Min {bIsDynamic=false, bIsEditable=true, bIsStaticSwitch=false, bIsVisible=true, type {classStructOrEnum {refPath=&NiagaraFloat}}}
      Lifetime {
       bIsDynamic=false
       bIsEditable=false
       bIsStaticSwitch=false
       bIsVisible=false
       type {classStructOrEnum {refPath=&NiagaraFloat}}
      }
      Lifetime Max {bIsDynamic=false, bIsEditable=true, bIsStaticSwitch=false, bIsVisible=true, type {classStructOrEnum {refPath=&NiagaraFloat}}}
      Lifetime Randomness Mode {
       bIsDynamic=false
       bIsEditable=true
       bIsStaticSwitch=true
       bIsVisible=true
       type {classStructOrEnum {refPath=&ENiagaraRandomnessMode}}
      }
      Color Mode {
       bIsDynamic=false
       bIsEditable=true
       bIsStaticSwitch=true
       bIsVisible=true
       type {classStructOrEnum {refPath=&ENiagara_ColorInitializationMode}}
      }
      Color Minimum {
       bIsDynamic=false
       bIsEditable=false
       bIsStaticSwitch=false
       bIsVisible=false
       type {classStructOrEnum {refPath=&LinearColor}}
      }
      Color {
       bIsDynamic=true
       bIsEditable=true
       bIsStaticSwitch=false
       bIsVisible=true
       type {classStructOrEnum {refPath=&LinearColor}}
      }
      AdjustHue {bIsDynamic=false, bIsEditable=false, bIsStaticSwitch=false, bIsVisible=false, type {classStructOrEnum {refPath=&NiagaraBool}}}
      Hue Shift Range {
       bIsDynamic=false
       bIsEditable=false
       bIsStaticSwitch=false
       bIsVisible=false
       type {classStructOrEnum {refPath=&Vector2f}}
      }
      Color Maximum {
       bIsDynamic=false
       bIsEditable=false
       bIsStaticSwitch=false
       bIsVisible=false
       type {classStructOrEnum {refPath=&LinearColor}}
      }
      AdjustSaturation {bIsDynamic=false, bIsEditable=false, bIsStaticSwitch=false, bIsVisible=false, type {classStructOrEnum {refPath=&NiagaraBool}}}
      Saturation Range {
       bIsDynamic=false
       bIsEditable=false
       bIsStaticSwitch=false
       bIsVisible=false
       type {classStructOrEnum {refPath=&Vector2f}}
      }
      AdjustValue {bIsDynamic=false, bIsEditable=false, bIsStaticSwitch=false, bIsVisible=false, type {classStructOrEnum {refPath=&NiagaraBool}}}
      Value Range {
       bIsDynamic=false
       bIsEditable=false
       bIsStaticSwitch=false
       bIsVisible=false
       type {classStructOrEnum {refPath=&Vector2f}}
      }
      AdjustAlpha {bIsDynamic=false, bIsEditable=false, bIsStaticSwitch=false, bIsVisible=false, type {classStructOrEnum {refPath=&NiagaraBool}}}
      Alpha Scale Range {
       bIsDynamic=false
       bIsEditable=false
       bIsStaticSwitch=false
       bIsVisible=false
       type {classStructOrEnum {refPath=&Vector2f}}
      }
      Color Channel Mode {
       bIsDynamic=false
       bIsEditable=false
       bIsStaticSwitch=true
       bIsVisible=false
       type {classStructOrEnum {refPath=&ENiagaraChannelCorrelation}}
      }
      Color Randomness Mode {
       bIsDynamic=false
       bIsEditable=false
       bIsStaticSwitch=true
       bIsVisible=false
       type {classStructOrEnum {refPath=&ENiagaraRandomnessMode}}
      }
      Position Mode {
       bIsDynamic=false
       bIsEditable=true
       bIsStaticSwitch=true
       bIsVisible=true
       type {classStructOrEnum {refPath=&ENiagara_PositionInitializationMode}}
      }
      Position {
       bIsDynamic=false
       bIsEditable=false
       bIsStaticSwitch=false
       bIsVisible=false
       type {classStructOrEnum {refPath=&NiagaraPosition}}
      }
      UsePositionOffset {bIsDynamic=false, bIsEditable=true, bIsStaticSwitch=false, bIsVisible=true, type {classStructOrEnum {refPath=&NiagaraBool}}}
      Position Offset {bIsDynamic=false, bIsEditable=true, bIsStaticSwitch=false, bIsVisible=true, type {classStructOrEnum {refPath=&Vector3f}}}
      Position Offset Coordinate Space {
       bIsDynamic=false
       bIsEditable=true
       bIsStaticSwitch=false
       bIsVisible=true
       type {classStructOrEnum {refPath=&ENiagaraCoordinateSpace}}
      }
      Mass Mode {
       bIsDynamic=false
       bIsEditable=true
       bIsStaticSwitch=true
       bIsVisible=true
       type {classStructOrEnum {refPath=&ENiagara_MassInitializationMode}}
      }
      Mass Min {
       bIsDynamic=false
       bIsEditable=false
       bIsStaticSwitch=false
       bIsVisible=false
       type {classStructOrEnum {refPath=&NiagaraFloat}}
      }
      Mass {
       bIsDynamic=false
       bIsEditable=false
       bIsStaticSwitch=false
       bIsVisible=false
       type {classStructOrEnum {refPath=&NiagaraFloat}}
      }
      Mass Max {
       bIsDynamic=false
       bIsEditable=false
       bIsStaticSwitch=false
       bIsVisible=false
       type {classStructOrEnum {refPath=&NiagaraFloat}}
      }
      Mass Randomness Mode {
       bIsDynamic=false
       bIsEditable=false
       bIsStaticSwitch=true
       bIsVisible=false
       type {classStructOrEnum {refPath=&ENiagaraRandomnessMode}}
      }
      Write Material Random {bIsDynamic=false, bIsEditable=true, bIsStaticSwitch=false, bIsVisible=true, type {classStructOrEnum {refPath=&NiagaraBool}}}
      Material Random {bIsDynamic=false, bIsEditable=true, bIsStaticSwitch=false, bIsVisible=true, type {classStructOrEnum {refPath=&NiagaraFloat}}}
      Sprite Size Mode {
       bIsDynamic=false
       bIsEditable=true
       bIsStaticSwitch=true
       bIsVisible=true
       type {classStructOrEnum {refPath=&ENiagara_SizeScaleMode}}
      }
      Sprite Size {
       bIsDynamic=false
       bIsEditable=false
       bIsStaticSwitch=false
       bIsVisible=false
       type {classStructOrEnum {refPath=&Vector2f}}
      }
      Uniform Sprite Size {bIsDynamic=true, bIsEditable=true, bIsStaticSwitch=false, bIsVisible=true, type {classStructOrEnum {refPath=&NiagaraFloat}}}
      Uniform Sprite Size Min {
       bIsDynamic=false
       bIsEditable=false
       bIsStaticSwitch=false
       bIsVisible=false
       type {classStructOrEnum {refPath=&NiagaraFloat}}
      }
      Sprite Size Min {
       bIsDynamic=false
       bIsEditable=false
       bIsStaticSwitch=false
       bIsVisible=false
       type {classStructOrEnum {refPath=&Vector2f}}
      }
      Sprite Size Max {
       bIsDynamic=false
       bIsEditable=false
       bIsStaticSwitch=false
       bIsVisible=false
       type {classStructOrEnum {refPath=&Vector2f}}
      }
      Uniform Sprite Size Max {
       bIsDynamic=false
       bIsEditable=false
       bIsStaticSwitch=false
       bIsVisible=false
       type {classStructOrEnum {refPath=&NiagaraFloat}}
      }
      Sprite Size Randomness Mode {
       bIsDynamic=false
       bIsEditable=false
       bIsStaticSwitch=true
       bIsVisible=false
       type {classStructOrEnum {refPath=&ENiagaraRandomnessMode}}
      }
      Sprite Rotation Mode {
       bIsDynamic=false
       bIsEditable=true
       bIsStaticSwitch=true
       bIsVisible=true
       type {classStructOrEnum {refPath=&ENiagara_SpriteRotationMode}}
      }
      Sprite Rotation Angle Min {
       bIsDynamic=false
       bIsEditable=false
       bIsStaticSwitch=false
       bIsVisible=false
       type {classStructOrEnum {refPath=&NiagaraFloat}}
      }
      Sprite Rotation Angle {
       bIsDynamic=false
       bIsEditable=false
       bIsStaticSwitch=false
       bIsVisible=false
       type {classStructOrEnum {refPath=&NiagaraFloat}}
      }
      Sprite Rotation Angle Max {
       bIsDynamic=false
       bIsEditable=false
       bIsStaticSwitch=false
       bIsVisible=false
       type {classStructOrEnum {refPath=&NiagaraFloat}}
      }
      Sprite Rotation Randomness Mode {
       bIsDynamic=false
       bIsEditable=false
       bIsStaticSwitch=true
       bIsVisible=false
       type {classStructOrEnum {refPath=&ENiagaraRandomnessMode}}
      }
      Sprite UV Mode {
       bIsDynamic=false
       bIsEditable=true
       bIsStaticSwitch=true
       bIsVisible=true
       type {classStructOrEnum {refPath=&ENiagara_UVFlippingMode}}
      }
      Sprite UV Randomness Mode {
       bIsDynamic=false
       bIsEditable=false
       bIsStaticSwitch=true
       bIsVisible=false
       type {classStructOrEnum {refPath=&ENiagaraRandomnessMode}}
      }
      Sprite UV Scale {
       bIsDynamic=false
       bIsEditable=false
       bIsStaticSwitch=false
       bIsVisible=false
       type {classStructOrEnum {refPath=&Vector2f}}
      }
      Mesh Scale Mode {
       bIsDynamic=false
       bIsEditable=true
       bIsStaticSwitch=true
       bIsVisible=true
       type {classStructOrEnum {refPath=&ENiagara_SizeScaleMode}}
      }
      Mesh Uniform Scale Min {
       bIsDynamic=false
       bIsEditable=false
       bIsStaticSwitch=false
       bIsVisible=false
       type {classStructOrEnum {refPath=&NiagaraFloat}}
      }
      Mesh Uniform Scale {
       bIsDynamic=false
       bIsEditable=false
       bIsStaticSwitch=false
       bIsVisible=false
       type {classStructOrEnum {refPath=&NiagaraFloat}}
      }
      Mesh Scale Min {bIsDynamic=false, bIsEditable=true, bIsStaticSwitch=false, bIsVisible=true, type {classStructOrEnum {refPath=&Vector3f}}}
      Mesh Scale {
       bIsDynamic=false
       bIsEditable=false
       bIsStaticSwitch=false
       bIsVisible=false
       type {classStructOrEnum {refPath=&Vector3f}}
      }
      Mesh Scale Max {bIsDynamic=false, bIsEditable=true, bIsStaticSwitch=false, bIsVisible=true, type {classStructOrEnum {refPath=&Vector3f}}}
      Mesh Uniform Scale Max {
       bIsDynamic=false
       bIsEditable=false
       bIsStaticSwitch=false
       bIsVisible=false
       type {classStructOrEnum {refPath=&NiagaraFloat}}
      }
      Mesh Scale Randomness Mode {
       bIsDynamic=false
       bIsEditable=true
       bIsStaticSwitch=true
       bIsVisible=true
       type {classStructOrEnum {refPath=&ENiagaraRandomnessMode}}
      }
      Mesh Renderer Array Visibility Mode {
       bIsDynamic=false
       bIsEditable=true
       bIsStaticSwitch=true
       bIsVisible=true
       type {classStructOrEnum {refPath=&ENiagara_UnsetDirectSetRandom}}
      }
      Mesh Renderer Mesh Index {
       bIsDynamic=false
       bIsEditable=false
       bIsStaticSwitch=false
       bIsVisible=false
       type {classStructOrEnum {refPath=&NiagaraInt32}}
      }
      Wrap Index By Mesh Count {bIsDynamic=false, bIsEditable=false, bIsStaticSwitch=true, bIsVisible=false, type {classStructOrEnum {refPath=&NiagaraBool}}}
      Mesh Renderer Info {
       bIsDynamic=false
       bIsEditable=false
       bIsStaticSwitch=false
       bIsVisible=false
       type {classStructOrEnum {refPath=&NiagaraDataInterfaceMeshRendererInfo}}
      }
      Mesh Renderer Random Seed {
       bIsDynamic=false
       bIsEditable=false
       bIsStaticSwitch=false
       bIsVisible=false
       type {classStructOrEnum {refPath=&NiagaraInt32}}
      }
      Ribbon Width Mode {
       bIsDynamic=false
       bIsEditable=true
       bIsStaticSwitch=true
       bIsVisible=true
       type {classStructOrEnum {refPath=&ENiagara_UnsetDirectSet}}
      }
      Ribbon Width {
       bIsDynamic=false
       bIsEditable=false
       bIsStaticSwitch=false
       bIsVisible=false
       type {classStructOrEnum {refPath=&NiagaraFloat}}
      }
      Ribbon Facing Mode {
       bIsDynamic=false
       bIsEditable=true
       bIsStaticSwitch=true
       bIsVisible=true
       type {classStructOrEnum {refPath=&ENiagara_UnsetDirectSet}}
      }
      Ribbon Facing Vector {
       bIsDynamic=false
       bIsEditable=false
       bIsStaticSwitch=false
       bIsVisible=false
       type {classStructOrEnum {refPath=&Vector3f}}
      }
      Ribbon Twist Mode {
       bIsDynamic=false
       bIsEditable=true
       bIsStaticSwitch=true
       bIsVisible=true
       type {classStructOrEnum {refPath=&ENiagara_UnsetDirectSet}}
      }
      "Ribbon Twist " {
       bIsDynamic=false
       bIsEditable=false
       bIsStaticSwitch=false
       bIsVisible=false
       type {classStructOrEnum {refPath=&NiagaraFloat}}
      }
     ]
     moduleScript {refPath=&InitializeParticle}
    }
    {
     bIsSetParametersModule=false
     enabled=true
     moduleName=SphereLocation
     inputs [
      Sphere Radius {bIsDynamic=false, bIsEditable=true, bIsStaticSwitch=false, bIsVisible=true, type {classStructOrEnum {refPath=&NiagaraFloat}}}
      UseOffset {bIsDynamic=false, bIsEditable=true, bIsStaticSwitch=false, bIsVisible=true, type {classStructOrEnum {refPath=&NiagaraBool}}}
      Offset {bIsDynamic=false, bIsEditable=true, bIsStaticSwitch=false, bIsVisible=true, type {classStructOrEnum {refPath=&Vector3f}}}
      Offset Coordinate Space {
       bIsDynamic=false
       bIsEditable=true
       bIsStaticSwitch=false
       bIsVisible=true
       type {classStructOrEnum {refPath=&ENiagaraCoordinateSpace}}
      }
      UseNonUniformScale {bIsDynamic=false, bIsEditable=true, bIsStaticSwitch=false, bIsVisible=true, type {classStructOrEnum {refPath=&NiagaraBool}}}
      Non Uniform Scale {bIsDynamic=false, bIsEditable=false, bIsStaticSwitch=false, bIsVisible=true, type {classStructOrEnum {refPath=&Vector3f}}}
      UseSphereOrientation {bIsDynamic=false, bIsEditable=true, bIsStaticSwitch=false, bIsVisible=true, type {classStructOrEnum {refPath=&NiagaraBool}}}
      Sphere Orientation Axis {bIsDynamic=false, bIsEditable=false, bIsStaticSwitch=false, bIsVisible=true, type {classStructOrEnum {refPath=&Vector3f}}}
      Sphere Origin {
       bIsDynamic=false
       bIsEditable=true
       bIsStaticSwitch=false
       bIsVisible=true
       type {classStructOrEnum {refPath=&NiagaraPosition}}
      }
      Sphere Coordinate Space {
       bIsDynamic=false
       bIsEditable=true
       bIsStaticSwitch=false
       bIsVisible=true
       type {classStructOrEnum {refPath=&ENiagaraCoordinateSpace}}
      }
      Write to Intrinsic Variables {bIsDynamic=false, bIsEditable=true, bIsStaticSwitch=true, bIsVisible=true, type {classStructOrEnum {refPath=&NiagaraBool}}}
      Sphere Distribution {
       bIsDynamic=false
       bIsEditable=true
       bIsStaticSwitch=true
       bIsVisible=true
       type {classStructOrEnum {refPath=&ENiagaraSphereDistributionMode}}
      }
      Evaluation Type {
       bIsDynamic=false
       bIsEditable=false
       bIsStaticSwitch=true
       bIsVisible=false
       type {classStructOrEnum {refPath=&ENiagaraRandomnessEvaluation}}
      }
      Hemisphere X {bIsDynamic=false, bIsEditable=true, bIsStaticSwitch=false, bIsVisible=true, type {classStructOrEnum {refPath=&NiagaraBool}}}
      Uniform Distribution {
       bIsDynamic=false
       bIsEditable=false
       bIsStaticSwitch=false
       bIsVisible=false
       type {classStructOrEnum {refPath=&NiagaraFloat}}
      }
      Hemisphere Y {bIsDynamic=false, bIsEditable=true, bIsStaticSwitch=false, bIsVisible=true, type {classStructOrEnum {refPath=&NiagaraBool}}}
      U Position {
       bIsDynamic=false
       bIsEditable=false
       bIsStaticSwitch=false
       bIsVisible=false
       type {classStructOrEnum {refPath=&NiagaraFloat}}
      }
      Uniform Spiral Amount {
       bIsDynamic=false
       bIsEditable=false
       bIsStaticSwitch=false
       bIsVisible=false
       type {classStructOrEnum {refPath=&NiagaraFloat}}
      }
      Hemisphere Z {bIsDynamic=false, bIsEditable=true, bIsStaticSwitch=false, bIsVisible=true, type {classStructOrEnum {refPath=&NiagaraBool}}}
      Uniform Endpoint Offset {
       bIsDynamic=false
       bIsEditable=false
       bIsStaticSwitch=false
       bIsVisible=false
       type {classStructOrEnum {refPath=&NiagaraFloat}}
      }
      V Position {
       bIsDynamic=false
       bIsEditable=false
       bIsStaticSwitch=false
       bIsVisible=false
       type {classStructOrEnum {refPath=&NiagaraFloat}}
      }
      Radius Position {
       bIsDynamic=false
       bIsEditable=false
       bIsStaticSwitch=false
       bIsVisible=false
       type {classStructOrEnum {refPath=&NiagaraFloat}}
      }
      Surface Only {bIsDynamic=false, bIsEditable=true, bIsStaticSwitch=false, bIsVisible=true, type {classStructOrEnum {refPath=&NiagaraBool}}}
      Surface Only Band Thickness {bIsDynamic=false, bIsEditable=true, bIsStaticSwitch=false, bIsVisible=true, type {classStructOrEnum {refPath=&NiagaraFloat}}}
      Surface Expansion Mode {
       bIsDynamic=false
       bIsEditable=true
       bIsStaticSwitch=false
       bIsVisible=true
       type {classStructOrEnum {refPath=&ENiagaraExpansionMode}}
      }
      Normalize Thickness {bIsDynamic=false, bIsEditable=true, bIsStaticSwitch=false, bIsVisible=true, type {classStructOrEnum {refPath=&NiagaraBool}}}
      Randomness Mode {
       bIsDynamic=false
       bIsEditable=true
       bIsStaticSwitch=true
       bIsVisible=true
       type {classStructOrEnum {refPath=&ENiagaraRandomnessMode}}
      }
      Override Seed {bIsDynamic=false, bIsEditable=true, bIsStaticSwitch=false, bIsVisible=true, type {classStructOrEnum {refPath=&NiagaraBool}}}
      Random Seed {bIsDynamic=false, bIsEditable=false, bIsStaticSwitch=false, bIsVisible=true, type {classStructOrEnum {refPath=&NiagaraInt32}}}
      Fixed Random Seed {bIsDynamic=false, bIsEditable=false, bIsStaticSwitch=true, bIsVisible=false, type {classStructOrEnum {refPath=&NiagaraBool}}}
      MaskSpawnGroup {bIsDynamic=false, bIsEditable=true, bIsStaticSwitch=false, bIsVisible=true, type {classStructOrEnum {refPath=&NiagaraBool}}}
      Spawn Group Mask {bIsDynamic=false, bIsEditable=false, bIsStaticSwitch=false, bIsVisible=true, type {classStructOrEnum {refPath=&NiagaraInt32}}}
     ]
     moduleScript {refPath=&SphereLocation}
    }
    {
     bIsSetParametersModule=false
     enabled=false
     moduleName=AddVelocityFromPoint
     inputs [
      Velocity Strength {bIsDynamic=true, bIsEditable=true, bIsStaticSwitch=false, bIsVisible=true, type {classStructOrEnum {refPath=&NiagaraFloat}}}
      Origin Offset {bIsDynamic=false, bIsEditable=true, bIsStaticSwitch=false, bIsVisible=true, type {classStructOrEnum {refPath=&Vector3f}}}
      Origin Offset Coordinate Space {
       bIsDynamic=false
       bIsEditable=true
       bIsStaticSwitch=false
       bIsVisible=true
       type {classStructOrEnum {refPath=&ENiagaraCoordinateSpace}}
      }
      Default Position {bIsDynamic=false, bIsEditable=true, bIsStaticSwitch=false, bIsVisible=true, type {classStructOrEnum {refPath=&Vector3f}}}
      Module.Use Falloff Distance {bIsDynamic=false, bIsEditable=true, bIsStaticSwitch=false, bIsVisible=true, type {classStructOrEnum {refPath=&NiagaraBool}}}
      Velocity Falloff Distance {bIsDynamic=false, bIsEditable=false, bIsStaticSwitch=false, bIsVisible=true, type {classStructOrEnum {refPath=&NiagaraFloat}}}
      Invert Velocity Falloff {bIsDynamic=false, bIsEditable=false, bIsStaticSwitch=false, bIsVisible=false, type {classStructOrEnum {refPath=&NiagaraBool}}}
      Velocity Origin {bIsDynamic=false, bIsEditable=true, bIsStaticSwitch=false, bIsVisible=true, type {classStructOrEnum {refPath=&Vector3f}}}
     ]
     moduleScript {refPath=&AddVelocityFromPoint}
    }
    {
     bIsSetParametersModule=false
     enabled=true
     moduleName=AddVelocity
     inputs [
      Velocity Mode {
       bIsDynamic=false
       bIsEditable=true
       bIsStaticSwitch=true
       bIsVisible=true
       type {classStructOrEnum {refPath=&ENiagara_VelocityMode}}
      }
      Velocity {
       bIsDynamic=false
       bIsEditable=false
       bIsStaticSwitch=false
       bIsVisible=false
       type {classStructOrEnum {refPath=&Vector3f}}
      }
      Velocity Speed {bIsDynamic=true, bIsEditable=true, bIsStaticSwitch=false, bIsVisible=true, type {classStructOrEnum {refPath=&NiagaraFloat}}}
      Velocity Speed Scale {
       bIsDynamic=false
       bIsEditable=false
       bIsStaticSwitch=false
       bIsVisible=false
       type {classStructOrEnum {refPath=&NiagaraFloat}}
      }
      Constrain To Radius {
       bIsDynamic=false
       bIsEditable=false
       bIsStaticSwitch=false
       bIsVisible=false
       type {classStructOrEnum {refPath=&NiagaraFloat}}
      }
      Module.UseVelDistribution {bIsDynamic=false, bIsEditable=true, bIsStaticSwitch=false, bIsVisible=true, type {classStructOrEnum {refPath=&NiagaraBool}}}
      Distribution Along Cone Axis {bIsDynamic=false, bIsEditable=false, bIsStaticSwitch=false, bIsVisible=true, type {classStructOrEnum {refPath=&NiagaraFloat}}}
      Radius Falloff Near / Far {
       bIsDynamic=false
       bIsEditable=false
       bIsStaticSwitch=false
       bIsVisible=false
       type {classStructOrEnum {refPath=&Vector2f}}
      }
      Radius Falloff Exponent {
       bIsDynamic=false
       bIsEditable=false
       bIsStaticSwitch=false
       bIsVisible=false
       type {classStructOrEnum {refPath=&NiagaraFloat}}
      }
      Module.UseConeFalloff {bIsDynamic=false, bIsEditable=true, bIsStaticSwitch=false, bIsVisible=true, type {classStructOrEnum {refPath=&NiagaraBool}}}
      Speed Falloff From Cone Axis {bIsDynamic=false, bIsEditable=false, bIsStaticSwitch=false, bIsVisible=true, type {classStructOrEnum {refPath=&NiagaraFloat}}}
      Invert Velocity Falloff {bIsDynamic=false, bIsEditable=false, bIsStaticSwitch=false, bIsVisible=false, type {classStructOrEnum {refPath=&NiagaraBool}}}
      Override Seed {bIsDynamic=false, bIsEditable=true, bIsStaticSwitch=false, bIsVisible=true, type {classStructOrEnum {refPath=&NiagaraBool}}}
      Random Seed {bIsDynamic=false, bIsEditable=false, bIsStaticSwitch=false, bIsVisible=true, type {classStructOrEnum {refPath=&NiagaraInt32}}}
      Fixed Random Seed {bIsDynamic=false, bIsEditable=false, bIsStaticSwitch=true, bIsVisible=false, type {classStructOrEnum {refPath=&NiagaraBool}}}
      Randomness Mode {
       bIsDynamic=false
       bIsEditable=true
       bIsStaticSwitch=true
       bIsVisible=true
       type {classStructOrEnum {refPath=&ENiagaraRandomnessMode}}
      }
      Cone Axis {bIsDynamic=false, bIsEditable=true, bIsStaticSwitch=false, bIsVisible=true, type {classStructOrEnum {refPath=&Vector3f}}}
      Cone Angle {bIsDynamic=false, bIsEditable=true, bIsStaticSwitch=false, bIsVisible=true, type {classStructOrEnum {refPath=&NiagaraFloat}}}
      Inner Cone Angle {bIsDynamic=false, bIsEditable=true, bIsStaticSwitch=false, bIsVisible=true, type {classStructOrEnum {refPath=&NiagaraFloat}}}
      Cone Angle Mode {
       bIsDynamic=false
       bIsEditable=true
       bIsStaticSwitch=true
       bIsVisible=true
       type {classStructOrEnum {refPath=&ENiagara_AngleInput}}
      }
      Rotation Mode {
       bIsDynamic=false
       bIsEditable=true
       bIsStaticSwitch=true
       bIsVisible=true
       type {classStructOrEnum {refPath=&ENiagara_RotationMode}}
      }
      Rotation Axis {
       bIsDynamic=false
       bIsEditable=false
       bIsStaticSwitch=false
       bIsVisible=false
       type {classStructOrEnum {refPath=&Vector3f}}
      }
      Yaw / Pitch / Roll {
       bIsDynamic=false
       bIsEditable=false
       bIsStaticSwitch=false
       bIsVisible=false
       type {classStructOrEnum {refPath=&Vector3f}}
      }
      Rotation Quaternion {bIsDynamic=false, bIsEditable=false, bIsStaticSwitch=false, bIsVisible=false, type {classStructOrEnum {refPath=&Quat4f}}}
      Rotation Matrix {
       bIsDynamic=false
       bIsEditable=false
       bIsStaticSwitch=false
       bIsVisible=false
       type {classStructOrEnum {refPath=&NiagaraMatrix}}
      }
      Rotation Angle {
       bIsDynamic=false
       bIsEditable=false
       bIsStaticSwitch=false
       bIsVisible=false
       type {classStructOrEnum {refPath=&NiagaraFloat}}
      }
      Invert Rotation Quaternion {bIsDynamic=false, bIsEditable=false, bIsStaticSwitch=false, bIsVisible=false, type {classStructOrEnum {refPath=&NiagaraBool}}}
      Additional Quaternion Rotation {bIsDynamic=false, bIsEditable=false, bIsStaticSwitch=false, bIsVisible=false, type {classStructOrEnum {refPath=&NiagaraBool}}}
      Additional Yaw / Pitch / Roll {
       bIsDynamic=false
       bIsEditable=false
       bIsStaticSwitch=false
       bIsVisible=false
       type {classStructOrEnum {refPath=&Vector3f}}
      }
      Rotation Angle Type {
       bIsDynamic=false
       bIsEditable=false
       bIsStaticSwitch=true
       bIsVisible=false
       type {classStructOrEnum {refPath=&ENiagara_AngleInput}}
      }
      Rotation Coordinate Space {
       bIsDynamic=false
       bIsEditable=true
       bIsStaticSwitch=false
       bIsVisible=true
       type {classStructOrEnum {refPath=&ENiagaraCoordinateSpace}}
      }
      Velocity Origin {
       bIsDynamic=false
       bIsEditable=false
       bIsStaticSwitch=false
       bIsVisible=false
       type {classStructOrEnum {refPath=&NiagaraPosition}}
      }
      Origin Offset {
       bIsDynamic=false
       bIsEditable=false
       bIsStaticSwitch=false
       bIsVisible=false
       type {classStructOrEnum {refPath=&Vector3f}}
      }
      Origin Offset Coordinate Space {
       bIsDynamic=false
       bIsEditable=false
       bIsStaticSwitch=false
       bIsVisible=false
       type {classStructOrEnum {refPath=&ENiagaraCoordinateSpace}}
      }
      Default Position {
       bIsDynamic=false
       bIsEditable=false
       bIsStaticSwitch=false
       bIsVisible=false
       type {classStructOrEnum {refPath=&NiagaraPosition}}
      }
     ]
     moduleScript {refPath=&AddVelocity}
    }
   ]
  }
  particleUpdateScript {
   scriptName=ParticleUpdateScript
   modules [
    {
     bIsSetParametersModule=false
     enabled=true
     moduleName=ParticleState
     inputs [
      Kill Particles When Lifetime Has Elapsed {bIsDynamic=false, bIsEditable=true, bIsStaticSwitch=true, bIsVisible=true, type {classStructOrEnum {refPath=&NiagaraBool}}}
      Loop Particles Lifetime {bIsDynamic=false, bIsEditable=false, bIsStaticSwitch=false, bIsVisible=false, type {classStructOrEnum {refPath=&NiagaraBool}}}
      Let Infinitely Lived Particles Die When Emitter Deactivates {bIsDynamic=false, bIsEditable=false, bIsStaticSwitch=false, bIsVisible=false, type {classStructOrEnum {refPath=&NiagaraBool}}}
      Lifetime {bIsDynamic=false, bIsEditable=true, bIsStaticSwitch=false, bIsVisible=true, type {classStructOrEnum {refPath=&NiagaraFloat}}}
      DeltaTime {bIsDynamic=false, bIsEditable=true, bIsStaticSwitch=false, bIsVisible=true, type {classStructOrEnum {refPath=&NiagaraFloat}}}
     ]
     moduleScript {refPath=&ParticleState}
    }
    {
     bIsSetParametersModule=false
     enabled=false
     moduleName=ScaleMeshSize
     inputs [
      Scale Factor {bIsDynamic=true, bIsEditable=true, bIsStaticSwitch=false, bIsVisible=true, type {classStructOrEnum {refPath=&Vector3f}}}
      Initial Mesh Scale {bIsDynamic=false, bIsEditable=true, bIsStaticSwitch=false, bIsVisible=true, type {classStructOrEnum {refPath=&Vector3f}}}
     ]
     moduleScript {refPath=&ScaleMeshSize}
    }
    {
     bIsSetParametersModule=false
     enabled=true
     moduleName=ScaleMeshSize001
     inputs [
      Scale Factor {bIsDynamic=true, bIsEditable=true, bIsStaticSwitch=false, bIsVisible=true, type {classStructOrEnum {refPath=&Vector3f}}}
      Initial Mesh Scale {bIsDynamic=false, bIsEditable=true, bIsStaticSwitch=false, bIsVisible=true, type {classStructOrEnum {refPath=&Vector3f}}}
     ]
     moduleScript {refPath=&ScaleMeshSize}
    }
    {
     bIsSetParametersModule=false
     enabled=false
     moduleName=CurlNoiseForce
     inputs [
      Noise Strength {bIsDynamic=false, bIsEditable=true, bIsStaticSwitch=false, bIsVisible=true, type {classStructOrEnum {refPath=&NiagaraFloat}}}
      Noise Frequency {bIsDynamic=false, bIsEditable=true, bIsStaticSwitch=false, bIsVisible=true, type {classStructOrEnum {refPath=&NiagaraFloat}}}
      Noise Quality / Cost {
       bIsDynamic=false
       bIsEditable=true
       bIsStaticSwitch=true
       bIsVisible=true
       type {classStructOrEnum {refPath=&ENiagaraCurlNoiseQuality}}
      }
      Pan Noise {bIsDynamic=false, bIsEditable=true, bIsStaticSwitch=false, bIsVisible=true, type {classStructOrEnum {refPath=&NiagaraBool}}}
      Pan Noise Field {bIsDynamic=false, bIsEditable=false, bIsStaticSwitch=false, bIsVisible=true, type {classStructOrEnum {refPath=&Vector3f}}}
      Write to Intrinsic Parameters {bIsDynamic=false, bIsEditable=true, bIsStaticSwitch=true, bIsVisible=true, type {classStructOrEnum {refPath=&NiagaraBool}}}
      Sample Position {bIsDynamic=false, bIsEditable=true, bIsStaticSwitch=false, bIsVisible=true, type {classStructOrEnum {refPath=&Vector3f}}}
      Randomize Noise Sample {bIsDynamic=false, bIsEditable=true, bIsStaticSwitch=true, bIsVisible=true, type {classStructOrEnum {refPath=&NiagaraBool}}}
      Random Seed {bIsDynamic=false, bIsEditable=true, bIsStaticSwitch=false, bIsVisible=true, type {classStructOrEnum {refPath=&NiagaraInt32}}}
      Randomization Vector {bIsDynamic=false, bIsEditable=true, bIsStaticSwitch=false, bIsVisible=true, type {classStructOrEnum {refPath=&Vector3f}}}
      Scale Curl Lookup By Engine Owner Scale {bIsDynamic=false, bIsEditable=true, bIsStaticSwitch=true, bIsVisible=true, type {classStructOrEnum {refPath=&NiagaraBool}}}
      CurlNoise {
       bIsDynamic=false
       bIsEditable=false
       bIsStaticSwitch=false
       bIsVisible=false
       type {classStructOrEnum {refPath=&NiagaraDataInterfaceCurlNoise}}
      }
      Mask Curl Noise {bIsDynamic=false, bIsEditable=true, bIsStaticSwitch=true, bIsVisible=true, type {classStructOrEnum {refPath=&NiagaraBool}}}
      Curl Noise Cone Mask Angle {
       bIsDynamic=false
       bIsEditable=false
       bIsStaticSwitch=false
       bIsVisible=false
       type {classStructOrEnum {refPath=&NiagaraFloat}}
      }
      Curl Noise Cone Mask Falloff Angle {
       bIsDynamic=false
       bIsEditable=false
       bIsStaticSwitch=false
       bIsVisible=false
       type {classStructOrEnum {refPath=&NiagaraFloat}}
      }
      Curl Noise Cone Mask Axis {
       bIsDynamic=false
       bIsEditable=false
       bIsStaticSwitch=false
       bIsVisible=false
       type {classStructOrEnum {refPath=&Vector3f}}
      }
     ]
     moduleScript {refPath=&CurlNoiseForce}
    }
    {
     bIsSetParametersModule=false
     enabled=true
     moduleName=Drag
     inputs [
      Use Linear Drag {bIsDynamic=false, bIsEditable=true, bIsStaticSwitch=false, bIsVisible=true, type {classStructOrEnum {refPath=&NiagaraBool}}}
      Drag {bIsDynamic=true, bIsEditable=true, bIsStaticSwitch=false, bIsVisible=true, type {classStructOrEnum {refPath=&NiagaraFloat}}}
      Use Rotational Drag {bIsDynamic=false, bIsEditable=true, bIsStaticSwitch=false, bIsVisible=true, type {classStructOrEnum {refPath=&NiagaraBool}}}
      Rotational Drag {bIsDynamic=false, bIsEditable=false, bIsStaticSwitch=false, bIsVisible=true, type {classStructOrEnum {refPath=&NiagaraFloat}}}
     ]
     moduleScript {refPath=&Drag}
    }
    {
     bIsSetParametersModule=false
     enabled=true
     moduleName=GravityForce
     inputs [
      Gravity {bIsDynamic=false, bIsEditable=true, bIsStaticSwitch=false, bIsVisible=true, type {classStructOrEnum {refPath=&Vector3f}}}
      Coordinate Space {
       bIsDynamic=false
       bIsEditable=true
       bIsStaticSwitch=true
       bIsVisible=true
       type {classStructOrEnum {refPath=&ENiagaraCoordinateSpace}}
      }
     ]
     moduleScript {refPath=&GravityForce}
    }
    {
     bIsSetParametersModule=false
     enabled=true
     moduleName=SolveForcesAndVelocity
     inputs [
      Clamp Velocity {bIsDynamic=false, bIsEditable=true, bIsStaticSwitch=false, bIsVisible=true, type {classStructOrEnum {refPath=&NiagaraBool}}}
      Speed Limit {bIsDynamic=false, bIsEditable=false, bIsStaticSwitch=false, bIsVisible=true, type {classStructOrEnum {refPath=&NiagaraFloat}}}
      Limit Acceleration {bIsDynamic=false, bIsEditable=true, bIsStaticSwitch=false, bIsVisible=true, type {classStructOrEnum {refPath=&NiagaraBool}}}
      Acceleration Limit {bIsDynamic=false, bIsEditable=false, bIsStaticSwitch=false, bIsVisible=true, type {classStructOrEnum {refPath=&NiagaraFloat}}}
      Manually Enable Rotational Solver {bIsDynamic=false, bIsEditable=true, bIsStaticSwitch=false, bIsVisible=true, type {classStructOrEnum {refPath=&NiagaraBool}}}
      Write to Intrinsic Properties {bIsDynamic=false, bIsEditable=true, bIsStaticSwitch=false, bIsVisible=true, type {classStructOrEnum {refPath=&NiagaraBool}}}
      Write to Presolve Properties {bIsDynamic=false, bIsEditable=true, bIsStaticSwitch=true, bIsVisible=true, type {classStructOrEnum {refPath=&NiagaraBool}}}
      DeltaTime {bIsDynamic=false, bIsEditable=true, bIsStaticSwitch=false, bIsVisible=true, type {classStructOrEnum {refPath=&NiagaraFloat}}}
      Mass {bIsDynamic=false, bIsEditable=true, bIsStaticSwitch=false, bIsVisible=true, type {classStructOrEnum {refPath=&NiagaraFloat}}}
      Position {
       bIsDynamic=false
       bIsEditable=true
       bIsStaticSwitch=false
       bIsVisible=true
       type {classStructOrEnum {refPath=&NiagaraPosition}}
      }
      Velocity {bIsDynamic=false, bIsEditable=true, bIsStaticSwitch=false, bIsVisible=true, type {classStructOrEnum {refPath=&Vector3f}}}
      PreviousVelocity {bIsDynamic=false, bIsEditable=true, bIsStaticSwitch=false, bIsVisible=true, type {classStructOrEnum {refPath=&Vector3f}}}
      Force {bIsDynamic=false, bIsEditable=true, bIsStaticSwitch=false, bIsVisible=true, type {classStructOrEnum {refPath=&Vector3f}}}
      Mesh Orientation {bIsDynamic=false, bIsEditable=false, bIsStaticSwitch=false, bIsVisible=false, type {classStructOrEnum {refPath=&Quat4f}}}
      Rotational Velocity {
       bIsDynamic=false
       bIsEditable=false
       bIsStaticSwitch=false
       bIsVisible=false
       type {classStructOrEnum {refPath=&Vector3f}}
      }
      Rotational Force {
       bIsDynamic=false
       bIsEditable=false
       bIsStaticSwitch=false
       bIsVisible=false
       type {classStructOrEnum {refPath=&Vector3f}}
      }
      Rotational Inertia {
       bIsDynamic=false
       bIsEditable=false
       bIsStaticSwitch=false
       bIsVisible=false
       type {classStructOrEnum {refPath=&NiagaraFloat}}
      }
     ]
     moduleScript {refPath=&SolveForcesAndVelocity}
    }
    {
     bIsSetParametersModule=false
     enabled=false
     moduleName=ScaleColor001
     inputs [
      Scale Mode {bIsDynamic=false, bIsEditable=true, bIsStaticSwitch=true, bIsVisible=true, type {classStructOrEnum {refPath=&NiagaraInt32}}}
      Linear Color Curve {
       bIsDynamic=false
       bIsEditable=false
       bIsStaticSwitch=false
       bIsVisible=false
       type {classStructOrEnum {refPath=&NiagaraDataInterfaceColorCurve}}
      }
      ScaleRGB {bIsDynamic=false, bIsEditable=true, bIsStaticSwitch=false, bIsVisible=true, type {classStructOrEnum {refPath=&NiagaraBool}}}
      Scale RGB {bIsDynamic=true, bIsEditable=true, bIsStaticSwitch=false, bIsVisible=true, type {classStructOrEnum {refPath=&Vector3f}}}
      ScaleRGBA {bIsDynamic=false, bIsEditable=false, bIsStaticSwitch=false, bIsVisible=false, type {classStructOrEnum {refPath=&NiagaraBool}}}
      Scale RGBA {
       bIsDynamic=false
       bIsEditable=false
       bIsStaticSwitch=false
       bIsVisible=false
       type {classStructOrEnum {refPath=&Vector4f}}
      }
      Curve Index {
       bIsDynamic=false
       bIsEditable=false
       bIsStaticSwitch=false
       bIsVisible=false
       type {classStructOrEnum {refPath=&NiagaraFloat}}
      }
      ScaleA {bIsDynamic=false, bIsEditable=true, bIsStaticSwitch=false, bIsVisible=true, type {classStructOrEnum {refPath=&NiagaraBool}}}
      Scale Alpha {bIsDynamic=false, bIsEditable=false, bIsStaticSwitch=false, bIsVisible=true, type {classStructOrEnum {refPath=&NiagaraFloat}}}
      Color Value To Scale {
       bIsDynamic=false
       bIsEditable=true
       bIsStaticSwitch=false
       bIsVisible=true
       type {classStructOrEnum {refPath=&LinearColor}}
      }
      Color Scale {
       bIsDynamic=false
       bIsEditable=false
       bIsStaticSwitch=false
       bIsVisible=false
       type {classStructOrEnum {refPath=&LinearColor}}
      }
     ]
     moduleScript {refPath=&ScaleColor}
    }
    {
     bIsSetParametersModule=false
     enabled=true
     moduleName=ScaleColor002
     inputs [
      Scale Mode {bIsDynamic=false, bIsEditable=true, bIsStaticSwitch=true, bIsVisible=true, type {classStructOrEnum {refPath=&NiagaraInt32}}}
      Linear Color Curve {
       bIsDynamic=false
       bIsEditable=false
       bIsStaticSwitch=false
       bIsVisible=false
       type {classStructOrEnum {refPath=&NiagaraDataInterfaceColorCurve}}
      }
      ScaleRGB {bIsDynamic=false, bIsEditable=true, bIsStaticSwitch=false, bIsVisible=true, type {classStructOrEnum {refPath=&NiagaraBool}}}
      Scale RGB {bIsDynamic=true, bIsEditable=true, bIsStaticSwitch=false, bIsVisible=true, type {classStructOrEnum {refPath=&Vector3f}}}
      ScaleRGBA {bIsDynamic=false, bIsEditable=false, bIsStaticSwitch=false, bIsVisible=false, type {classStructOrEnum {refPath=&NiagaraBool}}}
      Scale RGBA {
       bIsDynamic=false
       bIsEditable=false
       bIsStaticSwitch=false
       bIsVisible=false
       type {classStructOrEnum {refPath=&Vector4f}}
      }
      Curve Index {
       bIsDynamic=false
       bIsEditable=false
       bIsStaticSwitch=false
       bIsVisible=false
       type {classStructOrEnum {refPath=&NiagaraFloat}}
      }
      ScaleA {bIsDynamic=false, bIsEditable=true, bIsStaticSwitch=false, bIsVisible=true, type {classStructOrEnum {refPath=&NiagaraBool}}}
      Scale Alpha {bIsDynamic=false, bIsEditable=false, bIsStaticSwitch=false, bIsVisible=true, type {classStructOrEnum {refPath=&NiagaraFloat}}}
      Color Value To Scale {
       bIsDynamic=false
       bIsEditable=true
       bIsStaticSwitch=false
       bIsVisible=true
       type {classStructOrEnum {refPath=&LinearColor}}
      }
      Color Scale {
       bIsDynamic=false
       bIsEditable=false
       bIsStaticSwitch=false
       bIsVisible=false
       type {classStructOrEnum {refPath=&LinearColor}}
      }
     ]
     moduleScript {refPath=&ScaleColor}
    }
   ]
  }
  rendererClasses [{refPath=&NiagaraMeshRendererProperties}]
  renderers [
   {rendererIndex=0, rendererClass {refPath=&NiagaraMeshRendererProperties}}
   {rendererIndex=1, rendererClass {refPath=&NiagaraMeshRendererProperties}}
  ]
 }
 Snowflakedsff003 {
  emitterSpawnScript {scriptName=EmitterSpawnScript, modules []}
  emitterUpdateScript {
   scriptName=EmitterUpdateScript
   modules [
    {
     bIsSetParametersModule=false
     enabled=true
     moduleName=EmitterState
     inputs [
      Life Cycle Mode {
       bIsDynamic=false
       bIsEditable=true
       bIsStaticSwitch=true
       bIsVisible=true
       type {classStructOrEnum {refPath=&ENiagaraEmitterLifeCycleMode}}
      }
      Inactive Response {
       bIsDynamic=false
       bIsEditable=true
       bIsStaticSwitch=true
       bIsVisible=true
       type {classStructOrEnum {refPath=&ENiagaraInactiveMode}}
      }
      Loop Behavior {
       bIsDynamic=false
       bIsEditable=true
       bIsStaticSwitch=true
       bIsVisible=true
       type {classStructOrEnum {refPath=&ENiagara_EmitterStateOptions}}
      }
      Loop Duration Mode {
       bIsDynamic=false
       bIsEditable=true
       bIsStaticSwitch=true
       bIsVisible=true
       type {classStructOrEnum {refPath=&ENiagara_InfiniteLoopDuration}}
      }
      Loop Count {
       bIsDynamic=false
       bIsEditable=false
       bIsStaticSwitch=false
       bIsVisible=false
       type {classStructOrEnum {refPath=&NiagaraInt32}}
      }
      Loop Duration {bIsDynamic=false, bIsEditable=true, bIsStaticSwitch=false, bIsVisible=true, type {classStructOrEnum {refPath=&NiagaraFloat}}}
      Recalculate Duration Each Loop {bIsDynamic=false, bIsEditable=false, bIsStaticSwitch=false, bIsVisible=false, type {classStructOrEnum {refPath=&NiagaraBool}}}
      UseLoopDelay {bIsDynamic=false, bIsEditable=true, bIsStaticSwitch=false, bIsVisible=true, type {classStructOrEnum {refPath=&NiagaraBool}}}
      Loop Delay {bIsDynamic=false, bIsEditable=false, bIsStaticSwitch=false, bIsVisible=true, type {classStructOrEnum {refPath=&NiagaraFloat}}}
      Delay First Loop Only {bIsDynamic=false, bIsEditable=false, bIsStaticSwitch=false, bIsVisible=false, type {classStructOrEnum {refPath=&NiagaraBool}}}
      Scalability Mode {
       bIsDynamic=false
       bIsEditable=true
       bIsStaticSwitch=true
       bIsVisible=true
       type {classStructOrEnum {refPath=&ENiagaraEmitterScalabilityMode_Limited}}
      }
      Enable Distance Culling {bIsDynamic=false, bIsEditable=false, bIsStaticSwitch=true, bIsVisible=false, type {classStructOrEnum {refPath=&NiagaraBool}}}
      MinDistance {
       bIsDynamic=false
       bIsEditable=false
       bIsStaticSwitch=false
       bIsVisible=false
       type {classStructOrEnum {refPath=&NiagaraFloat}}
      }
      Min Distance Response {
       bIsDynamic=false
       bIsEditable=false
       bIsStaticSwitch=false
       bIsVisible=false
       type {classStructOrEnum {refPath=&ENiagaraExecutionStateManagement}}
      }
      MaxDistance {
       bIsDynamic=false
       bIsEditable=false
       bIsStaticSwitch=false
       bIsVisible=false
       type {classStructOrEnum {refPath=&NiagaraFloat}}
      }
      Max Distance Response {
       bIsDynamic=false
       bIsEditable=false
       bIsStaticSwitch=false
       bIsVisible=false
       type {classStructOrEnum {refPath=&ENiagaraExecutionStateManagement}}
      }
      Scale Spawn Count {bIsDynamic=false, bIsEditable=false, bIsStaticSwitch=true, bIsVisible=false, type {classStructOrEnum {refPath=&NiagaraBool}}}
      Spawn Count Scale {
       bIsDynamic=false
       bIsEditable=false
       bIsStaticSwitch=false
       bIsVisible=false
       type {classStructOrEnum {refPath=&NiagaraFloat}}
      }
      Spawn Count Scale By Distance {
       bIsDynamic=false
       bIsEditable=false
       bIsStaticSwitch=false
       bIsVisible=false
       type {classStructOrEnum {refPath=&NiagaraDataInterfaceCurve}}
      }
      Enable Visibility Culling {bIsDynamic=false, bIsEditable=false, bIsStaticSwitch=true, bIsVisible=false, type {classStructOrEnum {refPath=&NiagaraBool}}}
      Visibility Cull Response {
       bIsDynamic=false
       bIsEditable=false
       bIsStaticSwitch=false
       bIsVisible=false
       type {classStructOrEnum {refPath=&ENiagaraExecutionStateManagement}}
      }
      Visibility Cull Delay {
       bIsDynamic=false
       bIsEditable=false
       bIsStaticSwitch=false
       bIsVisible=false
       type {classStructOrEnum {refPath=&NiagaraFloat}}
      }
      Reset Age On Awaken {bIsDynamic=false, bIsEditable=false, bIsStaticSwitch=false, bIsVisible=false, type {classStructOrEnum {refPath=&NiagaraBool}}}
     ]
     moduleScript {refPath=&EmitterState}
    }
    {
     bIsSetParametersModule=false
     enabled=true
     moduleName=SpawnRate
     inputs [
      SpawnRate {bIsDynamic=false, bIsEditable=true, bIsStaticSwitch=false, bIsVisible=true, type {classStructOrEnum {refPath=&NiagaraFloat}}}
      Use Spawn Probability {bIsDynamic=false, bIsEditable=true, bIsStaticSwitch=false, bIsVisible=true, type {classStructOrEnum {refPath=&NiagaraBool}}}
      Spawn Probability {bIsDynamic=false, bIsEditable=false, bIsStaticSwitch=false, bIsVisible=true, type {classStructOrEnum {refPath=&NiagaraFloat}}}
      SpawnGroup {bIsDynamic=false, bIsEditable=true, bIsStaticSwitch=false, bIsVisible=true, type {classStructOrEnum {refPath=&NiagaraInt32}}}
     ]
     moduleScript {refPath=&SpawnRate}
    }
    {
     bIsSetParametersModule=false
     enabled=true
     moduleName=SpawnBurst_Instantaneous
     inputs [
      Spawn Count {bIsDynamic=false, bIsEditable=true, bIsStaticSwitch=false, bIsVisible=true, type {classStructOrEnum {refPath=&NiagaraInt32}}}
      Spawn Time {bIsDynamic=false, bIsEditable=true, bIsStaticSwitch=false, bIsVisible=true, type {classStructOrEnum {refPath=&NiagaraFloat}}}
      Use Spawn Probability {bIsDynamic=false, bIsEditable=true, bIsStaticSwitch=false, bIsVisible=true, type {classStructOrEnum {refPath=&NiagaraBool}}}
      Spawn Probability {bIsDynamic=false, bIsEditable=false, bIsStaticSwitch=false, bIsVisible=true, type {classStructOrEnum {refPath=&NiagaraFloat}}}
      UseLoopCountLimit {bIsDynamic=false, bIsEditable=true, bIsStaticSwitch=false, bIsVisible=true, type {classStructOrEnum {refPath=&NiagaraBool}}}
      Loop Count Limit {bIsDynamic=false, bIsEditable=false, bIsStaticSwitch=false, bIsVisible=true, type {classStructOrEnum {refPath=&NiagaraInt32}}}
      Spawn Group {bIsDynamic=false, bIsEditable=true, bIsStaticSwitch=false, bIsVisible=true, type {classStructOrEnum {refPath=&NiagaraInt32}}}
      Age {bIsDynamic=false, bIsEditable=true, bIsStaticSwitch=false, bIsVisible=true, type {classStructOrEnum {refPath=&NiagaraFloat}}}
     ]
     moduleScript {refPath=&SpawnBurst_Instantaneous}
    }
   ]
  }
  particleSpawnScript {
   scriptName=ParticleSpawnScript
   modules [
    {
     bIsSetParametersModule=false
     enabled=true
     moduleName=InitializeParticle001
     inputs [
      Lifetime Mode {
       bIsDynamic=false
       bIsEditable=true
       bIsStaticSwitch=true
       bIsVisible=true
       type {classStructOrEnum {refPath=&ENiagara_LifetimeMode}}
      }
      Lifetime Min {bIsDynamic=false, bIsEditable=true, bIsStaticSwitch=false, bIsVisible=true, type {classStructOrEnum {refPath=&NiagaraFloat}}}
      Lifetime {
       bIsDynamic=false
       bIsEditable=false
       bIsStaticSwitch=false
       bIsVisible=false
       type {classStructOrEnum {refPath=&NiagaraFloat}}
      }
      Lifetime Max {bIsDynamic=false, bIsEditable=true, bIsStaticSwitch=false, bIsVisible=true, type {classStructOrEnum {refPath=&NiagaraFloat}}}
      Lifetime Randomness Mode {
       bIsDynamic=false
       bIsEditable=true
       bIsStaticSwitch=true
       bIsVisible=true
       type {classStructOrEnum {refPath=&ENiagaraRandomnessMode}}
      }
      Color Mode {
       bIsDynamic=false
       bIsEditable=true
       bIsStaticSwitch=true
       bIsVisible=true
       type {classStructOrEnum {refPath=&ENiagara_ColorInitializationMode}}
      }
      Color Minimum {
       bIsDynamic=false
       bIsEditable=false
       bIsStaticSwitch=false
       bIsVisible=false
       type {classStructOrEnum {refPath=&LinearColor}}
      }
      Color {
       bIsDynamic=true
       bIsEditable=true
       bIsStaticSwitch=false
       bIsVisible=true
       type {classStructOrEnum {refPath=&LinearColor}}
      }
      AdjustHue {bIsDynamic=false, bIsEditable=false, bIsStaticSwitch=false, bIsVisible=false, type {classStructOrEnum {refPath=&NiagaraBool}}}
      Hue Shift Range {
       bIsDynamic=false
       bIsEditable=false
       bIsStaticSwitch=false
       bIsVisible=false
       type {classStructOrEnum {refPath=&Vector2f}}
      }
      Color Maximum {
       bIsDynamic=false
       bIsEditable=false
       bIsStaticSwitch=false
       bIsVisible=false
       type {classStructOrEnum {refPath=&LinearColor}}
      }
      AdjustSaturation {bIsDynamic=false, bIsEditable=false, bIsStaticSwitch=false, bIsVisible=false, type {classStructOrEnum {refPath=&NiagaraBool}}}
      Saturation Range {
       bIsDynamic=false
       bIsEditable=false
       bIsStaticSwitch=false
       bIsVisible=false
       type {classStructOrEnum {refPath=&Vector2f}}
      }
      AdjustValue {bIsDynamic=false, bIsEditable=false, bIsStaticSwitch=false, bIsVisible=false, type {classStructOrEnum {refPath=&NiagaraBool}}}
      Value Range {
       bIsDynamic=false
       bIsEditable=false
       bIsStaticSwitch=false
       bIsVisible=false
       type {classStructOrEnum {refPath=&Vector2f}}
      }
      AdjustAlpha {bIsDynamic=false, bIsEditable=false, bIsStaticSwitch=false, bIsVisible=false, type {classStructOrEnum {refPath=&NiagaraBool}}}
      Alpha Scale Range {
       bIsDynamic=false
       bIsEditable=false
       bIsStaticSwitch=false
       bIsVisible=false
       type {classStructOrEnum {refPath=&Vector2f}}
      }
      Color Channel Mode {
       bIsDynamic=false
       bIsEditable=false
       bIsStaticSwitch=true
       bIsVisible=false
       type {classStructOrEnum {refPath=&ENiagaraChannelCorrelation}}
      }
      Color Randomness Mode {
       bIsDynamic=false
       bIsEditable=false
       bIsStaticSwitch=true
       bIsVisible=false
       type {classStructOrEnum {refPath=&ENiagaraRandomnessMode}}
      }
      Position Mode {
       bIsDynamic=false
       bIsEditable=true
       bIsStaticSwitch=true
       bIsVisible=true
       type {classStructOrEnum {refPath=&ENiagara_PositionInitializationMode}}
      }
      Position {
       bIsDynamic=false
       bIsEditable=false
       bIsStaticSwitch=false
       bIsVisible=false
       type {classStructOrEnum {refPath=&NiagaraPosition}}
      }
      UsePositionOffset {bIsDynamic=false, bIsEditable=true, bIsStaticSwitch=false, bIsVisible=true, type {classStructOrEnum {refPath=&NiagaraBool}}}
      Position Offset {bIsDynamic=false, bIsEditable=true, bIsStaticSwitch=false, bIsVisible=true, type {classStructOrEnum {refPath=&Vector3f}}}
      Position Offset Coordinate Space {
       bIsDynamic=false
       bIsEditable=true
       bIsStaticSwitch=false
       bIsVisible=true
       type {classStructOrEnum {refPath=&ENiagaraCoordinateSpace}}
      }
      Mass Mode {
       bIsDynamic=false
       bIsEditable=true
       bIsStaticSwitch=true
       bIsVisible=true
       type {classStructOrEnum {refPath=&ENiagara_MassInitializationMode}}
      }
      Mass Min {
       bIsDynamic=false
       bIsEditable=false
       bIsStaticSwitch=false
       bIsVisible=false
       type {classStructOrEnum {refPath=&NiagaraFloat}}
      }
      Mass {
       bIsDynamic=false
       bIsEditable=false
       bIsStaticSwitch=false
       bIsVisible=false
       type {classStructOrEnum {refPath=&NiagaraFloat}}
      }
      Mass Max {
       bIsDynamic=false
       bIsEditable=false
       bIsStaticSwitch=false
       bIsVisible=false
       type {classStructOrEnum {refPath=&NiagaraFloat}}
      }
      Mass Randomness Mode {
       bIsDynamic=false
       bIsEditable=false
       bIsStaticSwitch=true
       bIsVisible=false
       type {classStructOrEnum {refPath=&ENiagaraRandomnessMode}}
      }
      Write Material Random {bIsDynamic=false, bIsEditable=true, bIsStaticSwitch=false, bIsVisible=true, type {classStructOrEnum {refPath=&NiagaraBool}}}
      Material Random {bIsDynamic=false, bIsEditable=true, bIsStaticSwitch=false, bIsVisible=true, type {classStructOrEnum {refPath=&NiagaraFloat}}}
      Sprite Size Mode {
       bIsDynamic=false
       bIsEditable=true
       bIsStaticSwitch=true
       bIsVisible=true
       type {classStructOrEnum {refPath=&ENiagara_SizeScaleMode}}
      }
      Sprite Size {
       bIsDynamic=false
       bIsEditable=false
       bIsStaticSwitch=false
       bIsVisible=false
       type {classStructOrEnum {refPath=&Vector2f}}
      }
      Uniform Sprite Size {bIsDynamic=true, bIsEditable=true, bIsStaticSwitch=false, bIsVisible=true, type {classStructOrEnum {refPath=&NiagaraFloat}}}
      Uniform Sprite Size Min {
       bIsDynamic=false
       bIsEditable=false
       bIsStaticSwitch=false
       bIsVisible=false
       type {classStructOrEnum {refPath=&NiagaraFloat}}
      }
      Sprite Size Min {
       bIsDynamic=false
       bIsEditable=false
       bIsStaticSwitch=false
       bIsVisible=false
       type {classStructOrEnum {refPath=&Vector2f}}
      }
      Sprite Size Max {
       bIsDynamic=false
       bIsEditable=false
       bIsStaticSwitch=false
       bIsVisible=false
       type {classStructOrEnum {refPath=&Vector2f}}
      }
      Uniform Sprite Size Max {
       bIsDynamic=false
       bIsEditable=false
       bIsStaticSwitch=false
       bIsVisible=false
       type {classStructOrEnum {refPath=&NiagaraFloat}}
      }
      Sprite Size Randomness Mode {
       bIsDynamic=false
       bIsEditable=false
       bIsStaticSwitch=true
       bIsVisible=false
       type {classStructOrEnum {refPath=&ENiagaraRandomnessMode}}
      }
      Sprite Rotation Mode {
       bIsDynamic=false
       bIsEditable=true
       bIsStaticSwitch=true
       bIsVisible=true
       type {classStructOrEnum {refPath=&ENiagara_SpriteRotationMode}}
      }
      Sprite Rotation Angle Min {
       bIsDynamic=false
       bIsEditable=false
       bIsStaticSwitch=false
       bIsVisible=false
       type {classStructOrEnum {refPath=&NiagaraFloat}}
      }
      Sprite Rotation Angle {
       bIsDynamic=false
       bIsEditable=false
       bIsStaticSwitch=false
       bIsVisible=false
       type {classStructOrEnum {refPath=&NiagaraFloat}}
      }
      Sprite Rotation Angle Max {
       bIsDynamic=false
       bIsEditable=false
       bIsStaticSwitch=false
       bIsVisible=false
       type {classStructOrEnum {refPath=&NiagaraFloat}}
      }
      Sprite Rotation Randomness Mode {
       bIsDynamic=false
       bIsEditable=false
       bIsStaticSwitch=true
       bIsVisible=false
       type {classStructOrEnum {refPath=&ENiagaraRandomnessMode}}
      }
      Sprite UV Mode {
       bIsDynamic=false
       bIsEditable=true
       bIsStaticSwitch=true
       bIsVisible=true
       type {classStructOrEnum {refPath=&ENiagara_UVFlippingMode}}
      }
      Sprite UV Randomness Mode {
       bIsDynamic=false
       bIsEditable=false
       bIsStaticSwitch=true
       bIsVisible=false
       type {classStructOrEnum {refPath=&ENiagaraRandomnessMode}}
      }
      Sprite UV Scale {
       bIsDynamic=false
       bIsEditable=false
       bIsStaticSwitch=false
       bIsVisible=false
       type {classStructOrEnum {refPath=&Vector2f}}
      }
      Mesh Scale Mode {
       bIsDynamic=false
       bIsEditable=true
       bIsStaticSwitch=true
       bIsVisible=true
       type {classStructOrEnum {refPath=&ENiagara_SizeScaleMode}}
      }
      Mesh Uniform Scale Min {
       bIsDynamic=false
       bIsEditable=false
       bIsStaticSwitch=false
       bIsVisible=false
       type {classStructOrEnum {refPath=&NiagaraFloat}}
      }
      Mesh Uniform Scale {
       bIsDynamic=false
       bIsEditable=false
       bIsStaticSwitch=false
       bIsVisible=false
       type {classStructOrEnum {refPath=&NiagaraFloat}}
      }
      Mesh Scale Min {bIsDynamic=false, bIsEditable=true, bIsStaticSwitch=false, bIsVisible=true, type {classStructOrEnum {refPath=&Vector3f}}}
      Mesh Scale {
       bIsDynamic=false
       bIsEditable=false
       bIsStaticSwitch=false
       bIsVisible=false
       type {classStructOrEnum {refPath=&Vector3f}}
      }
      Mesh Scale Max {bIsDynamic=false, bIsEditable=true, bIsStaticSwitch=false, bIsVisible=true, type {classStructOrEnum {refPath=&Vector3f}}}
      Mesh Uniform Scale Max {
       bIsDynamic=false
       bIsEditable=false
       bIsStaticSwitch=false
       bIsVisible=false
       type {classStructOrEnum {refPath=&NiagaraFloat}}
      }
      Mesh Scale Randomness Mode {
       bIsDynamic=false
       bIsEditable=true
       bIsStaticSwitch=true
       bIsVisible=true
       type {classStructOrEnum {refPath=&ENiagaraRandomnessMode}}
      }
      Mesh Renderer Array Visibility Mode {
       bIsDynamic=false
       bIsEditable=true
       bIsStaticSwitch=true
       bIsVisible=true
       type {classStructOrEnum {refPath=&ENiagara_UnsetDirectSetRandom}}
      }
      Mesh Renderer Mesh Index {
       bIsDynamic=false
       bIsEditable=false
       bIsStaticSwitch=false
       bIsVisible=false
       type {classStructOrEnum {refPath=&NiagaraInt32}}
      }
      Wrap Index By Mesh Count {bIsDynamic=false, bIsEditable=false, bIsStaticSwitch=true, bIsVisible=false, type {classStructOrEnum {refPath=&NiagaraBool}}}
      Mesh Renderer Info {
       bIsDynamic=false
       bIsEditable=false
       bIsStaticSwitch=false
       bIsVisible=false
       type {classStructOrEnum {refPath=&NiagaraDataInterfaceMeshRendererInfo}}
      }
      Mesh Renderer Random Seed {
       bIsDynamic=false
       bIsEditable=false
       bIsStaticSwitch=false
       bIsVisible=false
       type {classStructOrEnum {refPath=&NiagaraInt32}}
      }
      Ribbon Width Mode {
       bIsDynamic=false
       bIsEditable=true
       bIsStaticSwitch=true
       bIsVisible=true
       type {classStructOrEnum {refPath=&ENiagara_UnsetDirectSet}}
      }
      Ribbon Width {
       bIsDynamic=false
       bIsEditable=false
       bIsStaticSwitch=false
       bIsVisible=false
       type {classStructOrEnum {refPath=&NiagaraFloat}}
      }
      Ribbon Facing Mode {
       bIsDynamic=false
       bIsEditable=true
       bIsStaticSwitch=true
       bIsVisible=true
       type {classStructOrEnum {refPath=&ENiagara_UnsetDirectSet}}
      }
      Ribbon Facing Vector {
       bIsDynamic=false
       bIsEditable=false
       bIsStaticSwitch=false
       bIsVisible=false
       type {classStructOrEnum {refPath=&Vector3f}}
      }
      Ribbon Twist Mode {
       bIsDynamic=false
       bIsEditable=true
       bIsStaticSwitch=true
       bIsVisible=true
       type {classStructOrEnum {refPath=&ENiagara_UnsetDirectSet}}
      }
      "Ribbon Twist " {
       bIsDynamic=false
       bIsEditable=false
       bIsStaticSwitch=false
       bIsVisible=false
       type {classStructOrEnum {refPath=&NiagaraFloat}}
      }
     ]
     moduleScript {refPath=&InitializeParticle}
    }
    {
     bIsSetParametersModule=false
     enabled=true
     moduleName=SphereLocation
     inputs [
      Sphere Radius {bIsDynamic=false, bIsEditable=true, bIsStaticSwitch=false, bIsVisible=true, type {classStructOrEnum {refPath=&NiagaraFloat}}}
      UseOffset {bIsDynamic=false, bIsEditable=true, bIsStaticSwitch=false, bIsVisible=true, type {classStructOrEnum {refPath=&NiagaraBool}}}
      Offset {bIsDynamic=false, bIsEditable=true, bIsStaticSwitch=false, bIsVisible=true, type {classStructOrEnum {refPath=&Vector3f}}}
      Offset Coordinate Space {
       bIsDynamic=false
       bIsEditable=true
       bIsStaticSwitch=false
       bIsVisible=true
       type {classStructOrEnum {refPath=&ENiagaraCoordinateSpace}}
      }
      UseNonUniformScale {bIsDynamic=false, bIsEditable=true, bIsStaticSwitch=false, bIsVisible=true, type {classStructOrEnum {refPath=&NiagaraBool}}}
      Non Uniform Scale {bIsDynamic=false, bIsEditable=false, bIsStaticSwitch=false, bIsVisible=true, type {classStructOrEnum {refPath=&Vector3f}}}
      UseSphereOrientation {bIsDynamic=false, bIsEditable=true, bIsStaticSwitch=false, bIsVisible=true, type {classStructOrEnum {refPath=&NiagaraBool}}}
      Sphere Orientation Axis {bIsDynamic=false, bIsEditable=false, bIsStaticSwitch=false, bIsVisible=true, type {classStructOrEnum {refPath=&Vector3f}}}
      Sphere Origin {
       bIsDynamic=false
       bIsEditable=true
       bIsStaticSwitch=false
       bIsVisible=true
       type {classStructOrEnum {refPath=&NiagaraPosition}}
      }
      Sphere Coordinate Space {
       bIsDynamic=false
       bIsEditable=true
       bIsStaticSwitch=false
       bIsVisible=true
       type {classStructOrEnum {refPath=&ENiagaraCoordinateSpace}}
      }
      Write to Intrinsic Variables {bIsDynamic=false, bIsEditable=true, bIsStaticSwitch=true, bIsVisible=true, type {classStructOrEnum {refPath=&NiagaraBool}}}
      Sphere Distribution {
       bIsDynamic=false
       bIsEditable=true
       bIsStaticSwitch=true
       bIsVisible=true
       type {classStructOrEnum {refPath=&ENiagaraSphereDistributionMode}}
      }
      Evaluation Type {
       bIsDynamic=false
       bIsEditable=false
       bIsStaticSwitch=true
       bIsVisible=false
       type {classStructOrEnum {refPath=&ENiagaraRandomnessEvaluation}}
      }
      Hemisphere X {bIsDynamic=false, bIsEditable=true, bIsStaticSwitch=false, bIsVisible=true, type {classStructOrEnum {refPath=&NiagaraBool}}}
      Uniform Distribution {
       bIsDynamic=false
       bIsEditable=false
       bIsStaticSwitch=false
       bIsVisible=false
       type {classStructOrEnum {refPath=&NiagaraFloat}}
      }
      Hemisphere Y {bIsDynamic=false, bIsEditable=true, bIsStaticSwitch=false, bIsVisible=true, type {classStructOrEnum {refPath=&NiagaraBool}}}
      U Position {
       bIsDynamic=false
       bIsEditable=false
       bIsStaticSwitch=false
       bIsVisible=false
       type {classStructOrEnum {refPath=&NiagaraFloat}}
      }
      Uniform Spiral Amount {
       bIsDynamic=false
       bIsEditable=false
       bIsStaticSwitch=false
       bIsVisible=false
       type {classStructOrEnum {refPath=&NiagaraFloat}}
      }
      Hemisphere Z {bIsDynamic=false, bIsEditable=true, bIsStaticSwitch=false, bIsVisible=true, type {classStructOrEnum {refPath=&NiagaraBool}}}
      Uniform Endpoint Offset {
       bIsDynamic=false
       bIsEditable=false
       bIsStaticSwitch=false
       bIsVisible=false
       type {classStructOrEnum {refPath=&NiagaraFloat}}
      }
      V Position {
       bIsDynamic=false
       bIsEditable=false
       bIsStaticSwitch=false
       bIsVisible=false
       type {classStructOrEnum {refPath=&NiagaraFloat}}
      }
      Radius Position {
       bIsDynamic=false
       bIsEditable=false
       bIsStaticSwitch=false
       bIsVisible=false
       type {classStructOrEnum {refPath=&NiagaraFloat}}
      }
      Surface Only {bIsDynamic=false, bIsEditable=true, bIsStaticSwitch=false, bIsVisible=true, type {classStructOrEnum {refPath=&NiagaraBool}}}
      Surface Only Band Thickness {bIsDynamic=false, bIsEditable=true, bIsStaticSwitch=false, bIsVisible=true, type {classStructOrEnum {refPath=&NiagaraFloat}}}
      Surface Expansion Mode {
       bIsDynamic=false
       bIsEditable=true
       bIsStaticSwitch=false
       bIsVisible=true
       type {classStructOrEnum {refPath=&ENiagaraExpansionMode}}
      }
      Normalize Thickness {bIsDynamic=false, bIsEditable=true, bIsStaticSwitch=false, bIsVisible=true, type {classStructOrEnum {refPath=&NiagaraBool}}}
      Randomness Mode {
       bIsDynamic=false
       bIsEditable=true
       bIsStaticSwitch=true
       bIsVisible=true
       type {classStructOrEnum {refPath=&ENiagaraRandomnessMode}}
      }
      Override Seed {bIsDynamic=false, bIsEditable=true, bIsStaticSwitch=false, bIsVisible=true, type {classStructOrEnum {refPath=&NiagaraBool}}}
      Random Seed {bIsDynamic=false, bIsEditable=false, bIsStaticSwitch=false, bIsVisible=true, type {classStructOrEnum {refPath=&NiagaraInt32}}}
      Fixed Random Seed {bIsDynamic=false, bIsEditable=false, bIsStaticSwitch=true, bIsVisible=false, type {classStructOrEnum {refPath=&NiagaraBool}}}
      MaskSpawnGroup {bIsDynamic=false, bIsEditable=true, bIsStaticSwitch=false, bIsVisible=true, type {classStructOrEnum {refPath=&NiagaraBool}}}
      Spawn Group Mask {bIsDynamic=false, bIsEditable=false, bIsStaticSwitch=false, bIsVisible=true, type {classStructOrEnum {refPath=&NiagaraInt32}}}
     ]
     moduleScript {refPath=&SphereLocation}
    }
    {
     bIsSetParametersModule=false
     enabled=false
     moduleName=AddVelocityFromPoint
     inputs [
      Velocity Strength {bIsDynamic=true, bIsEditable=true, bIsStaticSwitch=false, bIsVisible=true, type {classStructOrEnum {refPath=&NiagaraFloat}}}
      Origin Offset {bIsDynamic=false, bIsEditable=true, bIsStaticSwitch=false, bIsVisible=true, type {classStructOrEnum {refPath=&Vector3f}}}
      Origin Offset Coordinate Space {
       bIsDynamic=false
       bIsEditable=true
       bIsStaticSwitch=false
       bIsVisible=true
       type {classStructOrEnum {refPath=&ENiagaraCoordinateSpace}}
      }
      Default Position {bIsDynamic=false, bIsEditable=true, bIsStaticSwitch=false, bIsVisible=true, type {classStructOrEnum {refPath=&Vector3f}}}
      Module.Use Falloff Distance {bIsDynamic=false, bIsEditable=true, bIsStaticSwitch=false, bIsVisible=true, type {classStructOrEnum {refPath=&NiagaraBool}}}
      Velocity Falloff Distance {bIsDynamic=false, bIsEditable=false, bIsStaticSwitch=false, bIsVisible=true, type {classStructOrEnum {refPath=&NiagaraFloat}}}
      Invert Velocity Falloff {bIsDynamic=false, bIsEditable=false, bIsStaticSwitch=false, bIsVisible=false, type {classStructOrEnum {refPath=&NiagaraBool}}}
      Velocity Origin {bIsDynamic=false, bIsEditable=true, bIsStaticSwitch=false, bIsVisible=true, type {classStructOrEnum {refPath=&Vector3f}}}
     ]
     moduleScript {refPath=&AddVelocityFromPoint}
    }
    {
     bIsSetParametersModule=false
     enabled=true
     moduleName=AddVelocity
     inputs [
      Velocity Mode {
       bIsDynamic=false
       bIsEditable=true
       bIsStaticSwitch=true
       bIsVisible=true
       type {classStructOrEnum {refPath=&ENiagara_VelocityMode}}
      }
      Velocity {
       bIsDynamic=false
       bIsEditable=false
       bIsStaticSwitch=false
       bIsVisible=false
       type {classStructOrEnum {refPath=&Vector3f}}
      }
      Velocity Speed {bIsDynamic=true, bIsEditable=true, bIsStaticSwitch=false, bIsVisible=true, type {classStructOrEnum {refPath=&NiagaraFloat}}}
      Velocity Speed Scale {
       bIsDynamic=false
       bIsEditable=false
       bIsStaticSwitch=false
       bIsVisible=false
       type {classStructOrEnum {refPath=&NiagaraFloat}}
      }
      Constrain To Radius {
       bIsDynamic=false
       bIsEditable=false
       bIsStaticSwitch=false
       bIsVisible=false
       type {classStructOrEnum {refPath=&NiagaraFloat}}
      }
      Module.UseVelDistribution {bIsDynamic=false, bIsEditable=true, bIsStaticSwitch=false, bIsVisible=true, type {classStructOrEnum {refPath=&NiagaraBool}}}
      Distribution Along Cone Axis {bIsDynamic=false, bIsEditable=false, bIsStaticSwitch=false, bIsVisible=true, type {classStructOrEnum {refPath=&NiagaraFloat}}}
      Radius Falloff Near / Far {
       bIsDynamic=false
       bIsEditable=false
       bIsStaticSwitch=false
       bIsVisible=false
       type {classStructOrEnum {refPath=&Vector2f}}
      }
      Radius Falloff Exponent {
       bIsDynamic=false
       bIsEditable=false
       bIsStaticSwitch=false
       bIsVisible=false
       type {classStructOrEnum {refPath=&NiagaraFloat}}
      }
      Module.UseConeFalloff {bIsDynamic=false, bIsEditable=true, bIsStaticSwitch=false, bIsVisible=true, type {classStructOrEnum {refPath=&NiagaraBool}}}
      Speed Falloff From Cone Axis {bIsDynamic=false, bIsEditable=false, bIsStaticSwitch=false, bIsVisible=true, type {classStructOrEnum {refPath=&NiagaraFloat}}}
      Invert Velocity Falloff {bIsDynamic=false, bIsEditable=false, bIsStaticSwitch=false, bIsVisible=false, type {classStructOrEnum {refPath=&NiagaraBool}}}
      Override Seed {bIsDynamic=false, bIsEditable=true, bIsStaticSwitch=false, bIsVisible=true, type {classStructOrEnum {refPath=&NiagaraBool}}}
      Random Seed {bIsDynamic=false, bIsEditable=false, bIsStaticSwitch=false, bIsVisible=true, type {classStructOrEnum {refPath=&NiagaraInt32}}}
      Fixed Random Seed {bIsDynamic=false, bIsEditable=false, bIsStaticSwitch=true, bIsVisible=false, type {classStructOrEnum {refPath=&NiagaraBool}}}
      Randomness Mode {
       bIsDynamic=false
       bIsEditable=true
       bIsStaticSwitch=true
       bIsVisible=true
       type {classStructOrEnum {refPath=&ENiagaraRandomnessMode}}
      }
      Cone Axis {bIsDynamic=false, bIsEditable=true, bIsStaticSwitch=false, bIsVisible=true, type {classStructOrEnum {refPath=&Vector3f}}}
      Cone Angle {bIsDynamic=false, bIsEditable=true, bIsStaticSwitch=false, bIsVisible=true, type {classStructOrEnum {refPath=&NiagaraFloat}}}
      Inner Cone Angle {bIsDynamic=false, bIsEditable=true, bIsStaticSwitch=false, bIsVisible=true, type {classStructOrEnum {refPath=&NiagaraFloat}}}
      Cone Angle Mode {
       bIsDynamic=false
       bIsEditable=true
       bIsStaticSwitch=true
       bIsVisible=true
       type {classStructOrEnum {refPath=&ENiagara_AngleInput}}
      }
      Rotation Mode {
       bIsDynamic=false
       bIsEditable=true
       bIsStaticSwitch=true
       bIsVisible=true
       type {classStructOrEnum {refPath=&ENiagara_RotationMode}}
      }
      Rotation Axis {
       bIsDynamic=false
       bIsEditable=false
       bIsStaticSwitch=false
       bIsVisible=false
       type {classStructOrEnum {refPath=&Vector3f}}
      }
      Yaw / Pitch / Roll {
       bIsDynamic=false
       bIsEditable=false
       bIsStaticSwitch=false
       bIsVisible=false
       type {classStructOrEnum {refPath=&Vector3f}}
      }
      Rotation Quaternion {bIsDynamic=false, bIsEditable=false, bIsStaticSwitch=false, bIsVisible=false, type {classStructOrEnum {refPath=&Quat4f}}}
      Rotation Matrix {
       bIsDynamic=false
       bIsEditable=false
       bIsStaticSwitch=false
       bIsVisible=false
       type {classStructOrEnum {refPath=&NiagaraMatrix}}
      }
      Rotation Angle {
       bIsDynamic=false
       bIsEditable=false
       bIsStaticSwitch=false
       bIsVisible=false
       type {classStructOrEnum {refPath=&NiagaraFloat}}
      }
      Invert Rotation Quaternion {bIsDynamic=false, bIsEditable=false, bIsStaticSwitch=false, bIsVisible=false, type {classStructOrEnum {refPath=&NiagaraBool}}}
      Additional Quaternion Rotation {bIsDynamic=false, bIsEditable=false, bIsStaticSwitch=false, bIsVisible=false, type {classStructOrEnum {refPath=&NiagaraBool}}}
      Additional Yaw / Pitch / Roll {
       bIsDynamic=false
       bIsEditable=false
       bIsStaticSwitch=false
       bIsVisible=false
       type {classStructOrEnum {refPath=&Vector3f}}
      }
      Rotation Angle Type {
       bIsDynamic=false
       bIsEditable=false
       bIsStaticSwitch=true
       bIsVisible=false
       type {classStructOrEnum {refPath=&ENiagara_AngleInput}}
      }
      Rotation Coordinate Space {
       bIsDynamic=false
       bIsEditable=true
       bIsStaticSwitch=false
       bIsVisible=true
       type {classStructOrEnum {refPath=&ENiagaraCoordinateSpace}}
      }
      Velocity Origin {
       bIsDynamic=false
       bIsEditable=false
       bIsStaticSwitch=false
       bIsVisible=false
       type {classStructOrEnum {refPath=&NiagaraPosition}}
      }
      Origin Offset {
       bIsDynamic=false
       bIsEditable=false
       bIsStaticSwitch=false
       bIsVisible=false
       type {classStructOrEnum {refPath=&Vector3f}}
      }
      Origin Offset Coordinate Space {
       bIsDynamic=false
       bIsEditable=false
       bIsStaticSwitch=false
       bIsVisible=false
       type {classStructOrEnum {refPath=&ENiagaraCoordinateSpace}}
      }
      Default Position {
       bIsDynamic=false
       bIsEditable=false
       bIsStaticSwitch=false
       bIsVisible=false
       type {classStructOrEnum {refPath=&NiagaraPosition}}
      }
     ]
     moduleScript {refPath=&AddVelocity}
    }
   ]
  }
  particleUpdateScript {
   scriptName=ParticleUpdateScript
   modules [
    {
     bIsSetParametersModule=false
     enabled=true
     moduleName=ParticleState
     inputs [
      Kill Particles When Lifetime Has Elapsed {bIsDynamic=false, bIsEditable=true, bIsStaticSwitch=true, bIsVisible=true, type {classStructOrEnum {refPath=&NiagaraBool}}}
      Loop Particles Lifetime {bIsDynamic=false, bIsEditable=false, bIsStaticSwitch=false, bIsVisible=false, type {classStructOrEnum {refPath=&NiagaraBool}}}
      Let Infinitely Lived Particles Die When Emitter Deactivates {bIsDynamic=false, bIsEditable=false, bIsStaticSwitch=false, bIsVisible=false, type {classStructOrEnum {refPath=&NiagaraBool}}}
      Lifetime {bIsDynamic=false, bIsEditable=true, bIsStaticSwitch=false, bIsVisible=true, type {classStructOrEnum {refPath=&NiagaraFloat}}}
      DeltaTime {bIsDynamic=false, bIsEditable=true, bIsStaticSwitch=false, bIsVisible=true, type {classStructOrEnum {refPath=&NiagaraFloat}}}
     ]
     moduleScript {refPath=&ParticleState}
    }
    {
     bIsSetParametersModule=false
     enabled=false
     moduleName=ScaleMeshSize
     inputs [
      Scale Factor {bIsDynamic=true, bIsEditable=true, bIsStaticSwitch=false, bIsVisible=true, type {classStructOrEnum {refPath=&Vector3f}}}
      Initial Mesh Scale {bIsDynamic=false, bIsEditable=true, bIsStaticSwitch=false, bIsVisible=true, type {classStructOrEnum {refPath=&Vector3f}}}
     ]
     moduleScript {refPath=&ScaleMeshSize}
    }
    {
     bIsSetParametersModule=false
     enabled=true
     moduleName=ScaleMeshSize001
     inputs [
      Scale Factor {bIsDynamic=true, bIsEditable=true, bIsStaticSwitch=false, bIsVisible=true, type {classStructOrEnum {refPath=&Vector3f}}}
      Initial Mesh Scale {bIsDynamic=false, bIsEditable=true, bIsStaticSwitch=false, bIsVisible=true, type {classStructOrEnum {refPath=&Vector3f}}}
     ]
     moduleScript {refPath=&ScaleMeshSize}
    }
    {
     bIsSetParametersModule=false
     enabled=false
     moduleName=CurlNoiseForce
     inputs [
      Noise Strength {bIsDynamic=false, bIsEditable=true, bIsStaticSwitch=false, bIsVisible=true, type {classStructOrEnum {refPath=&NiagaraFloat}}}
      Noise Frequency {bIsDynamic=false, bIsEditable=true, bIsStaticSwitch=false, bIsVisible=true, type {classStructOrEnum {refPath=&NiagaraFloat}}}
      Noise Quality / Cost {
       bIsDynamic=false
       bIsEditable=true
       bIsStaticSwitch=true
       bIsVisible=true
       type {classStructOrEnum {refPath=&ENiagaraCurlNoiseQuality}}
      }
      Pan Noise {bIsDynamic=false, bIsEditable=true, bIsStaticSwitch=false, bIsVisible=true, type {classStructOrEnum {refPath=&NiagaraBool}}}
      Pan Noise Field {bIsDynamic=false, bIsEditable=false, bIsStaticSwitch=false, bIsVisible=true, type {classStructOrEnum {refPath=&Vector3f}}}
      Write to Intrinsic Parameters {bIsDynamic=false, bIsEditable=true, bIsStaticSwitch=true, bIsVisible=true, type {classStructOrEnum {refPath=&NiagaraBool}}}
      Sample Position {bIsDynamic=false, bIsEditable=true, bIsStaticSwitch=false, bIsVisible=true, type {classStructOrEnum {refPath=&Vector3f}}}
      Randomize Noise Sample {bIsDynamic=false, bIsEditable=true, bIsStaticSwitch=true, bIsVisible=true, type {classStructOrEnum {refPath=&NiagaraBool}}}
      Random Seed {bIsDynamic=false, bIsEditable=true, bIsStaticSwitch=false, bIsVisible=true, type {classStructOrEnum {refPath=&NiagaraInt32}}}
      Randomization Vector {bIsDynamic=false, bIsEditable=true, bIsStaticSwitch=false, bIsVisible=true, type {classStructOrEnum {refPath=&Vector3f}}}
      Scale Curl Lookup By Engine Owner Scale {bIsDynamic=false, bIsEditable=true, bIsStaticSwitch=true, bIsVisible=true, type {classStructOrEnum {refPath=&NiagaraBool}}}
      CurlNoise {
       bIsDynamic=false
       bIsEditable=false
       bIsStaticSwitch=false
       bIsVisible=false
       type {classStructOrEnum {refPath=&NiagaraDataInterfaceCurlNoise}}
      }
      Mask Curl Noise {bIsDynamic=false, bIsEditable=true, bIsStaticSwitch=true, bIsVisible=true, type {classStructOrEnum {refPath=&NiagaraBool}}}
      Curl Noise Cone Mask Angle {
       bIsDynamic=false
       bIsEditable=false
       bIsStaticSwitch=false
       bIsVisible=false
       type {classStructOrEnum {refPath=&NiagaraFloat}}
      }
      Curl Noise Cone Mask Falloff Angle {
       bIsDynamic=false
       bIsEditable=false
       bIsStaticSwitch=false
       bIsVisible=false
       type {classStructOrEnum {refPath=&NiagaraFloat}}
      }
      Curl Noise Cone Mask Axis {
       bIsDynamic=false
       bIsEditable=false
       bIsStaticSwitch=false
       bIsVisible=false
       type {classStructOrEnum {refPath=&Vector3f}}
      }
     ]
     moduleScript {refPath=&CurlNoiseForce}
    }
    {
     bIsSetParametersModule=false
     enabled=true
     moduleName=Drag
     inputs [
      Use Linear Drag {bIsDynamic=false, bIsEditable=true, bIsStaticSwitch=false, bIsVisible=true, type {classStructOrEnum {refPath=&NiagaraBool}}}
      Drag {bIsDynamic=true, bIsEditable=true, bIsStaticSwitch=false, bIsVisible=true, type {classStructOrEnum {refPath=&NiagaraFloat}}}
      Use Rotational Drag {bIsDynamic=false, bIsEditable=true, bIsStaticSwitch=false, bIsVisible=true, type {classStructOrEnum {refPath=&NiagaraBool}}}
      Rotational Drag {bIsDynamic=false, bIsEditable=false, bIsStaticSwitch=false, bIsVisible=true, type {classStructOrEnum {refPath=&NiagaraFloat}}}
     ]
     moduleScript {refPath=&Drag}
    }
    {
     bIsSetParametersModule=false
     enabled=true
     moduleName=GravityForce
     inputs [
      Gravity {bIsDynamic=false, bIsEditable=true, bIsStaticSwitch=false, bIsVisible=true, type {classStructOrEnum {refPath=&Vector3f}}}
      Coordinate Space {
       bIsDynamic=false
       bIsEditable=true
       bIsStaticSwitch=true
       bIsVisible=true
       type {classStructOrEnum {refPath=&ENiagaraCoordinateSpace}}
      }
     ]
     moduleScript {refPath=&GravityForce}
    }
    {
     bIsSetParametersModule=false
     enabled=true
     moduleName=SolveForcesAndVelocity
     inputs [
      Clamp Velocity {bIsDynamic=false, bIsEditable=true, bIsStaticSwitch=false, bIsVisible=true, type {classStructOrEnum {refPath=&NiagaraBool}}}
      Speed Limit {bIsDynamic=false, bIsEditable=false, bIsStaticSwitch=false, bIsVisible=true, type {classStructOrEnum {refPath=&NiagaraFloat}}}
      Limit Acceleration {bIsDynamic=false, bIsEditable=true, bIsStaticSwitch=false, bIsVisible=true, type {classStructOrEnum {refPath=&NiagaraBool}}}
      Acceleration Limit {bIsDynamic=false, bIsEditable=false, bIsStaticSwitch=false, bIsVisible=true, type {classStructOrEnum {refPath=&NiagaraFloat}}}
      Manually Enable Rotational Solver {bIsDynamic=false, bIsEditable=true, bIsStaticSwitch=false, bIsVisible=true, type {classStructOrEnum {refPath=&NiagaraBool}}}
      Write to Intrinsic Properties {bIsDynamic=false, bIsEditable=true, bIsStaticSwitch=false, bIsVisible=true, type {classStructOrEnum {refPath=&NiagaraBool}}}
      Write to Presolve Properties {bIsDynamic=false, bIsEditable=true, bIsStaticSwitch=true, bIsVisible=true, type {classStructOrEnum {refPath=&NiagaraBool}}}
      DeltaTime {bIsDynamic=false, bIsEditable=true, bIsStaticSwitch=false, bIsVisible=true, type {classStructOrEnum {refPath=&NiagaraFloat}}}
      Mass {bIsDynamic=false, bIsEditable=true, bIsStaticSwitch=false, bIsVisible=true, type {classStructOrEnum {refPath=&NiagaraFloat}}}
      Position {
       bIsDynamic=false
       bIsEditable=true
       bIsStaticSwitch=false
       bIsVisible=true
       type {classStructOrEnum {refPath=&NiagaraPosition}}
      }
      Velocity {bIsDynamic=false, bIsEditable=true, bIsStaticSwitch=false, bIsVisible=true, type {classStructOrEnum {refPath=&Vector3f}}}
      PreviousVelocity {bIsDynamic=false, bIsEditable=true, bIsStaticSwitch=false, bIsVisible=true, type {classStructOrEnum {refPath=&Vector3f}}}
      Force {bIsDynamic=false, bIsEditable=true, bIsStaticSwitch=false, bIsVisible=true, type {classStructOrEnum {refPath=&Vector3f}}}
      Mesh Orientation {bIsDynamic=false, bIsEditable=false, bIsStaticSwitch=false, bIsVisible=false, type {classStructOrEnum {refPath=&Quat4f}}}
      Rotational Velocity {
       bIsDynamic=false
       bIsEditable=false
       bIsStaticSwitch=false
       bIsVisible=false
       type {classStructOrEnum {refPath=&Vector3f}}
      }
      Rotational Force {
       bIsDynamic=false
       bIsEditable=false
       bIsStaticSwitch=false
       bIsVisible=false
       type {classStructOrEnum {refPath=&Vector3f}}
      }
      Rotational Inertia {
       bIsDynamic=false
       bIsEditable=false
       bIsStaticSwitch=false
       bIsVisible=false
       type {classStructOrEnum {refPath=&NiagaraFloat}}
      }
     ]
     moduleScript {refPath=&SolveForcesAndVelocity}
    }
    {
     bIsSetParametersModule=false
     enabled=false
     moduleName=ScaleColor001
     inputs [
      Scale Mode {bIsDynamic=false, bIsEditable=true, bIsStaticSwitch=true, bIsVisible=true, type {classStructOrEnum {refPath=&NiagaraInt32}}}
      Linear Color Curve {
       bIsDynamic=false
       bIsEditable=false
       bIsStaticSwitch=false
       bIsVisible=false
       type {classStructOrEnum {refPath=&NiagaraDataInterfaceColorCurve}}
      }
      ScaleRGB {bIsDynamic=false, bIsEditable=true, bIsStaticSwitch=false, bIsVisible=true, type {classStructOrEnum {refPath=&NiagaraBool}}}
      Scale RGB {bIsDynamic=true, bIsEditable=true, bIsStaticSwitch=false, bIsVisible=true, type {classStructOrEnum {refPath=&Vector3f}}}
      ScaleRGBA {bIsDynamic=false, bIsEditable=false, bIsStaticSwitch=false, bIsVisible=false, type {classStructOrEnum {refPath=&NiagaraBool}}}
      Scale RGBA {
       bIsDynamic=false
       bIsEditable=false
       bIsStaticSwitch=false
       bIsVisible=false
       type {classStructOrEnum {refPath=&Vector4f}}
      }
      Curve Index {
       bIsDynamic=false
       bIsEditable=false
       bIsStaticSwitch=false
       bIsVisible=false
       type {classStructOrEnum {refPath=&NiagaraFloat}}
      }
      ScaleA {bIsDynamic=false, bIsEditable=true, bIsStaticSwitch=false, bIsVisible=true, type {classStructOrEnum {refPath=&NiagaraBool}}}
      Scale Alpha {bIsDynamic=false, bIsEditable=false, bIsStaticSwitch=false, bIsVisible=true, type {classStructOrEnum {refPath=&NiagaraFloat}}}
      Color Value To Scale {
       bIsDynamic=false
       bIsEditable=true
       bIsStaticSwitch=false
       bIsVisible=true
       type {classStructOrEnum {refPath=&LinearColor}}
      }
      Color Scale {
       bIsDynamic=false
       bIsEditable=false
       bIsStaticSwitch=false
       bIsVisible=false
       type {classStructOrEnum {refPath=&LinearColor}}
      }
     ]
     moduleScript {refPath=&ScaleColor}
    }
    {
     bIsSetParametersModule=false
     enabled=true
     moduleName=ScaleColor002
     inputs [
      Scale Mode {bIsDynamic=false, bIsEditable=true, bIsStaticSwitch=true, bIsVisible=true, type {classStructOrEnum {refPath=&NiagaraInt32}}}
      Linear Color Curve {
       bIsDynamic=false
       bIsEditable=false
       bIsStaticSwitch=false
       bIsVisible=false
       type {classStructOrEnum {refPath=&NiagaraDataInterfaceColorCurve}}
      }
      ScaleRGB {bIsDynamic=false, bIsEditable=true, bIsStaticSwitch=false, bIsVisible=true, type {classStructOrEnum {refPath=&NiagaraBool}}}
      Scale RGB {bIsDynamic=true, bIsEditable=true, bIsStaticSwitch=false, bIsVisible=true, type {classStructOrEnum {refPath=&Vector3f}}}
      ScaleRGBA {bIsDynamic=false, bIsEditable=false, bIsStaticSwitch=false, bIsVisible=false, type {classStructOrEnum {refPath=&NiagaraBool}}}
      Scale RGBA {
       bIsDynamic=false
       bIsEditable=false
       bIsStaticSwitch=false
       bIsVisible=false
       type {classStructOrEnum {refPath=&Vector4f}}
      }
      Curve Index {
       bIsDynamic=false
       bIsEditable=false
       bIsStaticSwitch=false
       bIsVisible=false
       type {classStructOrEnum {refPath=&NiagaraFloat}}
      }
      ScaleA {bIsDynamic=false, bIsEditable=true, bIsStaticSwitch=false, bIsVisible=true, type {classStructOrEnum {refPath=&NiagaraBool}}}
      Scale Alpha {bIsDynamic=false, bIsEditable=false, bIsStaticSwitch=false, bIsVisible=true, type {classStructOrEnum {refPath=&NiagaraFloat}}}
      Color Value To Scale {
       bIsDynamic=false
       bIsEditable=true
       bIsStaticSwitch=false
       bIsVisible=true
       type {classStructOrEnum {refPath=&LinearColor}}
      }
      Color Scale {
       bIsDynamic=false
       bIsEditable=false
       bIsStaticSwitch=false
       bIsVisible=false
       type {classStructOrEnum {refPath=&LinearColor}}
      }
     ]
     moduleScript {refPath=&ScaleColor}
    }
   ]
  }
  rendererClasses [{refPath=&NiagaraMeshRendererProperties}]
  renderers [
   {rendererIndex=0, rendererClass {refPath=&NiagaraMeshRendererProperties}}
   {rendererIndex=1, rendererClass {refPath=&NiagaraMeshRendererProperties}}
  ]
 }
 Snowflakes {
  bEnabled=false
  emitterSpawnScript {scriptName=EmitterSpawnScript, modules []}
  emitterUpdateScript {
   scriptName=EmitterUpdateScript
   modules [
    {
     bIsSetParametersModule=false
     enabled=true
     moduleName=EmitterState
     inputs [
      Life Cycle Mode {
       bIsDynamic=false
       bIsEditable=true
       bIsStaticSwitch=true
       bIsVisible=true
       type {classStructOrEnum {refPath=&ENiagaraEmitterLifeCycleMode}}
      }
      Inactive Response {
       bIsDynamic=false
       bIsEditable=true
       bIsStaticSwitch=true
       bIsVisible=true
       type {classStructOrEnum {refPath=&ENiagaraInactiveMode}}
      }
      Loop Behavior {
       bIsDynamic=false
       bIsEditable=true
       bIsStaticSwitch=true
       bIsVisible=true
       type {classStructOrEnum {refPath=&ENiagara_EmitterStateOptions}}
      }
      Loop Duration Mode {
       bIsDynamic=false
       bIsEditable=true
       bIsStaticSwitch=true
       bIsVisible=true
       type {classStructOrEnum {refPath=&ENiagara_InfiniteLoopDuration}}
      }
      Loop Count {
       bIsDynamic=false
       bIsEditable=false
       bIsStaticSwitch=false
       bIsVisible=false
       type {classStructOrEnum {refPath=&NiagaraInt32}}
      }
      Loop Duration {bIsDynamic=false, bIsEditable=true, bIsStaticSwitch=false, bIsVisible=true, type {classStructOrEnum {refPath=&NiagaraFloat}}}
      Recalculate Duration Each Loop {bIsDynamic=false, bIsEditable=false, bIsStaticSwitch=false, bIsVisible=false, type {classStructOrEnum {refPath=&NiagaraBool}}}
      UseLoopDelay {bIsDynamic=false, bIsEditable=true, bIsStaticSwitch=false, bIsVisible=true, type {classStructOrEnum {refPath=&NiagaraBool}}}
      Loop Delay {bIsDynamic=false, bIsEditable=false, bIsStaticSwitch=false, bIsVisible=true, type {classStructOrEnum {refPath=&NiagaraFloat}}}
      Delay First Loop Only {bIsDynamic=false, bIsEditable=false, bIsStaticSwitch=false, bIsVisible=false, type {classStructOrEnum {refPath=&NiagaraBool}}}
      Scalability Mode {
       bIsDynamic=false
       bIsEditable=true
       bIsStaticSwitch=true
       bIsVisible=true
       type {classStructOrEnum {refPath=&ENiagaraEmitterScalabilityMode_Limited}}
      }
      Enable Distance Culling {bIsDynamic=false, bIsEditable=false, bIsStaticSwitch=true, bIsVisible=false, type {classStructOrEnum {refPath=&NiagaraBool}}}
      MinDistance {
       bIsDynamic=false
       bIsEditable=false
       bIsStaticSwitch=false
       bIsVisible=false
       type {classStructOrEnum {refPath=&NiagaraFloat}}
      }
      Min Distance Response {
       bIsDynamic=false
       bIsEditable=false
       bIsStaticSwitch=false
       bIsVisible=false
       type {classStructOrEnum {refPath=&ENiagaraExecutionStateManagement}}
      }
      MaxDistance {
       bIsDynamic=false
       bIsEditable=false
       bIsStaticSwitch=false
       bIsVisible=false
       type {classStructOrEnum {refPath=&NiagaraFloat}}
      }
      Max Distance Response {
       bIsDynamic=false
       bIsEditable=false
       bIsStaticSwitch=false
       bIsVisible=false
       type {classStructOrEnum {refPath=&ENiagaraExecutionStateManagement}}
      }
      Scale Spawn Count {bIsDynamic=false, bIsEditable=false, bIsStaticSwitch=true, bIsVisible=false, type {classStructOrEnum {refPath=&NiagaraBool}}}
      Spawn Count Scale {
       bIsDynamic=false
       bIsEditable=false
       bIsStaticSwitch=false
       bIsVisible=false
       type {classStructOrEnum {refPath=&NiagaraFloat}}
      }
      Spawn Count Scale By Distance {
       bIsDynamic=false
       bIsEditable=false
       bIsStaticSwitch=false
       bIsVisible=false
       type {classStructOrEnum {refPath=&NiagaraDataInterfaceCurve}}
      }
      Enable Visibility Culling {bIsDynamic=false, bIsEditable=false, bIsStaticSwitch=true, bIsVisible=false, type {classStructOrEnum {refPath=&NiagaraBool}}}
      Visibility Cull Response {
       bIsDynamic=false
       bIsEditable=false
       bIsStaticSwitch=false
       bIsVisible=false
       type {classStructOrEnum {refPath=&ENiagaraExecutionStateManagement}}
      }
      Visibility Cull Delay {
       bIsDynamic=false
       bIsEditable=false
       bIsStaticSwitch=false
       bIsVisible=false
       type {classStructOrEnum {refPath=&NiagaraFloat}}
      }
      Reset Age On Awaken {bIsDynamic=false, bIsEditable=false, bIsStaticSwitch=false, bIsVisible=false, type {classStructOrEnum {refPath=&NiagaraBool}}}
     ]
     moduleScript {refPath=&EmitterState}
    }
    {
     bIsSetParametersModule=false
     enabled=true
     moduleName=SpawnBurst_Instantaneous
     inputs [
      Spawn Count {bIsDynamic=false, bIsEditable=true, bIsStaticSwitch=false, bIsVisible=true, type {classStructOrEnum {refPath=&NiagaraInt32}}}
      Spawn Time {bIsDynamic=false, bIsEditable=true, bIsStaticSwitch=false, bIsVisible=true, type {classStructOrEnum {refPath=&NiagaraFloat}}}
      Use Spawn Probability {bIsDynamic=false, bIsEditable=true, bIsStaticSwitch=false, bIsVisible=true, type {classStructOrEnum {refPath=&NiagaraBool}}}
      Spawn Probability {bIsDynamic=false, bIsEditable=false, bIsStaticSwitch=false, bIsVisible=true, type {classStructOrEnum {refPath=&NiagaraFloat}}}
      UseLoopCountLimit {bIsDynamic=false, bIsEditable=true, bIsStaticSwitch=false, bIsVisible=true, type {classStructOrEnum {refPath=&NiagaraBool}}}
      Loop Count Limit {bIsDynamic=false, bIsEditable=false, bIsStaticSwitch=false, bIsVisible=true, type {classStructOrEnum {refPath=&NiagaraInt32}}}
      Spawn Group {bIsDynamic=false, bIsEditable=true, bIsStaticSwitch=false, bIsVisible=true, type {classStructOrEnum {refPath=&NiagaraInt32}}}
      Age {bIsDynamic=false, bIsEditable=true, bIsStaticSwitch=false, bIsVisible=true, type {classStructOrEnum {refPath=&NiagaraFloat}}}
     ]
     moduleScript {refPath=&SpawnBurst_Instantaneous}
    }
   ]
  }
  particleSpawnScript {
   scriptName=ParticleSpawnScript
   modules [
    {
     bIsSetParametersModule=false
     enabled=true
     moduleName=InitializeParticle001
     inputs [
      Lifetime Mode {
       bIsDynamic=false
       bIsEditable=true
       bIsStaticSwitch=true
       bIsVisible=true
       type {classStructOrEnum {refPath=&ENiagara_LifetimeMode}}
      }
      Lifetime Min {bIsDynamic=false, bIsEditable=true, bIsStaticSwitch=false, bIsVisible=true, type {classStructOrEnum {refPath=&NiagaraFloat}}}
      Lifetime {
       bIsDynamic=false
       bIsEditable=false
       bIsStaticSwitch=false
       bIsVisible=false
       type {classStructOrEnum {refPath=&NiagaraFloat}}
      }
      Lifetime Max {bIsDynamic=false, bIsEditable=true, bIsStaticSwitch=false, bIsVisible=true, type {classStructOrEnum {refPath=&NiagaraFloat}}}
      Lifetime Randomness Mode {
       bIsDynamic=false
       bIsEditable=true
       bIsStaticSwitch=true
       bIsVisible=true
       type {classStructOrEnum {refPath=&ENiagaraRandomnessMode}}
      }
      Color Mode {
       bIsDynamic=false
       bIsEditable=true
       bIsStaticSwitch=true
       bIsVisible=true
       type {classStructOrEnum {refPath=&ENiagara_ColorInitializationMode}}
      }
      Color Minimum {
       bIsDynamic=false
       bIsEditable=false
       bIsStaticSwitch=false
       bIsVisible=false
       type {classStructOrEnum {refPath=&LinearColor}}
      }
      Color {
       bIsDynamic=true
       bIsEditable=true
       bIsStaticSwitch=false
       bIsVisible=true
       type {classStructOrEnum {refPath=&LinearColor}}
      }
      AdjustHue {bIsDynamic=false, bIsEditable=false, bIsStaticSwitch=false, bIsVisible=false, type {classStructOrEnum {refPath=&NiagaraBool}}}
      Hue Shift Range {
       bIsDynamic=false
       bIsEditable=false
       bIsStaticSwitch=false
       bIsVisible=false
       type {classStructOrEnum {refPath=&Vector2f}}
      }
      Color Maximum {
       bIsDynamic=false
       bIsEditable=false
       bIsStaticSwitch=false
       bIsVisible=false
       type {classStructOrEnum {refPath=&LinearColor}}
      }
      AdjustSaturation {bIsDynamic=false, bIsEditable=false, bIsStaticSwitch=false, bIsVisible=false, type {classStructOrEnum {refPath=&NiagaraBool}}}
      Saturation Range {
       bIsDynamic=false
       bIsEditable=false
       bIsStaticSwitch=false
       bIsVisible=false
       type {classStructOrEnum {refPath=&Vector2f}}
      }
      AdjustValue {bIsDynamic=false, bIsEditable=false, bIsStaticSwitch=false, bIsVisible=false, type {classStructOrEnum {refPath=&NiagaraBool}}}
      Value Range {
       bIsDynamic=false
       bIsEditable=false
       bIsStaticSwitch=false
       bIsVisible=false
       type {classStructOrEnum {refPath=&Vector2f}}
      }
      AdjustAlpha {bIsDynamic=false, bIsEditable=false, bIsStaticSwitch=false, bIsVisible=false, type {classStructOrEnum {refPath=&NiagaraBool}}}
      Alpha Scale Range {
       bIsDynamic=false
       bIsEditable=false
       bIsStaticSwitch=false
       bIsVisible=false
       type {classStructOrEnum {refPath=&Vector2f}}
      }
      Color Channel Mode {
       bIsDynamic=false
       bIsEditable=false
       bIsStaticSwitch=true
       bIsVisible=false
       type {classStructOrEnum {refPath=&ENiagaraChannelCorrelation}}
      }
      Color Randomness Mode {
       bIsDynamic=false
       bIsEditable=false
       bIsStaticSwitch=true
       bIsVisible=false
       type {classStructOrEnum {refPath=&ENiagaraRandomnessMode}}
      }
      Position Mode {
       bIsDynamic=false
       bIsEditable=true
       bIsStaticSwitch=true
       bIsVisible=true
       type {classStructOrEnum {refPath=&ENiagara_PositionInitializationMode}}
      }
      Position {
       bIsDynamic=false
       bIsEditable=false
       bIsStaticSwitch=false
       bIsVisible=false
       type {classStructOrEnum {refPath=&NiagaraPosition}}
      }
      UsePositionOffset {bIsDynamic=false, bIsEditable=true, bIsStaticSwitch=false, bIsVisible=true, type {classStructOrEnum {refPath=&NiagaraBool}}}
      Position Offset {bIsDynamic=false, bIsEditable=false, bIsStaticSwitch=false, bIsVisible=true, type {classStructOrEnum {refPath=&Vector3f}}}
      Position Offset Coordinate Space {
       bIsDynamic=false
       bIsEditable=false
       bIsStaticSwitch=false
       bIsVisible=false
       type {classStructOrEnum {refPath=&ENiagaraCoordinateSpace}}
      }
      Mass Mode {
       bIsDynamic=false
       bIsEditable=true
       bIsStaticSwitch=true
       bIsVisible=true
       type {classStructOrEnum {refPath=&ENiagara_MassInitializationMode}}
      }
      Mass Min {
       bIsDynamic=false
       bIsEditable=false
       bIsStaticSwitch=false
       bIsVisible=false
       type {classStructOrEnum {refPath=&NiagaraFloat}}
      }
      Mass {
       bIsDynamic=false
       bIsEditable=false
       bIsStaticSwitch=false
       bIsVisible=false
       type {classStructOrEnum {refPath=&NiagaraFloat}}
      }
      Mass Max {
       bIsDynamic=false
       bIsEditable=false
       bIsStaticSwitch=false
       bIsVisible=false
       type {classStructOrEnum {refPath=&NiagaraFloat}}
      }
      Mass Randomness Mode {
       bIsDynamic=false
       bIsEditable=false
       bIsStaticSwitch=true
       bIsVisible=false
       type {classStructOrEnum {refPath=&ENiagaraRandomnessMode}}
      }
      Write Material Random {bIsDynamic=false, bIsEditable=true, bIsStaticSwitch=false, bIsVisible=true, type {classStructOrEnum {refPath=&NiagaraBool}}}
      Material Random {bIsDynamic=false, bIsEditable=true, bIsStaticSwitch=false, bIsVisible=true, type {classStructOrEnum {refPath=&NiagaraFloat}}}
      Sprite Size Mode {
       bIsDynamic=false
       bIsEditable=true
       bIsStaticSwitch=true
       bIsVisible=true
       type {classStructOrEnum {refPath=&ENiagara_SizeScaleMode}}
      }
      Sprite Size {
       bIsDynamic=false
       bIsEditable=false
       bIsStaticSwitch=false
       bIsVisible=false
       type {classStructOrEnum {refPath=&Vector2f}}
      }
      Uniform Sprite Size {bIsDynamic=true, bIsEditable=true, bIsStaticSwitch=false, bIsVisible=true, type {classStructOrEnum {refPath=&NiagaraFloat}}}
      Uniform Sprite Size Min {
       bIsDynamic=false
       bIsEditable=false
       bIsStaticSwitch=false
       bIsVisible=false
       type {classStructOrEnum {refPath=&NiagaraFloat}}
      }
      Sprite Size Min {
       bIsDynamic=false
       bIsEditable=false
       bIsStaticSwitch=false
       bIsVisible=false
       type {classStructOrEnum {refPath=&Vector2f}}
      }
      Sprite Size Max {
       bIsDynamic=false
       bIsEditable=false
       bIsStaticSwitch=false
       bIsVisible=false
       type {classStructOrEnum {refPath=&Vector2f}}
      }
      Uniform Sprite Size Max {
       bIsDynamic=false
       bIsEditable=false
       bIsStaticSwitch=false
       bIsVisible=false
       type {classStructOrEnum {refPath=&NiagaraFloat}}
      }
      Sprite Size Randomness Mode {
       bIsDynamic=false
       bIsEditable=false
       bIsStaticSwitch=true
       bIsVisible=false
       type {classStructOrEnum {refPath=&ENiagaraRandomnessMode}}
      }
      Sprite Rotation Mode {
       bIsDynamic=false
       bIsEditable=true
       bIsStaticSwitch=true
       bIsVisible=true
       type {classStructOrEnum {refPath=&ENiagara_SpriteRotationMode}}
      }
      Sprite Rotation Angle Min {
       bIsDynamic=false
       bIsEditable=false
       bIsStaticSwitch=false
       bIsVisible=false
       type {classStructOrEnum {refPath=&NiagaraFloat}}
      }
      Sprite Rotation Angle {
       bIsDynamic=false
       bIsEditable=false
       bIsStaticSwitch=false
       bIsVisible=false
       type {classStructOrEnum {refPath=&NiagaraFloat}}
      }
      Sprite Rotation Angle Max {
       bIsDynamic=false
       bIsEditable=false
       bIsStaticSwitch=false
       bIsVisible=false
       type {classStructOrEnum {refPath=&NiagaraFloat}}
      }
      Sprite Rotation Randomness Mode {
       bIsDynamic=false
       bIsEditable=false
       bIsStaticSwitch=true
       bIsVisible=false
       type {classStructOrEnum {refPath=&ENiagaraRandomnessMode}}
      }
      Sprite UV Mode {
       bIsDynamic=false
       bIsEditable=true
       bIsStaticSwitch=true
       bIsVisible=true
       type {classStructOrEnum {refPath=&ENiagara_UVFlippingMode}}
      }
      Sprite UV Randomness Mode {
       bIsDynamic=false
       bIsEditable=false
       bIsStaticSwitch=true
       bIsVisible=false
       type {classStructOrEnum {refPath=&ENiagaraRandomnessMode}}
      }
      Sprite UV Scale {
       bIsDynamic=false
       bIsEditable=false
       bIsStaticSwitch=false
       bIsVisible=false
       type {classStructOrEnum {refPath=&Vector2f}}
      }
      Mesh Scale Mode {
       bIsDynamic=false
       bIsEditable=true
       bIsStaticSwitch=true
       bIsVisible=true
       type {classStructOrEnum {refPath=&ENiagara_SizeScaleMode}}
      }
      Mesh Uniform Scale Min {
       bIsDynamic=false
       bIsEditable=false
       bIsStaticSwitch=false
       bIsVisible=false
       type {classStructOrEnum {refPath=&NiagaraFloat}}
      }
      Mesh Uniform Scale {
       bIsDynamic=false
       bIsEditable=false
       bIsStaticSwitch=false
       bIsVisible=false
       type {classStructOrEnum {refPath=&NiagaraFloat}}
      }
      Mesh Scale Min {
       bIsDynamic=false
       bIsEditable=false
       bIsStaticSwitch=false
       bIsVisible=false
       type {classStructOrEnum {refPath=&Vector3f}}
      }
      Mesh Scale {
       bIsDynamic=false
       bIsEditable=false
       bIsStaticSwitch=false
       bIsVisible=false
       type {classStructOrEnum {refPath=&Vector3f}}
      }
      Mesh Scale Max {
       bIsDynamic=false
       bIsEditable=false
       bIsStaticSwitch=false
       bIsVisible=false
       type {classStructOrEnum {refPath=&Vector3f}}
      }
      Mesh Uniform Scale Max {
       bIsDynamic=false
       bIsEditable=false
       bIsStaticSwitch=false
       bIsVisible=false
       type {classStructOrEnum {refPath=&NiagaraFloat}}
      }
      Mesh Scale Randomness Mode {
       bIsDynamic=false
       bIsEditable=false
       bIsStaticSwitch=true
       bIsVisible=false
       type {classStructOrEnum {refPath=&ENiagaraRandomnessMode}}
      }
      Mesh Renderer Array Visibility Mode {
       bIsDynamic=false
       bIsEditable=true
       bIsStaticSwitch=true
       bIsVisible=true
       type {classStructOrEnum {refPath=&ENiagara_UnsetDirectSetRandom}}
      }
      Mesh Renderer Mesh Index {
       bIsDynamic=false
       bIsEditable=false
       bIsStaticSwitch=false
       bIsVisible=false
       type {classStructOrEnum {refPath=&NiagaraInt32}}
      }
      Wrap Index By Mesh Count {bIsDynamic=false, bIsEditable=false, bIsStaticSwitch=true, bIsVisible=false, type {classStructOrEnum {refPath=&NiagaraBool}}}
      Mesh Renderer Info {
       bIsDynamic=false
       bIsEditable=false
       bIsStaticSwitch=false
       bIsVisible=false
       type {classStructOrEnum {refPath=&NiagaraDataInterfaceMeshRendererInfo}}
      }
      Mesh Renderer Random Seed {
       bIsDynamic=false
       bIsEditable=false
       bIsStaticSwitch=false
       bIsVisible=false
       type {classStructOrEnum {refPath=&NiagaraInt32}}
      }
      Ribbon Width Mode {
       bIsDynamic=false
       bIsEditable=true
       bIsStaticSwitch=true
       bIsVisible=true
       type {classStructOrEnum {refPath=&ENiagara_UnsetDirectSet}}
      }
      Ribbon Width {
       bIsDynamic=false
       bIsEditable=false
       bIsStaticSwitch=false
       bIsVisible=false
       type {classStructOrEnum {refPath=&NiagaraFloat}}
      }
      Ribbon Facing Mode {
       bIsDynamic=false
       bIsEditable=true
       bIsStaticSwitch=true
       bIsVisible=true
       type {classStructOrEnum {refPath=&ENiagara_UnsetDirectSet}}
      }
      Ribbon Facing Vector {
       bIsDynamic=false
       bIsEditable=false
       bIsStaticSwitch=false
       bIsVisible=false
       type {classStructOrEnum {refPath=&Vector3f}}
      }
      Ribbon Twist Mode {
       bIsDynamic=false
       bIsEditable=true
       bIsStaticSwitch=true
       bIsVisible=true
       type {classStructOrEnum {refPath=&ENiagara_UnsetDirectSet}}
      }
      "Ribbon Twist " {
       bIsDynamic=false
       bIsEditable=false
       bIsStaticSwitch=false
       bIsVisible=false
       type {classStructOrEnum {refPath=&NiagaraFloat}}
      }
     ]
     moduleScript {refPath=&InitializeParticle}
    }
    {
     bIsSetParametersModule=false
     enabled=true
     moduleName=SphereLocation
     inputs [
      Sphere Radius {bIsDynamic=false, bIsEditable=true, bIsStaticSwitch=false, bIsVisible=true, type {classStructOrEnum {refPath=&NiagaraFloat}}}
      UseOffset {bIsDynamic=false, bIsEditable=true, bIsStaticSwitch=false, bIsVisible=true, type {classStructOrEnum {refPath=&NiagaraBool}}}
      Offset {bIsDynamic=false, bIsEditable=true, bIsStaticSwitch=false, bIsVisible=true, type {classStructOrEnum {refPath=&Vector3f}}}
      Offset Coordinate Space {
       bIsDynamic=false
       bIsEditable=true
       bIsStaticSwitch=false
       bIsVisible=true
       type {classStructOrEnum {refPath=&ENiagaraCoordinateSpace}}
      }
      UseNonUniformScale {bIsDynamic=false, bIsEditable=true, bIsStaticSwitch=false, bIsVisible=true, type {classStructOrEnum {refPath=&NiagaraBool}}}
      Non Uniform Scale {bIsDynamic=false, bIsEditable=false, bIsStaticSwitch=false, bIsVisible=true, type {classStructOrEnum {refPath=&Vector3f}}}
      UseSphereOrientation {bIsDynamic=false, bIsEditable=true, bIsStaticSwitch=false, bIsVisible=true, type {classStructOrEnum {refPath=&NiagaraBool}}}
      Sphere Orientation Axis {bIsDynamic=false, bIsEditable=false, bIsStaticSwitch=false, bIsVisible=true, type {classStructOrEnum {refPath=&Vector3f}}}
      Sphere Origin {
       bIsDynamic=false
       bIsEditable=true
       bIsStaticSwitch=false
       bIsVisible=true
       type {classStructOrEnum {refPath=&NiagaraPosition}}
      }
      Sphere Coordinate Space {
       bIsDynamic=false
       bIsEditable=true
       bIsStaticSwitch=false
       bIsVisible=true
       type {classStructOrEnum {refPath=&ENiagaraCoordinateSpace}}
      }
      Write to Intrinsic Variables {bIsDynamic=false, bIsEditable=true, bIsStaticSwitch=true, bIsVisible=true, type {classStructOrEnum {refPath=&NiagaraBool}}}
      Sphere Distribution {
       bIsDynamic=false
       bIsEditable=true
       bIsStaticSwitch=true
       bIsVisible=true
       type {classStructOrEnum {refPath=&ENiagaraSphereDistributionMode}}
      }
      Evaluation Type {
       bIsDynamic=false
       bIsEditable=false
       bIsStaticSwitch=true
       bIsVisible=false
       type {classStructOrEnum {refPath=&ENiagaraRandomnessEvaluation}}
      }
      Hemisphere X {bIsDynamic=false, bIsEditable=true, bIsStaticSwitch=false, bIsVisible=true, type {classStructOrEnum {refPath=&NiagaraBool}}}
      Uniform Distribution {
       bIsDynamic=false
       bIsEditable=false
       bIsStaticSwitch=false
       bIsVisible=false
       type {classStructOrEnum {refPath=&NiagaraFloat}}
      }
      Hemisphere Y {bIsDynamic=false, bIsEditable=true, bIsStaticSwitch=false, bIsVisible=true, type {classStructOrEnum {refPath=&NiagaraBool}}}
      U Position {
       bIsDynamic=false
       bIsEditable=false
       bIsStaticSwitch=false
       bIsVisible=false
       type {classStructOrEnum {refPath=&NiagaraFloat}}
      }
      Uniform Spiral Amount {
       bIsDynamic=false
       bIsEditable=false
       bIsStaticSwitch=false
       bIsVisible=false
       type {classStructOrEnum {refPath=&NiagaraFloat}}
      }
      Hemisphere Z {bIsDynamic=false, bIsEditable=true, bIsStaticSwitch=false, bIsVisible=true, type {classStructOrEnum {refPath=&NiagaraBool}}}
      Uniform Endpoint Offset {
       bIsDynamic=false
       bIsEditable=false
       bIsStaticSwitch=false
       bIsVisible=false
       type {classStructOrEnum {refPath=&NiagaraFloat}}
      }
      V Position {
       bIsDynamic=false
       bIsEditable=false
       bIsStaticSwitch=false
       bIsVisible=false
       type {classStructOrEnum {refPath=&NiagaraFloat}}
      }
      Radius Position {
       bIsDynamic=false
       bIsEditable=false
       bIsStaticSwitch=false
       bIsVisible=false
       type {classStructOrEnum {refPath=&NiagaraFloat}}
      }
      Surface Only {bIsDynamic=false, bIsEditable=true, bIsStaticSwitch=false, bIsVisible=true, type {classStructOrEnum {refPath=&NiagaraBool}}}
      Surface Only Band Thickness {bIsDynamic=false, bIsEditable=true, bIsStaticSwitch=false, bIsVisible=true, type {classStructOrEnum {refPath=&NiagaraFloat}}}
      Surface Expansion Mode {
       bIsDynamic=false
       bIsEditable=true
       bIsStaticSwitch=false
       bIsVisible=true
       type {classStructOrEnum {refPath=&ENiagaraExpansionMode}}
      }
      Normalize Thickness {bIsDynamic=false, bIsEditable=true, bIsStaticSwitch=false, bIsVisible=true, type {classStructOrEnum {refPath=&NiagaraBool}}}
      Randomness Mode {
       bIsDynamic=false
       bIsEditable=true
       bIsStaticSwitch=true
       bIsVisible=true
       type {classStructOrEnum {refPath=&ENiagaraRandomnessMode}}
      }
      Override Seed {bIsDynamic=false, bIsEditable=true, bIsStaticSwitch=false, bIsVisible=true, type {classStructOrEnum {refPath=&NiagaraBool}}}
      Random Seed {bIsDynamic=false, bIsEditable=false, bIsStaticSwitch=false, bIsVisible=true, type {classStructOrEnum {refPath=&NiagaraInt32}}}
      Fixed Random Seed {bIsDynamic=false, bIsEditable=false, bIsStaticSwitch=true, bIsVisible=false, type {classStructOrEnum {refPath=&NiagaraBool}}}
      MaskSpawnGroup {bIsDynamic=false, bIsEditable=true, bIsStaticSwitch=false, bIsVisible=true, type {classStructOrEnum {refPath=&NiagaraBool}}}
      Spawn Group Mask {bIsDynamic=false, bIsEditable=false, bIsStaticSwitch=false, bIsVisible=true, type {classStructOrEnum {refPath=&NiagaraInt32}}}
     ]
     moduleScript {refPath=&SphereLocation}
    }
    {
     bIsSetParametersModule=false
     enabled=true
     moduleName=AddVelocityFromPoint
     inputs [
      Velocity Strength {bIsDynamic=true, bIsEditable=true, bIsStaticSwitch=false, bIsVisible=true, type {classStructOrEnum {refPath=&NiagaraFloat}}}
      Origin Offset {bIsDynamic=false, bIsEditable=true, bIsStaticSwitch=false, bIsVisible=true, type {classStructOrEnum {refPath=&Vector3f}}}
      Origin Offset Coordinate Space {
       bIsDynamic=false
       bIsEditable=true
       bIsStaticSwitch=false
       bIsVisible=true
       type {classStructOrEnum {refPath=&ENiagaraCoordinateSpace}}
      }
      Default Position {bIsDynamic=false, bIsEditable=true, bIsStaticSwitch=false, bIsVisible=true, type {classStructOrEnum {refPath=&Vector3f}}}
      Module.Use Falloff Distance {bIsDynamic=false, bIsEditable=true, bIsStaticSwitch=false, bIsVisible=true, type {classStructOrEnum {refPath=&NiagaraBool}}}
      Velocity Falloff Distance {bIsDynamic=false, bIsEditable=false, bIsStaticSwitch=false, bIsVisible=true, type {classStructOrEnum {refPath=&NiagaraFloat}}}
      Invert Velocity Falloff {bIsDynamic=false, bIsEditable=false, bIsStaticSwitch=false, bIsVisible=false, type {classStructOrEnum {refPath=&NiagaraBool}}}
      Velocity Origin {bIsDynamic=false, bIsEditable=true, bIsStaticSwitch=false, bIsVisible=true, type {classStructOrEnum {refPath=&Vector3f}}}
     ]
     moduleScript {refPath=&AddVelocityFromPoint}
    }
   ]
  }
  particleUpdateScript {
   scriptName=ParticleUpdateScript
   modules [
    {
     bIsSetParametersModule=false
     enabled=true
     moduleName=ParticleState
     inputs [
      Kill Particles When Lifetime Has Elapsed {bIsDynamic=false, bIsEditable=true, bIsStaticSwitch=true, bIsVisible=true, type {classStructOrEnum {refPath=&NiagaraBool}}}
      Loop Particles Lifetime {bIsDynamic=false, bIsEditable=false, bIsStaticSwitch=false, bIsVisible=false, type {classStructOrEnum {refPath=&NiagaraBool}}}
      Let Infinitely Lived Particles Die When Emitter Deactivates {bIsDynamic=false, bIsEditable=false, bIsStaticSwitch=false, bIsVisible=false, type {classStructOrEnum {refPath=&NiagaraBool}}}
      Lifetime {bIsDynamic=false, bIsEditable=true, bIsStaticSwitch=false, bIsVisible=true, type {classStructOrEnum {refPath=&NiagaraFloat}}}
      DeltaTime {bIsDynamic=false, bIsEditable=true, bIsStaticSwitch=false, bIsVisible=true, type {classStructOrEnum {refPath=&NiagaraFloat}}}
     ]
     moduleScript {refPath=&ParticleState}
    }
    {
     bIsSetParametersModule=false
     enabled=true
     moduleName=CurlNoiseForce
     inputs [
      Noise Strength {bIsDynamic=false, bIsEditable=true, bIsStaticSwitch=false, bIsVisible=true, type {classStructOrEnum {refPath=&NiagaraFloat}}}
      Noise Frequency {bIsDynamic=false, bIsEditable=true, bIsStaticSwitch=false, bIsVisible=true, type {classStructOrEnum {refPath=&NiagaraFloat}}}
      Noise Quality / Cost {
       bIsDynamic=false
       bIsEditable=true
       bIsStaticSwitch=true
       bIsVisible=true
       type {classStructOrEnum {refPath=&ENiagaraCurlNoiseQuality}}
      }
      Pan Noise {bIsDynamic=false, bIsEditable=true, bIsStaticSwitch=false, bIsVisible=true, type {classStructOrEnum {refPath=&NiagaraBool}}}
      Pan Noise Field {bIsDynamic=false, bIsEditable=false, bIsStaticSwitch=false, bIsVisible=true, type {classStructOrEnum {refPath=&Vector3f}}}
      Write to Intrinsic Parameters {bIsDynamic=false, bIsEditable=true, bIsStaticSwitch=true, bIsVisible=true, type {classStructOrEnum {refPath=&NiagaraBool}}}
      Sample Position {bIsDynamic=false, bIsEditable=true, bIsStaticSwitch=false, bIsVisible=true, type {classStructOrEnum {refPath=&Vector3f}}}
      Randomize Noise Sample {bIsDynamic=false, bIsEditable=true, bIsStaticSwitch=true, bIsVisible=true, type {classStructOrEnum {refPath=&NiagaraBool}}}
      Random Seed {bIsDynamic=false, bIsEditable=true, bIsStaticSwitch=false, bIsVisible=true, type {classStructOrEnum {refPath=&NiagaraInt32}}}
      Randomization Vector {bIsDynamic=false, bIsEditable=true, bIsStaticSwitch=false, bIsVisible=true, type {classStructOrEnum {refPath=&Vector3f}}}
      Scale Curl Lookup By Engine Owner Scale {bIsDynamic=false, bIsEditable=true, bIsStaticSwitch=true, bIsVisible=true, type {classStructOrEnum {refPath=&NiagaraBool}}}
      CurlNoise {
       bIsDynamic=false
       bIsEditable=false
       bIsStaticSwitch=false
       bIsVisible=false
       type {classStructOrEnum {refPath=&NiagaraDataInterfaceCurlNoise}}
      }
      Mask Curl Noise {bIsDynamic=false, bIsEditable=true, bIsStaticSwitch=true, bIsVisible=true, type {classStructOrEnum {refPath=&NiagaraBool}}}
      Curl Noise Cone Mask Angle {
       bIsDynamic=false
       bIsEditable=false
       bIsStaticSwitch=false
       bIsVisible=false
       type {classStructOrEnum {refPath=&NiagaraFloat}}
      }
      Curl Noise Cone Mask Falloff Angle {
       bIsDynamic=false
       bIsEditable=false
       bIsStaticSwitch=false
       bIsVisible=false
       type {classStructOrEnum {refPath=&NiagaraFloat}}
      }
      Curl Noise Cone Mask Axis {
       bIsDynamic=false
       bIsEditable=false
       bIsStaticSwitch=false
       bIsVisible=false
       type {classStructOrEnum {refPath=&Vector3f}}
      }
     ]
     moduleScript {refPath=&CurlNoiseForce}
    }
    {
     bIsSetParametersModule=false
     enabled=true
     moduleName=Drag
     inputs [
      Use Linear Drag {bIsDynamic=false, bIsEditable=true, bIsStaticSwitch=false, bIsVisible=true, type {classStructOrEnum {refPath=&NiagaraBool}}}
      Drag {bIsDynamic=true, bIsEditable=true, bIsStaticSwitch=false, bIsVisible=true, type {classStructOrEnum {refPath=&NiagaraFloat}}}
      Use Rotational Drag {bIsDynamic=false, bIsEditable=true, bIsStaticSwitch=false, bIsVisible=true, type {classStructOrEnum {refPath=&NiagaraBool}}}
      Rotational Drag {bIsDynamic=false, bIsEditable=false, bIsStaticSwitch=false, bIsVisible=true, type {classStructOrEnum {refPath=&NiagaraFloat}}}
     ]
     moduleScript {refPath=&Drag}
    }
    {
     bIsSetParametersModule=false
     enabled=true
     moduleName=GravityForce
     inputs [
      Gravity {bIsDynamic=false, bIsEditable=true, bIsStaticSwitch=false, bIsVisible=true, type {classStructOrEnum {refPath=&Vector3f}}}
      Coordinate Space {
       bIsDynamic=false
       bIsEditable=true
       bIsStaticSwitch=true
       bIsVisible=true
       type {classStructOrEnum {refPath=&ENiagaraCoordinateSpace}}
      }
     ]
     moduleScript {refPath=&GravityForce}
    }
    {
     bIsSetParametersModule=false
     enabled=true
     moduleName=SolveForcesAndVelocity
     inputs [
      Clamp Velocity {bIsDynamic=false, bIsEditable=true, bIsStaticSwitch=false, bIsVisible=true, type {classStructOrEnum {refPath=&NiagaraBool}}}
      Speed Limit {bIsDynamic=false, bIsEditable=false, bIsStaticSwitch=false, bIsVisible=true, type {classStructOrEnum {refPath=&NiagaraFloat}}}
      Limit Acceleration {bIsDynamic=false, bIsEditable=true, bIsStaticSwitch=false, bIsVisible=true, type {classStructOrEnum {refPath=&NiagaraBool}}}
      Acceleration Limit {bIsDynamic=false, bIsEditable=false, bIsStaticSwitch=false, bIsVisible=true, type {classStructOrEnum {refPath=&NiagaraFloat}}}
      Manually Enable Rotational Solver {bIsDynamic=false, bIsEditable=true, bIsStaticSwitch=false, bIsVisible=true, type {classStructOrEnum {refPath=&NiagaraBool}}}
      Write to Intrinsic Properties {bIsDynamic=false, bIsEditable=true, bIsStaticSwitch=false, bIsVisible=true, type {classStructOrEnum {refPath=&NiagaraBool}}}
      Write to Presolve Properties {bIsDynamic=false, bIsEditable=true, bIsStaticSwitch=true, bIsVisible=true, type {classStructOrEnum {refPath=&NiagaraBool}}}
      DeltaTime {bIsDynamic=false, bIsEditable=true, bIsStaticSwitch=false, bIsVisible=true, type {classStructOrEnum {refPath=&NiagaraFloat}}}
      Mass {bIsDynamic=false, bIsEditable=true, bIsStaticSwitch=false, bIsVisible=true, type {classStructOrEnum {refPath=&NiagaraFloat}}}
      Position {
       bIsDynamic=false
       bIsEditable=true
       bIsStaticSwitch=false
       bIsVisible=true
       type {classStructOrEnum {refPath=&NiagaraPosition}}
      }
      Velocity {bIsDynamic=false, bIsEditable=true, bIsStaticSwitch=false, bIsVisible=true, type {classStructOrEnum {refPath=&Vector3f}}}
      PreviousVelocity {bIsDynamic=false, bIsEditable=true, bIsStaticSwitch=false, bIsVisible=true, type {classStructOrEnum {refPath=&Vector3f}}}
      Force {bIsDynamic=false, bIsEditable=true, bIsStaticSwitch=false, bIsVisible=true, type {classStructOrEnum {refPath=&Vector3f}}}
      Mesh Orientation {bIsDynamic=false, bIsEditable=false, bIsStaticSwitch=false, bIsVisible=false, type {classStructOrEnum {refPath=&Quat4f}}}
      Rotational Velocity {
       bIsDynamic=false
       bIsEditable=false
       bIsStaticSwitch=false
       bIsVisible=false
       type {classStructOrEnum {refPath=&Vector3f}}
      }
      Rotational Force {
       bIsDynamic=false
       bIsEditable=false
       bIsStaticSwitch=false
       bIsVisible=false
       type {classStructOrEnum {refPath=&Vector3f}}
      }
      Rotational Inertia {
       bIsDynamic=false
       bIsEditable=false
       bIsStaticSwitch=false
       bIsVisible=false
       type {classStructOrEnum {refPath=&NiagaraFloat}}
      }
     ]
     moduleScript {refPath=&SolveForcesAndVelocity}
    }
    {
     bIsSetParametersModule=false
     enabled=true
     moduleName=SpriteRotationRate
     inputs [
      Rotation Rate {bIsDynamic=true, bIsEditable=true, bIsStaticSwitch=false, bIsVisible=true, type {classStructOrEnum {refPath=&NiagaraFloat}}}
      Delta Time {bIsDynamic=false, bIsEditable=true, bIsStaticSwitch=false, bIsVisible=true, type {classStructOrEnum {refPath=&NiagaraFloat}}}
     ]
     moduleScript {refPath=&SpriteRotationRate}
    }
    {
     bIsSetParametersModule=false
     enabled=true
     moduleName=ScaleSpriteSize
     inputs [
      Scale Factor {bIsDynamic=true, bIsEditable=true, bIsStaticSwitch=false, bIsVisible=true, type {classStructOrEnum {refPath=&Vector2f}}}
      Initial Sprite Size {bIsDynamic=false, bIsEditable=true, bIsStaticSwitch=false, bIsVisible=true, type {classStructOrEnum {refPath=&Vector2f}}}
     ]
     moduleScript {refPath=&ScaleSpriteSize}
    }
    {
     bIsSetParametersModule=false
     enabled=true
     moduleName=ScaleColor
     inputs [
      Scale Mode {
       bIsDynamic=false
       bIsEditable=true
       bIsStaticSwitch=true
       bIsVisible=true
       type {classStructOrEnum {refPath=&ENiagaraScaleColorMode}}
      }
      Linear Color Curve {
       bIsDynamic=false
       bIsEditable=false
       bIsStaticSwitch=false
       bIsVisible=false
       type {classStructOrEnum {refPath=&NiagaraDataInterfaceColorCurve}}
      }
      ScaleRGB {bIsDynamic=false, bIsEditable=true, bIsStaticSwitch=false, bIsVisible=true, type {classStructOrEnum {refPath=&NiagaraBool}}}
      Scale RGB {bIsDynamic=false, bIsEditable=true, bIsStaticSwitch=false, bIsVisible=true, type {classStructOrEnum {refPath=&Vector3f}}}
      ScaleRGBA {bIsDynamic=false, bIsEditable=false, bIsStaticSwitch=false, bIsVisible=false, type {classStructOrEnum {refPath=&NiagaraBool}}}
      Scale RGBA {
       bIsDynamic=false
       bIsEditable=false
       bIsStaticSwitch=false
       bIsVisible=false
       type {classStructOrEnum {refPath=&Vector4f}}
      }
      Curve Index {
       bIsDynamic=false
       bIsEditable=false
       bIsStaticSwitch=false
       bIsVisible=false
       type {classStructOrEnum {refPath=&NiagaraFloat}}
      }
      ScaleA {bIsDynamic=false, bIsEditable=true, bIsStaticSwitch=false, bIsVisible=true, type {classStructOrEnum {refPath=&NiagaraBool}}}
      Scale Alpha {bIsDynamic=true, bIsEditable=true, bIsStaticSwitch=false, bIsVisible=true, type {classStructOrEnum {refPath=&NiagaraFloat}}}
      Color Value To Scale {
       bIsDynamic=false
       bIsEditable=true
       bIsStaticSwitch=false
       bIsVisible=true
       type {classStructOrEnum {refPath=&LinearColor}}
      }
     ]
     moduleScript {refPath=&ScaleColor}
    }
    {
     bIsSetParametersModule=false
     enabled=true
     moduleName=DynamicMaterialParameters
     inputs [
      Write Parameter Index 0 {bIsDynamic=false, bIsEditable=true, bIsStaticSwitch=true, bIsVisible=true, type {classStructOrEnum {refPath=&NiagaraBool}}}
      Module.Param0WriteEnabled {bIsDynamic=false, bIsEditable=true, bIsStaticSwitch=false, bIsVisible=true, type {classStructOrEnum {refPath=&NiagaraBool}}}
      Index 0 Param 1 {bIsDynamic=false, bIsEditable=true, bIsStaticSwitch=false, bIsVisible=true, type {classStructOrEnum {refPath=&NiagaraFloat}}}
      Module.Param1WriteEnabled {bIsDynamic=false, bIsEditable=true, bIsStaticSwitch=false, bIsVisible=true, type {classStructOrEnum {refPath=&NiagaraBool}}}
      Index 0 Param 2 {bIsDynamic=false, bIsEditable=true, bIsStaticSwitch=false, bIsVisible=true, type {classStructOrEnum {refPath=&NiagaraFloat}}}
      Module.Param2WriteEnabled {bIsDynamic=false, bIsEditable=true, bIsStaticSwitch=false, bIsVisible=true, type {classStructOrEnum {refPath=&NiagaraBool}}}
      Index 0 Param 3 {bIsDynamic=false, bIsEditable=true, bIsStaticSwitch=false, bIsVisible=true, type {classStructOrEnum {refPath=&NiagaraFloat}}}
      Module.Param3WriteEnabled {bIsDynamic=false, bIsEditable=true, bIsStaticSwitch=false, bIsVisible=true, type {classStructOrEnum {refPath=&NiagaraBool}}}
      Index 0 Param 4 {bIsDynamic=false, bIsEditable=true, bIsStaticSwitch=false, bIsVisible=true, type {classStructOrEnum {refPath=&NiagaraFloat}}}
      Write Parameter Index 1 {bIsDynamic=false, bIsEditable=true, bIsStaticSwitch=true, bIsVisible=true, type {classStructOrEnum {refPath=&NiagaraBool}}}
      Index 1 Param 1 {
       bIsDynamic=false
       bIsEditable=false
       bIsStaticSwitch=false
       bIsVisible=false
       type {classStructOrEnum {refPath=&NiagaraFloat}}
      }
      Index 1 Param 2 {
       bIsDynamic=false
       bIsEditable=false
       bIsStaticSwitch=false
       bIsVisible=false
       type {classStructOrEnum {refPath=&NiagaraFloat}}
      }
      Index 1 Param 3 {
       bIsDynamic=false
       bIsEditable=false
       bIsStaticSwitch=false
       bIsVisible=false
       type {classStructOrEnum {refPath=&NiagaraFloat}}
      }
      Index 1 Param 4 {
       bIsDynamic=false
       bIsEditable=false
       bIsStaticSwitch=false
       bIsVisible=false
       type {classStructOrEnum {refPath=&NiagaraFloat}}
      }
      Write Parameter Index 2 {bIsDynamic=false, bIsEditable=true, bIsStaticSwitch=true, bIsVisible=true, type {classStructOrEnum {refPath=&NiagaraBool}}}
      Index 2 Param 1 {
       bIsDynamic=false
       bIsEditable=false
       bIsStaticSwitch=false
       bIsVisible=false
       type {classStructOrEnum {refPath=&NiagaraFloat}}
      }
      Index 2 Param 2 {
       bIsDynamic=false
       bIsEditable=false
       bIsStaticSwitch=false
       bIsVisible=false
       type {classStructOrEnum {refPath=&NiagaraFloat}}
      }
      Index 2 Param 3 {
       bIsDynamic=false
       bIsEditable=false
       bIsStaticSwitch=false
       bIsVisible=false
       type {classStructOrEnum {refPath=&NiagaraFloat}}
      }
      Index 2 Param 4 {
       bIsDynamic=false
       bIsEditable=false
       bIsStaticSwitch=false
       bIsVisible=false
       type {classStructOrEnum {refPath=&NiagaraFloat}}
      }
      Write Parameter Index 3 {bIsDynamic=false, bIsEditable=true, bIsStaticSwitch=true, bIsVisible=true, type {classStructOrEnum {refPath=&NiagaraBool}}}
      Index 3 Param 1 {
       bIsDynamic=false
       bIsEditable=false
       bIsStaticSwitch=false
       bIsVisible=false
       type {classStructOrEnum {refPath=&NiagaraFloat}}
      }
      Index 3 Param 2 {
       bIsDynamic=false
       bIsEditable=false
       bIsStaticSwitch=false
       bIsVisible=false
       type {classStructOrEnum {refPath=&NiagaraFloat}}
      }
      Index 3 Param 3 {
       bIsDynamic=false
       bIsEditable=false
       bIsStaticSwitch=false
       bIsVisible=false
       type {classStructOrEnum {refPath=&NiagaraFloat}}
      }
      Index 3 Param 4 {
       bIsDynamic=false
       bIsEditable=false
       bIsStaticSwitch=false
       bIsVisible=false
       type {classStructOrEnum {refPath=&NiagaraFloat}}
      }
     ]
     moduleScript {refPath=&DynamicMaterialParameters}
    }
   ]
  }
  rendererClasses [{refPath=&NiagaraSpriteRendererProperties}]
  renderers [{rendererIndex=0, rendererClass {refPath=&NiagaraSpriteRendererProperties}}]
 }
 SnowflakesCPU {
  bEnabled=false
  simTarget=GPUComputeSim
  emitterSpawnScript {scriptName=EmitterSpawnScript, modules []}
  emitterUpdateScript {
   scriptName=EmitterUpdateScript
   modules [
    {
     bIsSetParametersModule=false
     enabled=true
     moduleName=EmitterState
     inputs [
      Life Cycle Mode {
       bIsDynamic=false
       bIsEditable=true
       bIsStaticSwitch=true
       bIsVisible=true
       type {classStructOrEnum {refPath=&ENiagaraEmitterLifeCycleMode}}
      }
      Inactive Response {
       bIsDynamic=false
       bIsEditable=true
       bIsStaticSwitch=true
       bIsVisible=true
       type {classStructOrEnum {refPath=&ENiagaraInactiveMode}}
      }
      Loop Behavior {
       bIsDynamic=false
       bIsEditable=true
       bIsStaticSwitch=true
       bIsVisible=true
       type {classStructOrEnum {refPath=&ENiagara_EmitterStateOptions}}
      }
      Loop Duration Mode {
       bIsDynamic=false
       bIsEditable=true
       bIsStaticSwitch=true
       bIsVisible=true
       type {classStructOrEnum {refPath=&ENiagara_InfiniteLoopDuration}}
      }
      Loop Count {
       bIsDynamic=false
       bIsEditable=false
       bIsStaticSwitch=false
       bIsVisible=false
       type {classStructOrEnum {refPath=&NiagaraInt32}}
      }
      Loop Duration {bIsDynamic=false, bIsEditable=true, bIsStaticSwitch=false, bIsVisible=true, type {classStructOrEnum {refPath=&NiagaraFloat}}}
      Recalculate Duration Each Loop {bIsDynamic=false, bIsEditable=false, bIsStaticSwitch=false, bIsVisible=false, type {classStructOrEnum {refPath=&NiagaraBool}}}
      UseLoopDelay {bIsDynamic=false, bIsEditable=true, bIsStaticSwitch=false, bIsVisible=true, type {classStructOrEnum {refPath=&NiagaraBool}}}
      Loop Delay {bIsDynamic=false, bIsEditable=false, bIsStaticSwitch=false, bIsVisible=true, type {classStructOrEnum {refPath=&NiagaraFloat}}}
      Delay First Loop Only {bIsDynamic=false, bIsEditable=false, bIsStaticSwitch=false, bIsVisible=false, type {classStructOrEnum {refPath=&NiagaraBool}}}
      Scalability Mode {
       bIsDynamic=false
       bIsEditable=true
       bIsStaticSwitch=true
       bIsVisible=true
       type {classStructOrEnum {refPath=&ENiagaraEmitterScalabilityMode_Limited}}
      }
      Enable Distance Culling {bIsDynamic=false, bIsEditable=false, bIsStaticSwitch=true, bIsVisible=false, type {classStructOrEnum {refPath=&NiagaraBool}}}
      MinDistance {
       bIsDynamic=false
       bIsEditable=false
       bIsStaticSwitch=false
       bIsVisible=false
       type {classStructOrEnum {refPath=&NiagaraFloat}}
      }
      Min Distance Response {
       bIsDynamic=false
       bIsEditable=false
       bIsStaticSwitch=false
       bIsVisible=false
       type {classStructOrEnum {refPath=&ENiagaraExecutionStateManagement}}
      }
      MaxDistance {
       bIsDynamic=false
       bIsEditable=false
       bIsStaticSwitch=false
       bIsVisible=false
       type {classStructOrEnum {refPath=&NiagaraFloat}}
      }
      Max Distance Response {
       bIsDynamic=false
       bIsEditable=false
       bIsStaticSwitch=false
       bIsVisible=false
       type {classStructOrEnum {refPath=&ENiagaraExecutionStateManagement}}
      }
      Scale Spawn Count {bIsDynamic=false, bIsEditable=false, bIsStaticSwitch=true, bIsVisible=false, type {classStructOrEnum {refPath=&NiagaraBool}}}
      Spawn Count Scale {
       bIsDynamic=false
       bIsEditable=false
       bIsStaticSwitch=false
       bIsVisible=false
       type {classStructOrEnum {refPath=&NiagaraFloat}}
      }
      Spawn Count Scale By Distance {
       bIsDynamic=false
       bIsEditable=false
       bIsStaticSwitch=false
       bIsVisible=false
       type {classStructOrEnum {refPath=&NiagaraDataInterfaceCurve}}
      }
      Enable Visibility Culling {bIsDynamic=false, bIsEditable=false, bIsStaticSwitch=true, bIsVisible=false, type {classStructOrEnum {refPath=&NiagaraBool}}}
      Visibility Cull Response {
       bIsDynamic=false
       bIsEditable=false
       bIsStaticSwitch=false
       bIsVisible=false
       type {classStructOrEnum {refPath=&ENiagaraExecutionStateManagement}}
      }
      Visibility Cull Delay {
       bIsDynamic=false
       bIsEditable=false
       bIsStaticSwitch=false
       bIsVisible=false
       type {classStructOrEnum {refPath=&NiagaraFloat}}
      }
      Reset Age On Awaken {bIsDynamic=false, bIsEditable=false, bIsStaticSwitch=false, bIsVisible=false, type {classStructOrEnum {refPath=&NiagaraBool}}}
     ]
     moduleScript {refPath=&EmitterState}
    }
    {
     bIsSetParametersModule=false
     enabled=true
     moduleName=SpawnRate
     inputs [
      SpawnRate {bIsDynamic=true, bIsEditable=true, bIsStaticSwitch=false, bIsVisible=true, type {classStructOrEnum {refPath=&NiagaraFloat}}}
      Use Spawn Probability {bIsDynamic=false, bIsEditable=true, bIsStaticSwitch=false, bIsVisible=true, type {classStructOrEnum {refPath=&NiagaraBool}}}
      Spawn Probability {bIsDynamic=false, bIsEditable=false, bIsStaticSwitch=false, bIsVisible=true, type {classStructOrEnum {refPath=&NiagaraFloat}}}
      SpawnGroup {bIsDynamic=false, bIsEditable=true, bIsStaticSwitch=false, bIsVisible=true, type {classStructOrEnum {refPath=&NiagaraInt32}}}
     ]
     moduleScript {refPath=&SpawnRate}
    }
   ]
  }
  particleSpawnScript {
   scriptName=ParticleSpawnScript
   modules [
    {
     bIsSetParametersModule=false
     enabled=true
     moduleName=InitializeParticle
     inputs [
      Lifetime Mode {
       bIsDynamic=false
       bIsEditable=true
       bIsStaticSwitch=true
       bIsVisible=true
       type {classStructOrEnum {refPath=&ENiagara_LifetimeMode}}
      }
      Lifetime Min {bIsDynamic=false, bIsEditable=true, bIsStaticSwitch=false, bIsVisible=true, type {classStructOrEnum {refPath=&NiagaraFloat}}}
      Lifetime {
       bIsDynamic=false
       bIsEditable=false
       bIsStaticSwitch=false
       bIsVisible=false
       type {classStructOrEnum {refPath=&NiagaraFloat}}
      }
      Lifetime Max {bIsDynamic=false, bIsEditable=true, bIsStaticSwitch=false, bIsVisible=true, type {classStructOrEnum {refPath=&NiagaraFloat}}}
      Lifetime Randomness Mode {
       bIsDynamic=false
       bIsEditable=true
       bIsStaticSwitch=true
       bIsVisible=true
       type {classStructOrEnum {refPath=&ENiagaraRandomnessMode}}
      }
      Color Mode {
       bIsDynamic=false
       bIsEditable=true
       bIsStaticSwitch=true
       bIsVisible=true
       type {classStructOrEnum {refPath=&ENiagara_ColorInitializationMode}}
      }
      Color Minimum {
       bIsDynamic=false
       bIsEditable=true
       bIsStaticSwitch=false
       bIsVisible=true
       type {classStructOrEnum {refPath=&LinearColor}}
      }
      Color {
       bIsDynamic=false
       bIsEditable=false
       bIsStaticSwitch=false
       bIsVisible=false
       type {classStructOrEnum {refPath=&LinearColor}}
      }
      AdjustHue {bIsDynamic=false, bIsEditable=false, bIsStaticSwitch=false, bIsVisible=false, type {classStructOrEnum {refPath=&NiagaraBool}}}
      Hue Shift Range {
       bIsDynamic=false
       bIsEditable=false
       bIsStaticSwitch=false
       bIsVisible=false
       type {classStructOrEnum {refPath=&Vector2f}}
      }
      Color Maximum {
       bIsDynamic=false
       bIsEditable=true
       bIsStaticSwitch=false
       bIsVisible=true
       type {classStructOrEnum {refPath=&LinearColor}}
      }
      AdjustSaturation {bIsDynamic=false, bIsEditable=false, bIsStaticSwitch=false, bIsVisible=false, type {classStructOrEnum {refPath=&NiagaraBool}}}
      Saturation Range {
       bIsDynamic=false
       bIsEditable=false
       bIsStaticSwitch=false
       bIsVisible=false
       type {classStructOrEnum {refPath=&Vector2f}}
      }
      AdjustValue {bIsDynamic=false, bIsEditable=false, bIsStaticSwitch=false, bIsVisible=false, type {classStructOrEnum {refPath=&NiagaraBool}}}
      Value Range {
       bIsDynamic=false
       bIsEditable=false
       bIsStaticSwitch=false
       bIsVisible=false
       type {classStructOrEnum {refPath=&Vector2f}}
      }
      AdjustAlpha {bIsDynamic=false, bIsEditable=false, bIsStaticSwitch=false, bIsVisible=false, type {classStructOrEnum {refPath=&NiagaraBool}}}
      Alpha Scale Range {
       bIsDynamic=false
       bIsEditable=false
       bIsStaticSwitch=false
       bIsVisible=false
       type {classStructOrEnum {refPath=&Vector2f}}
      }
      Color Channel Mode {
       bIsDynamic=false
       bIsEditable=true
       bIsStaticSwitch=true
       bIsVisible=true
       type {classStructOrEnum {refPath=&ENiagaraChannelCorrelation}}
      }
      Color Randomness Mode {
       bIsDynamic=false
       bIsEditable=true
       bIsStaticSwitch=true
       bIsVisible=true
       type {classStructOrEnum {refPath=&ENiagaraRandomnessMode}}
      }
      Position Mode {
       bIsDynamic=false
       bIsEditable=true
       bIsStaticSwitch=true
       bIsVisible=true
       type {classStructOrEnum {refPath=&ENiagara_PositionInitializationMode}}
      }
      Position {
       bIsDynamic=false
       bIsEditable=false
       bIsStaticSwitch=false
       bIsVisible=false
       type {classStructOrEnum {refPath=&NiagaraPosition}}
      }
      UsePositionOffset {bIsDynamic=false, bIsEditable=true, bIsStaticSwitch=false, bIsVisible=true, type {classStructOrEnum {refPath=&NiagaraBool}}}
      Position Offset {bIsDynamic=false, bIsEditable=false, bIsStaticSwitch=false, bIsVisible=true, type {classStructOrEnum {refPath=&Vector3f}}}
      Position Offset Coordinate Space {
       bIsDynamic=false
       bIsEditable=false
       bIsStaticSwitch=false
       bIsVisible=false
       type {classStructOrEnum {refPath=&ENiagaraCoordinateSpace}}
      }
      Mass Mode {
       bIsDynamic=false
       bIsEditable=true
       bIsStaticSwitch=true
       bIsVisible=true
       type {classStructOrEnum {refPath=&ENiagara_MassInitializationMode}}
      }
      Mass Min {
       bIsDynamic=false
       bIsEditable=false
       bIsStaticSwitch=false
       bIsVisible=false
       type {classStructOrEnum {refPath=&NiagaraFloat}}
      }
      Mass {
       bIsDynamic=false
       bIsEditable=false
       bIsStaticSwitch=false
       bIsVisible=false
       type {classStructOrEnum {refPath=&NiagaraFloat}}
      }
      Mass Max {
       bIsDynamic=false
       bIsEditable=false
       bIsStaticSwitch=false
       bIsVisible=false
       type {classStructOrEnum {refPath=&NiagaraFloat}}
      }
      Mass Randomness Mode {
       bIsDynamic=false
       bIsEditable=false
       bIsStaticSwitch=true
       bIsVisible=false
       type {classStructOrEnum {refPath=&ENiagaraRandomnessMode}}
      }
      Write Material Random {bIsDynamic=false, bIsEditable=true, bIsStaticSwitch=false, bIsVisible=true, type {classStructOrEnum {refPath=&NiagaraBool}}}
      Material Random {bIsDynamic=false, bIsEditable=true, bIsStaticSwitch=false, bIsVisible=true, type {classStructOrEnum {refPath=&NiagaraFloat}}}
      Sprite Size Mode {
       bIsDynamic=false
       bIsEditable=true
       bIsStaticSwitch=true
       bIsVisible=true
       type {classStructOrEnum {refPath=&ENiagara_SizeScaleMode}}
      }
      Sprite Size {
       bIsDynamic=false
       bIsEditable=false
       bIsStaticSwitch=false
       bIsVisible=false
       type {classStructOrEnum {refPath=&Vector2f}}
      }
      Uniform Sprite Size {
       bIsDynamic=false
       bIsEditable=false
       bIsStaticSwitch=false
       bIsVisible=false
       type {classStructOrEnum {refPath=&NiagaraFloat}}
      }
      Uniform Sprite Size Min {bIsDynamic=false, bIsEditable=true, bIsStaticSwitch=false, bIsVisible=true, type {classStructOrEnum {refPath=&NiagaraFloat}}}
      Sprite Size Min {
       bIsDynamic=false
       bIsEditable=false
       bIsStaticSwitch=false
       bIsVisible=false
       type {classStructOrEnum {refPath=&Vector2f}}
      }
      Sprite Size Max {
       bIsDynamic=false
       bIsEditable=false
       bIsStaticSwitch=false
       bIsVisible=false
       type {classStructOrEnum {refPath=&Vector2f}}
      }
      Uniform Sprite Size Max {bIsDynamic=false, bIsEditable=true, bIsStaticSwitch=false, bIsVisible=true, type {classStructOrEnum {refPath=&NiagaraFloat}}}
      Sprite Size Randomness Mode {
       bIsDynamic=false
       bIsEditable=true
       bIsStaticSwitch=true
       bIsVisible=true
       type {classStructOrEnum {refPath=&ENiagaraRandomnessMode}}
      }
      Sprite Rotation Mode {
       bIsDynamic=false
       bIsEditable=true
       bIsStaticSwitch=true
       bIsVisible=true
       type {classStructOrEnum {refPath=&ENiagara_SpriteRotationMode}}
      }
      Sprite Rotation Angle Min {bIsDynamic=false, bIsEditable=true, bIsStaticSwitch=false, bIsVisible=true, type {classStructOrEnum {refPath=&NiagaraFloat}}}
      Sprite Rotation Angle {
       bIsDynamic=false
       bIsEditable=false
       bIsStaticSwitch=false
       bIsVisible=false
       type {classStructOrEnum {refPath=&NiagaraFloat}}
      }
      Sprite Rotation Angle Max {bIsDynamic=false, bIsEditable=true, bIsStaticSwitch=false, bIsVisible=true, type {classStructOrEnum {refPath=&NiagaraFloat}}}
      Sprite Rotation Randomness Mode {
       bIsDynamic=false
       bIsEditable=true
       bIsStaticSwitch=true
       bIsVisible=true
       type {classStructOrEnum {refPath=&ENiagaraRandomnessMode}}
      }
      Sprite UV Mode {
       bIsDynamic=false
       bIsEditable=true
       bIsStaticSwitch=true
       bIsVisible=true
       type {classStructOrEnum {refPath=&ENiagara_UVFlippingMode}}
      }
      Sprite UV Randomness Mode {
       bIsDynamic=false
       bIsEditable=false
       bIsStaticSwitch=true
       bIsVisible=false
       type {classStructOrEnum {refPath=&ENiagaraRandomnessMode}}
      }
      Sprite UV Scale {
       bIsDynamic=false
       bIsEditable=false
       bIsStaticSwitch=false
       bIsVisible=false
       type {classStructOrEnum {refPath=&Vector2f}}
      }
      Mesh Scale Mode {
       bIsDynamic=false
       bIsEditable=true
       bIsStaticSwitch=true
       bIsVisible=true
       type {classStructOrEnum {refPath=&ENiagara_SizeScaleMode}}
      }
      Mesh Uniform Scale Min {
       bIsDynamic=false
       bIsEditable=false
       bIsStaticSwitch=false
       bIsVisible=false
       type {classStructOrEnum {refPath=&NiagaraFloat}}
      }
      Mesh Uniform Scale {
       bIsDynamic=false
       bIsEditable=false
       bIsStaticSwitch=false
       bIsVisible=false
       type {classStructOrEnum {refPath=&NiagaraFloat}}
      }
      Mesh Scale Min {
       bIsDynamic=false
       bIsEditable=false
       bIsStaticSwitch=false
       bIsVisible=false
       type {classStructOrEnum {refPath=&Vector3f}}
      }
      Mesh Scale {bIsDynamic=true, bIsEditable=false, bIsStaticSwitch=false, bIsVisible=false, type {classStructOrEnum {refPath=&Vector3f}}}
      Mesh Scale Max {
       bIsDynamic=false
       bIsEditable=false
       bIsStaticSwitch=false
       bIsVisible=false
       type {classStructOrEnum {refPath=&Vector3f}}
      }
      Mesh Uniform Scale Max {
       bIsDynamic=false
       bIsEditable=false
       bIsStaticSwitch=false
       bIsVisible=false
       type {classStructOrEnum {refPath=&NiagaraFloat}}
      }
      Mesh Scale Randomness Mode {
       bIsDynamic=false
       bIsEditable=false
       bIsStaticSwitch=true
       bIsVisible=false
       type {classStructOrEnum {refPath=&ENiagaraRandomnessMode}}
      }
      Mesh Renderer Array Visibility Mode {
       bIsDynamic=false
       bIsEditable=true
       bIsStaticSwitch=true
       bIsVisible=true
       type {classStructOrEnum {refPath=&ENiagara_UnsetDirectSetRandom}}
      }
      Mesh Renderer Mesh Index {
       bIsDynamic=false
       bIsEditable=false
       bIsStaticSwitch=false
       bIsVisible=false
       type {classStructOrEnum {refPath=&NiagaraInt32}}
      }
      Wrap Index By Mesh Count {bIsDynamic=false, bIsEditable=false, bIsStaticSwitch=true, bIsVisible=false, type {classStructOrEnum {refPath=&NiagaraBool}}}
      Mesh Renderer Info {
       bIsDynamic=false
       bIsEditable=false
       bIsStaticSwitch=false
       bIsVisible=false
       type {classStructOrEnum {refPath=&NiagaraDataInterfaceMeshRendererInfo}}
      }
      Mesh Renderer Random Seed {
       bIsDynamic=false
       bIsEditable=false
       bIsStaticSwitch=false
       bIsVisible=false
       type {classStructOrEnum {refPath=&NiagaraInt32}}
      }
      Ribbon Width Mode {
       bIsDynamic=false
       bIsEditable=true
       bIsStaticSwitch=true
       bIsVisible=true
       type {classStructOrEnum {refPath=&ENiagara_UnsetDirectSet}}
      }
      Ribbon Width {
       bIsDynamic=false
       bIsEditable=false
       bIsStaticSwitch=false
       bIsVisible=false
       type {classStructOrEnum {refPath=&NiagaraFloat}}
      }
      Ribbon Facing Mode {
       bIsDynamic=false
       bIsEditable=true
       bIsStaticSwitch=true
       bIsVisible=true
       type {classStructOrEnum {refPath=&ENiagara_UnsetDirectSet}}
      }
      Ribbon Facing Vector {
       bIsDynamic=false
       bIsEditable=false
       bIsStaticSwitch=false
       bIsVisible=false
       type {classStructOrEnum {refPath=&Vector3f}}
      }
      Ribbon Twist Mode {
       bIsDynamic=false
       bIsEditable=true
       bIsStaticSwitch=true
       bIsVisible=true
       type {classStructOrEnum {refPath=&ENiagara_UnsetDirectSet}}
      }
      "Ribbon Twist " {
       bIsDynamic=false
       bIsEditable=false
       bIsStaticSwitch=false
       bIsVisible=false
       type {classStructOrEnum {refPath=&NiagaraFloat}}
      }
     ]
     moduleScript {refPath=&InitializeParticle}
    }
    {
     bIsSetParametersModule=false
     enabled=true
     moduleName=ShapeLocation
     inputs [
      Shape Primitive {
       bIsDynamic=false
       bIsEditable=true
       bIsStaticSwitch=true
       bIsVisible=true
       type {classStructOrEnum {refPath=&ENiagara_LocationShapes}}
      }
      Box / Plane Mode {
       bIsDynamic=false
       bIsEditable=false
       bIsStaticSwitch=true
       bIsVisible=false
       type {classStructOrEnum {refPath=&ENiagara_BoxPlaneMode}}
      }
      Cone Mode {
       bIsDynamic=false
       bIsEditable=false
       bIsStaticSwitch=true
       bIsVisible=false
       type {classStructOrEnum {refPath=&ENiagara_ConeMode}}
      }
      Cylinder Height {
       bIsDynamic=false
       bIsEditable=false
       bIsStaticSwitch=false
       bIsVisible=false
       type {classStructOrEnum {refPath=&NiagaraFloat}}
      }
      Sphere Radius {bIsDynamic=false, bIsEditable=true, bIsStaticSwitch=false, bIsVisible=true, type {classStructOrEnum {refPath=&NiagaraFloat}}}
      Ring / Disc Mode {
       bIsDynamic=false
       bIsEditable=false
       bIsStaticSwitch=true
       bIsVisible=false
       type {classStructOrEnum {refPath=&ENiagaraRingDiscMode}}
      }
      Torus Mode {
       bIsDynamic=false
       bIsEditable=false
       bIsStaticSwitch=true
       bIsVisible=false
       type {classStructOrEnum {refPath=&ENiagaraShapeTorusMode}}
      }
      Knot Ring Loops {
       bIsDynamic=false
       bIsEditable=false
       bIsStaticSwitch=false
       bIsVisible=false
       type {classStructOrEnum {refPath=&NiagaraInt32}}
      }
      Knot Helix Loops {
       bIsDynamic=false
       bIsEditable=false
       bIsStaticSwitch=false
       bIsVisible=false
       type {classStructOrEnum {refPath=&NiagaraInt32}}
      }
      Enable Torus Knot Radius {bIsDynamic=false, bIsEditable=false, bIsStaticSwitch=false, bIsVisible=false, type {classStructOrEnum {refPath=&NiagaraBool}}}
      Torus Knot Tube Radius {
       bIsDynamic=false
       bIsEditable=false
       bIsStaticSwitch=false
       bIsVisible=false
       type {classStructOrEnum {refPath=&NiagaraFloat}}
      }
      Cylinder Radius {
       bIsDynamic=false
       bIsEditable=false
       bIsStaticSwitch=false
       bIsVisible=false
       type {classStructOrEnum {refPath=&NiagaraFloat}}
      }
      Cylinder Height Midpoint {
       bIsDynamic=false
       bIsEditable=false
       bIsStaticSwitch=false
       bIsVisible=false
       type {classStructOrEnum {refPath=&NiagaraFloat}}
      }
      Ring Radius {
       bIsDynamic=false
       bIsEditable=false
       bIsStaticSwitch=false
       bIsVisible=false
       type {classStructOrEnum {refPath=&NiagaraFloat}}
      }
      Box Size {
       bIsDynamic=false
       bIsEditable=false
       bIsStaticSwitch=false
       bIsVisible=false
       type {classStructOrEnum {refPath=&Vector3f}}
      }
      Box Midpoint {
       bIsDynamic=false
       bIsEditable=false
       bIsStaticSwitch=false
       bIsVisible=false
       type {classStructOrEnum {refPath=&Vector3f}}
      }
      Cone Length {
       bIsDynamic=false
       bIsEditable=false
       bIsStaticSwitch=false
       bIsVisible=false
       type {classStructOrEnum {refPath=&NiagaraFloat}}
      }
      Plane Size {
       bIsDynamic=false
       bIsEditable=false
       bIsStaticSwitch=false
       bIsVisible=false
       type {classStructOrEnum {refPath=&Vector2f}}
      }
      Wedge Horizontal Angle {
       bIsDynamic=false
       bIsEditable=false
       bIsStaticSwitch=false
       bIsVisible=false
       type {classStructOrEnum {refPath=&NiagaraFloat}}
      }
      Large Radius {
       bIsDynamic=false
       bIsEditable=false
       bIsStaticSwitch=false
       bIsVisible=false
       type {classStructOrEnum {refPath=&NiagaraFloat}}
      }
      Plane Midpoint {
       bIsDynamic=false
       bIsEditable=false
       bIsStaticSwitch=false
       bIsVisible=false
       type {classStructOrEnum {refPath=&Vector2f}}
      }
      Wedge Vertical Angle {
       bIsDynamic=false
       bIsEditable=false
       bIsStaticSwitch=false
       bIsVisible=false
       type {classStructOrEnum {refPath=&NiagaraFloat}}
      }
      Cone Angle {
       bIsDynamic=false
       bIsEditable=false
       bIsStaticSwitch=false
       bIsVisible=false
       type {classStructOrEnum {refPath=&NiagaraFloat}}
      }
      Handle Radius {
       bIsDynamic=false
       bIsEditable=false
       bIsStaticSwitch=false
       bIsVisible=false
       type {classStructOrEnum {refPath=&NiagaraFloat}}
      }
      Cone Inner Angle {
       bIsDynamic=false
       bIsEditable=false
       bIsStaticSwitch=false
       bIsVisible=false
       type {classStructOrEnum {refPath=&NiagaraFloat}}
      }
      Cone Radial Angle {
       bIsDynamic=false
       bIsEditable=false
       bIsStaticSwitch=false
       bIsVisible=false
       type {classStructOrEnum {refPath=&NiagaraFloat}}
      }
      Cone Angle Type {
       bIsDynamic=false
       bIsEditable=false
       bIsStaticSwitch=true
       bIsVisible=false
       type {classStructOrEnum {refPath=&ENiagara_AngleInput}}
      }
      Flatten Endcaps {bIsDynamic=false, bIsEditable=false, bIsStaticSwitch=true, bIsVisible=false, type {classStructOrEnum {refPath=&NiagaraBool}}}
      Apply To Particle Position {
       bIsDynamic=false
       bIsEditable=true
       bIsStaticSwitch=true
       bIsVisible=true
       type {classStructOrEnum {refPath=&ENiagara_AttributeSamplingApplyOutput}}
      }
      Cone Surface Distribution {
       bIsDynamic=false
       bIsEditable=false
       bIsStaticSwitch=false
       bIsVisible=false
       type {classStructOrEnum {refPath=&NiagaraFloat}}
      }
      Ring / Disc Distribution Mode {
       bIsDynamic=false
       bIsEditable=false
       bIsStaticSwitch=true
       bIsVisible=false
       type {classStructOrEnum {refPath=&ENiagaraSphereDistributionMode}}
      }
      Sphere Distribution {
       bIsDynamic=false
       bIsEditable=true
       bIsStaticSwitch=true
       bIsVisible=true
       type {classStructOrEnum {refPath=&ENiagaraSphereDistributionMode}}
      }
      Torus Distribution Mode {
       bIsDynamic=false
       bIsEditable=false
       bIsStaticSwitch=true
       bIsVisible=false
       type {classStructOrEnum {refPath=&ENiagaraTorusDistributionMode}}
      }
      Sphere Surface Distribution {bIsDynamic=false, bIsEditable=true, bIsStaticSwitch=false, bIsVisible=true, type {classStructOrEnum {refPath=&NiagaraFloat}}}
      Hemisphere Distribution {bIsDynamic=false, bIsEditable=true, bIsStaticSwitch=false, bIsVisible=true, type {classStructOrEnum {refPath=&Vector2f}}}
      Cylinder Distribution {
       bIsDynamic=false
       bIsEditable=false
       bIsStaticSwitch=true
       bIsVisible=false
       type {classStructOrEnum {refPath=&ENiagara_CylinderMode}}
      }
      Surface Only {bIsDynamic=false, bIsEditable=false, bIsStaticSwitch=false, bIsVisible=false, type {classStructOrEnum {refPath=&NiagaraBool}}}
      Surface Only Band Thickness {
       bIsDynamic=false
       bIsEditable=false
       bIsStaticSwitch=false
       bIsVisible=false
       type {classStructOrEnum {refPath=&NiagaraFloat}}
      }
      Use Endcaps In Surface Only Mode {bIsDynamic=false, bIsEditable=false, bIsStaticSwitch=true, bIsVisible=false, type {classStructOrEnum {refPath=&NiagaraBool}}}
      Hemicircle X {bIsDynamic=false, bIsEditable=false, bIsStaticSwitch=true, bIsVisible=false, type {classStructOrEnum {refPath=&NiagaraBool}}}
      Hemicircle Y {bIsDynamic=false, bIsEditable=false, bIsStaticSwitch=true, bIsVisible=false, type {classStructOrEnum {refPath=&NiagaraBool}}}
      Disc Coverage {
       bIsDynamic=false
       bIsEditable=false
       bIsStaticSwitch=false
       bIsVisible=false
       type {classStructOrEnum {refPath=&NiagaraFloat}}
      }
      Hemicircle Internal Cap {bIsDynamic=false, bIsEditable=false, bIsStaticSwitch=true, bIsVisible=false, type {classStructOrEnum {refPath=&NiagaraBool}}}
      U Position {
       bIsDynamic=false
       bIsEditable=false
       bIsStaticSwitch=false
       bIsVisible=false
       type {classStructOrEnum {refPath=&NiagaraFloat}}
      }
      Enable Lathe Profile {bIsDynamic=false, bIsEditable=false, bIsStaticSwitch=true, bIsVisible=false, type {classStructOrEnum {refPath=&NiagaraBool}}}
      Lathe Profile {
       bIsDynamic=false
       bIsEditable=false
       bIsStaticSwitch=false
       bIsVisible=false
       type {classStructOrEnum {refPath=&NiagaraDataInterfaceCurve}}
      }
      Surface Distribution {
       bIsDynamic=false
       bIsEditable=false
       bIsStaticSwitch=false
       bIsVisible=false
       type {classStructOrEnum {refPath=&NiagaraFloat}}
      }
      Surface Expansion Mode {
       bIsDynamic=false
       bIsEditable=false
       bIsStaticSwitch=false
       bIsVisible=false
       type {classStructOrEnum {refPath=&ENiagaraExpansionMode}}
      }
      U Distribution {
       bIsDynamic=false
       bIsEditable=false
       bIsStaticSwitch=false
       bIsVisible=false
       type {classStructOrEnum {refPath=&NiagaraFloat}}
      }
      V Position {
       bIsDynamic=false
       bIsEditable=false
       bIsStaticSwitch=false
       bIsVisible=false
       type {classStructOrEnum {refPath=&NiagaraFloat}}
      }
      Radius Position {
       bIsDynamic=false
       bIsEditable=false
       bIsStaticSwitch=false
       bIsVisible=false
       type {classStructOrEnum {refPath=&NiagaraFloat}}
      }
      Sphere Radius Position {
       bIsDynamic=false
       bIsEditable=false
       bIsStaticSwitch=false
       bIsVisible=false
       type {classStructOrEnum {refPath=&NiagaraFloat}}
      }
      V Distribution {
       bIsDynamic=false
       bIsEditable=false
       bIsStaticSwitch=false
       bIsVisible=false
       type {classStructOrEnum {refPath=&NiagaraFloat}}
      }
      Evaluation Type {
       bIsDynamic=false
       bIsEditable=false
       bIsStaticSwitch=true
       bIsVisible=false
       type {classStructOrEnum {refPath=&ENiagaraRandomnessEvaluation}}
      }
      Uniform Distribution {
       bIsDynamic=false
       bIsEditable=false
       bIsStaticSwitch=false
       bIsVisible=false
       type {classStructOrEnum {refPath=&NiagaraFloat}}
      }
      Uniform Spiral Amount {
       bIsDynamic=false
       bIsEditable=false
       bIsStaticSwitch=false
       bIsVisible=false
       type {classStructOrEnum {refPath=&NiagaraFloat}}
      }
      Uniform Spiral Falloff {
       bIsDynamic=false
       bIsEditable=false
       bIsStaticSwitch=false
       bIsVisible=false
       type {classStructOrEnum {refPath=&NiagaraFloat}}
      }
      Uniform Count {
       bIsDynamic=false
       bIsEditable=false
       bIsStaticSwitch=false
       bIsVisible=false
       type {classStructOrEnum {refPath=&NiagaraInt32}}
      }
      Uniform Particle Index {
       bIsDynamic=false
       bIsEditable=false
       bIsStaticSwitch=false
       bIsVisible=false
       type {classStructOrEnum {refPath=&NiagaraInt32}}
      }
      Hemisphere Angle Type {
       bIsDynamic=false
       bIsEditable=true
       bIsStaticSwitch=true
       bIsVisible=true
       type {classStructOrEnum {refPath=&ENiagara_AngleInput}}
      }
      Uniform Endpoint Offset {
       bIsDynamic=false
       bIsEditable=false
       bIsStaticSwitch=false
       bIsVisible=false
       type {classStructOrEnum {refPath=&NiagaraFloat}}
      }
      MaskSpawnGroup {bIsDynamic=false, bIsEditable=true, bIsStaticSwitch=false, bIsVisible=true, type {classStructOrEnum {refPath=&NiagaraBool}}}
      Spawn Group Mask {bIsDynamic=false, bIsEditable=false, bIsStaticSwitch=false, bIsVisible=true, type {classStructOrEnum {refPath=&NiagaraInt32}}}
      Spawn Group {
       bIsDynamic=false
       bIsEditable=false
       bIsStaticSwitch=false
       bIsVisible=false
       type {classStructOrEnum {refPath=&NiagaraInt32}}
      }
      Randomness Mode {
       bIsDynamic=false
       bIsEditable=true
       bIsStaticSwitch=true
       bIsVisible=true
       type {classStructOrEnum {refPath=&ENiagaraRandomnessMode}}
      }
      Override Seed {bIsDynamic=false, bIsEditable=true, bIsStaticSwitch=false, bIsVisible=true, type {classStructOrEnum {refPath=&NiagaraBool}}}
      Random Seed {bIsDynamic=false, bIsEditable=false, bIsStaticSwitch=false, bIsVisible=true, type {classStructOrEnum {refPath=&NiagaraInt32}}}
      Fixed Random Seed {bIsDynamic=false, bIsEditable=false, bIsStaticSwitch=true, bIsVisible=false, type {classStructOrEnum {refPath=&NiagaraBool}}}
      Transform Order {
       bIsDynamic=false
       bIsEditable=true
       bIsStaticSwitch=true
       bIsVisible=true
       type {classStructOrEnum {refPath=&ENiagara_TransformOrder}}
      }
      Scale Mode {
       bIsDynamic=false
       bIsEditable=true
       bIsStaticSwitch=true
       bIsVisible=true
       type {classStructOrEnum {refPath=&ENiagara_ScaleMode}}
      }
      Non Uniform Scale {bIsDynamic=false, bIsEditable=true, bIsStaticSwitch=false, bIsVisible=true, type {classStructOrEnum {refPath=&Vector3f}}}
      UseOwnerScale {bIsDynamic=false, bIsEditable=true, bIsStaticSwitch=false, bIsVisible=true, type {classStructOrEnum {refPath=&NiagaraBool}}}
      Apply Owner Scale {bIsDynamic=false, bIsEditable=true, bIsStaticSwitch=false, bIsVisible=true, type {classStructOrEnum {refPath=&Vector3f}}}
      Rotation Mode {
       bIsDynamic=false
       bIsEditable=true
       bIsStaticSwitch=true
       bIsVisible=true
       type {classStructOrEnum {refPath=&ENiagara_RotationMode}}
      }
      Rotation Axis {
       bIsDynamic=false
       bIsEditable=false
       bIsStaticSwitch=false
       bIsVisible=false
       type {classStructOrEnum {refPath=&Vector3f}}
      }
      Rotation Matrix {
       bIsDynamic=false
       bIsEditable=false
       bIsStaticSwitch=false
       bIsVisible=false
       type {classStructOrEnum {refPath=&NiagaraMatrix}}
      }
      Rotation Quaternion {bIsDynamic=false, bIsEditable=false, bIsStaticSwitch=false, bIsVisible=false, type {classStructOrEnum {refPath=&Quat4f}}}
      Yaw / Pitch / Roll {
       bIsDynamic=false
       bIsEditable=false
       bIsStaticSwitch=false
       bIsVisible=false
       type {classStructOrEnum {refPath=&Vector3f}}
      }
      Rotation Angle {
       bIsDynamic=false
       bIsEditable=false
       bIsStaticSwitch=false
       bIsVisible=false
       type {classStructOrEnum {refPath=&NiagaraFloat}}
      }
      Invert Rotation Quaternion {bIsDynamic=false, bIsEditable=false, bIsStaticSwitch=false, bIsVisible=false, type {classStructOrEnum {refPath=&NiagaraBool}}}
      Additional Quaternion Rotation {bIsDynamic=false, bIsEditable=false, bIsStaticSwitch=false, bIsVisible=false, type {classStructOrEnum {refPath=&NiagaraBool}}}
      Additional Yaw / Pitch / Roll {
       bIsDynamic=false
       bIsEditable=false
       bIsStaticSwitch=false
       bIsVisible=false
       type {classStructOrEnum {refPath=&Vector3f}}
      }
      Rotation Angle Type {
       bIsDynamic=false
       bIsEditable=false
       bIsStaticSwitch=true
       bIsVisible=false
       type {classStructOrEnum {refPath=&ENiagara_AngleInput}}
      }
      Rotation Coordinate Space {
       bIsDynamic=false
       bIsEditable=true
       bIsStaticSwitch=false
       bIsVisible=true
       type {classStructOrEnum {refPath=&ENiagaraCoordinateSpace}}
      }
      Offset Mode {
       bIsDynamic=false
       bIsEditable=true
       bIsStaticSwitch=true
       bIsVisible=true
       type {classStructOrEnum {refPath=&ENiagara_OffsetMode}}
      }
      Shape Origin {
       bIsDynamic=false
       bIsEditable=true
       bIsStaticSwitch=false
       bIsVisible=true
       type {classStructOrEnum {refPath=&NiagaraPosition}}
      }
      Offset {
       bIsDynamic=false
       bIsEditable=false
       bIsStaticSwitch=false
       bIsVisible=false
       type {classStructOrEnum {refPath=&Vector3f}}
      }
      Offset Coordinate Space {
       bIsDynamic=false
       bIsEditable=false
       bIsStaticSwitch=false
       bIsVisible=false
       type {classStructOrEnum {refPath=&ENiagaraCoordinateSpace}}
      }
      Custom Transform Matrix {
       bIsDynamic=false
       bIsEditable=false
       bIsStaticSwitch=false
       bIsVisible=false
       type {classStructOrEnum {refPath=&NiagaraMatrix}}
      }
      Transform Method {
       bIsDynamic=false
       bIsEditable=true
       bIsStaticSwitch=true
       bIsVisible=true
       type {classStructOrEnum {refPath=&ENiagara_TransformType}}
      }
     ]
     moduleScript {refPath=&ShapeLocation}
    }
    {
     bIsSetParametersModule=false
     enabled=true
     moduleName=AddVelocity
     inputs [
      Velocity Mode {
       bIsDynamic=false
       bIsEditable=true
       bIsStaticSwitch=true
       bIsVisible=true
       type {classStructOrEnum {refPath=&ENiagara_VelocityMode}}
      }
      Velocity {
       bIsDynamic=false
       bIsEditable=false
       bIsStaticSwitch=false
       bIsVisible=false
       type {classStructOrEnum {refPath=&Vector3f}}
      }
      Velocity Speed {bIsDynamic=true, bIsEditable=true, bIsStaticSwitch=false, bIsVisible=true, type {classStructOrEnum {refPath=&NiagaraFloat}}}
      Velocity Speed Scale {
       bIsDynamic=false
       bIsEditable=false
       bIsStaticSwitch=false
       bIsVisible=false
       type {classStructOrEnum {refPath=&NiagaraFloat}}
      }
      Constrain To Radius {
       bIsDynamic=false
       bIsEditable=false
       bIsStaticSwitch=false
       bIsVisible=false
       type {classStructOrEnum {refPath=&NiagaraFloat}}
      }
      Module.UseVelDistribution {bIsDynamic=false, bIsEditable=true, bIsStaticSwitch=false, bIsVisible=true, type {classStructOrEnum {refPath=&NiagaraBool}}}
      Distribution Along Cone Axis {bIsDynamic=false, bIsEditable=true, bIsStaticSwitch=false, bIsVisible=true, type {classStructOrEnum {refPath=&NiagaraFloat}}}
      Radius Falloff Near / Far {
       bIsDynamic=false
       bIsEditable=false
       bIsStaticSwitch=false
       bIsVisible=false
       type {classStructOrEnum {refPath=&Vector2f}}
      }
      Radius Falloff Exponent {
       bIsDynamic=false
       bIsEditable=false
       bIsStaticSwitch=false
       bIsVisible=false
       type {classStructOrEnum {refPath=&NiagaraFloat}}
      }
      Module.UseConeFalloff {bIsDynamic=false, bIsEditable=true, bIsStaticSwitch=false, bIsVisible=true, type {classStructOrEnum {refPath=&NiagaraBool}}}
      Speed Falloff From Cone Axis {bIsDynamic=false, bIsEditable=true, bIsStaticSwitch=false, bIsVisible=true, type {classStructOrEnum {refPath=&NiagaraFloat}}}
      Invert Velocity Falloff {bIsDynamic=false, bIsEditable=false, bIsStaticSwitch=false, bIsVisible=false, type {classStructOrEnum {refPath=&NiagaraBool}}}
      Override Seed {bIsDynamic=false, bIsEditable=true, bIsStaticSwitch=false, bIsVisible=true, type {classStructOrEnum {refPath=&NiagaraBool}}}
      Random Seed {bIsDynamic=false, bIsEditable=false, bIsStaticSwitch=false, bIsVisible=true, type {classStructOrEnum {refPath=&NiagaraInt32}}}
      Fixed Random Seed {bIsDynamic=false, bIsEditable=false, bIsStaticSwitch=true, bIsVisible=false, type {classStructOrEnum {refPath=&NiagaraBool}}}
      Randomness Mode {
       bIsDynamic=false
       bIsEditable=true
       bIsStaticSwitch=true
       bIsVisible=true
       type {classStructOrEnum {refPath=&ENiagaraRandomnessMode}}
      }
      Cone Axis {bIsDynamic=true, bIsEditable=true, bIsStaticSwitch=false, bIsVisible=true, type {classStructOrEnum {refPath=&Vector3f}}}
      Cone Angle {bIsDynamic=true, bIsEditable=true, bIsStaticSwitch=false, bIsVisible=true, type {classStructOrEnum {refPath=&NiagaraFloat}}}
      Inner Cone Angle {bIsDynamic=false, bIsEditable=true, bIsStaticSwitch=false, bIsVisible=true, type {classStructOrEnum {refPath=&NiagaraFloat}}}
      Cone Angle Mode {
       bIsDynamic=false
       bIsEditable=true
       bIsStaticSwitch=true
       bIsVisible=true
       type {classStructOrEnum {refPath=&ENiagara_AngleInput}}
      }
      Rotation Mode {
       bIsDynamic=false
       bIsEditable=true
       bIsStaticSwitch=true
       bIsVisible=true
       type {classStructOrEnum {refPath=&ENiagara_RotationMode}}
      }
      Rotation Axis {
       bIsDynamic=false
       bIsEditable=false
       bIsStaticSwitch=false
       bIsVisible=false
       type {classStructOrEnum {refPath=&Vector3f}}
      }
      Yaw / Pitch / Roll {
       bIsDynamic=false
       bIsEditable=false
       bIsStaticSwitch=false
       bIsVisible=false
       type {classStructOrEnum {refPath=&Vector3f}}
      }
      Rotation Quaternion {bIsDynamic=false, bIsEditable=false, bIsStaticSwitch=false, bIsVisible=false, type {classStructOrEnum {refPath=&Quat4f}}}
      Rotation Matrix {
       bIsDynamic=false
       bIsEditable=false
       bIsStaticSwitch=false
       bIsVisible=false
       type {classStructOrEnum {refPath=&NiagaraMatrix}}
      }
      Rotation Angle {
       bIsDynamic=false
       bIsEditable=false
       bIsStaticSwitch=false
       bIsVisible=false
       type {classStructOrEnum {refPath=&NiagaraFloat}}
      }
      Invert Rotation Quaternion {bIsDynamic=false, bIsEditable=false, bIsStaticSwitch=false, bIsVisible=false, type {classStructOrEnum {refPath=&NiagaraBool}}}
      Additional Quaternion Rotation {bIsDynamic=false, bIsEditable=false, bIsStaticSwitch=false, bIsVisible=false, type {classStructOrEnum {refPath=&NiagaraBool}}}
      Additional Yaw / Pitch / Roll {
       bIsDynamic=false
       bIsEditable=false
       bIsStaticSwitch=false
       bIsVisible=false
       type {classStructOrEnum {refPath=&Vector3f}}
      }
      Rotation Angle Type {
       bIsDynamic=false
       bIsEditable=false
       bIsStaticSwitch=true
       bIsVisible=false
       type {classStructOrEnum {refPath=&ENiagara_AngleInput}}
      }
      Rotation Coordinate Space {
       bIsDynamic=false
       bIsEditable=true
       bIsStaticSwitch=false
       bIsVisible=true
       type {classStructOrEnum {refPath=&ENiagaraCoordinateSpace}}
      }
      Velocity Origin {
       bIsDynamic=false
       bIsEditable=false
       bIsStaticSwitch=false
       bIsVisible=false
       type {classStructOrEnum {refPath=&NiagaraPosition}}
      }
      Origin Offset {
       bIsDynamic=false
       bIsEditable=false
       bIsStaticSwitch=false
       bIsVisible=false
       type {classStructOrEnum {refPath=&Vector3f}}
      }
      Origin Offset Coordinate Space {
       bIsDynamic=false
       bIsEditable=false
       bIsStaticSwitch=false
       bIsVisible=false
       type {classStructOrEnum {refPath=&ENiagaraCoordinateSpace}}
      }
      Default Position {
       bIsDynamic=false
       bIsEditable=false
       bIsStaticSwitch=false
       bIsVisible=false
       type {classStructOrEnum {refPath=&NiagaraPosition}}
      }
     ]
     moduleScript {refPath=&AddVelocity}
    }
    {
     bIsSetParametersModule=false
     enabled=true
     moduleName=SubUVAnimation
     inputs [
      Manual SubImage Size {
       bIsDynamic=false
       bIsEditable=false
       bIsStaticSwitch=false
       bIsVisible=false
       type {classStructOrEnum {refPath=&Vector2f}}
      }
      Is SubUV Blending Enabled On Renderer {bIsDynamic=false, bIsEditable=false, bIsStaticSwitch=false, bIsVisible=false, type {classStructOrEnum {refPath=&NiagaraBool}}}
      SubUV Animation Mode {
       bIsDynamic=false
       bIsEditable=true
       bIsStaticSwitch=true
       bIsVisible=true
       type {classStructOrEnum {refPath=&ENiagara_SubUVLookupModeV2}}
      }
      Renderer Type {
       bIsDynamic=false
       bIsEditable=true
       bIsStaticSwitch=true
       bIsVisible=true
       type {classStructOrEnum {refPath=&ENiagaraMeshOrSprite}}
      }
      Mesh Renderer {
       bIsDynamic=false
       bIsEditable=false
       bIsStaticSwitch=false
       bIsVisible=false
       type {classStructOrEnum {refPath=&NiagaraDataInterfaceMeshRendererInfo}}
      }
      "Sprite Renderer " {
       bIsDynamic=false
       bIsEditable=true
       bIsStaticSwitch=false
       bIsVisible=true
       type {classStructOrEnum {refPath=&NiagaraDataInterfaceSpriteRendererInfo}}
      }
      Define SubUV Setup Manually {
       bIsDynamic=false
       bIsEditable=true
       bIsStaticSwitch=true
       bIsVisible=true
       type {classStructOrEnum {refPath=&ENiagara_AutomaticManual}}
      }
      Direct Frame Index {
       bIsDynamic=false
       bIsEditable=false
       bIsStaticSwitch=false
       bIsVisible=false
       type {classStructOrEnum {refPath=&NiagaraFloat}}
      }
      Random Start Frame {bIsDynamic=false, bIsEditable=false, bIsStaticSwitch=true, bIsVisible=false, type {classStructOrEnum {refPath=&NiagaraBool}}}
      Start Frame Offset {
       bIsDynamic=false
       bIsEditable=false
       bIsStaticSwitch=false
       bIsVisible=false
       type {classStructOrEnum {refPath=&NiagaraInt32}}
      }
      Start Time Offset {
       bIsDynamic=false
       bIsEditable=false
       bIsStaticSwitch=false
       bIsVisible=false
       type {classStructOrEnum {refPath=&NiagaraFloat}}
      }
      SubUV Lookup Index {
       bIsDynamic=false
       bIsEditable=false
       bIsStaticSwitch=false
       bIsVisible=false
       type {classStructOrEnum {refPath=&NiagaraFloat}}
      }
      X Coordinate {
       bIsDynamic=false
       bIsEditable=false
       bIsStaticSwitch=false
       bIsVisible=false
       type {classStructOrEnum {refPath=&NiagaraFloat}}
      }
      Playback Mode {
       bIsDynamic=false
       bIsEditable=false
       bIsStaticSwitch=true
       bIsVisible=false
       type {classStructOrEnum {refPath=&ENiagara_FPSPlayrate}}
      }
      Frames Per Second {
       bIsDynamic=false
       bIsEditable=false
       bIsStaticSwitch=false
       bIsVisible=false
       type {classStructOrEnum {refPath=&NiagaraInt32}}
      }
      Play Rate {
       bIsDynamic=false
       bIsEditable=false
       bIsStaticSwitch=false
       bIsVisible=false
       type {classStructOrEnum {refPath=&NiagaraFloat}}
      }
      Playback Time {
       bIsDynamic=false
       bIsEditable=false
       bIsStaticSwitch=false
       bIsVisible=false
       type {classStructOrEnum {refPath=&NiagaraFloat}}
      }
      UsePlayRateScale {bIsDynamic=false, bIsEditable=false, bIsStaticSwitch=false, bIsVisible=false, type {classStructOrEnum {refPath=&NiagaraBool}}}
      Playback Scale Curve {
       bIsDynamic=false
       bIsEditable=false
       bIsStaticSwitch=false
       bIsVisible=false
       type {classStructOrEnum {refPath=&NiagaraDataInterfaceCurve}}
      }
      Playback Scale Curve Index {
       bIsDynamic=false
       bIsEditable=false
       bIsStaticSwitch=false
       bIsVisible=false
       type {classStructOrEnum {refPath=&NiagaraFloat}}
      }
      SubUV Curve {
       bIsDynamic=false
       bIsEditable=false
       bIsStaticSwitch=false
       bIsVisible=false
       type {classStructOrEnum {refPath=&NiagaraDataInterfaceCurve}}
      }
      SubUV Lookup Index Scale {
       bIsDynamic=false
       bIsEditable=false
       bIsStaticSwitch=false
       bIsVisible=false
       type {classStructOrEnum {refPath=&NiagaraFloat}}
      }
      Y Coordinate {
       bIsDynamic=false
       bIsEditable=false
       bIsStaticSwitch=false
       bIsVisible=false
       type {classStructOrEnum {refPath=&NiagaraFloat}}
      }
      SubUV Curve Index {
       bIsDynamic=false
       bIsEditable=false
       bIsStaticSwitch=false
       bIsVisible=false
       type {classStructOrEnum {refPath=&NiagaraFloat}}
      }
      Index By X/Y Coordinate {bIsDynamic=false, bIsEditable=false, bIsStaticSwitch=true, bIsVisible=false, type {classStructOrEnum {refPath=&NiagaraBool}}}
      Evaluation Type {
       bIsDynamic=false
       bIsEditable=true
       bIsStaticSwitch=true
       bIsVisible=true
       type {classStructOrEnum {refPath=&ENiagaraRandomnessEvaluation}}
      }
      Random Change Interval {
       bIsDynamic=false
       bIsEditable=false
       bIsStaticSwitch=false
       bIsVisible=false
       type {classStructOrEnum {refPath=&NiagaraFloat}}
      }
      Loop Range Boundary {
       bIsDynamic=false
       bIsEditable=false
       bIsStaticSwitch=true
       bIsVisible=false
       type {classStructOrEnum {refPath=&ENiagara_WrapClamp}}
      }
      Loop Boundary Min/Max {
       bIsDynamic=false
       bIsEditable=false
       bIsStaticSwitch=false
       bIsVisible=false
       type {classStructOrEnum {refPath=&Vector2f}}
      }
      Truncate Y Coordinate {bIsDynamic=false, bIsEditable=false, bIsStaticSwitch=true, bIsVisible=false, type {classStructOrEnum {refPath=&NiagaraBool}}}
      Random Change Interval Time {
       bIsDynamic=false
       bIsEditable=false
       bIsStaticSwitch=false
       bIsVisible=false
       type {classStructOrEnum {refPath=&NiagaraFloat}}
      }
      Randomness Mode {
       bIsDynamic=false
       bIsEditable=true
       bIsStaticSwitch=true
       bIsVisible=true
       type {classStructOrEnum {refPath=&ENiagaraRandomnessMode}}
      }
      Override Seed {bIsDynamic=false, bIsEditable=true, bIsStaticSwitch=true, bIsVisible=true, type {classStructOrEnum {refPath=&NiagaraBool}}}
      Random Seed {
       bIsDynamic=false
       bIsEditable=false
       bIsStaticSwitch=false
       bIsVisible=false
       type {classStructOrEnum {refPath=&NiagaraInt32}}
      }
      Fixed Random Seed {bIsDynamic=false, bIsEditable=true, bIsStaticSwitch=true, bIsVisible=true, type {classStructOrEnum {refPath=&NiagaraBool}}}
      Output Alternate Renderer Binding {bIsDynamic=false, bIsEditable=true, bIsStaticSwitch=true, bIsVisible=true, type {classStructOrEnum {refPath=&NiagaraBool}}}
      UseStartFrame {bIsDynamic=false, bIsEditable=true, bIsStaticSwitch=false, bIsVisible=true, type {classStructOrEnum {refPath=&NiagaraBool}}}
      Start Frame Range Override {bIsDynamic=false, bIsEditable=false, bIsStaticSwitch=false, bIsVisible=true, type {classStructOrEnum {refPath=&NiagaraInt32}}}
      UseEndFrame {bIsDynamic=false, bIsEditable=true, bIsStaticSwitch=false, bIsVisible=true, type {classStructOrEnum {refPath=&NiagaraBool}}}
      End Frame Range Override {bIsDynamic=false, bIsEditable=false, bIsStaticSwitch=false, bIsVisible=true, type {classStructOrEnum {refPath=&NiagaraInt32}}}
     ]
     moduleScript {refPath=&SubUVAnimation}
    }
    {
     bIsSetParametersModule=false
     enabled=true
     moduleName=InheritVelocity
     inputs [
      Inherited Velocity Amount Scale {bIsDynamic=false, bIsEditable=true, bIsStaticSwitch=false, bIsVisible=true, type {classStructOrEnum {refPath=&Vector3f}}}
      Speed Limit {bIsDynamic=false, bIsEditable=true, bIsStaticSwitch=false, bIsVisible=true, type {classStructOrEnum {refPath=&NiagaraBool}}}
      Inherited Velocity Speed Limit {bIsDynamic=false, bIsEditable=false, bIsStaticSwitch=false, bIsVisible=true, type {classStructOrEnum {refPath=&NiagaraFloat}}}
      UseSpeedThreshold {bIsDynamic=false, bIsEditable=true, bIsStaticSwitch=false, bIsVisible=true, type {classStructOrEnum {refPath=&NiagaraBool}}}
      Source Speed Threshold {bIsDynamic=false, bIsEditable=false, bIsStaticSwitch=false, bIsVisible=true, type {classStructOrEnum {refPath=&NiagaraFloat}}}
      Velocity Source {bIsDynamic=false, bIsEditable=true, bIsStaticSwitch=false, bIsVisible=true, type {classStructOrEnum {refPath=&Vector3f}}}
      Velocity Source Space {
       bIsDynamic=false
       bIsEditable=true
       bIsStaticSwitch=false
       bIsVisible=true
       type {classStructOrEnum {refPath=&ENiagaraCoordinateSpace}}
      }
      Velocity Destination Space {
       bIsDynamic=false
       bIsEditable=true
       bIsStaticSwitch=false
       bIsVisible=true
       type {classStructOrEnum {refPath=&ENiagaraCoordinateSpace}}
      }
     ]
     moduleScript {refPath=&InheritVelocity}
    }
   ]
  }
  particleUpdateScript {
   scriptName=ParticleUpdateScript
   modules [
    {
     bIsSetParametersModule=false
     enabled=true
     moduleName=ParticleState
     inputs [
      Kill Particles When Lifetime Has Elapsed {bIsDynamic=false, bIsEditable=true, bIsStaticSwitch=true, bIsVisible=true, type {classStructOrEnum {refPath=&NiagaraBool}}}
      Loop Particles Lifetime {bIsDynamic=false, bIsEditable=false, bIsStaticSwitch=false, bIsVisible=false, type {classStructOrEnum {refPath=&NiagaraBool}}}
      Let Infinitely Lived Particles Die When Emitter Deactivates {bIsDynamic=false, bIsEditable=false, bIsStaticSwitch=false, bIsVisible=false, type {classStructOrEnum {refPath=&NiagaraBool}}}
      Lifetime {bIsDynamic=false, bIsEditable=true, bIsStaticSwitch=false, bIsVisible=true, type {classStructOrEnum {refPath=&NiagaraFloat}}}
      DeltaTime {bIsDynamic=false, bIsEditable=true, bIsStaticSwitch=false, bIsVisible=true, type {classStructOrEnum {refPath=&NiagaraFloat}}}
     ]
     moduleScript {refPath=&ParticleState}
    }
    {
     bIsSetParametersModule=false
     enabled=true
     moduleName=Drag
     inputs [
      Use Linear Drag {bIsDynamic=false, bIsEditable=true, bIsStaticSwitch=false, bIsVisible=true, type {classStructOrEnum {refPath=&NiagaraBool}}}
      Drag {bIsDynamic=true, bIsEditable=true, bIsStaticSwitch=false, bIsVisible=true, type {classStructOrEnum {refPath=&NiagaraFloat}}}
      Use Rotational Drag {bIsDynamic=false, bIsEditable=true, bIsStaticSwitch=false, bIsVisible=true, type {classStructOrEnum {refPath=&NiagaraBool}}}
      Rotational Drag {bIsDynamic=false, bIsEditable=true, bIsStaticSwitch=false, bIsVisible=true, type {classStructOrEnum {refPath=&NiagaraFloat}}}
      Ignore Mass {bIsDynamic=false, bIsEditable=true, bIsStaticSwitch=false, bIsVisible=true, type {classStructOrEnum {refPath=&NiagaraBool}}}
     ]
     moduleScript {refPath=&Drag}
    }
    {
     bIsSetParametersModule=false
     enabled=true
     moduleName=GravityForce
     inputs [
      Gravity {bIsDynamic=true, bIsEditable=true, bIsStaticSwitch=false, bIsVisible=true, type {classStructOrEnum {refPath=&Vector3f}}}
      Coordinate Space {
       bIsDynamic=false
       bIsEditable=true
       bIsStaticSwitch=true
       bIsVisible=true
       type {classStructOrEnum {refPath=&ENiagaraCoordinateSpace}}
      }
     ]
     moduleScript {refPath=&GravityForce}
    }
    {
     bIsSetParametersModule=false
     enabled=true
     moduleName=AccelerationForce
     inputs [
      Acceleration {bIsDynamic=true, bIsEditable=true, bIsStaticSwitch=false, bIsVisible=true, type {classStructOrEnum {refPath=&Vector3f}}}
      Coordinate Space {
       bIsDynamic=false
       bIsEditable=true
       bIsStaticSwitch=false
       bIsVisible=true
       type {classStructOrEnum {refPath=&ENiagaraCoordinateSpace}}
      }
     ]
     moduleScript {refPath=&AccelerationForce}
    }
    {
     bIsSetParametersModule=false
     enabled=true
     moduleName=InheritSourceMovement
     inputs [
      ApplyPosition {bIsDynamic=false, bIsEditable=true, bIsStaticSwitch=false, bIsVisible=true, type {classStructOrEnum {refPath=&NiagaraBool}}}
      Applied Position Scale {bIsDynamic=false, bIsEditable=false, bIsStaticSwitch=false, bIsVisible=true, type {classStructOrEnum {refPath=&Vector3f}}}
      ApplyPositionFalloff {bIsDynamic=false, bIsEditable=false, bIsStaticSwitch=false, bIsVisible=false, type {classStructOrEnum {refPath=&NiagaraBool}}}
      Applied Position Falloff Distance {
       bIsDynamic=false
       bIsEditable=false
       bIsStaticSwitch=false
       bIsVisible=false
       type {classStructOrEnum {refPath=&NiagaraFloat}}
      }
      ApplyPositionFalloffCurve {bIsDynamic=false, bIsEditable=false, bIsStaticSwitch=false, bIsVisible=false, type {classStructOrEnum {refPath=&NiagaraBool}}}
      Applied Position Falloff Curve {
       bIsDynamic=false
       bIsEditable=false
       bIsStaticSwitch=false
       bIsVisible=false
       type {classStructOrEnum {refPath=&NiagaraDataInterfaceCurve}}
      }
      Applied Position Falloff Origin {
       bIsDynamic=false
       bIsEditable=false
       bIsStaticSwitch=false
       bIsVisible=false
       type {classStructOrEnum {refPath=&NiagaraPosition}}
      }
      ApplyForce {bIsDynamic=false, bIsEditable=true, bIsStaticSwitch=false, bIsVisible=true, type {classStructOrEnum {refPath=&NiagaraBool}}}
      Applied Force Scale {bIsDynamic=false, bIsEditable=true, bIsStaticSwitch=false, bIsVisible=true, type {classStructOrEnum {refPath=&Vector3f}}}
      Module.LimitForceSpeed {bIsDynamic=false, bIsEditable=true, bIsStaticSwitch=false, bIsVisible=true, type {classStructOrEnum {refPath=&NiagaraBool}}}
      Applied Force Speed Limit {bIsDynamic=false, bIsEditable=false, bIsStaticSwitch=false, bIsVisible=true, type {classStructOrEnum {refPath=&NiagaraFloat}}}
      ApplyForceFalloff {bIsDynamic=false, bIsEditable=true, bIsStaticSwitch=false, bIsVisible=true, type {classStructOrEnum {refPath=&NiagaraBool}}}
      Applied Force Falloff Distance {bIsDynamic=false, bIsEditable=false, bIsStaticSwitch=false, bIsVisible=true, type {classStructOrEnum {refPath=&NiagaraFloat}}}
      ApplyForceFalloffCurve {bIsDynamic=false, bIsEditable=false, bIsStaticSwitch=false, bIsVisible=false, type {classStructOrEnum {refPath=&NiagaraBool}}}
      Applied Force Falloff Curve {
       bIsDynamic=false
       bIsEditable=false
       bIsStaticSwitch=false
       bIsVisible=false
       type {classStructOrEnum {refPath=&NiagaraDataInterfaceCurve}}
      }
      Applied Force Falloff Origin {
       bIsDynamic=false
       bIsEditable=false
       bIsStaticSwitch=false
       bIsVisible=false
       type {classStructOrEnum {refPath=&NiagaraPosition}}
      }
      Applied Force Falloff Position {
       bIsDynamic=false
       bIsEditable=false
       bIsStaticSwitch=false
       bIsVisible=false
       type {classStructOrEnum {refPath=&NiagaraPosition}}
      }
      Source Velocity {bIsDynamic=false, bIsEditable=true, bIsStaticSwitch=false, bIsVisible=true, type {classStructOrEnum {refPath=&Vector3f}}}
     ]
     moduleScript {refPath=&InheritSourceMovement}
    }
    {
     bIsSetParametersModule=false
     enabled=true
     moduleName=SolveForcesAndVelocity
     inputs [
      Clamp Velocity {bIsDynamic=false, bIsEditable=true, bIsStaticSwitch=false, bIsVisible=true, type {classStructOrEnum {refPath=&NiagaraBool}}}
      Speed Limit {bIsDynamic=false, bIsEditable=false, bIsStaticSwitch=false, bIsVisible=true, type {classStructOrEnum {refPath=&NiagaraFloat}}}
      Limit Acceleration {bIsDynamic=false, bIsEditable=true, bIsStaticSwitch=false, bIsVisible=true, type {classStructOrEnum {refPath=&NiagaraBool}}}
      Acceleration Limit {bIsDynamic=false, bIsEditable=false, bIsStaticSwitch=false, bIsVisible=true, type {classStructOrEnum {refPath=&NiagaraFloat}}}
      Manually Enable Rotational Solver {bIsDynamic=false, bIsEditable=true, bIsStaticSwitch=false, bIsVisible=true, type {classStructOrEnum {refPath=&NiagaraBool}}}
      Write to Intrinsic Properties {bIsDynamic=false, bIsEditable=true, bIsStaticSwitch=false, bIsVisible=true, type {classStructOrEnum {refPath=&NiagaraBool}}}
      Write to Presolve Properties {bIsDynamic=false, bIsEditable=true, bIsStaticSwitch=true, bIsVisible=true, type {classStructOrEnum {refPath=&NiagaraBool}}}
      DeltaTime {bIsDynamic=false, bIsEditable=true, bIsStaticSwitch=false, bIsVisible=true, type {classStructOrEnum {refPath=&NiagaraFloat}}}
      Mass {bIsDynamic=false, bIsEditable=true, bIsStaticSwitch=false, bIsVisible=true, type {classStructOrEnum {refPath=&NiagaraFloat}}}
      Position {
       bIsDynamic=false
       bIsEditable=true
       bIsStaticSwitch=false
       bIsVisible=true
       type {classStructOrEnum {refPath=&NiagaraPosition}}
      }
      Velocity {bIsDynamic=false, bIsEditable=true, bIsStaticSwitch=false, bIsVisible=true, type {classStructOrEnum {refPath=&Vector3f}}}
      PreviousVelocity {bIsDynamic=false, bIsEditable=true, bIsStaticSwitch=false, bIsVisible=true, type {classStructOrEnum {refPath=&Vector3f}}}
      Force {bIsDynamic=false, bIsEditable=true, bIsStaticSwitch=false, bIsVisible=true, type {classStructOrEnum {refPath=&Vector3f}}}
      Mesh Orientation {bIsDynamic=false, bIsEditable=false, bIsStaticSwitch=false, bIsVisible=false, type {classStructOrEnum {refPath=&Quat4f}}}
      Rotational Velocity {
       bIsDynamic=false
       bIsEditable=false
       bIsStaticSwitch=false
       bIsVisible=false
       type {classStructOrEnum {refPath=&Vector3f}}
      }
      Rotational Force {
       bIsDynamic=false
       bIsEditable=false
       bIsStaticSwitch=false
       bIsVisible=false
       type {classStructOrEnum {refPath=&Vector3f}}
      }
      Rotational Inertia {
       bIsDynamic=false
       bIsEditable=false
       bIsStaticSwitch=false
       bIsVisible=false
       type {classStructOrEnum {refPath=&NiagaraFloat}}
      }
     ]
     moduleScript {refPath=&SolveForcesAndVelocity}
    }
    {
     bIsSetParametersModule=false
     enabled=true
     moduleName=SpriteRotationRate
     inputs [
      Rotation Rate {bIsDynamic=true, bIsEditable=true, bIsStaticSwitch=false, bIsVisible=true, type {classStructOrEnum {refPath=&NiagaraFloat}}}
      Delta Time {bIsDynamic=false, bIsEditable=true, bIsStaticSwitch=false, bIsVisible=true, type {classStructOrEnum {refPath=&NiagaraFloat}}}
     ]
     moduleScript {refPath=&SpriteRotationRate}
    }
    {
     bIsSetParametersModule=false
     enabled=true
     moduleName=ScaleSpriteSize
     inputs [
      Scale Sprite Size Mode {
       bIsDynamic=false
       bIsEditable=true
       bIsStaticSwitch=true
       bIsVisible=true
       type {classStructOrEnum {refPath=&ENiagara_ScaleSpriteSize}}
      }
      Uniform Curve Sprite Scale {
       bIsDynamic=false
       bIsEditable=false
       bIsStaticSwitch=false
       bIsVisible=false
       type {classStructOrEnum {refPath=&NiagaraDataInterfaceCurve}}
      }
      Non-Uniform Curve Sprite Scale {
       bIsDynamic=false
       bIsEditable=false
       bIsStaticSwitch=false
       bIsVisible=false
       type {classStructOrEnum {refPath=&NiagaraDataInterfaceVector2DCurve}}
      }
      Uniform Curve Index {
       bIsDynamic=false
       bIsEditable=false
       bIsStaticSwitch=false
       bIsVisible=false
       type {classStructOrEnum {refPath=&NiagaraFloat}}
      }
      Non-Uniform Curve Index {
       bIsDynamic=false
       bIsEditable=false
       bIsStaticSwitch=false
       bIsVisible=false
       type {classStructOrEnum {refPath=&NiagaraFloat}}
      }
      Non-Uniform Curve Scale {
       bIsDynamic=false
       bIsEditable=false
       bIsStaticSwitch=false
       bIsVisible=false
       type {classStructOrEnum {refPath=&Vector2f}}
      }
      Uniform Curve Scale {
       bIsDynamic=false
       bIsEditable=false
       bIsStaticSwitch=false
       bIsVisible=false
       type {classStructOrEnum {refPath=&NiagaraFloat}}
      }
      Scale Factor {bIsDynamic=true, bIsEditable=true, bIsStaticSwitch=false, bIsVisible=true, type {classStructOrEnum {refPath=&Vector2f}}}
      Uniform Scale Factor {
       bIsDynamic=false
       bIsEditable=false
       bIsStaticSwitch=false
       bIsVisible=false
       type {classStructOrEnum {refPath=&NiagaraFloat}}
      }
      Initial Sprite Size {bIsDynamic=false, bIsEditable=true, bIsStaticSwitch=false, bIsVisible=true, type {classStructOrEnum {refPath=&Vector2f}}}
     ]
     moduleScript {refPath=&ScaleSpriteSize}
    }
    {
     bIsSetParametersModule=false
     enabled=true
     moduleName=SpriteFacingAndAlignment
     inputs [
      WriteFacing {bIsDynamic=false, bIsEditable=true, bIsStaticSwitch=false, bIsVisible=true, type {classStructOrEnum {refPath=&NiagaraBool}}}
      Sprite Facing {bIsDynamic=false, bIsEditable=false, bIsStaticSwitch=false, bIsVisible=true, type {classStructOrEnum {refPath=&Vector3f}}}
      Facing Coordinate Space {
       bIsDynamic=false
       bIsEditable=false
       bIsStaticSwitch=false
       bIsVisible=false
       type {classStructOrEnum {refPath=&ENiagaraCoordinateSpace}}
      }
      WriteAlignment {bIsDynamic=false, bIsEditable=true, bIsStaticSwitch=false, bIsVisible=true, type {classStructOrEnum {refPath=&NiagaraBool}}}
      Sprite Alignment {bIsDynamic=true, bIsEditable=true, bIsStaticSwitch=false, bIsVisible=true, type {classStructOrEnum {refPath=&Vector3f}}}
      Alignment Coordinate Space {
       bIsDynamic=false
       bIsEditable=true
       bIsStaticSwitch=false
       bIsVisible=true
       type {classStructOrEnum {refPath=&ENiagaraCoordinateSpace}}
      }
     ]
     moduleScript {refPath=&SpriteFacingAndAlignment}
    }
    {
     bIsSetParametersModule=false
     enabled=true
     moduleName=ScaleColor
     inputs [
      Scale Mode {
       bIsDynamic=false
       bIsEditable=true
       bIsStaticSwitch=true
       bIsVisible=true
       type {classStructOrEnum {refPath=&ENiagaraScaleColorMode}}
      }
      Linear Color Curve {
       bIsDynamic=false
       bIsEditable=false
       bIsStaticSwitch=false
       bIsVisible=false
       type {classStructOrEnum {refPath=&NiagaraDataInterfaceColorCurve}}
      }
      ScaleRGB {bIsDynamic=false, bIsEditable=true, bIsStaticSwitch=false, bIsVisible=true, type {classStructOrEnum {refPath=&NiagaraBool}}}
      Scale RGB {bIsDynamic=true, bIsEditable=true, bIsStaticSwitch=false, bIsVisible=true, type {classStructOrEnum {refPath=&Vector3f}}}
      ScaleRGBA {bIsDynamic=false, bIsEditable=false, bIsStaticSwitch=false, bIsVisible=false, type {classStructOrEnum {refPath=&NiagaraBool}}}
      Scale RGBA {
       bIsDynamic=false
       bIsEditable=false
       bIsStaticSwitch=false
       bIsVisible=false
       type {classStructOrEnum {refPath=&Vector4f}}
      }
      Curve Index {
       bIsDynamic=false
       bIsEditable=false
       bIsStaticSwitch=false
       bIsVisible=false
       type {classStructOrEnum {refPath=&NiagaraFloat}}
      }
      ScaleA {bIsDynamic=false, bIsEditable=true, bIsStaticSwitch=false, bIsVisible=true, type {classStructOrEnum {refPath=&NiagaraBool}}}
      Scale Alpha {bIsDynamic=true, bIsEditable=true, bIsStaticSwitch=false, bIsVisible=true, type {classStructOrEnum {refPath=&NiagaraFloat}}}
      Color Value To Scale {
       bIsDynamic=false
       bIsEditable=true
       bIsStaticSwitch=false
       bIsVisible=true
       type {classStructOrEnum {refPath=&LinearColor}}
      }
     ]
     moduleScript {refPath=&ScaleColor}
    }
    {
     bIsSetParametersModule=true
     enabled=true
     moduleName=SetVariables_#16
     inputs [
      Particles.Sparkles Color {
       bIsDynamic=true
       bIsEditable=true
       bIsStaticSwitch=false
       bIsVisible=true
       type {classStructOrEnum {refPath=&LinearColor}}
      }
      Particles.SparklesSpriteRotation {bIsDynamic=false, bIsEditable=true, bIsStaticSwitch=false, bIsVisible=true, type {classStructOrEnum {refPath=&NiagaraFloat}}}
      Particles.SpriteSizeSparkles {bIsDynamic=true, bIsEditable=true, bIsStaticSwitch=false, bIsVisible=true, type {classStructOrEnum {refPath=&Vector2f}}}
     ]
     moduleScript {refPath=&self:SnowflakesCPU_5.NiagaraScriptSource_0.NiagaraGraph_0.NiagaraNodeAssignment_0.SetVariables_#16}
    }
   ]
  }
  rendererClasses [{refPath=&NiagaraSpriteRendererProperties}]
  renderers [
   {rendererIndex=0, rendererClass {refPath=&NiagaraSpriteRendererProperties}}
   {rendererIndex=1, rendererClass {refPath=&NiagaraSpriteRendererProperties}}
  ]
 }
 SoftShockwave {
  bEnabled=false
  emitterSpawnScript {scriptName=EmitterSpawnScript, modules []}
  emitterUpdateScript {
   scriptName=EmitterUpdateScript
   modules [
    {
     bIsSetParametersModule=false
     enabled=true
     moduleName=EmitterState
     inputs [
      Life Cycle Mode {
       bIsDynamic=false
       bIsEditable=true
       bIsStaticSwitch=true
       bIsVisible=true
       type {classStructOrEnum {refPath=&ENiagaraEmitterLifeCycleMode}}
      }
      Inactive Response {
       bIsDynamic=false
       bIsEditable=false
       bIsStaticSwitch=true
       bIsVisible=false
       type {classStructOrEnum {refPath=&ENiagaraInactiveMode}}
      }
      Loop Behavior {
       bIsDynamic=false
       bIsEditable=false
       bIsStaticSwitch=true
       bIsVisible=false
       type {classStructOrEnum {refPath=&ENiagara_EmitterStateOptions}}
      }
      Loop Duration Mode {
       bIsDynamic=false
       bIsEditable=false
       bIsStaticSwitch=true
       bIsVisible=false
       type {classStructOrEnum {refPath=&ENiagara_InfiniteLoopDuration}}
      }
      Loop Count {
       bIsDynamic=false
       bIsEditable=false
       bIsStaticSwitch=false
       bIsVisible=false
       type {classStructOrEnum {refPath=&NiagaraInt32}}
      }
      Loop Duration {
       bIsDynamic=false
       bIsEditable=false
       bIsStaticSwitch=false
       bIsVisible=false
       type {classStructOrEnum {refPath=&NiagaraFloat}}
      }
      Recalculate Duration Each Loop {bIsDynamic=false, bIsEditable=false, bIsStaticSwitch=false, bIsVisible=false, type {classStructOrEnum {refPath=&NiagaraBool}}}
      UseLoopDelay {bIsDynamic=false, bIsEditable=false, bIsStaticSwitch=false, bIsVisible=false, type {classStructOrEnum {refPath=&NiagaraBool}}}
      Loop Delay {
       bIsDynamic=false
       bIsEditable=false
       bIsStaticSwitch=false
       bIsVisible=false
       type {classStructOrEnum {refPath=&NiagaraFloat}}
      }
      Delay First Loop Only {bIsDynamic=false, bIsEditable=false, bIsStaticSwitch=false, bIsVisible=false, type {classStructOrEnum {refPath=&NiagaraBool}}}
      Scalability Mode {
       bIsDynamic=false
       bIsEditable=true
       bIsStaticSwitch=true
       bIsVisible=true
       type {classStructOrEnum {refPath=&ENiagaraEmitterScalabilityMode_Limited}}
      }
      Enable Distance Culling {bIsDynamic=false, bIsEditable=false, bIsStaticSwitch=true, bIsVisible=false, type {classStructOrEnum {refPath=&NiagaraBool}}}
      MinDistance {
       bIsDynamic=false
       bIsEditable=false
       bIsStaticSwitch=false
       bIsVisible=false
       type {classStructOrEnum {refPath=&NiagaraFloat}}
      }
      Min Distance Response {
       bIsDynamic=false
       bIsEditable=false
       bIsStaticSwitch=false
       bIsVisible=false
       type {classStructOrEnum {refPath=&ENiagaraExecutionStateManagement}}
      }
      MaxDistance {
       bIsDynamic=false
       bIsEditable=false
       bIsStaticSwitch=false
       bIsVisible=false
       type {classStructOrEnum {refPath=&NiagaraFloat}}
      }
      Max Distance Response {
       bIsDynamic=false
       bIsEditable=false
       bIsStaticSwitch=false
       bIsVisible=false
       type {classStructOrEnum {refPath=&ENiagaraExecutionStateManagement}}
      }
      Scale Spawn Count {bIsDynamic=false, bIsEditable=false, bIsStaticSwitch=true, bIsVisible=false, type {classStructOrEnum {refPath=&NiagaraBool}}}
      Spawn Count Scale {
       bIsDynamic=false
       bIsEditable=false
       bIsStaticSwitch=false
       bIsVisible=false
       type {classStructOrEnum {refPath=&NiagaraFloat}}
      }
      Spawn Count Scale By Distance {
       bIsDynamic=false
       bIsEditable=false
       bIsStaticSwitch=false
       bIsVisible=false
       type {classStructOrEnum {refPath=&NiagaraDataInterfaceCurve}}
      }
      Enable Visibility Culling {bIsDynamic=false, bIsEditable=false, bIsStaticSwitch=true, bIsVisible=false, type {classStructOrEnum {refPath=&NiagaraBool}}}
      Visibility Cull Response {
       bIsDynamic=false
       bIsEditable=false
       bIsStaticSwitch=false
       bIsVisible=false
       type {classStructOrEnum {refPath=&ENiagaraExecutionStateManagement}}
      }
      Visibility Cull Delay {
       bIsDynamic=false
       bIsEditable=false
       bIsStaticSwitch=false
       bIsVisible=false
       type {classStructOrEnum {refPath=&NiagaraFloat}}
      }
      Reset Age On Awaken {bIsDynamic=false, bIsEditable=false, bIsStaticSwitch=false, bIsVisible=false, type {classStructOrEnum {refPath=&NiagaraBool}}}
     ]
     moduleScript {refPath=&EmitterState}
    }
    {
     bIsSetParametersModule=false
     enabled=true
     moduleName=SpawnBurst_Instantaneous
     inputs [
      Spawn Count {bIsDynamic=false, bIsEditable=true, bIsStaticSwitch=false, bIsVisible=true, type {classStructOrEnum {refPath=&NiagaraInt32}}}
      Spawn Time {bIsDynamic=false, bIsEditable=true, bIsStaticSwitch=false, bIsVisible=true, type {classStructOrEnum {refPath=&NiagaraFloat}}}
      Use Spawn Probability {bIsDynamic=false, bIsEditable=true, bIsStaticSwitch=false, bIsVisible=true, type {classStructOrEnum {refPath=&NiagaraBool}}}
      Spawn Probability {bIsDynamic=false, bIsEditable=false, bIsStaticSwitch=false, bIsVisible=true, type {classStructOrEnum {refPath=&NiagaraFloat}}}
      UseLoopCountLimit {bIsDynamic=false, bIsEditable=true, bIsStaticSwitch=false, bIsVisible=true, type {classStructOrEnum {refPath=&NiagaraBool}}}
      Loop Count Limit {bIsDynamic=false, bIsEditable=false, bIsStaticSwitch=false, bIsVisible=true, type {classStructOrEnum {refPath=&NiagaraInt32}}}
      Spawn Group {bIsDynamic=false, bIsEditable=true, bIsStaticSwitch=false, bIsVisible=true, type {classStructOrEnum {refPath=&NiagaraInt32}}}
      Age {bIsDynamic=false, bIsEditable=true, bIsStaticSwitch=false, bIsVisible=true, type {classStructOrEnum {refPath=&NiagaraFloat}}}
     ]
     moduleScript {refPath=&SpawnBurst_Instantaneous}
    }
   ]
  }
  particleSpawnScript {
   scriptName=ParticleSpawnScript
   modules [{
     bIsSetParametersModule=false
     enabled=true
     moduleName=InitializeParticle
     inputs [
      Lifetime Mode {
       bIsDynamic=false
       bIsEditable=true
       bIsStaticSwitch=true
       bIsVisible=true
       type {classStructOrEnum {refPath=&ENiagara_LifetimeMode}}
      }
      Lifetime Min {
       bIsDynamic=false
       bIsEditable=false
       bIsStaticSwitch=false
       bIsVisible=false
       type {classStructOrEnum {refPath=&NiagaraFloat}}
      }
      Lifetime {bIsDynamic=false, bIsEditable=true, bIsStaticSwitch=false, bIsVisible=true, type {classStructOrEnum {refPath=&NiagaraFloat}}}
      Lifetime Max {
       bIsDynamic=false
       bIsEditable=false
       bIsStaticSwitch=false
       bIsVisible=false
       type {classStructOrEnum {refPath=&NiagaraFloat}}
      }
      Lifetime Randomness Mode {
       bIsDynamic=false
       bIsEditable=false
       bIsStaticSwitch=true
       bIsVisible=false
       type {classStructOrEnum {refPath=&ENiagaraRandomnessMode}}
      }
      Color Mode {
       bIsDynamic=false
       bIsEditable=true
       bIsStaticSwitch=true
       bIsVisible=true
       type {classStructOrEnum {refPath=&ENiagara_ColorInitializationMode}}
      }
      Color Minimum {
       bIsDynamic=false
       bIsEditable=false
       bIsStaticSwitch=false
       bIsVisible=false
       type {classStructOrEnum {refPath=&LinearColor}}
      }
      Color {
       bIsDynamic=false
       bIsEditable=true
       bIsStaticSwitch=false
       bIsVisible=true
       type {classStructOrEnum {refPath=&LinearColor}}
      }
      AdjustHue {bIsDynamic=false, bIsEditable=false, bIsStaticSwitch=false, bIsVisible=false, type {classStructOrEnum {refPath=&NiagaraBool}}}
      Hue Shift Range {
       bIsDynamic=false
       bIsEditable=false
       bIsStaticSwitch=false
       bIsVisible=false
       type {classStructOrEnum {refPath=&Vector2f}}
      }
      Color Maximum {
       bIsDynamic=false
       bIsEditable=false
       bIsStaticSwitch=false
       bIsVisible=false
       type {classStructOrEnum {refPath=&LinearColor}}
      }
      AdjustSaturation {bIsDynamic=false, bIsEditable=false, bIsStaticSwitch=false, bIsVisible=false, type {classStructOrEnum {refPath=&NiagaraBool}}}
      Saturation Range {
       bIsDynamic=false
       bIsEditable=false
       bIsStaticSwitch=false
       bIsVisible=false
       type {classStructOrEnum {refPath=&Vector2f}}
      }
      AdjustValue {bIsDynamic=false, bIsEditable=false, bIsStaticSwitch=false, bIsVisible=false, type {classStructOrEnum {refPath=&NiagaraBool}}}
      Value Range {
       bIsDynamic=false
       bIsEditable=false
       bIsStaticSwitch=false
       bIsVisible=false
       type {classStructOrEnum {refPath=&Vector2f}}
      }
      AdjustAlpha {bIsDynamic=false, bIsEditable=false, bIsStaticSwitch=false, bIsVisible=false, type {classStructOrEnum {refPath=&NiagaraBool}}}
      Alpha Scale Range {
       bIsDynamic=false
       bIsEditable=false
       bIsStaticSwitch=false
       bIsVisible=false
       type {classStructOrEnum {refPath=&Vector2f}}
      }
      Color Channel Mode {
       bIsDynamic=false
       bIsEditable=false
       bIsStaticSwitch=true
       bIsVisible=false
       type {classStructOrEnum {refPath=&ENiagaraChannelCorrelation}}
      }
      Color Randomness Mode {
       bIsDynamic=false
       bIsEditable=false
       bIsStaticSwitch=true
       bIsVisible=false
       type {classStructOrEnum {refPath=&ENiagaraRandomnessMode}}
      }
      Position Mode {
       bIsDynamic=false
       bIsEditable=true
       bIsStaticSwitch=true
       bIsVisible=true
       type {classStructOrEnum {refPath=&ENiagara_PositionInitializationMode}}
      }
      Position {
       bIsDynamic=false
       bIsEditable=false
       bIsStaticSwitch=false
       bIsVisible=false
       type {classStructOrEnum {refPath=&NiagaraPosition}}
      }
      UsePositionOffset {bIsDynamic=false, bIsEditable=true, bIsStaticSwitch=false, bIsVisible=true, type {classStructOrEnum {refPath=&NiagaraBool}}}
      Position Offset {bIsDynamic=false, bIsEditable=false, bIsStaticSwitch=false, bIsVisible=true, type {classStructOrEnum {refPath=&Vector3f}}}
      Position Offset Coordinate Space {
       bIsDynamic=false
       bIsEditable=false
       bIsStaticSwitch=false
       bIsVisible=false
       type {classStructOrEnum {refPath=&ENiagaraCoordinateSpace}}
      }
      Mass Mode {
       bIsDynamic=false
       bIsEditable=true
       bIsStaticSwitch=true
       bIsVisible=true
       type {classStructOrEnum {refPath=&ENiagara_MassInitializationMode}}
      }
      Mass Min {
       bIsDynamic=false
       bIsEditable=false
       bIsStaticSwitch=false
       bIsVisible=false
       type {classStructOrEnum {refPath=&NiagaraFloat}}
      }
      Mass {
       bIsDynamic=false
       bIsEditable=false
       bIsStaticSwitch=false
       bIsVisible=false
       type {classStructOrEnum {refPath=&NiagaraFloat}}
      }
      Mass Max {
       bIsDynamic=false
       bIsEditable=false
       bIsStaticSwitch=false
       bIsVisible=false
       type {classStructOrEnum {refPath=&NiagaraFloat}}
      }
      Mass Randomness Mode {
       bIsDynamic=false
       bIsEditable=false
       bIsStaticSwitch=true
       bIsVisible=false
       type {classStructOrEnum {refPath=&ENiagaraRandomnessMode}}
      }
      Write Material Random {bIsDynamic=false, bIsEditable=true, bIsStaticSwitch=false, bIsVisible=true, type {classStructOrEnum {refPath=&NiagaraBool}}}
      Material Random {bIsDynamic=false, bIsEditable=true, bIsStaticSwitch=false, bIsVisible=true, type {classStructOrEnum {refPath=&NiagaraFloat}}}
      Sprite Size Mode {
       bIsDynamic=false
       bIsEditable=true
       bIsStaticSwitch=true
       bIsVisible=true
       type {classStructOrEnum {refPath=&ENiagara_SizeScaleMode}}
      }
      Sprite Size {
       bIsDynamic=false
       bIsEditable=false
       bIsStaticSwitch=false
       bIsVisible=false
       type {classStructOrEnum {refPath=&Vector2f}}
      }
      Uniform Sprite Size {
       bIsDynamic=false
       bIsEditable=false
       bIsStaticSwitch=false
       bIsVisible=false
       type {classStructOrEnum {refPath=&NiagaraFloat}}
      }
      Uniform Sprite Size Min {
       bIsDynamic=false
       bIsEditable=false
       bIsStaticSwitch=false
       bIsVisible=false
       type {classStructOrEnum {refPath=&NiagaraFloat}}
      }
      Sprite Size Min {
       bIsDynamic=false
       bIsEditable=false
       bIsStaticSwitch=false
       bIsVisible=false
       type {classStructOrEnum {refPath=&Vector2f}}
      }
      Sprite Size Max {
       bIsDynamic=false
       bIsEditable=false
       bIsStaticSwitch=false
       bIsVisible=false
       type {classStructOrEnum {refPath=&Vector2f}}
      }
      Uniform Sprite Size Max {
       bIsDynamic=false
       bIsEditable=false
       bIsStaticSwitch=false
       bIsVisible=false
       type {classStructOrEnum {refPath=&NiagaraFloat}}
      }
      Sprite Size Randomness Mode {
       bIsDynamic=false
       bIsEditable=false
       bIsStaticSwitch=true
       bIsVisible=false
       type {classStructOrEnum {refPath=&ENiagaraRandomnessMode}}
      }
      Sprite Rotation Mode {
       bIsDynamic=false
       bIsEditable=true
       bIsStaticSwitch=true
       bIsVisible=true
       type {classStructOrEnum {refPath=&ENiagara_SpriteRotationMode}}
      }
      Sprite Rotation Angle Min {
       bIsDynamic=false
       bIsEditable=false
       bIsStaticSwitch=false
       bIsVisible=false
       type {classStructOrEnum {refPath=&NiagaraFloat}}
      }
      Sprite Rotation Angle {
       bIsDynamic=false
       bIsEditable=false
       bIsStaticSwitch=false
       bIsVisible=false
       type {classStructOrEnum {refPath=&NiagaraFloat}}
      }
      Sprite Rotation Angle Max {
       bIsDynamic=false
       bIsEditable=false
       bIsStaticSwitch=false
       bIsVisible=false
       type {classStructOrEnum {refPath=&NiagaraFloat}}
      }
      Sprite Rotation Randomness Mode {
       bIsDynamic=false
       bIsEditable=false
       bIsStaticSwitch=true
       bIsVisible=false
       type {classStructOrEnum {refPath=&ENiagaraRandomnessMode}}
      }
      Sprite UV Mode {
       bIsDynamic=false
       bIsEditable=true
       bIsStaticSwitch=true
       bIsVisible=true
       type {classStructOrEnum {refPath=&ENiagara_UVFlippingMode}}
      }
      Sprite UV Randomness Mode {
       bIsDynamic=false
       bIsEditable=false
       bIsStaticSwitch=true
       bIsVisible=false
       type {classStructOrEnum {refPath=&ENiagaraRandomnessMode}}
      }
      Sprite UV Scale {
       bIsDynamic=false
       bIsEditable=false
       bIsStaticSwitch=false
       bIsVisible=false
       type {classStructOrEnum {refPath=&Vector2f}}
      }
      Mesh Scale Mode {
       bIsDynamic=false
       bIsEditable=true
       bIsStaticSwitch=true
       bIsVisible=true
       type {classStructOrEnum {refPath=&ENiagara_SizeScaleMode}}
      }
      Mesh Uniform Scale Min {
       bIsDynamic=false
       bIsEditable=false
       bIsStaticSwitch=false
       bIsVisible=false
       type {classStructOrEnum {refPath=&NiagaraFloat}}
      }
      Mesh Uniform Scale {bIsDynamic=false, bIsEditable=true, bIsStaticSwitch=false, bIsVisible=true, type {classStructOrEnum {refPath=&NiagaraFloat}}}
      Mesh Scale Min {
       bIsDynamic=false
       bIsEditable=false
       bIsStaticSwitch=false
       bIsVisible=false
       type {classStructOrEnum {refPath=&Vector3f}}
      }
      Mesh Scale {
       bIsDynamic=false
       bIsEditable=false
       bIsStaticSwitch=false
       bIsVisible=false
       type {classStructOrEnum {refPath=&Vector3f}}
      }
      Mesh Scale Max {
       bIsDynamic=false
       bIsEditable=false
       bIsStaticSwitch=false
       bIsVisible=false
       type {classStructOrEnum {refPath=&Vector3f}}
      }
      Mesh Uniform Scale Max {
       bIsDynamic=false
       bIsEditable=false
       bIsStaticSwitch=false
       bIsVisible=false
       type {classStructOrEnum {refPath=&NiagaraFloat}}
      }
      Mesh Scale Randomness Mode {
       bIsDynamic=false
       bIsEditable=false
       bIsStaticSwitch=true
       bIsVisible=false
       type {classStructOrEnum {refPath=&ENiagaraRandomnessMode}}
      }
      Mesh Renderer Array Visibility Mode {
       bIsDynamic=false
       bIsEditable=true
       bIsStaticSwitch=true
       bIsVisible=true
       type {classStructOrEnum {refPath=&ENiagara_UnsetDirectSetRandom}}
      }
      Mesh Renderer Mesh Index {
       bIsDynamic=false
       bIsEditable=false
       bIsStaticSwitch=false
       bIsVisible=false
       type {classStructOrEnum {refPath=&NiagaraInt32}}
      }
      Wrap Index By Mesh Count {bIsDynamic=false, bIsEditable=false, bIsStaticSwitch=true, bIsVisible=false, type {classStructOrEnum {refPath=&NiagaraBool}}}
      Mesh Renderer Info {
       bIsDynamic=false
       bIsEditable=false
       bIsStaticSwitch=false
       bIsVisible=false
       type {classStructOrEnum {refPath=&NiagaraDataInterfaceMeshRendererInfo}}
      }
      Mesh Renderer Random Seed {
       bIsDynamic=false
       bIsEditable=false
       bIsStaticSwitch=false
       bIsVisible=false
       type {classStructOrEnum {refPath=&NiagaraInt32}}
      }
      Ribbon Width Mode {
       bIsDynamic=false
       bIsEditable=true
       bIsStaticSwitch=true
       bIsVisible=true
       type {classStructOrEnum {refPath=&ENiagara_UnsetDirectSet}}
      }
      Ribbon Width {
       bIsDynamic=false
       bIsEditable=false
       bIsStaticSwitch=false
       bIsVisible=false
       type {classStructOrEnum {refPath=&NiagaraFloat}}
      }
      Ribbon Facing Mode {
       bIsDynamic=false
       bIsEditable=true
       bIsStaticSwitch=true
       bIsVisible=true
       type {classStructOrEnum {refPath=&ENiagara_UnsetDirectSet}}
      }
      Ribbon Facing Vector {
       bIsDynamic=false
       bIsEditable=false
       bIsStaticSwitch=false
       bIsVisible=false
       type {classStructOrEnum {refPath=&Vector3f}}
      }
      Ribbon Twist Mode {
       bIsDynamic=false
       bIsEditable=true
       bIsStaticSwitch=true
       bIsVisible=true
       type {classStructOrEnum {refPath=&ENiagara_UnsetDirectSet}}
      }
      "Ribbon Twist " {
       bIsDynamic=false
       bIsEditable=false
       bIsStaticSwitch=false
       bIsVisible=false
       type {classStructOrEnum {refPath=&NiagaraFloat}}
      }
     ]
     moduleScript {refPath=&InitializeParticle}
    }]
  }
  particleUpdateScript {
   scriptName=ParticleUpdateScript
   modules [
    {
     bIsSetParametersModule=false
     enabled=true
     moduleName=ParticleState
     inputs [
      Kill Particles When Lifetime Has Elapsed {bIsDynamic=false, bIsEditable=true, bIsStaticSwitch=true, bIsVisible=true, type {classStructOrEnum {refPath=&NiagaraBool}}}
      Loop Particles Lifetime {bIsDynamic=false, bIsEditable=false, bIsStaticSwitch=false, bIsVisible=false, type {classStructOrEnum {refPath=&NiagaraBool}}}
      Let Infinitely Lived Particles Die When Emitter Deactivates {bIsDynamic=false, bIsEditable=false, bIsStaticSwitch=false, bIsVisible=false, type {classStructOrEnum {refPath=&NiagaraBool}}}
      Lifetime {bIsDynamic=false, bIsEditable=true, bIsStaticSwitch=false, bIsVisible=true, type {classStructOrEnum {refPath=&NiagaraFloat}}}
      DeltaTime {bIsDynamic=false, bIsEditable=true, bIsStaticSwitch=false, bIsVisible=true, type {classStructOrEnum {refPath=&NiagaraFloat}}}
     ]
     moduleScript {refPath=&ParticleState}
    }
    {
     bIsSetParametersModule=false
     enabled=true
     moduleName=ScaleColor
     inputs [
      Scale Mode {
       bIsDynamic=false
       bIsEditable=true
       bIsStaticSwitch=true
       bIsVisible=true
       type {classStructOrEnum {refPath=&ENiagaraScaleColorMode}}
      }
      Linear Color Curve {
       bIsDynamic=false
       bIsEditable=false
       bIsStaticSwitch=false
       bIsVisible=false
       type {classStructOrEnum {refPath=&NiagaraDataInterfaceColorCurve}}
      }
      ScaleRGB {bIsDynamic=false, bIsEditable=true, bIsStaticSwitch=false, bIsVisible=true, type {classStructOrEnum {refPath=&NiagaraBool}}}
      Scale RGB {bIsDynamic=false, bIsEditable=true, bIsStaticSwitch=false, bIsVisible=true, type {classStructOrEnum {refPath=&Vector3f}}}
      ScaleRGBA {bIsDynamic=false, bIsEditable=false, bIsStaticSwitch=false, bIsVisible=false, type {classStructOrEnum {refPath=&NiagaraBool}}}
      Scale RGBA {
       bIsDynamic=false
       bIsEditable=false
       bIsStaticSwitch=false
       bIsVisible=false
       type {classStructOrEnum {refPath=&Vector4f}}
      }
      Curve Index {
       bIsDynamic=false
       bIsEditable=false
       bIsStaticSwitch=false
       bIsVisible=false
       type {classStructOrEnum {refPath=&NiagaraFloat}}
      }
      ScaleA {bIsDynamic=false, bIsEditable=true, bIsStaticSwitch=false, bIsVisible=true, type {classStructOrEnum {refPath=&NiagaraBool}}}
      Scale Alpha {bIsDynamic=true, bIsEditable=true, bIsStaticSwitch=false, bIsVisible=true, type {classStructOrEnum {refPath=&NiagaraFloat}}}
      Color Value To Scale {
       bIsDynamic=false
       bIsEditable=true
       bIsStaticSwitch=false
       bIsVisible=true
       type {classStructOrEnum {refPath=&LinearColor}}
      }
     ]
     moduleScript {refPath=&ScaleColor}
    }
    {
     bIsSetParametersModule=false
     enabled=true
     moduleName=DynamicMaterialParameters
     inputs [
      Write Parameter Index 0 {bIsDynamic=false, bIsEditable=true, bIsStaticSwitch=true, bIsVisible=true, type {classStructOrEnum {refPath=&NiagaraBool}}}
      Module.Param0WriteEnabled {bIsDynamic=false, bIsEditable=true, bIsStaticSwitch=false, bIsVisible=true, type {classStructOrEnum {refPath=&NiagaraBool}}}
      Index 0 Param 1 {bIsDynamic=false, bIsEditable=true, bIsStaticSwitch=false, bIsVisible=true, type {classStructOrEnum {refPath=&NiagaraFloat}}}
      Module.Param1WriteEnabled {bIsDynamic=false, bIsEditable=true, bIsStaticSwitch=false, bIsVisible=true, type {classStructOrEnum {refPath=&NiagaraBool}}}
      Index 0 Param 2 {bIsDynamic=false, bIsEditable=true, bIsStaticSwitch=false, bIsVisible=true, type {classStructOrEnum {refPath=&NiagaraFloat}}}
      Module.Param2WriteEnabled {bIsDynamic=false, bIsEditable=true, bIsStaticSwitch=false, bIsVisible=true, type {classStructOrEnum {refPath=&NiagaraBool}}}
      Index 0 Param 3 {bIsDynamic=false, bIsEditable=true, bIsStaticSwitch=false, bIsVisible=true, type {classStructOrEnum {refPath=&NiagaraFloat}}}
      Module.Param3WriteEnabled {bIsDynamic=false, bIsEditable=true, bIsStaticSwitch=false, bIsVisible=true, type {classStructOrEnum {refPath=&NiagaraBool}}}
      Index 0 Param 4 {bIsDynamic=false, bIsEditable=true, bIsStaticSwitch=false, bIsVisible=true, type {classStructOrEnum {refPath=&NiagaraFloat}}}
      Write Parameter Index 1 {bIsDynamic=false, bIsEditable=true, bIsStaticSwitch=true, bIsVisible=true, type {classStructOrEnum {refPath=&NiagaraBool}}}
      Index 1 Param 1 {
       bIsDynamic=false
       bIsEditable=false
       bIsStaticSwitch=false
       bIsVisible=false
       type {classStructOrEnum {refPath=&NiagaraFloat}}
      }
      Index 1 Param 2 {
       bIsDynamic=false
       bIsEditable=false
       bIsStaticSwitch=false
       bIsVisible=false
       type {classStructOrEnum {refPath=&NiagaraFloat}}
      }
      Index 1 Param 3 {
       bIsDynamic=false
       bIsEditable=false
       bIsStaticSwitch=false
       bIsVisible=false
       type {classStructOrEnum {refPath=&NiagaraFloat}}
      }
      Index 1 Param 4 {
       bIsDynamic=false
       bIsEditable=false
       bIsStaticSwitch=false
       bIsVisible=false
       type {classStructOrEnum {refPath=&NiagaraFloat}}
      }
      Write Parameter Index 2 {bIsDynamic=false, bIsEditable=true, bIsStaticSwitch=true, bIsVisible=true, type {classStructOrEnum {refPath=&NiagaraBool}}}
      Index 2 Param 1 {
       bIsDynamic=false
       bIsEditable=false
       bIsStaticSwitch=false
       bIsVisible=false
       type {classStructOrEnum {refPath=&NiagaraFloat}}
      }
      Index 2 Param 2 {
       bIsDynamic=false
       bIsEditable=false
       bIsStaticSwitch=false
       bIsVisible=false
       type {classStructOrEnum {refPath=&NiagaraFloat}}
      }
      Index 2 Param 3 {
       bIsDynamic=false
       bIsEditable=false
       bIsStaticSwitch=false
       bIsVisible=false
       type {classStructOrEnum {refPath=&NiagaraFloat}}
      }
      Index 2 Param 4 {
       bIsDynamic=false
       bIsEditable=false
       bIsStaticSwitch=false
       bIsVisible=false
       type {classStructOrEnum {refPath=&NiagaraFloat}}
      }
      Write Parameter Index 3 {bIsDynamic=false, bIsEditable=true, bIsStaticSwitch=true, bIsVisible=true, type {classStructOrEnum {refPath=&NiagaraBool}}}
      Index 3 Param 1 {
       bIsDynamic=false
       bIsEditable=false
       bIsStaticSwitch=false
       bIsVisible=false
       type {classStructOrEnum {refPath=&NiagaraFloat}}
      }
      Index 3 Param 2 {
       bIsDynamic=false
       bIsEditable=false
       bIsStaticSwitch=false
       bIsVisible=false
       type {classStructOrEnum {refPath=&NiagaraFloat}}
      }
      Index 3 Param 3 {
       bIsDynamic=false
       bIsEditable=false
       bIsStaticSwitch=false
       bIsVisible=false
       type {classStructOrEnum {refPath=&NiagaraFloat}}
      }
      Index 3 Param 4 {
       bIsDynamic=false
       bIsEditable=false
       bIsStaticSwitch=false
       bIsVisible=false
       type {classStructOrEnum {refPath=&NiagaraFloat}}
      }
     ]
     moduleScript {refPath=&DynamicMaterialParameters}
    }
    {
     bIsSetParametersModule=false
     enabled=true
     moduleName=ScaleMeshSize
     inputs [
      Scale Factor {bIsDynamic=true, bIsEditable=true, bIsStaticSwitch=false, bIsVisible=true, type {classStructOrEnum {refPath=&Vector3f}}}
      Initial Mesh Scale {bIsDynamic=false, bIsEditable=true, bIsStaticSwitch=false, bIsVisible=true, type {classStructOrEnum {refPath=&Vector3f}}}
     ]
     moduleScript {refPath=&ScaleMeshSize}
    }
   ]
  }
  rendererClasses [{refPath=&NiagaraMeshRendererProperties}]
  renderers [{rendererIndex=0, rendererClass {refPath=&NiagaraMeshRendererProperties}}]
 }
 SpikeStreaksParticle {
  bEnabled=false
  emitterSpawnScript {scriptName=EmitterSpawnScript, modules []}
  emitterUpdateScript {
   scriptName=EmitterUpdateScript
   modules [
    {
     bIsSetParametersModule=false
     enabled=true
     moduleName=EmitterState
     inputs [
      Life Cycle Mode {
       bIsDynamic=false
       bIsEditable=true
       bIsStaticSwitch=true
       bIsVisible=true
       type {classStructOrEnum {refPath=&ENiagaraEmitterLifeCycleMode}}
      }
      Inactive Response {
       bIsDynamic=false
       bIsEditable=false
       bIsStaticSwitch=true
       bIsVisible=false
       type {classStructOrEnum {refPath=&ENiagaraInactiveMode}}
      }
      Loop Behavior {
       bIsDynamic=false
       bIsEditable=false
       bIsStaticSwitch=true
       bIsVisible=false
       type {classStructOrEnum {refPath=&ENiagara_EmitterStateOptions}}
      }
      Loop Duration Mode {
       bIsDynamic=false
       bIsEditable=false
       bIsStaticSwitch=true
       bIsVisible=false
       type {classStructOrEnum {refPath=&ENiagara_InfiniteLoopDuration}}
      }
      Loop Count {
       bIsDynamic=false
       bIsEditable=false
       bIsStaticSwitch=false
       bIsVisible=false
       type {classStructOrEnum {refPath=&NiagaraInt32}}
      }
      Loop Duration {
       bIsDynamic=false
       bIsEditable=false
       bIsStaticSwitch=false
       bIsVisible=false
       type {classStructOrEnum {refPath=&NiagaraFloat}}
      }
      Recalculate Duration Each Loop {bIsDynamic=false, bIsEditable=false, bIsStaticSwitch=false, bIsVisible=false, type {classStructOrEnum {refPath=&NiagaraBool}}}
      UseLoopDelay {bIsDynamic=false, bIsEditable=false, bIsStaticSwitch=false, bIsVisible=false, type {classStructOrEnum {refPath=&NiagaraBool}}}
      Loop Delay {
       bIsDynamic=false
       bIsEditable=false
       bIsStaticSwitch=false
       bIsVisible=false
       type {classStructOrEnum {refPath=&NiagaraFloat}}
      }
      Delay First Loop Only {bIsDynamic=false, bIsEditable=false, bIsStaticSwitch=false, bIsVisible=false, type {classStructOrEnum {refPath=&NiagaraBool}}}
      Scalability Mode {
       bIsDynamic=false
       bIsEditable=true
       bIsStaticSwitch=true
       bIsVisible=true
       type {classStructOrEnum {refPath=&ENiagaraEmitterScalabilityMode_Limited}}
      }
      Enable Distance Culling {bIsDynamic=false, bIsEditable=true, bIsStaticSwitch=true, bIsVisible=true, type {classStructOrEnum {refPath=&NiagaraBool}}}
      Module.bUseMinDistance {bIsDynamic=false, bIsEditable=true, bIsStaticSwitch=false, bIsVisible=true, type {classStructOrEnum {refPath=&NiagaraBool}}}
      MinDistance {bIsDynamic=false, bIsEditable=false, bIsStaticSwitch=false, bIsVisible=true, type {classStructOrEnum {refPath=&NiagaraFloat}}}
      Min Distance Response {
       bIsDynamic=false
       bIsEditable=false
       bIsStaticSwitch=false
       bIsVisible=false
       type {classStructOrEnum {refPath=&ENiagaraExecutionStateManagement}}
      }
      Module.bUseMaxDistance {bIsDynamic=false, bIsEditable=true, bIsStaticSwitch=false, bIsVisible=true, type {classStructOrEnum {refPath=&NiagaraBool}}}
      MaxDistance {bIsDynamic=false, bIsEditable=true, bIsStaticSwitch=false, bIsVisible=true, type {classStructOrEnum {refPath=&NiagaraFloat}}}
      Max Distance Response {
       bIsDynamic=false
       bIsEditable=true
       bIsStaticSwitch=false
       bIsVisible=true
       type {classStructOrEnum {refPath=&ENiagaraExecutionStateManagement}}
      }
      Scale Spawn Count {bIsDynamic=false, bIsEditable=true, bIsStaticSwitch=true, bIsVisible=true, type {classStructOrEnum {refPath=&NiagaraBool}}}
      Spawn Count Scale {
       bIsDynamic=false
       bIsEditable=false
       bIsStaticSwitch=false
       bIsVisible=false
       type {classStructOrEnum {refPath=&NiagaraFloat}}
      }
      Spawn Count Scale By Distance {
       bIsDynamic=false
       bIsEditable=false
       bIsStaticSwitch=false
       bIsVisible=false
       type {classStructOrEnum {refPath=&NiagaraDataInterfaceCurve}}
      }
      Enable Visibility Culling {bIsDynamic=false, bIsEditable=true, bIsStaticSwitch=true, bIsVisible=true, type {classStructOrEnum {refPath=&NiagaraBool}}}
      Visibility Cull Response {
       bIsDynamic=false
       bIsEditable=false
       bIsStaticSwitch=false
       bIsVisible=false
       type {classStructOrEnum {refPath=&ENiagaraExecutionStateManagement}}
      }
      Visibility Cull Delay {
       bIsDynamic=false
       bIsEditable=false
       bIsStaticSwitch=false
       bIsVisible=false
       type {classStructOrEnum {refPath=&NiagaraFloat}}
      }
      Reset Age On Awaken {bIsDynamic=false, bIsEditable=true, bIsStaticSwitch=false, bIsVisible=true, type {classStructOrEnum {refPath=&NiagaraBool}}}
     ]
     moduleScript {refPath=&EmitterState}
    }
    {
     bIsSetParametersModule=false
     enabled=true
     moduleName=SpawnBurst_Instantaneous
     inputs [
      Spawn Count {bIsDynamic=false, bIsEditable=true, bIsStaticSwitch=false, bIsVisible=true, type {classStructOrEnum {refPath=&NiagaraInt32}}}
      Spawn Time {bIsDynamic=false, bIsEditable=true, bIsStaticSwitch=false, bIsVisible=true, type {classStructOrEnum {refPath=&NiagaraFloat}}}
      Use Spawn Probability {bIsDynamic=false, bIsEditable=true, bIsStaticSwitch=false, bIsVisible=true, type {classStructOrEnum {refPath=&NiagaraBool}}}
      Spawn Probability {bIsDynamic=false, bIsEditable=false, bIsStaticSwitch=false, bIsVisible=true, type {classStructOrEnum {refPath=&NiagaraFloat}}}
      Spawn Group {bIsDynamic=false, bIsEditable=true, bIsStaticSwitch=false, bIsVisible=true, type {classStructOrEnum {refPath=&NiagaraInt32}}}
      Age {bIsDynamic=false, bIsEditable=true, bIsStaticSwitch=false, bIsVisible=true, type {classStructOrEnum {refPath=&NiagaraFloat}}}
     ]
     moduleScript {refPath=&SpawnBurst_Instantaneous}
    }
   ]
  }
  particleSpawnScript {
   scriptName=ParticleSpawnScript
   modules [
    {
     bIsSetParametersModule=false
     enabled=true
     moduleName=InitializeParticle
     inputs [
      Module.Write Lifetime {bIsDynamic=false, bIsEditable=true, bIsStaticSwitch=false, bIsVisible=true, type {classStructOrEnum {refPath=&NiagaraBool}}}
      Lifetime {bIsDynamic=true, bIsEditable=true, bIsStaticSwitch=false, bIsVisible=true, type {classStructOrEnum {refPath=&NiagaraFloat}}}
      Module.Write Position {bIsDynamic=false, bIsEditable=true, bIsStaticSwitch=false, bIsVisible=true, type {classStructOrEnum {refPath=&NiagaraBool}}}
      Position {
       bIsDynamic=true
       bIsEditable=true
       bIsStaticSwitch=false
       bIsVisible=true
       type {classStructOrEnum {refPath=&NiagaraPosition}}
      }
      Module.Write Mass {bIsDynamic=false, bIsEditable=true, bIsStaticSwitch=false, bIsVisible=true, type {classStructOrEnum {refPath=&NiagaraBool}}}
      Mass {bIsDynamic=false, bIsEditable=true, bIsStaticSwitch=false, bIsVisible=true, type {classStructOrEnum {refPath=&NiagaraFloat}}}
      Module.Write Color {bIsDynamic=false, bIsEditable=true, bIsStaticSwitch=false, bIsVisible=true, type {classStructOrEnum {refPath=&NiagaraBool}}}
      Color {
       bIsDynamic=false
       bIsEditable=true
       bIsStaticSwitch=false
       bIsVisible=true
       type {classStructOrEnum {refPath=&LinearColor}}
      }
      Module.Write Material Random {bIsDynamic=false, bIsEditable=true, bIsStaticSwitch=false, bIsVisible=true, type {classStructOrEnum {refPath=&NiagaraBool}}}
      Material Random {bIsDynamic=false, bIsEditable=true, bIsStaticSwitch=false, bIsVisible=true, type {classStructOrEnum {refPath=&NiagaraFloat}}}
      Module.Write SpriteSize {bIsDynamic=false, bIsEditable=true, bIsStaticSwitch=false, bIsVisible=true, type {classStructOrEnum {refPath=&NiagaraBool}}}
      Sprite Size {bIsDynamic=true, bIsEditable=true, bIsStaticSwitch=false, bIsVisible=true, type {classStructOrEnum {refPath=&Vector2f}}}
      Module.Write SpriteRotation {bIsDynamic=false, bIsEditable=true, bIsStaticSwitch=false, bIsVisible=true, type {classStructOrEnum {refPath=&NiagaraBool}}}
      Sprite Rotation {bIsDynamic=false, bIsEditable=false, bIsStaticSwitch=false, bIsVisible=true, type {classStructOrEnum {refPath=&NiagaraFloat}}}
      Module.Write Scale {bIsDynamic=false, bIsEditable=true, bIsStaticSwitch=false, bIsVisible=true, type {classStructOrEnum {refPath=&NiagaraBool}}}
      Mesh Scale {bIsDynamic=false, bIsEditable=false, bIsStaticSwitch=false, bIsVisible=true, type {classStructOrEnum {refPath=&Vector3f}}}
     ]
     moduleScript {refPath=/Niagara/Modules/Spawn/Initialization/InitializeParticle}
    }
    {
     bIsSetParametersModule=false
     enabled=false
     moduleName=InitializeParticle001
     inputs [
      Lifetime Mode {
       bIsDynamic=false
       bIsEditable=true
       bIsStaticSwitch=true
       bIsVisible=true
       type {classStructOrEnum {refPath=&ENiagara_LifetimeMode}}
      }
      Lifetime Min {
       bIsDynamic=false
       bIsEditable=false
       bIsStaticSwitch=false
       bIsVisible=false
       type {classStructOrEnum {refPath=&NiagaraFloat}}
      }
      Lifetime {bIsDynamic=true, bIsEditable=true, bIsStaticSwitch=false, bIsVisible=true, type {classStructOrEnum {refPath=&NiagaraFloat}}}
      Lifetime Max {
       bIsDynamic=false
       bIsEditable=false
       bIsStaticSwitch=false
       bIsVisible=false
       type {classStructOrEnum {refPath=&NiagaraFloat}}
      }
      Lifetime Randomness Mode {
       bIsDynamic=false
       bIsEditable=false
       bIsStaticSwitch=true
       bIsVisible=false
       type {classStructOrEnum {refPath=&ENiagaraRandomnessMode}}
      }
      Color Mode {
       bIsDynamic=false
       bIsEditable=true
       bIsStaticSwitch=true
       bIsVisible=true
       type {classStructOrEnum {refPath=&ENiagara_ColorInitializationMode}}
      }
      Color Minimum {
       bIsDynamic=false
       bIsEditable=false
       bIsStaticSwitch=false
       bIsVisible=false
       type {classStructOrEnum {refPath=&LinearColor}}
      }
      Color {
       bIsDynamic=false
       bIsEditable=true
       bIsStaticSwitch=false
       bIsVisible=true
       type {classStructOrEnum {refPath=&LinearColor}}
      }
      AdjustHue {bIsDynamic=false, bIsEditable=false, bIsStaticSwitch=false, bIsVisible=false, type {classStructOrEnum {refPath=&NiagaraBool}}}
      Hue Shift Range {
       bIsDynamic=false
       bIsEditable=false
       bIsStaticSwitch=false
       bIsVisible=false
       type {classStructOrEnum {refPath=&Vector2f}}
      }
      Color Maximum {
       bIsDynamic=false
       bIsEditable=false
       bIsStaticSwitch=false
       bIsVisible=false
       type {classStructOrEnum {refPath=&LinearColor}}
      }
      AdjustSaturation {bIsDynamic=false, bIsEditable=false, bIsStaticSwitch=false, bIsVisible=false, type {classStructOrEnum {refPath=&NiagaraBool}}}
      Saturation Range {
       bIsDynamic=false
       bIsEditable=false
       bIsStaticSwitch=false
       bIsVisible=false
       type {classStructOrEnum {refPath=&Vector2f}}
      }
      AdjustValue {bIsDynamic=false, bIsEditable=false, bIsStaticSwitch=false, bIsVisible=false, type {classStructOrEnum {refPath=&NiagaraBool}}}
      Value Range {
       bIsDynamic=false
       bIsEditable=false
       bIsStaticSwitch=false
       bIsVisible=false
       type {classStructOrEnum {refPath=&Vector2f}}
      }
      AdjustAlpha {bIsDynamic=false, bIsEditable=false, bIsStaticSwitch=false, bIsVisible=false, type {classStructOrEnum {refPath=&NiagaraBool}}}
      Alpha Scale Range {
       bIsDynamic=false
       bIsEditable=false
       bIsStaticSwitch=false
       bIsVisible=false
       type {classStructOrEnum {refPath=&Vector2f}}
      }
      Color Channel Mode {
       bIsDynamic=false
       bIsEditable=false
       bIsStaticSwitch=true
       bIsVisible=false
       type {classStructOrEnum {refPath=&ENiagaraChannelCorrelation}}
      }
      Color Randomness Mode {
       bIsDynamic=false
       bIsEditable=false
       bIsStaticSwitch=true
       bIsVisible=false
       type {classStructOrEnum {refPath=&ENiagaraRandomnessMode}}
      }
      Position Mode {
       bIsDynamic=false
       bIsEditable=true
       bIsStaticSwitch=true
       bIsVisible=true
       type {classStructOrEnum {refPath=&ENiagara_PositionInitializationMode}}
      }
      Position {
       bIsDynamic=true
       bIsEditable=false
       bIsStaticSwitch=false
       bIsVisible=false
       type {classStructOrEnum {refPath=&NiagaraPosition}}
      }
      UsePositionOffset {bIsDynamic=false, bIsEditable=true, bIsStaticSwitch=false, bIsVisible=true, type {classStructOrEnum {refPath=&NiagaraBool}}}
      Position Offset {bIsDynamic=false, bIsEditable=false, bIsStaticSwitch=false, bIsVisible=true, type {classStructOrEnum {refPath=&Vector3f}}}
      Position Offset Coordinate Space {
       bIsDynamic=false
       bIsEditable=false
       bIsStaticSwitch=false
       bIsVisible=false
       type {classStructOrEnum {refPath=&ENiagaraCoordinateSpace}}
      }
      Mass Mode {
       bIsDynamic=false
       bIsEditable=true
       bIsStaticSwitch=true
       bIsVisible=true
       type {classStructOrEnum {refPath=&ENiagara_MassInitializationMode}}
      }
      Mass Min {
       bIsDynamic=false
       bIsEditable=false
       bIsStaticSwitch=false
       bIsVisible=false
       type {classStructOrEnum {refPath=&NiagaraFloat}}
      }
      Mass {
       bIsDynamic=false
       bIsEditable=false
       bIsStaticSwitch=false
       bIsVisible=false
       type {classStructOrEnum {refPath=&NiagaraFloat}}
      }
      Mass Max {
       bIsDynamic=false
       bIsEditable=false
       bIsStaticSwitch=false
       bIsVisible=false
       type {classStructOrEnum {refPath=&NiagaraFloat}}
      }
      Mass Randomness Mode {
       bIsDynamic=false
       bIsEditable=false
       bIsStaticSwitch=true
       bIsVisible=false
       type {classStructOrEnum {refPath=&ENiagaraRandomnessMode}}
      }
      Write Material Random {bIsDynamic=false, bIsEditable=true, bIsStaticSwitch=false, bIsVisible=true, type {classStructOrEnum {refPath=&NiagaraBool}}}
      Material Random {bIsDynamic=false, bIsEditable=true, bIsStaticSwitch=false, bIsVisible=true, type {classStructOrEnum {refPath=&NiagaraFloat}}}
      Sprite Size Mode {
       bIsDynamic=false
       bIsEditable=true
       bIsStaticSwitch=true
       bIsVisible=true
       type {classStructOrEnum {refPath=&ENiagara_SizeScaleMode}}
      }
      Sprite Size {bIsDynamic=true, bIsEditable=false, bIsStaticSwitch=false, bIsVisible=false, type {classStructOrEnum {refPath=&Vector2f}}}
      Uniform Sprite Size {
       bIsDynamic=false
       bIsEditable=false
       bIsStaticSwitch=false
       bIsVisible=false
       type {classStructOrEnum {refPath=&NiagaraFloat}}
      }
      Uniform Sprite Size Min {
       bIsDynamic=false
       bIsEditable=false
       bIsStaticSwitch=false
       bIsVisible=false
       type {classStructOrEnum {refPath=&NiagaraFloat}}
      }
      Sprite Size Min {
       bIsDynamic=false
       bIsEditable=false
       bIsStaticSwitch=false
       bIsVisible=false
       type {classStructOrEnum {refPath=&Vector2f}}
      }
      Sprite Size Max {
       bIsDynamic=false
       bIsEditable=false
       bIsStaticSwitch=false
       bIsVisible=false
       type {classStructOrEnum {refPath=&Vector2f}}
      }
      Uniform Sprite Size Max {
       bIsDynamic=false
       bIsEditable=false
       bIsStaticSwitch=false
       bIsVisible=false
       type {classStructOrEnum {refPath=&NiagaraFloat}}
      }
      Sprite Size Randomness Mode {
       bIsDynamic=false
       bIsEditable=false
       bIsStaticSwitch=true
       bIsVisible=false
       type {classStructOrEnum {refPath=&ENiagaraRandomnessMode}}
      }
      Sprite Rotation Mode {
       bIsDynamic=false
       bIsEditable=true
       bIsStaticSwitch=true
       bIsVisible=true
       type {classStructOrEnum {refPath=&ENiagara_SpriteRotationMode}}
      }
      Sprite Rotation Angle Min {
       bIsDynamic=false
       bIsEditable=false
       bIsStaticSwitch=false
       bIsVisible=false
       type {classStructOrEnum {refPath=&NiagaraFloat}}
      }
      Sprite Rotation Angle {
       bIsDynamic=false
       bIsEditable=false
       bIsStaticSwitch=false
       bIsVisible=false
       type {classStructOrEnum {refPath=&NiagaraFloat}}
      }
      Sprite Rotation Angle Max {
       bIsDynamic=false
       bIsEditable=false
       bIsStaticSwitch=false
       bIsVisible=false
       type {classStructOrEnum {refPath=&NiagaraFloat}}
      }
      Sprite Rotation Randomness Mode {
       bIsDynamic=false
       bIsEditable=false
       bIsStaticSwitch=true
       bIsVisible=false
       type {classStructOrEnum {refPath=&ENiagaraRandomnessMode}}
      }
      Sprite UV Mode {
       bIsDynamic=false
       bIsEditable=true
       bIsStaticSwitch=true
       bIsVisible=true
       type {classStructOrEnum {refPath=&ENiagara_UVFlippingMode}}
      }
      Sprite UV Randomness Mode {
       bIsDynamic=false
       bIsEditable=false
       bIsStaticSwitch=true
       bIsVisible=false
       type {classStructOrEnum {refPath=&ENiagaraRandomnessMode}}
      }
      Sprite UV Scale {
       bIsDynamic=false
       bIsEditable=false
       bIsStaticSwitch=false
       bIsVisible=false
       type {classStructOrEnum {refPath=&Vector2f}}
      }
      Mesh Scale Mode {
       bIsDynamic=false
       bIsEditable=true
       bIsStaticSwitch=true
       bIsVisible=true
       type {classStructOrEnum {refPath=&ENiagara_SizeScaleMode}}
      }
      Mesh Uniform Scale Min {
       bIsDynamic=false
       bIsEditable=false
       bIsStaticSwitch=false
       bIsVisible=false
       type {classStructOrEnum {refPath=&NiagaraFloat}}
      }
      Mesh Uniform Scale {
       bIsDynamic=false
       bIsEditable=false
       bIsStaticSwitch=false
       bIsVisible=false
       type {classStructOrEnum {refPath=&NiagaraFloat}}
      }
      Mesh Scale Min {
       bIsDynamic=false
       bIsEditable=false
       bIsStaticSwitch=false
       bIsVisible=false
       type {classStructOrEnum {refPath=&Vector3f}}
      }
      Mesh Scale {
       bIsDynamic=false
       bIsEditable=false
       bIsStaticSwitch=false
       bIsVisible=false
       type {classStructOrEnum {refPath=&Vector3f}}
      }
      Mesh Scale Max {
       bIsDynamic=false
       bIsEditable=false
       bIsStaticSwitch=false
       bIsVisible=false
       type {classStructOrEnum {refPath=&Vector3f}}
      }
      Mesh Uniform Scale Max {
       bIsDynamic=false
       bIsEditable=false
       bIsStaticSwitch=false
       bIsVisible=false
       type {classStructOrEnum {refPath=&NiagaraFloat}}
      }
      Mesh Scale Randomness Mode {
       bIsDynamic=false
       bIsEditable=false
       bIsStaticSwitch=true
       bIsVisible=false
       type {classStructOrEnum {refPath=&ENiagaraRandomnessMode}}
      }
      Mesh Renderer Array Visibility Mode {
       bIsDynamic=false
       bIsEditable=true
       bIsStaticSwitch=true
       bIsVisible=true
       type {classStructOrEnum {refPath=&ENiagara_UnsetDirectSetRandom}}
      }
      Mesh Renderer Mesh Index {
       bIsDynamic=false
       bIsEditable=false
       bIsStaticSwitch=false
       bIsVisible=false
       type {classStructOrEnum {refPath=&NiagaraInt32}}
      }
      Wrap Index By Mesh Count {bIsDynamic=false, bIsEditable=false, bIsStaticSwitch=true, bIsVisible=false, type {classStructOrEnum {refPath=&NiagaraBool}}}
      Mesh Renderer Info {
       bIsDynamic=false
       bIsEditable=false
       bIsStaticSwitch=false
       bIsVisible=false
       type {classStructOrEnum {refPath=&NiagaraDataInterfaceMeshRendererInfo}}
      }
      Mesh Renderer Random Seed {
       bIsDynamic=false
       bIsEditable=false
       bIsStaticSwitch=false
       bIsVisible=false
       type {classStructOrEnum {refPath=&NiagaraInt32}}
      }
      Ribbon Width Mode {
       bIsDynamic=false
       bIsEditable=true
       bIsStaticSwitch=true
       bIsVisible=true
       type {classStructOrEnum {refPath=&ENiagara_UnsetDirectSet}}
      }
      Ribbon Width {
       bIsDynamic=false
       bIsEditable=false
       bIsStaticSwitch=false
       bIsVisible=false
       type {classStructOrEnum {refPath=&NiagaraFloat}}
      }
      Ribbon Facing Mode {
       bIsDynamic=false
       bIsEditable=true
       bIsStaticSwitch=true
       bIsVisible=true
       type {classStructOrEnum {refPath=&ENiagara_UnsetDirectSet}}
      }
      Ribbon Facing Vector {
       bIsDynamic=false
       bIsEditable=false
       bIsStaticSwitch=false
       bIsVisible=false
       type {classStructOrEnum {refPath=&Vector3f}}
      }
      Ribbon Twist Mode {
       bIsDynamic=false
       bIsEditable=true
       bIsStaticSwitch=true
       bIsVisible=true
       type {classStructOrEnum {refPath=&ENiagara_UnsetDirectSet}}
      }
      "Ribbon Twist " {
       bIsDynamic=false
       bIsEditable=false
       bIsStaticSwitch=false
       bIsVisible=false
       type {classStructOrEnum {refPath=&NiagaraFloat}}
      }
     ]
     moduleScript {refPath=/Niagara/Modules/Spawn/Initialization/V2/InitializeParticle}
    }
    {
     bIsSetParametersModule=false
     enabled=true
     moduleName=AddVelocityFromPoint
     inputs [
      Velocity Strength {bIsDynamic=false, bIsEditable=true, bIsStaticSwitch=false, bIsVisible=true, type {classStructOrEnum {refPath=&NiagaraFloat}}}
      Origin Offset {bIsDynamic=false, bIsEditable=true, bIsStaticSwitch=false, bIsVisible=true, type {classStructOrEnum {refPath=&Vector3f}}}
      Origin Offset Coordinate Space {
       bIsDynamic=false
       bIsEditable=true
       bIsStaticSwitch=false
       bIsVisible=true
       type {classStructOrEnum {refPath=&ENiagaraCoordinateSpace}}
      }
      Default Position {bIsDynamic=false, bIsEditable=true, bIsStaticSwitch=false, bIsVisible=true, type {classStructOrEnum {refPath=&Vector3f}}}
      Module.Use Falloff Distance {bIsDynamic=false, bIsEditable=true, bIsStaticSwitch=false, bIsVisible=true, type {classStructOrEnum {refPath=&NiagaraBool}}}
      Velocity Falloff Distance {bIsDynamic=false, bIsEditable=false, bIsStaticSwitch=false, bIsVisible=true, type {classStructOrEnum {refPath=&NiagaraFloat}}}
      Invert Velocity Falloff {bIsDynamic=false, bIsEditable=false, bIsStaticSwitch=false, bIsVisible=false, type {classStructOrEnum {refPath=&NiagaraBool}}}
      Velocity Origin {bIsDynamic=false, bIsEditable=true, bIsStaticSwitch=false, bIsVisible=true, type {classStructOrEnum {refPath=&Vector3f}}}
     ]
     moduleScript {refPath=&AddVelocityFromPoint}
    }
   ]
  }
  particleUpdateScript {
   scriptName=ParticleUpdateScript
   modules [
    {
     bIsSetParametersModule=false
     enabled=true
     moduleName=ParticleState
     inputs [
      Kill Particles When Lifetime Has Elapsed {bIsDynamic=false, bIsEditable=true, bIsStaticSwitch=true, bIsVisible=true, type {classStructOrEnum {refPath=&NiagaraBool}}}
      Loop Particles Lifetime {bIsDynamic=false, bIsEditable=false, bIsStaticSwitch=false, bIsVisible=false, type {classStructOrEnum {refPath=&NiagaraBool}}}
      Let Infinitely Lived Particles Die When Emitter Deactivates {bIsDynamic=false, bIsEditable=false, bIsStaticSwitch=false, bIsVisible=false, type {classStructOrEnum {refPath=&NiagaraBool}}}
      Lifetime {bIsDynamic=false, bIsEditable=true, bIsStaticSwitch=false, bIsVisible=true, type {classStructOrEnum {refPath=&NiagaraFloat}}}
      DeltaTime {bIsDynamic=false, bIsEditable=true, bIsStaticSwitch=false, bIsVisible=true, type {classStructOrEnum {refPath=&NiagaraFloat}}}
     ]
     moduleScript {refPath=&ParticleState}
    }
    {
     bIsSetParametersModule=false
     enabled=true
     moduleName=CameraOffset
     inputs [Camera Offset Amount {bIsDynamic=false, bIsEditable=true, bIsStaticSwitch=false, bIsVisible=true, type {classStructOrEnum {refPath=&NiagaraFloat}}}]
     moduleScript {refPath=&CameraOffset}
    }
    {
     bIsSetParametersModule=false
     enabled=true
     moduleName=SolveForcesAndVelocity
     inputs [
      Clamp Velocity {bIsDynamic=false, bIsEditable=true, bIsStaticSwitch=false, bIsVisible=true, type {classStructOrEnum {refPath=&NiagaraBool}}}
      Speed Limit {bIsDynamic=false, bIsEditable=false, bIsStaticSwitch=false, bIsVisible=true, type {classStructOrEnum {refPath=&NiagaraFloat}}}
      Limit Acceleration {bIsDynamic=false, bIsEditable=true, bIsStaticSwitch=false, bIsVisible=true, type {classStructOrEnum {refPath=&NiagaraBool}}}
      Acceleration Limit {bIsDynamic=false, bIsEditable=false, bIsStaticSwitch=false, bIsVisible=true, type {classStructOrEnum {refPath=&NiagaraFloat}}}
      Manually Enable Rotational Solver {bIsDynamic=false, bIsEditable=true, bIsStaticSwitch=false, bIsVisible=true, type {classStructOrEnum {refPath=&NiagaraBool}}}
      Write to Intrinsic Properties {bIsDynamic=false, bIsEditable=true, bIsStaticSwitch=false, bIsVisible=true, type {classStructOrEnum {refPath=&NiagaraBool}}}
      Write to Presolve Properties {bIsDynamic=false, bIsEditable=true, bIsStaticSwitch=true, bIsVisible=true, type {classStructOrEnum {refPath=&NiagaraBool}}}
      DeltaTime {bIsDynamic=false, bIsEditable=true, bIsStaticSwitch=false, bIsVisible=true, type {classStructOrEnum {refPath=&NiagaraFloat}}}
      Mass {bIsDynamic=false, bIsEditable=true, bIsStaticSwitch=false, bIsVisible=true, type {classStructOrEnum {refPath=&NiagaraFloat}}}
      Position {
       bIsDynamic=false
       bIsEditable=true
       bIsStaticSwitch=false
       bIsVisible=true
       type {classStructOrEnum {refPath=&NiagaraPosition}}
      }
      Velocity {bIsDynamic=false, bIsEditable=true, bIsStaticSwitch=false, bIsVisible=true, type {classStructOrEnum {refPath=&Vector3f}}}
      PreviousVelocity {bIsDynamic=false, bIsEditable=true, bIsStaticSwitch=false, bIsVisible=true, type {classStructOrEnum {refPath=&Vector3f}}}
      Force {bIsDynamic=false, bIsEditable=true, bIsStaticSwitch=false, bIsVisible=true, type {classStructOrEnum {refPath=&Vector3f}}}
      Mesh Orientation {bIsDynamic=false, bIsEditable=false, bIsStaticSwitch=false, bIsVisible=false, type {classStructOrEnum {refPath=&Quat4f}}}
      Rotational Velocity {
       bIsDynamic=false
       bIsEditable=false
       bIsStaticSwitch=false
       bIsVisible=false
       type {classStructOrEnum {refPath=&Vector3f}}
      }
      Rotational Force {
       bIsDynamic=false
       bIsEditable=false
       bIsStaticSwitch=false
       bIsVisible=false
       type {classStructOrEnum {refPath=&Vector3f}}
      }
      Rotational Inertia {
       bIsDynamic=false
       bIsEditable=false
       bIsStaticSwitch=false
       bIsVisible=false
       type {classStructOrEnum {refPath=&NiagaraFloat}}
      }
     ]
     moduleScript {refPath=&SolveForcesAndVelocity}
    }
    {
     bIsSetParametersModule=false
     enabled=true
     moduleName=ScaleSpriteSize
     inputs [
      Scale Factor {bIsDynamic=true, bIsEditable=true, bIsStaticSwitch=false, bIsVisible=true, type {classStructOrEnum {refPath=&Vector2f}}}
      Initial Sprite Size {bIsDynamic=false, bIsEditable=true, bIsStaticSwitch=false, bIsVisible=true, type {classStructOrEnum {refPath=&Vector2f}}}
     ]
     moduleScript {refPath=&ScaleSpriteSize}
    }
    {
     bIsSetParametersModule=false
     enabled=true
     moduleName=ScaleColor
     inputs [
      Scale Mode {
       bIsDynamic=false
       bIsEditable=true
       bIsStaticSwitch=true
       bIsVisible=true
       type {classStructOrEnum {refPath=&ENiagaraScaleColorMode}}
      }
      Linear Color Curve {
       bIsDynamic=false
       bIsEditable=false
       bIsStaticSwitch=false
       bIsVisible=false
       type {classStructOrEnum {refPath=&NiagaraDataInterfaceColorCurve}}
      }
      ScaleRGB {bIsDynamic=false, bIsEditable=true, bIsStaticSwitch=false, bIsVisible=true, type {classStructOrEnum {refPath=&NiagaraBool}}}
      Scale RGB {bIsDynamic=true, bIsEditable=true, bIsStaticSwitch=false, bIsVisible=true, type {classStructOrEnum {refPath=&Vector3f}}}
      ScaleRGBA {bIsDynamic=false, bIsEditable=false, bIsStaticSwitch=false, bIsVisible=false, type {classStructOrEnum {refPath=&NiagaraBool}}}
      Scale RGBA {
       bIsDynamic=false
       bIsEditable=false
       bIsStaticSwitch=false
       bIsVisible=false
       type {classStructOrEnum {refPath=&Vector4f}}
      }
      Curve Index {
       bIsDynamic=false
       bIsEditable=false
       bIsStaticSwitch=false
       bIsVisible=false
       type {classStructOrEnum {refPath=&NiagaraFloat}}
      }
      ScaleA {bIsDynamic=false, bIsEditable=true, bIsStaticSwitch=false, bIsVisible=true, type {classStructOrEnum {refPath=&NiagaraBool}}}
      Scale Alpha {bIsDynamic=true, bIsEditable=true, bIsStaticSwitch=false, bIsVisible=true, type {classStructOrEnum {refPath=&NiagaraFloat}}}
      Color Value To Scale {
       bIsDynamic=false
       bIsEditable=true
       bIsStaticSwitch=false
       bIsVisible=true
       type {classStructOrEnum {refPath=&LinearColor}}
      }
     ]
     moduleScript {refPath=&ScaleColor}
    }
   ]
  }
  rendererClasses [{refPath=&NiagaraSpriteRendererProperties}]
  renderers [{rendererIndex=0, rendererClass {refPath=&NiagaraSpriteRendererProperties}}]
 }
 sparkles {
  emitterSpawnScript {scriptName=EmitterSpawnScript, modules []}
  emitterUpdateScript {
   scriptName=EmitterUpdateScript
   modules [
    {
     bIsSetParametersModule=false
     enabled=true
     moduleName=EmitterState
     inputs [
      Life Cycle Mode {
       bIsDynamic=false
       bIsEditable=true
       bIsStaticSwitch=true
       bIsVisible=true
       type {classStructOrEnum {refPath=&ENiagaraEmitterLifeCycleMode}}
      }
      Inactive Response {
       bIsDynamic=false
       bIsEditable=true
       bIsStaticSwitch=true
       bIsVisible=true
       type {classStructOrEnum {refPath=&ENiagaraInactiveMode}}
      }
      Loop Behavior {
       bIsDynamic=false
       bIsEditable=true
       bIsStaticSwitch=true
       bIsVisible=true
       type {classStructOrEnum {refPath=&ENiagara_EmitterStateOptions}}
      }
      Loop Duration Mode {
       bIsDynamic=false
       bIsEditable=true
       bIsStaticSwitch=true
       bIsVisible=true
       type {classStructOrEnum {refPath=&ENiagara_InfiniteLoopDuration}}
      }
      Loop Count {
       bIsDynamic=false
       bIsEditable=false
       bIsStaticSwitch=false
       bIsVisible=false
       type {classStructOrEnum {refPath=&NiagaraInt32}}
      }
      Loop Duration {bIsDynamic=false, bIsEditable=true, bIsStaticSwitch=false, bIsVisible=true, type {classStructOrEnum {refPath=&NiagaraFloat}}}
      Recalculate Duration Each Loop {bIsDynamic=false, bIsEditable=false, bIsStaticSwitch=false, bIsVisible=false, type {classStructOrEnum {refPath=&NiagaraBool}}}
      UseLoopDelay {bIsDynamic=false, bIsEditable=true, bIsStaticSwitch=false, bIsVisible=true, type {classStructOrEnum {refPath=&NiagaraBool}}}
      Loop Delay {bIsDynamic=false, bIsEditable=true, bIsStaticSwitch=false, bIsVisible=true, type {classStructOrEnum {refPath=&NiagaraFloat}}}
      Delay First Loop Only {bIsDynamic=false, bIsEditable=false, bIsStaticSwitch=false, bIsVisible=false, type {classStructOrEnum {refPath=&NiagaraBool}}}
      Scalability Mode {
       bIsDynamic=false
       bIsEditable=true
       bIsStaticSwitch=true
       bIsVisible=true
       type {classStructOrEnum {refPath=&ENiagaraEmitterScalabilityMode_Limited}}
      }
      Enable Distance Culling {bIsDynamic=false, bIsEditable=false, bIsStaticSwitch=true, bIsVisible=false, type {classStructOrEnum {refPath=&NiagaraBool}}}
      MinDistance {
       bIsDynamic=false
       bIsEditable=false
       bIsStaticSwitch=false
       bIsVisible=false
       type {classStructOrEnum {refPath=&NiagaraFloat}}
      }
      Min Distance Response {
       bIsDynamic=false
       bIsEditable=false
       bIsStaticSwitch=false
       bIsVisible=false
       type {classStructOrEnum {refPath=&ENiagaraExecutionStateManagement}}
      }
      MaxDistance {
       bIsDynamic=false
       bIsEditable=false
       bIsStaticSwitch=false
       bIsVisible=false
       type {classStructOrEnum {refPath=&NiagaraFloat}}
      }
      Max Distance Response {
       bIsDynamic=false
       bIsEditable=false
       bIsStaticSwitch=false
       bIsVisible=false
       type {classStructOrEnum {refPath=&ENiagaraExecutionStateManagement}}
      }
      Scale Spawn Count {bIsDynamic=false, bIsEditable=false, bIsStaticSwitch=true, bIsVisible=false, type {classStructOrEnum {refPath=&NiagaraBool}}}
      Spawn Count Scale {
       bIsDynamic=false
       bIsEditable=false
       bIsStaticSwitch=false
       bIsVisible=false
       type {classStructOrEnum {refPath=&NiagaraFloat}}
      }
      Spawn Count Scale By Distance {
       bIsDynamic=false
       bIsEditable=false
       bIsStaticSwitch=false
       bIsVisible=false
       type {classStructOrEnum {refPath=&NiagaraDataInterfaceCurve}}
      }
      Enable Visibility Culling {bIsDynamic=false, bIsEditable=false, bIsStaticSwitch=true, bIsVisible=false, type {classStructOrEnum {refPath=&NiagaraBool}}}
      Visibility Cull Response {
       bIsDynamic=false
       bIsEditable=false
       bIsStaticSwitch=false
       bIsVisible=false
       type {classStructOrEnum {refPath=&ENiagaraExecutionStateManagement}}
      }
      Visibility Cull Delay {
       bIsDynamic=false
       bIsEditable=false
       bIsStaticSwitch=false
       bIsVisible=false
       type {classStructOrEnum {refPath=&NiagaraFloat}}
      }
      Reset Age On Awaken {bIsDynamic=false, bIsEditable=false, bIsStaticSwitch=false, bIsVisible=false, type {classStructOrEnum {refPath=&NiagaraBool}}}
     ]
     moduleScript {refPath=&EmitterState}
    }
    {
     bIsSetParametersModule=false
     enabled=true
     moduleName=SpawnRate
     inputs [
      SpawnRate {bIsDynamic=false, bIsEditable=true, bIsStaticSwitch=false, bIsVisible=true, type {classStructOrEnum {refPath=&NiagaraFloat}}}
      Use Spawn Probability {bIsDynamic=false, bIsEditable=true, bIsStaticSwitch=false, bIsVisible=true, type {classStructOrEnum {refPath=&NiagaraBool}}}
      Spawn Probability {bIsDynamic=false, bIsEditable=false, bIsStaticSwitch=false, bIsVisible=true, type {classStructOrEnum {refPath=&NiagaraFloat}}}
      SpawnGroup {bIsDynamic=false, bIsEditable=true, bIsStaticSwitch=false, bIsVisible=true, type {classStructOrEnum {refPath=&NiagaraInt32}}}
     ]
     moduleScript {refPath=&SpawnRate}
    }
    {
     bIsSetParametersModule=false
     enabled=true
     moduleName=SpawnBurst_Instantaneous
     inputs [
      Spawn Count {bIsDynamic=false, bIsEditable=true, bIsStaticSwitch=false, bIsVisible=true, type {classStructOrEnum {refPath=&NiagaraInt32}}}
      Spawn Time {bIsDynamic=false, bIsEditable=true, bIsStaticSwitch=false, bIsVisible=true, type {classStructOrEnum {refPath=&NiagaraFloat}}}
      Use Spawn Probability {bIsDynamic=false, bIsEditable=true, bIsStaticSwitch=false, bIsVisible=true, type {classStructOrEnum {refPath=&NiagaraBool}}}
      Spawn Probability {bIsDynamic=false, bIsEditable=false, bIsStaticSwitch=false, bIsVisible=true, type {classStructOrEnum {refPath=&NiagaraFloat}}}
      UseLoopCountLimit {bIsDynamic=false, bIsEditable=true, bIsStaticSwitch=false, bIsVisible=true, type {classStructOrEnum {refPath=&NiagaraBool}}}
      Loop Count Limit {bIsDynamic=false, bIsEditable=false, bIsStaticSwitch=false, bIsVisible=true, type {classStructOrEnum {refPath=&NiagaraInt32}}}
      Spawn Group {bIsDynamic=false, bIsEditable=true, bIsStaticSwitch=false, bIsVisible=true, type {classStructOrEnum {refPath=&NiagaraInt32}}}
      Age {bIsDynamic=false, bIsEditable=true, bIsStaticSwitch=false, bIsVisible=true, type {classStructOrEnum {refPath=&NiagaraFloat}}}
     ]
     moduleScript {refPath=&SpawnBurst_Instantaneous}
    }
   ]
  }
  particleSpawnScript {
   scriptName=ParticleSpawnScript
   modules [
    {
     bIsSetParametersModule=false
     enabled=true
     moduleName=InitializeParticle001
     inputs [
      Lifetime Mode {
       bIsDynamic=false
       bIsEditable=true
       bIsStaticSwitch=true
       bIsVisible=true
       type {classStructOrEnum {refPath=&ENiagara_LifetimeMode}}
      }
      Lifetime Min {bIsDynamic=false, bIsEditable=true, bIsStaticSwitch=false, bIsVisible=true, type {classStructOrEnum {refPath=&NiagaraFloat}}}
      Lifetime {
       bIsDynamic=false
       bIsEditable=false
       bIsStaticSwitch=false
       bIsVisible=false
       type {classStructOrEnum {refPath=&NiagaraFloat}}
      }
      Lifetime Max {bIsDynamic=false, bIsEditable=true, bIsStaticSwitch=false, bIsVisible=true, type {classStructOrEnum {refPath=&NiagaraFloat}}}
      Lifetime Randomness Mode {
       bIsDynamic=false
       bIsEditable=true
       bIsStaticSwitch=true
       bIsVisible=true
       type {classStructOrEnum {refPath=&ENiagaraRandomnessMode}}
      }
      Color Mode {
       bIsDynamic=false
       bIsEditable=true
       bIsStaticSwitch=true
       bIsVisible=true
       type {classStructOrEnum {refPath=&ENiagara_ColorInitializationMode}}
      }
      Color Minimum {
       bIsDynamic=false
       bIsEditable=false
       bIsStaticSwitch=false
       bIsVisible=false
       type {classStructOrEnum {refPath=&LinearColor}}
      }
      Color {
       bIsDynamic=true
       bIsEditable=true
       bIsStaticSwitch=false
       bIsVisible=true
       type {classStructOrEnum {refPath=&LinearColor}}
      }
      AdjustHue {bIsDynamic=false, bIsEditable=false, bIsStaticSwitch=false, bIsVisible=false, type {classStructOrEnum {refPath=&NiagaraBool}}}
      Hue Shift Range {
       bIsDynamic=false
       bIsEditable=false
       bIsStaticSwitch=false
       bIsVisible=false
       type {classStructOrEnum {refPath=&Vector2f}}
      }
      Color Maximum {
       bIsDynamic=false
       bIsEditable=false
       bIsStaticSwitch=false
       bIsVisible=false
       type {classStructOrEnum {refPath=&LinearColor}}
      }
      AdjustSaturation {bIsDynamic=false, bIsEditable=false, bIsStaticSwitch=false, bIsVisible=false, type {classStructOrEnum {refPath=&NiagaraBool}}}
      Saturation Range {
       bIsDynamic=false
       bIsEditable=false
       bIsStaticSwitch=false
       bIsVisible=false
       type {classStructOrEnum {refPath=&Vector2f}}
      }
      AdjustValue {bIsDynamic=false, bIsEditable=false, bIsStaticSwitch=false, bIsVisible=false, type {classStructOrEnum {refPath=&NiagaraBool}}}
      Value Range {
       bIsDynamic=false
       bIsEditable=false
       bIsStaticSwitch=false
       bIsVisible=false
       type {classStructOrEnum {refPath=&Vector2f}}
      }
      AdjustAlpha {bIsDynamic=false, bIsEditable=false, bIsStaticSwitch=false, bIsVisible=false, type {classStructOrEnum {refPath=&NiagaraBool}}}
      Alpha Scale Range {
       bIsDynamic=false
       bIsEditable=false
       bIsStaticSwitch=false
       bIsVisible=false
       type {classStructOrEnum {refPath=&Vector2f}}
      }
      Color Channel Mode {
       bIsDynamic=false
       bIsEditable=false
       bIsStaticSwitch=true
       bIsVisible=false
       type {classStructOrEnum {refPath=&ENiagaraChannelCorrelation}}
      }
      Color Randomness Mode {
       bIsDynamic=false
       bIsEditable=false
       bIsStaticSwitch=true
       bIsVisible=false
       type {classStructOrEnum {refPath=&ENiagaraRandomnessMode}}
      }
      Position Mode {
       bIsDynamic=false
       bIsEditable=true
       bIsStaticSwitch=true
       bIsVisible=true
       type {classStructOrEnum {refPath=&ENiagara_PositionInitializationMode}}
      }
      Position {
       bIsDynamic=false
       bIsEditable=false
       bIsStaticSwitch=false
       bIsVisible=false
       type {classStructOrEnum {refPath=&NiagaraPosition}}
      }
      UsePositionOffset {bIsDynamic=false, bIsEditable=true, bIsStaticSwitch=false, bIsVisible=true, type {classStructOrEnum {refPath=&NiagaraBool}}}
      Position Offset {bIsDynamic=false, bIsEditable=true, bIsStaticSwitch=false, bIsVisible=true, type {classStructOrEnum {refPath=&Vector3f}}}
      Position Offset Coordinate Space {
       bIsDynamic=false
       bIsEditable=true
       bIsStaticSwitch=false
       bIsVisible=true
       type {classStructOrEnum {refPath=&ENiagaraCoordinateSpace}}
      }
      Mass Mode {
       bIsDynamic=false
       bIsEditable=true
       bIsStaticSwitch=true
       bIsVisible=true
       type {classStructOrEnum {refPath=&ENiagara_MassInitializationMode}}
      }
      Mass Min {
       bIsDynamic=false
       bIsEditable=false
       bIsStaticSwitch=false
       bIsVisible=false
       type {classStructOrEnum {refPath=&NiagaraFloat}}
      }
      Mass {
       bIsDynamic=false
       bIsEditable=false
       bIsStaticSwitch=false
       bIsVisible=false
       type {classStructOrEnum {refPath=&NiagaraFloat}}
      }
      Mass Max {
       bIsDynamic=false
       bIsEditable=false
       bIsStaticSwitch=false
       bIsVisible=false
       type {classStructOrEnum {refPath=&NiagaraFloat}}
      }
      Mass Randomness Mode {
       bIsDynamic=false
       bIsEditable=false
       bIsStaticSwitch=true
       bIsVisible=false
       type {classStructOrEnum {refPath=&ENiagaraRandomnessMode}}
      }
      Write Material Random {bIsDynamic=false, bIsEditable=true, bIsStaticSwitch=false, bIsVisible=true, type {classStructOrEnum {refPath=&NiagaraBool}}}
      Material Random {bIsDynamic=false, bIsEditable=true, bIsStaticSwitch=false, bIsVisible=true, type {classStructOrEnum {refPath=&NiagaraFloat}}}
      Sprite Size Mode {
       bIsDynamic=false
       bIsEditable=true
       bIsStaticSwitch=true
       bIsVisible=true
       type {classStructOrEnum {refPath=&ENiagara_SizeScaleMode}}
      }
      Sprite Size {
       bIsDynamic=false
       bIsEditable=false
       bIsStaticSwitch=false
       bIsVisible=false
       type {classStructOrEnum {refPath=&Vector2f}}
      }
      Uniform Sprite Size {bIsDynamic=true, bIsEditable=true, bIsStaticSwitch=false, bIsVisible=true, type {classStructOrEnum {refPath=&NiagaraFloat}}}
      Uniform Sprite Size Min {
       bIsDynamic=false
       bIsEditable=false
       bIsStaticSwitch=false
       bIsVisible=false
       type {classStructOrEnum {refPath=&NiagaraFloat}}
      }
      Sprite Size Min {
       bIsDynamic=false
       bIsEditable=false
       bIsStaticSwitch=false
       bIsVisible=false
       type {classStructOrEnum {refPath=&Vector2f}}
      }
      Sprite Size Max {
       bIsDynamic=false
       bIsEditable=false
       bIsStaticSwitch=false
       bIsVisible=false
       type {classStructOrEnum {refPath=&Vector2f}}
      }
      Uniform Sprite Size Max {
       bIsDynamic=false
       bIsEditable=false
       bIsStaticSwitch=false
       bIsVisible=false
       type {classStructOrEnum {refPath=&NiagaraFloat}}
      }
      Sprite Size Randomness Mode {
       bIsDynamic=false
       bIsEditable=false
       bIsStaticSwitch=true
       bIsVisible=false
       type {classStructOrEnum {refPath=&ENiagaraRandomnessMode}}
      }
      Sprite Rotation Mode {
       bIsDynamic=false
       bIsEditable=true
       bIsStaticSwitch=true
       bIsVisible=true
       type {classStructOrEnum {refPath=&ENiagara_SpriteRotationMode}}
      }
      Sprite Rotation Angle Min {
       bIsDynamic=false
       bIsEditable=false
       bIsStaticSwitch=false
       bIsVisible=false
       type {classStructOrEnum {refPath=&NiagaraFloat}}
      }
      Sprite Rotation Angle {
       bIsDynamic=false
       bIsEditable=false
       bIsStaticSwitch=false
       bIsVisible=false
       type {classStructOrEnum {refPath=&NiagaraFloat}}
      }
      Sprite Rotation Angle Max {
       bIsDynamic=false
       bIsEditable=false
       bIsStaticSwitch=false
       bIsVisible=false
       type {classStructOrEnum {refPath=&NiagaraFloat}}
      }
      Sprite Rotation Randomness Mode {
       bIsDynamic=false
       bIsEditable=false
       bIsStaticSwitch=true
       bIsVisible=false
       type {classStructOrEnum {refPath=&ENiagaraRandomnessMode}}
      }
      Sprite UV Mode {
       bIsDynamic=false
       bIsEditable=true
       bIsStaticSwitch=true
       bIsVisible=true
       type {classStructOrEnum {refPath=&ENiagara_UVFlippingMode}}
      }
      Sprite UV Randomness Mode {
       bIsDynamic=false
       bIsEditable=false
       bIsStaticSwitch=true
       bIsVisible=false
       type {classStructOrEnum {refPath=&ENiagaraRandomnessMode}}
      }
      Sprite UV Scale {
       bIsDynamic=false
       bIsEditable=false
       bIsStaticSwitch=false
       bIsVisible=false
       type {classStructOrEnum {refPath=&Vector2f}}
      }
      Mesh Scale Mode {
       bIsDynamic=false
       bIsEditable=true
       bIsStaticSwitch=true
       bIsVisible=true
       type {classStructOrEnum {refPath=&ENiagara_SizeScaleMode}}
      }
      Mesh Uniform Scale Min {
       bIsDynamic=false
       bIsEditable=false
       bIsStaticSwitch=false
       bIsVisible=false
       type {classStructOrEnum {refPath=&NiagaraFloat}}
      }
      Mesh Uniform Scale {
       bIsDynamic=false
       bIsEditable=false
       bIsStaticSwitch=false
       bIsVisible=false
       type {classStructOrEnum {refPath=&NiagaraFloat}}
      }
      Mesh Scale Min {
       bIsDynamic=false
       bIsEditable=false
       bIsStaticSwitch=false
       bIsVisible=false
       type {classStructOrEnum {refPath=&Vector3f}}
      }
      Mesh Scale {
       bIsDynamic=false
       bIsEditable=false
       bIsStaticSwitch=false
       bIsVisible=false
       type {classStructOrEnum {refPath=&Vector3f}}
      }
      Mesh Scale Max {
       bIsDynamic=false
       bIsEditable=false
       bIsStaticSwitch=false
       bIsVisible=false
       type {classStructOrEnum {refPath=&Vector3f}}
      }
      Mesh Uniform Scale Max {
       bIsDynamic=false
       bIsEditable=false
       bIsStaticSwitch=false
       bIsVisible=false
       type {classStructOrEnum {refPath=&NiagaraFloat}}
      }
      Mesh Scale Randomness Mode {
       bIsDynamic=false
       bIsEditable=false
       bIsStaticSwitch=true
       bIsVisible=false
       type {classStructOrEnum {refPath=&ENiagaraRandomnessMode}}
      }
      Mesh Renderer Array Visibility Mode {
       bIsDynamic=false
       bIsEditable=true
       bIsStaticSwitch=true
       bIsVisible=true
       type {classStructOrEnum {refPath=&ENiagara_UnsetDirectSetRandom}}
      }
      Mesh Renderer Mesh Index {
       bIsDynamic=false
       bIsEditable=false
       bIsStaticSwitch=false
       bIsVisible=false
       type {classStructOrEnum {refPath=&NiagaraInt32}}
      }
      Wrap Index By Mesh Count {bIsDynamic=false, bIsEditable=false, bIsStaticSwitch=true, bIsVisible=false, type {classStructOrEnum {refPath=&NiagaraBool}}}
      Mesh Renderer Info {
       bIsDynamic=false
       bIsEditable=false
       bIsStaticSwitch=false
       bIsVisible=false
       type {classStructOrEnum {refPath=&NiagaraDataInterfaceMeshRendererInfo}}
      }
      Mesh Renderer Random Seed {
       bIsDynamic=false
       bIsEditable=false
       bIsStaticSwitch=false
       bIsVisible=false
       type {classStructOrEnum {refPath=&NiagaraInt32}}
      }
      Ribbon Width Mode {
       bIsDynamic=false
       bIsEditable=true
       bIsStaticSwitch=true
       bIsVisible=true
       type {classStructOrEnum {refPath=&ENiagara_UnsetDirectSet}}
      }
      Ribbon Width {
       bIsDynamic=false
       bIsEditable=false
       bIsStaticSwitch=false
       bIsVisible=false
       type {classStructOrEnum {refPath=&NiagaraFloat}}
      }
      Ribbon Facing Mode {
       bIsDynamic=false
       bIsEditable=true
       bIsStaticSwitch=true
       bIsVisible=true
       type {classStructOrEnum {refPath=&ENiagara_UnsetDirectSet}}
      }
      Ribbon Facing Vector {
       bIsDynamic=false
       bIsEditable=false
       bIsStaticSwitch=false
       bIsVisible=false
       type {classStructOrEnum {refPath=&Vector3f}}
      }
      Ribbon Twist Mode {
       bIsDynamic=false
       bIsEditable=true
       bIsStaticSwitch=true
       bIsVisible=true
       type {classStructOrEnum {refPath=&ENiagara_UnsetDirectSet}}
      }
      "Ribbon Twist " {
       bIsDynamic=false
       bIsEditable=false
       bIsStaticSwitch=false
       bIsVisible=false
       type {classStructOrEnum {refPath=&NiagaraFloat}}
      }
     ]
     moduleScript {refPath=&InitializeParticle}
    }
    {
     bIsSetParametersModule=false
     enabled=true
     moduleName=SphereLocation
     inputs [
      Sphere Radius {bIsDynamic=false, bIsEditable=true, bIsStaticSwitch=false, bIsVisible=true, type {classStructOrEnum {refPath=&NiagaraFloat}}}
      UseOffset {bIsDynamic=false, bIsEditable=true, bIsStaticSwitch=false, bIsVisible=true, type {classStructOrEnum {refPath=&NiagaraBool}}}
      Offset {bIsDynamic=false, bIsEditable=true, bIsStaticSwitch=false, bIsVisible=true, type {classStructOrEnum {refPath=&Vector3f}}}
      Offset Coordinate Space {
       bIsDynamic=false
       bIsEditable=true
       bIsStaticSwitch=false
       bIsVisible=true
       type {classStructOrEnum {refPath=&ENiagaraCoordinateSpace}}
      }
      UseNonUniformScale {bIsDynamic=false, bIsEditable=true, bIsStaticSwitch=false, bIsVisible=true, type {classStructOrEnum {refPath=&NiagaraBool}}}
      Non Uniform Scale {bIsDynamic=false, bIsEditable=false, bIsStaticSwitch=false, bIsVisible=true, type {classStructOrEnum {refPath=&Vector3f}}}
      UseSphereOrientation {bIsDynamic=false, bIsEditable=true, bIsStaticSwitch=false, bIsVisible=true, type {classStructOrEnum {refPath=&NiagaraBool}}}
      Sphere Orientation Axis {bIsDynamic=false, bIsEditable=false, bIsStaticSwitch=false, bIsVisible=true, type {classStructOrEnum {refPath=&Vector3f}}}
      Sphere Origin {
       bIsDynamic=false
       bIsEditable=true
       bIsStaticSwitch=false
       bIsVisible=true
       type {classStructOrEnum {refPath=&NiagaraPosition}}
      }
      Sphere Coordinate Space {
       bIsDynamic=false
       bIsEditable=true
       bIsStaticSwitch=false
       bIsVisible=true
       type {classStructOrEnum {refPath=&ENiagaraCoordinateSpace}}
      }
      Write to Intrinsic Variables {bIsDynamic=false, bIsEditable=true, bIsStaticSwitch=true, bIsVisible=true, type {classStructOrEnum {refPath=&NiagaraBool}}}
      Sphere Distribution {
       bIsDynamic=false
       bIsEditable=true
       bIsStaticSwitch=true
       bIsVisible=true
       type {classStructOrEnum {refPath=&ENiagaraSphereDistributionMode}}
      }
      Evaluation Type {
       bIsDynamic=false
       bIsEditable=false
       bIsStaticSwitch=true
       bIsVisible=false
       type {classStructOrEnum {refPath=&ENiagaraRandomnessEvaluation}}
      }
      Hemisphere X {bIsDynamic=false, bIsEditable=true, bIsStaticSwitch=false, bIsVisible=true, type {classStructOrEnum {refPath=&NiagaraBool}}}
      Uniform Distribution {
       bIsDynamic=false
       bIsEditable=false
       bIsStaticSwitch=false
       bIsVisible=false
       type {classStructOrEnum {refPath=&NiagaraFloat}}
      }
      Hemisphere Y {bIsDynamic=false, bIsEditable=true, bIsStaticSwitch=false, bIsVisible=true, type {classStructOrEnum {refPath=&NiagaraBool}}}
      U Position {
       bIsDynamic=false
       bIsEditable=false
       bIsStaticSwitch=false
       bIsVisible=false
       type {classStructOrEnum {refPath=&NiagaraFloat}}
      }
      Uniform Spiral Amount {
       bIsDynamic=false
       bIsEditable=false
       bIsStaticSwitch=false
       bIsVisible=false
       type {classStructOrEnum {refPath=&NiagaraFloat}}
      }
      Hemisphere Z {bIsDynamic=false, bIsEditable=true, bIsStaticSwitch=false, bIsVisible=true, type {classStructOrEnum {refPath=&NiagaraBool}}}
      Uniform Endpoint Offset {
       bIsDynamic=false
       bIsEditable=false
       bIsStaticSwitch=false
       bIsVisible=false
       type {classStructOrEnum {refPath=&NiagaraFloat}}
      }
      V Position {
       bIsDynamic=false
       bIsEditable=false
       bIsStaticSwitch=false
       bIsVisible=false
       type {classStructOrEnum {refPath=&NiagaraFloat}}
      }
      Radius Position {
       bIsDynamic=false
       bIsEditable=false
       bIsStaticSwitch=false
       bIsVisible=false
       type {classStructOrEnum {refPath=&NiagaraFloat}}
      }
      Surface Only {bIsDynamic=false, bIsEditable=true, bIsStaticSwitch=false, bIsVisible=true, type {classStructOrEnum {refPath=&NiagaraBool}}}
      Surface Only Band Thickness {bIsDynamic=false, bIsEditable=true, bIsStaticSwitch=false, bIsVisible=true, type {classStructOrEnum {refPath=&NiagaraFloat}}}
      Surface Expansion Mode {
       bIsDynamic=false
       bIsEditable=true
       bIsStaticSwitch=false
       bIsVisible=true
       type {classStructOrEnum {refPath=&ENiagaraExpansionMode}}
      }
      Normalize Thickness {bIsDynamic=false, bIsEditable=true, bIsStaticSwitch=false, bIsVisible=true, type {classStructOrEnum {refPath=&NiagaraBool}}}
      Randomness Mode {
       bIsDynamic=false
       bIsEditable=true
       bIsStaticSwitch=true
       bIsVisible=true
       type {classStructOrEnum {refPath=&ENiagaraRandomnessMode}}
      }
      Override Seed {bIsDynamic=false, bIsEditable=true, bIsStaticSwitch=false, bIsVisible=true, type {classStructOrEnum {refPath=&NiagaraBool}}}
      Random Seed {bIsDynamic=false, bIsEditable=false, bIsStaticSwitch=false, bIsVisible=true, type {classStructOrEnum {refPath=&NiagaraInt32}}}
      Fixed Random Seed {bIsDynamic=false, bIsEditable=false, bIsStaticSwitch=true, bIsVisible=false, type {classStructOrEnum {refPath=&NiagaraBool}}}
      MaskSpawnGroup {bIsDynamic=false, bIsEditable=true, bIsStaticSwitch=false, bIsVisible=true, type {classStructOrEnum {refPath=&NiagaraBool}}}
      Spawn Group Mask {bIsDynamic=false, bIsEditable=false, bIsStaticSwitch=false, bIsVisible=true, type {classStructOrEnum {refPath=&NiagaraInt32}}}
     ]
     moduleScript {refPath=&SphereLocation}
    }
    {
     bIsSetParametersModule=false
     enabled=false
     moduleName=AddVelocityFromPoint
     inputs [
      Velocity Strength {bIsDynamic=true, bIsEditable=true, bIsStaticSwitch=false, bIsVisible=true, type {classStructOrEnum {refPath=&NiagaraFloat}}}
      Origin Offset {bIsDynamic=false, bIsEditable=true, bIsStaticSwitch=false, bIsVisible=true, type {classStructOrEnum {refPath=&Vector3f}}}
      Origin Offset Coordinate Space {
       bIsDynamic=false
       bIsEditable=true
       bIsStaticSwitch=false
       bIsVisible=true
       type {classStructOrEnum {refPath=&ENiagaraCoordinateSpace}}
      }
      Default Position {bIsDynamic=false, bIsEditable=true, bIsStaticSwitch=false, bIsVisible=true, type {classStructOrEnum {refPath=&Vector3f}}}
      Module.Use Falloff Distance {bIsDynamic=false, bIsEditable=true, bIsStaticSwitch=false, bIsVisible=true, type {classStructOrEnum {refPath=&NiagaraBool}}}
      Velocity Falloff Distance {bIsDynamic=false, bIsEditable=false, bIsStaticSwitch=false, bIsVisible=true, type {classStructOrEnum {refPath=&NiagaraFloat}}}
      Invert Velocity Falloff {bIsDynamic=false, bIsEditable=false, bIsStaticSwitch=false, bIsVisible=false, type {classStructOrEnum {refPath=&NiagaraBool}}}
      Velocity Origin {bIsDynamic=false, bIsEditable=true, bIsStaticSwitch=false, bIsVisible=true, type {classStructOrEnum {refPath=&Vector3f}}}
     ]
     moduleScript {refPath=&AddVelocityFromPoint}
    }
    {
     bIsSetParametersModule=false
     enabled=true
     moduleName=AddVelocity
     inputs [
      Velocity Mode {
       bIsDynamic=false
       bIsEditable=true
       bIsStaticSwitch=true
       bIsVisible=true
       type {classStructOrEnum {refPath=&ENiagara_VelocityMode}}
      }
      Velocity {
       bIsDynamic=false
       bIsEditable=false
       bIsStaticSwitch=false
       bIsVisible=false
       type {classStructOrEnum {refPath=&Vector3f}}
      }
      Velocity Speed {bIsDynamic=true, bIsEditable=true, bIsStaticSwitch=false, bIsVisible=true, type {classStructOrEnum {refPath=&NiagaraFloat}}}
      Velocity Speed Scale {
       bIsDynamic=false
       bIsEditable=false
       bIsStaticSwitch=false
       bIsVisible=false
       type {classStructOrEnum {refPath=&NiagaraFloat}}
      }
      Constrain To Radius {
       bIsDynamic=false
       bIsEditable=false
       bIsStaticSwitch=false
       bIsVisible=false
       type {classStructOrEnum {refPath=&NiagaraFloat}}
      }
      Module.UseVelDistribution {bIsDynamic=false, bIsEditable=true, bIsStaticSwitch=false, bIsVisible=true, type {classStructOrEnum {refPath=&NiagaraBool}}}
      Distribution Along Cone Axis {bIsDynamic=false, bIsEditable=false, bIsStaticSwitch=false, bIsVisible=true, type {classStructOrEnum {refPath=&NiagaraFloat}}}
      Radius Falloff Near / Far {
       bIsDynamic=false
       bIsEditable=false
       bIsStaticSwitch=false
       bIsVisible=false
       type {classStructOrEnum {refPath=&Vector2f}}
      }
      Radius Falloff Exponent {
       bIsDynamic=false
       bIsEditable=false
       bIsStaticSwitch=false
       bIsVisible=false
       type {classStructOrEnum {refPath=&NiagaraFloat}}
      }
      Module.UseConeFalloff {bIsDynamic=false, bIsEditable=true, bIsStaticSwitch=false, bIsVisible=true, type {classStructOrEnum {refPath=&NiagaraBool}}}
      Speed Falloff From Cone Axis {bIsDynamic=false, bIsEditable=false, bIsStaticSwitch=false, bIsVisible=true, type {classStructOrEnum {refPath=&NiagaraFloat}}}
      Invert Velocity Falloff {bIsDynamic=false, bIsEditable=false, bIsStaticSwitch=false, bIsVisible=false, type {classStructOrEnum {refPath=&NiagaraBool}}}
      Override Seed {bIsDynamic=false, bIsEditable=true, bIsStaticSwitch=false, bIsVisible=true, type {classStructOrEnum {refPath=&NiagaraBool}}}
      Random Seed {bIsDynamic=false, bIsEditable=false, bIsStaticSwitch=false, bIsVisible=true, type {classStructOrEnum {refPath=&NiagaraInt32}}}
      Fixed Random Seed {bIsDynamic=false, bIsEditable=false, bIsStaticSwitch=true, bIsVisible=false, type {classStructOrEnum {refPath=&NiagaraBool}}}
      Randomness Mode {
       bIsDynamic=false
       bIsEditable=true
       bIsStaticSwitch=true
       bIsVisible=true
       type {classStructOrEnum {refPath=&ENiagaraRandomnessMode}}
      }
      Cone Axis {bIsDynamic=false, bIsEditable=true, bIsStaticSwitch=false, bIsVisible=true, type {classStructOrEnum {refPath=&Vector3f}}}
      Cone Angle {bIsDynamic=false, bIsEditable=true, bIsStaticSwitch=false, bIsVisible=true, type {classStructOrEnum {refPath=&NiagaraFloat}}}
      Inner Cone Angle {bIsDynamic=false, bIsEditable=true, bIsStaticSwitch=false, bIsVisible=true, type {classStructOrEnum {refPath=&NiagaraFloat}}}
      Cone Angle Mode {
       bIsDynamic=false
       bIsEditable=true
       bIsStaticSwitch=true
       bIsVisible=true
       type {classStructOrEnum {refPath=&ENiagara_AngleInput}}
      }
      Rotation Mode {
       bIsDynamic=false
       bIsEditable=true
       bIsStaticSwitch=true
       bIsVisible=true
       type {classStructOrEnum {refPath=&ENiagara_RotationMode}}
      }
      Rotation Axis {
       bIsDynamic=false
       bIsEditable=false
       bIsStaticSwitch=false
       bIsVisible=false
       type {classStructOrEnum {refPath=&Vector3f}}
      }
      Yaw / Pitch / Roll {
       bIsDynamic=false
       bIsEditable=false
       bIsStaticSwitch=false
       bIsVisible=false
       type {classStructOrEnum {refPath=&Vector3f}}
      }
      Rotation Quaternion {bIsDynamic=false, bIsEditable=false, bIsStaticSwitch=false, bIsVisible=false, type {classStructOrEnum {refPath=&Quat4f}}}
      Rotation Matrix {
       bIsDynamic=false
       bIsEditable=false
       bIsStaticSwitch=false
       bIsVisible=false
       type {classStructOrEnum {refPath=&NiagaraMatrix}}
      }
      Rotation Angle {
       bIsDynamic=false
       bIsEditable=false
       bIsStaticSwitch=false
       bIsVisible=false
       type {classStructOrEnum {refPath=&NiagaraFloat}}
      }
      Invert Rotation Quaternion {bIsDynamic=false, bIsEditable=false, bIsStaticSwitch=false, bIsVisible=false, type {classStructOrEnum {refPath=&NiagaraBool}}}
      Additional Quaternion Rotation {bIsDynamic=false, bIsEditable=false, bIsStaticSwitch=false, bIsVisible=false, type {classStructOrEnum {refPath=&NiagaraBool}}}
      Additional Yaw / Pitch / Roll {
       bIsDynamic=false
       bIsEditable=false
       bIsStaticSwitch=false
       bIsVisible=false
       type {classStructOrEnum {refPath=&Vector3f}}
      }
      Rotation Angle Type {
       bIsDynamic=false
       bIsEditable=false
       bIsStaticSwitch=true
       bIsVisible=false
       type {classStructOrEnum {refPath=&ENiagara_AngleInput}}
      }
      Rotation Coordinate Space {
       bIsDynamic=false
       bIsEditable=true
       bIsStaticSwitch=false
       bIsVisible=true
       type {classStructOrEnum {refPath=&ENiagaraCoordinateSpace}}
      }
      Velocity Origin {
       bIsDynamic=false
       bIsEditable=false
       bIsStaticSwitch=false
       bIsVisible=false
       type {classStructOrEnum {refPath=&NiagaraPosition}}
      }
      Origin Offset {
       bIsDynamic=false
       bIsEditable=false
       bIsStaticSwitch=false
       bIsVisible=false
       type {classStructOrEnum {refPath=&Vector3f}}
      }
      Origin Offset Coordinate Space {
       bIsDynamic=false
       bIsEditable=false
       bIsStaticSwitch=false
       bIsVisible=false
       type {classStructOrEnum {refPath=&ENiagaraCoordinateSpace}}
      }
      Default Position {
       bIsDynamic=false
       bIsEditable=false
       bIsStaticSwitch=false
       bIsVisible=false
       type {classStructOrEnum {refPath=&NiagaraPosition}}
      }
     ]
     moduleScript {refPath=&AddVelocity}
    }
   ]
  }
  particleUpdateScript {
   scriptName=ParticleUpdateScript
   modules [
    {
     bIsSetParametersModule=false
     enabled=true
     moduleName=ParticleState
     inputs [
      Kill Particles When Lifetime Has Elapsed {bIsDynamic=false, bIsEditable=true, bIsStaticSwitch=true, bIsVisible=true, type {classStructOrEnum {refPath=&NiagaraBool}}}
      Loop Particles Lifetime {bIsDynamic=false, bIsEditable=false, bIsStaticSwitch=false, bIsVisible=false, type {classStructOrEnum {refPath=&NiagaraBool}}}
      Let Infinitely Lived Particles Die When Emitter Deactivates {bIsDynamic=false, bIsEditable=false, bIsStaticSwitch=false, bIsVisible=false, type {classStructOrEnum {refPath=&NiagaraBool}}}
      Lifetime {bIsDynamic=false, bIsEditable=true, bIsStaticSwitch=false, bIsVisible=true, type {classStructOrEnum {refPath=&NiagaraFloat}}}
      DeltaTime {bIsDynamic=false, bIsEditable=true, bIsStaticSwitch=false, bIsVisible=true, type {classStructOrEnum {refPath=&NiagaraFloat}}}
     ]
     moduleScript {refPath=&ParticleState}
    }
    {
     bIsSetParametersModule=false
     enabled=true
     moduleName=CurlNoiseForce
     inputs [
      Noise Strength {bIsDynamic=false, bIsEditable=true, bIsStaticSwitch=false, bIsVisible=true, type {classStructOrEnum {refPath=&NiagaraFloat}}}
      Noise Frequency {bIsDynamic=false, bIsEditable=true, bIsStaticSwitch=false, bIsVisible=true, type {classStructOrEnum {refPath=&NiagaraFloat}}}
      Noise Quality / Cost {
       bIsDynamic=false
       bIsEditable=true
       bIsStaticSwitch=true
       bIsVisible=true
       type {classStructOrEnum {refPath=&ENiagaraCurlNoiseQuality}}
      }
      Pan Noise {bIsDynamic=false, bIsEditable=true, bIsStaticSwitch=false, bIsVisible=true, type {classStructOrEnum {refPath=&NiagaraBool}}}
      Pan Noise Field {bIsDynamic=false, bIsEditable=false, bIsStaticSwitch=false, bIsVisible=true, type {classStructOrEnum {refPath=&Vector3f}}}
      Write to Intrinsic Parameters {bIsDynamic=false, bIsEditable=true, bIsStaticSwitch=true, bIsVisible=true, type {classStructOrEnum {refPath=&NiagaraBool}}}
      Sample Position {bIsDynamic=false, bIsEditable=true, bIsStaticSwitch=false, bIsVisible=true, type {classStructOrEnum {refPath=&Vector3f}}}
      Randomize Noise Sample {bIsDynamic=false, bIsEditable=true, bIsStaticSwitch=true, bIsVisible=true, type {classStructOrEnum {refPath=&NiagaraBool}}}
      Random Seed {bIsDynamic=false, bIsEditable=true, bIsStaticSwitch=false, bIsVisible=true, type {classStructOrEnum {refPath=&NiagaraInt32}}}
      Randomization Vector {bIsDynamic=false, bIsEditable=true, bIsStaticSwitch=false, bIsVisible=true, type {classStructOrEnum {refPath=&Vector3f}}}
      Scale Curl Lookup By Engine Owner Scale {bIsDynamic=false, bIsEditable=true, bIsStaticSwitch=true, bIsVisible=true, type {classStructOrEnum {refPath=&NiagaraBool}}}
      CurlNoise {
       bIsDynamic=false
       bIsEditable=false
       bIsStaticSwitch=false
       bIsVisible=false
       type {classStructOrEnum {refPath=&NiagaraDataInterfaceCurlNoise}}
      }
      Mask Curl Noise {bIsDynamic=false, bIsEditable=true, bIsStaticSwitch=true, bIsVisible=true, type {classStructOrEnum {refPath=&NiagaraBool}}}
      Curl Noise Cone Mask Angle {
       bIsDynamic=false
       bIsEditable=false
       bIsStaticSwitch=false
       bIsVisible=false
       type {classStructOrEnum {refPath=&NiagaraFloat}}
      }
      Curl Noise Cone Mask Falloff Angle {
       bIsDynamic=false
       bIsEditable=false
       bIsStaticSwitch=false
       bIsVisible=false
       type {classStructOrEnum {refPath=&NiagaraFloat}}
      }
      Curl Noise Cone Mask Axis {
       bIsDynamic=false
       bIsEditable=false
       bIsStaticSwitch=false
       bIsVisible=false
       type {classStructOrEnum {refPath=&Vector3f}}
      }
     ]
     moduleScript {refPath=&CurlNoiseForce}
    }
    {
     bIsSetParametersModule=false
     enabled=true
     moduleName=Drag
     inputs [
      Use Linear Drag {bIsDynamic=false, bIsEditable=true, bIsStaticSwitch=false, bIsVisible=true, type {classStructOrEnum {refPath=&NiagaraBool}}}
      Drag {bIsDynamic=true, bIsEditable=true, bIsStaticSwitch=false, bIsVisible=true, type {classStructOrEnum {refPath=&NiagaraFloat}}}
      Use Rotational Drag {bIsDynamic=false, bIsEditable=true, bIsStaticSwitch=false, bIsVisible=true, type {classStructOrEnum {refPath=&NiagaraBool}}}
      Rotational Drag {bIsDynamic=false, bIsEditable=false, bIsStaticSwitch=false, bIsVisible=true, type {classStructOrEnum {refPath=&NiagaraFloat}}}
     ]
     moduleScript {refPath=&Drag}
    }
    {
     bIsSetParametersModule=false
     enabled=true
     moduleName=GravityForce
     inputs [
      Gravity {bIsDynamic=false, bIsEditable=true, bIsStaticSwitch=false, bIsVisible=true, type {classStructOrEnum {refPath=&Vector3f}}}
      Coordinate Space {
       bIsDynamic=false
       bIsEditable=true
       bIsStaticSwitch=true
       bIsVisible=true
       type {classStructOrEnum {refPath=&ENiagaraCoordinateSpace}}
      }
     ]
     moduleScript {refPath=&GravityForce}
    }
    {
     bIsSetParametersModule=false
     enabled=true
     moduleName=SolveForcesAndVelocity
     inputs [
      Clamp Velocity {bIsDynamic=false, bIsEditable=true, bIsStaticSwitch=false, bIsVisible=true, type {classStructOrEnum {refPath=&NiagaraBool}}}
      Speed Limit {bIsDynamic=false, bIsEditable=false, bIsStaticSwitch=false, bIsVisible=true, type {classStructOrEnum {refPath=&NiagaraFloat}}}
      Limit Acceleration {bIsDynamic=false, bIsEditable=true, bIsStaticSwitch=false, bIsVisible=true, type {classStructOrEnum {refPath=&NiagaraBool}}}
      Acceleration Limit {bIsDynamic=false, bIsEditable=false, bIsStaticSwitch=false, bIsVisible=true, type {classStructOrEnum {refPath=&NiagaraFloat}}}
      Manually Enable Rotational Solver {bIsDynamic=false, bIsEditable=true, bIsStaticSwitch=false, bIsVisible=true, type {classStructOrEnum {refPath=&NiagaraBool}}}
      Write to Intrinsic Properties {bIsDynamic=false, bIsEditable=true, bIsStaticSwitch=false, bIsVisible=true, type {classStructOrEnum {refPath=&NiagaraBool}}}
      Write to Presolve Properties {bIsDynamic=false, bIsEditable=true, bIsStaticSwitch=true, bIsVisible=true, type {classStructOrEnum {refPath=&NiagaraBool}}}
      DeltaTime {bIsDynamic=false, bIsEditable=true, bIsStaticSwitch=false, bIsVisible=true, type {classStructOrEnum {refPath=&NiagaraFloat}}}
      Mass {bIsDynamic=false, bIsEditable=true, bIsStaticSwitch=false, bIsVisible=true, type {classStructOrEnum {refPath=&NiagaraFloat}}}
      Position {
       bIsDynamic=false
       bIsEditable=true
       bIsStaticSwitch=false
       bIsVisible=true
       type {classStructOrEnum {refPath=&NiagaraPosition}}
      }
      Velocity {bIsDynamic=false, bIsEditable=true, bIsStaticSwitch=false, bIsVisible=true, type {classStructOrEnum {refPath=&Vector3f}}}
      PreviousVelocity {bIsDynamic=false, bIsEditable=true, bIsStaticSwitch=false, bIsVisible=true, type {classStructOrEnum {refPath=&Vector3f}}}
      Force {bIsDynamic=false, bIsEditable=true, bIsStaticSwitch=false, bIsVisible=true, type {classStructOrEnum {refPath=&Vector3f}}}
      Mesh Orientation {bIsDynamic=false, bIsEditable=false, bIsStaticSwitch=false, bIsVisible=false, type {classStructOrEnum {refPath=&Quat4f}}}
      Rotational Velocity {
       bIsDynamic=false
       bIsEditable=false
       bIsStaticSwitch=false
       bIsVisible=false
       type {classStructOrEnum {refPath=&Vector3f}}
      }
      Rotational Force {
       bIsDynamic=false
       bIsEditable=false
       bIsStaticSwitch=false
       bIsVisible=false
       type {classStructOrEnum {refPath=&Vector3f}}
      }
      Rotational Inertia {
       bIsDynamic=false
       bIsEditable=false
       bIsStaticSwitch=false
       bIsVisible=false
       type {classStructOrEnum {refPath=&NiagaraFloat}}
      }
     ]
     moduleScript {refPath=&SolveForcesAndVelocity}
    }
    {
     bIsSetParametersModule=false
     enabled=true
     moduleName=SpriteRotationRate
     inputs [
      Rotation Rate {bIsDynamic=true, bIsEditable=true, bIsStaticSwitch=false, bIsVisible=true, type {classStructOrEnum {refPath=&NiagaraFloat}}}
      Delta Time {bIsDynamic=false, bIsEditable=true, bIsStaticSwitch=false, bIsVisible=true, type {classStructOrEnum {refPath=&NiagaraFloat}}}
     ]
     moduleScript {refPath=&SpriteRotationRate}
    }
    {
     bIsSetParametersModule=false
     enabled=true
     moduleName=ScaleSpriteSize
     inputs [
      Scale Factor {bIsDynamic=true, bIsEditable=true, bIsStaticSwitch=false, bIsVisible=true, type {classStructOrEnum {refPath=&Vector2f}}}
      Initial Sprite Size {bIsDynamic=false, bIsEditable=true, bIsStaticSwitch=false, bIsVisible=true, type {classStructOrEnum {refPath=&Vector2f}}}
     ]
     moduleScript {refPath=&ScaleSpriteSize}
    }
    {
     bIsSetParametersModule=false
     enabled=true
     moduleName=ScaleColor
     inputs [
      Scale Mode {
       bIsDynamic=false
       bIsEditable=true
       bIsStaticSwitch=true
       bIsVisible=true
       type {classStructOrEnum {refPath=&ENiagaraScaleColorMode}}
      }
      Linear Color Curve {
       bIsDynamic=false
       bIsEditable=false
       bIsStaticSwitch=false
       bIsVisible=false
       type {classStructOrEnum {refPath=&NiagaraDataInterfaceColorCurve}}
      }
      ScaleRGB {bIsDynamic=false, bIsEditable=true, bIsStaticSwitch=false, bIsVisible=true, type {classStructOrEnum {refPath=&NiagaraBool}}}
      Scale RGB {bIsDynamic=false, bIsEditable=true, bIsStaticSwitch=false, bIsVisible=true, type {classStructOrEnum {refPath=&Vector3f}}}
      ScaleRGBA {bIsDynamic=false, bIsEditable=false, bIsStaticSwitch=false, bIsVisible=false, type {classStructOrEnum {refPath=&NiagaraBool}}}
      Scale RGBA {
       bIsDynamic=false
       bIsEditable=false
       bIsStaticSwitch=false
       bIsVisible=false
       type {classStructOrEnum {refPath=&Vector4f}}
      }
      Curve Index {
       bIsDynamic=false
       bIsEditable=false
       bIsStaticSwitch=false
       bIsVisible=false
       type {classStructOrEnum {refPath=&NiagaraFloat}}
      }
      ScaleA {bIsDynamic=false, bIsEditable=true, bIsStaticSwitch=false, bIsVisible=true, type {classStructOrEnum {refPath=&NiagaraBool}}}
      Scale Alpha {bIsDynamic=true, bIsEditable=true, bIsStaticSwitch=false, bIsVisible=true, type {classStructOrEnum {refPath=&NiagaraFloat}}}
      Color Value To Scale {
       bIsDynamic=false
       bIsEditable=true
       bIsStaticSwitch=false
       bIsVisible=true
       type {classStructOrEnum {refPath=&LinearColor}}
      }
     ]
     moduleScript {refPath=&ScaleColor}
    }
    {
     bIsSetParametersModule=false
     enabled=true
     moduleName=DynamicMaterialParameters
     inputs [
      Write Parameter Index 0 {bIsDynamic=false, bIsEditable=true, bIsStaticSwitch=true, bIsVisible=true, type {classStructOrEnum {refPath=&NiagaraBool}}}
      Module.Param0WriteEnabled {bIsDynamic=false, bIsEditable=true, bIsStaticSwitch=false, bIsVisible=true, type {classStructOrEnum {refPath=&NiagaraBool}}}
      Index 0 Param 1 {bIsDynamic=false, bIsEditable=true, bIsStaticSwitch=false, bIsVisible=true, type {classStructOrEnum {refPath=&NiagaraFloat}}}
      Module.Param1WriteEnabled {bIsDynamic=false, bIsEditable=true, bIsStaticSwitch=false, bIsVisible=true, type {classStructOrEnum {refPath=&NiagaraBool}}}
      Index 0 Param 2 {bIsDynamic=false, bIsEditable=true, bIsStaticSwitch=false, bIsVisible=true, type {classStructOrEnum {refPath=&NiagaraFloat}}}
      Module.Param2WriteEnabled {bIsDynamic=false, bIsEditable=true, bIsStaticSwitch=false, bIsVisible=true, type {classStructOrEnum {refPath=&NiagaraBool}}}
      Index 0 Param 3 {bIsDynamic=false, bIsEditable=true, bIsStaticSwitch=false, bIsVisible=true, type {classStructOrEnum {refPath=&NiagaraFloat}}}
      Module.Param3WriteEnabled {bIsDynamic=false, bIsEditable=true, bIsStaticSwitch=false, bIsVisible=true, type {classStructOrEnum {refPath=&NiagaraBool}}}
      Index 0 Param 4 {bIsDynamic=false, bIsEditable=true, bIsStaticSwitch=false, bIsVisible=true, type {classStructOrEnum {refPath=&NiagaraFloat}}}
      Write Parameter Index 1 {bIsDynamic=false, bIsEditable=true, bIsStaticSwitch=true, bIsVisible=true, type {classStructOrEnum {refPath=&NiagaraBool}}}
      Index 1 Param 1 {
       bIsDynamic=false
       bIsEditable=false
       bIsStaticSwitch=false
       bIsVisible=false
       type {classStructOrEnum {refPath=&NiagaraFloat}}
      }
      Index 1 Param 2 {
       bIsDynamic=false
       bIsEditable=false
       bIsStaticSwitch=false
       bIsVisible=false
       type {classStructOrEnum {refPath=&NiagaraFloat}}
      }
      Index 1 Param 3 {
       bIsDynamic=false
       bIsEditable=false
       bIsStaticSwitch=false
       bIsVisible=false
       type {classStructOrEnum {refPath=&NiagaraFloat}}
      }
      Index 1 Param 4 {
       bIsDynamic=false
       bIsEditable=false
       bIsStaticSwitch=false
       bIsVisible=false
       type {classStructOrEnum {refPath=&NiagaraFloat}}
      }
      Write Parameter Index 2 {bIsDynamic=false, bIsEditable=true, bIsStaticSwitch=true, bIsVisible=true, type {classStructOrEnum {refPath=&NiagaraBool}}}
      Index 2 Param 1 {
       bIsDynamic=false
       bIsEditable=false
       bIsStaticSwitch=false
       bIsVisible=false
       type {classStructOrEnum {refPath=&NiagaraFloat}}
      }
      Index 2 Param 2 {
       bIsDynamic=false
       bIsEditable=false
       bIsStaticSwitch=false
       bIsVisible=false
       type {classStructOrEnum {refPath=&NiagaraFloat}}
      }
      Index 2 Param 3 {
       bIsDynamic=false
       bIsEditable=false
       bIsStaticSwitch=false
       bIsVisible=false
       type {classStructOrEnum {refPath=&NiagaraFloat}}
      }
      Index 2 Param 4 {
       bIsDynamic=false
       bIsEditable=false
       bIsStaticSwitch=false
       bIsVisible=false
       type {classStructOrEnum {refPath=&NiagaraFloat}}
      }
      Write Parameter Index 3 {bIsDynamic=false, bIsEditable=true, bIsStaticSwitch=true, bIsVisible=true, type {classStructOrEnum {refPath=&NiagaraBool}}}
      Index 3 Param 1 {
       bIsDynamic=false
       bIsEditable=false
       bIsStaticSwitch=false
       bIsVisible=false
       type {classStructOrEnum {refPath=&NiagaraFloat}}
      }
      Index 3 Param 2 {
       bIsDynamic=false
       bIsEditable=false
       bIsStaticSwitch=false
       bIsVisible=false
       type {classStructOrEnum {refPath=&NiagaraFloat}}
      }
      Index 3 Param 3 {
       bIsDynamic=false
       bIsEditable=false
       bIsStaticSwitch=false
       bIsVisible=false
       type {classStructOrEnum {refPath=&NiagaraFloat}}
      }
      Index 3 Param 4 {
       bIsDynamic=false
       bIsEditable=false
       bIsStaticSwitch=false
       bIsVisible=false
       type {classStructOrEnum {refPath=&NiagaraFloat}}
      }
     ]
     moduleScript {refPath=&DynamicMaterialParameters}
    }
    {
     bIsSetParametersModule=true
     enabled=true
     moduleName=SetVariables_#17
     inputs [
      Particles.Sparkles Color {
       bIsDynamic=true
       bIsEditable=true
       bIsStaticSwitch=false
       bIsVisible=true
       type {classStructOrEnum {refPath=&LinearColor}}
      }
      Particles.SpriteSizeSparkles {bIsDynamic=true, bIsEditable=true, bIsStaticSwitch=false, bIsVisible=true, type {classStructOrEnum {refPath=&Vector2f}}}
      Particles.SparklesSpriteRotation {bIsDynamic=false, bIsEditable=true, bIsStaticSwitch=false, bIsVisible=true, type {classStructOrEnum {refPath=&NiagaraFloat}}}
     ]
     moduleScript {refPath=&self:sparkles_0.NiagaraScriptSource_0.NiagaraGraph_0.NiagaraNodeAssignment_0.SetVariables_#17}
    }
    {
     bIsSetParametersModule=false
     enabled=true
     moduleName=ScaleColor001
     inputs [
      Scale Mode {bIsDynamic=false, bIsEditable=true, bIsStaticSwitch=true, bIsVisible=true, type {classStructOrEnum {refPath=&NiagaraInt32}}}
      Linear Color Curve {
       bIsDynamic=false
       bIsEditable=false
       bIsStaticSwitch=false
       bIsVisible=false
       type {classStructOrEnum {refPath=&NiagaraDataInterfaceColorCurve}}
      }
      ScaleRGB {bIsDynamic=false, bIsEditable=true, bIsStaticSwitch=false, bIsVisible=true, type {classStructOrEnum {refPath=&NiagaraBool}}}
      Scale RGB {bIsDynamic=true, bIsEditable=true, bIsStaticSwitch=false, bIsVisible=true, type {classStructOrEnum {refPath=&Vector3f}}}
      ScaleRGBA {bIsDynamic=false, bIsEditable=false, bIsStaticSwitch=false, bIsVisible=false, type {classStructOrEnum {refPath=&NiagaraBool}}}
      Scale RGBA {
       bIsDynamic=false
       bIsEditable=false
       bIsStaticSwitch=false
       bIsVisible=false
       type {classStructOrEnum {refPath=&Vector4f}}
      }
      Curve Index {
       bIsDynamic=false
       bIsEditable=false
       bIsStaticSwitch=false
       bIsVisible=false
       type {classStructOrEnum {refPath=&NiagaraFloat}}
      }
      ScaleA {bIsDynamic=false, bIsEditable=true, bIsStaticSwitch=false, bIsVisible=true, type {classStructOrEnum {refPath=&NiagaraBool}}}
      Scale Alpha {bIsDynamic=false, bIsEditable=false, bIsStaticSwitch=false, bIsVisible=true, type {classStructOrEnum {refPath=&NiagaraFloat}}}
      Color Value To Scale {
       bIsDynamic=false
       bIsEditable=true
       bIsStaticSwitch=false
       bIsVisible=true
       type {classStructOrEnum {refPath=&LinearColor}}
      }
      Color Scale {
       bIsDynamic=false
       bIsEditable=false
       bIsStaticSwitch=false
       bIsVisible=false
       type {classStructOrEnum {refPath=&LinearColor}}
      }
     ]
     moduleScript {refPath=&ScaleColor}
    }
   ]
  }
  rendererClasses [{refPath=&NiagaraSpriteRendererProperties}]
  renderers [
   {rendererIndex=0, rendererClass {refPath=&NiagaraSpriteRendererProperties}}
   {rendererIndex=1, rendererClass {refPath=&NiagaraSpriteRendererProperties}}
  ]
 }
}
summary {
 systemName=NS_TicketPoke_StartUp
 emitters [
  {
   bEnabled=false
   emitterName=Flash001
   rendererClasses [{refPath=&NiagaraSpriteRendererProperties}, {refPath=&NiagaraMeshRendererProperties}]
  }
  {
   bEnabled=false
   emitterName=Flare
   rendererClasses [{refPath=&NiagaraSpriteRendererProperties}, {refPath=&NiagaraLightRendererProperties}]
  }
  {bEnabled=false, emitterName=Snowflakes, rendererClasses [{refPath=&NiagaraSpriteRendererProperties}]}
  {bEnabled=false, emitterName=NE_Debris_02, rendererClasses [{refPath=&NiagaraMeshRendererProperties}]}
  {
   bEnabled=false
   emitterName=Head
   rendererClasses [{refPath=&NiagaraLightRendererProperties}, {refPath=&NiagaraSpriteRendererProperties}]
  }
  {bEnabled=false, emitterName=NE_flameTrail, rendererClasses [{refPath=&NiagaraRibbonRendererProperties}]}
  {bEnabled=false, emitterName=SpikeStreaksParticle, rendererClasses [{refPath=&NiagaraSpriteRendererProperties}]}
  {bEnabled=false, emitterName=SoftShockwave, rendererClasses [{refPath=&NiagaraMeshRendererProperties}]}
  {emitterName=Snowflakedsff, rendererClasses [{refPath=&NiagaraSpriteRendererProperties}]}
  {emitterName=Snow_Powder, rendererClasses [{refPath=&NiagaraSpriteRendererProperties}]}
  {emitterName=HandGlow, rendererClasses [{refPath=&NiagaraSpriteRendererProperties}]}
  {emitterName=IceBlast, rendererClasses [{refPath=&NiagaraSpriteRendererProperties}]}
  {emitterName=IceSpikes, rendererClasses [{refPath=&NiagaraMeshRendererProperties}]}
  {
   bEnabled=false
   emitterName=SnowflakesCPU
   simTarget=GPUComputeSim
   rendererClasses [{refPath=&NiagaraSpriteRendererProperties}]
  }
  {emitterName=sparkles, rendererClasses [{refPath=&NiagaraSpriteRendererProperties}]}
  {
   bEnabled=false
   emitterName=Snowflakedsff001
   rendererClasses [{refPath=&NiagaraSpriteRendererProperties}, {refPath=&NiagaraMeshRendererProperties}]
  }
  {emitterName=Snowflakedsff002, rendererClasses [{refPath=&NiagaraMeshRendererProperties}]}
  {emitterName=Snowflakedsff003, rendererClasses [{refPath=&NiagaraMeshRendererProperties}]}
 ]
 userVariables [
  User.bIsFrontEnd {
   description=""
   defaultValue {struct {refPath=&NiagaraBool}, value {value=0}}
   type {classStructOrEnum {refPath=&NiagaraBool}}
  }
  User.bIsFrontEndPreview {
   description=""
   defaultValue {struct {refPath=&NiagaraBool}, value {value=0}}
   type {classStructOrEnum {refPath=&NiagaraBool}}
  }
  User.BoolYay {
   description=""
   defaultValue {struct {refPath=&NiagaraFloat}, value {value=0}}
   type {classStructOrEnum {refPath=&NiagaraFloat}}
  }
  User.Color {
   description=""
   defaultValue {struct {refPath=&LinearColor}, value=<rgba 1.359400987625122 2.4773459434509277 6 1>}
   type {classStructOrEnum {refPath=&LinearColor}}
  }
  User.Color_Blade_B {
   description=""
   defaultValue {struct {refPath=&LinearColor}, value=<rgba 0.22656679153442383 0.41289055347442627 1 1>}
   type {classStructOrEnum {refPath=&LinearColor}}
  }
  User.exp_Delay {
   description=""
   defaultValue {struct {refPath=&NiagaraFloat}, value {value=0.032499998807907104}}
   type {classStructOrEnum {refPath=&NiagaraFloat}}
  }
  User.UseCustomColor {
   description=""
   defaultValue {struct {refPath=&NiagaraFloat}, value {value=0}}
   type {classStructOrEnum {refPath=&NiagaraFloat}}
  }
  User.UseCustomColor1 {
   description=""
   defaultValue {struct {refPath=&NiagaraFloat}, value {value=0}}
   type {classStructOrEnum {refPath=&NiagaraFloat}}
  }
 ]
}
