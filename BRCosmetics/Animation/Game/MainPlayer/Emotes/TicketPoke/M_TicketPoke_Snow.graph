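# Snow material for the TicketPoke emote, enabled for Niagara mesh particles and using the Subsurface shading model.
# The graph holds two pale vector parameters, Color and ColorSSS, plus a dynamic parameter whose first channel is named ao.
# An Athena landscape snow normal map is sampled and multiplied by a constant zero, which cancels its detail.
# The file does not show which material inputs these nodes drive.
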Material /BRCosmetics/Animation/Game/MainPlayer/Emotes/TicketPoke/M_TicketPoke_Snow
graph {
 properties {bUsedWithNiagaraMeshParticles=true, pixelDepthOffsetMode=PDOM_Legacy, shadingModel=MSM_Subsurface, usedShadingModels=Subsurface}
 nodes [
  DynamicParameter() {paramNames [ao,Param2,Param3,Param4]}
  Multiply(A=TextureSample(UVs=TextureCoordinate()) {samplerType=SAMPLERTYPE_Normal, texture=/Game/Athena/Environments/Landscape/Textures/T_Athena_Snow_01_N}.RGB) {constB=0}
  VectorParameter() {parameterName=Color, defaultValue=<rgba 0.7665 0.8046 0.8698 0>}
  VectorParameter() {parameterName=ColorSSS, defaultValue=<rgba 0.4055 0.7381 0.8698 0>}
 ]
}
